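# UI material for the Full Combo icon in Pilgrim Quickplay, drawing a tinted icon with a glow and shine.
# Two SmoothStep thresholds on one grayscale texture produce the icon and glow masks, with parameterized ranges and colors.
# The icon mask blends the base color, and Linear Dodge adds the shared ShineSwipe function on top.
# Both masks, scaled by their color alphas, are summed and saturated into a combined alpha.
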Material /PilgrimQuickplay/Widgets/Materials/M_UI_FullCombo
graph {
 properties {blendMode=BLEND_Translucent, materialDomain=MD_UI, pixelDepthOffsetMode=PDOM_Legacy, usedShadingModels=Default Lit}
 nodes [
  n1 = VectorParameter() {group=Glow, parameterName=Glow Color, sortPriority=34, defaultValue=<rgba 1 0.6683 0 1>}
  n2 = TextureSampleParameter2D(UVs=TextureCoordinate()) {
   group=Icon
   parameterName=Icon Texture
   samplerType=SAMPLERTYPE_Grayscale
   sortPriority=33
   texture=/PilgrimQuickplay/Widgets/Textures/T_Icon_FullCombo
  }
  n3 = Multiply(
   A=@n1.A
   B=SmoothStep(
    Min=ScalarParameter() {defaultValue=0.01, group=Glow, parameterName=Glow Min, sliderMax=0.99, sliderMin=0.005, sortPriority=34}
    Max=ScalarParameter() {defaultValue=0.4, group=Glow, parameterName=Glow Max, sliderMax=0.99, sliderMin=0.005, sortPriority=34}
    Value=@n2.RGB
   )
  )
  LinearInterpolate(B=@n1.RGB, Alpha=@n3)
  n4 = VectorParameter() {group=Icon, parameterName=Icon Color, sortPriority=33, defaultValue=<rgba 1 0.62 0 1>}
  n5 = SmoothStep(
   Min=ScalarParameter() {defaultValue=0.33, group=Icon, parameterName=Icon Min, sliderMax=0.99, sliderMin=0.01, sortPriority=33}
   Max=ScalarParameter() {defaultValue=0.5, group=Icon, parameterName=Icon Max, sliderMax=0.99, sliderMin=0.01, sortPriority=33}
   Value=@n2.RGB
  )
  MaterialFunctionCall!(
   Blend=MaterialFunctionCall!(In_Uv) {desc=Replaces the shine nodes, materialFunction=/Game/UI/Foundation/Materials/M-Function-ShineSwipe}.Result
   Base=LinearInterpolate(A=If(A=@n1.A, "A > B"=@n1.A, "A == B"=@n4.RGB, "A < B"=@n4.RGB), B=@n4.RGB, Alpha=@n5)
  ) {materialFunction=/Engine/Functions/Engine_MaterialFunctions03/Blends/Blend_LinearDodge}
  Saturate(None=Add(A=Multiply(A=@n4.A, B=@n5), B=@n3))
 ]
}
