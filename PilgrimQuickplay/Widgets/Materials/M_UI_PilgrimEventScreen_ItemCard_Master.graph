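# Translucent UI master material for Pilgrim event screen item cards, drawing a rounded tile with a progress fill.
# OverlayProgress feeds masks built on a rotated box edge, separating the fill color from the unfilled color, which mixes over the Slate post buffer.
# Color-dodge and overlay blends layer radial gradient, inner glow and leading-edge glow into the background.
# The rounded-shape SDF supplies border masking, and panned noise forms a bubbles mask.
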Material /PilgrimQuickplay/Widgets/Materials/M_UI_PilgrimEventScreen_ItemCard_Master
graph {
 properties {blendMode=BLEND_Translucent, materialDomain=MD_UI, pixelDepthOffsetMode=PDOM_Legacy, usedShadingModels=Default Lit}
 nodes [
  n1 = Multiply(
   A=MaterialFunctionCall(UV=NamedRerouteUsage()) {materialFunction=&MF_UI_UVScaling}.ScaleX
   B=AppendVector(A=Constant() {r=3}, B=Constant() {r=2})
  )
  Add(A=@n1)
  Constant() {r=0.3}
  LinearInterpolate(
   A=NamedRerouteUsage()
   B=VectorParameter() {parameterName=BoderLineColor, defaultValue=<rgba 0.0953 0.0284 0.6105 1>}.RGB
   Alpha=NamedRerouteUsage()
  )
  n2 = MaterialFunctionCall(
   Corner Radius=ScalarParameter() {defaultValue=20, group=Shape, parameterName=Corner_Radiues}
   Radius in Pixels=Constant() {r=1}
   Outline Thickness=Constant() {r=6}
  ) {materialFunction=&MF_UI_CoreUI_RoundedMask}
  LinearInterpolate(
   B=LinearInterpolate(A=ScalarParameter() {defaultValue=0.75, parameterName=Alpha_Background}, Alpha=NamedRerouteUsage())
   Alpha=NamedRerouteDeclaration(None=@n2.Shape Mask) {name=Rounded Mask, nodeColor=<rgba 0 0 0 1>}
  )
  n3 = NamedRerouteUsage()
  LinearInterpolate(B=@n3) {constAlpha=0.8}
  MaterialFunctionCall() {materialFunction=&ScreenAlignedUVs}
  n4 = NamedRerouteUsage()
  n5 = NamedRerouteUsage()
  n6 = Add(
   A=MaterialFunctionCall(
    UVs=MaterialFunctionCall(UV=@n4) {materialFunction=&MF_UI_UVScaling}.ScaleX
    "Rotation Angle (0-1)"=Constant() {r=0.25}
   ) {materialFunction=&CustomRotator}.Rotated Values
   B=AppendVector(A=NamedRerouteUsage(), B=OneMinus(None=@n5))
  )
  MaterialFunctionCall!(
   UV=@n6
   Wave Height=Constant() {r=0.01}
   Wave Elevation
   Wave Length=Constant() {r=12}
   Wave Speed=Constant() {r=-0.5}
   GlowMin
   GlowMax=Constant() {r=0.025}
   Outline Thickness
   Outline Alignment
   Custom Curve
  ) {materialFunction=/FortUILibrary/MaterialFunctions/Shapes/MF_UI_Wave}
  n7 = ScalarParameter() {defaultValue=0.5, group=Inner Glow, parameterName=Inner Glow Size, sliderMax=1}
  n8 = Saturate(None=Add(A=NamedRerouteUsage(), B=LinearInterpolate(Alpha=@n7) {constB=0.5}) {constB=0.5})
  n9 = NamedRerouteDeclaration(None=Saturate(None=Multiply(A=@n8) {constB=2})) {name=Inner Glow Mask, nodeColor=<rgba 0 0 0 1>}
  Multiply(A=@n9, B=ScalarParameter() {defaultValue=1, group=Inner Glow, parameterName=Inner Glow Opacity, sliderMax=1})
  n10 = NamedRerouteUsage()
  n11 = Subtract(A=NamedRerouteUsage()) {constB=0.5}
  n12 = SmoothStep(Value=ComponentMask(None=Subtract(A=@n10, B=@n11)) {r=true})
  n13 = NamedRerouteDeclaration(None=MaterialFunctionCall(UVs=@n6, Size=Constant() {r=1}) {materialFunction=&MF_UI_Box}.SDF) {name=SineWaveEdge, nodeColor=<rgba 0 0 0 1>}
  n14 = OneMinus(None=Subtract(A=Saturate(None=@n13)) {constB=-0.2})
  n15 = NamedRerouteUsage()
  n16 = Multiply(
   A=Multiply(A=Multiply(A=@n12, B=@n14), B=LinearInterpolate(Alpha=@n15) {constA=0.5})
   B=ScalarParameter() {defaultValue=2.22, parameterName=LeadingEdgeGlow}
  )
  n17 = NamedRerouteUsage()
  n18 = OneMinus(None=@n17)
  n19 = LinearInterpolate(
   A=LinearInterpolate(
    A=@n18
    B=MaterialFunctionCall(DistanceField=@n13) {materialFunction=&MF_UI_AntiAliasing}.Outer
    Alpha=Ceil(None=@n17)
   )
   B=@n18
   Alpha=Floor(None=NamedRerouteUsage())
  )
  n20 = NamedRerouteDeclaration(None=OneMinus(None=NamedRerouteDeclaration(None=@n19) {name=Progress Mask, nodeColor=<rgba 0 0 0 1>})) {name=Inverted Progress Mask, nodeColor=<rgba 0 0 0 1>}
  NamedRerouteDeclaration(None=Saturate(None=Min(A=@n16, B=@n20))) {name=LeadingEdgeGlow, nodeColor=<rgba 0 0 0 1>}
  n21 = MaterialFunctionCall() {materialFunction=&GetUserInterfaceUV}
  n22 = NamedRerouteDeclaration(None=@n21.Normalized UV) {name=Normalized UV, nodeColor=<rgba 0 0 0 1>}
  NamedRerouteDeclaration(None=ComponentMask(None=Sine(None=@n22) {period=2}) {g=true}) {name=VSineGrad, nodeColor=<rgba 0 0 0 1>}
  n23 = ScalarParameter() {defaultValue=0.1, parameterName=BorderSize}
  n24 = MaterialFunctionCall(
   DistanceField=SmoothStep(Min=Multiply(A=@n23) {constB=-1}, Value=NamedRerouteUsage()) {constMax=0, constMin=-0.1}
   Threshold=Constant() {r=0.5}
  ) {materialFunction=&MF_UI_AntiAliasing}
  NamedRerouteDeclaration(None=Multiply(A=Saturate(None=@n24.Outer), B=ScalarParameter() {defaultValue=0.25, parameterName=BorderAmount})) {name=Border, nodeColor=<rgba 0 1 0.6953 1>}
  NamedRerouteDeclaration(None=@n21.9-Slice UV) {name=9 Slice UV, nodeColor=<rgba 0 0 0 1>}
  NamedRerouteDeclaration(None=@n9) {name=Inner Glow Opacity, nodeColor=<rgba 0 0 0 1>}
  NamedRerouteDeclaration(None=ComponentMask(None=@n22) {r=true}) {name=HGrad, nodeColor=<rgba 0 0 0 1>}
  NamedRerouteDeclaration(None=TextureObject() {samplerType=SAMPLERTYPE_Grayscale, texture=/FortWater/Textures/LowResBlurredNoise}) {name=Noise Texture, nodeColor=<rgba 0 0 0 1>}
  NamedRerouteDeclaration(None=ScalarParameter() {group=Edge Noise, parameterName=RandomSeed, sliderMax=1}) {name=Random Seed, nodeColor=<rgba 0 0 0 1>}
  n25 = LinearInterpolate(Alpha=MaterialFunctionCall() {materialFunction=&DebugTimeSine}.Result) {constB=0.6}
  n26 = Constant() {r=0.2}
  n27 = Constant() {r=0.1}
  n28 = TextureSample(UVs=Panner(Coordinate=@n1, Speed=AppendVector(A=@n26, B=@n27)) {speedX=0.2}, Tex=NamedRerouteUsage()) {samplerType=SAMPLERTYPE_Grayscale, texture=/FortUILibrary/UIKit/Buttons/Textures/LowResBlurredNoise}
  n29 = MaterialFunctionCall() {materialFunction=&LinearGradient}
  n30 = LinearInterpolate(B=Min(A=SmoothStep(Min=@n25, Value=@n28.R) {constMax=0.7, constMin=0.6}, B=OneMinus(None=@n29.UGradient)))
  n31 = LinearInterpolate(B=SmoothStep(Value=@n28.R) {constMax=0.7, constMin=0.5}) {constAlpha=0.1}
  NamedRerouteDeclaration(None=Saturate(None=Max(A=@n30, B=@n31))) {name=Bubbles, nodeColor=<rgba 1 0 0.4922 1>}
  NamedRerouteDeclaration(None=ScalarParameter() {defaultValue=0.5, group=Shape, parameterName=OverlayProgress, sliderMax=1, sortPriority=2}) {name=Progression Value, nodeColor=<rgba 0 0 0 1>}
  n32 = ScalarParameter() {defaultValue=12, group=Inner Glow, parameterName=Inner Glow Opacity, sliderMax=1}
  n33 = NamedRerouteUsage()
  n34 = NamedRerouteDeclaration(None=Saturate(None=LinearInterpolate(B=@n32, Alpha=@n33) {constB=1.25})) {name=InnerGlowAdjusted, nodeColor=<rgba 0 1 0.6484 1>}
  n35 = MaterialFunctionCall!("UVs (Optional)") {materialFunction=/Engine/Functions/UserInterface/GetSlatePost0}
  n36 = NamedRerouteUsage()
  n37 = ScalarParameter() {defaultValue=0.5, parameterName=RadialGradient_Amount}
  n38 = Saturate(None=MaterialFunctionCall!(
    Blend=Multiply(A=@n36, B=@n37) {constB=0.5}
    Base=VectorParameter() {group=TileBkgd, parameterName=Tile Bkgd Color NoFill, defaultValue=<rgba 0.0354 0 0.5 1>}.RGB
   ) {materialFunction=/Engine/Functions/Engine_MaterialFunctions03/Blends/Blend_ColorDodge}.Result)
  n39 = ScalarParameter() {defaultValue=0.75, parameterName=Alpha_Background}
  n40 = NamedRerouteUsage()
  n41 = LinearInterpolate(B=VectorParameter() {group=Inner Glow, parameterName=LeadginGlow, defaultValue=<rgba 0 0.3601 1 1>}.RGB, Alpha=@n40)
  n42 = NamedRerouteDeclaration(None=NamedRerouteDeclaration(None=VectorParameter() {group=TileBkgd, parameterName=Tile Bkgd Color Fill, defaultValue=<rgba 0.7 0.0525 0.1928 1>}.RGB) {name=BG Color, nodeColor=<rgba 0 0 0 1>}) {name=Background Colors, nodeColor=<rgba 0 0 0 1>}
  n43 = Saturate(None=MaterialFunctionCall(Blend=@n41, Base=@n42) {materialFunction=&Blend_ColorDodge}.Result)
  n44 = Min(A=@n3, B=OneMinus(None=NamedRerouteUsage()))
  n45 = LinearInterpolate(
   A=MaterialFunctionCall(
    Base=@n42
    Blend=MaterialFunctionCall(Blend=@n42, Base=@n44) {materialFunction=&Blend_ColorDodge}.Result
   ) {materialFunction=&Blend_Overlay}.Result
   B=@n41
   Alpha=@n40
  )
  n46 = MaterialFunctionCall(
   Blend=Saturate(None=NamedRerouteUsage())
   Base=LinearInterpolate(A=LinearInterpolate(A=@n43, B=@n45), B=LinearInterpolate(A=@n35.LinearRGB, B=@n38, Alpha=@n39), Alpha=NamedRerouteUsage()) {constA=1}
  ) {materialFunction=&Blend_ColorDodge}
  n47 = NamedRerouteDeclaration(None=Saturate(None=MaterialFunctionCall(Blend=@n34, Base=@n46.Result) {materialFunction=&Blend_ColorDodge}.Result)) {name=Progression Background, nodeColor=<rgba 0 0 0 1>}
  n48 = MaterialFunctionCall(Blend=NamedRerouteUsage(), Base=@n47) {materialFunction=&Blend_ColorDodge}
  n49 = NamedRerouteUsage()
  n50 = Constant() {r=1}
  NamedRerouteDeclaration(None=LinearInterpolate(
    A=Saturate(None=@n48.Result)
    B=@n47
    Alpha=MaterialFunctionCall!(A=@n49, B=@n50) {materialFunction=/FortUILibrary/MaterialFunctions/Basics/MF_UI_BooleanOperations}.IsGreaterOrEqual
   )) {name=Combined Background, nodeColor=<rgba 0 0 0 1>}
  NamedRerouteDeclaration(None=@n2.SDF) {name=Rounded Shape SDF, nodeColor=<rgba 0 0 0 1>}
  NamedRerouteDeclaration(None=MaterialFunctionCall(UV=NamedRerouteUsage()) {materialFunction=&MF_UI_RadialGradient}.Smooth) {name=Radial Gradient, nodeColor=<rgba 0 0 0 1>}
  NamedRerouteDeclaration(None=Time()) {name=Time, nodeColor=<rgba 0 0 0 1>}
 ]
}
