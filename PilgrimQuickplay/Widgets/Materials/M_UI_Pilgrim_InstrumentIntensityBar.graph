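# Translucent UI material for a segment or bar of the Pilgrim Quickplay instrument intensity widget, drawing a rounded box.
# MF_UI_Box builds a box SDF from texture coordinates, a fixed size and a Corner Roundness parameter.
# The inverted, saturated SDF passes through a narrow SmoothStep to form a soft-edged mask.
# A Filled static switch selects that mask at full strength or dimmed for the empty state.
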Material /PilgrimQuickplay/Widgets/Materials/M_UI_Pilgrim_InstrumentIntensityBar
graph {
 properties {blendMode=BLEND_Translucent, materialDomain=MD_UI, pixelDepthOffsetMode=PDOM_Legacy, usedShadingModels=Default Lit}
 nodes [
  Constant() {r=1}
  MaterialFunctionCall() {materialFunction=&GetUserInterfaceUV}
  n1 = MaterialFunctionCall(UVs=TextureCoordinate(), Size=Constant() {r=0.65}, CornerRoundness=ScalarParameter() {defaultValue=0.2, parameterName=Corner Roundness}) {materialFunction=&MF_UI_Box}
  n2 = SmoothStep(Value=Saturate(None=OneMinus(None=@n1.SDF))) {constMax=0.8, constMin=0.69}
  StaticSwitchParameter(True=@n2, False=Multiply(A=@n2) {constB=0.2}) {defaultValue=true, parameterName=Filled}
 ]
}
