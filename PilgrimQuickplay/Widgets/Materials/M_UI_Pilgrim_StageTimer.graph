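# Translucent UI material for the Pilgrim stage timer, drawing a circular disc with a ring that fills radially by Progress.
# Exponential radial gradients with smoothstep edges form the disc, and subtracting two of them isolates the ring.
# Floor of Progress plus a rotated UV angle gives the sweep mask, with origin, offset, axis inversion and normalization parameters.
# That mask blends BorderColor to BorderColorFill on the ring, over a BgColor interior.
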Material /PilgrimQuickplay/Widgets/Materials/M_UI_Pilgrim_StageTimer
graph {
 properties {blendMode=BLEND_Translucent, materialDomain=MD_UI, pixelDepthOffsetMode=PDOM_Legacy, usedShadingModels=Default Lit}
 nodes [
  n1 = Constant() {r=0.1}
  n2 = Add(A=@n1) {constB=0.5}
  n3 = ScalarParameter() {defaultValue=0.5, desc=GlobalRadius, parameterName=Scale}
  n4 = Constant() {r=30}
  n5 = MaterialFunctionCall(CenterPosition=Constant2Vector() {g=0.5, r=0.5}, Radius=Subtract(A=@n3, B=Constant()), Density=@n4) {materialFunction=&RadialGradientExponential}
  n6 = Constant() {r=0.1}
  n7 = Add(A=@n6) {constB=0.5}
  n8 = MaterialFunctionCall(
   CenterPosition=Constant2Vector() {g=0.5, r=0.5}
   Radius=Subtract(A=@n3, B=ScalarParameter() {defaultValue=0.1, parameterName=Ring Thickness})
   Density=@n4
  ) {materialFunction=&RadialGradientExponential}
  n9 = Saturate(None=Subtract(A=SmoothStep(Min=@n1, Max=@n2, Value=@n5.RadialGradientExponential), B=SmoothStep(Min=@n6, Max=@n7, Value=@n8.RadialGradientExponential)))
  n10 = Constant() {r=0.1}
  n11 = Constant2Vector() {g=0.5, r=0.5}
  n12 = Constant() {r=30}
  n13 = Saturate(None=SmoothStep(
    Min=@n10
    Max=Add(A=@n10) {constB=0.5}
    Value=MaterialFunctionCall(CenterPosition=@n11, Radius=@n3, Density=@n12) {materialFunction=&RadialGradientExponential}.RadialGradientExponential
   ))
  Add(A=@n9, B=@n13)
  n14 = VectorParameter() {parameterName=BorderColor, defaultValue=<rgba 0.0039 0.3725 0.8314 1>}
  n15 = VectorParameter() {parameterName=BorderColorFill, defaultValue=<rgba 1 1 1 1>}
  n16 = Constant() {r=0.1}
  n17 = Constant2Vector() {g=0.5, r=0.5}
  n18 = Constant() {r=30}
  n19 = ScalarParameter() {parameterName=Progress}
  n20 = ScalarParameter() {parameterName=Offset_0-1}
  n21 = LinearInterpolate(A=@n19, B=Multiply(A=@n19, B=OneMinus(None=@n20)), Alpha=ScalarParameter() {parameterName=bNormalizeAmount}) {desc="If normalized, amount reduced by (1 - offset)"}
  n22 = Clamp(None=@n20) {desc="Removes gap when Amount  = 1.0", minDefault=0.001}
  n23 = ScalarParameter() {defaultValue=0.25, parameterName=Origin_0-1}
  n24 = LinearInterpolate(Alpha=ScalarParameter() {parameterName=bInvertX}) {constA=-1}
  n25 = LinearInterpolate(Alpha=ScalarParameter() {parameterName=bInvertY}) {constA=1, constB=-1}
  n26 = MaterialFunctionCall(UVs=TextureCoordinate(), "Rotation Angle (0-1)"=Multiply(A=Add(A=@n22, B=@n23), B=Multiply(A=@n24, B=@n25))) {materialFunction=&CustomRotator}
  n27 = Constant2Vector() {g=2, r=2}
  n28 = AppendVector(A=@n24, B=@n25)
  n29 = MaterialFunctionCall(Vector or UVs=Multiply(A=Subtract(A=@n26.Rotated Values) {constB=0.5}, B=Multiply(A=@n27, B=@n28)) {constB=2}) {materialFunction=&VectorToRadialValue}
  n30 = Multiply(
   A=SmoothStep(
    Min=@n16
    Max=Add(A=@n16) {constB=0.5}
    Value=MaterialFunctionCall(CenterPosition=@n17, Radius=@n3, Density=@n18) {materialFunction=&RadialGradientExponential}.RadialGradientExponential
   )
   B=Floor(None=Add(A=@n21, B=@n29.Vector Converted to Angle))
  )
  n31 = Saturate(None=Multiply(A=@n9, B=Saturate(None=@n30)))
  Multiply(
   A=@n13
   B=LinearInterpolate(
    A=VectorParameter() {parameterName=BgColor, defaultValue=<rgba 0 0 0 1>}.RGB
    B=LinearInterpolate(A=@n14.RGB, B=@n15.RGB, Alpha=@n31)
    Alpha=@n9
   )
  )
 ]
}
