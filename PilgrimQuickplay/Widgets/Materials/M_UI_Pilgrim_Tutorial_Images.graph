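# Translucent UI-domain material for tutorial images in the Pilgrim Quickplay widgets, judging by its name and path.
# A single texture sample reads from a texture object parameter named Image, which exposes the picture as an override for material instances.
# The graph holds no other processing, such as tinting, masking or animation.
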Material /PilgrimQuickplay/Widgets/Materials/M_UI_Pilgrim_Tutorial_Images
graph {
 properties {blendMode=BLEND_Translucent, materialDomain=MD_UI, pixelDepthOffsetMode=PDOM_Legacy, usedShadingModels=Default Lit}
 nodes [TextureSample(Tex=TextureObjectParameter() {parameterName=Image})]
}
