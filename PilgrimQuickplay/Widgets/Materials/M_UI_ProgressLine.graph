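# Translucent UI material for a thin rounded progress line that slides horizontally as the Progress parameter changes.
# Progress is remapped across limits from box width times a hand-tuned ProgressFudge, offsetting the UVs fed into MF_UI_Box.
# Fill and outline masks carry separate opacities and tints before being summed.
# As wired, fill subtracts itself, leaving only the outline, and tints read only each color's red channel.
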Material /PilgrimQuickplay/Widgets/Materials/M_UI_ProgressLine
graph {
 properties {
  bAutomaticallySetUsageInEditor=false
  bUsedWithStaticMesh=false
  blendMode=BLEND_Translucent
  materialDomain=MD_UI
  usedShadingModels=Default Lit
 }
 nodes [
  n1 = VectorParameter() {group="Color && Opacity", parameterName=Color, sortPriority=1, defaultValue=<rgba 0.3789 0.8899 1 1>}
  n2 = VectorParameter() {group="Color && Opacity", parameterName=Outline Color, sortPriority=3, defaultValue=<rgba 0.6514 1 1 1>}
  Add(
   A=Multiply(
    A=MaterialFunctionCall(X=@n1.R, Y=@n1.R, Z=@n1.R, A=@n1.R) {materialFunction=&MakeFloat4}.Result
    B=NamedRerouteUsage()
   )
   B=Multiply(
    A=NamedRerouteUsage()
    B=MaterialFunctionCall(X=@n2.R, Y=@n2.R, Z=@n2.R, A=@n2.R) {materialFunction=&MakeFloat4}.Result
   )
  )
  Add(A=NamedRerouteUsage(), B=NamedRerouteUsage())
  n3 = ScalarParameter() {defaultValue=0.011, group=Box Settings, parameterName=Box Width, sliderMax=1, sortPriority=0}
  n4 = VectorParameter() {group=Box Settings, parameterName=Box Corners Radius, sortPriority=2, defaultValue=<rgba 0.2 0.2 0.2 0.2>}
  n5 = MaterialFunctionCall(
   UVs=NamedRerouteUsage()
   Size=MaterialFunctionCall(
    X=NamedRerouteDeclaration(None=@n3) {name=BoxWidth, nodeColor=<rgba 1 0 0.7266 1>}
    Y=ScalarParameter() {defaultValue=0.85, group=Box Settings, parameterName=Box Height, sliderMax=1, sortPriority=1}
   ) {materialFunction=&MakeFloat2}.Result
   CornerRoundness=@n4.RGBA
   OutlineThickness=ScalarParameter() {defaultValue=0.05, group=Box Settings, parameterName=Outline Thickness, sliderMax=100, sortPriority=3}
   OutlineAlignment=Constant() {r=-1}
   KeepAspectRatio=StaticBool() {value=true}
   PixelSize=NamedRerouteUsage()
  ) {materialFunction=&MF_UI_Box}
  n6 = Subtract(A=@n5.Fill, B=@n5.Fill)
  n7 = ScalarParameter() {defaultValue=1, group="Color && Opacity", parameterName=Opacity, sliderMax=1, sortPriority=0}
  NamedRerouteDeclaration(None=Multiply(A=Saturate(None=@n6), B=Saturate(None=@n7))) {name=Shape, nodeColor=<rgba 0 0 0 1>}
  n8 = ScalarParameter() {defaultValue=1, group="Color && Opacity", parameterName=Outline Opacity, sliderMax=1, sortPriority=2}
  NamedRerouteDeclaration(None=Multiply(A=@n5.Outline, B=Saturate(None=@n8))) {name=Outline, nodeColor=<rgba 0 0 0 1>}
  NamedRerouteDeclaration(None=MaterialFunctionCall() {materialFunction=&GetUserInterfaceUV}.Pixel Size) {name=Pixel Size, nodeColor=<rgba 0 0 0 1>}
  n9 = ScalarParameter() {
   defaultValue=44.95
   desc="Can't figure the math right now. Adjust this to have the box kiss the edge of the image at progress 0.0 and 1.0 as you adjust the width. Some kind of power of the Box width. Known Values:\r\nBox  Fudge\r\n===    =======\r\n0.5     0.5\r\n1.0     0.0\r\n0.75   0.1666\r\n0.25   1.5\r\n0.125 3.5"
   group=Progress
   parameterName=ProgressFudge
  }
  n10 = Multiply(A=NamedRerouteUsage(), B=@n9) {constB=3.5}
  NamedRerouteDeclaration(None=@n10) {name=ProgressLow, nodeColor=<rgba 1 0.6328 0 1>}
  NamedRerouteDeclaration(None=Multiply(A=@n10) {constB=-1}) {name=ProgressHigh, nodeColor=<rgba 0.3125 0 1 1>}
  n11 = MaterialFunctionCall(
   Input=Saturate(None=ScalarParameter() {group=Progress, parameterName=Progress})
   Target Low=NamedRerouteUsage()
   Target High=NamedRerouteUsage()
  ) {materialFunction=&RemapValueRange}
  n12 = Constant()
  NamedRerouteDeclaration(None=Add(
    A=TextureCoordinate()
    B=MaterialFunctionCall(X=@n11.Result, Y=@n12) {materialFunction=&MakeFloat2}.Result
   ) {constB=0}) {name=ProgressUVs, nodeColor=<rgba 0.0078 1 0 1>}
 ]
}
