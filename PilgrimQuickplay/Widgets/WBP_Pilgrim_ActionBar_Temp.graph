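# WBP_Pilgrim_ActionBar_Temp is a CommonUserWidget that provides a bottom action bar for Pilgrim Quickplay UI, apparently as a temporary version.
# An InvalidationBox wraps a bottom-aligned, undrawn Border and a SafeZone with top padding disabled.
# Inside, a right-aligned CommonBoundActionBar creates its buttons from the AthenaHUDBoundActionButton class.
# The widget is layout only, with no graph logic or variables.
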WidgetBlueprint /PilgrimQuickplay/Widgets/WBP_Pilgrim_ActionBar_Temp
graphs [EventGraph {links [], nodes []}]
variables []
widgets {
 info {inheritedWidgetCount=0, parentClass=&CommonUserWidget, rootWidgetClass=/Script/UMG.InvalidationBox, widgetCount=4}
 widgets [
  InvalidationRoot {widgetClassPath=&InvalidationBox, widget=InvalidationRoot}
  Border_BottomBorder {
   widgetClassPath=&Border
   bIsVariable=true
   parent=InvalidationRoot
   widget=Border_BottomBorder
   properties {
    verticalAlignment=VAlign_Bottom
    visibility=SelfHitTestInvisible
    background {
     bIsDynamicallyLoaded=false
     drawAs=NoDrawType
     imageType=NoImage
     mirroring=NoMirror
     resourceName=None
     resourceObject=null
     tiling=NoTile
     imageSize=<32 32>
     margin=<ltrb 0 0 0 0>
     outlineSettings {
      bUseBrushTransparency=false
      roundingType=HalfHeightRadius
      width=0
      color {colorUseRule=UseColor_Specified, specifiedColor=<rgba 0 0 0 0>}
      cornerRadii=<0 0 0 0>
     }
     tintColor {colorUseRule=UseColor_Specified, specifiedColor=<rgba 0.0075 0.0091 0.0242 0.8>}
     uVRegion {bIsValid=false, max=<0 0>, min=<0 0>}
    }
    padding=<ltrb 0 0 0 0>
   }
   slot {class=PanelSlot}
  }
  SafeZone {
   widgetClassPath=&SafeZone
   bIsVariable=true
   parent=Border_BottomBorder
   widget=SafeZone
   properties {padTop=false}
   slot {class=BorderSlot, properties {verticalAlignment=VAlign_Bottom, padding=<ltrb 0 0 0 0>}}
  }
  BoundActionBar {
   widgetClassPath=/Script/CommonUI.CommonBoundActionBar
   bIsVariable=true
   parent=SafeZone
   widget=BoundActionBar
   properties {
    actionButtonClass=/Game/UI/Frontend/BottomBar/AthenaHUDBoundActionButton.AthenaHUDBoundActionButton_C
    entryHorizontalAlignment=HAlign_Right
    entrySpacing=<4 0>
   }
   slot {class=SafeZoneSlot, properties {hAlign=HAlign_Right, padding=<ltrb 0 0 48 16>}}
  }
 ]
}
