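# Minimal Pilgrim Quickplay HUD, derived from PilgrimHUDBase, layering a primary content stack, the Athena HUD menu and a bottom action bar in an overlay.
# On Construct it collapses the Athena bottom bar, then waits for the player's event router before registering listeners.
# Main-menu events toggle the menu, one green room timer state force-closes settings, and opening splitscreen settings reshows the bottom bar after a short delay.
# Destruct stops all three listeners.
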WidgetBlueprint /PilgrimQuickplay/Widgets/PilgrimMinimalHUD
graphs [EventGraph {nodes [
   Event Construct()
   > SetAthenaBottomBarVisibility!(execute, self:Fort Null HUD Object Reference, NewVisibility:ESlateVisibility Enum=Collapsed, bIsUsingLegacyBottomBar:bool=false) -> (then)
   > n1 = PilgrimWaitForEventRouter!(execute:Exec, ContextActor:Actor Object Reference=GetOwningPlayer()) -> (then:Exec, OnEventRouterAvailable:Exec, EventRouter:Gameplay Event Router Component Object Reference) {is=title K2Node_AsyncAction}
   n2 = Sequence(execute=@n1.OnEventRouterAvailable)
   > n3 = StartListeningToEvent!(
    execute:Exec
    Target:Gameplay Event Router Component Object Reference=@n1
    EventType:Script Struct Object Reference=/PilgrimQuickplay/Items/PilgrimEvent_GreenRoomTimerStateChanged
    Context:Object Reference
    EventBubblingRule:EEventBubblingRule Enum
   ) -> (
    then:Exec
    AsyncTaskProxy:Async Action Start Listening To Event Object Reference
    OnEventReceived:Exec
    Payload_GreenRoomTimerState_2_#1:GreenRoomTimerState Enum
    Payload_FinalTimerLength_5_#2:double=~
   )
   > Set GreenRoomTimerHandle(GreenRoomTimerHandle:Async Action Start Listening To Event Object Reference=@n3) -> (Output_Get:Async Action Start Listening To Event Object Reference)
   > n4 = StartListeningToEvent!(execute, Target=GetGlobalEventRouter(), EventType=/FortniteCoreUI/TextChat/ToggleMainMenuEvent, Context=GetOwningPlayer(), EventBubblingRule) -> (then, AsyncTaskProxy, OnEventReceived, Payload:Toggle Main Menu Event Structure)
   > Set MainMenuHandle(MainMenuHandle:Async Action Start Listening To Event Object Reference=@n4) -> (Output_Get:Async Action Start Listening To Event Object Reference)
   ToggleMainMenu(execute=@n4.OnEventReceived, self:Pilgrim HUDBase Object Reference)
   n5 = Switch on GreenRoomTimerState!(execute=@n3.OnEventReceived, Selection:GreenRoomTimerState Enum=@n3.Payload_GreenRoomTimerState_2_#1) -> (NewEnumerator0:Exec, NewEnumerator1:Exec, NewEnumerator2:Exec, NewEnumerator3:Exec, NewEnumerator4:Exec) {is=title K2Node_SwitchEnum}
   ForceCloseSettingsMenu(execute=@n5.NewEnumerator2, self:Pilgrim HUDBase Object Reference)
   n6 = StartListeningToEvent(execute=@n2.then_1, Target=@n1, EventType=/Script/PilgrimSettingsRuntime.PilgrimEvent_PilgrimSettings_SplitscreenSettingsOpened) -> (Payload:Pilgrim Event Pilgrim Settings Splitscreen Settings Opened Structure)
   > Set SplitScreenSettingsHandle(SplitScreenSettingsHandle:Async Action Start Listening To Event Object Reference=@n6) -> (Output_Get:Async Action Start Listening To Event Object Reference)
   Delay(execute=@n6.OnEventReceived, Duration=0.1)
   > SetAthenaBottomBarVisibility(NewVisibility=SelfHitTestInvisible, bIsUsingLegacyBottomBar=false)
   Event Destruct()
   > StopListeningToEvent(self:Async Action Start Listening To Event Object Reference=Get MainMenuHandle() -> (MainMenuHandle:Async Action Start Listening To Event Object Reference))
   > StopListeningToEvent(self=Get GreenRoomTimerHandle() -> (GreenRoomTimerHandle:Async Action Start Listening To Event Object Reference))
   > StopListeningToEvent(self=Get SplitScreenSettingsHandle() -> (SplitScreenSettingsHandle:Async Action Start Listening To Event Object Reference))
  ]}]
variables [
 GreenRoomTimerHandle {type=AsyncAction_StartListeningToEvent, default=null}
 MainMenuHandle {type=AsyncAction_StartListeningToEvent, default=null}
 SplitScreenSettingsHandle {type=AsyncAction_StartListeningToEvent, default=null}
]
widgets {
 info {inheritedWidgetCount=3, parentClass=/Script/PilgrimQuickplayUI.PilgrimHUDBase, rootWidgetClass=&Overlay, widgetCount=4}
 widgets [
  Overlay_0 {widgetClassPath=&Overlay, widget=Overlay_0}
  WidgetStack_PrimaryContent {
   widgetClassPath=/Script/CommonUI.CommonActivatableWidgetStack
   bInherited=true
   bIsVariable=true
   parent=Overlay_0
   widget=WidgetStack_PrimaryContent
   properties {transitionDuration=0, visibility=SelfHitTestInvisible}
   slot {class=OverlaySlot, properties {horizontalAlignment=HAlign_Fill, verticalAlignment=VAlign_Fill}}
  }
  AthenaHUDMenu {
   widgetClassPath=/FortniteCoreUI/HUD/WBP_AthenaHUDMenu
   bInherited=true
   parent=Overlay_0
   widget=AthenaHUDMenu
   slot {class=OverlaySlot, properties {horizontalAlignment=HAlign_Fill, verticalAlignment=VAlign_Fill}}
  }
  ActionBar_BottomBar {
   widgetClassPath=&/WBP_Pilgrim_ActionBar_Temp
   bInherited=true
   bIsVariable=true
   parent=Overlay_0
   widget=ActionBar_BottomBar
   slot {class=OverlaySlot, properties {horizontalAlignment=HAlign_Right, verticalAlignment=VAlign_Bottom}}
  }
 ]
}
