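# Debug overlay widget that plots auto-calibration offset (labelled Latency) and sample average as two scrolling graphs with numeric readouts.
# It shows or collapses itself following the debug flag on the owning player's FMCalibrationControllerComponent.
# When debug is first enabled, it waits once for the event router, then listens for AutoCalibration OffsetChanged events.
# Plotted values are scaled and centred at 0.5, while text shows raw values and replaces the graphs' hidden built-in labels.
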WidgetBlueprint /PilgrimQuickplay/Widgets/WBP_AutocalibratorDebug
graphs [
 AddGraphPoint {nodes [
   n1 = AddGraphPoint!() -> (then, Graph:WB Audio Analysis Float Graph Object Reference, Value:double=~) {is=title K2Node_FunctionEntry}
   > Set Value!(execute, self:WB Audio Analysis Float Graph Object Reference=@n1, value:double=ScaleValue!(Value:double=@n1.Value) -> (Result:double)) -> (then) {is=title K2Node_CallFunction}
  ]}
 ConfigureGraphs {nodes [
   ConfigureGraphs() {is=title K2Node_FunctionEntry}
   > SetVisibility(
    self=Get Value!(self:WB Audio Analysis Float Graph Object Reference=Get LatencyGraph() -> (LatencyGraph:WB Audio Analysis Float Graph Object Reference)) -> (Value:Text Object Reference)
    InVisibility=Hidden
   )
   > SetVisibility(self=Get Value(self=Get SampleAverageGraph() -> (SampleAverageGraph:WB Audio Analysis Float Graph Object Reference)), InVisibility=Hidden)
  ]}
 EventGraph {nodes [
   n1 = OnDebugUpdated!() -> (OutputDelegate, then, CalibrationControllerComponent:FMCalibration Controller Component Object Reference, IsCalibrationDebugEnabled:bool=~) {is=title K2Node_CustomEvent}
   > UpdateVisibility!(execute, self, Index:bool=@n1.IsCalibrationDebugEnabled) -> (then)
   > n2 = Branch(Condition=@n1.IsCalibrationDebugEnabled)
   Do Once(execute=[@n9.then, @n2.then])
   > n3 = PilgrimWaitForEventRouter!(execute:Exec, ContextActor:Actor Object Reference=GetOwningPlayer()) -> (then:Exec, OnEventRouterAvailable:Exec, EventRouter:Gameplay Event Router Component Object Reference) {is=title K2Node_AsyncAction}
   n4 = StartListeningToEvent!(
    execute:Exec=@n3.OnEventRouterAvailable
    Target:Gameplay Event Router Component Object Reference=@n3
    EventType:Script Struct Object Reference=/Script/PilgrimCoreRuntime.PilgrimEvent_AutoCalibration_OffsetChanged
    Context:Object Reference
    EventBubblingRule:EEventBubblingRule Enum
   ) -> (
    then:Exec
    AsyncTaskProxy:Async Action Start Listening To Event Object Reference
    OnEventReceived:Exec
    Payload:Pilgrim Event Auto Calibration Offset Changed Structure
   )
   n5 = Sequence(execute=@n4.OnEventReceived)
   n6 = Break Pilgrim Event Auto Calibration Offset Changed!(PilgrimEvent_AutoCalibration_OffsetChanged:"Pilgrim Event Auto Calibration Offset Changed Structure (by ref)"=@n4.Payload) -> (Offset:float=~, SampleAverage:float=~) {is=title K2Node_BreakStruct}
   > AddGraphPoint!(execute, self, Graph:WB Audio Analysis Float Graph Object Reference=Get LatencyGraph(), Value:double=@n6) -> (then) {is=title K2Node_CallFunction}
   > "SetText (Text)"(
    self=Get Latency() -> (Latency:Common Text Object Reference)
    InText="To Text (Float)"(Value=@n6, MinimumIntegralDigits=1, MaximumIntegralDigits=324, MinimumFractionalDigits=0, MaximumFractionalDigits=3)
   )
   AddGraphPoint(execute=@n5.then_1, Graph=Get SampleAverageGraph(), Value=@n6.SampleAverage)
   > "SetText (Text)"(
    self=Get SampleAverage() -> (SampleAverage:Common Text Object Reference)
    InText="To Text (Float)"(Value=@n6.SampleAverage, MinimumIntegralDigits=1, MaximumIntegralDigits=324, MinimumFractionalDigits=0, MaximumFractionalDigits=3)
   )
   Event OnInitialized()
   > ConfigureGraphs() {is=title K2Node_CallFunction}
   n7 = GetComponentByClass!(
    self:Actor Object Reference=GetOwningPlayer()
    ComponentClass:Actor Component Class Reference=/Script/FMCalibrationRuntime.FMCalibrationControllerComponent
   ) -> (ReturnValue:FMCalibration Controller Component Object Reference)
   > Bind Event to On Calibration Debug Updated(self:FMCalibration Controller Component Object Reference=@n7, Delegate=@n1) {is=title K2Node_AddDelegate}
   n8 = GetIsCalibrationDebugEnabled!(self:FMCalibration Controller Component Object Reference=@n7) -> (ReturnValue:bool)
   > UpdateVisibility(Index=@n8)
   > n9 = Branch(Condition=@n8)
   Event PreConstruct()
   Event Construct()
   Event Tick()
  ]}
 ScaleValue {nodes [
   n1 = ScaleValue!() -> (then, Value:double=~) {is=title K2Node_FunctionEntry}
   > Return Node!(execute, Result:double=float + float(A=float / float(A=@n1, B=200), B=0.5))
  ]}
 UpdateVisibility {nodes [
   n1 = UpdateVisibility!() -> (then, Index:bool=~) {is=title K2Node_FunctionEntry}
   > SetVisibility(InVisibility=Select(Option 0=Collapsed, Option 1=HitTestInvisible, Index=@n1))
  ]}
]
variables []
widgets {
 info {inheritedWidgetCount=0, parentClass=&CommonUserWidget, rootWidgetClass=&Overlay, widgetCount=8}
 widgets [
  Overlay_26 {widgetClassPath=&Overlay, widget=Overlay_26}
  HorizontalBox_0 {
   widgetClassPath=&HorizontalBox
   parent=Overlay_26
   widget=HorizontalBox_0
   slot {class=OverlaySlot, properties {horizontalAlignment=HAlign_Fill, verticalAlignment=VAlign_Fill}}
  }
  VerticalBox_0 {widgetClassPath=&VerticalBox, parent=HorizontalBox_0, widget=VerticalBox_0, slot {class=HorizontalBoxSlot}}
  LatencyGraph {
   widgetClassPath=/Game/Sounds/AudioAnalysis/DebugHUD/WB_AudioAnalysis_FloatGraph
   bIsVariable=true
   parent=VerticalBox_0
   widget=LatencyGraph
   properties {numPoints=1200, parameter=Latency, color=<rgba 0.0671 0 1 0.5>, size=<600 100>}
   slot {class=VerticalBoxSlot}
  }
  Latency {
   widgetClassPath=&CommonTextBlock
   bIsVariable=true
   parent=VerticalBox_0
   widget=Latency
   properties {
    style=&TextStyle-Base_C
    font {
     bForceMonospaced=false
     bMaterialIsStencil=false
     fontMaterial=null
     fontObject=&NotoSans
     letterSpacing=0
     monospacedWidth=1
     size=24
     skewAmount=0
     typefaceFontName=NotoSans-Regular
     outlineSettings {
      bApplyOutlineToDropShadows=false
      bMiteredCorners=false
      bSeparateFillAlpha=false
      outlineMaterial=null
      outlineSize=0
      outlineColor=<rgba 0 0 0 1>
     }
    }
    shadowOffset=<0 0>
   }
   slot {class=VerticalBoxSlot, properties {horizontalAlignment=HAlign_Right}}
  }
  VerticalBox_1 {widgetClassPath=&VerticalBox, parent=HorizontalBox_0, widget=VerticalBox_1, slot {class=HorizontalBoxSlot}}
  SampleAverageGraph {
   widgetClassPath=&WB_AudioAnalysis_FloatGraph
   bIsVariable=true
   parent=VerticalBox_1
   widget=SampleAverageGraph
   properties {numPoints=1200, parameter=Sample Average, color=<rgba 0.0671 0 1 0.5>, size=<600 100>}
   slot {class=VerticalBoxSlot}
  }
  SampleAverage {
   widgetClassPath=&CommonTextBlock
   bIsVariable=true
   parent=VerticalBox_1
   widget=SampleAverage
   properties {
    style=&TextStyle-Base_C
    font {
     bForceMonospaced=false
     bMaterialIsStencil=false
     fontMaterial=null
     fontObject=&NotoSans
     letterSpacing=0
     monospacedWidth=1
     size=24
     skewAmount=0
     typefaceFontName=NotoSans-Regular
     outlineSettings {
      bApplyOutlineToDropShadows=false
      bMiteredCorners=false
      bSeparateFillAlpha=false
      outlineMaterial=null
      outlineSize=0
      outlineColor=<rgba 0 0 0 1>
     }
    }
    shadowOffset=<0 0>
   }
   slot {class=VerticalBoxSlot, properties {horizontalAlignment=HAlign_Right}}
  }
 ]
}
