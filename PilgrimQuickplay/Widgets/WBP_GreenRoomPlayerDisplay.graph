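# Player card widget for the Pilgrim Quickplay green room, showing a player's name, avatar, instrument, difficulty letter and ready state.
# It starts hidden; SetPlayer shows it for a valid PlayerState and reads instrument and difficulty from the Quickplay player-state component.
# The avatar loads asynchronously, retrying each second up to a limit while its texture reference is invalid.
# Instrument colors derive from item rarity, and a separate function updates ready text and background tint.
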WidgetBlueprint /PilgrimQuickplay/Widgets/WBP_GreenRoomPlayerDisplay
graphs [
 EventGraph {nodes [
   n1 = SetPlayerImage!() -> (OutputDelegate, then, FortPlayerStateAthena:Fort Player State Athena Object Reference) {is=title K2Node_CustomEvent}
   > n2 = GetDynamicMaterial(self=Get PlayerImage() -> (PlayerImage:Image Object Reference))
   > SetVectorParameterValue(
    self=@n2
    ParameterName=BG_Color
    Value=GetPinColor!(self:Fort Player State Zone Object Reference=@n1.FortPlayerStateAthena) -> (ReturnValue:Linear Color Structure)
   )
   > n3 = Set PlayerIconRetryCount(PlayerIconRetryCount:int=Get PlayerIconRetryCount_Default() -> (PlayerIconRetryCount_Default:int=0)) -> (Output_Get:int=0)
   n4 = GetSocialAvatarBrush!(self:Fort Player State Zone Object Reference=@n1.FortPlayerStateAthena, bSmallImage:bool=true) -> (ReturnValue:Texture 2D Soft Object Reference)
   n5 = Branch(execute=[@n9.then, @n3.then], Condition=IsValidSoftObjectReference(SoftObjectReference=@n4))
   > n6 = Async Load Asset(Asset=@n4)
   n7 = Cast To Texture2D(execute=@n6.Completed, Object=@n6)
   > SetTextureParameterValue(self=@n2, ParameterName=PlayerIcon, Value=@n7)
   n8 = Branch(execute=@n5.else, Condition="integer > integer"(A=Get PlayerIconRetryCount() -> (PlayerIconRetryCount:int=0)))
   > Decrement Int(Value=Get PlayerIconRetryCount() -> (PlayerIconRetryCount=0))
   > n9 = Delay(Duration=1)
   PrintString!(execute=@n8.else, InString=Failed to get a vaild texture reference for player, bPrintToScreen, bPrintToLog, TextColor, Duration, Key) -> (then)
   Event Construct()
   > HidePlayerDisplay()
  ]}
 HidePlayerDisplay {nodes [
   HidePlayerDisplay() {is=title K2Node_FunctionEntry}
   > SetVisibility(self=Self-Reference(), InVisibility=Hidden)
  ]}
 Set Instrument Type {nodes [
   n1 = GetDynamicMaterial(self=Get InstrumentImage() -> (InstrumentImage:Image Object Reference))
   > n2 = Sequence()
   n3 = Select!(
    TrackGuitar:Fort World Item Definition Object Reference=/PilgrimCore/Weapons/WID_Pilgrim_BandQuickPlay_Guitar
    TrackBass:Fort World Item Definition Object Reference=/PilgrimCore/Weapons/WID_Pilgrim_BandQuickPlay_Bass
    TrackVocals:Fort World Item Definition Object Reference=/PilgrimCore/Weapons/WID_Pilgrim_BandQuickPlay_Mic
    TrackDrum:Fort World Item Definition Object Reference=/PilgrimCore/Weapons/WID_Pilgrim_BandQuickPlay_Drums
    TrackPlasticGuitar:Fort World Item Definition Object Reference=&WID_Pilgrim_BandQuickPlay_Guitar
    TrackPlasticBass:Fort World Item Definition Object Reference=&WID_Pilgrim_BandQuickPlay_Bass
    TrackPlasticDrum:Fort World Item Definition Object Reference=&WID_Pilgrim_BandQuickPlay_Drums
    TrackProVocals:Fort World Item Definition Object Reference=&WID_Pilgrim_BandQuickPlay_Mic
    TrackProCymbals:Fort World Item Definition Object Reference=&WID_Pilgrim_BandQuickPlay_Drums
    TrackEvents:Fort World Item Definition Object Reference
    TrackNone:Fort World Item Definition Object Reference=&WID_Pilgrim_BandQuickPlay_Guitar
    Index:EPilgrimTrackType Enum=TrackGuitar
   ) -> (ReturnValue:Fort World Item Definition Object Reference)
   > n4 = GetRarityColorPalette!(execute, Rarity:EFortRarity Enum=GetRarity!(self:Fort Item Definition Object Reference=@n3) -> (ReturnValue:EFortRarity Enum=Common)) -> (then, ReturnValue:Fort Color Palette Structure)
   n5 = Break Fort Color Palette!(FortColorPalette:"Fort Color Palette Structure (by ref)"=@n4) -> (
    Color1:Linear Color Structure
    Color2:Linear Color Structure
    Color3:Linear Color Structure
    Color4:Linear Color Structure
    Color5:Linear Color Structure
   ) {is=title K2Node_BreakStruct}
   > SetVectorParameterValue(self=@n1, ParameterName=BG_Hotspot_Color_Hover, Value=@n5)
   > SetVectorParameterValue(self=@n1, ParameterName=BG_Hotspot_Color, Value=@n5.Color2)
   > SetVectorParameterValue(self=@n1, ParameterName=BG_Falloff_Color, Value=@n5.Color3)
   SetTextureParameterValue(
    execute=@n2.then_1
    self=@n1
    ParameterName=ItemIcon
    Value=Resolve Soft Reference!(Input:Texture 2D Soft Object Reference=GetLargePreviewImage!(self:Fort Item Definition Object Reference=@n3) -> (ReturnValue:Texture 2D Soft Object Reference)) -> (Output:Texture 2D Object Reference)
   )
   n6 = Set Instrument Type!() -> (then, TrackType:EPilgrimTrackType Enum=TrackGuitar) {is=title K2Node_FunctionEntry}
   > SetBrushFromSoftMaterial!(
    execute
    self:Image Object Reference=Get InstrumentImage()
    SoftMaterial:Material Interface Soft Object Reference=GetIconMaterialByInstrument!(
     Instrument:EPilgrimTrackType Enum=@n6
     SongMetaData:Pilgrim Song Metadata Object Reference
     SparksInstrument:ESparksInstrumentType Enum=None
     bUseKeytarIcon:bool=false
     UseProIcons:bool=false
    ) -> (IconMaterial:Material Instance Soft Object Reference)
   ) -> (then)
  ]}
 Set Ready State UI {nodes [
   n1 = Set Ready State UI!() -> (then, IsReady:bool=~) {is=title K2Node_FunctionEntry}
   n2 = MakeColor!(R:float=0.1, G:float=0.75, B:float=0.1, A:float=1) -> (ReturnValue:Linear Color Structure)
   n3 = MakeColor(R=0.05, G=0.05, B=0.05, A=1)
   > SetBrushTintColor!(
    execute
    self:Image Object Reference=Get Background() -> (Background:Image Object Reference)
    TintColor:Slate Color Structure=Select!(
     Option 0:Slate Color Structure=Make Slate Color(SpecifiedColor=@n3)
     Option 1:Slate Color Structure=Make Slate Color(SpecifiedColor=@n2)
     Index:bool=@n1
    ) -> (ReturnValue:Slate Color Structure)
   ) -> (then)
   > Set Text(
    Text=Select(Option 0:Text="LOCTEXT(\"Not Ready\")", Option 1:Text="LOCTEXT(\"READY!\")", Index=@n1) -> (ReturnValue:Text)
    self=Get ReadyText() -> (ReadyText:Common Text Object Reference)
   )
  ]}
 SetDifficultyDisplay {nodes [
   n1 = SetDifficultyDisplay!() -> (then, Song Difficulty:EPilgrimSongDifficulty Enum=DifficultyEasy) {is=title K2Node_FunctionEntry}
   > "SetText (Text)"(
    self=Get DifficultyName() -> (DifficultyName:Common Text Object Reference)
    InText=Select!(
     DifficultyEasy:Text="LOCTEXT(\"E\")"
     DifficultyMedium:Text="LOCTEXT(\"M\")"
     DifficultyHard:Text="LOCTEXT(\"H\")"
     DifficultyExpert:Text="LOCTEXT(\"X\")"
     None:Text="LOCTEXT(\"?\")"
     Index:EPilgrimSongDifficulty Enum=@n1
    ) -> (ReturnValue)
   )
  ]}
 SetPlayer {nodes [
   n1 = SetPlayer!() -> (then, PlayerState:Player State Object Reference) {is=title K2Node_FunctionEntry}
   > n2 = Is Valid(InputObject=@n1)
   HidePlayerDisplay(execute=@n2.Is Not Valid) {is=title K2Node_CallFunction}
   ShowPlayerDisplay(execute=@n2.Is Valid)
   > n3 = Set MyPlayerState(MyPlayerState:Player State Object Reference=@n1) -> (Output_Get:Player State Object Reference)
   > "SetText (Text)"(
    self=Get PlayerName() -> (PlayerName:Common Text Object Reference)
    InText="To Text (String)"(InString=GetPlayerNameSafe!(Actor:Actor Object Reference=@n3) -> (ReturnValue:String))
   )
   > n4 = Cast To FortPlayerStateAthena!(execute, Object=@n3) -> (then, CastFailed, AsFort Player State Athena:Fort Player State Athena Object Reference) {is=title K2Node_DynamicCast}
   > n5 = Sequence()
   > SetPlayerImage!(execute, self, FortPlayerStateAthena:Fort Player State Athena Object Reference=@n4) -> (then) {is=title K2Node_CallFunction}
   n6 = GetPilgrimQuickplayPlayerStateComponent!(execute=@n5.then_1, PlayerState:Player State Object Reference=Get MyPlayerState() -> (MyPlayerState:Player State Object Reference)) -> (then, ReturnValue:Pilgrim Quickplay Player State Component Object Reference)
   > SetDifficultyDisplay!(
    execute
    self
    Song Difficulty:EPilgrimSongDifficulty Enum=GetSongDifficulty!(self:Pilgrim Core Player State Component Object Reference=@n6) -> (ReturnValue:EPilgrimSongDifficulty Enum=DifficultyEasy)
   ) -> (then) {is=title K2Node_CallFunction}
   > Set Instrument Type!(
    execute
    self
    TrackType:EPilgrimTrackType Enum=GetInstrumentChoice!(self:Pilgrim Core Player State Component Object Reference=@n6) -> (ReturnValue:EPilgrimTrackType Enum=TrackGuitar)
   ) -> (then) {is=title K2Node_CallFunction}
   > SetPlayerVote!(execute, self, PilgrimSongMetadata:Pilgrim Song Metadata Object Reference) -> (then)
   PrintString(execute=@n4.CastFailed, InString="CAST FAILED!", TextColor=(R=1,G=0.078,B=0,A=1), Duration=8)
  ]}
 SetPlayerVote {nodes [
   n1 = SetPlayerVote!() -> (then, PilgrimSongMetadata:Pilgrim Song Metadata Object Reference) {is=title K2Node_FunctionEntry}
   > n2 = Is Valid(InputObject=@n1)
   "SetText (Text)"(execute=@n2.Is Not Valid, self=Get PlayerVote() -> (PlayerVote:Common Text Object Reference))
   "SetText (Text)"(execute=@n2.Is Valid, self=Get PlayerVote(), InText=GetTitleText!(self:Sparks Song Data Object Reference=@n1) -> (ReturnValue:Text))
  ]}
 SetToDefault {nodes [SetToDefault() {is=title K2Node_FunctionEntry}]}
 ShowPlayerDisplay {nodes [
   ShowPlayerDisplay() {is=title K2Node_FunctionEntry}
   > SetVisibility(self=Self-Reference(), InVisibility=SelfHitTestInvisible)
  ]}
]
variables [
 MyPlayerState {type=PlayerState, default=null}
 PlayerIconRetryCount {type=integer, default=10}
 PlayerIconRetryCount_Default {type=integer, default=10}
]
widgets {
 info {inheritedWidgetCount=0, parentClass=&UserWidget, rootWidgetClass=&Overlay, widgetCount=12}
 widgets [
  Overlay_11 {widgetClassPath=&Overlay, widget=Overlay_11}
  Background {
   widgetClassPath=&Image
   bIsVariable=true
   parent=Overlay_11
   widget=Background
   properties {brush {
     bIsDynamicallyLoaded=false
     drawAs=Image
     imageType=NoImage
     mirroring=NoMirror
     resourceName=None
     resourceObject=/Game/Environments/World/Utility/128gray
     tiling=NoTile
     imageSize=<500 250>
     margin=<ltrb 0 0 0 0>
     outlineSettings {
      bUseBrushTransparency=false
      roundingType=HalfHeightRadius
      width=0
      color {colorUseRule=UseColor_Specified, specifiedColor=<rgba 0 0 0 0>}
      cornerRadii=<0 0 0 0>
     }
     tintColor {colorUseRule=UseColor_Specified, specifiedColor=<rgba 0.05 0.05 0.05 1>}
     uVRegion {bIsValid=false, max=<0 0>, min=<0 0>}
    }}
   slot {class=OverlaySlot, properties {horizontalAlignment=HAlign_Fill, verticalAlignment=VAlign_Fill}}
  }
  VerticalBox_115 {
   widgetClassPath=&VerticalBox
   parent=Overlay_11
   widget=VerticalBox_115
   slot {class=OverlaySlot, properties {horizontalAlignment=HAlign_Fill, verticalAlignment=VAlign_Center}}
  }
  PlayerName {
   widgetClassPath=&CommonTextBlock
   bIsVariable=true
   parent=VerticalBox_115
   widget=PlayerName
   properties {
    scrollStyle=&TextScrollStyle-Base_C
    style=/FortUILibrary/Styles/Typography/Foundation/TextStyle_BurbankSmall_M_White.TextStyle_BurbankSmall_M_White_C
    text="{PlayerName}"
    font {
     bForceMonospaced=false
     bMaterialIsStencil=false
     fontMaterial=null
     fontObject=&BurbankSmall
     letterSpacing=0
     monospacedWidth=1
     size=24
     skewAmount=0
     typefaceFontName=Black
     outlineSettings {
      bApplyOutlineToDropShadows=false
      bMiteredCorners=false
      bSeparateFillAlpha=false
      outlineMaterial=null
      outlineSize=0
      outlineColor=<rgba 0 0 0 1>
     }
    }
    shadowOffset=<0 0>
   }
   slot {class=VerticalBoxSlot, properties {horizontalAlignment=HAlign_Center, verticalAlignment=VAlign_Center}}
  }
  Spacer_153 {widgetClassPath=&Spacer, parent=VerticalBox_115, widget=Spacer_153, properties {size=<1 10>}, slot {class=VerticalBoxSlot}}
  HorizontalBox_0 {
   widgetClassPath=&HorizontalBox
   parent=VerticalBox_115
   widget=HorizontalBox_0
   slot {class=VerticalBoxSlot, properties {verticalAlignment=VAlign_Center}}
  }
  InstrumentImage {
   widgetClassPath=&Image
   bIsVariable=true
   parent=HorizontalBox_0
   widget=InstrumentImage
   properties {brush {
     bIsDynamicallyLoaded=false
     drawAs=Image
     imageType=NoImage
     mirroring=NoMirror
     resourceName=None
     resourceObject=/PilgrimQuickplay/Widgets/Materials/MI_UI_Icon_Instrument_Type_Guitar
     tiling=NoTile
     imageSize=<52 52>
     margin=<ltrb 0 0 0 0>
     outlineSettings {
      bUseBrushTransparency=false
      roundingType=HalfHeightRadius
      width=0
      color {colorUseRule=UseColor_Specified, specifiedColor=<rgba 0 0 0 0>}
      cornerRadii=<0 0 0 0>
     }
     tintColor {colorUseRule=UseColor_Specified, specifiedColor=<rgba 1 1 1 1>}
     uVRegion {bIsValid=false, max=<0 0>, min=<0 0>}
    }}
   slot {class=HorizontalBoxSlot, properties {size {sizeRule=Fill, value=1}}}
  }
  DifficultyName {
   widgetClassPath=&CommonTextBlock
   bIsVariable=true
   parent=HorizontalBox_0
   widget=DifficultyName
   properties {
    justification=Center
    style=/FortUILibrary/Styles/Typography/_TextStyle_SupersizeH3._TextStyle_SupersizeH3_C
    text=D
    font {
     bForceMonospaced=false
     bMaterialIsStencil=false
     fontMaterial=null
     fontObject=&Burbank
     letterSpacing=0
     monospacedWidth=1
     size=84
     skewAmount=0
     typefaceFontName=BurbankRegular-Black
     outlineSettings {
      bApplyOutlineToDropShadows=false
      bMiteredCorners=false
      bSeparateFillAlpha=false
      outlineMaterial=null
      outlineSize=0
      outlineColor=<rgba 0 0 0 1>
     }
    }
    shadowOffset=<0 0>
    strikeBrush {
     bIsDynamicallyLoaded=false
     drawAs=NoDrawType
     imageType=NoImage
     mirroring=NoMirror
     resourceName=None
     resourceObject=null
     tiling=NoTile
     imageSize=<32 32>
     margin=<ltrb 0 0 0 0>
     outlineSettings {
      bUseBrushTransparency=false
      roundingType=HalfHeightRadius
      width=0
      color {colorUseRule=UseColor_Specified, specifiedColor=<rgba 0 0 0 0>}
      cornerRadii=<0 0 0 0>
     }
     tintColor {colorUseRule=UseColor_Specified, specifiedColor=<rgba 1 1 1 1>}
     uVRegion {bIsValid=false, max=<0 0>, min=<0 0>}
    }
   }
   slot {class=HorizontalBoxSlot, properties {verticalAlignment=VAlign_Center, size {sizeRule=Fill, value=1}}}
  }
  PlayerImage {
   widgetClassPath=&Image
   bIsVariable=true
   parent=HorizontalBox_0
   widget=PlayerImage
   properties {brush {
     bIsDynamicallyLoaded=false
     drawAs=Image
     imageType=FullColor
     mirroring=NoMirror
     resourceName=None
     resourceObject=/SparksUICommon/Materials/MI_UI_BandMember_Avatar
     tiling=NoTile
     imageSize=<128 128>
     margin=<ltrb 0 0 0 0>
     outlineSettings {
      bUseBrushTransparency=false
      roundingType=HalfHeightRadius
      width=0
      color {colorUseRule=UseColor_Specified, specifiedColor=<rgba 0 0 0 0>}
      cornerRadii=<0 0 0 0>
     }
     tintColor {colorUseRule=UseColor_Specified, specifiedColor=<rgba 1 1 1 1>}
     uVRegion {bIsValid=false, max=<0 0>, min=<0 0>}
    }}
   slot {
    class=HorizontalBoxSlot
    properties {horizontalAlignment=HAlign_Right, verticalAlignment=VAlign_Center, padding=<ltrb 10 10 10 10>, size {sizeRule=Fill, value=1}}
   }
  }
  Spacer_219 {widgetClassPath=&Spacer, parent=VerticalBox_115, widget=Spacer_219, properties {size=<1 10>}, slot {class=VerticalBoxSlot}}
  ReadyText {
   widgetClassPath=&CommonTextBlock
   bIsVariable=true
   parent=VerticalBox_115
   widget=ReadyText
   properties {
    scrollStyle=&TextScrollStyle-Base_C
    style=&TextStyle_BurbankSmall_M_White_C
    text=Not Ready
    visibility=HitTestInvisible
    font {
     bForceMonospaced=false
     bMaterialIsStencil=false
     fontMaterial=null
     fontObject=&BurbankSmall
     letterSpacing=0
     monospacedWidth=1
     size=24
     skewAmount=0
     typefaceFontName=Black
     outlineSettings {
      bApplyOutlineToDropShadows=false
      bMiteredCorners=false
      bSeparateFillAlpha=false
      outlineMaterial=null
      outlineSize=0
      outlineColor=<rgba 0 0 0 1>
     }
    }
    shadowOffset=<0 0>
   }
   slot {class=VerticalBoxSlot, properties {horizontalAlignment=HAlign_Center, verticalAlignment=VAlign_Center}}
  }
  PlayerVote {
   widgetClassPath=&CommonTextBlock
   bIsVariable=true
   parent=VerticalBox_115
   widget=PlayerVote
   properties {
    scrollStyle=&TextScrollStyle-Base_C
    style=&TextStyle_BurbankSmall_M_White_C
    text=Not Ready
    visibility=Collapsed
    font {
     bForceMonospaced=false
     bMaterialIsStencil=false
     fontMaterial=null
     fontObject=&BurbankSmall
     letterSpacing=0
     monospacedWidth=1
     size=24
     skewAmount=0
     typefaceFontName=Black
     outlineSettings {
      bApplyOutlineToDropShadows=false
      bMiteredCorners=false
      bSeparateFillAlpha=false
      outlineMaterial=null
      outlineSize=0
      outlineColor=<rgba 0 0 0 1>
     }
    }
    shadowOffset=<0 0>
   }
   slot {class=VerticalBoxSlot, properties {horizontalAlignment=HAlign_Center, verticalAlignment=VAlign_Center}}
  }
 ]
}
