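# FM_VoteDone is a Sound Cue in SparksMaps that, judging by its name, signals the end of a vote.
# A Mixer layers two waves: VoteDone, pitched up to double through a Modulator, and VoteGo, held back a quarter second by a fixed Delay.
# The cue runs at reduced volume under the in-game SFX sound class.
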SoundCue /SparksMaps/Audio/FM_VoteDone
root=Mixer_0
nodes [
 {id=Delay_0, children [WavePlayer_5], properties {delayMax=0.25, delayMin=0.25}}
 {id=Mixer_0, children [Modulator_0,Delay_0], properties {inputVolume [1,1]}}
 {id=Modulator_0, children [WavePlayer_0], properties {pitchMax=2, pitchMin=2}}
 {id=WavePlayer_0, properties {soundWaveAssetPtr=/SparksMaps/Audio/Wavs/VoteDone}}
 {id=WavePlayer_5, properties {soundWaveAssetPtr=/SparksMaps/Audio/Wavs/VoteGo}}
]
properties {maxDistance=2097152, soundClassObject=&Fort_SFX_InGame, volumeMultiplier=0.6}
