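# Sound cue for a vote tumbler sound effect in SparksMaps, judging by the asset and wave names.
# A Random node at the root picks one of nine VoteTumbler wave variations with equal weight.
# The cue plays at reduced volume through the Fort_SFX_InGame sound class, with a very large max distance.
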SoundCue /SparksMaps/Audio/FM_VoteTumblir
root=Random_0
nodes [
 {
  id=Random_0
  children [WavePlayer_0,WavePlayer_1,WavePlayer_2,WavePlayer_3,WavePlayer_4,WavePlayer_5,WavePlayer_6,WavePlayer_7,WavePlayer_8]
  properties {weights [1,1,1,1,1,1,1,1,1]}
 }
 {id=WavePlayer_0, properties {soundWaveAssetPtr=/SparksMaps/Audio/Wavs/VoteTumbler_01}}
 {id=WavePlayer_1, properties {soundWaveAssetPtr=/SparksMaps/Audio/Wavs/VoteTumbler_02}}
 {id=WavePlayer_2, properties {soundWaveAssetPtr=/SparksMaps/Audio/Wavs/VoteTumbler_03}}
 {id=WavePlayer_3, properties {soundWaveAssetPtr=/SparksMaps/Audio/Wavs/VoteTumbler_04}}
 {id=WavePlayer_4, properties {soundWaveAssetPtr=/SparksMaps/Audio/Wavs/VoteTumbler_05}}
 {id=WavePlayer_5, properties {soundWaveAssetPtr=/SparksMaps/Audio/Wavs/VoteTumbler_06}}
 {id=WavePlayer_6, properties {soundWaveAssetPtr=/SparksMaps/Audio/Wavs/VoteTumbler_07}}
 {id=WavePlayer_7, properties {soundWaveAssetPtr=/SparksMaps/Audio/Wavs/VoteTumbler_08}}
 {id=WavePlayer_8, properties {soundWaveAssetPtr=/SparksMaps/Audio/Wavs/VoteTumbler_09}}
]
properties {maxDistance=2097152, soundClassObject=&Fort_SFX_InGame, volumeMultiplier=0.3}
