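# Green room vote option widget for Pilgrim quickplay, showing a setlist title and avatars of up to four players who voted for it.
# RefreshVoters filters the vote map by setlist, then tints each avatar's dynamic material with the player's pin color and async-loads their social avatar.
# Cycle and winner animations flash a selection pointer and carry event tracks.
# Sequence events play tumbler and final sounds at the parent actor's location.
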WidgetBlueprint /PilgrimQuickplay/Widgets/WBP_GreenRoomVote
animations [
 A_CyclePointer {
  binding_count=1
  length_seconds=0.25
  object=&self:A_CyclePointer
  sequence {
   bindings [SelectionPointer {
     id=#1
     possessed_class=Image
     tracks [
      {
       type=MovieSceneFloatTrack
       display_name=Render Opacity
       sections [{type=MovieSceneFloatSection, end_frame=5, start_frame=0, channels [None {type=MovieSceneScriptingFloatChannel, keys [[0,1],[3,1],[5,0]]}]}]
      }
      {
       type=MovieSceneByteTrack
       display_name=Visibility
       sections [{type=MovieSceneByteSection, end_frame=5, start_frame=0, channels [None {type=MovieSceneScriptingByteChannel, keys [[0,3],[5,4]]}]}]
      }
     ]
    }]
   tracks [{
     type=MovieSceneEventTrack
     sections [{
       type=MovieSceneEventTriggerSection
       channels [None {type=MovieSceneScriptingEventChannel, const ["<Struct 'MovieSceneEvent' (0x468e4c182a0) {}>", 1]}]
      }]
    }]
  }
 }
 A_Winner {
  binding_count=1
  length_seconds=2
  object=&self:A_Winner
  sequence {
   bindings [SelectionPointer {
     id=#1
     possessed_class=Image
     tracks [
      {
       type=MovieSceneFloatTrack
       display_name=Render Opacity
       sections [{
         type=MovieSceneFloatSection
         end_frame=40
         start_frame=0
         channels [None {type=MovieSceneScriptingFloatChannel, thinned_from=5, keys [[0,0],[2,1],[35,1],[40,0]]}]
        }]
      }
      {
       type=MovieSceneByteTrack
       display_name=Visibility
       sections [{type=MovieSceneByteSection, end_frame=40, start_frame=0, channels [None {type=MovieSceneScriptingByteChannel, keys [[0,3],[10,4],[40,4]]}]}]
      }
     ]
    }]
   tracks [{
     type=MovieSceneEventTrack
     sections [{
       type=MovieSceneEventTriggerSection
       channels [None {type=MovieSceneScriptingEventChannel, const ["<Struct 'MovieSceneEvent' (0x468e4c19ba0) {}>", 1]}]
      }]
    }]
  }
 }
]
graphs [
 EventGraph {nodes [
   n1 = SetupVoterIcon!() -> (OutputDelegate, then, Image:Image Object Reference, PlayerState:Player State Object Reference) {is=title K2Node_CustomEvent}
   > n2 = Cast To FortPlayerStateAthena!(execute, Object=@n1.PlayerState) -> (then, CastFailed, AsFort Player State Athena:Fort Player State Athena Object Reference) {is=title K2Node_DynamicCast}
   > n3 = GetDynamicMaterial(self=@n1.Image)
   > SetVectorParameterValue(
    self=@n3
    ParameterName=BG_Color
    Value=GetPinColor!(self:Fort Player State Zone Object Reference=@n2) -> (ReturnValue:Linear Color Structure)
   )
   > n4 = Async Load Asset(Asset=GetSocialAvatarBrush!(self:Fort Player State Zone Object Reference=@n2, bSmallImage:bool=true) -> (ReturnValue:Texture 2D Soft Object Reference))
   n5 = Cast To Texture2D(execute=@n4.Completed, Object=@n4)
   > SetTextureParameterValue(self=@n3, ParameterName=PlayerIcon, Value=@n5)
   > SetVisibility(self=@n1.Image, InVisibility=HitTestInvisible)
   PrintString!(execute=@n2.CastFailed, InString="CAST FAILED!", bPrintToScreen, bPrintToLog, TextColor=(R=1,G=0.078,B=0,A=1), Duration=8, Key) -> (then)
   Event Construct()
   > Set VoterIcons(VoterIcons:Array of Image Object References=Make Array!(
     "[0]":Image Object Reference=Get P1() -> (P1:Image Object Reference)
     "[1]":Image Object Reference=Get P2() -> (P2:Image Object Reference)
     "[2]":Image Object Reference=Get P3() -> (P3:Image Object Reference)
     "[3]":Image Object Reference=Get P4() -> (P4:Image Object Reference)
    ) -> (Array:Array of Image Object References)) -> (Output_Get:Array of Image Object References)
  ]}
 GetTitle {nodes [
   GetTitle() {is=title K2Node_FunctionEntry}
   > Return Node!(execute, Text:Text=Get Text(self:Text Object Reference=Get VoteActionLabel() -> (VoteActionLabel:Common Text Object Reference)))
  ]}
 RefreshVoters {nodes [
   n1 = RefreshVoters!() -> (
    then
    PlayerVotes:Map of Player State Object References to Pilgrim Premade Setlist Object References
    Setlist:Pilgrim Premade Setlist Object Reference
   ) {is=title K2Node_FunctionEntry}
   > SetVisibility(self=[Get P1(), Get P3(), Get P4(), Get P2()], InVisibility=Collapsed)
   > n2 = Keys!(execute, TargetMap:Map of Player State Object References to Pilgrim Premade Setlist Object References=@n1) -> (then, Keys:Array of Player State Object References)
   > n3 = Reverse for Each Loop!(Exec:Exec, Array:Array of Player State Object References=@n2) -> (LoopBody:Exec, ArrayIndex:int, ArrayElement:Player State Object Reference, Completed:Exec) {is=title K2Node_MacroInstance}
   n4 = For Each Loop(Exec=@n3.Completed, Array:Array of Image Object References=Get VoterIcons() -> (VoterIcons:Array of Image Object References)) -> (Array Element:Image Object Reference)
   n5 = Get VotingPlayers() -> (VotingPlayers:Array of Player State Object References)
   > Branch(Condition=Is Valid Index!(TargetArray:Array of Player State Object References=@n5, IndexToTest:int=@n4.Array Index) -> (ReturnValue:bool))
   > SetupVoterIcon!(
    execute
    self
    Image:Image Object Reference=@n4
    PlayerState:Player State Object Reference="Get (a copy)"(Array:Array of Player State Object References=@n5, Dimension 1=@n4.Array Index) -> (Output:Player State Object Reference)
   ) -> (then) {is=title K2Node_CallFunction}
   Branch(
    execute=@n3.LoopBody
    Condition="Equal (Object)"(
     A=Find(
      TargetMap:Map of Player State Object References to Pilgrim Premade Setlist Object References=@n1
      Key:"Player State Object Reference (by ref)"=@n3.ArrayElement
     ) -> (Value:Pilgrim Premade Setlist Object Reference)
     B=@n1.Setlist
    )
   )
   > Add(TargetArray:Array of Player State Object References=Get VotingPlayers(), NewItem:"Player State Object Reference (by ref)"=@n3.ArrayElement) {is=title K2Node_CallArrayFunction}
  ]}
 Sequencer Events {nodes [
   PointerAudio_SequenceEvent() {is=title K2Node_CustomEvent}
   n1 = Get WidgetParent() -> (WidgetParent:Actor Object Reference)
   > Is Valid(InputObject=@n1)
   > PlaySoundAtLocation!(
    execute
    Sound:Sound Base Object Reference=Get BasePointerSound() -> (BasePointerSound:Sound Base Object Reference)
    Location:Vector=Get Actor Location(self=@n1)
    Rotation:Rotator="0, 0, 0"
    VolumeMultiplier:float=1
    PitchMultiplier:float=1
    StartTime:float=0
    AttenuationSettings:Sound Attenuation Object Reference
    ConcurrencySettings:Sound Concurrency Object Reference
    OwningActor:Actor Object Reference
    InitialParams:Initial Active Sound Params Object Reference
   ) -> (then)
   PointerAudioFinal_SequenceEvent() {is=title K2Node_CustomEvent}
   n2 = Get WidgetParent()
   > Is Valid(InputObject=@n2)
   > PlaySoundAtLocation(
    Sound=Get FinalPointerSound() -> (FinalPointerSound:Sound Base Object Reference)
    Location=Get Actor Location(self=@n2)
    Rotation="0, 0, 0"
    VolumeMultiplier=1
    PitchMultiplier=1
    StartTime=0
   )
  ]}
 SetTitle {nodes [
   n1 = SetTitle!() -> (then, TitleText:Text) {is=title K2Node_FunctionEntry}
   > Set Text(Text=@n1, self=Get VoteActionLabel())
  ]}
 SetWidgetParent {nodes [
   n1 = SetWidgetParent!() -> (then, WidgetActor:Actor Object Reference) {is=title K2Node_FunctionEntry}
   > Set WidgetParent(WidgetParent:Actor Object Reference=@n1) -> (Output_Get:Actor Object Reference)
  ]}
]
variables [
 BasePointerSound {type=SoundBase, default=/SparksMaps/Audio/FM_VoteTumblir}
 FinalPointerSound {type=SoundBase, default=/SparksMaps/Audio/FM_VoteDone}
 VoterIcons {type="array<Image>", default []}
 WidgetParent {type=Actor, default=null}
]
widgets {
 info {inheritedWidgetCount=0, parentClass=&UserWidget, rootWidgetClass=&Overlay, widgetCount=11}
 widgets [
  Overlay_16 {widgetClassPath=&Overlay, widget=Overlay_16}
  Background {
   widgetClassPath=&Image
   bIsVariable=true
   parent=Overlay_16
   widget=Background
   properties {brush {
     bIsDynamicallyLoaded=false
     drawAs=Image
     imageType=NoImage
     mirroring=NoMirror
     resourceName=None
     resourceObject=/Game/Environments/World/Utility/128gray
     tiling=NoTile
     imageSize=<300 64>
     margin=<ltrb 0 0 0 0>
     outlineSettings {
      bUseBrushTransparency=false
      roundingType=HalfHeightRadius
      width=0
      color {colorUseRule=UseColor_Specified, specifiedColor=<rgba 0 0 0 0>}
      cornerRadii=<0 0 0 0>
     }
     tintColor {colorUseRule=UseColor_Specified, specifiedColor=<rgba 0.05 0.05 0.05 1>}
     uVRegion {bIsValid=false, max=<0 0>, min=<0 0>}
    }}
   slot {class=OverlaySlot, properties {horizontalAlignment=HAlign_Fill, verticalAlignment=VAlign_Fill}}
  }
  HorizontalBox_57 {
   widgetClassPath=&HorizontalBox
   parent=Overlay_16
   widget=HorizontalBox_57
   slot {class=OverlaySlot, properties {horizontalAlignment=HAlign_Center, verticalAlignment=VAlign_Center, padding=<ltrb 8 0 8 0>}}
  }
  VoteActionLabel {
   widgetClassPath=&CommonTextBlock
   bIsVariable=true
   parent=HorizontalBox_57
   widget=VoteActionLabel
   properties {
    scrollStyle=&TextScrollStyle-Base_C
    style=/FortUILibrary/Styles/Typography/Foundation/TextStyle-BurbankSmall-M-Blue.TextStyle-BurbankSmall-M-Blue_C
    text=SetlistName
    colorAndOpacity {colorUseRule=UseColor_Specified, specifiedColor=<rgba 0 0.624 1 1>}
    font {
     bForceMonospaced=false
     bMaterialIsStencil=false
     fontMaterial=null
     fontObject=&BurbankSmall
     letterSpacing=0
     monospacedWidth=1
     size=24
     skewAmount=0
     typefaceFontName=Black
     outlineSettings {
      bApplyOutlineToDropShadows=false
      bMiteredCorners=false
      bSeparateFillAlpha=false
      outlineMaterial=null
      outlineSize=0
      outlineColor=<rgba 0 0 0 1>
     }
    }
    shadowOffset=<0 0>
   }
   slot {class=HorizontalBoxSlot, properties {horizontalAlignment=HAlign_Center, verticalAlignment=VAlign_Center, size {sizeRule=Fill, value=1}}}
  }
  HorizontalBox_101 {
   widgetClassPath=&HorizontalBox
   parent=HorizontalBox_57
   widget=HorizontalBox_101
   slot {class=HorizontalBoxSlot, properties {padding=<ltrb 8 0 0 0>, size {sizeRule=Automatic, value=0.7}}}
  }
  P2 {
   widgetClassPath=&Image
   bIsVariable=true
   parent=HorizontalBox_101
   widget=P2
   properties {brush {
     bIsDynamicallyLoaded=false
     drawAs=Image
     imageType=FullColor
     mirroring=NoMirror
     resourceName=None
     resourceObject=/SparksUICommon/Materials/MI_UI_BandMember_Avatar
     tiling=NoTile
     imageSize=<32 32>
     margin=<ltrb 0 0 0 0>
     outlineSettings {
      bUseBrushTransparency=false
      roundingType=HalfHeightRadius
      width=0
      color {colorUseRule=UseColor_Specified, specifiedColor=<rgba 0 0 0 0>}
      cornerRadii=<0 0 0 0>
     }
     tintColor {colorUseRule=UseColor_Specified, specifiedColor=<rgba 1 1 1 1>}
     uVRegion {bIsValid=false, max=<0 0>, min=<0 0>}
    }}
   slot {class=HorizontalBoxSlot, properties {horizontalAlignment=HAlign_Right, verticalAlignment=VAlign_Center}}
  }
  P1 {
   widgetClassPath=&Image
   bIsVariable=true
   parent=HorizontalBox_101
   widget=P1
   properties {brush {
     bIsDynamicallyLoaded=false
     drawAs=Image
     imageType=FullColor
     mirroring=NoMirror
     resourceName=None
     resourceObject=&MI_UI_BandMember_Avatar
     tiling=NoTile
     imageSize=<32 32>
     margin=<ltrb 0 0 0 0>
     outlineSettings {
      bUseBrushTransparency=false
      roundingType=HalfHeightRadius
      width=0
      color {colorUseRule=UseColor_Specified, specifiedColor=<rgba 0 0 0 0>}
      cornerRadii=<0 0 0 0>
     }
     tintColor {colorUseRule=UseColor_Specified, specifiedColor=<rgba 1 1 1 1>}
     uVRegion {bIsValid=false, max=<0 0>, min=<0 0>}
    }}
   slot {class=HorizontalBoxSlot, properties {horizontalAlignment=HAlign_Right, verticalAlignment=VAlign_Center}}
  }
  P3 {
   widgetClassPath=&Image
   bIsVariable=true
   parent=HorizontalBox_101
   widget=P3
   properties {brush {
     bIsDynamicallyLoaded=false
     drawAs=Image
     imageType=FullColor
     mirroring=NoMirror
     resourceName=None
     resourceObject=&MI_UI_BandMember_Avatar
     tiling=NoTile
     imageSize=<32 32>
     margin=<ltrb 0 0 0 0>
     outlineSettings {
      bUseBrushTransparency=false
      roundingType=HalfHeightRadius
      width=0
      color {colorUseRule=UseColor_Specified, specifiedColor=<rgba 0 0 0 0>}
      cornerRadii=<0 0 0 0>
     }
     tintColor {colorUseRule=UseColor_Specified, specifiedColor=<rgba 1 1 1 1>}
     uVRegion {bIsValid=false, max=<0 0>, min=<0 0>}
    }}
   slot {class=HorizontalBoxSlot, properties {horizontalAlignment=HAlign_Right, verticalAlignment=VAlign_Center}}
  }
  P4 {
   widgetClassPath=&Image
   bIsVariable=true
   parent=HorizontalBox_101
   widget=P4
   properties {brush {
     bIsDynamicallyLoaded=false
     drawAs=Image
     imageType=FullColor
     mirroring=NoMirror
     resourceName=None
     resourceObject=&MI_UI_BandMember_Avatar
     tiling=NoTile
     imageSize=<32 32>
     margin=<ltrb 0 0 0 0>
     outlineSettings {
      bUseBrushTransparency=false
      roundingType=HalfHeightRadius
      width=0
      color {colorUseRule=UseColor_Specified, specifiedColor=<rgba 0 0 0 0>}
      cornerRadii=<0 0 0 0>
     }
     tintColor {colorUseRule=UseColor_Specified, specifiedColor=<rgba 1 1 1 1>}
     uVRegion {bIsValid=false, max=<0 0>, min=<0 0>}
    }}
   slot {class=HorizontalBoxSlot, properties {horizontalAlignment=HAlign_Right, verticalAlignment=VAlign_Center}}
  }
  Overlay_47 {
   widgetClassPath=&Overlay
   parent=Overlay_16
   widget=Overlay_47
   slot {class=OverlaySlot, properties {horizontalAlignment=HAlign_Center}}
  }
  SelectionPointer {
   widgetClassPath=&Image
   bIsVariable=true
   parent=Overlay_47
   widget=SelectionPointer
   properties {
    visibility=Hidden
    brush {
     bIsDynamicallyLoaded=false
     drawAs=Image
     imageType=NoImage
     mirroring=NoMirror
     resourceName=None
     resourceObject=/Game/UI/CodeReferenced/Radial_Pointer
     tiling=NoTile
     imageSize=<64 64>
     margin=<ltrb 0 0 0 0>
     outlineSettings {
      bUseBrushTransparency=false
      roundingType=HalfHeightRadius
      width=0
      color {colorUseRule=UseColor_Specified, specifiedColor=<rgba 0 0 0 0>}
      cornerRadii=<0 0 0 0>
     }
     tintColor {colorUseRule=UseColor_Specified, specifiedColor=<rgba 1 1 1 1>}
     uVRegion {bIsValid=false, max=<0 0>, min=<0 0>}
    }
    renderTransform {angle=90, scale=<1 1>, shear=<0 0>, translation=<0 -45>}
   }
   slot {class=OverlaySlot}
  }
 ]
}
