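# Green room widget that shows up to four setlist vote options and animates the reveal of the winning setlist.
# RefreshDisplay fills and shows one vote item per setlist, passing player votes to each.
# Pick Winner repeats the visible items into a longer list, trims it to end on the title-matched winner, then caps its length.
# Chained animation-finished callbacks cycle a pointer along that list, play the winner animation and collapse the losers.
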WidgetBlueprint /PilgrimQuickplay/Widgets/WBP_GreenRoomVotesDisplay
graphs [
 EventGraph {nodes [
   n1 = RefreshDisplay!() -> (
    OutputDelegate
    then
    Setlists:Array of Pilgrim Premade Setlist Object References
    PlayerVotes:Map of Player State Object References to Pilgrim Premade Setlist Object References
   ) {is=title K2Node_CustomEvent}
   > ClearVotesDisplay()
   > n2 = For Each Loop(Array:Array of WBP Green Room Vote Object References=Get GreenroomSetlistVoteItems() -> (GreenroomSetlistVoteItems:Array of WBP Green Room Vote Object References)) -> (Array Element:WBP Green Room Vote Object Reference)
   > Branch(Condition=Is Valid Index!(TargetArray:Array of Pilgrim Premade Setlist Object References=@n1.Setlists, IndexToTest:int=@n2.Array Index) -> (ReturnValue:bool))
   n3 = "Get (a copy)"(Array:Array of Pilgrim Premade Setlist Object References=@n1.Setlists, Dimension 1=@n2.Array Index) -> (Output:Pilgrim Premade Setlist Object Reference)
   > RefreshVoters!(
    execute
    self:WBP Green Room Vote Object Reference=@n2
    PlayerVotes:Map of Player State Object References to Pilgrim Premade Setlist Object References=@n1.PlayerVotes
    Setlist:Pilgrim Premade Setlist Object Reference=@n3
   ) -> (then)
   > SetTitle!(
    execute
    self:WBP Green Room Vote Object Reference=@n2
    TitleText:Text=Get SetlistTitle!(self:Pilgrim Premade Setlist Object Reference=@n3) -> (SetlistTitle:Text)
   ) -> (then)
   > SetVisibility(self=@n2, InVisibility=HitTestInvisible)
   n4 = StartCycleThroughPickers!() -> (OutputDelegate, then, ActiveVoteObjects:Array of WBP Green Room Vote Object References) {is=title K2Node_CustomEvent}
   > Set PickerIndex(PickerIndex:int=0) -> (Output_Get:int=0)
   > AnimatePickerForPickerIndexAndAdvance!(execute, self, ActiveVoteObjects:Array of WBP Green Room Vote Object References=@n4.ActiveVoteObjects) -> (then)
   ClearVotesDisplay() {is=title K2Node_CustomEvent}
   > SetVisibility(
    self=[
     Get WBP_GreenRoomVote_1() -> (WBP_GreenRoomVote_1:WBP Green Room Vote Object Reference)
     Get WBP_GreenRoomVote_2() -> (WBP_GreenRoomVote_2:WBP Green Room Vote Object Reference)
     Get WBP_GreenRoomVote_3() -> (WBP_GreenRoomVote_3:WBP Green Room Vote Object Reference)
     Get WBP_GreenRoomVote_4() -> (WBP_GreenRoomVote_4:WBP Green Room Vote Object Reference)
    ]
    InVisibility=Collapsed
   )
   n5 = AnimatePickerForPickerIndexAndAdvance!() -> (OutputDelegate, then, ActiveVoteObjects:Array of WBP Green Room Vote Object References) {is=title K2Node_CustomEvent}
   n6 = "Get (a copy)"(Array:Array of WBP Green Room Vote Object References=@n5.ActiveVoteObjects, Dimension 1=Get PickerIndex() -> (PickerIndex:int=0)) -> (Output:WBP Green Room Vote Object Reference)
   > n7 = "Play Animation with Finished event (legacy)"!(
    execute:Exec
    Widget:User Widget Object Reference=@n6
    InAnimation:Widget Animation Object Reference=Get A_CyclePointer!(self:WBP Green Room Vote Object Reference=@n6) -> (A_CyclePointer:Widget Animation Object Reference)
    StartAtTime:float=0
    NumLoopsToPlay:int=1
    PlayMode:EUMGSequencePlayMode Enum
    PlaybackSpeed:float=1
   ) -> (then:Exec, Result:UMGSequence Player Object Reference, Finished:Exec) {is=title K2Node_PlayAnimation}
   n8 = Increment Int(" "=@n7.Finished, Value=Get PickerIndex() -> (PickerIndex=0))
   > n9 = Branch(Condition="integer < integer"(A=@n8, B=Last Index!(TargetArray:Array of WBP Green Room Vote Object References=@n5.ActiveVoteObjects) -> (ReturnValue=0)))
   > AnimatePickerForPickerIndexAndAdvance(ActiveVoteObjects=@n5.ActiveVoteObjects) {is=title K2Node_CallFunction}
   AnimatePickerForWinner!(execute=@n9.else, self, ActiveVoteObjects:Array of WBP Green Room Vote Object References=@n5.ActiveVoteObjects) -> (then)
   n10 = AnimatePickerForWinner!() -> (OutputDelegate, then, ActiveVoteObjects:Array of WBP Green Room Vote Object References) {is=title K2Node_CustomEvent}
   n11 = "Get (a copy)"(Array=@n10.ActiveVoteObjects, Dimension 1=Last Index(TargetArray=@n10.ActiveVoteObjects) -> (ReturnValue=0))
   > n12 = "Play Animation with Finished event (legacy)"(
    Widget=@n11
    InAnimation=Get A_Winner!(self:WBP Green Room Vote Object Reference=@n11) -> (A_Winner:Widget Animation Object Reference)
    StartAtTime=0
    NumLoopsToPlay=1
    PlaybackSpeed=1
   )
   n13 = For Each Loop(Exec=@n12.Finished, Array=@n10.ActiveVoteObjects)
   > Branch(Condition="Not Equal (Object)"(A:Object Reference=@n13, B:Object Reference=@n11) -> (ReturnValue:bool=~) {is=title K2Node_PromotableOperator})
   > SetVisibility(self=@n13, InVisibility=Collapsed)
   Event Construct()
   > Set GreenroomSetlistVoteItems(GreenroomSetlistVoteItems:Array of WBP Green Room Vote Object References=Make Array!(
     "[0]":WBP Green Room Vote Object Reference=Get WBP_GreenRoomVote_1()
     "[1]":WBP Green Room Vote Object Reference=Get WBP_GreenRoomVote_2()
     "[2]":WBP Green Room Vote Object Reference=Get WBP_GreenRoomVote_3()
     "[3]":WBP Green Room Vote Object Reference=Get WBP_GreenRoomVote_4()
    ) -> (Array:Array of WBP Green Room Vote Object References)) -> (Output_Get:Array of WBP Green Room Vote Object References)
  ]}
 Pick Winner {nodes [
   n1 = Pick Winner!() -> (then, WinnerRef:Pilgrim Premade Setlist Object Reference) {is=title K2Node_FunctionEntry}
   > Can Execute Cosmetic Events!(execute:Exec) -> (True:Exec, False:Exec) {is=title K2Node_MacroInstance}
   > n2 = For Each Loop(Array=Get GreenroomSetlistVoteItems())
   Append Array!(
    execute=@n2.Completed
    TargetArray:Array of WBP Green Room Vote Object References=Get ActiveGreenroomVoteItems() -> (ActiveGreenroomVoteItems:Array of WBP Green Room Vote Object References)
    SourceArray:Array of WBP Green Room Vote Object References=Get ActiveGreenroomVoteItems()
   ) -> (then)
   > Append Array(TargetArray=Get ActiveGreenroomVoteItems(), SourceArray=Get ActiveGreenroomVoteItems())
   > n3 = Reverse for Each Loop!(Exec:Exec, Array:Array of WBP Green Room Vote Object References=Get ActiveGreenroomVoteItems()) -> (LoopBody:Exec, ArrayIndex:int, ArrayElement:WBP Green Room Vote Object Reference, Completed:Exec) {is=title K2Node_MacroInstance}
   > n4 = Branch(Condition=Get EndWinnerFound() -> (EndWinnerFound:bool))
   n5 = Branch(
    execute=@n4.else
    Condition="Not Equal Exactly (Text)"!(
     A:"Text (by ref)"=GetTitle!(self:WBP Green Room Vote Object Reference=@n3.ArrayElement) -> (Text:Text)
     B:"Text (by ref)"=Get SetlistTitle(self=@n1)
    ) -> (ReturnValue:bool)
   )
   > Remove Index!(execute, TargetArray:Array of WBP Green Room Vote Object References=Get ActiveGreenroomVoteItems(), IndexToRemove:int=@n3) -> (then)
   Set EndWinnerFound(execute=@n5.else, EndWinnerFound:bool=true) -> (Output_Get:bool)
   n6 = Length(TargetArray:Array of WBP Green Room Vote Object References=Get ActiveGreenroomVoteItems())
   n7 = Branch(execute=@n3.Completed, Condition="integer > integer"(A=@n6, B=8))
   > n8 = For Loop(LastIndex=int - int(A=@n6, B=8))
   > Remove Index(TargetArray=Get ActiveGreenroomVoteItems(), IndexToRemove=0)
   StartCycleThroughPickers!(execute=[@n8.Completed, @n7.else], self, ActiveVoteObjects:Array of WBP Green Room Vote Object References=Get ActiveGreenroomVoteItems()) -> (then) {is=title K2Node_CallFunction}
   Branch(execute=@n2.LoopBody, Condition=IsVisible(self=@n2))
   > Add(
    TargetArray:Array of WBP Green Room Vote Object References=Get ActiveGreenroomVoteItems()
    NewItem:"WBP Green Room Vote Object Reference (by ref)"=@n2
   ) {is=title K2Node_CallArrayFunction}
  ]}
 SetWidgetParent {nodes [
   n1 = SetWidgetParent!() -> (then, WidgetParent:Actor Object Reference) {is=title K2Node_FunctionEntry}
   > n2 = Set WidgetParent(WidgetParent:Actor Object Reference=@n1) -> (Output_Get:Actor Object Reference)
   > SetWidgetParent!(
    execute
    self:WBP Green Room Vote Object Reference=[Get WBP_GreenRoomVote_1(), Get WBP_GreenRoomVote_2(), Get WBP_GreenRoomVote_3(), Get WBP_GreenRoomVote_4()]
    WidgetActor:Actor Object Reference
   ) -> (then) {is=title K2Node_CallFunction}
   > Return Node!(execute, WidgetParent_Passthrough:Actor Object Reference=@n2)
  ]}
]
variables [
 GreenroomSetlistVoteItems {type="array<WBP_GreenRoomVote_C>", default []}
 PickerIndex {type=integer}
 WidgetParent {type=Actor, default=null}
]
widgets {
 info {inheritedWidgetCount=0, parentClass=&UserWidget, rootWidgetClass=&Overlay, widgetCount=8}
 widgets [
  Overlay_16 {widgetClassPath=&Overlay, widget=Overlay_16}
  Background {
   widgetClassPath=&Image
   bIsVariable=true
   parent=Overlay_16
   widget=Background
   properties {brush {
     bIsDynamicallyLoaded=false
     drawAs=Image
     imageType=NoImage
     mirroring=NoMirror
     resourceName=None
     resourceObject=/Game/Environments/World/Utility/128gray
     tiling=NoTile
     imageSize=<300 64>
     margin=<ltrb 0 0 0 0>
     outlineSettings {
      bUseBrushTransparency=false
      roundingType=HalfHeightRadius
      width=0
      color {colorUseRule=UseColor_Specified, specifiedColor=<rgba 0 0 0 0>}
      cornerRadii=<0 0 0 0>
     }
     tintColor {colorUseRule=UseColor_Specified, specifiedColor=<rgba 0.05 0.05 0.05 1>}
     uVRegion {bIsValid=false, max=<0 0>, min=<0 0>}
    }}
   slot {class=OverlaySlot, properties {horizontalAlignment=HAlign_Fill, verticalAlignment=VAlign_Fill}}
  }
  VoteActionLabel {
   widgetClassPath=&CommonTextBlock
   bIsVariable=true
   parent=Overlay_16
   widget=VoteActionLabel
   properties {
    scrollStyle=&TextScrollStyle-Base_C
    style=/FortUILibrary/Styles/Typography/Foundation/TextStyle-BurbankSmall-M-Blue.TextStyle-BurbankSmall-M-Blue_C
    text=Vote For Setlist...
    visibility=Hidden
    colorAndOpacity {colorUseRule=UseColor_Specified, specifiedColor=<rgba 0 0.624 1 1>}
    font {
     bForceMonospaced=false
     bMaterialIsStencil=false
     fontMaterial=null
     fontObject=&BurbankSmall
     letterSpacing=0
     monospacedWidth=1
     size=24
     skewAmount=0
     typefaceFontName=Black
     outlineSettings {
      bApplyOutlineToDropShadows=false
      bMiteredCorners=false
      bSeparateFillAlpha=false
      outlineMaterial=null
      outlineSize=0
      outlineColor=<rgba 0 0 0 1>
     }
    }
    shadowOffset=<0 0>
   }
   slot {class=OverlaySlot, properties {horizontalAlignment=HAlign_Center, verticalAlignment=VAlign_Center}}
  }
  HorizontalBox_102 {
   widgetClassPath=&HorizontalBox
   parent=Overlay_16
   widget=HorizontalBox_102
   slot {class=OverlaySlot, properties {horizontalAlignment=HAlign_Fill, verticalAlignment=VAlign_Fill}}
  }
  WBP_GreenRoomVote_1 {
   widgetClassPath=&/WBP_GreenRoomVote
   bIsVariable=true
   parent=HorizontalBox_102
   widget=WBP_GreenRoomVote_1
   properties {visibility=Collapsed}
   slot {class=HorizontalBoxSlot, properties {size {sizeRule=Fill, value=1}}}
  }
  WBP_GreenRoomVote_2 {
   widgetClassPath=&WBP_GreenRoomVote
   bIsVariable=true
   parent=HorizontalBox_102
   widget=WBP_GreenRoomVote_2
   properties {visibility=Collapsed}
   slot {class=HorizontalBoxSlot, properties {size {sizeRule=Fill, value=1}}}
  }
  WBP_GreenRoomVote_3 {
   widgetClassPath=&WBP_GreenRoomVote
   bIsVariable=true
   parent=HorizontalBox_102
   widget=WBP_GreenRoomVote_3
   properties {visibility=Collapsed}
   slot {class=HorizontalBoxSlot, properties {size {sizeRule=Fill, value=1}}}
  }
  WBP_GreenRoomVote_4 {
   widgetClassPath=&WBP_GreenRoomVote
   bIsVariable=true
   parent=HorizontalBox_102
   widget=WBP_GreenRoomVote_4
   properties {visibility=Collapsed}
   slot {class=HorizontalBoxSlot, properties {size {sizeRule=Fill, value=1}}}
  }
 ]
}
