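# Countdown timer widget for the Pilgrim Quickplay greenroom stage, showing a number over a timer image in an overlay that starts hidden.
# PulseNumber sets the text to the number, or "GO!" at zero or below, and plays a scale pulse.
# At or below AudioPingThreshold, it plays a base or final ping at the WidgetParent actor's location; the final sound defaults to unset.
# ShowTimer and HideTimer toggle the overlay, and Set Widget Parent stores that actor.
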WidgetBlueprint /PilgrimQuickplay/Widgets/WBP_PilgrimQuickplay_Greenroom_StageTimer
animations [A_Pulse {
  binding_count=1
  length_seconds=0.75
  object=&self:A_Pulse
  sequence {bindings [NumerText {
     id=#1
     possessed_class=CommonTextBlock
     tracks [{
       type=MovieScene2DTransformTrack
       sections [{
         type=MovieScene2DTransformSection
         end_frame=15
         start_frame=0
         channels [
          Translation.X {type=MovieSceneScriptingFloatChannel}
          Translation.Y {type=MovieSceneScriptingFloatChannel}
          Angle {type=MovieSceneScriptingFloatChannel}
          Scale.X {type=MovieSceneScriptingFloatChannel, keys [[0,1],[5,1.25],[15,1]]}
          Scale.Y {type=MovieSceneScriptingFloatChannel, keys [[0,1],[5,1.25],[15,1]]}
          Shear.X {type=MovieSceneScriptingFloatChannel}
          Shear.Y {type=MovieSceneScriptingFloatChannel}
         ]
        }]
      }]
    }]}
 }]
graphs [
 EventGraph {nodes [
   n1 = PulseNumber!() -> (OutputDelegate, then, Number:int) {is=title K2Node_CustomEvent}
   n2 = "integer > integer"(A=@n1.Number)
   > Set Text(
    Text=Select(Option 0:Text="LOCTEXT(\"GO!\")", Option 1:Text="To Text (Integer)"(Value=@n1.Number), Index=@n2) -> (ReturnValue:Text)
    self=Get NumerText() -> (NumerText:Common Text Object Reference)
   )
   > PlayAnimationForward(InAnimation=Get A_Pulse() -> (A_Pulse:Widget Animation Object Reference))
   > Branch(Condition="integer <= integer"(A=@n1.Number, B=Get AudioPingThreshold() -> (AudioPingThreshold:int=0)))
   n3 = Get WidgetParent() -> (WidgetParent:Actor Object Reference)
   > Is Valid(InputObject=@n3)
   > PlaySoundAtLocation!(
    execute
    Sound:Sound Base Object Reference=Select(
     Option 0:Sound Base Object Reference=Get FinalPingSound() -> (FinalPingSound:Sound Base Object Reference)
     Option 1:Sound Base Object Reference=Get BasePingSound() -> (BasePingSound:Sound Base Object Reference)
     Index=@n2
    ) -> (ReturnValue:Sound Base Object Reference)
    Location:Vector=Get Actor Location(self=@n3)
    Rotation:Rotator="0, 0, 0"
    VolumeMultiplier:float=1
    PitchMultiplier:float=1
    StartTime:float=0
    AttenuationSettings:Sound Attenuation Object Reference
    ConcurrencySettings:Sound Concurrency Object Reference
    OwningActor:Actor Object Reference
    InitialParams:Initial Active Sound Params Object Reference
   ) -> (then)
   ShowTimer() {is=title K2Node_CustomEvent}
   > SetVisibility(self=Get TimerOverlay() -> (TimerOverlay:Overlay Object Reference), InVisibility=HitTestInvisible)
   HideTimer() {is=title K2Node_CustomEvent}
   > SetVisibility(self=Get TimerOverlay(), InVisibility=Collapsed)
  ]}
 Set Widget Parent {nodes [
   n1 = Set Widget Parent!() -> (then, WidgetParent:Actor Object Reference) {is=title K2Node_FunctionEntry}
   > n2 = Set WidgetParent(WidgetParent:Actor Object Reference=@n1) -> (Output_Get:Actor Object Reference)
   > Return Node!(execute, WidgetParent_Passthrough:Actor Object Reference=@n2)
  ]}
]
variables [
 AudioPingThreshold {type=integer, default=10}
 BasePingSound {type=SoundBase, default=/SparksMaps/Audio/FM_VoteCountdown}
 FinalPingSound {type=SoundBase, default=null}
 WidgetParent {type=Actor, default=null}
]
widgets {
 info {inheritedWidgetCount=0, parentClass=&CommonUserWidget, rootWidgetClass=&Overlay, widgetCount=4}
 widgets [
  TimerOverlay {widgetClassPath=&Overlay, bIsVariable=true, widget=TimerOverlay, properties {visibility=Hidden}}
  Overlay_42 {widgetClassPath=&Overlay, parent=TimerOverlay, widget=Overlay_42, slot {class=OverlaySlot}}
  TimerImage {
   widgetClassPath=&Image
   bIsVariable=true
   parent=Overlay_42
   widget=TimerImage
   properties {brush {
     bIsDynamicallyLoaded=false
     drawAs=Image
     imageType=NoImage
     mirroring=NoMirror
     resourceName=None
     resourceObject=/PilgrimQuickplay/Widgets/Materials/MI_UI_Pilgrim_StageTimer
     tiling=NoTile
     imageSize=<128 128>
     margin=<ltrb 0 0 0 0>
     outlineSettings {
      bUseBrushTransparency=false
      roundingType=HalfHeightRadius
      width=0
      color {colorUseRule=UseColor_Specified, specifiedColor=<rgba 0 0 0 0>}
      cornerRadii=<0 0 0 0>
     }
     tintColor {colorUseRule=UseColor_Specified, specifiedColor=<rgba 1 1 1 1>}
     uVRegion {bIsValid=false, max=<0 0>, min=<0 0>}
    }}
   slot {class=OverlaySlot}
  }
  NumerText {
   widgetClassPath=&CommonTextBlock
   bIsVariable=true
   parent=Overlay_42
   widget=NumerText
   properties {
    style=/FortUILibrary/Styles/Typography/Foundation/TextStyle-Header-L.TextStyle-Header-L_C
    text=30
    font {
     bForceMonospaced=false
     bMaterialIsStencil=false
     fontMaterial=null
     fontObject=&Burbank
     letterSpacing=0
     monospacedWidth=1
     size=48
     skewAmount=0
     typefaceFontName=Burbank-Black
     outlineSettings {
      bApplyOutlineToDropShadows=false
      bMiteredCorners=false
      bSeparateFillAlpha=false
      outlineMaterial=null
      outlineSize=0
      outlineColor=<rgba 0 0 0 1>
     }
    }
    shadowOffset=<0 0>
   }
   slot {class=OverlaySlot, properties {horizontalAlignment=HAlign_Center, verticalAlignment=VAlign_Center}}
  }
 ]
}
