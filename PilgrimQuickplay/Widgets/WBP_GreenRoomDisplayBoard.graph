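# Widget Blueprint for the Pilgrim Quickplay green room display board, showing four player displays, a votes display and a stage timer.
# ResetBoard hides the timer, clears the votes display and sets all four player displays to not ready.
# SetWidgetParent stores an Actor reference and passes it on to the stage timer and votes display.
# The event graph is empty, so the board holds no event logic of its own.
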WidgetBlueprint /PilgrimQuickplay/Widgets/WBP_GreenRoomDisplayBoard
graphs [
 EventGraph {links [], nodes []}
 ResetBoard {nodes [
   ResetBoard() {is=title K2Node_FunctionEntry}
   > HideTimer(self:WBP Pilgrim Quickplay Greenroom Stage Timer Object Reference=Get StageTimer() -> (StageTimer:WBP Pilgrim Quickplay Greenroom Stage Timer Object Reference))
   > ClearVotesDisplay(self:WBP Green Room Votes Display Object Reference=Get VotesDisplay() -> (VotesDisplay:WBP Green Room Votes Display Object Reference))
   > Set Ready State UI!(
    execute
    self:WBP Green Room Player Display Object Reference=[
     Get Player1_Widget() -> (Player1_Widget:WBP Green Room Player Display Object Reference)
     Get Player2_Widget() -> (Player2_Widget:WBP Green Room Player Display Object Reference)
     Get Player3_Widget() -> (Player3_Widget:WBP Green Room Player Display Object Reference)
     Get Player4_Widget() -> (Player4_Widget:WBP Green Room Player Display Object Reference)
    ]
    IsReady:bool=false
   ) -> (then) {is=title K2Node_CallFunction}
  ]}
 SetWidgetParent {nodes [
   n1 = SetWidgetParent!() -> (then, WidgetParent:Actor Object Reference) {is=title K2Node_FunctionEntry}
   > n2 = Set WidgetParent(WidgetParent:Actor Object Reference=@n1) -> (Output_Get:Actor Object Reference)
   > n3 = Set Widget Parent!(execute, self:WBP Pilgrim Quickplay Greenroom Stage Timer Object Reference=Get StageTimer(), WidgetParent:Actor Object Reference=@n2) -> (then, WidgetParent_Passthrough:Actor Object Reference) {is=title K2Node_CallFunction}
   > SetWidgetParent!(execute, self:WBP Green Room Votes Display Object Reference=Get VotesDisplay(), WidgetParent:Actor Object Reference=@n3) -> (then, WidgetParent_Passthrough:Actor Object Reference) {is=title K2Node_CallFunction}
  ]}
]
variables [WidgetParent {type=Actor, default=null}]
widgets {
 info {inheritedWidgetCount=0, parentClass=&UserWidget, rootWidgetClass=&Overlay, widgetCount=11}
 widgets [
  Overlay_0 {widgetClassPath=&Overlay, widget=Overlay_0}
  Background {
   widgetClassPath=&Image
   bIsVariable=true
   parent=Overlay_0
   widget=Background
   properties {colorAndOpacity=<rgba 0.9873 1 0.9341 1>}
   slot {class=OverlaySlot, properties {horizontalAlignment=HAlign_Fill, verticalAlignment=VAlign_Fill}}
  }
  VerticalBox_0 {widgetClassPath=&VerticalBox, parent=Overlay_0, widget=VerticalBox_0, slot {class=OverlaySlot}}
  HorizontalBox_46 {widgetClassPath=&HorizontalBox, parent=VerticalBox_0, widget=HorizontalBox_46, slot {class=VerticalBoxSlot}}
  Player1_Widget {
   widgetClassPath=&/WBP_GreenRoomPlayerDisplay
   bIsVariable=true
   parent=HorizontalBox_46
   widget=Player1_Widget
   slot {class=HorizontalBoxSlot, properties {padding=<ltrb 5 5 5 5>}}
  }
  Player2_Widget {
   widgetClassPath=&WBP_GreenRoomPlayerDisplay
   bIsVariable=true
   parent=HorizontalBox_46
   widget=Player2_Widget
   slot {class=HorizontalBoxSlot, properties {padding=<ltrb 5 5 5 5>}}
  }
  Player3_Widget {
   widgetClassPath=&WBP_GreenRoomPlayerDisplay
   bIsVariable=true
   parent=HorizontalBox_46
   widget=Player3_Widget
   slot {class=HorizontalBoxSlot, properties {padding=<ltrb 5 5 5 5>}}
  }
  Player4_Widget {
   widgetClassPath=&WBP_GreenRoomPlayerDisplay
   bIsVariable=true
   parent=HorizontalBox_46
   widget=Player4_Widget
   slot {class=HorizontalBoxSlot, properties {padding=<ltrb 5 5 5 5>}}
  }
  SizeBox_1 {
   widgetClassPath=&SizeBox
   parent=VerticalBox_0
   widget=SizeBox_1
   properties {bOverride_MaxDesiredWidth=true, maxDesiredWidth=1200}
   slot {class=VerticalBoxSlot, properties {verticalAlignment=VAlign_Top}}
  }
  VotesDisplay {
   widgetClassPath=&/WBP_GreenRoomVotesDisplay
   bIsVariable=true
   parent=SizeBox_1
   widget=VotesDisplay
   slot {class=SizeBoxSlot}
  }
  StageTimer {
   widgetClassPath=&/WBP_PilgrimQuickplay_Greenroom_StageTimer
   bIsVariable=true
   parent=Overlay_0
   widget=StageTimer
   slot {class=OverlaySlot, properties {horizontalAlignment=HAlign_Center, verticalAlignment=VAlign_Bottom, padding=<ltrb 0 0 0 50>}}
  }
 ]
}
