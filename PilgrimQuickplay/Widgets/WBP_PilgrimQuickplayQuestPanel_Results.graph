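# Widget Blueprint for the quest panel on the Pilgrim Quickplay results screen, derived from SparksQuestScreenContainer.
# Its only widget is an inherited BP_AthenaMapScreenContainer named AthenaScreenContainer, set to Collapsed with an empty default tab list.
# The Blueprint adds no variables or graph logic, so any behavior comes from the parent class and the container widget.
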WidgetBlueprint /PilgrimQuickplay/Widgets/WBP_PilgrimQuickplayQuestPanel_Results
graphs [EventGraph {links [], nodes []}]
variables []
widgets {
 info {
  inheritedWidgetCount=1
  parentClass=/Script/SparksCoreUI.SparksQuestScreenContainer
  rootWidgetClass=/FortniteCoreUI/HUD/BP_AthenaMapScreenContainer.BP_AthenaMapScreenContainer_C
  widgetCount=1
 }
 widgets [AthenaScreenContainer {
   widgetClassPath=/FortniteCoreUI/HUD/BP_AthenaMapScreenContainer
   bInherited=true
   bIsVariable=true
   widget=AthenaScreenContainer
   properties {visibility=Collapsed, defaultTabs []}
  }]
}
