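# Score display widget for Pilgrim Quickplay, layering monospaced score text and a glow over a backing image and shadow.
# PreConstruct and Construct share one styling chain that uses exposed booleans to toggle the backing, align the text left or right, and mirror the backing.
# For Arabic on non-touch UI, FlipReadDirectionArabic mirrors the text overlay with negative X render scale.
# The graph holds no score-update logic.
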WidgetBlueprint /PilgrimQuickplay/Widgets/WBP_PilgrimQuickplay_Score
graphs [
 EventGraph {nodes [
   n1 = Event PreConstruct()
   > n2 = Parent: PreConstruct(IsDesignTime=@n1.IsDesignTime)
   SetVisibility(
    execute=[@n2.then, @n6.then]
    self=Get Backing() -> (Backing:Image Object Reference)
    InVisibility=Select(Option 0=Hidden, Option 1=SelfHitTestInvisible, Index=Get ShouldShowBorder() -> (ShouldShowBorder:bool))
   )
   > SetHorizontalAlignment(
    self=SlotAsOverlaySlot(Widget=Get TextOverlay() -> (TextOverlay:Overlay Object Reference))
    InHorizontalAlignment=Select(
     Option 0:EHorizontalAlignment Enum=HAlign_Right
     Option 1:EHorizontalAlignment Enum=HAlign_Left
     Index=Get AlignmentLeft() -> (AlignmentLeft:bool)
    ) -> (ReturnValue:EHorizontalAlignment Enum)
   )
   n3 = Get FlipBacking() -> (FlipBacking:bool)
   > SetRenderScale!(
    execute
    self:Widget Object Reference=Get Backing()
    Scale:Vector 2D Structure=MakeVector2D(X=Select(Option 0:double=1, Option 1:double=-1, Index=@n3) -> (ReturnValue:double), Y=1)
   ) -> (then)
   n4 = GetCurrentLanguage!() -> (ReturnValue:String)
   n5 = ShouldUseTouchUIForLocalPlayer!() -> (ReturnValue:bool)
   > FlipReadDirectionArabic!(execute, self, IsAr:bool=AND Boolean(A="Equal, Case Insensitive (String)"(A=@n4, B=ar), B=NOT Boolean(A=@n5))) -> (then)
   Event Construct()
   > n6 = Parent: Construct()
   n7 = Event Tick()
   > Parent: Tick!(execute, MyGeometry:Geometry Structure=@n7.MyGeometry, InDeltaTime:float=@n7.InDeltaTime) -> (then) {is=title K2Node_CallParentFunction}
  ]}
 FlipReadDirectionArabic {nodes [
   n1 = FlipReadDirectionArabic!() -> (then, IsAr:bool=~) {is=title K2Node_FunctionEntry}
   > SetRenderScale(self=Get TextOverlay(), Scale=MakeVector2D(X=Select(Option 0=1, Option 1=-1, Index=@n1), Y=1))
  ]}
]
variables [
 AlignmentLeft {type=boolean}
 FlipBacking {type=boolean}
 ShouldShowBorder {type=boolean, default=true}
]
widgets {
 info {inheritedWidgetCount=0, parentClass=/Script/PilgrimCoreUI.PilgrimGameWidgetBase, rootWidgetClass=&Overlay, widgetCount=6}
 widgets [
  Overlay_All {widgetClassPath=&Overlay, bIsVariable=true, widget=Overlay_All}
  Shadow {
   widgetClassPath=&Image
   bIsVariable=true
   parent=Overlay_All
   widget=Shadow
   properties {brush {
     bIsDynamicallyLoaded=false
     drawAs=Image
     imageType=NoImage
     mirroring=NoMirror
     resourceName=None
     resourceObject=/PilgrimCore/Widgets/Materials/M_UI_Pilgrim_RoundedShadow_Box
     tiling=NoTile
     imageSize=<32 32>
     margin=<ltrb 0 0 0 0>
     outlineSettings {
      bUseBrushTransparency=false
      roundingType=HalfHeightRadius
      width=0
      color {colorUseRule=UseColor_Specified, specifiedColor=<rgba 0 0 0 0>}
      cornerRadii=<0 0 0 0>
     }
     tintColor {colorUseRule=UseColor_Specified, specifiedColor=<rgba 1 1 1 1>}
     uVRegion {bIsValid=false, max=<0 0>, min=<0 0>}
    }}
   slot {class=OverlaySlot, properties {horizontalAlignment=HAlign_Fill, verticalAlignment=VAlign_Fill, padding=<ltrb -16 -16 -16 -16>}}
  }
  Backing {
   widgetClassPath=&Image
   bIsVariable=true
   parent=Overlay_All
   widget=Backing
   properties {brush {
     bIsDynamicallyLoaded=false
     drawAs=Image
     imageType=NoImage
     mirroring=NoMirror
     resourceName=None
     resourceObject=/PilgrimCore/Widgets/Materials/MI_UI_Pilgrim_Score_Backing
     tiling=NoTile
     imageSize=<32 32>
     margin=<ltrb 0 0 0 0>
     outlineSettings {
      bUseBrushTransparency=false
      roundingType=HalfHeightRadius
      width=0
      color {colorUseRule=UseColor_Specified, specifiedColor=<rgba 0 0 0 0>}
      cornerRadii=<0 0 0 0>
     }
     tintColor {colorUseRule=UseColor_Specified, specifiedColor=<rgba 1 1 1 1>}
     uVRegion {bIsValid=false, max=<0 0>, min=<0 0>}
    }}
   slot {class=OverlaySlot, properties {horizontalAlignment=HAlign_Fill, verticalAlignment=VAlign_Fill}}
  }
  TextOverlay {
   widgetClassPath=&Overlay
   bIsVariable=true
   parent=Overlay_All
   widget=TextOverlay
   slot {class=OverlaySlot, properties {horizontalAlignment=HAlign_Right, verticalAlignment=VAlign_Center, padding=<ltrb 16 0 16 1>}}
  }
  ScoreGlow {
   widgetClassPath=&Image
   bIsVariable=true
   parent=TextOverlay
   widget=ScoreGlow
   properties {
    clipping=ClipToBoundsWithoutIntersecting
    brush {
     bIsDynamicallyLoaded=false
     drawAs=Image
     imageType=NoImage
     mirroring=NoMirror
     resourceName=None
     resourceObject=/PilgrimCore/Widgets/Materials/MI_UI_Pilgrim_ScoreGlow
     tiling=NoTile
     imageSize=<32 32>
     margin=<ltrb 0 0 0 0>
     outlineSettings {
      bUseBrushTransparency=false
      roundingType=HalfHeightRadius
      width=0
      color {colorUseRule=UseColor_Specified, specifiedColor=<rgba 0 0 0 0>}
      cornerRadii=<0 0 0 0>
     }
     tintColor {colorUseRule=UseColor_Specified, specifiedColor=<rgba 1 1 1 1>}
     uVRegion {bIsValid=false, max=<0 0>, min=<0 0>}
    }
   }
   slot {class=OverlaySlot, properties {horizontalAlignment=HAlign_Fill, verticalAlignment=VAlign_Fill, padding=<ltrb -10 -2 -10 -2>}}
  }
  ScoreText {
   widgetClassPath=&CommonTextBlock
   bIsVariable=true
   parent=TextOverlay
   widget=ScoreText
   properties {
    justification=Right
    text=0
    font {
     bForceMonospaced=true
     bMaterialIsStencil=false
     fontMaterial=null
     fontObject=&HeadingNow
     letterSpacing=0
     monospacedWidth=1
     size=21
     skewAmount=0
     typefaceFontName=76 Bold
     outlineSettings {
      bApplyOutlineToDropShadows=false
      bMiteredCorners=false
      bSeparateFillAlpha=false
      outlineMaterial=&M-FontShadow
      outlineSize=2
      outlineColor=<rgba 1 0.5 1 1>
     }
    }
    shadowColorAndOpacity=<rgba 0 0 0 1>
    shadowOffset=<0 0>
    shapedTextOptions {bOverride_TextFlowDirection=false, bOverride_TextShapingMethod=false, textFlowDirection=LeftToRight, textShapingMethod=Auto}
   }
   slot {class=OverlaySlot, properties {horizontalAlignment=HAlign_Center, verticalAlignment=VAlign_Center}}
  }
 ]
}
