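# Quickplay HUD widget that shows the band score: a score sub-widget with an uppercase "Band Score" label, collapsed by default.
# PreConstruct mirrors the overlay for Arabic by negative X render scale, counter-flipping the label box to keep it readable, and creates a PilgrimBandScoreVM.
# Construct subscribes to tag-driven visibility changes.
# When the music battle instance initializes, the overlay shows only if there are multiple player states.
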WidgetBlueprint /PilgrimQuickplay/Widgets/WBP_PilgrimQuickplay_BandScore
graphs [
 EventGraph {nodes [
   n1 = Event PreConstruct()
   FlipReadDirectionArabic!(execute=[@n3.then, @n1.then], self, IsAr:bool="Equal, Case Insensitive (String)"(A=GetCurrentLanguage!() -> (ReturnValue:String), B=ar)) -> (then)
   > n2 = Construct Object from Class(Class=/Script/PilgrimCoreUI.PilgrimBandScoreVM) -> (ReturnValue:Pilgrim Band Score VM Object Reference)
   > Set PilgrimBandScoreVM(PilgrimBandScoreVM:Pilgrim Band Score VM Object Reference=@n2) -> (Output_Get:Pilgrim Band Score VM Object Reference)
   Event Construct()
   > n3 = SubscribeForChangingVisibilityByTag!(
    execute
    self:Pilgrim Game Widget Base Object Reference
    GameplayTag:"Gameplay Tag Structure (by ref)"=MakeLiteralGameplayTag!(Value:Gameplay Tag Structure=(TagName="Pilgrim.Widget.BandScore")) -> (ReturnValue:Gameplay Tag Structure)
   ) -> (then)
   Event OnMusicBattleInstanceInitializeEvent() {is=title K2Node_Event}
   n4 = GetMusicBattleData!(self:Pilgrim Music Battle Instance Object Reference=Get MusicBattleInstance() -> (MusicBattleInstance:Pilgrim Music Battle Instance Object Reference)) -> (ReturnValue:"Pilgrim Music Battle Data Structure (by ref)")
   n5 = int - int(A=Length(TargetArray:Array of Pilgrim Music Battle Player State Structures=Break Pilgrim Music Battle Data!(PilgrimMusicBattleData:"Pilgrim Music Battle Data Structure (by ref)"=@n4) -> (MusicBattleStates:Array of Pilgrim Music Battle Player State Structures, BandState:Pilgrim Band State Structure) {is=title K2Node_BreakStruct}))
   > SetVisibility(
    self=Get BandScoreOverlay() -> (BandScoreOverlay:Overlay Object Reference)
    InVisibility=Select(Option 0=Collapsed, Option 1=SelfHitTestInvisible, Index="integer > integer"(A=@n5))
   )
  ]}
 FlipReadDirectionArabic {nodes [
   n1 = FlipReadDirectionArabic!() -> (then, IsAr:bool=~) {is=title K2Node_FunctionEntry}
   > SetRenderScale!(
    execute
    self:Widget Object Reference=[Get BandScoreOverlay(), Get LabelSizeBox() -> (LabelSizeBox:Size Box Object Reference)]
    Scale:Vector 2D Structure=MakeVector2D(X=Select(Option 0:double=1, Option 1:double=-1, Index=@n1) -> (ReturnValue:double), Y=1)
   ) -> (then)
  ]}
]
variables []
widgets {
 info {inheritedWidgetCount=0, parentClass=/Script/PilgrimCoreUI.PilgrimGameWidgetBase, rootWidgetClass=&Overlay, widgetCount=6}
 widgets [
  BandScoreOverlay {widgetClassPath=&Overlay, bIsVariable=true, widget=BandScoreOverlay, properties {visibility=Collapsed}}
  SizeControl {
   widgetClassPath=&Spacer
   parent=BandScoreOverlay
   widget=SizeControl
   properties {size=<374 56>}
   slot {class=OverlaySlot, properties {horizontalAlignment=HAlign_Fill, verticalAlignment=VAlign_Fill}}
  }
  WBP_PilgrimQuickplay_Score {
   widgetClassPath=&/WBP_PilgrimQuickplay_Score
   bIsVariable=true
   parent=BandScoreOverlay
   widget=WBP_PilgrimQuickplay_Score
   properties {alignmentLeft=true}
   slot {class=OverlaySlot, properties {horizontalAlignment=HAlign_Fill, verticalAlignment=VAlign_Fill}}
  }
  LabelSizeBox {
   widgetClassPath=&SizeBox
   bIsVariable=true
   parent=BandScoreOverlay
   widget=LabelSizeBox
   properties {bOverride_MaxDesiredWidth=true, maxDesiredWidth=144, minDesiredWidth=144}
   slot {class=OverlaySlot, properties {horizontalAlignment=HAlign_Right, verticalAlignment=VAlign_Center, padding=<ltrb 12 0 16 0>}}
  }
  ScaleBox_0 {
   widgetClassPath=&ScaleBox
   parent=LabelSizeBox
   widget=ScaleBox_0
   properties {stretch=ScaleToFitX, stretchDirection=DownOnly}
   slot {class=SizeBoxSlot}
  }
  CommonTextBlock_81 {
   widgetClassPath=&CommonTextBlock
   parent=ScaleBox_0
   widget=CommonTextBlock_81
   properties {
    justification=Right
    text=Band Score
    textTransformPolicy=ToUpper
    font {
     bForceMonospaced=false
     bMaterialIsStencil=false
     fontMaterial=/PilgrimCore/Widgets/Materials/MI_UI_Piligrim_HUDTextColor
     fontObject=&HeadingNow
     letterSpacing=20
     monospacedWidth=1
     size=12
     skewAmount=0
     typefaceFontName=76 Bold
     outlineSettings {
      bApplyOutlineToDropShadows=false
      bMiteredCorners=false
      bSeparateFillAlpha=true
      outlineMaterial=&M-FontShadow
      outlineSize=2
      outlineColor=<rgba 0 0 0 1>
     }
    }
    shadowOffset=<0 0>
    strikeBrush {
     bIsDynamicallyLoaded=false
     drawAs=NoDrawType
     imageType=NoImage
     mirroring=NoMirror
     resourceName=None
     resourceObject=null
     tiling=NoTile
     imageSize=<32 32>
     margin=<ltrb 0 0 0 0>
     outlineSettings {
      bUseBrushTransparency=false
      roundingType=HalfHeightRadius
      width=0
      color {colorUseRule=UseColor_Specified, specifiedColor=<rgba 0 0 0 0>}
      cornerRadii=<0 0 0 0>
     }
     tintColor {colorUseRule=UseColor_Specified, specifiedColor=<rgba 1 1 1 1>}
     uVRegion {bIsValid=false, max=<0 0>, min=<0 0>}
    }
   }
   slot {class=ScaleBoxSlot, properties {horizontalAlignment=HAlign_Right}}
  }
 ]
}
