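# Widget for a band star in Pilgrim Quickplay, showing progress, completed and gold states, each completion with burst, flare and particle animations.
# On construct it caches dynamic materials, randomizes FX parameters and listens for the music battle start event to push band size into star materials.
# Functions set progress, swap FX materials between normal and gold, and either play completion animations or jump to their end via ShowCompleteAnim.
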WidgetBlueprint /PilgrimQuickplay/Widgets/WBP_PilgrimStar_Band
animations [
 anim_completed {
  binding_count=10
  length_seconds=1.25
  object=&self:anim_completed
  sequence {
   bindings [
    ParticlesFX {
     id=#1
     possessed_class=VerticalBox
     tracks [
      {
       type=MovieSceneFloatTrack
       display_name=Render Opacity
       sections [{type=MovieSceneFloatSection, start_frame=0, channels [None {type=MovieSceneScriptingFloatChannel, keys [[9,0],[12,1],[18,1],[75,0]]}]}]
      }
      {
       type=MovieSceneByteTrack
       display_name=Visibility
       sections [{type=MovieSceneByteSection, channels [None {type=MovieSceneScriptingByteChannel, keys [[9,4],[75,1]]}]}]
      }
     ]
    }
    RingFlare {
     id=#2
     possessed_class=Image
     tracks [
      {
       type=MovieSceneFloatTrack
       display_name=Render Opacity
       sections [{type=MovieSceneFloatSection, channels [None {type=MovieSceneScriptingFloatChannel, keys [[15,0],[18,1],[33,1],[62,0]]}]}]
      }
      {
       type=MovieScene2DTransformTrack
       sections [{
         type=MovieScene2DTransformSection
         channels [
          Scale.X {type=MovieSceneScriptingFloatChannel, keys [[12,1],[59,5]]}
          Scale.Y {type=MovieSceneScriptingFloatChannel, keys [[12,1],[59,5]]}
         ]
        }]
      }
      {
       type=MovieSceneWidgetMaterialTrack
       sections [{type=MovieSceneParameterSection, channels [TaperGlow {type=MovieSceneScriptingFloatChannel, keys [[15,1],[62,4]]}]}]
      }
      {
       type=MovieSceneByteTrack
       display_name=Visibility
       sections [{type=MovieSceneByteSection, end_frame=59, start_frame=0, channels [None {type=MovieSceneScriptingByteChannel, keys [[0,4],[59,2]]}]}]
      }
     ]
    }
    StarBurst {
     id=#3
     possessed_class=Image
     tracks [
      {
       type=MovieSceneFloatTrack
       display_name=Render Opacity
       sections [{type=MovieSceneFloatSection, channels [None {type=MovieSceneScriptingFloatChannel, keys [[9,0],[12,1],[40,0]]}]}]
      }
      {
       type=MovieScene2DTransformTrack
       sections [{
         type=MovieScene2DTransformSection
         channels [
          Scale.X {type=MovieSceneScriptingFloatChannel, keys [[9,1],[40,3]]}
          Scale.Y {type=MovieSceneScriptingFloatChannel, keys [[9,1],[40,3]]}
         ]
        }]
      }
      {
       type=MovieSceneByteTrack
       display_name=Visibility
       sections [{type=MovieSceneByteSection, end_frame=45, start_frame=0, channels [None {type=MovieSceneScriptingByteChannel, keys [[0,4],[45,1]]}]}]
      }
     ]
    }
    BackingBase {
     id=#4
     possessed_class=Image
     tracks [
      {
       type=MovieSceneWidgetMaterialTrack
       sections [{
         type=MovieSceneParameterSection
         channels [
          Anim_SimpleColor {type=MovieSceneScriptingFloatChannel, keys [[3,0],[12,8],[75,1]]}
          Alpha_Final {type=MovieSceneScriptingFloatChannel, keys [[0,0.5],[12,1],[75,0.7]]}
         ]
        }]
      }
      {
       type=MovieScene2DTransformTrack
       sections [{
         type=MovieScene2DTransformSection
         channels [
          Scale.X {type=MovieSceneScriptingFloatChannel, keys [[3,1],[6,1.25],[20,1]]}
          Scale.Y {type=MovieSceneScriptingFloatChannel, keys [[3,1],[6,1.25],[20,1]]}
         ]
        }]
      }
     ]
    }
    Gradient {
     id=#5
     possessed_class=Image
     tracks [{
       type=MovieSceneWidgetMaterialTrack
       sections [{type=MovieSceneParameterSection, channels [Alpha_Final {type=MovieSceneScriptingFloatChannel, keys [[0,3],[3,5],[30,1]]}]}]
      }]
    }
    StarComplete {
     id=#6
     possessed_class=Image
     tracks [
      {
       type=MovieScene2DTransformTrack
       sections [{
         type=MovieScene2DTransformSection
         channels [
          Translation.X {type=MovieSceneScriptingFloatChannel}
          Translation.Y {type=MovieSceneScriptingFloatChannel}
          Angle {type=MovieSceneScriptingFloatChannel}
          Scale.X {type=MovieSceneScriptingFloatChannel, keys [[0,1],[3,0.95],[6,1.333],[24,1]]}
          Scale.Y {type=MovieSceneScriptingFloatChannel, keys [[0,1],[3,0.95],[6,1.333],[24,1]]}
          Shear.X {type=MovieSceneScriptingFloatChannel}
          Shear.Y {type=MovieSceneScriptingFloatChannel}
         ]
        }]
      }
      {
       type=MovieSceneWidgetMaterialTrack
       sections [{
         type=MovieSceneParameterSection
         channels [
          HotCore_Amount {type=MovieSceneScriptingFloatChannel, keys [[0,0.5],[9,8],[45,0.5]]}
          GlowAnimateAmount {type=MovieSceneScriptingFloatChannel, keys [[0,0.25],[9,0],[45,0.25]]}
         ]
        }]
      }
     ]
    }
    Particles_Up {
     id=#7
     possessed_class=Image
     tracks [{
       type=MovieSceneWidgetMaterialTrack
       sections [{type=MovieSceneParameterSection, channels [Anim_Scale {type=MovieSceneScriptingFloatChannel, keys [[9,0.75],[18,0],[25,0],[75,0.75]]}]}]
      }]
    }
    Completed_Glow {
     id=#8
     possessed_class=Image
     tracks [
      {
       type=MovieSceneFloatTrack
       display_name=Render Opacity
       sections [{type=MovieSceneFloatSection, end_frame=20, start_frame=0, channels [None {type=MovieSceneScriptingFloatChannel, keys [[0,0],[7,1]]}]}]
      }
      {
       type=MovieScene2DTransformTrack
       sections [{
         type=MovieScene2DTransformSection
         end_frame=30
         start_frame=0
         channels [
          Scale.X {type=MovieSceneScriptingFloatChannel, keys [[0,1.333],[7,7],[30,1.333]]}
          Scale.Y {type=MovieSceneScriptingFloatChannel, const [1.333,2]}
         ]
        }]
      }
     ]
    }
    FlareFX {
     id=#9
     possessed_class=Overlay
     tracks [
      {
       type=MovieScene2DTransformTrack
       sections [{
         type=MovieScene2DTransformSection
         channels [
          Scale.X {type=MovieSceneScriptingFloatChannel, const [1,1]}
          Scale.Y {type=MovieSceneScriptingFloatChannel, keys [[0,0.75],[3,1],[8,1],[30,0.75]]}
         ]
        }]
      }
      {
       type=MovieSceneFloatTrack
       display_name=Render Opacity
       sections [{type=MovieSceneFloatSection, channels [None {type=MovieSceneScriptingFloatChannel, keys [[0,0],[2,1],[12,1],[30,0]]}]}]
      }
      {
       type=MovieSceneByteTrack
       display_name=Visibility
       sections [{type=MovieSceneByteSection, end_frame=30, start_frame=0, channels [None {type=MovieSceneScriptingByteChannel, keys [[0,4],[30,2]]}]}]
      }
     ]
    }
    Particles_Down {
     id=#10
     possessed_class=Image
     tracks [{
       type=MovieSceneWidgetMaterialTrack
       sections [{type=MovieSceneParameterSection, channels [Anim_Scale {type=MovieSceneScriptingFloatChannel, keys [[9,0.75],[18,0],[30,0],[75,0.75]]}]}]
      }]
    }
   ]
   display_rate {numerator=60}
  }
 }
 anim_gold_completed {
  binding_count=11
  length_seconds=1.25
  object=&self:anim_gold_completed
  sequence {bindings [
    GoldStarComplete {
     id=#11
     possessed_class=Image
     tracks [
      {
       type=MovieScene2DTransformTrack
       sections [{
         type=MovieScene2DTransformSection
         channels [
          Translation.X {type=MovieSceneScriptingFloatChannel}
          Translation.Y {type=MovieSceneScriptingFloatChannel}
          Angle {type=MovieSceneScriptingFloatChannel}
          Scale.X {type=MovieSceneScriptingFloatChannel, keys [[0,1],[1,0.95],[2,1.333],[8,1]]}
          Scale.Y {type=MovieSceneScriptingFloatChannel, keys [[0,1],[1,0.95],[2,1.333],[8,1]]}
          Shear.X {type=MovieSceneScriptingFloatChannel}
          Shear.Y {type=MovieSceneScriptingFloatChannel}
         ]
        }]
      }
      {
       type=MovieSceneWidgetMaterialTrack
       sections [{
         type=MovieSceneParameterSection
         channels [
          Completed {type=MovieSceneScriptingFloatChannel, keys [[2,0],[6,1]]}
          DistortionAmount {type=MovieSceneScriptingFloatChannel, keys [[3,0],[4,4]]}
         ]
        }]
      }
     ]
    }
    Particles_Down {
     id=#12
     possessed_class=Image
     tracks [{
       type=MovieSceneWidgetMaterialTrack
       sections [{type=MovieSceneParameterSection, channels [Anim_Scale {type=MovieSceneScriptingFloatChannel, keys [[3,1],[6,0],[10,0],[25,0.666]]}]}]
      }]
    }
    Gradient {
     id=#13
     possessed_class=Image
     tracks [
      {
       type=MovieSceneWidgetMaterialTrack
       sections [{type=MovieSceneParameterSection, channels [Alpha_Final {type=MovieSceneScriptingFloatChannel, keys [[0,3],[1,5],[10,1]]}]}]
      }
      {
       type=MovieScene2DTransformTrack
       sections [{
         type=MovieScene2DTransformSection
         end_frame=3
         start_frame=0
         channels [
          Translation.X {type=MovieSceneScriptingFloatChannel, const [0,1]}
          Translation.Y {type=MovieSceneScriptingFloatChannel, const [0,1]}
          Angle {type=MovieSceneScriptingFloatChannel, const [90,1]}
          Scale.X {type=MovieSceneScriptingFloatChannel, const [16,1]}
          Scale.Y {type=MovieSceneScriptingFloatChannel, const [0.75,1]}
          Shear.X {type=MovieSceneScriptingFloatChannel, const [0,1]}
          Shear.Y {type=MovieSceneScriptingFloatChannel, const [0,1]}
         ]
        }]
      }
     ]
    }
    RingFlare {
     id=#14
     possessed_class=Image
     tracks [
      {
       type=MovieSceneFloatTrack
       display_name=Render Opacity
       sections [{type=MovieSceneFloatSection, channels [None {type=MovieSceneScriptingFloatChannel, keys [[5,0],[6,1],[11,1],[20.6667,0]]}]}]
      }
      {
       type=MovieScene2DTransformTrack
       sections [{
         type=MovieScene2DTransformSection
         channels [
          Scale.X {type=MovieSceneScriptingFloatChannel, keys [[4,1],[19.6667,5]]}
          Scale.Y {type=MovieSceneScriptingFloatChannel, keys [[4,1],[19.6667,5]]}
         ]
        }]
      }
      {
       type=MovieSceneWidgetMaterialTrack
       sections [{type=MovieSceneParameterSection, channels [TaperGlow {type=MovieSceneScriptingFloatChannel, keys [[5,1],[20.6667,4]]}]}]
      }
      {
       type=MovieSceneByteTrack
       display_name=Visibility
       sections [{type=MovieSceneByteSection, end_frame=19, start_frame=0, channels [None {type=MovieSceneScriptingByteChannel, keys [[0,4],[19.6667,2]]}]}]
      }
     ]
    }
    FlareFX {
     id=#15
     possessed_class=Overlay
     tracks [
      {
       type=MovieScene2DTransformTrack
       sections [{
         type=MovieScene2DTransformSection
         channels [
          Scale.X {type=MovieSceneScriptingFloatChannel, const [1,2]}
          Scale.Y {type=MovieSceneScriptingFloatChannel, keys [[0,0.75],[1,1],[3,1],[6,0.75]]}
         ]
        }]
      }
      {
       type=MovieSceneFloatTrack
       display_name=Render Opacity
       sections [{type=MovieSceneFloatSection, channels [None {type=MovieSceneScriptingFloatChannel, keys [[0,0],[0.6667,1],[4,1],[10,0]]}]}]
      }
      {
       type=MovieSceneByteTrack
       display_name=Visibility
       sections [{type=MovieSceneByteSection, end_frame=10, start_frame=0, channels [None {type=MovieSceneScriptingByteChannel, const [4,1]}]}]
      }
     ]
    }
    StarBurst_Gold {
     id=#16
     possessed_class=Image
     tracks [
      {
       type=MovieSceneFloatTrack
       display_name=Render Opacity
       sections [{type=MovieSceneFloatSection, channels [None {type=MovieSceneScriptingFloatChannel, keys [[3,0],[4,1],[13.3333,0]]}]}]
      }
      {
       type=MovieScene2DTransformTrack
       sections [{
         type=MovieScene2DTransformSection
         channels [
          Scale.X {type=MovieSceneScriptingFloatChannel, keys [[3,1],[13.3333,3]]}
          Scale.Y {type=MovieSceneScriptingFloatChannel, keys [[3,1],[13.3333,3]]}
         ]
        }]
      }
      {
       type=MovieSceneByteTrack
       display_name=Visibility
       sections [{type=MovieSceneByteSection, end_frame=15, start_frame=0, channels [None {type=MovieSceneScriptingByteChannel, keys [[0,4],[15,1]]}]}]
      }
     ]
    }
    Particles_Up {
     id=#17
     possessed_class=Image
     tracks [{
       type=MovieSceneWidgetMaterialTrack
       sections [{type=MovieSceneParameterSection, channels [Anim_Scale {type=MovieSceneScriptingFloatChannel, keys [[3,1],[6,0],[10,0],[25,0.666]]}]}]
      }]
    }
    BackingGold {
     id=#18
     possessed_class=Image
     tracks [
      {
       type=MovieSceneWidgetMaterialTrack
       sections [{
         type=MovieSceneParameterSection
         channels [
          Anim_SimpleColor {type=MovieSceneScriptingFloatChannel, keys [[1,0],[4,8],[25,1]]}
          Alpha_Final {type=MovieSceneScriptingFloatChannel, keys [[0,0.5],[2,0.5],[5,1],[25,0.8]]}
         ]
        }]
      }
      {
       type=MovieScene2DTransformTrack
       sections [{
         type=MovieScene2DTransformSection
         channels [
          Scale.X {type=MovieSceneScriptingFloatChannel, keys [[1,1],[2,1.25],[6.6667,1]]}
          Scale.Y {type=MovieSceneScriptingFloatChannel, keys [[1,1],[2,1.25],[6.6667,1]]}
         ]
        }]
      }
     ]
    }
    ParticlesFX {
     id=#19
     possessed_class=VerticalBox
     tracks [
      {
       type=MovieSceneFloatTrack
       display_name=Render Opacity
       sections [{type=MovieSceneFloatSection, start_frame=0, channels [None {type=MovieSceneScriptingFloatChannel, keys [[3,0],[4,1],[24,1],[25,0]]}]}]
      }
      {
       type=MovieSceneByteTrack
       display_name=Visibility
       sections [{type=MovieSceneByteSection, channels [None {type=MovieSceneScriptingByteChannel, const [4,2]}]}]
      }
     ]
    }
    GoldStarComplete_Glow_Radial {
     id=#20
     possessed_class=Image
     tracks [{
       type=MovieSceneWidgetMaterialTrack
       sections [{type=MovieSceneParameterSection, channels [Final_Alpha {type=MovieSceneScriptingFloatChannel, keys [[3,0.25],[6,1],[25,0.25]]}]}]
      }]
    }
    Completed_Glow_Gold {
     id=#21
     possessed_class=Image
     tracks [
      {
       type=MovieSceneFloatTrack
       display_name=Render Opacity
       sections [{type=MovieSceneFloatSection, end_frame=6, start_frame=0, channels [None {type=MovieSceneScriptingFloatChannel, keys [[0,0],[2.3333,1]]}]}]
      }
      {
       type=MovieScene2DTransformTrack
       sections [{
         type=MovieScene2DTransformSection
         channels [
          Scale.X {type=MovieSceneScriptingFloatChannel, keys [[0,1.333],[1,8],[10,2]]}
          Scale.Y {type=MovieSceneScriptingFloatChannel, keys [[0,1.333],[2,2],[10,2]]}
         ]
        }]
      }
     ]
    }
   ]}
 }
]
graphs [
 EventGraph {nodes [
   Event Destruct()
   > Parent: Destruct()
   n1 = Get PilgrimMBIStartedEventHandle() -> (PilgrimMBIStartedEventHandle:Async Action Start Listening To Stateful Event Object Reference)
   > Is Valid(InputObject=@n1)
   > StopListeningToStatefulEvent(self:Async Action Start Listening To Stateful Event Object Reference=@n1)
   > Set PilgrimMBIStartedEventHandle(PilgrimMBIStartedEventHandle:Async Action Start Listening To Stateful Event Object Reference) -> (Output_Get:Async Action Start Listening To Stateful Event Object Reference)
   Event Construct()
   > SetRenderOpacity(self=Get All() -> (All:Overlay Object Reference), InOpacity=0.5)
   > RandomizeFXParameters()
   > Set Materials() {is=title K2Node_CallFunction}
   > n2 = PilgrimWaitForEventRouter!(execute:Exec, ContextActor:Actor Object Reference=GetOwningPlayer()) -> (then:Exec, OnEventRouterAvailable:Exec, EventRouter:Gameplay Event Router Component Object Reference) {is=title K2Node_AsyncAction}
   n3 = StartListeningToStatefulEvent!(
    execute:Exec=@n2.OnEventRouterAvailable
    Target:Gameplay Event Router Component Object Reference=@n2
    EventType:Script Struct Object Reference=/Script/PilgrimCoreRuntime.PilgrimEvent_MBI_Started
    Context:Object Reference
    EventBubblingRule:EEventBubblingRule Enum
   ) -> (
    then:Exec
    AsyncTaskProxy:Async Action Start Listening To Stateful Event Object Reference
    OnEventReceived:Exec
    OnSavedEventStateExists:Exec
    OnEventStateCleared:Exec
    Payload:Pilgrim Event MBI Started Structure
   ) {is=title K2Node_StartListeningToStatefulEventAsyncAction}
   > Set PilgrimMBIStartedEventHandle(PilgrimMBIStartedEventHandle=@n3)
   n4 = Cast To MusicBattleInstance!(
    execute=[@n3.OnEventReceived, @n3.OnSavedEventStateExists]
    Object=Break Pilgrim Event MBI Event Base!(PilgrimEvent_MBI_EventBase:"Pilgrim Event MBI Event Base Structure (by ref)"=@n3.Payload) -> (MusicBattleInstance:Pilgrim Music Battle Instance Weak object) {is=title K2Node_BreakStruct}
   ) -> (then, CastFailed, AsMusic Battle Instance:Music Battle Instance Object Reference) {is=title K2Node_DynamicCast}
   n5 = GetActorTeam!(Actor:Actor Object Reference=GetOwningPlayer()) -> (ReturnValue:Byte=0)
   n6 = GetBandState!(ContextActor:Actor Object Reference=@n4) -> (ReturnValue:Pilgrim Quickplay Band State Component Object Reference)
   > SetBandSize!(
    execute
    self
    BandSize:int=Length(TargetArray:Array of Player State Object References=GetPlayersInBand!(self:Pilgrim Quickplay Band State Component Object Reference=@n6, Team:Byte=@n5) -> (ReturnValue:Array of Player State Object References))
   ) -> (then)
   Branch()
   > SetFXColors!(execute, self, Gold FX:bool=true) -> (then)
   > RandomizeFXParameters()
   Event PreConstruct()
  ]}
 RandomizeFXParameters {nodes [
   RandomizeFXParameters() {is=title K2Node_FunctionEntry}
   > n1 = GetDynamicMaterial(self=Get Flare() -> (Flare:Image Object Reference))
   > SetScalarParameterValue(self=@n1, ParameterName=NoiseSet1_Tile_Alter, Value=RandomFloatInRange(Min=0.5, Max=1.5))
   > SetScalarParameterValue(self=@n1, ParameterName=NoiseSet2_Tile_Alter, Value=RandomFloatInRange(Min=0.5, Max=1.5))
   > n2 = GetDynamicMaterial(self=Get Particles_Up() -> (Particles_Up:Image Object Reference))
   > n3 = GetDynamicMaterial(self=Get Particles_Down() -> (Particles_Down:Image Object Reference))
   > SetScalarParameterValue(self=[@n3, @n2], ParameterName=Set1_SeedAdjustment, Value=RandomFloatInRange(Min=0.5, Max=2))
   > SetScalarParameterValue(self=[@n3, @n2], ParameterName=Set2_SeedAdjustment, Value=RandomFloatInRange(Min=0.5, Max=2))
   > n4 = GetDynamicMaterial(self=Get Completed_Glow() -> (Completed_Glow:Image Object Reference))
   > n5 = GetDynamicMaterial(self=Get Completed_Glow_Gold() -> (Completed_Glow_Gold:Image Object Reference))
   > SetScalarParameterValue(self=[@n5, @n4], ParameterName=SpeedMultiplier, Value=RandomFloatInRange(Min=18, Max=22))
   > SetRenderTransformAngle!(
    execute
    self:Widget Object Reference=[Get Completed_Glow_Gold(), Get Completed_Glow()]
    Angle:float=float * float(A=RandomFloatInRange(Min=0, Max=10), B=0.1) {is=title K2Node_PromotableOperator}
   ) -> (then)
  ]}
 Reset Widget {nodes [
   n1 = Reset Widget() {is=title K2Node_FunctionEntry}
   n2 = Get anim_gold_completed() -> (anim_gold_completed:Widget Animation Object Reference)
   PlayAnimation(execute=[@n1.then, @n4.then], InAnimation=@n2, StartAtTime=GetEndTime(self=@n2), PlayMode=Reverse)
   n3 = Get anim_completed() -> (anim_completed:Widget Animation Object Reference)
   > PlayAnimation(InAnimation=@n3, StartAtTime=GetEndTime(self=@n3), PlayMode=Reverse)
   > SetVisibility(self=Get BG() -> (BG:Image Object Reference), InVisibility=SelfHitTestInvisible)
   > SetVisibility(
    self=[
     Get StarIncomplete() -> (StarIncomplete:Image Object Reference)
     Get GoldStarComplete() -> (GoldStarComplete:Image Object Reference)
     Get GoldStarProgress() -> (GoldStarProgress:Image Object Reference)
     Get Image_Timer_ProgressBar() -> (Image_Timer_ProgressBar:Image Object Reference)
     Get CompletedStar_Overlay() -> (CompletedStar_Overlay:Overlay Object Reference)
     Get Completed_Glow()
     Get GoldStarOverlay() -> (GoldStarOverlay:Overlay Object Reference)
    ]
    InVisibility=Collapsed
   )
   > SetScalarParameterValue(
    self=[
     Get GoldStarProgressMat() -> (GoldStarProgressMat:Material Instance Dynamic Object Reference)
     Get StarProgressMat() -> (StarProgressMat:Material Instance Dynamic Object Reference)
     Get GoldStarGlowMat() -> (GoldStarGlowMat:Material Instance Dynamic Object Reference)
    ]
    ParameterName=Progress
    Value=0
   )
   > SetRenderOpacity(self=Get All(), InOpacity=1)
   > SetActiveWidgetIndex(self=Get BackingTypeSwitcher() -> (BackingTypeSwitcher:Common Activatable Widget Switcher Object Reference), Index=0)
   n4 = Set ShowCompleteAnim(ShowCompleteAnim:bool=true) -> (Output_Get:bool)
  ]}
 Set Gold Star Progress {nodes [
   n1 = Set Gold Star Progress!() -> (then, Progress:double=~) {is=title K2Node_FunctionEntry}
   > SetScalarParameterValue(self=[Get GoldStarProgressMat(), Get GoldStarGlowMat()], ParameterName=Progress, Value=@n1)
   > SetVisibility(
    self=[Get GoldStarProgress(), Get GoldStarComplete_Glow_Radial() -> (GoldStarComplete_Glow_Radial:Image Object Reference), Get GoldStarOverlay()]
    InVisibility=SelfHitTestInvisible
   )
   > SetActiveWidgetIndex(self=Get BackingTypeSwitcher(), Index=1)
  ]}
 Set Materials {nodes [
   Set Materials() {is=title K2Node_FunctionEntry}
   > n1 = GetDynamicMaterial(self=Get Image_Timer_ProgressBar())
   > Set StarProgressMat(StarProgressMat:Material Instance Dynamic Object Reference=@n1) -> (Output_Get:Material Instance Dynamic Object Reference)
   > n2 = GetDynamicMaterial(self=Get BG())
   > Set BGMat(BGMat:Material Instance Dynamic Object Reference=@n2) -> (Output_Get:Material Instance Dynamic Object Reference)
   > n3 = GetDynamicMaterial(self=Get StarIncomplete())
   > Set StarIncompleteMat(StarIncompleteMat:Material Instance Dynamic Object Reference=@n3) -> (Output_Get:Material Instance Dynamic Object Reference)
   > n4 = GetDynamicMaterial(self=Get StarComplete() -> (StarComplete:Image Object Reference))
   > Set StarCompleteMat(StarCompleteMat:Material Instance Dynamic Object Reference=@n4) -> (Output_Get:Material Instance Dynamic Object Reference)
   > n5 = GetDynamicMaterial(self=Get GoldStarComplete())
   > Set GoldStarCompleteMat(GoldStarCompleteMat:Material Instance Dynamic Object Reference=@n5) -> (Output_Get:Material Instance Dynamic Object Reference)
   > n6 = GetDynamicMaterial(self=Get GoldStarProgress())
   > Set GoldStarProgressMat(GoldStarProgressMat:Material Instance Dynamic Object Reference=@n6) -> (Output_Get:Material Instance Dynamic Object Reference)
   > n7 = GetDynamicMaterial(self=Get GoldStarComplete_Glow_Radial())
   > Set GoldStarGlowMat(GoldStarGlowMat:Material Instance Dynamic Object Reference=@n7) -> (Output_Get:Material Instance Dynamic Object Reference)
  ]}
 SetBandSize {nodes [SetScalarParameterValue(
    self=[
     Get BGMat() -> (BGMat:Material Instance Dynamic Object Reference)
     Get StarIncompleteMat() -> (StarIncompleteMat:Material Instance Dynamic Object Reference)
     Get StarCompleteMat() -> (StarCompleteMat:Material Instance Dynamic Object Reference)
     Get GoldStarProgressMat()
     Get GoldStarCompleteMat() -> (GoldStarCompleteMat:Material Instance Dynamic Object Reference)
    ]
    ParameterName=BandSize
    Value="To Float (Integer)"(InInt=SetBandSize!() -> (then, BandSize:int) {is=title K2Node_FunctionEntry})
   )]}
 SetComplete {nodes [
   SetComplete() {is=title K2Node_FunctionEntry}
   > SetFXColors(Gold FX=false)
   > SetVisibility(self=[Get BG(), Get Image_Timer_ProgressBar(), Get StarIncomplete()], InVisibility=Collapsed)
   > SetVisibility(self=[Get CompletedStar_Overlay(), Get Completed_Glow()], InVisibility=SelfHitTestInvisible)
   > n1 = Branch(Condition=Get ShowCompleteAnim() -> (ShowCompleteAnim:bool))
   > n2 = PlayAnimation(InAnimation=Get anim_completed())
   SetRenderOpacity(execute=[@n2.then, @n4.then], self=Get All(), InOpacity=1)
   n3 = GetEndTime(self=Get anim_completed())
   n4 = PlayAnimationTimeRange!(
    execute=@n1.else
    self:User Widget Object Reference
    InAnimation:Widget Animation Object Reference=Get anim_completed()
    StartAtTime:float=@n3
    EndAtTime:float=@n3
    NumLoopsToPlay:int=1
    PlayMode:EUMGSequencePlayMode Enum
    PlaybackSpeed:float=1
    bRestoreState:bool=false
   ) -> (then, ReturnValue:Widget Animation Handle Structure)
  ]}
 SetFXColors {nodes [
   n1 = SetFXColors!() -> (then, Gold FX:bool=~) {is=title K2Node_FunctionEntry}
   > SetBrushFromMaterial(
    self=Get Flare()
    Material=Select(
     Option 0:Material Interface Object Reference=/PilgrimCore/Widgets/Materials/MI_UI_Pilgrim_StarComplete_Flare
     Option 1:Material Interface Object Reference=/PilgrimCore/Widgets/Materials/MI_UI_Pilgrim_StarComplete_Flare_Gold
     Index=@n1
    ) -> (ReturnValue:Material Interface Object Reference)
   )
   > SetBrushFromMaterial(
    self=Get Gradient() -> (Gradient:Image Object Reference)
    Material=Select(
     Option 0=/PilgrimCore/Widgets/Materials/MI_UI_Pilgrim_StarComplete_GradientBurst
     Option 1=/PilgrimCore/Widgets/Materials/MI_UI_Pilgrim_StarComplete_GradientBurst_Gold
     Index=@n1
    )
   )
   > SetBrushFromMaterial(
    self=Get Particles_Up()
    Material=Select(
     Option 0=/PilgrimCore/Widgets/Materials/MI_UI_Piligrim_SquareParticles
     Option 1=/PilgrimCore/Widgets/Materials/MI_UI_Piligrim_SquareParticles_Gold
     Index=@n1
    )
   )
   > SetBrushFromMaterial(
    self=Get Particles_Down()
    Material=Select(
     Option 0=/PilgrimCore/Widgets/Materials/MI_UI_Piligrim_SquareParticles_Down
     Option 1=/PilgrimCore/Widgets/Materials/MI_UI_Piligrim_SquareParticles_Gold_Down
     Index=@n1
    )
   )
   > SetBrushFromMaterial(
    self=Get RingFlare() -> (RingFlare:Image Object Reference)
    Material=Select(
     Option 0=/PilgrimCore/Widgets/Materials/MI_UI_Piligrim_GemBreak_Pulse
     Option 1=/PilgrimCore/Widgets/Materials/MI_UI_Piligrim_GemBreak_Pulse_Gold
     Index=@n1
    )
   )
  ]}
 SetGoldStarComplete {nodes [
   SetGoldStarComplete() {is=title K2Node_FunctionEntry}
   > SetFXColors(Gold FX=true) {is=title K2Node_CallFunction}
   > RandomizeFXParameters() {is=title K2Node_CallFunction}
   > SetVisibility(self=[Get CompletedStar_Overlay(), Get GoldStarProgress()], InVisibility=Hidden)
   > SetVisibility(
    self=[
     Get StarBurst_Gold() -> (StarBurst_Gold:Image Object Reference)
     Get GoldStarOverlay()
     Get GoldStarComplete()
     Get GoldStarComplete_Glow_Radial()
     Get Completed_Glow_Gold()
    ]
    InVisibility=SelfHitTestInvisible
   )
   > n1 = Branch(Condition=Get ShowCompleteAnim())
   n2 = Get anim_gold_completed()
   > n3 = PlayAnimation(InAnimation=@n2)
   SetActiveWidgetIndex(execute=[@n3.then, @n4.then], self=Get BackingTypeSwitcher(), Index=1)
   n4 = PlayAnimation(execute=@n1.else, InAnimation=@n2, StartAtTime=GetEndTime(self=@n2), PlayMode=Reverse)
  ]}
 SetStarProgress {nodes [
   n1 = SetStarProgress!() -> (then, SetStarProgress:double=~) {is=title K2Node_FunctionEntry}
   > SetScalarParameterValue(self=Get StarProgressMat(), ParameterName=Progress, Value=@n1)
   > SetVisibility(self=[Get StarIncomplete(), Get Image_Timer_ProgressBar()], InVisibility=SelfHitTestInvisible)
   > SetVisibility(self=Get BG(), InVisibility=Collapsed)
   > SetRenderOpacity(self=Get All(), InOpacity=1)
  ]}
]
variables [
 BGMat {type=MaterialInstanceDynamic}
 GoldStarCompleteMat {type=MaterialInstanceDynamic}
 GoldStarGlowMat {type=MaterialInstanceDynamic}
 GoldStarProgressMat {type=MaterialInstanceDynamic}
 PilgrimMBIStartedEventHandle {type=AsyncAction_StartListeningToStatefulEvent, default=null}
 ShowCompleteAnim {type=boolean}
 StarCompleteMat {type=MaterialInstanceDynamic}
 StarIncompleteMat {type=MaterialInstanceDynamic}
 StarProgressMat {type=MaterialInstanceDynamic}
]
widgets {
 info {inheritedWidgetCount=0, parentClass=&UserWidget, rootWidgetClass=&SizeBox, widgetCount=30}
 widgets [
  SizeBox_5 {
   widgetClassPath=&SizeBox
   widget=SizeBox_5
   properties {bOverride_HeightOverride=true, bOverride_WidthOverride=true, heightOverride=70, widthOverride=70}
  }
  All {widgetClassPath=&Overlay, bIsVariable=true, parent=SizeBox_5, widget=All, slot {class=SizeBoxSlot}}
  BackingsOverlay {
   widgetClassPath=&Overlay
   parent=All
   widget=BackingsOverlay
   slot {class=OverlaySlot, properties {horizontalAlignment=HAlign_Fill, verticalAlignment=VAlign_Fill}}
  }
  Shadow {
   widgetClassPath=&Image
   parent=BackingsOverlay
   widget=Shadow
   properties {
    visibility=SelfHitTestInvisible
    brush {
     bIsDynamicallyLoaded=false
     drawAs=Image
     imageType=NoImage
     mirroring=NoMirror
     resourceName=None
     resourceObject=/PilgrimCore/Widgets/Materials/M_UI_Pilgrim_RoundedShadow_Round_Star
     tiling=NoTile
     imageSize=<72 72>
     margin=<ltrb 0 0 0 0>
     outlineSettings {
      bUseBrushTransparency=false
      roundingType=HalfHeightRadius
      width=0
      color {colorUseRule=UseColor_Specified, specifiedColor=<rgba 0 0 0 0>}
      cornerRadii=<0 0 0 0>
     }
     tintColor {colorUseRule=UseColor_Specified, specifiedColor=<rgba 1 1 1 1>}
     uVRegion {bIsValid=false, max=<0 0>, min=<0 0>}
    }
    renderTransform {angle=0, scale=<1.2 1.22>, shear=<0 0>, translation=<0 0>}
   }
   slot {class=OverlaySlot}
  }
  RingFlare {
   widgetClassPath=&Image
   bIsVariable=true
   parent=BackingsOverlay
   widget=RingFlare
   properties {
    visibility=Hidden
    brush {
     bIsDynamicallyLoaded=false
     drawAs=Image
     imageType=NoImage
     mirroring=NoMirror
     resourceName=None
     resourceObject=&MI_UI_Piligrim_GemBreak_Pulse
     tiling=NoTile
     imageSize=<32 32>
     margin=<ltrb 0 0 0 0>
     outlineSettings {
      bUseBrushTransparency=false
      roundingType=HalfHeightRadius
      width=0
      color {colorUseRule=UseColor_Specified, specifiedColor=<rgba 0 0 0 0>}
      cornerRadii=<0 0 0 0>
     }
     tintColor {colorUseRule=UseColor_Specified, specifiedColor=<rgba 1 1 1 1>}
     uVRegion {bIsValid=false, max=<0 0>, min=<0 0>}
    }
   }
   slot {class=OverlaySlot, properties {horizontalAlignment=HAlign_Fill, verticalAlignment=VAlign_Fill}}
  }
  BackingTypeSwitcher {
   widgetClassPath=&CommonActivatableWidgetSwitcher
   bIsVariable=true
   parent=BackingsOverlay
   widget=BackingTypeSwitcher
   properties {transitionDuration=1}
   slot {class=OverlaySlot, properties {horizontalAlignment=HAlign_Fill, verticalAlignment=VAlign_Fill}}
  }
  BackingBase {
   widgetClassPath=&Image
   bIsVariable=true
   parent=BackingTypeSwitcher
   widget=BackingBase
   properties {
    visibility=SelfHitTestInvisible
    brush {
     bIsDynamicallyLoaded=false
     drawAs=Image
     imageType=NoImage
     mirroring=NoMirror
     resourceName=None
     resourceObject=/PilgrimCore/Widgets/Materials/M_UI_Pilgrim_HUD_StarBacking
     tiling=NoTile
     imageSize=<72 72>
     margin=<ltrb 0 0 0 0>
     outlineSettings {
      bUseBrushTransparency=false
      roundingType=HalfHeightRadius
      width=0
      color {colorUseRule=UseColor_Specified, specifiedColor=<rgba 0 0 0 0>}
      cornerRadii=<0 0 0 0>
     }
     tintColor {colorUseRule=UseColor_Specified, specifiedColor=<rgba 1 1 1 1>}
     uVRegion {bIsValid=false, max=<0 0>, min=<0 0>}
    }
   }
   slot {class=WidgetSwitcherSlot}
  }
  BackingGold {
   widgetClassPath=&Image
   bIsVariable=true
   parent=BackingTypeSwitcher
   widget=BackingGold
   properties {
    visibility=SelfHitTestInvisible
    brush {
     bIsDynamicallyLoaded=false
     drawAs=Image
     imageType=NoImage
     mirroring=NoMirror
     resourceName=None
     resourceObject=/PilgrimCore/Widgets/Materials/MI_UI_Pilgrim_HUD_StarBacking_Gold
     tiling=NoTile
     imageSize=<72 72>
     margin=<ltrb 0 0 0 0>
     outlineSettings {
      bUseBrushTransparency=false
      roundingType=HalfHeightRadius
      width=0
      color {colorUseRule=UseColor_Specified, specifiedColor=<rgba 0 0 0 0>}
      cornerRadii=<0 0 0 0>
     }
     tintColor {colorUseRule=UseColor_Specified, specifiedColor=<rgba 1 1 1 1>}
     uVRegion {bIsValid=false, max=<0 0>, min=<0 0>}
    }
   }
   slot {class=WidgetSwitcherSlot}
  }
  FlareFX {
   widgetClassPath=&Overlay
   parent=All
   widget=FlareFX
   properties {visibility=Hidden}
   slot {class=OverlaySlot, properties {horizontalAlignment=HAlign_Fill, verticalAlignment=VAlign_Fill}}
  }
  BeamShadow {
   widgetClassPath=&Image
   parent=FlareFX
   widget=BeamShadow
   properties {
    visibility=SelfHitTestInvisible
    brush {
     bIsDynamicallyLoaded=false
     drawAs=Image
     imageType=NoImage
     mirroring=NoMirror
     resourceName=None
     resourceObject=/PilgrimCore/Widgets/Materials/MI_UI_Pilgrim_StarComplete_GradientBurst_Shadow
     tiling=NoTile
     imageSize=<64 48>
     margin=<ltrb 0 0 0 0>
     outlineSettings {
      bUseBrushTransparency=false
      roundingType=HalfHeightRadius
      width=0
      color {colorUseRule=UseColor_Specified, specifiedColor=<rgba 0 0 0 0>}
      cornerRadii=<0 0 0 0>
     }
     tintColor {colorUseRule=UseColor_Specified, specifiedColor=<rgba 1 1 1 1>}
     uVRegion {bIsValid=false, max=<0 0>, min=<0 0>}
    }
    renderTransform {angle=90, scale=<12 2>, shear=<0 0>, translation=<0 0>}
   }
   slot {class=OverlaySlot, properties {horizontalAlignment=HAlign_Center, verticalAlignment=VAlign_Center}}
  }
  Gradient {
   widgetClassPath=&Image
   bIsVariable=true
   parent=FlareFX
   widget=Gradient
   properties {
    visibility=SelfHitTestInvisible
    brush {
     bIsDynamicallyLoaded=false
     drawAs=Image
     imageType=NoImage
     mirroring=NoMirror
     resourceName=None
     resourceObject=&MI_UI_Pilgrim_StarComplete_GradientBurst
     tiling=NoTile
     imageSize=<64 48>
     margin=<ltrb 0 0 0 0>
     outlineSettings {
      bUseBrushTransparency=false
      roundingType=HalfHeightRadius
      width=0
      color {colorUseRule=UseColor_Specified, specifiedColor=<rgba 0 0 0 0>}
      cornerRadii=<0 0 0 0>
     }
     tintColor {colorUseRule=UseColor_Specified, specifiedColor=<rgba 1 1 1 1>}
     uVRegion {bIsValid=false, max=<0 0>, min=<0 0>}
    }
    renderTransform {angle=90, scale=<16 1>, shear=<0 0>, translation=<0 0>}
   }
   slot {class=OverlaySlot, properties {horizontalAlignment=HAlign_Center, verticalAlignment=VAlign_Center}}
  }
  Flare {
   widgetClassPath=&Image
   bIsVariable=true
   parent=FlareFX
   widget=Flare
   properties {
    visibility=SelfHitTestInvisible
    brush {
     bIsDynamicallyLoaded=false
     drawAs=Image
     imageType=NoImage
     mirroring=NoMirror
     resourceName=None
     resourceObject=&MI_UI_Pilgrim_StarComplete_Flare
     tiling=NoTile
     imageSize=<64 32>
     margin=<ltrb 0 0 0 0>
     outlineSettings {
      bUseBrushTransparency=false
      roundingType=HalfHeightRadius
      width=0
      color {colorUseRule=UseColor_Specified, specifiedColor=<rgba 0 0 0 0>}
      cornerRadii=<0 0 0 0>
     }
     tintColor {colorUseRule=UseColor_Specified, specifiedColor=<rgba 1 1 1 1>}
     uVRegion {bIsValid=false, max=<0 0>, min=<0 0>}
    }
    renderTransform {angle=90, scale=<10 1>, shear=<0 0>, translation=<0 0>}
   }
   slot {class=OverlaySlot, properties {horizontalAlignment=HAlign_Center, verticalAlignment=VAlign_Center}}
  }
  FortVisualAttachment_204 {
   widgetClassPath=&FortVisualAttachment
   parent=All
   widget=FortVisualAttachment_204
   properties {
    bOverride_HeightOverride=true
    bOverride_WidthOverride=true
    heightOverride=512
    widthOverride=64
    contentAnchor=<0.5 0.5>
    renderTransform {angle=0, scale=<1 1.25>, shear=<0 0>, translation=<0 0>}
   }
   slot {class=OverlaySlot, properties {horizontalAlignment=HAlign_Center, verticalAlignment=VAlign_Center}}
  }
  ParticlesFX {
   widgetClassPath=&VerticalBox
   parent=FortVisualAttachment_204
   widget=ParticlesFX
   properties {visibility=Collapsed, renderTransform {angle=0, scale=<1 0.75>, shear=<0 0>, translation=<0 0>}}
   slot {class=SizeBoxSlot, properties {horizontalAlignment=HAlign_Center}}
  }
  Particles_Up {
   widgetClassPath=&Image
   bIsVariable=true
   parent=ParticlesFX
   widget=Particles_Up
   properties {
    visibility=SelfHitTestInvisible
    brush {
     bIsDynamicallyLoaded=false
     drawAs=Image
     imageType=NoImage
     mirroring=NoMirror
     resourceName=None
     resourceObject=&MI_UI_Piligrim_SquareParticles
     tiling=NoTile
     imageSize=<48 32>
     margin=<ltrb 0 0 0 0>
     outlineSettings {
      bUseBrushTransparency=false
      roundingType=HalfHeightRadius
      width=0
      color {colorUseRule=UseColor_Specified, specifiedColor=<rgba 0 0 0 0>}
      cornerRadii=<0 0 0 0>
     }
     tintColor {colorUseRule=UseColor_Specified, specifiedColor=<rgba 1 1 1 1>}
     uVRegion {bIsValid=false, max=<0 0>, min=<0 0>}
    }
    renderTransformPivot=<0.5 1>
   }
   slot {class=VerticalBoxSlot, properties {size {sizeRule=Fill, value=1}}}
  }
  Particles_Down {
   widgetClassPath=&Image
   bIsVariable=true
   parent=ParticlesFX
   widget=Particles_Down
   properties {
    visibility=SelfHitTestInvisible
    brush {
     bIsDynamicallyLoaded=false
     drawAs=Image
     imageType=NoImage
     mirroring=NoMirror
     resourceName=None
     resourceObject=&MI_UI_Piligrim_SquareParticles_Down
     tiling=NoTile
     imageSize=<48 32>
     margin=<ltrb 0 0 0 0>
     outlineSettings {
      bUseBrushTransparency=false
      roundingType=HalfHeightRadius
      width=0
      color {colorUseRule=UseColor_Specified, specifiedColor=<rgba 0 0 0 0>}
      cornerRadii=<0 0 0 0>
     }
     tintColor {colorUseRule=UseColor_Specified, specifiedColor=<rgba 1 1 1 1>}
     uVRegion {bIsValid=false, max=<0 0>, min=<0 0>}
    }
    renderTransform {angle=0, scale=<-1 1>, shear=<0 0>, translation=<0 0>}
    renderTransformPivot=<0.5 1>
   }
   slot {class=VerticalBoxSlot, properties {size {sizeRule=Fill, value=1}}}
  }
  Image_Timer_ProgressBar {
   widgetClassPath=&Image
   bIsVariable=true
   parent=All
   widget=Image_Timer_ProgressBar
   properties {
    visibility=Hidden
    brush {
     bIsDynamicallyLoaded=false
     drawAs=Image
     imageType=NoImage
     mirroring=NoMirror
     resourceName=None
     resourceObject=/PilgrimCore/Widgets/Materials/M_UI_Pilgrim_HUD_StarBar
     tiling=NoTile
     imageSize=<64 64>
     margin=<ltrb 0 0 0 0>
     outlineSettings {
      bUseBrushTransparency=false
      roundingType=HalfHeightRadius
      width=0
      color {colorUseRule=UseColor_Specified, specifiedColor=<rgba 0 0 0 0>}
      cornerRadii=<0 0 0 0>
     }
     tintColor {colorUseRule=UseColor_Specified, specifiedColor=<rgba 1 1 1 1>}
     uVRegion {bIsValid=false, max=<0 0>, min=<0 0>}
    }
   }
   slot {class=OverlaySlot, properties {horizontalAlignment=HAlign_Center, verticalAlignment=VAlign_Center}}
  }
  BG {
   widgetClassPath=&Image
   bIsVariable=true
   parent=All
   widget=BG
   properties {
    visibility=SelfHitTestInvisible
    brush {
     bIsDynamicallyLoaded=false
     drawAs=Image
     imageType=NoImage
     mirroring=NoMirror
     resourceName=None
     resourceObject=/PilgrimCore/Widgets/Materials/MI_UI_Pilgrim_BandStar_NotStarted
     tiling=NoTile
     imageSize=<56 56>
     margin=<ltrb 0 0 0 0>
     outlineSettings {
      bUseBrushTransparency=false
      roundingType=HalfHeightRadius
      width=0
      color {colorUseRule=UseColor_Specified, specifiedColor=<rgba 0 0 0 0>}
      cornerRadii=<0 0 0 0>
     }
     tintColor {colorUseRule=UseColor_Specified, specifiedColor=<rgba 1 1 1 1>}
     uVRegion {bIsValid=false, max=<0 0>, min=<0 0>}
    }
   }
   slot {class=OverlaySlot, properties {horizontalAlignment=HAlign_Center, verticalAlignment=VAlign_Center}}
  }
  StarIncomplete {
   widgetClassPath=&Image
   bIsVariable=true
   parent=All
   widget=StarIncomplete
   properties {
    visibility=Hidden
    brush {
     bIsDynamicallyLoaded=false
     drawAs=Image
     imageType=NoImage
     mirroring=NoMirror
     resourceName=None
     resourceObject=/PilgrimCore/Widgets/Materials/MI_UI_Pilgrim_BandStar_Incomplete
     tiling=NoTile
     imageSize=<56 56>
     margin=<ltrb 0 0 0 0>
     outlineSettings {
      bUseBrushTransparency=false
      roundingType=HalfHeightRadius
      width=0
      color {colorUseRule=UseColor_Specified, specifiedColor=<rgba 0 0 0 0>}
      cornerRadii=<0 0 0 0>
     }
     tintColor {colorUseRule=UseColor_Specified, specifiedColor=<rgba 1 1 1 1>}
     uVRegion {bIsValid=false, max=<0 0>, min=<0 0>}
    }
   }
   slot {class=OverlaySlot, properties {horizontalAlignment=HAlign_Center, verticalAlignment=VAlign_Center}}
  }
  CompletedStar_Overlay {
   widgetClassPath=&Overlay
   bIsVariable=true
   parent=All
   widget=CompletedStar_Overlay
   properties {visibility=Hidden}
   slot {class=OverlaySlot, properties {horizontalAlignment=HAlign_Fill, verticalAlignment=VAlign_Fill}}
  }
  Completed_Glow {
   widgetClassPath=&Image
   bIsVariable=true
   parent=CompletedStar_Overlay
   widget=Completed_Glow
   properties {
    visibility=SelfHitTestInvisible
    brush {
     bIsDynamicallyLoaded=false
     drawAs=Image
     imageType=NoImage
     mirroring=NoMirror
     resourceName=None
     resourceObject=/PilgrimCore/Widgets/Materials/MI_Pilgrim_LightRays_Star
     tiling=NoTile
     imageSize=<32 32>
     margin=<ltrb 0 0 0 0>
     outlineSettings {
      bUseBrushTransparency=false
      roundingType=HalfHeightRadius
      width=0
      color {colorUseRule=UseColor_Specified, specifiedColor=<rgba 0 0 0 0>}
      cornerRadii=<0 0 0 0>
     }
     tintColor {colorUseRule=UseColor_Specified, specifiedColor=<rgba 1 1 1 1>}
     uVRegion {bIsValid=false, max=<0 0>, min=<0 0>}
    }
    renderTransform {angle=0, scale=<1.333 1.333>, shear=<0 0>, translation=<0 0>}
   }
   slot {class=OverlaySlot, properties {horizontalAlignment=HAlign_Fill, verticalAlignment=VAlign_Fill}}
  }
  DarkenCenterGlow {
   widgetClassPath=&Image
   bIsVariable=true
   parent=CompletedStar_Overlay
   widget=DarkenCenterGlow
   properties {
    visibility=SelfHitTestInvisible
    brush {
     bIsDynamicallyLoaded=false
     drawAs=Image
     imageType=NoImage
     mirroring=NoMirror
     resourceName=None
     resourceObject=/PilgrimCore/Widgets/Materials/MI_Piligrim_GlowDarken_BandStar
     tiling=NoTile
     imageSize=<48 48>
     margin=<ltrb 0 0 0 0>
     outlineSettings {
      bUseBrushTransparency=false
      roundingType=HalfHeightRadius
      width=0
      color {colorUseRule=UseColor_Specified, specifiedColor=<rgba 0 0 0 0>}
      cornerRadii=<0 0 0 0>
     }
     tintColor {colorUseRule=UseColor_Specified, specifiedColor=<rgba 1 1 1 1>}
     uVRegion {bIsValid=false, max=<0 0>, min=<0 0>}
    }
   }
   slot {class=OverlaySlot, properties {horizontalAlignment=HAlign_Center, verticalAlignment=VAlign_Center}}
  }
  StarBurst {
   widgetClassPath=&Image
   parent=CompletedStar_Overlay
   widget=StarBurst
   properties {
    visibility=SelfHitTestInvisible
    brush {
     bIsDynamicallyLoaded=false
     drawAs=Image
     imageType=NoImage
     mirroring=NoMirror
     resourceName=None
     resourceObject=/PilgrimCore/Widgets/Materials/MI_UI_Pilgrim_BandStar_Outline
     tiling=NoTile
     imageSize=<72 72>
     margin=<ltrb 0 0 0 0>
     outlineSettings {
      bUseBrushTransparency=false
      roundingType=HalfHeightRadius
      width=0
      color {colorUseRule=UseColor_Specified, specifiedColor=<rgba 0 0 0 0>}
      cornerRadii=<0 0 0 0>
     }
     tintColor {colorUseRule=UseColor_Specified, specifiedColor=<rgba 1 1 1 1>}
     uVRegion {bIsValid=false, max=<0 0>, min=<0 0>}
    }
   }
   slot {class=OverlaySlot, properties {horizontalAlignment=HAlign_Center, verticalAlignment=VAlign_Center}}
  }
  StarComplete {
   widgetClassPath=&Image
   bIsVariable=true
   parent=CompletedStar_Overlay
   widget=StarComplete
   properties {
    visibility=SelfHitTestInvisible
    brush {
     bIsDynamicallyLoaded=false
     drawAs=Image
     imageType=NoImage
     mirroring=NoMirror
     resourceName=None
     resourceObject=/PilgrimCore/Widgets/Materials/MI_UI_Pilgrim_BandStar
     tiling=NoTile
     imageSize=<72 72>
     margin=<ltrb 0 0 0 0>
     outlineSettings {
      bUseBrushTransparency=false
      roundingType=HalfHeightRadius
      width=0
      color {colorUseRule=UseColor_Specified, specifiedColor=<rgba 0 0 0 0>}
      cornerRadii=<0 0 0 0>
     }
     tintColor {colorUseRule=UseColor_Specified, specifiedColor=<rgba 1 1 1 1>}
     uVRegion {bIsValid=false, max=<0 0>, min=<0 0>}
    }
   }
   slot {class=OverlaySlot, properties {horizontalAlignment=HAlign_Center, verticalAlignment=VAlign_Center}}
  }
  GoldStarOverlay {
   widgetClassPath=&Overlay
   bIsVariable=true
   parent=All
   widget=GoldStarOverlay
   properties {visibility=Hidden}
   slot {class=OverlaySlot, properties {horizontalAlignment=HAlign_Fill, verticalAlignment=VAlign_Fill, padding=<ltrb 1 1 1 1>}}
  }
  GoldStarComplete_Glow_Radial {
   widgetClassPath=&Image
   bIsVariable=true
   parent=GoldStarOverlay
   widget=GoldStarComplete_Glow_Radial
   properties {
    visibility=Hidden
    brush {
     bIsDynamicallyLoaded=false
     drawAs=Image
     imageType=NoImage
     mirroring=NoMirror
     resourceName=None
     resourceObject=/PilgrimCore/Widgets/Materials/MI_UI_Pilgrim_GoldStar_Glow
     tiling=NoTile
     imageSize=<64 64>
     margin=<ltrb 0 0 0 0>
     outlineSettings {
      bUseBrushTransparency=false
      roundingType=HalfHeightRadius
      width=0
      color {colorUseRule=UseColor_Specified, specifiedColor=<rgba 0 0 0 0>}
      cornerRadii=<0 0 0 0>
     }
     tintColor {colorUseRule=UseColor_Specified, specifiedColor=<rgba 1 1 1 1>}
     uVRegion {bIsValid=false, max=<0 0>, min=<0 0>}
    }
    renderTransform {angle=0, scale=<1.75 1.75>, shear=<0 0>, translation=<0 0>}
   }
   slot {class=OverlaySlot, properties {horizontalAlignment=HAlign_Fill, verticalAlignment=VAlign_Fill, padding=<ltrb 1 1 1 1>}}
  }
  Completed_Glow_Gold {
   widgetClassPath=&Image
   bIsVariable=true
   parent=GoldStarOverlay
   widget=Completed_Glow_Gold
   properties {
    visibility=Hidden
    brush {
     bIsDynamicallyLoaded=false
     drawAs=Image
     imageType=NoImage
     mirroring=NoMirror
     resourceName=None
     resourceObject=/PilgrimCore/Widgets/Materials/MI_Pilgrim_LightRays_Star_Gold
     tiling=NoTile
     imageSize=<32 32>
     margin=<ltrb 0 0 0 0>
     outlineSettings {
      bUseBrushTransparency=false
      roundingType=HalfHeightRadius
      width=0
      color {colorUseRule=UseColor_Specified, specifiedColor=<rgba 0 0 0 0>}
      cornerRadii=<0 0 0 0>
     }
     tintColor {colorUseRule=UseColor_Specified, specifiedColor=<rgba 1 1 1 1>}
     uVRegion {bIsValid=false, max=<0 0>, min=<0 0>}
    }
    renderTransform {angle=0, scale=<1.333 1.333>, shear=<0 0>, translation=<0 0>}
   }
   slot {class=OverlaySlot, properties {horizontalAlignment=HAlign_Fill, verticalAlignment=VAlign_Fill}}
  }
  StarBurst_Gold {
   widgetClassPath=&Image
   bIsVariable=true
   parent=GoldStarOverlay
   widget=StarBurst_Gold
   properties {
    visibility=Collapsed
    brush {
     bIsDynamicallyLoaded=false
     drawAs=Image
     imageType=NoImage
     mirroring=NoMirror
     resourceName=None
     resourceObject=/PilgrimCore/Widgets/Materials/MI_UI_Pilgrim_Star_OutlneBand_Gold
     tiling=NoTile
     imageSize=<64 64>
     margin=<ltrb 0 0 0 0>
     outlineSettings {
      bUseBrushTransparency=false
      roundingType=HalfHeightRadius
      width=0
      color {colorUseRule=UseColor_Specified, specifiedColor=<rgba 0 0 0 0>}
      cornerRadii=<0 0 0 0>
     }
     tintColor {colorUseRule=UseColor_Specified, specifiedColor=<rgba 1 1 1 1>}
     uVRegion {bIsValid=false, max=<0 0>, min=<0 0>}
    }
    renderTransform {angle=0, scale=<1 1>, shear=<0 0>, translation=<0 -6>}
   }
   slot {class=OverlaySlot, properties {horizontalAlignment=HAlign_Center, verticalAlignment=VAlign_Center}}
  }
  GoldStarComplete {
   widgetClassPath=&Image
   bIsVariable=true
   parent=GoldStarOverlay
   widget=GoldStarComplete
   properties {
    visibility=Hidden
    brush {
     bIsDynamicallyLoaded=false
     drawAs=Image
     imageType=NoImage
     mirroring=NoMirror
     resourceName=None
     resourceObject=/PilgrimCore/Widgets/Materials/MI_UI_Pilgrim_GoldStar_Band_Complete
     tiling=NoTile
     imageSize=<64 64>
     margin=<ltrb 0 0 0 0>
     outlineSettings {
      bUseBrushTransparency=false
      roundingType=HalfHeightRadius
      width=0
      color {colorUseRule=UseColor_Specified, specifiedColor=<rgba 0 0 0 0>}
      cornerRadii=<0 0 0 0>
     }
     tintColor {colorUseRule=UseColor_Specified, specifiedColor=<rgba 1 1 1 1>}
     uVRegion {bIsValid=false, max=<0 0>, min=<0 0>}
    }
   }
   slot {class=OverlaySlot, properties {horizontalAlignment=HAlign_Center, verticalAlignment=VAlign_Center, padding=<ltrb 1 1 1 1>}}
  }
  GoldStarProgress {
   widgetClassPath=&Image
   bIsVariable=true
   parent=GoldStarOverlay
   widget=GoldStarProgress
   properties {
    visibility=SelfHitTestInvisible
    brush {
     bIsDynamicallyLoaded=false
     drawAs=Image
     imageType=NoImage
     mirroring=NoMirror
     resourceName=None
     resourceObject=/PilgrimCore/Widgets/Materials/MI_UI_Pilgrim_GoldStar_Band
     tiling=NoTile
     imageSize=<64 64>
     margin=<ltrb 0 0 0 0>
     outlineSettings {
      bUseBrushTransparency=false
      roundingType=HalfHeightRadius
      width=0
      color {colorUseRule=UseColor_Specified, specifiedColor=<rgba 0 0 0 0>}
      cornerRadii=<0 0 0 0>
     }
     tintColor {colorUseRule=UseColor_Specified, specifiedColor=<rgba 1 1 1 1>}
     uVRegion {bIsValid=false, max=<0 0>, min=<0 0>}
    }
   }
   slot {class=OverlaySlot, properties {horizontalAlignment=HAlign_Center, verticalAlignment=VAlign_Center}}
  }
 ]
}
