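# WBP_PilgrimQuickplay_BandStarBar is a UMG widget showing the band's score as five radial star widgets, with an optional gold-star tier.
# Handle Star Update filters unchanged values, updates completed stars and fills the current one, switching all five to gold progress at the top tier.
# For results, SetupStarsResults sets final or reset visuals, and StaggeredStarReveal completes stars on a delay with rising-pitch sounds.
# PreConstruct applies scaling, animation flags and a designer preview.
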WidgetBlueprint /PilgrimQuickplay/Widgets/WBP_PilgrimQuickplay_BandStarBar
graphs [
 Check Is Star Changed {nodes [
   n1 = Check Is Star Changed!() -> (then, CurrentStarIndex:int, CurrentStarProgress:double=~) {is=title K2Node_FunctionEntry}
   n2 = Get LastStarProgress() -> (LastStarProgress:double)
   n3 = Get LastStarIndex() -> (LastStarIndex:int=0)
   > n4 = Set HasChanged(HasChanged:bool=OR Boolean(A="Not Equal (Integer)"(A=@n1, B=@n3), B="float < float"(A=@n2, B=@n1.CurrentStarProgress))) -> (Output_Get:bool)
   > Set LastStarProgress(LastStarProgress:double=@n1.CurrentStarProgress) -> (Output_Get:double)
   > Return Node!(execute, IsChanged:bool=@n4)
  ]}
 CheckStarCompletion {nodes [
   n1 = Get LastStarIndex() -> (LastStarIndex=0)
   n2 = CheckStarCompletion!() -> (then, CurrentStar:int) {is=title K2Node_FunctionEntry}
   n3 = For Loop(execute=@n10."<", FirstIndex=@n2, LastIndex=@n1)
   n4 = Get Star Widget!(execute, self, Index:int=@n3) -> (then, StarWidget:WBP Pilgrim Star Band Object Reference) {is=title K2Node_CallFunction}
   > SetVisibility(
    self=Get StarProgress!(self:WBP Pilgrim Star Band Object Reference=@n4) -> (StarProgress:Radial Slider Object Reference)
    InVisibility=Hidden
   )
   > SetVisibility(self=Get StarProgress(self=@n4), InVisibility=Hidden)
   n5 = int - int(A=@n2)
   n6 = For Loop(execute=@n9.then, FirstIndex=Get LastStarIndex() -> (LastStarIndex=0), LastIndex=@n5)
   > n7 = Get Star Widget(Index=@n6)
   > SetComplete(self:WBP Pilgrim Star Band Object Reference=@n7)
   n8 = Get Star Widget(execute=@n3.Completed, Index=@n5)
   > SetComplete(self=@n8)
   Set LastStarIndex(execute=@n6.Completed, LastStarIndex:int=@n2) -> (Output_Get:int=0)
   n9 = Branch(execute=@n10.">")
   n10 = Compare Int!(Exec:Exec=@n2.then, Input:int=@n2, CompareWith:int=@n1) -> (">":Exec, "==":Exec, "<":Exec) {is=title K2Node_MacroInstance}
  ]}
 EventGraph {nodes [
   n1 = StaggeredStarReveal!() -> (OutputDelegate, then, StarCount:int) {is=title K2Node_CustomEvent}
   > Branch(Condition="integer > integer"(A=@n1.StarCount))
   > n2 = For Each Loop with Delay!(
    Exec:Exec
    Array:Array of WBP Pilgrim Star Band Object References=Make Array!(
     "[0]":WBP Pilgrim Star Band Object Reference=Get RadialStar1() -> (RadialStar1:WBP Pilgrim Star Band Object Reference)
     "[1]":WBP Pilgrim Star Band Object Reference=Get RadialStar2() -> (RadialStar2:WBP Pilgrim Star Band Object Reference)
     "[2]":WBP Pilgrim Star Band Object Reference=Get RadialStar3() -> (RadialStar3:WBP Pilgrim Star Band Object Reference)
     "[3]":WBP Pilgrim Star Band Object Reference=Get RadialStar4() -> (RadialStar4:WBP Pilgrim Star Band Object Reference)
     "[4]":WBP Pilgrim Star Band Object Reference=Get RadialStar5() -> (RadialStar5:WBP Pilgrim Star Band Object Reference)
    ) -> (Array:Array of WBP Pilgrim Star Band Object References)
    Delay:double=0.25
   ) -> (Loop Body:Exec, Array Element:WBP Pilgrim Star Band Object Reference, Array Index:int, Completed:Exec) {is=title K2Node_MacroInstance}
   > Branch(Condition="integer < integer"(A=@n2.Array Index, B=@n1.StarCount))
   n3 = "To Float (Integer)"(InInt=@n2.Array Index)
   > PlaySound2D(
    Sound=Get StarCompleteSound() -> (StarCompleteSound:Sound Base Object Reference)
    PitchMultiplier=float + float(A=float * float(A=@n3, B=0.2) {is=title K2Node_PromotableOperator}, B=0.08)
   )
   > n4 = Branch(Condition=Get bIsGoldStars() -> (bIsGoldStars:bool))
   > SetGoldStarComplete(self:WBP Pilgrim Star Band Object Reference=@n2)
   SetComplete(execute=@n4.else, self=@n2)
   n5 = Event PreConstruct()
   > n6 = Sequence!(execute) -> (then_0, then_1, then_2:Exec)
   Branch(execute=@n6.then_1, Condition=@n5.IsDesignTime)
   > Handle Star Update!(execute, self, CurrentStar:int=Get Current Star() -> (Current Star:int=0), StarProgress:double=Get Star Progress() -> (Star Progress:double)) -> (then)
   n7 = Get ShouldStarsAnimateComplete() -> (ShouldStarsAnimateComplete:bool)
   Set ShowCompleteAnim!(execute=@n6.then_2, ShowCompleteAnim:bool=@n7, self:WBP Pilgrim Star Band Object Reference=Get RadialStar1()) -> (then, Output_Get:bool)
   > Set ShowCompleteAnim(ShowCompleteAnim=@n7, self=Get RadialStar2())
   > Set ShowCompleteAnim(ShowCompleteAnim=@n7, self=Get RadialStar3())
   > Set ShowCompleteAnim(ShowCompleteAnim=@n7, self=Get RadialStar4())
   > Set ShowCompleteAnim(ShowCompleteAnim=@n7, self=Get RadialStar5())
   n8 = Get ScaleBox() -> (ScaleBox:Scale Box Object Reference)
   SetStretch!(execute=@n6.then_0, self:Scale Box Object Reference=@n8, InStretch:EStretch Enum=Get Scale Stretch() -> (Scale Stretch:EStretch Enum=None)) -> (then)
   > SetUserSpecifiedScale!(execute, self:Scale Box Object Reference=@n8, InUserSpecifiedScale:float=Get In User Specified Scale() -> (In User Specified Scale:double)) -> (then)
   Reroute Node(InputPin:Wildcard) -> (OutputPin:Wildcard)
  ]}
 Get Star Widget {nodes [
   n1 = Get Star Widget!() -> (then, Index:int) {is=title K2Node_FunctionEntry}
   > Return Node!(
    execute
    StarWidget:WBP Pilgrim Star Band Object Reference=Select!(
     Option 0:WBP Pilgrim Star Band Object Reference=Get RadialStar1()
     Option 1:WBP Pilgrim Star Band Object Reference=Get RadialStar2()
     Option 2:WBP Pilgrim Star Band Object Reference=Get RadialStar3()
     Option 3:WBP Pilgrim Star Band Object Reference=Get RadialStar4()
     Option 4:WBP Pilgrim Star Band Object Reference=Get RadialStar5()
     Index:int=@n1
    ) -> (ReturnValue:WBP Pilgrim Star Band Object Reference)
   )
  ]}
 Handle Star Update {nodes [
   n1 = Handle Star Update!() -> (then, CurrentStar:int, StarProgress:double=~) {is=title K2Node_FunctionEntry}
   > n2 = Check Is Star Changed!(execute, self, CurrentStarIndex:int=@n1, CurrentStarProgress:double=@n1.StarProgress) -> (then, IsChanged:bool) {is=title K2Node_CallFunction}
   > Branch(Condition=@n2)
   > CheckStarCompletion!(execute, self, CurrentStar:int=@n1) -> (then) {is=title K2Node_CallFunction}
   n3 = Get MaxStars() -> (MaxStars:int=0)
   > n4 = Branch(Condition="integer <= integer"(A=int - int(A=@n3), B=@n1))
   > Update Gold Star Progress Widgets!(execute, self, CurrentStar:int=@n1, GoldStarProgress:double=@n1.StarProgress) -> (then)
   Update Current Star!(execute=@n4.else, self, CurrentStar:int=@n1, NextStarProgress:double=@n1.StarProgress) -> (then)
  ]}
 ResetCompletedVisuals {nodes [
   ResetCompletedVisuals() {is=title K2Node_FunctionEntry}
   > n1 = For Each Loop(Array:Array of WBP Pilgrim Star Band Object References=Make Array("[0]"=Get RadialStar1(), "[1]"=Get RadialStar2(), "[2]"=Get RadialStar3(), "[3]"=Get RadialStar4(), "[4]"=Get RadialStar5())) -> (Array Element:WBP Pilgrim Star Band Object Reference)
   > Reset Widget(self:WBP Pilgrim Star Band Object Reference=@n1)
  ]}
 SetupStarsResults {nodes [
   n1 = SetupStarsResults!() -> (then, StarCount:int) {is=title K2Node_FunctionEntry}
   > n2 = For Loop(LastIndex="Min (Integer)"(A=int - int(A=@n1), B=4) -> (ReturnValue=0))
   Branch(execute=@n2.Completed, Condition=Get bIsSequencedAnimation() -> (bIsSequencedAnimation:bool))
   > ResetCompletedVisuals() {is=title K2Node_CallFunction}
   n3 = Get Star Widget(execute=@n2.LoopBody, Index=@n2) {is=title K2Node_CallFunction}
   > n4 = Set bIsGoldStars(bIsGoldStars:bool=AND Boolean(A=Get AllowGoldStar() -> (AllowGoldStar:bool), B="integer > integer"(A=@n1, B=5))) -> (Output_Get:bool)
   > n5 = Branch(Condition=@n4)
   > SetGoldStarComplete(self=@n3)
   Branch(execute=@n5.else, Condition="integer > integer"(A=@n1, B=@n2))
   > SetComplete(self=@n3)
  ]}
 ShowGoldStarIndicator {nodes [
   ShowGoldStarIndicator() {is=title K2Node_FunctionEntry}
   > SetVisibility(self=Get GoldStarIndicator() -> (GoldStarIndicator:Border Object Reference), InVisibility=Visible)
  ]}
 Update Current Star {nodes [
   n1 = Update Current Star!() -> (then, CurrentStar:int, NextStarProgress:double=~) {is=title K2Node_FunctionEntry}
   > Branch(Condition="float > float"(A=@n1.NextStarProgress))
   > n2 = Get Star Widget(Index=@n1)
   > SetStarProgress!(execute, self:WBP Pilgrim Star Band Object Reference=@n2, SetStarProgress:double=@n1.NextStarProgress) -> (then)
   Reroute Node()
   Reroute Node()
  ]}
 Update Gold Star Progress Widgets {nodes [
   n1 = Update Gold Star Progress Widgets!() -> (then, CurrentStar:int, GoldStarProgress:double=~) {is=title K2Node_FunctionEntry}
   > Branch(Condition=Get AllowGoldStar())
   > n2 = For Loop(LastIndex=4)
   > n3 = Get Star Widget(Index=@n2)
   n4 = "integer >= integer"(A=@n1, B=Get MaxStars() -> (MaxStars=0))
   > n5 = Branch(Condition=@n4)
   > SetGoldStarComplete(self=@n3)
   Set Gold Star Progress!(execute=@n5.else, self:WBP Pilgrim Star Band Object Reference=@n3, Progress:double=@n1.GoldStarProgress) -> (then) {is=title K2Node_CallFunction}
   Branch(execute=@n2.Completed, Condition=@n4)
   > ShowGoldStarIndicator() {is=title K2Node_CallFunction}
  ]}
]
variables [
 AllowGoldStar {type=boolean, default=true}
 bIsGoldStars {type=boolean}
 bIsSequencedAnimation {type=boolean}
 Current Star {type=integer, default=1}
 In User Specified Scale {type=number, default=1}
 LastStarIndex {type=integer}
 LastStarProgress {type=number}
 MaxStars {type=integer, default=6}
 Scale Stretch {type=EStretch, default=ScaleToFit}
 ShouldStarsAnimateComplete {type=boolean, default=true}
 Star Count {type=integer}
 Star Progress {type=number}
 StarCompleteSound {type=SoundBase, default=/PilgrimQuickplay/Audio/Results/MSS_FM_ResultsVictory_TallyBandAppear}
 TallySound {type=MetaSoundSource, default=/PilgrimQuickplay/Audio/Results/MSS_FM_ResultsVictory_TallyBandNums, class=~, properties=~}
 TallySoundReference {type=AudioComponent, default=null}
]
widgets {
 info {inheritedWidgetCount=0, parentClass=&UserWidget, rootWidgetClass=&Overlay, widgetCount=13}
 widgets [
  StarsPanel {widgetClassPath=&Overlay, widget=StarsPanel}
  GoldStarIndicator {
   widgetClassPath=&Border
   bIsVariable=true
   parent=StarsPanel
   widget=GoldStarIndicator
   properties {
    visibility=Hidden
    background {
     bIsDynamicallyLoaded=false
     drawAs=Box
     imageType=NoImage
     mirroring=NoMirror
     resourceName=None
     resourceObject=null
     tiling=NoTile
     imageSize=<128 128>
     margin=<ltrb 1 1 1 1>
     outlineSettings {
      bUseBrushTransparency=false
      roundingType=HalfHeightRadius
      width=0
      color {colorUseRule=UseColor_Specified, specifiedColor=<rgba 0 0 0 0>}
      cornerRadii=<0 0 0 0>
     }
     tintColor {colorUseRule=UseColor_Specified, specifiedColor=<rgba 1 1 0 0>}
     uVRegion {bIsValid=false, max=<0 0>, min=<0 0>}
    }
    desiredSizeScale=<2 2>
    padding=<ltrb 0 0 0 0>
   }
   slot {class=OverlaySlot, properties {horizontalAlignment=HAlign_Fill, verticalAlignment=VAlign_Fill}}
  }
  ScaleBox {
   widgetClassPath=&ScaleBox
   bIsVariable=true
   parent=StarsPanel
   widget=ScaleBox
   slot {class=OverlaySlot, properties {horizontalAlignment=HAlign_Fill, verticalAlignment=VAlign_Fill}}
  }
  StarContainer {widgetClassPath=&HorizontalBox, bIsVariable=true, parent=ScaleBox, widget=StarContainer, slot {class=ScaleBoxSlot}}
  RadialStar1 {
   widgetClassPath=&/WBP_PilgrimStar_Band
   bIsVariable=true
   parent=StarContainer
   widget=RadialStar1
   slot {class=HorizontalBoxSlot}
  }
  Spacer_85 {widgetClassPath=&Spacer, parent=StarContainer, widget=Spacer_85, properties {size=<6 1>}, slot {class=HorizontalBoxSlot}}
  RadialStar2 {
   widgetClassPath=&WBP_PilgrimStar_Band
   bIsVariable=true
   parent=StarContainer
   widget=RadialStar2
   slot {class=HorizontalBoxSlot}
  }
  Spacer {widgetClassPath=&Spacer, parent=StarContainer, widget=Spacer, properties {size=<6 1>}, slot {class=HorizontalBoxSlot}}
  RadialStar3 {
   widgetClassPath=&WBP_PilgrimStar_Band
   bIsVariable=true
   parent=StarContainer
   widget=RadialStar3
   slot {class=HorizontalBoxSlot}
  }
  Spacer_1 {widgetClassPath=&Spacer, parent=StarContainer, widget=Spacer_1, properties {size=<6 1>}, slot {class=HorizontalBoxSlot}}
  RadialStar4 {
   widgetClassPath=&WBP_PilgrimStar_Band
   bIsVariable=true
   parent=StarContainer
   widget=RadialStar4
   slot {class=HorizontalBoxSlot}
  }
  Spacer_2 {widgetClassPath=&Spacer, parent=StarContainer, widget=Spacer_2, properties {size=<6 1>}, slot {class=HorizontalBoxSlot}}
  RadialStar5 {
   widgetClassPath=&WBP_PilgrimStar_Band
   bIsVariable=true
   parent=StarContainer
   widget=RadialStar5
   slot {class=HorizontalBoxSlot}
  }
 ]
}
