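# Quickplay band star display widget that hosts a star bar showing the band's shared star progress during a Pilgrim music battle.
# On battle initialization it shows the bar only for multiple players and disallows the gold star if any player is below Expert difficulty.
# Replicated score data feeds the band's star count and next-star progress into the bar.
# It subscribes to tag-driven visibility and passes its animate-complete setting to the bar.
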WidgetBlueprint /PilgrimQuickplay/Widgets/WBP_PilgrimQuickplay_BandStarDisplay
graphs [EventGraph {nodes [
   n1 = Event PreConstruct()
   > Parent: PreConstruct(IsDesignTime=@n1.IsDesignTime)
   > Set ShouldStarsAnimateComplete!(
    execute
    ShouldStarsAnimateComplete:bool=Get ShouldStarsAnimateComplete() -> (ShouldStarsAnimateComplete:bool)
    self:WBP Pilgrim Quickplay Band Star Bar Object Reference=Get StarMeter() -> (StarMeter:WBP Pilgrim Quickplay Band Star Bar Object Reference)
   ) -> (then, Output_Get:bool)
   Event Construct()
   > SubscribeForChangingVisibilityByTag!(
    execute
    self:Pilgrim Game Widget Base Object Reference
    GameplayTag:"Gameplay Tag Structure (by ref)"=MakeLiteralGameplayTag!(Value:Gameplay Tag Structure=(TagName="Pilgrim.Widget.BandStarDisplay")) -> (ReturnValue:Gameplay Tag Structure)
   ) -> (then)
   Event OnMusicBattleInstanceInitializeEvent() {is=title K2Node_Event}
   > Sequence()
   n2 = Break Pilgrim Music Battle Data!(PilgrimMusicBattleData:"Pilgrim Music Battle Data Structure (by ref)"=GetMusicBattleData!(self:Pilgrim Music Battle Instance Object Reference=Get MusicBattleInstance() -> (MusicBattleInstance:Pilgrim Music Battle Instance Object Reference)) -> (ReturnValue:"Pilgrim Music Battle Data Structure (by ref)")) -> (MusicBattleStates:Array of Pilgrim Music Battle Player State Structures, BandState:Pilgrim Band State Structure) {is=title K2Node_BreakStruct}
   n3 = int - int(A=Length(TargetArray:Array of Pilgrim Music Battle Player State Structures=@n2))
   > SetVisibility(self=Get StarMeter(), InVisibility=Select(Option 0=Collapsed, Option 1=SelfHitTestInvisible, Index="integer > integer"(A=@n3)))
   > n4 = For Each Loop(Array:Array of Pilgrim Music Battle Player State Structures=@n2) -> (Array Element:Pilgrim Music Battle Player State Structure)
   n5 = Break Pilgrim Music Battle Player State!(PilgrimMusicBattlePlayerState:"Pilgrim Music Battle Player State Structure (by ref)"=@n4) -> (
    PlayerState:Fort Player State Athena Object Reference
    Team:Byte
    TrackType:EPilgrimTrackType Enum
    InstrumentTags:Gameplay Tag Container Structure
    PilgrimGameProgress:Pilgrim Game Progress Structure
    ServerScoringChanges:Pilgrim Server Scoring Changes Structure
   ) {is=title K2Node_BreakStruct}
   n6 = Get MusicBattleInstance()
   > n7 = Branch(Condition="Equal (Enum)"(
     A:EPilgrimSongDifficulty Enum=GetSongDifficultyForPlayer!(self:Pilgrim Music Battle Instance Object Reference=@n6, Player:Fort Player State Athena Object Reference=@n5) -> (ReturnValue:EPilgrimSongDifficulty Enum=DifficultyEasy)
     B:EPilgrimSongDifficulty Enum=DifficultyExpert
    ))
   Set AllowGoldStar!(execute=@n7.else, AllowGoldStar:bool=false, self:WBP Pilgrim Quickplay Band Star Bar Object Reference=Get StarMeter()) -> (then, Output_Get:bool)
   n8 = Event OnScoreDataReplicated!() -> (OutputDelegate, then, MusicBattleData:"Pilgrim Music Battle Data Structure (by ref)") {is=title K2Node_Event}
   n9 = Break Pilgrim Band State!(PilgrimBandState:"Pilgrim Band State Structure (by ref)"=Break Pilgrim Music Battle Data(PilgrimMusicBattleData=@n8.MusicBattleData).BandState) -> (
    BandModifierBonus:float=~
    CurrentScoreFromOverdrive:float=~
    TotalScoreFromOverdrive:float=~
    BandStar:int
    NextBandStarProgress:float=~
    BaseBandScore:float=~
    bAnyPlayersHaveOverdriveActive:bool=~
   ) {is=title K2Node_BreakStruct}
   > Handle Star Update!(
    execute
    self:WBP Pilgrim Quickplay Band Star Bar Object Reference=Get StarMeter()
    CurrentStar:int=@n9.BandStar
    StarProgress:double=@n9.NextBandStarProgress
   ) -> (then)
  ]}]
variables [ShouldStarsAnimateComplete {type=boolean, default=true}]
widgets {
 info {inheritedWidgetCount=0, parentClass=/Script/PilgrimCoreUI.PilgrimGameWidgetBase, rootWidgetClass=&Overlay, widgetCount=2}
 widgets [
  Overlay_0 {widgetClassPath=&Overlay, widget=Overlay_0}
  StarMeter {
   widgetClassPath=&/WBP_PilgrimQuickplay_BandStarBar
   bIsVariable=true
   parent=Overlay_0
   widget=StarMeter
   properties {visibility=Collapsed}
   slot {class=OverlaySlot}
  }
 ]
}
