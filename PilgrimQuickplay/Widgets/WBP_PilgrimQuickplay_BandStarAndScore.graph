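# Pilgrim Quickplay widget that stacks the band star display above the band score in a vertical box, separated by a spacer.
# On Construct it plays Anim_Intro, a short animation that slides both elements in horizontally from a positive X offset while fading them in.
# The score's keys trail the stars' slightly, giving a staggered entrance.
# The widget holds no variables or other logic, leaving content to the two child widgets.
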WidgetBlueprint /PilgrimQuickplay/Widgets/WBP_PilgrimQuickplay_BandStarAndScore
animations [Anim_Intro {
  binding_count=2
  length_seconds=0.5
  object=&self:Anim_Intro
  sequence {bindings [
    BandStarDisplay {
     id=#1
     possessed_class=WBP_PilgrimQuickplay_BandStarDisplay_C
     tracks [
      {
       type=MovieSceneFloatTrack
       display_name=Render Opacity
       sections [{type=MovieSceneFloatSection, channels [None {type=MovieSceneScriptingFloatChannel, keys [[7,0],[13,1]]}]}]
      }
      {
       type=MovieScene2DTransformTrack
       sections [{type=MovieScene2DTransformSection, channels [Translation.X {type=MovieSceneScriptingFloatChannel, keys [[5,384],[13,0]]}]}]
      }
     ]
    }
    BandScore {
     id=#2
     possessed_class=WBP_PilgrimQuickplay_BandScore_C
     tracks [
      {
       type=MovieSceneFloatTrack
       display_name=Render Opacity
       sections [{type=MovieSceneFloatSection, channels [None {type=MovieSceneScriptingFloatChannel, keys [[9,0],[15,1]]}]}]
      }
      {
       type=MovieScene2DTransformTrack
       sections [{type=MovieScene2DTransformSection, channels [Translation.X {type=MovieSceneScriptingFloatChannel, keys [[7,384],[15,0]]}]}]
      }
     ]
    }
   ]}
 }]
graphs [EventGraph {nodes [
   Event Construct()
   > PlayAnimation(InAnimation=Get Anim_Intro() -> (Anim_Intro:Widget Animation Object Reference))
  ]}]
variables []
widgets {
 info {inheritedWidgetCount=0, parentClass=/Script/PilgrimCoreUI.PilgrimGameWidgetBase, rootWidgetClass=&VerticalBox, widgetCount=4}
 widgets [
  Box_Main {widgetClassPath=&VerticalBox, widget=Box_Main}
  BandStarDisplay {
   widgetClassPath=&/WBP_PilgrimQuickplay_BandStarDisplay
   parent=Box_Main
   widget=BandStarDisplay
   slot {class=VerticalBoxSlot}
  }
  Spacer_101 {widgetClassPath=&Spacer, parent=Box_Main, widget=Spacer_101, properties {size=<1 16>}, slot {class=VerticalBoxSlot}}
  BandScore {
   widgetClassPath=&/WBP_PilgrimQuickplay_BandScore
   parent=Box_Main
   widget=BandScore
   properties {visibility=SelfHitTestInvisible}
   slot {class=VerticalBoxSlot}
  }
 ]
}
