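# Debug HUD widget for Pilgrim music battles that shows the local player's live song position, streak, hit accuracy, timing deviation and latency offsets.
# Visibility follows the calibration controller's debug flag and also toggles through an Enhanced Input action.
# After retrying until it finds the player's performance tracker, a Delay loop refreshes text from the tracker, game and battle instance, querying only while visible.
# Battle completion resets a DoOnceThenDo gate in that loop.
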WidgetBlueprint /PilgrimQuickplay/Widgets/WBP_PilgrimQuickplay_DEBUG_SongStats
graphs [
 EventGraph {nodes [
   n1 = Music Battle Finished!() -> (OutputDelegate, then, EndReason:EMusicBattleEndReason Enum=CompletedNormally) {is=title K2Node_CustomEvent}
   n2 = Do Once Then Do!(execute:Exec=@n3.then, Reset:Exec=@n1.then, StartClosed:bool=true) -> (DoOnce:Exec, ThenDo:Exec) {is=title K2Node_MacroInstance}
   UpdateSongStats(execute=[@n7.Is Valid, @n2.ThenDo])
   > n3 = Delay(Duration=Get RefreshRate() -> (RefreshRate:double))
   n4 = OnDebugUpdated!() -> (OutputDelegate, then, CalibrationControllerComponent:FMCalibration Controller Component Object Reference, IsCalibrationDebugEnabled:bool=~) {is=title K2Node_CustomEvent}
   > SetVisibility(InVisibility=Select(Option 0=Collapsed, Option 1=HitTestInvisible, Index=@n4.IsCalibrationDebugEnabled))
   n5 = EnhancedInputAction IA_PilgrimToggleSongStats!() -> (
    Triggered:Exec
    Started:Exec
    Ongoing:Exec
    Canceled:Exec
    Completed:Exec
    ActionValue:bool
    ElapsedSeconds:double=~
    TriggeredSeconds:double=~
    InputAction:Input Action Object Reference=/PilgrimCore/Input/InputActions/IA_PilgrimToggleSongStats
   ) {is=title K2Node_EnhancedInputAction}
   ToggleVisibility(execute=@n5.Started)
   Event OnMusicBattleInstanceInitializeEvent() {is=title K2Node_Event}
   > Bind Event to On Music Battle Finished(
    self:Pilgrim Music Battle Instance Object Reference=Get MusicBattleInstance() -> (MusicBattleInstance:Pilgrim Music Battle Instance Object Reference)
    Delegate=Create Event()
   ) {is=title K2Node_AddDelegate}
   Event OnInitialized()
   > Parent: OnInitialized() {is=title K2Node_CallParentFunction}
   > Bind Event to On Calibration Debug Updated(
    self:FMCalibration Controller Component Object Reference=GetComponentByClass!(
     self:Actor Object Reference=GetOwningPlayer()
     ComponentClass:Actor Component Class Reference=/Script/FMCalibrationRuntime.FMCalibrationControllerComponent
    ) -> (ReturnValue:FMCalibration Controller Component Object Reference)
    Delegate=@n4
   ) {is=title K2Node_AddDelegate}
   > n6 = UpdateVisibility()
   FindAndSetPerformanceTrackerComponent(execute=[@n8.then, @n6.then])
   > n7 = Is Valid(InputObject=Get PerformanceTracker() -> (PerformanceTracker:BP Pilgrim Performance Tracker Component Object Reference))
   n8 = Delay(execute=@n7.Is Not Valid, Duration=0.2)
  ]}
 FindAndSetPerformanceTrackerComponent {nodes [
   FindAndSetPerformanceTrackerComponent() {is=title K2Node_FunctionEntry}
   > Set PerformanceTracker(PerformanceTracker:BP Pilgrim Performance Tracker Component Object Reference=GetComponentByClass(
     self=GetOwningPlayer()
     ComponentClass=/PilgrimCore/Player/Components/BP_PilgrimPerformanceTrackerComponent.BP_PilgrimPerformanceTrackerComponent_C
    ) -> (ReturnValue:BP Pilgrim Performance Tracker Component Object Reference)) -> (Output_Get:BP Pilgrim Performance Tracker Component Object Reference)
  ]}
 GetCurrentMeasureBeatTickText {nodes [
   GetCurrentMeasureBeatTickText() {is=title K2Node_FunctionEntry}
   > n1 = Get Pilgrim Game!(in:Exec) -> (Succeeded:Exec, Failed:Exec, PilgrimGameInstance:Pilgrim Game Object Reference) {is=title K2Node_MacroInstance}
   n2 = Break Pilgrim Song Measure Position!(PilgrimSongMeasurePosition:"Pilgrim Song Measure Position Structure (by ref)"=GetCurrentMeasureBeatTick!(self:Pilgrim Game Object Reference=@n1) -> (ReturnValue:Pilgrim Song Measure Position Structure)) -> (Measure:int, Beat:int, Tick:int) {is=title K2Node_BreakStruct}
   > Return Node!(execute, ReturnValue:Text="To Text (Integer)"(Value=@n2), ReturnValue2:Text="To Text (Integer)"(Value=@n2.Beat))
  ]}
 GetLocalPlayerIndex {nodes [
   GetLocalPlayerIndex() {is=title K2Node_FunctionEntry}
   > n1 = Get!(execute:Exec) -> (then:Exec, else:Exec, MusicBattleInstance:Pilgrim Music Battle Instance Object Reference)
   > Return Node!(
    execute
    NewParam:int=GetPlayerIndexFromController!(self:Pilgrim Music Battle Instance Object Reference=@n1, LocalPC:Player Controller Object Reference=GetOwningPlayer()) -> (ReturnValue:int=0)
   )
   Return Node(execute=@n1.else, NewParam=-1)
  ]}
 GetPilgrimGame {nodes [
   Inputs!() -> (in:Exec)
   n1 = Get MusicBattleInstance()
   > n2 = Branch(Condition=IsValid(Object=@n1))
   n3 = GetLocalPilgrimGameForPlayer!(self:Pilgrim Music Battle Instance Object Reference=@n1, InPC:Player Controller Object Reference=GetOwningPlayer()) -> (ReturnValue:Pilgrim Game Object Reference)
   > n4 = Branch(Condition=IsValid(Object=@n3))
   > Outputs!(Succeeded:Exec, Failed:Exec=[@n4.else, @n2.else], PilgrimGameInstance:Pilgrim Game Object Reference=@n3)
  ]}
 SetAudioLatencyMs {nodes [
   n1 = SetAudioLatencyMs!() -> (then, PilgrimGame:Pilgrim Game Object Reference) {is=title K2Node_FunctionEntry}
   > Is Valid(InputObject=@n1)
   > n2 = GetMeasuredVideoToAudioRenderOffsetMs!(execute) -> (then, ReturnValue:float)
   > "SetText (Text)"(
    self=Get AVOffsetMsText() -> (AVOffsetMsText:Common Text Object Reference)
    InText=Format Text!(Format="LOCTEXT(\"{value}{ending}\")", value:float=@n2, ending:Text=Get EndingMilliseconds() -> (EndingMilliseconds:Text)) -> (Result)
   )
  ]}
 SetAverageText {nodes [
   n1 = SetAverageText!() -> (then, PilgrimEventPerformanceDataResult:"Pilgrim Event Scoring Performance Data Result Structure (by ref)") {is=title K2Node_FunctionEntry}
   n2 = Break Pilgrim Event Scoring Performance Data Result!(PilgrimEvent_Scoring_PerformanceDataResult:"Pilgrim Event Scoring Performance Data Result Structure (by ref)"=@n1) -> (
    ForPlayer:Fort Player Pawn Object Reference
    TotalOffset:double=~
    TotalSamples:int
    TotalGems:int
    TotalHits:int
    TotalMisses:int
    TotalPasses:int
    TotalNotes:int
    MaxValue:float=~
    MinValue:float=~
    StandardDeviation:float=~
    AverageOffset:float=~
    MaxStreak:int
    TimeInOverdriveMs:float=~
    Histogram:Pilgrim Performance Tracker Histogram Structure
    AccuracyTierCounts:Array of Integers
   ) {is=title K2Node_BreakStruct}
   > "SetText (Text)"(
    self=Get AvgText() -> (AvgText:Common Text Object Reference)
    InText=Format Text(
     Format="LOCTEXT(\"{value}{ending}\")"
     value:Text="To Text (Float)"(Value=@n2.AverageOffset, MinimumIntegralDigits=1, MaximumIntegralDigits=324, MinimumFractionalDigits=0, MaximumFractionalDigits=2)
     ending=Get EndingMilliseconds()
    )
   )
  ]}
 SetCurrentAccuracyText {nodes [
   n1 = SetCurrentAccuracyText!() -> (then, PilgrimEventPerformanceDataResult:"Pilgrim Event Scoring Performance Data Result Structure (by ref)") {is=title K2Node_FunctionEntry}
   n2 = Break Pilgrim Event Scoring Performance Data Result(PilgrimEvent_Scoring_PerformanceDataResult=@n1) -> (TotalOffset=~, MaxValue=~, MinValue=~, StandardDeviation=~, AverageOffset=~, TimeInOverdriveMs=~)
   n3 = float * float(A=SafeDivide(A="To Float (Integer)"(InInt=@n2.TotalHits), B="To Float (Integer)"(InInt=@n2.TotalGems)), B=100) {is=title K2Node_PromotableOperator}
   > "SetText (Text)"(
    self=Get AccuracyText() -> (AccuracyText:Common Text Object Reference)
    InText=Format Text(
     Format="LOCTEXT(\"{value}{ending}\")"
     value="To Text (Float)"(Value=@n3, MinimumIntegralDigits=1, MaximumIntegralDigits=324, MinimumFractionalDigits=0, MaximumFractionalDigits=0)
     ending=Get EndingPercent() -> (EndingPercent:Text)
    )
   )
  ]}
 SetCurrentGemsText {nodes [
   n1 = SetCurrentGemsText!() -> (then, PilgrimEventPerformanceDataResult:"Pilgrim Event Scoring Performance Data Result Structure (by ref)") {is=title K2Node_FunctionEntry}
   > Set Text(
    Text="To Text (Integer)"(Value=Break Pilgrim Event Scoring Performance Data Result(PilgrimEvent_Scoring_PerformanceDataResult=@n1) -> (TotalOffset=~, MaxValue=~, MinValue=~, StandardDeviation=~, AverageOffset=~, TimeInOverdriveMs=~).TotalGems)
    self=Get GemsText() -> (GemsText:Common Text Object Reference)
   )
  ]}
 SetCurrentHitsText {nodes [
   n1 = SetCurrentHitsText!() -> (then, PilgrimEventPerformanceDataResult:"Pilgrim Event Scoring Performance Data Result Structure (by ref)") {is=title K2Node_FunctionEntry}
   > Set Text(
    Text="To Text (Integer)"(Value=Break Pilgrim Event Scoring Performance Data Result(PilgrimEvent_Scoring_PerformanceDataResult=@n1) -> (TotalOffset=~, MaxValue=~, MinValue=~, StandardDeviation=~, AverageOffset=~, TimeInOverdriveMs=~).TotalHits)
    self=Get HitsText() -> (HitsText:Common Text Object Reference)
   )
  ]}
 SetCurrentSongMsText {nodes [
   n1 = SetCurrentSongMsText!() -> (then, PilgrimGame:Pilgrim Game Object Reference) {is=title K2Node_FunctionEntry}
   n2 = GetMusicClock!(self:Pilgrim Game Object Reference=@n1) -> (ReturnValue:Music Clock Object Reference)
   > Is Valid(InputObject=@n2)
   n3 = float * float(
    A=Break MIDI Song Position!(MidiSongPos:"MIDI Song Position Structure (by ref)"=GetSongPos!(self:Music Clock Object Reference=@n2, Timebase:ECalibratedMusicTimebase Enum=AudioRenderTime) -> (ReturnValue:"MIDI Song Position Structure (by ref)")) -> (
     SecondsFromBarOne:float=~
     SecondsIncludingCountIn:float=~
     TimeSigNumerator:int
     TimeSigDenominator:int
     Tempo:float=~
     BarsIncludingCountIn:float=~
     BeatsIncludingCountIn:float=~
     BeatType:EMusicalBeatType Enum
     Timestamp:Music Timestamp Structure
    ) {is=title K2Node_BreakStruct}
    B=1000
   )
   > "SetText (Text)"(
    self=Get MsText() -> (MsText:Common Text Object Reference)
    InText=Format Text(
     Format="LOCTEXT(\"{value}{ending}\")"
     value="To Text (Float)"(Value=@n3, MinimumIntegralDigits=1, MaximumIntegralDigits=324, MinimumFractionalDigits=0, MaximumFractionalDigits=3)
     ending=Get EndingMilliseconds()
    )
   )
  ]}
 SetCurrentStreakText {nodes [
   SetCurrentStreakText() {is=title K2Node_FunctionEntry}
   n1 = GetMusicBattleData!(self:Pilgrim Music Battle Instance Object Reference=Get MusicBattleInstance()) -> (ReturnValue:"Pilgrim Music Battle Data Structure (by ref)")
   n2 = Break Pilgrim Music Battle Player State!(PilgrimMusicBattlePlayerState:"Pilgrim Music Battle Player State Structure (by ref)"="Get (a ref)"(
     Array:Array of Pilgrim Music Battle Player State Structures=Break Pilgrim Music Battle Data!(PilgrimMusicBattleData:"Pilgrim Music Battle Data Structure (by ref)"=@n1) -> (MusicBattleStates:Array of Pilgrim Music Battle Player State Structures, BandState:Pilgrim Band State Structure) {is=title K2Node_BreakStruct}
     Dimension 1=GetLocalPlayerIndex!() -> (NewParam:int=0) {is=title K2Node_CallFunction}
    ) -> (Output:"Pilgrim Music Battle Player State Structure (by ref)")) -> (
    PlayerState:Fort Player State Athena Object Reference
    Team:Byte
    TrackType:EPilgrimTrackType Enum
    InstrumentTags:Gameplay Tag Container Structure
    PilgrimGameProgress:Pilgrim Game Progress Structure
    ServerScoringChanges:Pilgrim Server Scoring Changes Structure
   ) {is=title K2Node_BreakStruct}
   > "SetText (Text)"(
    self=Get CurrentStreakText() -> (CurrentStreakText:Common Text Object Reference)
    InText="To Text (Integer)"(Value=Break Pilgrim Game Progress!(PilgrimGameProgress:"Pilgrim Game Progress Structure (by ref)"=@n2.PilgrimGameProgress) -> (
      NumHits:int
      NumMisses:int
      NumStreakBreakingMisses:int
      NumPasses:int
      bHasFullCombo:bool=~
      bHasPerfectCombo:bool=~
      CurrentTick:int
      NoteStreak:int
      CurrentScore:float=~
      RawCurrentScore:float=~
      CurrentModifierBonus:float=~
      CurrentStar:int
      NextStarProgress:float=~
      StreakScoreMultiplier:int
      NotesPastLastMultiplier:int
      NotesToNextMultiplier:int
      CurrentHealth:float=~
      IsOverdriveActive:bool=~
      CurrentOverdrive:float=~
      CurrentOverdriveSectionNum:int
      BandScoreContribution:float=~
      BandModifierBonusContribution:float=~
      OverdriveScoringState:EOverdriveScoringState Enum
      MaxScore:float=~
      RawMaxScore:float=~
      MaxStreakLength:int
      AverageStreakMultiplierPerHitOrPass:float=~
      bIsAtMaxStreak:bool=~
      CurrentSectionCompletionPercentage:float=~
     ) {is=title K2Node_BreakStruct}.NoteStreak)
   )
  ]}
 SetMaxStreakText {nodes [
   n1 = SetMaxStreakText!() -> (then, PilgrimEventPerformanceDataResult:"Pilgrim Event Scoring Performance Data Result Structure (by ref)") {is=title K2Node_FunctionEntry}
   > "SetText (Text)"(
    self=Get MaxStreakText() -> (MaxStreakText:Common Text Object Reference)
    InText="To Text (Integer)"(Value=Break Pilgrim Event Scoring Performance Data Result(PilgrimEvent_Scoring_PerformanceDataResult=@n1) -> (TotalOffset=~, MaxValue=~, MinValue=~, StandardDeviation=~, AverageOffset=~, TimeInOverdriveMs=~).MaxStreak)
   )
  ]}
 SetMeasureBeatTickTexts {nodes [
   SetMeasureBeatTickTexts() {is=title K2Node_FunctionEntry}
   > n1 = Sequence()
   n2 = GetCurrentMeasureBeatTickText!() -> (ReturnValue:Text, ReturnValue2:Text) {is=title K2Node_CallFunction}
   "SetText (Text)"(execute=@n1.then_1, self=Get MeasureBeatText() -> (MeasureBeatText:Common Text Object Reference), InText=@n2.ReturnValue2)
   "SetText (Text)"(execute=@n1.then_0, self=Get MeasureText() -> (MeasureText:Common Text Object Reference), InText=@n2)
  ]}
 SetStandardDeviationText {nodes [
   n1 = SetStandardDeviationText!() -> (then, PilgrimEventPerformanceDataResult:"Pilgrim Event Scoring Performance Data Result Structure (by ref)") {is=title K2Node_FunctionEntry}
   n2 = Break Pilgrim Event Scoring Performance Data Result(PilgrimEvent_Scoring_PerformanceDataResult=@n1) -> (TotalOffset=~, MaxValue=~, MinValue=~, StandardDeviation=~, AverageOffset=~, TimeInOverdriveMs=~)
   > "SetText (Text)"(
    self=Get StndDvText() -> (StndDvText:Common Text Object Reference)
    InText=Format Text(
     Format="LOCTEXT(\"{value}{ending}\")"
     value="To Text (Float)"(Value=@n2.StandardDeviation, MinimumIntegralDigits=1, MaximumIntegralDigits=324, MinimumFractionalDigits=0, MaximumFractionalDigits=2)
     ending=Get EndingMilliseconds()
    )
   )
  ]}
 SetVideoLatencyMs {nodes [
   n1 = SetVideoLatencyMs!() -> (then, PilgrimGame:Pilgrim Game Object Reference) {is=title K2Node_FunctionEntry}
   > Is Valid(InputObject=@n1)
   > n2 = GetMeasuredUserExperienceAndReactionToAudioRenderOffsetMs!(execute) -> (then, ReturnValue:float)
   > "SetText (Text)"(
    self=Get InputOffsetMsText() -> (InputOffsetMsText:Common Text Object Reference)
    InText=Format Text(Format="LOCTEXT(\"{value}{ending}\")", value:float=@n2, ending=Get EndingMilliseconds())
   )
  ]}
 ToggleVisibility {nodes [
   ToggleVisibility() {is=title K2Node_FunctionEntry}
   > SetVisibility(InVisibility=Select(Option 0=HitTestInvisible, Option 1=Collapsed, Index=IsVisible()))
  ]}
 UpdatePerformanceTrackerInformation {nodes [
   UpdatePerformanceTrackerInformation() {is=title K2Node_FunctionEntry}
   > Is Valid(InputObject=Get PerformanceTracker())
   > Set ResultData(ResultData:Pilgrim Event Scoring Performance Data Result Structure=GetResultData!(self:Pilgrim Performance Tracker Component Object Reference=Get PerformanceTracker()) -> (ReturnValue:Pilgrim Event Scoring Performance Data Result Structure)) -> (Output_Get:Pilgrim Event Scoring Performance Data Result Structure)
   > SetAverageText!(
    execute
    self
    PilgrimEventPerformanceDataResult:"Pilgrim Event Scoring Performance Data Result Structure (by ref)"=Get ResultData() -> (ResultData:Pilgrim Event Scoring Performance Data Result Structure)
   ) -> (then) {is=title K2Node_CallFunction}
   > SetStandardDeviationText!(execute, self, PilgrimEventPerformanceDataResult:"Pilgrim Event Scoring Performance Data Result Structure (by ref)"=Get ResultData()) -> (then) {is=title K2Node_CallFunction}
   > SetCurrentGemsText!(execute, self, PilgrimEventPerformanceDataResult:"Pilgrim Event Scoring Performance Data Result Structure (by ref)"=Get ResultData()) -> (then) {is=title K2Node_CallFunction}
   > SetCurrentHitsText!(execute, self, PilgrimEventPerformanceDataResult:"Pilgrim Event Scoring Performance Data Result Structure (by ref)"=Get ResultData()) -> (then) {is=title K2Node_CallFunction}
   > SetCurrentAccuracyText!(execute, self, PilgrimEventPerformanceDataResult:"Pilgrim Event Scoring Performance Data Result Structure (by ref)"=Get ResultData()) -> (then) {is=title K2Node_CallFunction}
   > SetMaxStreakText!(execute, self, PilgrimEventPerformanceDataResult:"Pilgrim Event Scoring Performance Data Result Structure (by ref)"=Get ResultData()) -> (then) {is=title K2Node_CallFunction}
   > n1 = Get Pilgrim Game()
   > Set PilgrimGameInstance(PilgrimGameInstance:Pilgrim Game Object Reference=@n1) -> (Output_Get:Pilgrim Game Object Reference)
   > SetAudioLatencyMs!(execute, self, PilgrimGame:Pilgrim Game Object Reference=Get PilgrimGameInstance() -> (PilgrimGameInstance:Pilgrim Game Object Reference)) -> (then) {is=title K2Node_CallFunction}
   > SetVideoLatencyMs!(execute, self, PilgrimGame:Pilgrim Game Object Reference=Get PilgrimGameInstance()) -> (then) {is=title K2Node_CallFunction}
  ]}
 UpdatePilgrimGameInformation {nodes [
   UpdatePilgrimGameInformation() {is=title K2Node_FunctionEntry}
   > Is Valid(InputObject=Get MusicBattleInstance())
   > n1 = Get Pilgrim Game()
   > Is Valid(InputObject=@n1)
   > SetCurrentSongMsText!(execute, self, PilgrimGame:Pilgrim Game Object Reference=@n1) -> (then) {is=title K2Node_CallFunction}
   > SetMeasureBeatTickTexts() {is=title K2Node_CallFunction}
   > SetCurrentStreakText() {is=title K2Node_CallFunction}
  ]}
 UpdateSongStats {nodes [
   UpdateSongStats() {is=title K2Node_FunctionEntry}
   > Branch(Condition=IsVisible())
   > UpdatePilgrimGameInformation() {is=title K2Node_CallFunction}
   > UpdatePerformanceTrackerInformation() {is=title K2Node_CallFunction}
  ]}
 UpdateVisibility {nodes [
   UpdateVisibility() {is=title K2Node_FunctionEntry}
   n1 = GetComponentByClass(self=GetOwningPlayer(), ComponentClass=&FMCalibrationControllerComponent) -> (ReturnValue:FMCalibration Controller Component Object Reference)
   > SetVisibility(InVisibility=Select(
     Option 0=Collapsed
     Option 1=HitTestInvisible
     Index=GetIsCalibrationDebugEnabled!(self:FMCalibration Controller Component Object Reference=@n1) -> (ReturnValue:bool)
    ))
   > Return Node!(execute)
  ]}
]
variables [
 EndingMilliseconds {type=string, default=" ms"}
 EndingPercent {type=string, default=%}
 EndingTick {type=string, default=" tick"}
 PerformanceTracker {type=BP_PilgrimPerformanceTrackerComponent_C, default=null}
 RefreshRate {type=number, default=0.1}
]
widgets {
 info {inheritedWidgetCount=0, parentClass=/Script/PilgrimCoreUI.PilgrimGameWidgetBase, rootWidgetClass=&CanvasPanel, widgetCount=38}
 widgets [
  CanvasPanel_30 {widgetClassPath=&CanvasPanel, widget=CanvasPanel_30}
  HorizontalBox_0 {
   widgetClassPath=&HorizontalBox
   parent=CanvasPanel_30
   widget=HorizontalBox_0
   slot {
    class=CanvasPanelSlot
    properties {bAutoSize=true, layoutData {alignment=<0 0>, anchors {maximum=<0.1 0.1767>, minimum=<0.1 0.1767>}, offsets=<ltrb -0.1922 0.9518 100 80>}}
   }
  }
  HorizontalBox_4 {widgetClassPath=&HorizontalBox, parent=HorizontalBox_0, widget=HorizontalBox_4, slot {class=HorizontalBoxSlot}}
  VerticalBox_0 {widgetClassPath=&VerticalBox, parent=HorizontalBox_4, widget=VerticalBox_0, slot {class=HorizontalBoxSlot}}
  CommonTextBlock_8 {
   widgetClassPath=&CommonTextBlock
   parent=VerticalBox_0
   widget=CommonTextBlock_8
   properties {
    style=/FortUILibrary/Styles/Typography/_TextStyle_TagExtraSmall._TextStyle_TagExtraSmall_C
    text=Measure:
    font {
     bForceMonospaced=false
     bMaterialIsStencil=false
     fontMaterial=null
     fontObject=&Burbank
     letterSpacing=0
     monospacedWidth=1
     size=12
     skewAmount=0
     typefaceFontName=Burbank-PlayerPower
     outlineSettings {
      bApplyOutlineToDropShadows=false
      bMiteredCorners=false
      bSeparateFillAlpha=false
      outlineMaterial=null
      outlineSize=0
      outlineColor=<rgba 0 0 0 1>
     }
    }
    shadowOffset=<0 0>
    strikeBrush {
     bIsDynamicallyLoaded=false
     drawAs=NoDrawType
     imageType=NoImage
     mirroring=NoMirror
     resourceName=None
     resourceObject=null
     tiling=NoTile
     imageSize=<32 32>
     margin=<ltrb 0 0 0 0>
     outlineSettings {
      bUseBrushTransparency=false
      roundingType=HalfHeightRadius
      width=0
      color {colorUseRule=UseColor_Specified, specifiedColor=<rgba 0 0 0 0>}
      cornerRadii=<0 0 0 0>
     }
     tintColor {colorUseRule=UseColor_Specified, specifiedColor=<rgba 1 1 1 1>}
     uVRegion {bIsValid=false, max=<0 0>, min=<0 0>}
    }
   }
   slot {class=VerticalBoxSlot, properties {horizontalAlignment=HAlign_Right}}
  }
  CommonTextBlock_9 {
   widgetClassPath=&CommonTextBlock
   parent=VerticalBox_0
   widget=CommonTextBlock_9
   properties {
    style=&_TextStyle_TagExtraSmall_C
    text=Beat:
    font {
     bForceMonospaced=false
     bMaterialIsStencil=false
     fontMaterial=null
     fontObject=&Burbank
     letterSpacing=0
     monospacedWidth=1
     size=12
     skewAmount=0
     typefaceFontName=Burbank-PlayerPower
     outlineSettings {
      bApplyOutlineToDropShadows=false
      bMiteredCorners=false
      bSeparateFillAlpha=false
      outlineMaterial=null
      outlineSize=0
      outlineColor=<rgba 0 0 0 1>
     }
    }
    shadowOffset=<0 0>
    strikeBrush {
     bIsDynamicallyLoaded=false
     drawAs=NoDrawType
     imageType=NoImage
     mirroring=NoMirror
     resourceName=None
     resourceObject=null
     tiling=NoTile
     imageSize=<32 32>
     margin=<ltrb 0 0 0 0>
     outlineSettings {
      bUseBrushTransparency=false
      roundingType=HalfHeightRadius
      width=0
      color {colorUseRule=UseColor_Specified, specifiedColor=<rgba 0 0 0 0>}
      cornerRadii=<0 0 0 0>
     }
     tintColor {colorUseRule=UseColor_Specified, specifiedColor=<rgba 1 1 1 1>}
     uVRegion {bIsValid=false, max=<0 0>, min=<0 0>}
    }
   }
   slot {class=VerticalBoxSlot, properties {horizontalAlignment=HAlign_Right}}
  }
  CommonTextBlock_11 {
   widgetClassPath=&CommonTextBlock
   parent=VerticalBox_0
   widget=CommonTextBlock_11
   properties {
    style=&_TextStyle_TagExtraSmall_C
    text=---
    font {
     bForceMonospaced=false
     bMaterialIsStencil=false
     fontMaterial=null
     fontObject=&Burbank
     letterSpacing=0
     monospacedWidth=1
     size=12
     skewAmount=0
     typefaceFontName=Burbank-PlayerPower
     outlineSettings {
      bApplyOutlineToDropShadows=false
      bMiteredCorners=false
      bSeparateFillAlpha=false
      outlineMaterial=null
      outlineSize=0
      outlineColor=<rgba 0 0 0 1>
     }
    }
    shadowOffset=<0 0>
    strikeBrush {
     bIsDynamicallyLoaded=false
     drawAs=NoDrawType
     imageType=NoImage
     mirroring=NoMirror
     resourceName=None
     resourceObject=null
     tiling=NoTile
     imageSize=<32 32>
     margin=<ltrb 0 0 0 0>
     outlineSettings {
      bUseBrushTransparency=false
      roundingType=HalfHeightRadius
      width=0
      color {colorUseRule=UseColor_Specified, specifiedColor=<rgba 0 0 0 0>}
      cornerRadii=<0 0 0 0>
     }
     tintColor {colorUseRule=UseColor_Specified, specifiedColor=<rgba 1 1 1 1>}
     uVRegion {bIsValid=false, max=<0 0>, min=<0 0>}
    }
   }
   slot {class=VerticalBoxSlot, properties {horizontalAlignment=HAlign_Right}}
  }
  CommonTextBlock_3 {
   widgetClassPath=&CommonTextBlock
   parent=VerticalBox_0
   widget=CommonTextBlock_3
   properties {
    style=&_TextStyle_TagExtraSmall_C
    text=Song MS:
    font {
     bForceMonospaced=false
     bMaterialIsStencil=false
     fontMaterial=null
     fontObject=&Burbank
     letterSpacing=0
     monospacedWidth=1
     size=12
     skewAmount=0
     typefaceFontName=Burbank-PlayerPower
     outlineSettings {
      bApplyOutlineToDropShadows=false
      bMiteredCorners=false
      bSeparateFillAlpha=false
      outlineMaterial=null
      outlineSize=0
      outlineColor=<rgba 0 0 0 1>
     }
    }
    shadowOffset=<0 0>
    strikeBrush {
     bIsDynamicallyLoaded=false
     drawAs=NoDrawType
     imageType=NoImage
     mirroring=NoMirror
     resourceName=None
     resourceObject=null
     tiling=NoTile
     imageSize=<32 32>
     margin=<ltrb 0 0 0 0>
     outlineSettings {
      bUseBrushTransparency=false
      roundingType=HalfHeightRadius
      width=0
      color {colorUseRule=UseColor_Specified, specifiedColor=<rgba 0 0 0 0>}
      cornerRadii=<0 0 0 0>
     }
     tintColor {colorUseRule=UseColor_Specified, specifiedColor=<rgba 1 1 1 1>}
     uVRegion {bIsValid=false, max=<0 0>, min=<0 0>}
    }
   }
   slot {class=VerticalBoxSlot, properties {horizontalAlignment=HAlign_Right}}
  }
  CommonTextBlock_16 {
   widgetClassPath=&CommonTextBlock
   parent=VerticalBox_0
   widget=CommonTextBlock_16
   properties {
    style=&_TextStyle_TagExtraSmall_C
    text=A/V Offset:
    font {
     bForceMonospaced=false
     bMaterialIsStencil=false
     fontMaterial=null
     fontObject=&Burbank
     letterSpacing=0
     monospacedWidth=1
     size=12
     skewAmount=0
     typefaceFontName=Burbank-PlayerPower
     outlineSettings {
      bApplyOutlineToDropShadows=false
      bMiteredCorners=false
      bSeparateFillAlpha=false
      outlineMaterial=null
      outlineSize=0
      outlineColor=<rgba 0 0 0 1>
     }
    }
    shadowOffset=<0 0>
    strikeBrush {
     bIsDynamicallyLoaded=false
     drawAs=NoDrawType
     imageType=NoImage
     mirroring=NoMirror
     resourceName=None
     resourceObject=null
     tiling=NoTile
     imageSize=<32 32>
     margin=<ltrb 0 0 0 0>
     outlineSettings {
      bUseBrushTransparency=false
      roundingType=HalfHeightRadius
      width=0
      color {colorUseRule=UseColor_Specified, specifiedColor=<rgba 0 0 0 0>}
      cornerRadii=<0 0 0 0>
     }
     tintColor {colorUseRule=UseColor_Specified, specifiedColor=<rgba 1 1 1 1>}
     uVRegion {bIsValid=false, max=<0 0>, min=<0 0>}
    }
   }
   slot {class=VerticalBoxSlot, properties {horizontalAlignment=HAlign_Right}}
  }
  CommonTextBlock_17 {
   widgetClassPath=&CommonTextBlock
   parent=VerticalBox_0
   widget=CommonTextBlock_17
   properties {
    style=&_TextStyle_TagExtraSmall_C
    text=Input Offset:
    font {
     bForceMonospaced=false
     bMaterialIsStencil=false
     fontMaterial=null
     fontObject=&Burbank
     letterSpacing=0
     monospacedWidth=1
     size=12
     skewAmount=0
     typefaceFontName=Burbank-PlayerPower
     outlineSettings {
      bApplyOutlineToDropShadows=false
      bMiteredCorners=false
      bSeparateFillAlpha=false
      outlineMaterial=null
      outlineSize=0
      outlineColor=<rgba 0 0 0 1>
     }
    }
    shadowOffset=<0 0>
    strikeBrush {
     bIsDynamicallyLoaded=false
     drawAs=NoDrawType
     imageType=NoImage
     mirroring=NoMirror
     resourceName=None
     resourceObject=null
     tiling=NoTile
     imageSize=<32 32>
     margin=<ltrb 0 0 0 0>
     outlineSettings {
      bUseBrushTransparency=false
      roundingType=HalfHeightRadius
      width=0
      color {colorUseRule=UseColor_Specified, specifiedColor=<rgba 0 0 0 0>}
      cornerRadii=<0 0 0 0>
     }
     tintColor {colorUseRule=UseColor_Specified, specifiedColor=<rgba 1 1 1 1>}
     uVRegion {bIsValid=false, max=<0 0>, min=<0 0>}
    }
   }
   slot {class=VerticalBoxSlot, properties {horizontalAlignment=HAlign_Right}}
  }
  CommonTextBlock {
   widgetClassPath=&CommonTextBlock
   parent=VerticalBox_0
   widget=CommonTextBlock
   properties {
    style=&_TextStyle_TagExtraSmall_C
    text=Avg:
    font {
     bForceMonospaced=false
     bMaterialIsStencil=false
     fontMaterial=null
     fontObject=&Burbank
     letterSpacing=0
     monospacedWidth=1
     size=12
     skewAmount=0
     typefaceFontName=Burbank-PlayerPower
     outlineSettings {
      bApplyOutlineToDropShadows=false
      bMiteredCorners=false
      bSeparateFillAlpha=false
      outlineMaterial=null
      outlineSize=0
      outlineColor=<rgba 0 0 0 1>
     }
    }
    shadowOffset=<0 0>
    strikeBrush {
     bIsDynamicallyLoaded=false
     drawAs=NoDrawType
     imageType=NoImage
     mirroring=NoMirror
     resourceName=None
     resourceObject=null
     tiling=NoTile
     imageSize=<32 32>
     margin=<ltrb 0 0 0 0>
     outlineSettings {
      bUseBrushTransparency=false
      roundingType=HalfHeightRadius
      width=0
      color {colorUseRule=UseColor_Specified, specifiedColor=<rgba 0 0 0 0>}
      cornerRadii=<0 0 0 0>
     }
     tintColor {colorUseRule=UseColor_Specified, specifiedColor=<rgba 1 1 1 1>}
     uVRegion {bIsValid=false, max=<0 0>, min=<0 0>}
    }
   }
   slot {class=VerticalBoxSlot, properties {horizontalAlignment=HAlign_Right}}
  }
  CommonTextBlock_1 {
   widgetClassPath=&CommonTextBlock
   parent=VerticalBox_0
   widget=CommonTextBlock_1
   properties {
    style=&_TextStyle_TagExtraSmall_C
    text=StdDv:
    font {
     bForceMonospaced=false
     bMaterialIsStencil=false
     fontMaterial=null
     fontObject=&Burbank
     letterSpacing=0
     monospacedWidth=1
     size=12
     skewAmount=0
     typefaceFontName=Burbank-PlayerPower
     outlineSettings {
      bApplyOutlineToDropShadows=false
      bMiteredCorners=false
      bSeparateFillAlpha=false
      outlineMaterial=null
      outlineSize=0
      outlineColor=<rgba 0 0 0 1>
     }
    }
    shadowOffset=<0 0>
    strikeBrush {
     bIsDynamicallyLoaded=false
     drawAs=NoDrawType
     imageType=NoImage
     mirroring=NoMirror
     resourceName=None
     resourceObject=null
     tiling=NoTile
     imageSize=<32 32>
     margin=<ltrb 0 0 0 0>
     outlineSettings {
      bUseBrushTransparency=false
      roundingType=HalfHeightRadius
      width=0
      color {colorUseRule=UseColor_Specified, specifiedColor=<rgba 0 0 0 0>}
      cornerRadii=<0 0 0 0>
     }
     tintColor {colorUseRule=UseColor_Specified, specifiedColor=<rgba 1 1 1 1>}
     uVRegion {bIsValid=false, max=<0 0>, min=<0 0>}
    }
   }
   slot {class=VerticalBoxSlot, properties {horizontalAlignment=HAlign_Right}}
  }
  CommonTextBlock_13 {
   widgetClassPath=&CommonTextBlock
   parent=VerticalBox_0
   widget=CommonTextBlock_13
   properties {
    style=&_TextStyle_TagExtraSmall_C
    text=---
    font {
     bForceMonospaced=false
     bMaterialIsStencil=false
     fontMaterial=null
     fontObject=&Burbank
     letterSpacing=0
     monospacedWidth=1
     size=12
     skewAmount=0
     typefaceFontName=Burbank-PlayerPower
     outlineSettings {
      bApplyOutlineToDropShadows=false
      bMiteredCorners=false
      bSeparateFillAlpha=false
      outlineMaterial=null
      outlineSize=0
      outlineColor=<rgba 0 0 0 1>
     }
    }
    shadowOffset=<0 0>
    strikeBrush {
     bIsDynamicallyLoaded=false
     drawAs=NoDrawType
     imageType=NoImage
     mirroring=NoMirror
     resourceName=None
     resourceObject=null
     tiling=NoTile
     imageSize=<32 32>
     margin=<ltrb 0 0 0 0>
     outlineSettings {
      bUseBrushTransparency=false
      roundingType=HalfHeightRadius
      width=0
      color {colorUseRule=UseColor_Specified, specifiedColor=<rgba 0 0 0 0>}
      cornerRadii=<0 0 0 0>
     }
     tintColor {colorUseRule=UseColor_Specified, specifiedColor=<rgba 1 1 1 1>}
     uVRegion {bIsValid=false, max=<0 0>, min=<0 0>}
    }
   }
   slot {class=VerticalBoxSlot, properties {horizontalAlignment=HAlign_Right}}
  }
  CommonTextBlock_4 {
   widgetClassPath=&CommonTextBlock
   parent=VerticalBox_0
   widget=CommonTextBlock_4
   properties {
    style=&_TextStyle_TagExtraSmall_C
    text=Hits:
    font {
     bForceMonospaced=false
     bMaterialIsStencil=false
     fontMaterial=null
     fontObject=&Burbank
     letterSpacing=0
     monospacedWidth=1
     size=12
     skewAmount=0
     typefaceFontName=Burbank-PlayerPower
     outlineSettings {
      bApplyOutlineToDropShadows=false
      bMiteredCorners=false
      bSeparateFillAlpha=false
      outlineMaterial=null
      outlineSize=0
      outlineColor=<rgba 0 0 0 1>
     }
    }
    shadowOffset=<0 0>
    strikeBrush {
     bIsDynamicallyLoaded=false
     drawAs=NoDrawType
     imageType=NoImage
     mirroring=NoMirror
     resourceName=None
     resourceObject=null
     tiling=NoTile
     imageSize=<32 32>
     margin=<ltrb 0 0 0 0>
     outlineSettings {
      bUseBrushTransparency=false
      roundingType=HalfHeightRadius
      width=0
      color {colorUseRule=UseColor_Specified, specifiedColor=<rgba 0 0 0 0>}
      cornerRadii=<0 0 0 0>
     }
     tintColor {colorUseRule=UseColor_Specified, specifiedColor=<rgba 1 1 1 1>}
     uVRegion {bIsValid=false, max=<0 0>, min=<0 0>}
    }
   }
   slot {class=VerticalBoxSlot, properties {horizontalAlignment=HAlign_Right}}
  }
  CommonTextBlock_5 {
   widgetClassPath=&CommonTextBlock
   parent=VerticalBox_0
   widget=CommonTextBlock_5
   properties {
    style=&_TextStyle_TagExtraSmall_C
    text=Gems:
    font {
     bForceMonospaced=false
     bMaterialIsStencil=false
     fontMaterial=null
     fontObject=&Burbank
     letterSpacing=0
     monospacedWidth=1
     size=12
     skewAmount=0
     typefaceFontName=Burbank-PlayerPower
     outlineSettings {
      bApplyOutlineToDropShadows=false
      bMiteredCorners=false
      bSeparateFillAlpha=false
      outlineMaterial=null
      outlineSize=0
      outlineColor=<rgba 0 0 0 1>
     }
    }
    shadowOffset=<0 0>
    strikeBrush {
     bIsDynamicallyLoaded=false
     drawAs=NoDrawType
     imageType=NoImage
     mirroring=NoMirror
     resourceName=None
     resourceObject=null
     tiling=NoTile
     imageSize=<32 32>
     margin=<ltrb 0 0 0 0>
     outlineSettings {
      bUseBrushTransparency=false
      roundingType=HalfHeightRadius
      width=0
      color {colorUseRule=UseColor_Specified, specifiedColor=<rgba 0 0 0 0>}
      cornerRadii=<0 0 0 0>
     }
     tintColor {colorUseRule=UseColor_Specified, specifiedColor=<rgba 1 1 1 1>}
     uVRegion {bIsValid=false, max=<0 0>, min=<0 0>}
    }
   }
   slot {class=VerticalBoxSlot, properties {horizontalAlignment=HAlign_Right}}
  }
  CommonTextBlock_6 {
   widgetClassPath=&CommonTextBlock
   parent=VerticalBox_0
   widget=CommonTextBlock_6
   properties {
    style=&_TextStyle_TagExtraSmall_C
    text=Acc:
    font {
     bForceMonospaced=false
     bMaterialIsStencil=false
     fontMaterial=null
     fontObject=&Burbank
     letterSpacing=0
     monospacedWidth=1
     size=12
     skewAmount=0
     typefaceFontName=Burbank-PlayerPower
     outlineSettings {
      bApplyOutlineToDropShadows=false
      bMiteredCorners=false
      bSeparateFillAlpha=false
      outlineMaterial=null
      outlineSize=0
      outlineColor=<rgba 0 0 0 1>
     }
    }
    shadowOffset=<0 0>
    strikeBrush {
     bIsDynamicallyLoaded=false
     drawAs=NoDrawType
     imageType=NoImage
     mirroring=NoMirror
     resourceName=None
     resourceObject=null
     tiling=NoTile
     imageSize=<32 32>
     margin=<ltrb 0 0 0 0>
     outlineSettings {
      bUseBrushTransparency=false
      roundingType=HalfHeightRadius
      width=0
      color {colorUseRule=UseColor_Specified, specifiedColor=<rgba 0 0 0 0>}
      cornerRadii=<0 0 0 0>
     }
     tintColor {colorUseRule=UseColor_Specified, specifiedColor=<rgba 1 1 1 1>}
     uVRegion {bIsValid=false, max=<0 0>, min=<0 0>}
    }
   }
   slot {class=VerticalBoxSlot, properties {horizontalAlignment=HAlign_Right}}
  }
  CommonTextBlock_7 {
   widgetClassPath=&CommonTextBlock
   parent=VerticalBox_0
   widget=CommonTextBlock_7
   properties {
    style=&_TextStyle_TagExtraSmall_C
    text=Max Streak:
    font {
     bForceMonospaced=false
     bMaterialIsStencil=false
     fontMaterial=null
     fontObject=&Burbank
     letterSpacing=0
     monospacedWidth=1
     size=12
     skewAmount=0
     typefaceFontName=Burbank-PlayerPower
     outlineSettings {
      bApplyOutlineToDropShadows=false
      bMiteredCorners=false
      bSeparateFillAlpha=false
      outlineMaterial=null
      outlineSize=0
      outlineColor=<rgba 0 0 0 1>
     }
    }
    shadowOffset=<0 0>
    strikeBrush {
     bIsDynamicallyLoaded=false
     drawAs=NoDrawType
     imageType=NoImage
     mirroring=NoMirror
     resourceName=None
     resourceObject=null
     tiling=NoTile
     imageSize=<32 32>
     margin=<ltrb 0 0 0 0>
     outlineSettings {
      bUseBrushTransparency=false
      roundingType=HalfHeightRadius
      width=0
      color {colorUseRule=UseColor_Specified, specifiedColor=<rgba 0 0 0 0>}
      cornerRadii=<0 0 0 0>
     }
     tintColor {colorUseRule=UseColor_Specified, specifiedColor=<rgba 1 1 1 1>}
     uVRegion {bIsValid=false, max=<0 0>, min=<0 0>}
    }
   }
   slot {class=VerticalBoxSlot, properties {horizontalAlignment=HAlign_Right}}
  }
  CommonTextBlock_18 {
   widgetClassPath=&CommonTextBlock
   parent=VerticalBox_0
   widget=CommonTextBlock_18
   properties {
    style=&_TextStyle_TagExtraSmall_C
    text=Cur Streak:
    font {
     bForceMonospaced=false
     bMaterialIsStencil=false
     fontMaterial=null
     fontObject=&Burbank
     letterSpacing=0
     monospacedWidth=1
     size=12
     skewAmount=0
     typefaceFontName=Burbank-PlayerPower
     outlineSettings {
      bApplyOutlineToDropShadows=false
      bMiteredCorners=false
      bSeparateFillAlpha=false
      outlineMaterial=null
      outlineSize=0
      outlineColor=<rgba 0 0 0 1>
     }
    }
    shadowOffset=<0 0>
    strikeBrush {
     bIsDynamicallyLoaded=false
     drawAs=NoDrawType
     imageType=NoImage
     mirroring=NoMirror
     resourceName=None
     resourceObject=null
     tiling=NoTile
     imageSize=<32 32>
     margin=<ltrb 0 0 0 0>
     outlineSettings {
      bUseBrushTransparency=false
      roundingType=HalfHeightRadius
      width=0
      color {colorUseRule=UseColor_Specified, specifiedColor=<rgba 0 0 0 0>}
      cornerRadii=<0 0 0 0>
     }
     tintColor {colorUseRule=UseColor_Specified, specifiedColor=<rgba 1 1 1 1>}
     uVRegion {bIsValid=false, max=<0 0>, min=<0 0>}
    }
   }
   slot {class=VerticalBoxSlot, properties {horizontalAlignment=HAlign_Right}}
  }
  Spacer_63 {widgetClassPath=&Spacer, parent=VerticalBox_0, widget=Spacer_63, properties {size=<60 1>}, slot {class=VerticalBoxSlot}}
  HorizontalBox_1 {widgetClassPath=&HorizontalBox, parent=HorizontalBox_0, widget=HorizontalBox_1, slot {class=HorizontalBoxSlot}}
  Spacer_1 {widgetClassPath=&Spacer, parent=HorizontalBox_1, widget=Spacer_1, properties {size=<8 1>}, slot {class=HorizontalBoxSlot}}
  HorizontalBox_5 {widgetClassPath=&HorizontalBox, parent=HorizontalBox_0, widget=HorizontalBox_5, slot {class=HorizontalBoxSlot}}
  VerticalBox_1 {
   widgetClassPath=&VerticalBox
   parent=HorizontalBox_5
   widget=VerticalBox_1
   slot {class=HorizontalBoxSlot, properties {horizontalAlignment=HAlign_Right, size {sizeRule=Fill, value=1}}}
  }
  MeasureText {
   widgetClassPath=&CommonTextBlock
   bIsVariable=true
   parent=VerticalBox_1
   widget=MeasureText
   properties {
    minDesiredWidth=120
    style=&_TextStyle_TagExtraSmall_C
    text=--
    font {
     bForceMonospaced=false
     bMaterialIsStencil=false
     fontMaterial=null
     fontObject=&Burbank
     letterSpacing=0
     monospacedWidth=1
     size=12
     skewAmount=0
     typefaceFontName=Burbank-PlayerPower
     outlineSettings {
      bApplyOutlineToDropShadows=false
      bMiteredCorners=false
      bSeparateFillAlpha=false
      outlineMaterial=null
      outlineSize=0
      outlineColor=<rgba 0 0 0 1>
     }
    }
    shadowOffset=<0 0>
    strikeBrush {
     bIsDynamicallyLoaded=false
     drawAs=NoDrawType
     imageType=NoImage
     mirroring=NoMirror
     resourceName=None
     resourceObject=null
     tiling=NoTile
     imageSize=<32 32>
     margin=<ltrb 0 0 0 0>
     outlineSettings {
      bUseBrushTransparency=false
      roundingType=HalfHeightRadius
      width=0
      color {colorUseRule=UseColor_Specified, specifiedColor=<rgba 0 0 0 0>}
      cornerRadii=<0 0 0 0>
     }
     tintColor {colorUseRule=UseColor_Specified, specifiedColor=<rgba 1 1 1 1>}
     uVRegion {bIsValid=false, max=<0 0>, min=<0 0>}
    }
   }
   slot {class=VerticalBoxSlot, properties {horizontalAlignment=HAlign_Left}}
  }
  MeasureBeatText {
   widgetClassPath=&CommonTextBlock
   bIsVariable=true
   parent=VerticalBox_1
   widget=MeasureBeatText
   properties {
    minDesiredWidth=120
    style=&_TextStyle_TagExtraSmall_C
    text=--
    font {
     bForceMonospaced=false
     bMaterialIsStencil=false
     fontMaterial=null
     fontObject=&Burbank
     letterSpacing=0
     monospacedWidth=1
     size=12
     skewAmount=0
     typefaceFontName=Burbank-PlayerPower
     outlineSettings {
      bApplyOutlineToDropShadows=false
      bMiteredCorners=false
      bSeparateFillAlpha=false
      outlineMaterial=null
      outlineSize=0
      outlineColor=<rgba 0 0 0 1>
     }
    }
    shadowOffset=<0 0>
    strikeBrush {
     bIsDynamicallyLoaded=false
     drawAs=NoDrawType
     imageType=NoImage
     mirroring=NoMirror
     resourceName=None
     resourceObject=null
     tiling=NoTile
     imageSize=<32 32>
     margin=<ltrb 0 0 0 0>
     outlineSettings {
      bUseBrushTransparency=false
      roundingType=HalfHeightRadius
      width=0
      color {colorUseRule=UseColor_Specified, specifiedColor=<rgba 0 0 0 0>}
      cornerRadii=<0 0 0 0>
     }
     tintColor {colorUseRule=UseColor_Specified, specifiedColor=<rgba 1 1 1 1>}
     uVRegion {bIsValid=false, max=<0 0>, min=<0 0>}
    }
   }
   slot {class=VerticalBoxSlot, properties {horizontalAlignment=HAlign_Left}}
  }
  CommonTextBlock_12 {
   widgetClassPath=&CommonTextBlock
   parent=VerticalBox_1
   widget=CommonTextBlock_12
   properties {
    minDesiredWidth=120
    style=&_TextStyle_TagExtraSmall_C
    font {
     bForceMonospaced=false
     bMaterialIsStencil=false
     fontMaterial=null
     fontObject=&Burbank
     letterSpacing=0
     monospacedWidth=1
     size=12
     skewAmount=0
     typefaceFontName=Burbank-PlayerPower
     outlineSettings {
      bApplyOutlineToDropShadows=false
      bMiteredCorners=false
      bSeparateFillAlpha=false
      outlineMaterial=null
      outlineSize=0
      outlineColor=<rgba 0 0 0 1>
     }
    }
    shadowOffset=<0 0>
    strikeBrush {
     bIsDynamicallyLoaded=false
     drawAs=NoDrawType
     imageType=NoImage
     mirroring=NoMirror
     resourceName=None
     resourceObject=null
     tiling=NoTile
     imageSize=<32 32>
     margin=<ltrb 0 0 0 0>
     outlineSettings {
      bUseBrushTransparency=false
      roundingType=HalfHeightRadius
      width=0
      color {colorUseRule=UseColor_Specified, specifiedColor=<rgba 0 0 0 0>}
      cornerRadii=<0 0 0 0>
     }
     tintColor {colorUseRule=UseColor_Specified, specifiedColor=<rgba 1 1 1 1>}
     uVRegion {bIsValid=false, max=<0 0>, min=<0 0>}
    }
   }
   slot {class=VerticalBoxSlot, properties {horizontalAlignment=HAlign_Right}}
  }
  MsText {
   widgetClassPath=&CommonTextBlock
   bIsVariable=true
   parent=VerticalBox_1
   widget=MsText
   properties {
    minDesiredWidth=120
    style=&_TextStyle_TagExtraSmall_C
    text=--
    font {
     bForceMonospaced=false
     bMaterialIsStencil=false
     fontMaterial=null
     fontObject=&Burbank
     letterSpacing=0
     monospacedWidth=1
     size=12
     skewAmount=0
     typefaceFontName=Burbank-PlayerPower
     outlineSettings {
      bApplyOutlineToDropShadows=false
      bMiteredCorners=false
      bSeparateFillAlpha=false
      outlineMaterial=null
      outlineSize=0
      outlineColor=<rgba 0 0 0 1>
     }
    }
    shadowOffset=<0 0>
    strikeBrush {
     bIsDynamicallyLoaded=false
     drawAs=NoDrawType
     imageType=NoImage
     mirroring=NoMirror
     resourceName=None
     resourceObject=null
     tiling=NoTile
     imageSize=<32 32>
     margin=<ltrb 0 0 0 0>
     outlineSettings {
      bUseBrushTransparency=false
      roundingType=HalfHeightRadius
      width=0
      color {colorUseRule=UseColor_Specified, specifiedColor=<rgba 0 0 0 0>}
      cornerRadii=<0 0 0 0>
     }
     tintColor {colorUseRule=UseColor_Specified, specifiedColor=<rgba 1 1 1 1>}
     uVRegion {bIsValid=false, max=<0 0>, min=<0 0>}
    }
   }
   slot {class=VerticalBoxSlot, properties {horizontalAlignment=HAlign_Left, size {sizeRule=Fill, value=1}}}
  }
  AVOffsetMsText {
   widgetClassPath=&CommonTextBlock
   bIsVariable=true
   parent=VerticalBox_1
   widget=AVOffsetMsText
   properties {
    minDesiredWidth=120
    style=&_TextStyle_TagExtraSmall_C
    text=--
    font {
     bForceMonospaced=false
     bMaterialIsStencil=false
     fontMaterial=null
     fontObject=&Burbank
     letterSpacing=0
     monospacedWidth=1
     size=12
     skewAmount=0
     typefaceFontName=Burbank-PlayerPower
     outlineSettings {
      bApplyOutlineToDropShadows=false
      bMiteredCorners=false
      bSeparateFillAlpha=false
      outlineMaterial=null
      outlineSize=0
      outlineColor=<rgba 0 0 0 1>
     }
    }
    shadowOffset=<0 0>
    strikeBrush {
     bIsDynamicallyLoaded=false
     drawAs=NoDrawType
     imageType=NoImage
     mirroring=NoMirror
     resourceName=None
     resourceObject=null
     tiling=NoTile
     imageSize=<32 32>
     margin=<ltrb 0 0 0 0>
     outlineSettings {
      bUseBrushTransparency=false
      roundingType=HalfHeightRadius
      width=0
      color {colorUseRule=UseColor_Specified, specifiedColor=<rgba 0 0 0 0>}
      cornerRadii=<0 0 0 0>
     }
     tintColor {colorUseRule=UseColor_Specified, specifiedColor=<rgba 1 1 1 1>}
     uVRegion {bIsValid=false, max=<0 0>, min=<0 0>}
    }
   }
   slot {class=VerticalBoxSlot, properties {horizontalAlignment=HAlign_Right}}
  }
  InputOffsetMsText {
   widgetClassPath=&CommonTextBlock
   bIsVariable=true
   parent=VerticalBox_1
   widget=InputOffsetMsText
   properties {
    minDesiredWidth=120
    style=&_TextStyle_TagExtraSmall_C
    text=--
    font {
     bForceMonospaced=false
     bMaterialIsStencil=false
     fontMaterial=null
     fontObject=&Burbank
     letterSpacing=0
     monospacedWidth=1
     size=12
     skewAmount=0
     typefaceFontName=Burbank-PlayerPower
     outlineSettings {
      bApplyOutlineToDropShadows=false
      bMiteredCorners=false
      bSeparateFillAlpha=false
      outlineMaterial=null
      outlineSize=0
      outlineColor=<rgba 0 0 0 1>
     }
    }
    shadowOffset=<0 0>
    strikeBrush {
     bIsDynamicallyLoaded=false
     drawAs=NoDrawType
     imageType=NoImage
     mirroring=NoMirror
     resourceName=None
     resourceObject=null
     tiling=NoTile
     imageSize=<32 32>
     margin=<ltrb 0 0 0 0>
     outlineSettings {
      bUseBrushTransparency=false
      roundingType=HalfHeightRadius
      width=0
      color {colorUseRule=UseColor_Specified, specifiedColor=<rgba 0 0 0 0>}
      cornerRadii=<0 0 0 0>
     }
     tintColor {colorUseRule=UseColor_Specified, specifiedColor=<rgba 1 1 1 1>}
     uVRegion {bIsValid=false, max=<0 0>, min=<0 0>}
    }
   }
   slot {class=VerticalBoxSlot, properties {horizontalAlignment=HAlign_Right}}
  }
  AvgText {
   widgetClassPath=&CommonTextBlock
   bIsVariable=true
   parent=VerticalBox_1
   widget=AvgText
   properties {
    minDesiredWidth=120
    style=&_TextStyle_TagExtraSmall_C
    text=--
    font {
     bForceMonospaced=false
     bMaterialIsStencil=false
     fontMaterial=null
     fontObject=&Burbank
     letterSpacing=0
     monospacedWidth=1
     size=12
     skewAmount=0
     typefaceFontName=Burbank-PlayerPower
     outlineSettings {
      bApplyOutlineToDropShadows=false
      bMiteredCorners=false
      bSeparateFillAlpha=false
      outlineMaterial=null
      outlineSize=0
      outlineColor=<rgba 0 0 0 1>
     }
    }
    shadowOffset=<0 0>
    strikeBrush {
     bIsDynamicallyLoaded=false
     drawAs=NoDrawType
     imageType=NoImage
     mirroring=NoMirror
     resourceName=None
     resourceObject=null
     tiling=NoTile
     imageSize=<32 32>
     margin=<ltrb 0 0 0 0>
     outlineSettings {
      bUseBrushTransparency=false
      roundingType=HalfHeightRadius
      width=0
      color {colorUseRule=UseColor_Specified, specifiedColor=<rgba 0 0 0 0>}
      cornerRadii=<0 0 0 0>
     }
     tintColor {colorUseRule=UseColor_Specified, specifiedColor=<rgba 1 1 1 1>}
     uVRegion {bIsValid=false, max=<0 0>, min=<0 0>}
    }
   }
   slot {class=VerticalBoxSlot, properties {horizontalAlignment=HAlign_Left, size {sizeRule=Fill, value=1}}}
  }
  StndDvText {
   widgetClassPath=&CommonTextBlock
   bIsVariable=true
   parent=VerticalBox_1
   widget=StndDvText
   properties {
    minDesiredWidth=120
    style=&_TextStyle_TagExtraSmall_C
    text=--
    font {
     bForceMonospaced=false
     bMaterialIsStencil=false
     fontMaterial=null
     fontObject=&Burbank
     letterSpacing=0
     monospacedWidth=1
     size=12
     skewAmount=0
     typefaceFontName=Burbank-PlayerPower
     outlineSettings {
      bApplyOutlineToDropShadows=false
      bMiteredCorners=false
      bSeparateFillAlpha=false
      outlineMaterial=null
      outlineSize=0
      outlineColor=<rgba 0 0 0 1>
     }
    }
    shadowOffset=<0 0>
    strikeBrush {
     bIsDynamicallyLoaded=false
     drawAs=NoDrawType
     imageType=NoImage
     mirroring=NoMirror
     resourceName=None
     resourceObject=null
     tiling=NoTile
     imageSize=<32 32>
     margin=<ltrb 0 0 0 0>
     outlineSettings {
      bUseBrushTransparency=false
      roundingType=HalfHeightRadius
      width=0
      color {colorUseRule=UseColor_Specified, specifiedColor=<rgba 0 0 0 0>}
      cornerRadii=<0 0 0 0>
     }
     tintColor {colorUseRule=UseColor_Specified, specifiedColor=<rgba 1 1 1 1>}
     uVRegion {bIsValid=false, max=<0 0>, min=<0 0>}
    }
   }
   slot {class=VerticalBoxSlot, properties {horizontalAlignment=HAlign_Left, size {sizeRule=Fill, value=1}}}
  }
  CommonTextBlock_14 {
   widgetClassPath=&CommonTextBlock
   parent=VerticalBox_1
   widget=CommonTextBlock_14
   properties {
    minDesiredWidth=120
    style=&_TextStyle_TagExtraSmall_C
    font {
     bForceMonospaced=false
     bMaterialIsStencil=false
     fontMaterial=null
     fontObject=&Burbank
     letterSpacing=0
     monospacedWidth=1
     size=12
     skewAmount=0
     typefaceFontName=Burbank-PlayerPower
     outlineSettings {
      bApplyOutlineToDropShadows=false
      bMiteredCorners=false
      bSeparateFillAlpha=false
      outlineMaterial=null
      outlineSize=0
      outlineColor=<rgba 0 0 0 1>
     }
    }
    shadowOffset=<0 0>
    strikeBrush {
     bIsDynamicallyLoaded=false
     drawAs=NoDrawType
     imageType=NoImage
     mirroring=NoMirror
     resourceName=None
     resourceObject=null
     tiling=NoTile
     imageSize=<32 32>
     margin=<ltrb 0 0 0 0>
     outlineSettings {
      bUseBrushTransparency=false
      roundingType=HalfHeightRadius
      width=0
      color {colorUseRule=UseColor_Specified, specifiedColor=<rgba 0 0 0 0>}
      cornerRadii=<0 0 0 0>
     }
     tintColor {colorUseRule=UseColor_Specified, specifiedColor=<rgba 1 1 1 1>}
     uVRegion {bIsValid=false, max=<0 0>, min=<0 0>}
    }
   }
   slot {class=VerticalBoxSlot, properties {horizontalAlignment=HAlign_Right}}
  }
  HitsText {
   widgetClassPath=&CommonTextBlock
   bIsVariable=true
   parent=VerticalBox_1
   widget=HitsText
   properties {
    minDesiredWidth=120
    style=&_TextStyle_TagExtraSmall_C
    text=--
    font {
     bForceMonospaced=false
     bMaterialIsStencil=false
     fontMaterial=null
     fontObject=&Burbank
     letterSpacing=0
     monospacedWidth=1
     size=12
     skewAmount=0
     typefaceFontName=Burbank-PlayerPower
     outlineSettings {
      bApplyOutlineToDropShadows=false
      bMiteredCorners=false
      bSeparateFillAlpha=false
      outlineMaterial=null
      outlineSize=0
      outlineColor=<rgba 0 0 0 1>
     }
    }
    shadowOffset=<0 0>
    strikeBrush {
     bIsDynamicallyLoaded=false
     drawAs=NoDrawType
     imageType=NoImage
     mirroring=NoMirror
     resourceName=None
     resourceObject=null
     tiling=NoTile
     imageSize=<32 32>
     margin=<ltrb 0 0 0 0>
     outlineSettings {
      bUseBrushTransparency=false
      roundingType=HalfHeightRadius
      width=0
      color {colorUseRule=UseColor_Specified, specifiedColor=<rgba 0 0 0 0>}
      cornerRadii=<0 0 0 0>
     }
     tintColor {colorUseRule=UseColor_Specified, specifiedColor=<rgba 1 1 1 1>}
     uVRegion {bIsValid=false, max=<0 0>, min=<0 0>}
    }
   }
   slot {class=VerticalBoxSlot, properties {horizontalAlignment=HAlign_Left, size {sizeRule=Fill, value=1}}}
  }
  GemsText {
   widgetClassPath=&CommonTextBlock
   bIsVariable=true
   parent=VerticalBox_1
   widget=GemsText
   properties {
    minDesiredWidth=120
    style=&_TextStyle_TagExtraSmall_C
    text=--
    font {
     bForceMonospaced=false
     bMaterialIsStencil=false
     fontMaterial=null
     fontObject=&Burbank
     letterSpacing=0
     monospacedWidth=1
     size=12
     skewAmount=0
     typefaceFontName=Burbank-PlayerPower
     outlineSettings {
      bApplyOutlineToDropShadows=false
      bMiteredCorners=false
      bSeparateFillAlpha=false
      outlineMaterial=null
      outlineSize=0
      outlineColor=<rgba 0 0 0 1>
     }
    }
    shadowOffset=<0 0>
    strikeBrush {
     bIsDynamicallyLoaded=false
     drawAs=NoDrawType
     imageType=NoImage
     mirroring=NoMirror
     resourceName=None
     resourceObject=null
     tiling=NoTile
     imageSize=<32 32>
     margin=<ltrb 0 0 0 0>
     outlineSettings {
      bUseBrushTransparency=false
      roundingType=HalfHeightRadius
      width=0
      color {colorUseRule=UseColor_Specified, specifiedColor=<rgba 0 0 0 0>}
      cornerRadii=<0 0 0 0>
     }
     tintColor {colorUseRule=UseColor_Specified, specifiedColor=<rgba 1 1 1 1>}
     uVRegion {bIsValid=false, max=<0 0>, min=<0 0>}
    }
   }
   slot {class=VerticalBoxSlot, properties {horizontalAlignment=HAlign_Left, size {sizeRule=Fill, value=1}}}
  }
  AccuracyText {
   widgetClassPath=&CommonTextBlock
   bIsVariable=true
   parent=VerticalBox_1
   widget=AccuracyText
   properties {
    minDesiredWidth=120
    style=&_TextStyle_TagExtraSmall_C
    text=--
    font {
     bForceMonospaced=false
     bMaterialIsStencil=false
     fontMaterial=null
     fontObject=&Burbank
     letterSpacing=0
     monospacedWidth=1
     size=12
     skewAmount=0
     typefaceFontName=Burbank-PlayerPower
     outlineSettings {
      bApplyOutlineToDropShadows=false
      bMiteredCorners=false
      bSeparateFillAlpha=false
      outlineMaterial=null
      outlineSize=0
      outlineColor=<rgba 0 0 0 1>
     }
    }
    shadowOffset=<0 0>
    strikeBrush {
     bIsDynamicallyLoaded=false
     drawAs=NoDrawType
     imageType=NoImage
     mirroring=NoMirror
     resourceName=None
     resourceObject=null
     tiling=NoTile
     imageSize=<32 32>
     margin=<ltrb 0 0 0 0>
     outlineSettings {
      bUseBrushTransparency=false
      roundingType=HalfHeightRadius
      width=0
      color {colorUseRule=UseColor_Specified, specifiedColor=<rgba 0 0 0 0>}
      cornerRadii=<0 0 0 0>
     }
     tintColor {colorUseRule=UseColor_Specified, specifiedColor=<rgba 1 1 1 1>}
     uVRegion {bIsValid=false, max=<0 0>, min=<0 0>}
    }
   }
   slot {class=VerticalBoxSlot, properties {horizontalAlignment=HAlign_Left, size {sizeRule=Fill, value=1}}}
  }
  MaxStreakText {
   widgetClassPath=&CommonTextBlock
   bIsVariable=true
   parent=VerticalBox_1
   widget=MaxStreakText
   properties {
    minDesiredWidth=120
    style=&_TextStyle_TagExtraSmall_C
    text=--
    font {
     bForceMonospaced=false
     bMaterialIsStencil=false
     fontMaterial=null
     fontObject=&Burbank
     letterSpacing=0
     monospacedWidth=1
     size=12
     skewAmount=0
     typefaceFontName=Burbank-PlayerPower
     outlineSettings {
      bApplyOutlineToDropShadows=false
      bMiteredCorners=false
      bSeparateFillAlpha=false
      outlineMaterial=null
      outlineSize=0
      outlineColor=<rgba 0 0 0 1>
     }
    }
    shadowOffset=<0 0>
    strikeBrush {
     bIsDynamicallyLoaded=false
     drawAs=NoDrawType
     imageType=NoImage
     mirroring=NoMirror
     resourceName=None
     resourceObject=null
     tiling=NoTile
     imageSize=<32 32>
     margin=<ltrb 0 0 0 0>
     outlineSettings {
      bUseBrushTransparency=false
      roundingType=HalfHeightRadius
      width=0
      color {colorUseRule=UseColor_Specified, specifiedColor=<rgba 0 0 0 0>}
      cornerRadii=<0 0 0 0>
     }
     tintColor {colorUseRule=UseColor_Specified, specifiedColor=<rgba 1 1 1 1>}
     uVRegion {bIsValid=false, max=<0 0>, min=<0 0>}
    }
   }
   slot {class=VerticalBoxSlot, properties {horizontalAlignment=HAlign_Left, size {sizeRule=Fill, value=1}}}
  }
  CurrentStreakText {
   widgetClassPath=&CommonTextBlock
   bIsVariable=true
   parent=VerticalBox_1
   widget=CurrentStreakText
   properties {
    minDesiredWidth=120
    style=&_TextStyle_TagExtraSmall_C
    text=--
    font {
     bForceMonospaced=false
     bMaterialIsStencil=false
     fontMaterial=null
     fontObject=&Burbank
     letterSpacing=0
     monospacedWidth=1
     size=12
     skewAmount=0
     typefaceFontName=Burbank-PlayerPower
     outlineSettings {
      bApplyOutlineToDropShadows=false
      bMiteredCorners=false
      bSeparateFillAlpha=false
      outlineMaterial=null
      outlineSize=0
      outlineColor=<rgba 0 0 0 1>
     }
    }
    shadowOffset=<0 0>
    strikeBrush {
     bIsDynamicallyLoaded=false
     drawAs=NoDrawType
     imageType=NoImage
     mirroring=NoMirror
     resourceName=None
     resourceObject=null
     tiling=NoTile
     imageSize=<32 32>
     margin=<ltrb 0 0 0 0>
     outlineSettings {
      bUseBrushTransparency=false
      roundingType=HalfHeightRadius
      width=0
      color {colorUseRule=UseColor_Specified, specifiedColor=<rgba 0 0 0 0>}
      cornerRadii=<0 0 0 0>
     }
     tintColor {colorUseRule=UseColor_Specified, specifiedColor=<rgba 1 1 1 1>}
     uVRegion {bIsValid=false, max=<0 0>, min=<0 0>}
    }
   }
   slot {class=VerticalBoxSlot, properties {horizontalAlignment=HAlign_Left, size {sizeRule=Fill, value=1}}}
  }
  Spacer {widgetClassPath=&Spacer, parent=VerticalBox_1, widget=Spacer, properties {size=<60 1>}, slot {class=VerticalBoxSlot}}
 ]
}
