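# Pilgrim Quickplay debug widget that shows the latest calculated input calibration offset in four text blocks, one per track lane.
# On init it binds to the owning player's FMCalibrationControllerComponent, so visibility follows the calibration debug flag, and an Enhanced Input action can also toggle it.
# Once the event router is available it listens for InputOffsetCalculated events, writing gem index and offset into the matching lane's text.
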WidgetBlueprint /PilgrimQuickplay/Widgets/WBP_PilgrimQuickplay_DebugCalibrationOffsets
graphs [
 EventGraph {nodes [
   n1 = OnDebugUpdated!() -> (OutputDelegate, then, CalibrationControllerComponent:FMCalibration Controller Component Object Reference, IsCalibrationDebugEnabled:bool=~) {is=title K2Node_CustomEvent}
   > SetVisibility(InVisibility=Select(Option 0=Collapsed, Option 1=HitTestInvisible, Index=@n1.IsCalibrationDebugEnabled))
   n2 = EnhancedInputAction IA_PilgrimDebug_ToggleCalibrationLanes!() -> (
    Triggered:Exec
    Started:Exec
    Ongoing:Exec
    Canceled:Exec
    Completed:Exec
    ActionValue:bool
    ElapsedSeconds:double=~
    TriggeredSeconds:double=~
    InputAction:Input Action Object Reference=/PilgrimQuickplay/Widgets/DebugTools/PanelDebugStats/IA_PilgrimDebug_ToggleCalibrationLanes
   ) {is=title K2Node_EnhancedInputAction}
   ToggleVisibility(execute=@n2.Started)
   Event OnInitialized()
   > Bind Event to On Calibration Debug Updated(
    self:FMCalibration Controller Component Object Reference=GetComponentByClass!(
     self:Actor Object Reference=GetOwningPlayer()
     ComponentClass:Actor Component Class Reference=/Script/FMCalibrationRuntime.FMCalibrationControllerComponent
    ) -> (ReturnValue:FMCalibration Controller Component Object Reference)
    Delegate=@n1
   ) {is=title K2Node_AddDelegate}
   > SetInitialVisibility()
   > n3 = PilgrimWaitForEventRouter!(execute:Exec, ContextActor:Actor Object Reference=GetOwningPlayer()) -> (then:Exec, OnEventRouterAvailable:Exec, EventRouter:Gameplay Event Router Component Object Reference) {is=title K2Node_AsyncAction}
   n4 = StartListeningToEvent!(
    execute:Exec=@n3.OnEventRouterAvailable
    Target:Gameplay Event Router Component Object Reference=@n3
    EventType:Script Struct Object Reference=/Script/PilgrimCoreRuntime.PilgrimEvent_InputOffsetCalculated
    Context:Object Reference
    EventBubblingRule:EEventBubblingRule Enum
    None:%=@n3
   ) -> (
    then:Exec
    AsyncTaskProxy:Async Action Start Listening To Event Object Reference
    OnEventReceived:Exec
    Payload:Pilgrim Event Input Offset Calculated Structure
   )
   > Add(
    TargetArray:Array of Async Action Start Listening To Event Object References=Get GameplayEventHandles() -> (GameplayEventHandles:Array of Async Action Start Listening To Event Object References)
    NewItem:"Async Action Start Listening To Event Object Reference (by ref)"=@n4
   ) {is=title K2Node_CallArrayFunction}
   n5 = Break Pilgrim Event Input Offset Calculated!(PilgrimEvent_InputOffsetCalculated:"Pilgrim Event Input Offset Calculated Structure (by ref)"=@n4.Payload) -> (Lane:EPilgrimTrackLane Enum, Offset:float=~, TrackerGemIndex:int, GemInfo:Track Item Info Structure) {is=title K2Node_BreakStruct}
   n6 = Select!(
    TrackLane0:Common Text Object Reference=Get Lane1Offset() -> (Lane1Offset:Common Text Object Reference)
    TrackLane1:Common Text Object Reference=Get Lane2Offset() -> (Lane2Offset:Common Text Object Reference)
    TrackLane2:Common Text Object Reference=Get Lane3Offset() -> (Lane3Offset:Common Text Object Reference)
    TrackLane3:Common Text Object Reference=Get Lane4Offset() -> (Lane4Offset:Common Text Object Reference)
    TrackLane4:Common Text Object Reference
    TrackLane5:Common Text Object Reference
    Index:EPilgrimTrackLane Enum=@n5
   ) -> (ReturnValue:Common Text Object Reference)
   Is Valid(exec=@n4.OnEventReceived, InputObject=@n6)
   n7 = "To String (Integer)"(InInt=@n5.TrackerGemIndex)
   n8 = "To String (Text)"(InText="To Text (Float)"(Value=@n5.Offset, MinimumIntegralDigits=1, MaximumIntegralDigits=324, MinimumFractionalDigits=0, MaximumFractionalDigits=1))
   > "SetText (Text)"(self=@n6, InText="To Text (String)"(InString=Append!(A="G: ", B=@n7, C:String="  O: ", D:String=@n8) -> (ReturnValue)))
   Event Construct()
   Event Tick()
  ]}
 SetInitialVisibility {nodes [
   SetInitialVisibility() {is=title K2Node_FunctionEntry}
   n1 = GetComponentByClass(self=GetOwningPlayer(), ComponentClass=&FMCalibrationControllerComponent)
   > SetVisibility(InVisibility=Select!(
     Option 0:ESlateVisibility Enum=Collapsed
     Option 1:ESlateVisibility Enum=HitTestInvisible
     Index:bool=GetIsCalibrationDebugEnabled!(self:FMCalibration Controller Component Object Reference=@n1) -> (ReturnValue:bool)
    ) -> (ReturnValue:ESlateVisibility Enum))
   > Return Node()
  ]}
 ToggleVisibility {nodes [
   ToggleVisibility() {is=title K2Node_FunctionEntry}
   > SetVisibility(InVisibility=Select(Option 0=SelfHitTestInvisible, Option 1=Collapsed, Index=IsVisible()))
  ]}
]
variables [
 isPlayerRoaming {type=boolean}
 LabelGemMap {default {}}
]
widgets {
 info {inheritedWidgetCount=0, parentClass=/Script/PilgrimCoreUI.PilgrimGameWidgetBase, rootWidgetClass=&Overlay, widgetCount=10}
 widgets [
  Overlay_0 {widgetClassPath=&Overlay, widget=Overlay_0}
  SizeBox_153 {
   widgetClassPath=&SizeBox
   parent=Overlay_0
   widget=SizeBox_153
   properties {bOverride_HeightOverride=true, bOverride_WidthOverride=true, heightOverride=200, widthOverride=800}
   slot {class=OverlaySlot, properties {horizontalAlignment=HAlign_Fill, verticalAlignment=VAlign_Fill}}
  }
  HorizontalBox {
   widgetClassPath=&HorizontalBox
   parent=SizeBox_153
   widget=HorizontalBox
   slot {class=SizeBoxSlot, properties {verticalAlignment=VAlign_Bottom}}
  }
  Lane1Offset {
   widgetClassPath=&CommonTextBlock
   bIsVariable=true
   parent=HorizontalBox
   widget=Lane1Offset
   properties {
    style=&TextStyle-Base_C
    text=0
    font {
     bForceMonospaced=false
     bMaterialIsStencil=false
     fontMaterial=null
     fontObject=&NotoSans
     letterSpacing=0
     monospacedWidth=1
     size=24
     skewAmount=0
     typefaceFontName=NotoSans-Regular
     outlineSettings {
      bApplyOutlineToDropShadows=false
      bMiteredCorners=false
      bSeparateFillAlpha=false
      outlineMaterial=null
      outlineSize=0
      outlineColor=<rgba 0 0 0 1>
     }
    }
    shadowOffset=<0 0>
   }
   slot {class=HorizontalBoxSlot}
  }
  Spacer_63 {
   widgetClassPath=&Spacer
   parent=HorizontalBox
   widget=Spacer_63
   slot {class=HorizontalBoxSlot, properties {size {sizeRule=Fill, value=1}}}
  }
  Lane2Offset {
   widgetClassPath=&CommonTextBlock
   bIsVariable=true
   parent=HorizontalBox
   widget=Lane2Offset
   properties {
    style=&TextStyle-Base_C
    text=0
    font {
     bForceMonospaced=false
     bMaterialIsStencil=false
     fontMaterial=null
     fontObject=&NotoSans
     letterSpacing=0
     monospacedWidth=1
     size=24
     skewAmount=0
     typefaceFontName=NotoSans-Regular
     outlineSettings {
      bApplyOutlineToDropShadows=false
      bMiteredCorners=false
      bSeparateFillAlpha=false
      outlineMaterial=null
      outlineSize=0
      outlineColor=<rgba 0 0 0 1>
     }
    }
    shadowOffset=<0 0>
   }
   slot {class=HorizontalBoxSlot}
  }
  Spacer {
   widgetClassPath=&Spacer
   parent=HorizontalBox
   widget=Spacer
   slot {class=HorizontalBoxSlot, properties {size {sizeRule=Fill, value=1}}}
  }
  Lane3Offset {
   widgetClassPath=&CommonTextBlock
   bIsVariable=true
   parent=HorizontalBox
   widget=Lane3Offset
   properties {
    style=&TextStyle-Base_C
    text=0
    font {
     bForceMonospaced=false
     bMaterialIsStencil=false
     fontMaterial=null
     fontObject=&NotoSans
     letterSpacing=0
     monospacedWidth=1
     size=24
     skewAmount=0
     typefaceFontName=NotoSans-Regular
     outlineSettings {
      bApplyOutlineToDropShadows=false
      bMiteredCorners=false
      bSeparateFillAlpha=false
      outlineMaterial=null
      outlineSize=0
      outlineColor=<rgba 0 0 0 1>
     }
    }
    shadowOffset=<0 0>
   }
   slot {class=HorizontalBoxSlot}
  }
  Spacer_112 {
   widgetClassPath=&Spacer
   parent=HorizontalBox
   widget=Spacer_112
   slot {class=HorizontalBoxSlot, properties {size {sizeRule=Fill, value=1}}}
  }
  Lane4Offset {
   widgetClassPath=&CommonTextBlock
   bIsVariable=true
   parent=HorizontalBox
   widget=Lane4Offset
   properties {
    style=&TextStyle-Base_C
    text=0
    font {
     bForceMonospaced=false
     bMaterialIsStencil=false
     fontMaterial=null
     fontObject=&NotoSans
     letterSpacing=0
     monospacedWidth=1
     size=24
     skewAmount=0
     typefaceFontName=NotoSans-Regular
     outlineSettings {
      bApplyOutlineToDropShadows=false
      bMiteredCorners=false
      bSeparateFillAlpha=false
      outlineMaterial=null
      outlineSize=0
      outlineColor=<rgba 0 0 0 1>
     }
    }
    shadowOffset=<0 0>
   }
   slot {class=HorizontalBoxSlot}
  }
 ]
}
