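# Music battle widget for Pilgrim Quickplay that presents a "Variety Bonus": four instrument icons and a band score percentage.
# On battle initialization it darkens all icons through dynamic materials, then brightens those matching players' track types, collected uniquely.
# The count of active icons selects a multiplier shown as bonus text before the FadeOut animation plays.
# Both options of the visibility Select are Collapsed, so setup collapses the widget regardless of player count.
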WidgetBlueprint /PilgrimQuickplay/Widgets/WBP_PilgrimQuickplay_InstrumentVarietyMultiplier
animations [FadeOut {
  binding_count=1
  length_seconds=5.2
  object=&self:FadeOut
  sequence {bindings [WBP_PilgrimQuickplay_InstrumentVarietyMultiplier_C_113 {
     id=#1
     possessed_class=WBP_PilgrimQuickplay_InstrumentVarietyMultiplier_C
     tracks [{
       type=MovieSceneFloatTrack
       display_name=Render Opacity
       sections [{
         type=MovieSceneFloatSection
         end_frame=100
         start_frame=-4
         channels [None {type=MovieSceneScriptingFloatChannel, keys [[-4,1],[60,1],[100,0]]}]
        }]
      }]
    }]}
 }]
graphs [
 BrightenActiveIcons {nodes [
   n1 = BrightenActiveIcons!() -> (then, MusicBattleStates:Array of Pilgrim Music Battle Player State Structures) {is=title K2Node_FunctionEntry}
   > Clear(TargetArray:Array of Image Object References=Get ActiveInstrumentIcons() -> (ActiveInstrumentIcons:Array of Image Object References))
   > n2 = For Each Loop!(Exec, Array:Array of Pilgrim Music Battle Player State Structures=@n1) -> (
    LoopBody
    Array Element_PlayerState:Fort Player State Athena Object Reference
    Array Element_Team:Byte
    Array Element_TrackType:EPilgrimTrackType Enum
    Array Element_InstrumentTags:Gameplay Tag Container Structure
    Array Element_PilgrimGameProgress:Pilgrim Game Progress Structure
    Array Element_ServerScoringChanges:Pilgrim Server Scoring Changes Structure
    Array Index
    Completed
   )
   > n3 = Set CurrentIcon(CurrentIcon:Image Object Reference=Select!(
     Guitar:Image Object Reference=Get GuitarIcon() -> (GuitarIcon:Image Object Reference)
     Bass:Image Object Reference=Get BassIcon() -> (BassIcon:Image Object Reference)
     Vocals:Image Object Reference=Get VocalsIcon() -> (VocalsIcon:Image Object Reference)
     Drum:Image Object Reference=Get DrumsIcon() -> (DrumsIcon:Image Object Reference)
     Keyboard:Image Object Reference
     None:Image Object Reference
     Index:ESparksInstrumentType Enum=PilgrimTrackTypeToSparksInstrument!(PilgrimType:EPilgrimTrackType Enum=@n2.Array Element_TrackType) -> (ReturnValue:ESparksInstrumentType Enum=Guitar)
    ) -> (ReturnValue:Image Object Reference)) -> (Output_Get:Image Object Reference)
   > Is Valid(InputObject=@n3)
   > Add Unique(TargetArray:Array of Image Object References=Get ActiveInstrumentIcons(), NewItem:"Image Object Reference (by ref)"=@n3)
   > n4 = GetDynamicMaterial(self=@n3)
   n5 = Get BrightenInstrumentIcon() -> (BrightenInstrumentIcon:float)
   > SetVectorParameterValue!(
    execute
    self=@n4
    ParameterName=Icon Color
    Value_R:float=@n5
    Value_G:float=@n5
    Value_B:float=@n5
    Value_A:float=Get InstrumentIconVisibility() -> (InstrumentIconVisibility:float)
   ) -> (then)
  ]}
 EventGraph {nodes [
   SetupUIVisuals() {is=title K2Node_CustomEvent}
   n1 = Get MusicBattleInstance() -> (MusicBattleInstance:Pilgrim Music Battle Instance Object Reference)
   > Is Valid(InputObject=@n1)
   n2 = Break Pilgrim Music Battle Data!(PilgrimMusicBattleData:"Pilgrim Music Battle Data Structure (by ref)"=GetMusicBattleData!(self:Pilgrim Music Battle Instance Object Reference=@n1) -> (ReturnValue:"Pilgrim Music Battle Data Structure (by ref)")) -> (MusicBattleStates:Array of Pilgrim Music Battle Player State Structures, BandState:Pilgrim Band State Structure) {is=title K2Node_BreakStruct}
   n3 = int - int(A=Length(TargetArray:Array of Pilgrim Music Battle Player State Structures=@n2))
   > SetVisibility(InVisibility=Select!(Option 0:ESlateVisibility Enum=Collapsed, Option 1:ESlateVisibility Enum=Collapsed, Index:bool="integer > integer"(A=@n3)) -> (ReturnValue:ESlateVisibility Enum))
   > n4 = Sequence()
   > n5 = For Loop(LastIndex=3)
   > n6 = GetDynamicMaterial(self=Select!(
     Option 0:Image Object Reference=Get GuitarIcon()
     Option 1:Image Object Reference=Get BassIcon()
     Option 2:Image Object Reference=Get VocalsIcon()
     Option 3:Image Object Reference=Get DrumsIcon()
     Index:int=@n5
    ) -> (ReturnValue:Image Object Reference))
   n7 = Get DarkenInstrumentIcon() -> (DarkenInstrumentIcon:float)
   > SetVectorParameterValue(self=@n6, ParameterName=Icon Color, Value_R=@n7, Value_G=@n7, Value_B=@n7, Value_A=Get InstrumentIconVisibility())
   Set IsInitialized(execute=@n4.then_1, IsInitialized:bool=true) -> (Output_Get:bool)
   > Branch(Condition=Get IsInitialized() -> (IsInitialized:bool))
   > Set LocalPlayerIndex(LocalPlayerIndex:int=Get Local Player Index() -> (Local Player Index:int=0)) -> (Output_Get:int=0)
   > n8 = Sequence()
   > BrightenActiveIcons!(execute, self, MusicBattleStates:Array of Pilgrim Music Battle Player State Structures=@n2) -> (then) {is=title K2Node_CallFunction}
   n9 = int - int(A=Length(TargetArray:Array of Image Object References=Get ActiveInstrumentIcons()))
   n10 = Get 1InstrumentTypeMutiplier() -> (1InstrumentTypeMutiplier:double)
   n11 = Get 4InstrumentTypeMutiplier() -> (4InstrumentTypeMutiplier:double)
   n12 = Get 2InstrumentTypeMutiplier() -> (2InstrumentTypeMutiplier:double)
   n13 = Get 3InstrumentTypeMutiplier() -> (3InstrumentTypeMutiplier:double)
   n14 = float - float(
    A=float * float(A=Select(Option 0:double=@n10, Option 1:double=@n12, Option 2:double=@n13, Option 3:double=@n11, Index=@n9) -> (ReturnValue:double), B=100) {is=title K2Node_PromotableOperator}
    B=100
   )
   Set Text(
    execute=@n8.then_1
    Text=Format Text!(Format="INVTEXT(\"+{Value}% band score\")", Value:Integer64="To Integer64 (Float)"!(InDouble:double=@n14) -> (ReturnValue:Integer64=0)) -> (Result)
    self=Get Text_Message() -> (Text_Message:Common Text Object Reference)
   )
   > PlayAnimation(InAnimation=Get FadeOut() -> (FadeOut:Widget Animation Object Reference))
   Event OnMusicBattleInstanceInitializeEvent() {is=title K2Node_Event}
   > SetupUIVisuals() {is=title K2Node_CallFunction}
  ]}
]
variables [
 1InstrumentTypeMutiplier {type=number, default=1}
 2InstrumentTypeMutiplier {type=number, default=1.2}
 3InstrumentTypeMutiplier {type=number, default=1.4}
 4InstrumentTypeMutiplier {type=number, default=1.6}
 ActiveInstrumentIcons {type="array<Image>", default []}
 BassInstrumentInt {type=integer, default=1}
 BassMaxStreak {type=integer, default=6}
 BrightenInstrumentIcon {type=number, default=1}
 Collection {type=MaterialParameterCollection, default=null}
 DarkenInstrumentIcon {type=number, default=0.15}
 Filler to make UE happy {type=number}
 InitializeBandmateUniqueTraits {type="array<boolean>", default []}
 InstrumentIconColor {type=string, default=None}
 InstrumentIconVisibility {type=number, default=1}
 IsInitialized {type=boolean}
 Local Player Index {type=integer}
 LocalPlayerIndex {type=integer}
 NonBassMaxStreak {type=integer, default=4}
 Start at Time {type=number, default=35}
 Target {type=MaterialInstanceDynamic}
]
widgets {
 info {inheritedWidgetCount=0, parentClass=/Script/PilgrimCoreUI.PilgrimGameWidgetBase, rootWidgetClass=&VerticalBox, widgetCount=14}
 widgets [
  VerticalBox_0 {widgetClassPath=&VerticalBox, widget=VerticalBox_0}
  VarietyBonusText {
   widgetClassPath=&Overlay
   parent=VerticalBox_0
   widget=VarietyBonusText
   slot {class=VerticalBoxSlot, properties {horizontalAlignment=HAlign_Center}}
  }
  Image {
   widgetClassPath=&Image
   bIsVariable=true
   parent=VarietyBonusText
   widget=Image
   properties {brush {
     bIsDynamicallyLoaded=false
     drawAs=Image
     imageType=NoImage
     mirroring=NoMirror
     resourceName=None
     resourceObject=null
     tiling=NoTile
     imageSize=<32 32>
     margin=<ltrb 0 0 0 0>
     outlineSettings {
      bUseBrushTransparency=false
      roundingType=HalfHeightRadius
      width=0
      color {colorUseRule=UseColor_Specified, specifiedColor=<rgba 0 0 0 0>}
      cornerRadii=<0 0 0 0>
     }
     tintColor {colorUseRule=UseColor_Specified, specifiedColor=<rgba 1 1 1 0.25>}
     uVRegion {bIsValid=false, max=<0 0>, min=<0 0>}
    }}
   slot {class=OverlaySlot, properties {horizontalAlignment=HAlign_Fill, verticalAlignment=VAlign_Fill}}
  }
  Text_Title {
   widgetClassPath=&CommonTextBlock
   parent=VarietyBonusText
   widget=Text_Title
   properties {
    justification=Center
    text=Variety Bonus
    textTransformPolicy=ToUpper
    font {
     bForceMonospaced=false
     bMaterialIsStencil=false
     fontMaterial=null
     fontObject=&HeadingNow
     letterSpacing=0
     monospacedWidth=1
     size=24
     skewAmount=0
     typefaceFontName=76 Bold Italic
     outlineSettings {
      bApplyOutlineToDropShadows=false
      bMiteredCorners=false
      bSeparateFillAlpha=false
      outlineMaterial=null
      outlineSize=0
      outlineColor=<rgba 0 0 0 1>
     }
    }
    shadowOffset=<0 0>
    strikeBrush {
     bIsDynamicallyLoaded=false
     drawAs=NoDrawType
     imageType=NoImage
     mirroring=NoMirror
     resourceName=None
     resourceObject=null
     tiling=NoTile
     imageSize=<32 32>
     margin=<ltrb 0 0 0 0>
     outlineSettings {
      bUseBrushTransparency=false
      roundingType=HalfHeightRadius
      width=0
      color {colorUseRule=UseColor_Specified, specifiedColor=<rgba 0 0 0 0>}
      cornerRadii=<0 0 0 0>
     }
     tintColor {colorUseRule=UseColor_Specified, specifiedColor=<rgba 1 1 1 1>}
     uVRegion {bIsValid=false, max=<0 0>, min=<0 0>}
    }
   }
   slot {class=OverlaySlot, properties {horizontalAlignment=HAlign_Fill, verticalAlignment=VAlign_Fill, padding=<ltrb 8 0 8 0>}}
  }
  Overlay_0 {
   widgetClassPath=&Overlay
   parent=VerticalBox_0
   widget=Overlay_0
   slot {class=VerticalBoxSlot, properties {horizontalAlignment=HAlign_Center, verticalAlignment=VAlign_Center}}
  }
  FortLazyImage_152 {
   widgetClassPath=&FortLazyImage
   bIsVariable=true
   parent=Overlay_0
   widget=FortLazyImage_152
   properties {brush {
     bIsDynamicallyLoaded=false
     drawAs=RoundedBox
     imageType=NoImage
     mirroring=NoMirror
     resourceName=None
     resourceObject=null
     tiling=NoTile
     imageSize=<470 112>
     margin=<ltrb 0 0 0 0>
     outlineSettings {
      bUseBrushTransparency=false
      roundingType=HalfHeightRadius
      width=0
      color {colorUseRule=UseColor_Specified, specifiedColor=<rgba 0 0 0 0>}
      cornerRadii=<0.4 0.4 0.4 0.4>
     }
     tintColor {colorUseRule=UseColor_Specified, specifiedColor=<rgba 0 0 0 1>}
     uVRegion {bIsValid=false, max=<0 0>, min=<0 0>}
    }}
   slot {class=OverlaySlot, properties {horizontalAlignment=HAlign_Center, verticalAlignment=VAlign_Center}}
  }
  InstrumentIcons {
   widgetClassPath=&HorizontalBox
   parent=Overlay_0
   widget=InstrumentIcons
   slot {class=OverlaySlot, properties {horizontalAlignment=HAlign_Center, verticalAlignment=VAlign_Center}}
  }
  GuitarIcon {
   widgetClassPath=&Image
   bIsVariable=true
   parent=InstrumentIcons
   widget=GuitarIcon
   properties {brush {
     bIsDynamicallyLoaded=false
     drawAs=Image
     imageType=NoImage
     mirroring=NoMirror
     resourceName=None
     resourceObject=/PilgrimQuickplay/Widgets/Materials/MI_UI_Icon_Instrument_Type_Guitar
     tiling=NoTile
     imageSize=<142 142>
     margin=<ltrb 0 0 0 0>
     outlineSettings {
      bUseBrushTransparency=false
      roundingType=HalfHeightRadius
      width=0
      color {colorUseRule=UseColor_Specified, specifiedColor=<rgba 0 0 0 0>}
      cornerRadii=<0 0 0 0>
     }
     tintColor {colorUseRule=UseColor_Specified, specifiedColor=<rgba 1 1 1 1>}
     uVRegion {bIsValid=false, max=<0 0>, min=<0 0>}
    }}
   slot {class=HorizontalBoxSlot, properties {padding=<ltrb 0 0 -24 0>}}
  }
  DrumsIcon {
   widgetClassPath=&Image
   bIsVariable=true
   parent=InstrumentIcons
   widget=DrumsIcon
   properties {brush {
     bIsDynamicallyLoaded=false
     drawAs=Image
     imageType=NoImage
     mirroring=NoMirror
     resourceName=None
     resourceObject=/PilgrimQuickplay/Widgets/Materials/MI_UI_Icon_Instrument_Type_Drums
     tiling=NoTile
     imageSize=<142 142>
     margin=<ltrb 0 0 0 0>
     outlineSettings {
      bUseBrushTransparency=false
      roundingType=HalfHeightRadius
      width=0
      color {colorUseRule=UseColor_Specified, specifiedColor=<rgba 0 0 0 0>}
      cornerRadii=<0 0 0 0>
     }
     tintColor {colorUseRule=UseColor_Specified, specifiedColor=<rgba 1 1 1 1>}
     uVRegion {bIsValid=false, max=<0 0>, min=<0 0>}
    }}
   slot {class=HorizontalBoxSlot}
  }
  VocalsIcon {
   widgetClassPath=&Image
   bIsVariable=true
   parent=InstrumentIcons
   widget=VocalsIcon
   properties {brush {
     bIsDynamicallyLoaded=false
     drawAs=Image
     imageType=NoImage
     mirroring=NoMirror
     resourceName=None
     resourceObject=/PilgrimQuickplay/Widgets/Materials/MI_UI_Icon_Instrument_Type_Vocals
     tiling=NoTile
     imageSize=<142 142>
     margin=<ltrb 0 0 0 0>
     outlineSettings {
      bUseBrushTransparency=false
      roundingType=HalfHeightRadius
      width=0
      color {colorUseRule=UseColor_Specified, specifiedColor=<rgba 0 0 0 0>}
      cornerRadii=<0 0 0 0>
     }
     tintColor {colorUseRule=UseColor_Specified, specifiedColor=<rgba 1 1 1 1>}
     uVRegion {bIsValid=false, max=<0 0>, min=<0 0>}
    }}
   slot {class=HorizontalBoxSlot, properties {padding=<ltrb -24 0 0 0>}}
  }
  BassIcon {
   widgetClassPath=&Image
   bIsVariable=true
   parent=InstrumentIcons
   widget=BassIcon
   properties {brush {
     bIsDynamicallyLoaded=false
     drawAs=Image
     imageType=NoImage
     mirroring=NoMirror
     resourceName=None
     resourceObject=/PilgrimQuickplay/Widgets/Materials/MI_UI_Icon_Instrument_Type_Bass
     tiling=NoTile
     imageSize=<142 142>
     margin=<ltrb 0 0 0 0>
     outlineSettings {
      bUseBrushTransparency=false
      roundingType=HalfHeightRadius
      width=0
      color {colorUseRule=UseColor_Specified, specifiedColor=<rgba 0 0 0 0>}
      cornerRadii=<0 0 0 0>
     }
     tintColor {colorUseRule=UseColor_Specified, specifiedColor=<rgba 1 1 1 1>}
     uVRegion {bIsValid=false, max=<0 0>, min=<0 0>}
    }}
   slot {class=HorizontalBoxSlot, properties {padding=<ltrb -24 0 0 0>}}
  }
  BandScoreMultiplierText {
   widgetClassPath=&Overlay
   parent=VerticalBox_0
   widget=BandScoreMultiplierText
   slot {class=VerticalBoxSlot, properties {horizontalAlignment=HAlign_Center}}
  }
  Image_102 {
   widgetClassPath=&Image
   bIsVariable=true
   parent=BandScoreMultiplierText
   widget=Image_102
   properties {brush {
     bIsDynamicallyLoaded=false
     drawAs=Image
     imageType=NoImage
     mirroring=NoMirror
     resourceName=None
     resourceObject=null
     tiling=NoTile
     imageSize=<32 32>
     margin=<ltrb 0 0 0 0>
     outlineSettings {
      bUseBrushTransparency=false
      roundingType=HalfHeightRadius
      width=0
      color {colorUseRule=UseColor_Specified, specifiedColor=<rgba 0 0 0 0>}
      cornerRadii=<0 0 0 0>
     }
     tintColor {colorUseRule=UseColor_Specified, specifiedColor=<rgba 1 1 1 0.25>}
     uVRegion {bIsValid=false, max=<0 0>, min=<0 0>}
    }}
   slot {class=OverlaySlot, properties {horizontalAlignment=HAlign_Fill, verticalAlignment=VAlign_Fill}}
  }
  Text_Message {
   widgetClassPath=&CommonTextBlock
   bIsVariable=true
   parent=BandScoreMultiplierText
   widget=Text_Message
   properties {
    justification=Center
    text="+{}% band score"
    font {
     bForceMonospaced=false
     bMaterialIsStencil=false
     fontMaterial=null
     fontObject=&HeadingNow
     letterSpacing=0
     monospacedWidth=1
     size=24
     skewAmount=0
     typefaceFontName=75 Medium
     outlineSettings {
      bApplyOutlineToDropShadows=false
      bMiteredCorners=false
      bSeparateFillAlpha=false
      outlineMaterial=null
      outlineSize=0
      outlineColor=<rgba 0 0 0 1>
     }
    }
    shadowOffset=<0 0>
    strikeBrush {
     bIsDynamicallyLoaded=false
     drawAs=NoDrawType
     imageType=NoImage
     mirroring=NoMirror
     resourceName=None
     resourceObject=null
     tiling=NoTile
     imageSize=<32 32>
     margin=<ltrb 0 0 0 0>
     outlineSettings {
      bUseBrushTransparency=false
      roundingType=HalfHeightRadius
      width=0
      color {colorUseRule=UseColor_Specified, specifiedColor=<rgba 0 0 0 0>}
      cornerRadii=<0 0 0 0>
     }
     tintColor {colorUseRule=UseColor_Specified, specifiedColor=<rgba 1 1 1 1>}
     uVRegion {bIsValid=false, max=<0 0>, min=<0 0>}
    }
   }
   slot {class=OverlaySlot, properties {horizontalAlignment=HAlign_Fill, verticalAlignment=VAlign_Fill, padding=<ltrb 8 0 8 0>}}
  }
 ]
}
