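# Pilgrim Quickplay HUD widget that shows the local player's score beside an instrument-specific label.
# On construct it subscribes to tag-driven visibility (Pilgrim.Widget.LocalScore) and runs UpdateInstrumentText.
# That function takes the local player override from the playspace's player manager, falling back to the owning player's PlayerState, then formats that player's instrument choice as "{Instrument} Score".
# SetScore writes supplied text into the glowing score readout.
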WidgetBlueprint /PilgrimQuickplay/Widgets/WBP_Pilgrim_LocalPlayerScore
graphs [
 EventGraph {nodes [
   Event Construct()
   > SubscribeForChangingVisibilityByTag!(
    execute
    self:Pilgrim Game Widget Base Object Reference
    GameplayTag:"Gameplay Tag Structure (by ref)"=MakeLiteralGameplayTag!(Value:Gameplay Tag Structure=(TagName="Pilgrim.Widget.LocalScore")) -> (ReturnValue:Gameplay Tag Structure)
   ) -> (then)
   > UpdateInstrumentText()
  ]}
 SetScore {nodes [
   n1 = SetScore!() -> (then, Score:Text) {is=title K2Node_FunctionEntry}
   > Set Text(Text=@n1, self=Get ScoreText() -> (ScoreText:Common Text Object Reference))
  ]}
 UpdateInstrumentText {nodes [
   UpdateInstrumentText() {is=title K2Node_FunctionEntry}
   > n1 = Sequence()
   n2 = GetSparksMusicPlayspaceOfClass!(
    PlayspaceType:Sparks Music Playspace Class Reference=/Script/PilgrimQuickplayRuntime.PilgrimQuickplayPlayspace
    Actor:Actor Object Reference=GetOwningPlayer()
    bMatchChildTypes:bool=true
   ) -> (ReturnValue:Pilgrim Quickplay Playspace Object Reference)
   > n3 = Is Valid(InputObject=@n2)
   n4 = GetComponentByClass!(
    self:Actor Object Reference=@n2
    ComponentClass:Actor Component Class Reference=/Script/PilgrimQuickplayRuntime.PilgrimQuickplayPlayspaceComponent_PlayerManager
   ) -> (ReturnValue:Pilgrim Quickplay Playspace Component Player Manager Object Reference)
   > n5 = Is Valid(InputObject=@n4)
   n6 = GetLocalPlayerOverride!(self:Pilgrim Quickplay Playspace Component Player Manager Object Reference=@n4) -> (ReturnValue:Player State Object Reference)
   > n7 = Is Valid(InputObject=@n6)
   > n8 = Set PlayerToUse(PlayerToUse:Player State Object Reference=@n6) -> (Output_Get:Player State Object Reference)
   n9 = GetPilgrimQuickplayPlayerStateComponent!(execute=[@n10.then, @n8.then], PlayerState:Player State Object Reference=Get PlayerToUse() -> (PlayerToUse:Player State Object Reference)) -> (then, ReturnValue:Pilgrim Quickplay Player State Component Object Reference)
   > Set TrackType(TrackType:EPilgrimTrackType Enum=GetInstrumentChoice!(self:Pilgrim Core Player State Component Object Reference=@n9) -> (ReturnValue:EPilgrimTrackType Enum=TrackGuitar)) -> (Output_Get:EPilgrimTrackType Enum=TrackGuitar)
   n10 = Set PlayerToUse(
    execute=[@n7.Is Not Valid, @n5.Is Not Valid, @n3.Is Not Valid]
    PlayerToUse=Get PlayerState!(self:Controller Object Reference=GetOwningPlayer()) -> (PlayerState:Player State Object Reference)
   )
   n11 = Get TrackType() -> (TrackType:EPilgrimTrackType Enum=TrackGuitar)
   Set Text(
    execute=@n1.then_1
    Text=Format Text!(
     Format="LOCTEXT(\"{Instrument} Score\")"
     Instrument:Text=GetInstrumentNameFromPart!(TrackType:EPilgrimTrackType Enum=@n11) -> (InstrumentText:Text)
    ) -> (Result)
    self=Get PlayerScore_TextBlock() -> (PlayerScore_TextBlock:Common Text Object Reference)
   )
  ]}
]
variables [GameplayTagForVisibility {type=GameplayTag, default {tagName=None}}]
widgets {
 info {inheritedWidgetCount=0, parentClass=/Script/PilgrimCoreUI.PilgrimGameWidgetBase, rootWidgetClass=&SizeBox, widgetCount=10}
 widgets [
  SizeBox_0 {
   widgetClassPath=&SizeBox
   widget=SizeBox_0
   properties {bOverride_HeightOverride=true, bOverride_WidthOverride=true, heightOverride=56, widthOverride=374}
  }
  Overlay_All {widgetClassPath=&Overlay, bIsVariable=true, parent=SizeBox_0, widget=Overlay_All, slot {class=SizeBoxSlot}}
  Shadow {
   widgetClassPath=&Image
   bIsVariable=true
   parent=Overlay_All
   widget=Shadow
   properties {brush {
     bIsDynamicallyLoaded=false
     drawAs=Image
     imageType=NoImage
     mirroring=NoMirror
     resourceName=None
     resourceObject=/PilgrimCore/Widgets/Materials/M_UI_Pilgrim_RoundedShadow_Box
     tiling=NoTile
     imageSize=<32 32>
     margin=<ltrb 0 0 0 0>
     outlineSettings {
      bUseBrushTransparency=false
      roundingType=HalfHeightRadius
      width=0
      color {colorUseRule=UseColor_Specified, specifiedColor=<rgba 0 0 0 0>}
      cornerRadii=<0 0 0 0>
     }
     tintColor {colorUseRule=UseColor_Specified, specifiedColor=<rgba 1 1 1 1>}
     uVRegion {bIsValid=false, max=<0 0>, min=<0 0>}
    }}
   slot {class=OverlaySlot, properties {horizontalAlignment=HAlign_Fill, verticalAlignment=VAlign_Fill, padding=<ltrb -16 -16 -16 -16>}}
  }
  Backing {
   widgetClassPath=&Image
   bIsVariable=true
   parent=Overlay_All
   widget=Backing
   properties {brush {
     bIsDynamicallyLoaded=false
     drawAs=Image
     imageType=NoImage
     mirroring=NoMirror
     resourceName=None
     resourceObject=/PilgrimCore/Widgets/Materials/MI_UI_Pilgrim_Score_Backing
     tiling=NoTile
     imageSize=<32 32>
     margin=<ltrb 0 0 0 0>
     outlineSettings {
      bUseBrushTransparency=false
      roundingType=HalfHeightRadius
      width=0
      color {colorUseRule=UseColor_Specified, specifiedColor=<rgba 0 0 0 0>}
      cornerRadii=<0 0 0 0>
     }
     tintColor {colorUseRule=UseColor_Specified, specifiedColor=<rgba 1 1 1 1>}
     uVRegion {bIsValid=false, max=<0 0>, min=<0 0>}
    }}
   slot {class=OverlaySlot, properties {horizontalAlignment=HAlign_Fill, verticalAlignment=VAlign_Fill}}
  }
  HorizontalBox_0 {
   widgetClassPath=&HorizontalBox
   parent=Overlay_All
   widget=HorizontalBox_0
   properties {flowDirectionPreference=Culture}
   slot {class=OverlaySlot, properties {horizontalAlignment=HAlign_Fill, verticalAlignment=VAlign_Fill, padding=<ltrb 16 0 16 1>}}
  }
  PlayerScore_TextBlock {
   widgetClassPath=&CommonTextBlock
   bIsVariable=true
   parent=HorizontalBox_0
   widget=PlayerScore_TextBlock
   properties {
    scrollStyle=&TextScrollStyle-Base_C
    text=Pro Vocals Score With LOC And More Words
    textTransformPolicy=ToUpper
    font {
     bForceMonospaced=false
     bMaterialIsStencil=false
     fontMaterial=/PilgrimCore/Widgets/Materials/MI_UI_Piligrim_HUDTextColor
     fontObject=&HeadingNow
     letterSpacing=20
     monospacedWidth=1
     size=12
     skewAmount=0
     typefaceFontName=76 Bold
     outlineSettings {
      bApplyOutlineToDropShadows=false
      bMiteredCorners=false
      bSeparateFillAlpha=true
      outlineMaterial=&M-FontShadow
      outlineSize=2
      outlineColor=<rgba 0 0 0 1>
     }
    }
    shadowOffset=<0 0>
    strikeBrush {
     bIsDynamicallyLoaded=false
     drawAs=NoDrawType
     imageType=NoImage
     mirroring=NoMirror
     resourceName=None
     resourceObject=null
     tiling=NoTile
     imageSize=<32 32>
     margin=<ltrb 0 0 0 0>
     outlineSettings {
      bUseBrushTransparency=false
      roundingType=HalfHeightRadius
      width=0
      color {colorUseRule=UseColor_Specified, specifiedColor=<rgba 0 0 0 0>}
      cornerRadii=<0 0 0 0>
     }
     tintColor {colorUseRule=UseColor_Specified, specifiedColor=<rgba 1 1 1 1>}
     uVRegion {bIsValid=false, max=<0 0>, min=<0 0>}
    }
   }
   slot {class=HorizontalBoxSlot, properties {horizontalAlignment=HAlign_Left, verticalAlignment=VAlign_Center, size {sizeRule=Fill, value=1}}}
  }
  Spacer_83 {widgetClassPath=&Spacer, parent=HorizontalBox_0, widget=Spacer_83, properties {size=<10 1>}, slot {class=HorizontalBoxSlot}}
  TextOverlay {widgetClassPath=&Overlay, bIsVariable=true, parent=HorizontalBox_0, widget=TextOverlay, slot {class=HorizontalBoxSlot}}
  ScoreGlow {
   widgetClassPath=&Image
   bIsVariable=true
   parent=TextOverlay
   widget=ScoreGlow
   properties {
    clipping=ClipToBoundsWithoutIntersecting
    brush {
     bIsDynamicallyLoaded=false
     drawAs=Image
     imageType=NoImage
     mirroring=NoMirror
     resourceName=None
     resourceObject=/PilgrimCore/Widgets/Materials/MI_UI_Pilgrim_ScoreGlow
     tiling=NoTile
     imageSize=<32 32>
     margin=<ltrb 0 0 0 0>
     outlineSettings {
      bUseBrushTransparency=false
      roundingType=HalfHeightRadius
      width=0
      color {colorUseRule=UseColor_Specified, specifiedColor=<rgba 0 0 0 0>}
      cornerRadii=<0 0 0 0>
     }
     tintColor {colorUseRule=UseColor_Specified, specifiedColor=<rgba 1 1 1 1>}
     uVRegion {bIsValid=false, max=<0 0>, min=<0 0>}
    }
   }
   slot {class=OverlaySlot, properties {horizontalAlignment=HAlign_Fill, verticalAlignment=VAlign_Fill, padding=<ltrb -10 -2 -10 -2>}}
  }
  ScoreText {
   widgetClassPath=&CommonTextBlock
   bIsVariable=true
   parent=TextOverlay
   widget=ScoreText
   properties {
    justification=Right
    text="21,345"
    font {
     bForceMonospaced=true
     bMaterialIsStencil=false
     fontMaterial=null
     fontObject=&HeadingNow
     letterSpacing=0
     monospacedWidth=1
     size=21
     skewAmount=0
     typefaceFontName=76 Bold
     outlineSettings {
      bApplyOutlineToDropShadows=false
      bMiteredCorners=false
      bSeparateFillAlpha=false
      outlineMaterial=&M-FontShadow
      outlineSize=2
      outlineColor=<rgba 1 0.5 1 1>
     }
    }
    shadowColorAndOpacity=<rgba 0 0 0 1>
    shadowOffset=<0 0>
    shapedTextOptions {bOverride_TextFlowDirection=false, bOverride_TextShapingMethod=false, textFlowDirection=LeftToRight, textShapingMethod=Auto}
   }
   slot {class=OverlaySlot, properties {horizontalAlignment=HAlign_Center, verticalAlignment=VAlign_Center}}
  }
 ]
}
