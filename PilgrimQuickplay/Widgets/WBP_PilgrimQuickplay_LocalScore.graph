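# Quickplay variant of the Pilgrim local player score widget, a child of WBP_Pilgrim_LocalPlayerScore that adds no widgets of its own.
# Its event graph holds only a loose Set Text chain that prefixes "Pro " to an unconnected text input, with no event triggering it.
# It declares a gameplay tag variable for visibility, which defaults to None and is not used in this graph.
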WidgetBlueprint /PilgrimQuickplay/Widgets/WBP_PilgrimQuickplay_LocalScore
graphs [EventGraph {nodes [
   n1 = "To String (Text)"(InText=MakeLiteralText(Value="LOCTEXT(\"Pro \")"))
   n2 = "To String (Text)"()
   Set Text(Text="To Text (String)"(InString=Append(A=@n1, B=@n2)))
  ]}]
variables [GameplayTagForVisibility_0 {type=GameplayTag, default {tagName=None}}]
widgets {
 info {
  inheritedWidgetCount=0
  parentClass=&/WBP_Pilgrim_LocalPlayerScore.WBP_Pilgrim_LocalPlayerScore_C
  rootWidgetClass=null
  widgetCount=0
 }
 widgets []
}
