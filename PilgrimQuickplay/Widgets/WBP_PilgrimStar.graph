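# WBP_PilgrimStar is a Pilgrim Quickplay star widget showing not-started, progress, completed, gold-progress and gold-completed states.
# Progress functions drive cached dynamic materials, while completion functions switch layer visibility and play burst animations or, without ShowCompleteAnim, apply the end state.
# SetFXColors swaps shared effect materials between normal and gold, and RandomizeFXParameters varies noise, seeds, glow speed and angle.
# ResetCompletedVisuals restores the dimmed initial look and clears the gold-complete flag.
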WidgetBlueprint /PilgrimQuickplay/Widgets/WBP_PilgrimStar
animations [
 anim_gold_completed {
  binding_count=11
  length_seconds=1.25
  object=&self:anim_gold_completed
  sequence {bindings [
    GoldStarComplete_Glow_Radial {
     id=#1
     possessed_class=Image
     tracks [{
       type=MovieSceneWidgetMaterialTrack
       sections [{type=MovieSceneParameterSection, channels [Final_Alpha {type=MovieSceneScriptingFloatChannel, keys [[3,0.25],[6,1],[25,0.25]]}]}]
      }]
    }
    Particles_Down {
     id=#2
     possessed_class=Image
     tracks [{
       type=MovieSceneWidgetMaterialTrack
       sections [{type=MovieSceneParameterSection, channels [Anim_Scale {type=MovieSceneScriptingFloatChannel, keys [[3,1],[6,0],[10,0],[25,0.666]]}]}]
      }]
    }
    StarBurst_Gold {
     id=#3
     possessed_class=Image
     tracks [
      {
       type=MovieSceneFloatTrack
       display_name=Render Opacity
       sections [{type=MovieSceneFloatSection, channels [None {type=MovieSceneScriptingFloatChannel, keys [[3,0],[4,1],[13.3333,0]]}]}]
      }
      {
       type=MovieScene2DTransformTrack
       sections [{
         type=MovieScene2DTransformSection
         channels [
          Scale.X {type=MovieSceneScriptingFloatChannel, keys [[3,1],[13.3333,3]]}
          Scale.Y {type=MovieSceneScriptingFloatChannel, keys [[3,1],[13.3333,3]]}
         ]
        }]
      }
      {
       type=MovieSceneByteTrack
       display_name=Visibility
       sections [{type=MovieSceneByteSection, end_frame=15, start_frame=0, channels [None {type=MovieSceneScriptingByteChannel, keys [[0,4],[15,1]]}]}]
      }
     ]
    }
    FlareFX {
     id=#4
     possessed_class=Overlay
     tracks [
      {
       type=MovieScene2DTransformTrack
       sections [{
         type=MovieScene2DTransformSection
         channels [
          Scale.X {type=MovieSceneScriptingFloatChannel, const [1,2]}
          Scale.Y {type=MovieSceneScriptingFloatChannel, keys [[0,0.25],[1,2],[3,2],[6,0.25]]}
         ]
        }]
      }
      {
       type=MovieSceneFloatTrack
       display_name=Render Opacity
       sections [{type=MovieSceneFloatSection, channels [None {type=MovieSceneScriptingFloatChannel, thinned_from=5, keys [[0,0],[0.6667,1],[25,1]]}]}]
      }
      {
       type=MovieSceneByteTrack
       display_name=Visibility
       sections [{type=MovieSceneByteSection, end_frame=10, start_frame=0, channels [None {type=MovieSceneScriptingByteChannel, const [4,1]}]}]
      }
     ]
    }
    BackingGold {
     id=#5
     possessed_class=Image
     tracks [
      {
       type=MovieSceneWidgetMaterialTrack
       sections [{
         type=MovieSceneParameterSection
         channels [
          Anim_SimpleColor {type=MovieSceneScriptingFloatChannel, keys [[1,0],[4,8],[25,1]]}
          Alpha_Final {type=MovieSceneScriptingFloatChannel, keys [[0,0.5],[2,0.5],[5,1],[25,0.8]]}
         ]
        }]
      }
      {
       type=MovieScene2DTransformTrack
       sections [{
         type=MovieScene2DTransformSection
         channels [
          Scale.X {type=MovieSceneScriptingFloatChannel, keys [[1,1],[2,1.25],[6.6667,1]]}
          Scale.Y {type=MovieSceneScriptingFloatChannel, keys [[1,1],[2,1.25],[6.6667,1]]}
         ]
        }]
      }
     ]
    }
    ParticlesFX {
     id=#6
     possessed_class=VerticalBox
     tracks [
      {
       type=MovieSceneFloatTrack
       display_name=Render Opacity
       sections [{type=MovieSceneFloatSection, start_frame=0, channels [None {type=MovieSceneScriptingFloatChannel, keys [[3,0],[4,1]]}]}]
      }
      {
       type=MovieSceneByteTrack
       display_name=Visibility
       sections [{type=MovieSceneByteSection, channels [None {type=MovieSceneScriptingByteChannel, const [4,2]}]}]
      }
     ]
    }
    Gradient {
     id=#7
     possessed_class=Image
     tracks [{
       type=MovieSceneWidgetMaterialTrack
       sections [{type=MovieSceneParameterSection, channels [Alpha_Final {type=MovieSceneScriptingFloatChannel, keys [[0,3],[1,5],[10,1]]}]}]
      }]
    }
    GoldStarComplete {
     id=#8
     possessed_class=Image
     tracks [
      {
       type=MovieScene2DTransformTrack
       sections [{
         type=MovieScene2DTransformSection
         channels [
          Translation.X {type=MovieSceneScriptingFloatChannel}
          Translation.Y {type=MovieSceneScriptingFloatChannel}
          Angle {type=MovieSceneScriptingFloatChannel}
          Scale.X {type=MovieSceneScriptingFloatChannel, keys [[0,1],[1,0.95],[2,1.333],[8,1]]}
          Scale.Y {type=MovieSceneScriptingFloatChannel, keys [[0,1],[1,0.95],[2,1.333],[8,1]]}
          Shear.X {type=MovieSceneScriptingFloatChannel}
          Shear.Y {type=MovieSceneScriptingFloatChannel}
         ]
        }]
      }
      {
       type=MovieSceneWidgetMaterialTrack
       sections [{
         type=MovieSceneParameterSection
         channels [
          Completed {type=MovieSceneScriptingFloatChannel, keys [[2,0],[6,1]]}
          DistortionAmount {type=MovieSceneScriptingFloatChannel, keys [[3,0],[4,4]]}
         ]
        }]
      }
     ]
    }
    RingFlare {
     id=#9
     possessed_class=Image
     tracks [
      {
       type=MovieSceneFloatTrack
       display_name=Render Opacity
       sections [{type=MovieSceneFloatSection, channels [None {type=MovieSceneScriptingFloatChannel, keys [[5,0],[6,1],[11,1],[20.6667,0]]}]}]
      }
      {
       type=MovieScene2DTransformTrack
       sections [{
         type=MovieScene2DTransformSection
         channels [
          Scale.X {type=MovieSceneScriptingFloatChannel, keys [[4,1],[19.6667,5]]}
          Scale.Y {type=MovieSceneScriptingFloatChannel, keys [[4,1],[19.6667,5]]}
         ]
        }]
      }
      {
       type=MovieSceneWidgetMaterialTrack
       sections [{type=MovieSceneParameterSection, channels [TaperGlow {type=MovieSceneScriptingFloatChannel, keys [[5,1],[20.6667,4]]}]}]
      }
      {
       type=MovieSceneByteTrack
       display_name=Visibility
       sections [{type=MovieSceneByteSection, end_frame=19, start_frame=0, channels [None {type=MovieSceneScriptingByteChannel, keys [[0,4],[19.6667,2]]}]}]
      }
     ]
    }
    Completed_Glow_Gold {
     id=#10
     possessed_class=Image
     tracks [
      {
       type=MovieSceneFloatTrack
       display_name=Render Opacity
       sections [{type=MovieSceneFloatSection, end_frame=6, start_frame=0, channels [None {type=MovieSceneScriptingFloatChannel, keys [[0,0],[2.3333,1]]}]}]
      }
      {
       type=MovieScene2DTransformTrack
       sections [{
         type=MovieScene2DTransformSection
         channels [
          Scale.X {type=MovieSceneScriptingFloatChannel, keys [[0,1.333],[1,8],[10,2]]}
          Scale.Y {type=MovieSceneScriptingFloatChannel, keys [[0,1.333],[2,2],[10,2]]}
         ]
        }]
      }
     ]
    }
    Particles_Up {
     id=#11
     possessed_class=Image
     tracks [{
       type=MovieSceneWidgetMaterialTrack
       sections [{type=MovieSceneParameterSection, channels [Anim_Scale {type=MovieSceneScriptingFloatChannel, keys [[3,1],[6,0],[10,0],[25,0.666]]}]}]
      }]
    }
   ]}
 }
 anim_completed {
  binding_count=10
  length_seconds=1.25
  object=&self:anim_completed
  sequence {
   bindings [
    BackingBase {
     id=#12
     possessed_class=Image
     tracks [
      {
       type=MovieSceneWidgetMaterialTrack
       sections [{
         type=MovieSceneParameterSection
         channels [
          Anim_SimpleColor {type=MovieSceneScriptingFloatChannel, keys [[3,0],[12,8],[75,1]]}
          Alpha_Final {type=MovieSceneScriptingFloatChannel, keys [[0,0.5],[12,0.5],[15,1],[75,0.75]]}
         ]
        }]
      }
      {
       type=MovieScene2DTransformTrack
       sections [{
         type=MovieScene2DTransformSection
         channels [
          Scale.X {type=MovieSceneScriptingFloatChannel, keys [[3,1],[6,1.25],[20,1]]}
          Scale.Y {type=MovieSceneScriptingFloatChannel, keys [[3,1],[6,1.25],[20,1]]}
         ]
        }]
      }
     ]
    }
    Completed_Glow {
     id=#13
     possessed_class=Image
     tracks [
      {
       type=MovieSceneFloatTrack
       display_name=Render Opacity
       sections [{type=MovieSceneFloatSection, end_frame=20, start_frame=0, channels [None {type=MovieSceneScriptingFloatChannel, keys [[0,0],[7,1]]}]}]
      }
      {
       type=MovieScene2DTransformTrack
       sections [{
         type=MovieScene2DTransformSection
         end_frame=30
         start_frame=0
         channels [
          Scale.X {type=MovieSceneScriptingFloatChannel, keys [[0,1.333],[7,7],[30,1.333]]}
          Scale.Y {type=MovieSceneScriptingFloatChannel, const [1.333,2]}
         ]
        }]
      }
     ]
    }
    StarComplete {
     id=#14
     possessed_class=Image
     tracks [
      {
       type=MovieScene2DTransformTrack
       sections [{
         type=MovieScene2DTransformSection
         channels [
          Translation.X {type=MovieSceneScriptingFloatChannel}
          Translation.Y {type=MovieSceneScriptingFloatChannel}
          Angle {type=MovieSceneScriptingFloatChannel}
          Scale.X {type=MovieSceneScriptingFloatChannel, keys [[0,1],[3,0.95],[6,1.333],[24,1]]}
          Scale.Y {type=MovieSceneScriptingFloatChannel, keys [[0,1],[3,0.95],[6,1.333],[24,1]]}
          Shear.X {type=MovieSceneScriptingFloatChannel}
          Shear.Y {type=MovieSceneScriptingFloatChannel}
         ]
        }]
      }
      {
       type=MovieSceneWidgetMaterialTrack
       sections [{
         type=MovieSceneParameterSection
         channels [
          HotCore_Amount {type=MovieSceneScriptingFloatChannel, keys [[0,0.5],[9,8],[45,0.5]]}
          GlowAnimateAmount {type=MovieSceneScriptingFloatChannel, keys [[0,0.25],[9,0],[45,0.25]]}
         ]
        }]
      }
     ]
    }
    ParticlesFX {
     id=#15
     possessed_class=VerticalBox
     tracks [
      {
       type=MovieSceneFloatTrack
       display_name=Render Opacity
       sections [{type=MovieSceneFloatSection, start_frame=0, channels [None {type=MovieSceneScriptingFloatChannel, keys [[9,0],[12,1],[18,1],[75,0]]}]}]
      }
      {
       type=MovieSceneByteTrack
       display_name=Visibility
       sections [{type=MovieSceneByteSection, channels [None {type=MovieSceneScriptingByteChannel, keys [[9,4],[75,1]]}]}]
      }
     ]
    }
    FlareFX {
     id=#16
     possessed_class=Overlay
     tracks [
      {
       type=MovieScene2DTransformTrack
       sections [{
         type=MovieScene2DTransformSection
         channels [
          Scale.X {type=MovieSceneScriptingFloatChannel, const [1,1]}
          Scale.Y {type=MovieSceneScriptingFloatChannel, keys [[0,1],[3,2],[8,2],[30,1]]}
         ]
        }]
      }
      {
       type=MovieSceneFloatTrack
       display_name=Render Opacity
       sections [{type=MovieSceneFloatSection, channels [None {type=MovieSceneScriptingFloatChannel, keys [[0,0],[2,1],[12,1],[30,0]]}]}]
      }
      {
       type=MovieSceneByteTrack
       display_name=Visibility
       sections [{type=MovieSceneByteSection, end_frame=30, start_frame=0, channels [None {type=MovieSceneScriptingByteChannel, keys [[0,4],[30,2]]}]}]
      }
     ]
    }
    StarBurst {
     id=#17
     possessed_class=Image
     tracks [
      {
       type=MovieSceneFloatTrack
       display_name=Render Opacity
       sections [{type=MovieSceneFloatSection, channels [None {type=MovieSceneScriptingFloatChannel, keys [[9,0],[12,1],[40,0]]}]}]
      }
      {
       type=MovieScene2DTransformTrack
       sections [{
         type=MovieScene2DTransformSection
         channels [
          Scale.X {type=MovieSceneScriptingFloatChannel, keys [[9,1],[40,3]]}
          Scale.Y {type=MovieSceneScriptingFloatChannel, keys [[9,1],[40,3]]}
         ]
        }]
      }
      {
       type=MovieSceneByteTrack
       display_name=Visibility
       sections [{type=MovieSceneByteSection, end_frame=45, start_frame=0, channels [None {type=MovieSceneScriptingByteChannel, keys [[0,4],[45,1]]}]}]
      }
     ]
    }
    Particles_Down {
     id=#18
     possessed_class=Image
     tracks [{
       type=MovieSceneWidgetMaterialTrack
       sections [{type=MovieSceneParameterSection, channels [Anim_Scale {type=MovieSceneScriptingFloatChannel, keys [[9,1],[18,0],[30,0],[75,1]]}]}]
      }]
    }
    RingFlare {
     id=#19
     possessed_class=Image
     tracks [
      {
       type=MovieSceneFloatTrack
       display_name=Render Opacity
       sections [{type=MovieSceneFloatSection, channels [None {type=MovieSceneScriptingFloatChannel, keys [[15,0],[18,1],[33,1],[62,0]]}]}]
      }
      {
       type=MovieScene2DTransformTrack
       sections [{
         type=MovieScene2DTransformSection
         channels [
          Scale.X {type=MovieSceneScriptingFloatChannel, keys [[12,1],[59,5]]}
          Scale.Y {type=MovieSceneScriptingFloatChannel, keys [[12,1],[59,5]]}
         ]
        }]
      }
      {
       type=MovieSceneWidgetMaterialTrack
       sections [{type=MovieSceneParameterSection, channels [TaperGlow {type=MovieSceneScriptingFloatChannel, keys [[15,1],[62,4]]}]}]
      }
      {
       type=MovieSceneByteTrack
       display_name=Visibility
       sections [{type=MovieSceneByteSection, end_frame=59, start_frame=0, channels [None {type=MovieSceneScriptingByteChannel, keys [[0,4],[59,2]]}]}]
      }
     ]
    }
    Gradient {
     id=#20
     possessed_class=Image
     tracks [{
       type=MovieSceneWidgetMaterialTrack
       sections [{type=MovieSceneParameterSection, channels [Alpha_Final {type=MovieSceneScriptingFloatChannel, keys [[0,3],[3,5],[30,1]]}]}]
      }]
    }
    Particles_Up {
     id=#21
     possessed_class=Image
     tracks [{
       type=MovieSceneWidgetMaterialTrack
       sections [{type=MovieSceneParameterSection, channels [Anim_Scale {type=MovieSceneScriptingFloatChannel, keys [[9,1],[18,0],[30,0],[75,1]]}]}]
      }]
    }
   ]
   display_rate {numerator=60}
  }
 }
 testgoldprogressanim {
  binding_count=2
  length_seconds=5
  object=&self:testgoldprogressanim
  sequence {bindings [
    GoldStarComplete_Glow_Radial {
     id=#22
     possessed_class=Image
     tracks [{
       type=MovieSceneWidgetMaterialTrack
       sections [{type=MovieSceneParameterSection, channels [Progress {type=MovieSceneScriptingFloatChannel, keys [[0,0],[100,1]]}]}]
      }]
    }
    GoldStarProgress {
     id=#23
     possessed_class=Image
     tracks [{
       type=MovieSceneWidgetMaterialTrack
       sections [{type=MovieSceneParameterSection, channels [Progress {type=MovieSceneScriptingFloatChannel, keys [[0,0],[100,1]]}]}]
      }]
    }
   ]}
 }
]
graphs [
 EventGraph {nodes [
   n1 = Event PreConstruct()
   > Branch(Condition=@n1.IsDesignTime)
   > SetFXColors!(execute, self, GoldFX:bool=true) -> (then)
   > RandomizeFXParameters()
   Event Construct()
   > RandomizeFXParameters()
   > SetDynamicMaterialInstances()
   > SetRenderOpacity(self=Get All() -> (All:Overlay Object Reference), InOpacity=0.5)
  ]}
 RandomizeFXParameters {nodes [
   RandomizeFXParameters() {is=title K2Node_FunctionEntry}
   > n1 = GetDynamicMaterial(self=Get Flare() -> (Flare:Image Object Reference))
   > SetScalarParameterValue(self=@n1, ParameterName=NoiseSet1_Tile_Alter, Value=RandomFloatInRange(Min=0.5, Max=1.5))
   > SetScalarParameterValue(self=@n1, ParameterName=NoiseSet2_Tile_Alter, Value=RandomFloatInRange(Min=0.5, Max=1.5))
   > n2 = GetDynamicMaterial(self=Get Particles_Up() -> (Particles_Up:Image Object Reference))
   > n3 = GetDynamicMaterial(self=Get Particles_Down() -> (Particles_Down:Image Object Reference))
   > SetScalarParameterValue(self=[@n3, @n2], ParameterName=Set1_SeedAdjustment, Value=RandomFloatInRange(Min=0.5, Max=2))
   > SetScalarParameterValue(self=[@n3, @n2], ParameterName=Set2_SeedAdjustment, Value=RandomFloatInRange(Min=0.5, Max=2))
   > n4 = GetDynamicMaterial(self=Get Completed_Glow() -> (Completed_Glow:Image Object Reference))
   > n5 = GetDynamicMaterial(self=Get Completed_Glow_Gold() -> (Completed_Glow_Gold:Image Object Reference))
   > SetScalarParameterValue(self=[@n5, @n4], ParameterName=SpeedMultiplier, Value=RandomFloatInRange(Min=18, Max=22))
   > SetRenderTransformAngle!(
    execute
    self:Widget Object Reference=[Get Completed_Glow_Gold(), Get Completed_Glow()]
    Angle:float=float * float(A=RandomFloatInRange(Min=0, Max=10), B=0.1) {is=title K2Node_PromotableOperator}
   ) -> (then)
  ]}
 ResetCompletedVisuals {nodes [
   ResetCompletedVisuals() {is=title K2Node_FunctionEntry}
   > n1 = Sequence!(execute) -> (then_0, then_1, then_2:Exec, then_3:Exec)
   > n2 = GetDynamicMaterial(self=Get BackingBase() -> (BackingBase:Image Object Reference))
   > SetScalarParameterValue(self=@n2, ParameterName=Anim_SimpleColor, Value=0)
   > SetScalarParameterValue(self=@n2, ParameterName=Alpha_Final, Value=0.5)
   SetVisibility(execute=@n1.then_3, self=Get FlareFX() -> (FlareFX:Overlay Object Reference), InVisibility=Hidden)
   > SetRenderOpacity(self=[Get ParticlesFX() -> (ParticlesFX:Vertical Box Object Reference), Get Completed_Glow_Gold()], InOpacity=0)
   > n3 = GetDynamicMaterial(self=Get GoldStarComplete() -> (GoldStarComplete:Image Object Reference))
   > SetScalarParameterValue(self=@n3, ParameterName=Completed, Value=0)
   > SetScalarParameterValue(self=@n3, ParameterName=DistortionAmount, Value=0.5)
   > SetRenderScale!(execute, self:Widget Object Reference=Get FlareFX(), Scale:Vector 2D Structure=(X=1,Y=0.25)) -> (then)
   > SetVisibility(self=Get ParticlesFX(), InVisibility=Collapsed)
   n4 = Get GoldStarCompleteMat() -> (GoldStarCompleteMat:Material Instance Dynamic Object Reference)
   > SetScalarParameterValue(self=@n4, ParameterName=Completed, Value=1)
   > SetScalarParameterValue(self=@n4, ParameterName=DistortionAmount, Value=4)
   > Set IsGoldStarComplete(IsGoldStarComplete:bool=false) -> (Output_Get:bool)
   n5 = GetDynamicMaterial(execute=@n1.then_1, self=Get BackingGold() -> (BackingGold:Image Object Reference))
   > SetScalarParameterValue(self=@n5, ParameterName=Anim_SimpleColor, Value=0)
   > SetScalarParameterValue(self=@n5, ParameterName=Alpha_Final, Value=0.5)
   SetStarProgress!(execute=@n1.then_2, self, SetStarProgress:double=0) -> (then)
   > SetVisibility(
    self=[
     Get GoldStarOverlay() -> (GoldStarOverlay:Overlay Object Reference)
     Get GoldStarComplete_Glow_Radial() -> (GoldStarComplete_Glow_Radial:Image Object Reference)
     Get GoldStarProgress() -> (GoldStarProgress:Image Object Reference)
     Get Completed_Glow()
     Get Star_Incomplete() -> (Star_Incomplete:Image Object Reference)
     Get Image_Timer_ProgressBar() -> (Image_Timer_ProgressBar:Image Object Reference)
     Get StarProgress() -> (StarProgress:Radial Slider Object Reference)
     Get GoldStarComplete()
     Get CompletedStarOverlay() -> (CompletedStarOverlay:Overlay Object Reference)
     Get Completed_Glow_Gold()
     Get StarBurst_Gold() -> (StarBurst_Gold:Image Object Reference)
    ]
    InVisibility=Hidden
   )
   > SetVisibility(self=Get Star_BG() -> (Star_BG:Image Object Reference), InVisibility=SelfHitTestInvisible)
   > SetRenderOpacity(self=Get All(), InOpacity=0.5)
   > SetActiveWidgetIndex(self=Get BackingTypeSwitcher() -> (BackingTypeSwitcher:Common Activatable Widget Switcher Object Reference), Index=0)
   n6 = Get StarBurst() -> (StarBurst:Image Object Reference)
   > SetRenderScale(self=@n6, Scale=(X=1,Y=1))
   > SetVisibility(self=@n6, InVisibility=Collapsed)
   > SetRenderScale(self=Get Completed_Glow(), Scale=(X=1.333,Y=1.333))
  ]}
 Set Gold Star Progress {nodes [
   n1 = Set Gold Star Progress!() -> (then, Progress:double=~) {is=title K2Node_FunctionEntry}
   > SetScalarParameterValue(
    self=[
     Get GoldStarProgressMat() -> (GoldStarProgressMat:Material Instance Dynamic Object Reference)
     Get GoldStarGlowMat() -> (GoldStarGlowMat:Material Instance Dynamic Object Reference)
    ]
    ParameterName=Progress
    Value=@n1
   )
   > SetVisibility(self=[Get GoldStarProgress(), Get GoldStarComplete_Glow_Radial(), Get GoldStarOverlay()], InVisibility=SelfHitTestInvisible)
   > SetActiveWidgetIndex(self=Get BackingTypeSwitcher(), Index=1)
  ]}
 SetComplete {nodes [
   n1 = GetEndTime(self=Get anim_completed() -> (anim_completed:Widget Animation Object Reference))
   n2 = PlayAnimationTimeRange!(
    execute
    self:User Widget Object Reference
    InAnimation:Widget Animation Object Reference=Get anim_completed()
    StartAtTime:float=@n1
    EndAtTime:float=@n1
    NumLoopsToPlay:int=1
    PlayMode:EUMGSequencePlayMode Enum
    PlaybackSpeed:float=1
    bRestoreState:bool=false
   ) -> (then, ReturnValue:Widget Animation Handle Structure)
   SetRenderOpacity(execute=[@n2.then, @n3.then], self=Get All(), InOpacity=1)
   SetComplete() {is=title K2Node_FunctionEntry}
   > Branch(Condition=NOT Boolean(A=Get IsGoldStarComplete() -> (IsGoldStarComplete:bool)))
   > SetFXColors(GoldFX=false)
   > SetVisibility(self=[Get StarProgress(), Get Image_Timer_ProgressBar(), Get Star_Incomplete(), Get Star_BG()], InVisibility=Hidden)
   > SetVisibility(self=[Get Completed_Glow(), Get CompletedStarOverlay()], InVisibility=HitTestInvisible)
   > Branch(Condition=Get ShowCompleteAnim() -> (ShowCompleteAnim:bool))
   > n3 = PlayAnimation(InAnimation=Get anim_completed())
  ]}
 SetDynamicMaterialInstances {nodes [
   SetDynamicMaterialInstances() {is=title K2Node_FunctionEntry}
   > n1 = GetDynamicMaterial(self=Get Image_Timer_ProgressBar())
   > Set StarProgressMat(StarProgressMat:Material Instance Dynamic Object Reference=@n1) -> (Output_Get:Material Instance Dynamic Object Reference)
   > n2 = GetDynamicMaterial(self=Get Star_BG())
   > Set StarBGMat(StarBGMat:Material Instance Dynamic Object Reference=@n2) -> (Output_Get:Material Instance Dynamic Object Reference)
   > n3 = GetDynamicMaterial(self=Get Star_Incomplete())
   > Set StarIncompleteMat(StarIncompleteMat:Material Instance Dynamic Object Reference=@n3) -> (Output_Get:Material Instance Dynamic Object Reference)
   > n4 = GetDynamicMaterial(self=Get StarComplete() -> (StarComplete:Image Object Reference))
   > Set StarCompleteMat(StarCompleteMat:Material Instance Dynamic Object Reference=@n4) -> (Output_Get:Material Instance Dynamic Object Reference)
   > n5 = GetDynamicMaterial(self=Get GoldStarComplete())
   > Set GoldStarCompleteMat(GoldStarCompleteMat:Material Instance Dynamic Object Reference=@n5) -> (Output_Get:Material Instance Dynamic Object Reference)
   > n6 = GetDynamicMaterial(self=Get GoldStarProgress())
   > Set GoldStarProgressMat(GoldStarProgressMat:Material Instance Dynamic Object Reference=@n6) -> (Output_Get:Material Instance Dynamic Object Reference)
   > n7 = GetDynamicMaterial(self=Get GoldStarComplete_Glow_Radial())
   > Set GoldStarGlowMat(GoldStarGlowMat:Material Instance Dynamic Object Reference=@n7) -> (Output_Get:Material Instance Dynamic Object Reference)
  ]}
 SetFXColors {nodes [
   n1 = SetFXColors!() -> (then, GoldFX:bool=~) {is=title K2Node_FunctionEntry}
   > SetBrushFromMaterial(
    self=Get Flare()
    Material=Select(
     Option 0:Material Interface Object Reference=/PilgrimCore/Widgets/Materials/MI_UI_Pilgrim_StarComplete_Flare
     Option 1:Material Interface Object Reference=/PilgrimCore/Widgets/Materials/MI_UI_Pilgrim_StarComplete_Flare_Gold
     Index=@n1
    ) -> (ReturnValue:Material Interface Object Reference)
   )
   > SetBrushFromMaterial(
    self=Get Gradient() -> (Gradient:Image Object Reference)
    Material=Select(
     Option 0=/PilgrimCore/Widgets/Materials/MI_UI_Pilgrim_StarComplete_GradientBurst
     Option 1=/PilgrimCore/Widgets/Materials/MI_UI_Pilgrim_StarComplete_GradientBurst_Gold
     Index=@n1
    )
   )
   > SetBrushFromMaterial(
    self=Get Particles_Up()
    Material=Select(
     Option 0=/PilgrimCore/Widgets/Materials/MI_UI_Piligrim_SquareParticles
     Option 1=/PilgrimCore/Widgets/Materials/MI_UI_Piligrim_SquareParticles_Gold
     Index=@n1
    )
   )
   > SetBrushFromMaterial(
    self=Get Particles_Down()
    Material=Select(
     Option 0=/PilgrimCore/Widgets/Materials/MI_UI_Piligrim_SquareParticles_Down
     Option 1=/PilgrimCore/Widgets/Materials/MI_UI_Piligrim_SquareParticles_Gold_Down
     Index=@n1
    )
   )
   > SetBrushFromMaterial(
    self=Get RingFlare() -> (RingFlare:Image Object Reference)
    Material=Select(
     Option 0=/PilgrimCore/Widgets/Materials/MI_UI_Piligrim_GemBreak_Pulse
     Option 1=/PilgrimCore/Widgets/Materials/MI_UI_Piligrim_GemBreak_Pulse_Gold
     Index=@n1
    )
   )
  ]}
 SetGoldStarComplete {nodes [
   SetGoldStarComplete() {is=title K2Node_FunctionEntry}
   > Branch(Condition=NOT Boolean(A=Get IsGoldStarComplete()))
   > SetFXColors(GoldFX=true) {is=title K2Node_CallFunction}
   > RandomizeFXParameters() {is=title K2Node_CallFunction}
   > SetActiveWidgetIndex(self=Get BackingTypeSwitcher(), Index=1)
   > SetVisibility(self=Get StarProgress(), InVisibility=Hidden)
   > SetVisibility(self=Get CompletedStarOverlay(), InVisibility=Hidden)
   > SetVisibility(self=Get GoldStarProgress(), InVisibility=Hidden)
   > SetVisibility(
    self=[Get GoldStarOverlay(), Get GoldStarComplete_Glow_Radial(), Get GoldStarComplete(), Get Completed_Glow_Gold(), Get StarBurst_Gold()]
    InVisibility=SelfHitTestInvisible
   )
   > n1 = Sequence!(execute) -> (then_0, then_1)
   > n2 = Branch(Condition=Get ShowCompleteAnim())
   > PlayAnimation(InAnimation=Get anim_gold_completed() -> (anim_gold_completed:Widget Animation Object Reference))
   n3 = Get FlareFX()
   SetVisibility(execute=@n2.else, self=@n3, InVisibility=HitTestInvisible)
   > SetRenderScale(self=@n3, Scale=(X=1,Y=0.25))
   > SetRenderOpacity(self=@n3, InOpacity=1)
   > n4 = GetDynamicMaterial(self=Get Gradient())
   > SetScalarParameterValue(self=@n4, ParameterName=Alpha_Final, Value=1)
   n5 = Get GoldStarCompleteMat()
   > SetScalarParameterValue(self=@n5, ParameterName=Alpha_Final, Value=0.25)
   > SetScalarParameterValue(self=@n5, ParameterName=Completed, Value=1)
   > SetScalarParameterValue(self=@n5, ParameterName=DistortionAmount, Value=4)
   > n6 = GetDynamicMaterial(self=Get BackingGold())
   > SetScalarParameterValue(self=@n6, ParameterName=Anim_SimpleColor, Value=1)
   > SetScalarParameterValue(self=@n6, ParameterName=Alpha_Final, Value=0.8)
   > n7 = GetDynamicMaterial(self=Get Particles_Up())
   > n8 = GetDynamicMaterial(self=Get Particles_Down())
   > SetScalarParameterValue(self=[@n8, @n7], ParameterName=Anim_Scale, Value=0.666)
   > SetVisibility(self=Get ParticlesFX(), InVisibility=HitTestInvisible)
   > SetRenderOpacity(self=Get ParticlesFX(), InOpacity=1)
   n9 = Get Completed_Glow_Gold()
   > SetRenderOpacity(self=@n9, InOpacity=1)
   > SetRenderScale(self=@n9, Scale=(X=2,Y=2))
   > n10 = GetDynamicMaterial(self=Get GoldStarComplete_Glow_Radial())
   > SetScalarParameterValue(self=@n10, ParameterName=Alpha_Final, Value=0.25)
   Set IsGoldStarComplete(execute=@n1.then_1, IsGoldStarComplete=true)
  ]}
 SetStarProgress {nodes [
   n1 = SetStarProgress!() -> (then, SetStarProgress:double=~) {is=title K2Node_FunctionEntry}
   > Branch(Condition=NOT Boolean(A=Get IsGoldStarComplete()))
   > SetScalarParameterValue(self=Get StarProgressMat() -> (StarProgressMat:Material Instance Dynamic Object Reference), ParameterName=Progress, Value=@n1)
   > SetVisibility(self=[Get Image_Timer_ProgressBar(), Get Star_Incomplete()], InVisibility=HitTestInvisible)
   > SetVisibility(self=Get Star_BG(), InVisibility=Hidden)
   > SetRenderOpacity(self=Get All(), InOpacity=1)
  ]}
 SetupStaggeredAnimationVisibiltyStates {nodes [
   SetupStaggeredAnimationVisibiltyStates() {is=title K2Node_FunctionEntry}
   > SetVisibility(self=[Get CompletedStarOverlay(), Get GoldStarOverlay()], InVisibility=SelfHitTestInvisible)
   > SetVisibility(self=Get Star_BG(), InVisibility=Hidden)
  ]}
]
variables [
 GoldStarCompleteMat {type=MaterialInstanceDynamic}
 GoldStarGlowMat {type=MaterialInstanceDynamic}
 GoldStarProgressMat {type=MaterialInstanceDynamic}
 IsGoldFX {type=boolean}
 IsGoldStarComplete {type=boolean}
 ShowCompleteAnim {type=boolean}
 StarBGMat {type=MaterialInstanceDynamic}
 StarCompleteMat {type=MaterialInstanceDynamic}
 StarIncompleteMat {type=MaterialInstanceDynamic}
 StarProgressMat {type=MaterialInstanceDynamic}
 Value {type=number}
]
widgets {
 info {inheritedWidgetCount=0, parentClass=&UserWidget, rootWidgetClass=&SizeBox, widgetCount=30}
 widgets [
  SizeBox_0 {
   widgetClassPath=&SizeBox
   widget=SizeBox_0
   properties {bOverride_HeightOverride=true, bOverride_WidthOverride=true, heightOverride=70, widthOverride=70}
  }
  All {widgetClassPath=&Overlay, bIsVariable=true, parent=SizeBox_0, widget=All, slot {class=SizeBoxSlot}}
  BackingsOverlay {
   widgetClassPath=&Overlay
   parent=All
   widget=BackingsOverlay
   slot {class=OverlaySlot, properties {horizontalAlignment=HAlign_Fill, verticalAlignment=VAlign_Fill}}
  }
  Shadow {
   widgetClassPath=&Image
   parent=BackingsOverlay
   widget=Shadow
   properties {
    visibility=SelfHitTestInvisible
    brush {
     bIsDynamicallyLoaded=false
     drawAs=Image
     imageType=NoImage
     mirroring=NoMirror
     resourceName=None
     resourceObject=/PilgrimCore/Widgets/Materials/M_UI_Pilgrim_RoundedShadow_Round_Star
     tiling=NoTile
     imageSize=<72 72>
     margin=<ltrb 0 0 0 0>
     outlineSettings {
      bUseBrushTransparency=false
      roundingType=HalfHeightRadius
      width=0
      color {colorUseRule=UseColor_Specified, specifiedColor=<rgba 0 0 0 0>}
      cornerRadii=<0 0 0 0>
     }
     tintColor {colorUseRule=UseColor_Specified, specifiedColor=<rgba 1 1 1 1>}
     uVRegion {bIsValid=false, max=<0 0>, min=<0 0>}
    }
    renderTransform {angle=0, scale=<1.2 1.22>, shear=<0 0>, translation=<0 0>}
   }
   slot {class=OverlaySlot}
  }
  BackingTypeSwitcher {
   widgetClassPath=&CommonActivatableWidgetSwitcher
   bIsVariable=true
   parent=BackingsOverlay
   widget=BackingTypeSwitcher
   properties {transitionDuration=1}
   slot {class=OverlaySlot, properties {horizontalAlignment=HAlign_Fill, verticalAlignment=VAlign_Fill}}
  }
  BackingBase {
   widgetClassPath=&Image
   bIsVariable=true
   parent=BackingTypeSwitcher
   widget=BackingBase
   properties {
    visibility=SelfHitTestInvisible
    brush {
     bIsDynamicallyLoaded=false
     drawAs=Image
     imageType=NoImage
     mirroring=NoMirror
     resourceName=None
     resourceObject=/PilgrimCore/Widgets/Materials/M_UI_Pilgrim_HUD_StarBacking
     tiling=NoTile
     imageSize=<72 72>
     margin=<ltrb 0 0 0 0>
     outlineSettings {
      bUseBrushTransparency=false
      roundingType=HalfHeightRadius
      width=0
      color {colorUseRule=UseColor_Specified, specifiedColor=<rgba 0 0 0 0>}
      cornerRadii=<0 0 0 0>
     }
     tintColor {colorUseRule=UseColor_Specified, specifiedColor=<rgba 1 1 1 1>}
     uVRegion {bIsValid=false, max=<0 0>, min=<0 0>}
    }
   }
   slot {class=WidgetSwitcherSlot}
  }
  BackingGold {
   widgetClassPath=&Image
   bIsVariable=true
   parent=BackingTypeSwitcher
   widget=BackingGold
   properties {
    visibility=SelfHitTestInvisible
    brush {
     bIsDynamicallyLoaded=false
     drawAs=Image
     imageType=NoImage
     mirroring=NoMirror
     resourceName=None
     resourceObject=/PilgrimCore/Widgets/Materials/MI_UI_Pilgrim_HUD_StarBacking_Gold
     tiling=NoTile
     imageSize=<72 72>
     margin=<ltrb 0 0 0 0>
     outlineSettings {
      bUseBrushTransparency=false
      roundingType=HalfHeightRadius
      width=0
      color {colorUseRule=UseColor_Specified, specifiedColor=<rgba 0 0 0 0>}
      cornerRadii=<0 0 0 0>
     }
     tintColor {colorUseRule=UseColor_Specified, specifiedColor=<rgba 1 1 1 1>}
     uVRegion {bIsValid=false, max=<0 0>, min=<0 0>}
    }
   }
   slot {class=WidgetSwitcherSlot}
  }
  RingFlare {
   widgetClassPath=&Image
   bIsVariable=true
   parent=BackingsOverlay
   widget=RingFlare
   properties {
    visibility=Hidden
    brush {
     bIsDynamicallyLoaded=false
     drawAs=Image
     imageType=NoImage
     mirroring=NoMirror
     resourceName=None
     resourceObject=&MI_UI_Piligrim_GemBreak_Pulse
     tiling=NoTile
     imageSize=<32 32>
     margin=<ltrb 0 0 0 0>
     outlineSettings {
      bUseBrushTransparency=false
      roundingType=HalfHeightRadius
      width=0
      color {colorUseRule=UseColor_Specified, specifiedColor=<rgba 0 0 0 0>}
      cornerRadii=<0 0 0 0>
     }
     tintColor {colorUseRule=UseColor_Specified, specifiedColor=<rgba 1 1 1 1>}
     uVRegion {bIsValid=false, max=<0 0>, min=<0 0>}
    }
   }
   slot {class=OverlaySlot, properties {horizontalAlignment=HAlign_Fill, verticalAlignment=VAlign_Fill}}
  }
  FlareFX {
   widgetClassPath=&Overlay
   bIsVariable=true
   parent=All
   widget=FlareFX
   properties {visibility=Hidden, renderTransform {angle=0, scale=<1 0.25>, shear=<0 0>, translation=<0 0>}}
   slot {class=OverlaySlot, properties {horizontalAlignment=HAlign_Fill, verticalAlignment=VAlign_Fill}}
  }
  BeamShadow {
   widgetClassPath=&Image
   parent=FlareFX
   widget=BeamShadow
   properties {
    visibility=SelfHitTestInvisible
    brush {
     bIsDynamicallyLoaded=false
     drawAs=Image
     imageType=NoImage
     mirroring=NoMirror
     resourceName=None
     resourceObject=/PilgrimCore/Widgets/Materials/MI_UI_Pilgrim_StarComplete_GradientBurst_Shadow
     tiling=NoTile
     imageSize=<64 48>
     margin=<ltrb 0 0 0 0>
     outlineSettings {
      bUseBrushTransparency=false
      roundingType=HalfHeightRadius
      width=0
      color {colorUseRule=UseColor_Specified, specifiedColor=<rgba 0 0 0 0>}
      cornerRadii=<0 0 0 0>
     }
     tintColor {colorUseRule=UseColor_Specified, specifiedColor=<rgba 1 1 1 1>}
     uVRegion {bIsValid=false, max=<0 0>, min=<0 0>}
    }
    renderTransform {angle=90, scale=<12 2>, shear=<0 0>, translation=<0 0>}
   }
   slot {class=OverlaySlot, properties {horizontalAlignment=HAlign_Center, verticalAlignment=VAlign_Center}}
  }
  Gradient {
   widgetClassPath=&Image
   bIsVariable=true
   parent=FlareFX
   widget=Gradient
   properties {
    visibility=SelfHitTestInvisible
    brush {
     bIsDynamicallyLoaded=false
     drawAs=Image
     imageType=NoImage
     mirroring=NoMirror
     resourceName=None
     resourceObject=&MI_UI_Pilgrim_StarComplete_GradientBurst
     tiling=NoTile
     imageSize=<64 48>
     margin=<ltrb 0 0 0 0>
     outlineSettings {
      bUseBrushTransparency=false
      roundingType=HalfHeightRadius
      width=0
      color {colorUseRule=UseColor_Specified, specifiedColor=<rgba 0 0 0 0>}
      cornerRadii=<0 0 0 0>
     }
     tintColor {colorUseRule=UseColor_Specified, specifiedColor=<rgba 1 1 1 1>}
     uVRegion {bIsValid=false, max=<0 0>, min=<0 0>}
    }
    renderTransform {angle=90, scale=<10 1>, shear=<0 0>, translation=<0 0>}
   }
   slot {class=OverlaySlot, properties {horizontalAlignment=HAlign_Center, verticalAlignment=VAlign_Center}}
  }
  Flare {
   widgetClassPath=&Image
   bIsVariable=true
   parent=FlareFX
   widget=Flare
   properties {
    visibility=SelfHitTestInvisible
    brush {
     bIsDynamicallyLoaded=false
     drawAs=Image
     imageType=NoImage
     mirroring=NoMirror
     resourceName=None
     resourceObject=&MI_UI_Pilgrim_StarComplete_Flare
     tiling=NoTile
     imageSize=<64 32>
     margin=<ltrb 0 0 0 0>
     outlineSettings {
      bUseBrushTransparency=false
      roundingType=HalfHeightRadius
      width=0
      color {colorUseRule=UseColor_Specified, specifiedColor=<rgba 0 0 0 0>}
      cornerRadii=<0 0 0 0>
     }
     tintColor {colorUseRule=UseColor_Specified, specifiedColor=<rgba 1 1 1 1>}
     uVRegion {bIsValid=false, max=<0 0>, min=<0 0>}
    }
    renderTransform {angle=90, scale=<10 1>, shear=<0 0>, translation=<0 0>}
   }
   slot {class=OverlaySlot, properties {horizontalAlignment=HAlign_Center, verticalAlignment=VAlign_Center}}
  }
  FortVisualAttachment_204 {
   widgetClassPath=&FortVisualAttachment
   parent=All
   widget=FortVisualAttachment_204
   properties {
    bOverride_HeightOverride=true
    bOverride_WidthOverride=true
    heightOverride=100
    widthOverride=64
    contentAnchor=<0.5 0.5>
    renderTransform {angle=0, scale=<1 1.25>, shear=<0 0>, translation=<0 0>}
   }
   slot {class=OverlaySlot, properties {horizontalAlignment=HAlign_Center, verticalAlignment=VAlign_Center}}
  }
  ParticlesFX {
   widgetClassPath=&VerticalBox
   bIsVariable=true
   parent=FortVisualAttachment_204
   widget=ParticlesFX
   properties {visibility=Collapsed}
   slot {class=SizeBoxSlot, properties {horizontalAlignment=HAlign_Center}}
  }
  Particles_Up {
   widgetClassPath=&Image
   bIsVariable=true
   parent=ParticlesFX
   widget=Particles_Up
   properties {
    visibility=SelfHitTestInvisible
    brush {
     bIsDynamicallyLoaded=false
     drawAs=Image
     imageType=NoImage
     mirroring=NoMirror
     resourceName=None
     resourceObject=&MI_UI_Piligrim_SquareParticles
     tiling=NoTile
     imageSize=<48 32>
     margin=<ltrb 0 0 0 0>
     outlineSettings {
      bUseBrushTransparency=false
      roundingType=HalfHeightRadius
      width=0
      color {colorUseRule=UseColor_Specified, specifiedColor=<rgba 0 0 0 0>}
      cornerRadii=<0 0 0 0>
     }
     tintColor {colorUseRule=UseColor_Specified, specifiedColor=<rgba 1 1 1 1>}
     uVRegion {bIsValid=false, max=<0 0>, min=<0 0>}
    }
    renderTransformPivot=<0.5 1>
   }
   slot {class=VerticalBoxSlot, properties {size {sizeRule=Fill, value=1}}}
  }
  Particles_Down {
   widgetClassPath=&Image
   bIsVariable=true
   parent=ParticlesFX
   widget=Particles_Down
   properties {
    visibility=SelfHitTestInvisible
    brush {
     bIsDynamicallyLoaded=false
     drawAs=Image
     imageType=NoImage
     mirroring=NoMirror
     resourceName=None
     resourceObject=&MI_UI_Piligrim_SquareParticles_Down
     tiling=NoTile
     imageSize=<48 32>
     margin=<ltrb 0 0 0 0>
     outlineSettings {
      bUseBrushTransparency=false
      roundingType=HalfHeightRadius
      width=0
      color {colorUseRule=UseColor_Specified, specifiedColor=<rgba 0 0 0 0>}
      cornerRadii=<0 0 0 0>
     }
     tintColor {colorUseRule=UseColor_Specified, specifiedColor=<rgba 1 1 1 1>}
     uVRegion {bIsValid=false, max=<0 0>, min=<0 0>}
    }
    renderTransform {angle=0, scale=<-1 1>, shear=<0 0>, translation=<0 0>}
   }
   slot {class=VerticalBoxSlot, properties {size {sizeRule=Fill, value=1}}}
  }
  Image_Timer_ProgressBar {
   widgetClassPath=&Image
   bIsVariable=true
   parent=All
   widget=Image_Timer_ProgressBar
   properties {
    visibility=Hidden
    brush {
     bIsDynamicallyLoaded=false
     drawAs=Image
     imageType=NoImage
     mirroring=NoMirror
     resourceName=None
     resourceObject=/PilgrimCore/Widgets/Materials/M_UI_Pilgrim_HUD_StarBar
     tiling=NoTile
     imageSize=<64 64>
     margin=<ltrb 0 0 0 0>
     outlineSettings {
      bUseBrushTransparency=false
      roundingType=HalfHeightRadius
      width=0
      color {colorUseRule=UseColor_Specified, specifiedColor=<rgba 0 0 0 0>}
      cornerRadii=<0 0 0 0>
     }
     tintColor {colorUseRule=UseColor_Specified, specifiedColor=<rgba 1 1 1 1>}
     uVRegion {bIsValid=false, max=<0 0>, min=<0 0>}
    }
   }
   slot {class=OverlaySlot, properties {horizontalAlignment=HAlign_Center, verticalAlignment=VAlign_Center}}
  }
  Star_BG {
   widgetClassPath=&Image
   bIsVariable=true
   parent=All
   widget=Star_BG
   properties {
    visibility=HitTestInvisible
    brush {
     bIsDynamicallyLoaded=false
     drawAs=Image
     imageType=NoImage
     mirroring=NoMirror
     resourceName=None
     resourceObject=/PilgrimCore/Widgets/Materials/MI_UI_Pilgrim_Star_NotStarted
     tiling=NoTile
     imageSize=<50 50>
     margin=<ltrb 0 0 0 0>
     outlineSettings {
      bUseBrushTransparency=false
      roundingType=HalfHeightRadius
      width=0
      color {colorUseRule=UseColor_Specified, specifiedColor=<rgba 0 0 0 0>}
      cornerRadii=<0 0 0 0>
     }
     tintColor {colorUseRule=UseColor_Specified, specifiedColor=<rgba 1 1 1 1>}
     uVRegion {bIsValid=false, max=<0 0>, min=<0 0>}
    }
    renderTransform {angle=0, scale=<1 1>, shear=<0 0>, translation=<0 -2>}
   }
   slot {class=OverlaySlot, properties {horizontalAlignment=HAlign_Center, verticalAlignment=VAlign_Center}}
  }
  Star_Incomplete {
   widgetClassPath=&Image
   bIsVariable=true
   parent=All
   widget=Star_Incomplete
   properties {
    visibility=Hidden
    brush {
     bIsDynamicallyLoaded=false
     drawAs=Image
     imageType=NoImage
     mirroring=NoMirror
     resourceName=None
     resourceObject=/PilgrimCore/Widgets/Materials/MI_UI_Pilgrim_Star_Incomplete
     tiling=NoTile
     imageSize=<50 50>
     margin=<ltrb 0 0 0 0>
     outlineSettings {
      bUseBrushTransparency=false
      roundingType=HalfHeightRadius
      width=0
      color {colorUseRule=UseColor_Specified, specifiedColor=<rgba 0 0 0 0>}
      cornerRadii=<0 0 0 0>
     }
     tintColor {colorUseRule=UseColor_Specified, specifiedColor=<rgba 1 1 1 1>}
     uVRegion {bIsValid=false, max=<0 0>, min=<0 0>}
    }
    renderTransform {angle=0, scale=<1 1>, shear=<0 0>, translation=<0 -2>}
   }
   slot {class=OverlaySlot, properties {horizontalAlignment=HAlign_Center, verticalAlignment=VAlign_Center}}
  }
  CompletedStarOverlay {
   widgetClassPath=&Overlay
   bIsVariable=true
   parent=All
   widget=CompletedStarOverlay
   properties {visibility=Hidden}
   slot {class=OverlaySlot, properties {horizontalAlignment=HAlign_Center, verticalAlignment=VAlign_Center, padding=<ltrb 1 1 1 1>}}
  }
  Completed_Glow {
   widgetClassPath=&Image
   bIsVariable=true
   parent=CompletedStarOverlay
   widget=Completed_Glow
   properties {
    visibility=SelfHitTestInvisible
    brush {
     bIsDynamicallyLoaded=false
     drawAs=Image
     imageType=NoImage
     mirroring=NoMirror
     resourceName=None
     resourceObject=/PilgrimCore/Widgets/Materials/MI_Pilgrim_LightRays_Star
     tiling=NoTile
     imageSize=<32 32>
     margin=<ltrb 0 0 0 0>
     outlineSettings {
      bUseBrushTransparency=false
      roundingType=HalfHeightRadius
      width=0
      color {colorUseRule=UseColor_Specified, specifiedColor=<rgba 0 0 0 0>}
      cornerRadii=<0 0 0 0>
     }
     tintColor {colorUseRule=UseColor_Specified, specifiedColor=<rgba 1 1 1 1>}
     uVRegion {bIsValid=false, max=<0 0>, min=<0 0>}
    }
    renderTransform {angle=0, scale=<1.333 1.333>, shear=<0 0>, translation=<0 0>}
   }
   slot {class=OverlaySlot, properties {horizontalAlignment=HAlign_Fill, verticalAlignment=VAlign_Fill}}
  }
  StarBurst {
   widgetClassPath=&Image
   bIsVariable=true
   parent=CompletedStarOverlay
   widget=StarBurst
   properties {
    visibility=Collapsed
    brush {
     bIsDynamicallyLoaded=false
     drawAs=Image
     imageType=NoImage
     mirroring=NoMirror
     resourceName=None
     resourceObject=/PilgrimCore/Widgets/Materials/MI_UI_Pilgrim_Star_Outline
     tiling=NoTile
     imageSize=<68 68>
     margin=<ltrb 0 0 0 0>
     outlineSettings {
      bUseBrushTransparency=false
      roundingType=HalfHeightRadius
      width=0
      color {colorUseRule=UseColor_Specified, specifiedColor=<rgba 0 0 0 0>}
      cornerRadii=<0 0 0 0>
     }
     tintColor {colorUseRule=UseColor_Specified, specifiedColor=<rgba 1 1 1 1>}
     uVRegion {bIsValid=false, max=<0 0>, min=<0 0>}
    }
    renderTransform {angle=0, scale=<1 1>, shear=<0 0>, translation=<0 -6>}
   }
   slot {class=OverlaySlot, properties {horizontalAlignment=HAlign_Center, verticalAlignment=VAlign_Center}}
  }
  StarComplete {
   widgetClassPath=&Image
   bIsVariable=true
   parent=CompletedStarOverlay
   widget=StarComplete
   properties {
    visibility=SelfHitTestInvisible
    brush {
     bIsDynamicallyLoaded=false
     drawAs=Image
     imageType=NoImage
     mirroring=NoMirror
     resourceName=None
     resourceObject=/PilgrimCore/Widgets/Materials/MI_UI_Pilgrim_Star
     tiling=NoTile
     imageSize=<68 68>
     margin=<ltrb 0 0 0 0>
     outlineSettings {
      bUseBrushTransparency=false
      roundingType=HalfHeightRadius
      width=0
      color {colorUseRule=UseColor_Specified, specifiedColor=<rgba 0 0 0 0>}
      cornerRadii=<0 0 0 0>
     }
     tintColor {colorUseRule=UseColor_Specified, specifiedColor=<rgba 1 1 1 1>}
     uVRegion {bIsValid=false, max=<0 0>, min=<0 0>}
    }
    renderTransform {angle=0, scale=<1 1>, shear=<0 0>, translation=<0 -2>}
   }
   slot {class=OverlaySlot, properties {horizontalAlignment=HAlign_Center, verticalAlignment=VAlign_Center}}
  }
  StarProgress {
   widgetClassPath=/Script/AdvancedWidgets.RadialSlider
   bIsVariable=true
   parent=All
   widget=StarProgress
   properties {
    angularOffset=180
    showSliderHandle=false
    sliderHandleEndAngle=360
    sliderHandleStartAngle=0
    value=0.424
    visibility=Hidden
    sliderBarColor=<rgba 0 0 0 0>
    sliderHandleColor=<rgba 1 1 1 0>
    sliderProgressColor=<rgba 0.0044 0.3712 0.8308 1>
   }
   slot {class=OverlaySlot, properties {horizontalAlignment=HAlign_Fill, verticalAlignment=VAlign_Fill}}
  }
  GoldStarOverlay {
   widgetClassPath=&Overlay
   bIsVariable=true
   parent=All
   widget=GoldStarOverlay
   properties {visibility=Hidden}
   slot {class=OverlaySlot, properties {horizontalAlignment=HAlign_Fill, verticalAlignment=VAlign_Fill, padding=<ltrb 1 1 1 1>}}
  }
  GoldStarComplete_Glow_Radial {
   widgetClassPath=&Image
   bIsVariable=true
   parent=GoldStarOverlay
   widget=GoldStarComplete_Glow_Radial
   properties {
    visibility=Hidden
    brush {
     bIsDynamicallyLoaded=false
     drawAs=Image
     imageType=NoImage
     mirroring=NoMirror
     resourceName=None
     resourceObject=/PilgrimCore/Widgets/Materials/MI_UI_Pilgrim_GoldStar_Glow
     tiling=NoTile
     imageSize=<64 64>
     margin=<ltrb 0 0 0 0>
     outlineSettings {
      bUseBrushTransparency=false
      roundingType=HalfHeightRadius
      width=0
      color {colorUseRule=UseColor_Specified, specifiedColor=<rgba 0 0 0 0>}
      cornerRadii=<0 0 0 0>
     }
     tintColor {colorUseRule=UseColor_Specified, specifiedColor=<rgba 1 1 1 1>}
     uVRegion {bIsValid=false, max=<0 0>, min=<0 0>}
    }
    renderTransform {angle=0, scale=<1.75 1.75>, shear=<0 0>, translation=<0 0>}
   }
   slot {class=OverlaySlot, properties {horizontalAlignment=HAlign_Fill, verticalAlignment=VAlign_Fill, padding=<ltrb 1 1 1 1>}}
  }
  Completed_Glow_Gold {
   widgetClassPath=&Image
   bIsVariable=true
   parent=GoldStarOverlay
   widget=Completed_Glow_Gold
   properties {
    visibility=Hidden
    brush {
     bIsDynamicallyLoaded=false
     drawAs=Image
     imageType=NoImage
     mirroring=NoMirror
     resourceName=None
     resourceObject=/PilgrimCore/Widgets/Materials/MI_Pilgrim_LightRays_Star_Gold
     tiling=NoTile
     imageSize=<32 32>
     margin=<ltrb 0 0 0 0>
     outlineSettings {
      bUseBrushTransparency=false
      roundingType=HalfHeightRadius
      width=0
      color {colorUseRule=UseColor_Specified, specifiedColor=<rgba 0 0 0 0>}
      cornerRadii=<0 0 0 0>
     }
     tintColor {colorUseRule=UseColor_Specified, specifiedColor=<rgba 1 1 1 1>}
     uVRegion {bIsValid=false, max=<0 0>, min=<0 0>}
    }
    renderTransform {angle=0, scale=<1.333 1.333>, shear=<0 0>, translation=<0 0>}
   }
   slot {class=OverlaySlot, properties {horizontalAlignment=HAlign_Fill, verticalAlignment=VAlign_Fill}}
  }
  StarBurst_Gold {
   widgetClassPath=&Image
   bIsVariable=true
   parent=GoldStarOverlay
   widget=StarBurst_Gold
   properties {
    visibility=Collapsed
    brush {
     bIsDynamicallyLoaded=false
     drawAs=Image
     imageType=NoImage
     mirroring=NoMirror
     resourceName=None
     resourceObject=/PilgrimCore/Widgets/Materials/MI_UI_Pilgrim_Star_Outline_Gold
     tiling=NoTile
     imageSize=<68 68>
     margin=<ltrb 0 0 0 0>
     outlineSettings {
      bUseBrushTransparency=false
      roundingType=HalfHeightRadius
      width=0
      color {colorUseRule=UseColor_Specified, specifiedColor=<rgba 0 0 0 0>}
      cornerRadii=<0 0 0 0>
     }
     tintColor {colorUseRule=UseColor_Specified, specifiedColor=<rgba 1 1 1 1>}
     uVRegion {bIsValid=false, max=<0 0>, min=<0 0>}
    }
    renderTransform {angle=0, scale=<1 1>, shear=<0 0>, translation=<0 -6>}
   }
   slot {class=OverlaySlot, properties {horizontalAlignment=HAlign_Center, verticalAlignment=VAlign_Center}}
  }
  GoldStarComplete {
   widgetClassPath=&Image
   bIsVariable=true
   parent=GoldStarOverlay
   widget=GoldStarComplete
   properties {
    visibility=Hidden
    brush {
     bIsDynamicallyLoaded=false
     drawAs=Image
     imageType=NoImage
     mirroring=NoMirror
     resourceName=None
     resourceObject=/PilgrimCore/Widgets/Materials/MI_UI_Pilgrim_GoldStar_Complete
     tiling=NoTile
     imageSize=<64 64>
     margin=<ltrb 0 0 0 0>
     outlineSettings {
      bUseBrushTransparency=false
      roundingType=HalfHeightRadius
      width=0
      color {colorUseRule=UseColor_Specified, specifiedColor=<rgba 0 0 0 0>}
      cornerRadii=<0 0 0 0>
     }
     tintColor {colorUseRule=UseColor_Specified, specifiedColor=<rgba 1 1 1 1>}
     uVRegion {bIsValid=false, max=<0 0>, min=<0 0>}
    }
   }
   slot {class=OverlaySlot, properties {horizontalAlignment=HAlign_Center, verticalAlignment=VAlign_Center, padding=<ltrb 1 1 1 1>}}
  }
  GoldStarProgress {
   widgetClassPath=&Image
   bIsVariable=true
   parent=GoldStarOverlay
   widget=GoldStarProgress
   properties {brush {
     bIsDynamicallyLoaded=false
     drawAs=Image
     imageType=NoImage
     mirroring=NoMirror
     resourceName=None
     resourceObject=/PilgrimCore/Widgets/Materials/MI_UI_Pilgrim_GoldStar
     tiling=NoTile
     imageSize=<64 64>
     margin=<ltrb 0 0 0 0>
     outlineSettings {
      bUseBrushTransparency=false
      roundingType=HalfHeightRadius
      width=0
      color {colorUseRule=UseColor_Specified, specifiedColor=<rgba 0 0 0 0>}
      cornerRadii=<0 0 0 0>
     }
     tintColor {colorUseRule=UseColor_Specified, specifiedColor=<rgba 1 1 1 1>}
     uVRegion {bIsValid=false, max=<0 0>, min=<0 0>}
    }}
   slot {class=OverlaySlot, properties {horizontalAlignment=HAlign_Center, verticalAlignment=VAlign_Center}}
  }
 ]
}
