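# Five-star progress bar widget for a Pilgrim music game that shows the local or spectated player's score stars.
# Score updates complete or roll back stars on change and fill the current one; Expert difficulty adds gold-star progress past the maximum.
# Spectator switches pause updates while stars rebuild, and hit, miss and pass deltas broadcast note-reaction events through the event router.
# It also offers a staggered results reveal with rising-pitch sound.
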WidgetBlueprint /PilgrimQuickplay/Widgets/WBP_PilgrimQuickplay_LocalStarBar
graphs [
 Check Is Star Changed {nodes [
   n1 = Check Is Star Changed!() -> (then, CurrentStarIndex:int, CurrentStarProgress:double=~) {is=title K2Node_FunctionEntry}
   n2 = Get LastStarProgress() -> (LastStarProgress:double)
   n3 = Get LastStarIndex() -> (LastStarIndex:int=0)
   > n4 = Set HasChanged(HasChanged:bool=OR Boolean(A="Not Equal (Integer)"(A=@n1, B=@n3), B="float < float"(A=@n2, B=@n1.CurrentStarProgress))) -> (Output_Get:bool)
   > Set LastStarProgress(LastStarProgress:double=@n1.CurrentStarProgress) -> (Output_Get:double)
   > Return Node!(execute, IsChanged:bool=@n4)
  ]}
 CheckStarCompletion {nodes [
   n1 = CheckStarCompletion!() -> (then, CurrentStar:int) {is=title K2Node_FunctionEntry}
   n2 = Get LastStarIndex() -> (LastStarIndex=0)
   > n3 = Compare Int!(Exec:Exec, Input:int=@n1, CompareWith:int=@n2) -> (">":Exec, "==":Exec, "<":Exec) {is=title K2Node_MacroInstance}
   n4 = For Loop(execute=@n3."<", FirstIndex=@n1, LastIndex=@n2)
   n5 = int - int(A=@n1)
   n6 = Get Star Widget!(execute=@n4.Completed, self, Index:int=@n5) -> (then, StarWidget:WBP Pilgrim Star Object Reference) {is=title K2Node_CallFunction}
   > Is Valid(InputObject=@n6)
   > SetComplete(self:WBP Pilgrim Star Object Reference=@n6)
   n7 = Get Star Widget(execute=@n4.LoopBody, Index=@n4)
   > Is Valid(InputObject=@n7)
   > SetVisibility(self=Get StarProgress!(self:WBP Pilgrim Star Object Reference=@n7) -> (StarProgress:Radial Slider Object Reference), InVisibility=Hidden)
   > SetVisibility(self=Get StarComplete!(self:WBP Pilgrim Star Object Reference=@n7) -> (StarComplete:Image Object Reference), InVisibility=Hidden)
   n8 = For Loop(execute=@n3.">", FirstIndex=Get LastStarIndex() -> (LastStarIndex=0), LastIndex=@n5)
   Set LastStarIndex(execute=@n8.Completed, LastStarIndex:int=@n1) -> (Output_Get:int=0)
   n9 = Get Star Widget(execute=@n8.LoopBody, Index=@n8)
   > Is Valid(InputObject=@n9)
   > SetComplete(self=@n9)
  ]}
 EventGraph {nodes [
   n1 = StaggeredStarReveal!() -> (OutputDelegate, then, StarCount:int) {is=title K2Node_CustomEvent}
   > n2 = Set Star Count(Star Count:int=@n1.StarCount) -> (Output_Get:int=0)
   > Branch(Condition="integer > integer"(A=@n2))
   > n3 = For Each Loop with Delay!(
    Exec:Exec
    Array:Array of WBP Pilgrim Star Object References=Get Stars() -> (Stars:Array of WBP Pilgrim Star Object References)
    Delay:double=0.25
   ) -> (Loop Body:Exec, Array Element:WBP Pilgrim Star Object Reference, Array Index:int, Completed:Exec) {is=title K2Node_MacroInstance}
   > Branch(Condition="integer < integer"(A=@n3.Array Index, B=Get Star Count() -> (Star Count:int=0)))
   > n4 = Branch(Condition=IsVisible())
   n5 = "To Float (Integer)"(InInt=@n3.Array Index)
   n6 = Get StarSFXPitchIncrement() -> (StarSFXPitchIncrement:double)
   > n7 = PlaySound2D(
    Sound=Get StarCompleteSound() -> (StarCompleteSound:Sound Base Object Reference)
    PitchMultiplier=float + float(A=float * float(A=@n5, B=@n6) {is=title K2Node_PromotableOperator}, B=Get StartingStarSFXPitch() -> (StartingStarSFXPitch:double))
   )
   n8 = Branch(execute=[@n7.then, @n4.else], Condition=Get bIsGoldStars() -> (bIsGoldStars:bool))
   > SetGoldStarComplete(self:WBP Pilgrim Star Object Reference=@n3)
   SetComplete(execute=@n8.else, self=@n3)
   n9 = RefreshStars!() -> (OutputDelegate, then, PIlgrimGameProgress:"Pilgrim Game Progress Structure (by ref)") {is=title K2Node_CustomEvent}
   > n10 = Branch(Condition=Get isPausingUpdates() -> (isPausingUpdates:bool))
   n11 = Sequence!(execute=@n10.else) -> (then_0, then_1, then_2:Exec)
   n12 = Break Pilgrim Game Progress!(PilgrimGameProgress:"Pilgrim Game Progress Structure (by ref)"=@n9.PIlgrimGameProgress) -> (
    NumStreakBreakingMisses:int
    bHasPerfectCombo:bool=~
    RawCurrentScore:float=~
    CurrentModifierBonus:float=~
    CurrentStar:int
    NextStarProgress:float=~
    BandScoreContribution:float=~
    BandModifierBonusContribution:float=~
    RawMaxScore:float=~
    MaxStreakLength:int
    AverageStreakMultiplierPerHitOrPass:float=~
    CurrentSectionCompletionPercentage:float=~
   ) {is=title K2Node_BreakStruct}
   > Handle Star Update!(execute, self, CurrentStar:int=@n12.CurrentStar, StarProgress:double=@n12.NextStarProgress) -> (then)
   Branch(
    execute=@n11.then_1
    Condition="integer > integer"(A=@n12.CurrentStar, B=Get LocalPlayerPreviousStarValue() -> (LocalPlayerPreviousStarValue:int=0))
   )
   > Set LocalPlayerPreviousStarValue(LocalPlayerPreviousStarValue:int=@n12.CurrentStar) -> (Output_Get:int=0)
   n13 = Get!(execute:Exec=@n11.then_2) -> (then:Exec, else:Exec, _SpectatedPlayerState:Fort Player State Object Reference)
   n14 = Break Pilgrim Game Progress!(PilgrimGameProgress=@n9.PIlgrimGameProgress) -> (NumHits:int, NumMisses:int, NumPasses:int, NoteStreak:int, CurrentSectionCompletionPercentage=~)
   n15 = Get PrintDebugInfo() -> (PrintDebugInfo:bool)
   PrintString!(
    execute=@n13.else
    InString=Append!(
     A=Score Event...Num Hits:
     B="To String (Integer)"(InInt=@n14)
     C:String="   Num Misses: "
     D:String="To String (Integer)"(InInt=@n14.NumMisses)
     E:String="  Num Passes:"
     F:String="To String (Integer)"(InInt=@n14.NumPasses)
    ) -> (ReturnValue)
    bPrintToScreen=@n15
    bPrintToLog=@n15
    TextColor
    Duration
    Key
   ) -> (then)
   > n16 = Sequence!(execute) -> (then_0, then_1, then_2, then_3:Exec, then_4:Exec, then_5:Exec)
   n17 = Get Prev Hits() -> (Prev Hits:int=0)
   > Branch(Condition=AND Boolean(A="integer > integer"(A=@n14, B=@n17), B="integer >= integer"(A=@n14.NoteStreak, B=4)))
   n18 = Get Event Router() -> (Event Router:Gameplay Event Router Component Object Reference)
   > Is Valid(InputObject=@n18)
   > n19 = Cast To FortPlayerController(Object=GetOwningPlayer())
   > Broadcast Event!(
    execute
    self:Gameplay Event Router Component Object Reference=@n18
    EventData:Pilgrim Event Gameplay Note Reaction Structure=Make Pilgrim Event Gameplay Note Reaction!(
     FortPlayerController_3_#1:Fort Player Controller Object Reference=@n19
     PlayerNoteReaction_6_#2:PlayerNoteReactionEnum Enum=NewEnumerator2
    ) -> (PilgrimEvent_Gameplay_NoteReaction:Pilgrim Event Gameplay Note Reaction Structure) {is=title K2Node_MakeStruct}
    EventContext:Object Reference
   ) -> (then, ReturnValue:bool)
   Set Prev Hits(execute=@n16.then_1, Prev Hits:int=@n14) -> (Output_Get:int=0)
   Set Prev Misses(execute=@n16.then_3, Prev Misses:int=@n14.NumMisses) -> (Output_Get:int=0)
   Set Prev Passes(execute=@n16.then_5, Prev Passes:int=@n14.NumPasses) -> (Output_Get:int=0)
   Branch(execute=@n16.then_2, Condition="integer > integer"(A=@n14.NumMisses, B=Get Prev Misses() -> (Prev Misses:int=0)))
   n20 = Get Event Router()
   > Is Valid(InputObject=@n20)
   > n21 = Cast To FortPlayerController(Object=GetOwningPlayer())
   > Broadcast Event(
    self=@n20
    EventData=Make Pilgrim Event Gameplay Note Reaction(FortPlayerController_3_#1=@n21, PlayerNoteReaction_6_#2=NewEnumerator0)
   )
   Branch(execute=@n16.then_4, Condition="integer > integer"(A=@n14.NumPasses, B=Get Prev Passes() -> (Prev Passes:int=0)))
   n22 = Get Event Router()
   > Is Valid(InputObject=@n22)
   > n23 = Cast To FortPlayerController(Object=GetOwningPlayer())
   > Broadcast Event(
    self=@n22
    EventData=Make Pilgrim Event Gameplay Note Reaction(FortPlayerController_3_#1=@n23, PlayerNoteReaction_6_#2=NewEnumerator1)
   )
   n24 = SetSpectatedPlayer!() -> (OutputDelegate, then, SpectatedPlayerState:Fort Player State Object Reference, PilgrimGameProgress:Pilgrim Game Progress Structure) {is=title K2Node_CustomEvent}
   > n25 = Branch(Condition="Not Equal (Object)"(
     A:Object Reference=@n24.SpectatedPlayerState
     B:Object Reference=Get _SpectatedPlayerState() -> (_SpectatedPlayerState:Fort Player State Object Reference)
    ) -> (ReturnValue:bool=~) {is=title K2Node_PromotableOperator})
   > Set _SpectatedPlayerState(_SpectatedPlayerState:Fort Player State Object Reference=@n24.SpectatedPlayerState) -> (Output_Get:Fort Player State Object Reference)
   > Set isPausingUpdates(isPausingUpdates:bool=true) -> (Output_Get:bool)
   > SetStarsToDefault()
   > SetStarsForPlayer!(execute, self, PilgrimGameProgress:"Pilgrim Game Progress Structure (by ref)"=@n24.PilgrimGameProgress) -> (then)
   SetStarsForPlayer(execute=@n25.else, PilgrimGameProgress=@n24.PilgrimGameProgress)
   ClearSpectatedPlayer() {is=title K2Node_CustomEvent}
   > Set _SpectatedPlayerState()
   > SetStarsToDefault()
   n26 = SetStarsForPlayer!() -> (OutputDelegate, then, PilgrimGameProgress:"Pilgrim Game Progress Structure (by ref)") {is=title K2Node_CustomEvent}
   n27 = Break Pilgrim Game Progress!(PilgrimGameProgress=@n26.PilgrimGameProgress) -> (CurrentStar:int, NextStarProgress:float=~, CurrentSectionCompletionPercentage=~)
   > Set LocalPlayerPreviousStarValue(LocalPlayerPreviousStarValue=@n27) -> (Output_Get=0)
   > Set LastStarIndex(LastStarIndex=@n27) -> (Output_Get=0)
   > Set LastStarProgress(LastStarProgress=@n27.NextStarProgress)
   > n28 = For Each Loop(Array:Array of WBP Pilgrim Star Object References=Get Stars()) -> (Array Element:WBP Pilgrim Star Object Reference)
   > n29 = Branch(Condition="integer < integer"(A=@n28.Array Index, B=@n27))
   > SetStarProgress!(execute, self:WBP Pilgrim Star Object Reference=@n28, SetStarProgress:double=1) -> (then)
   > Set ShowCompleteAnim!(execute, ShowCompleteAnim:bool=false, self:WBP Pilgrim Star Object Reference=@n28) -> (then, Output_Get:bool)
   > SetComplete(self=@n28)
   Set ShowCompleteAnim(execute=@n29.else, ShowCompleteAnim=true, self=@n28)
   > Branch(Condition="Equal (Integer)"(A=@n28.Array Index, B=@n27))
   > SetStarProgress(self=@n28, SetStarProgress=@n27.NextStarProgress)
   n30 = Branch(execute=@n28.Completed, Condition="integer <= integer"(A=Get MaxStars() -> (MaxStars:int=0), B=@n27))
   > Update Gold Star Progress Widgets!(execute, self, CurrentStar:int=@n27, GoldStarProgress:double=@n27.NextStarProgress) -> (then)
   > n31 = For Each Loop(Array=Get Stars())
   > Set ShowCompleteAnim(ShowCompleteAnim="float < float"(A=@n27.NextStarProgress, B=1), self=@n31)
   Set LocalPlayerPreviousStarValue(execute=[@n31.Completed, @n30.else], LocalPlayerPreviousStarValue=@n27) -> (Output_Get=0)
   > Set isPausingUpdates(isPausingUpdates=false)
   Event PreConstruct()
   > SetStretch!(
    execute
    self:Scale Box Object Reference=Get ScaleBox() -> (ScaleBox:Scale Box Object Reference)
    InStretch:EStretch Enum=Get Scale Stretch() -> (Scale Stretch:EStretch Enum=None)
   ) -> (then)
   > SetUserSpecifiedScale!(
    execute
    self:Scale Box Object Reference=Get ScaleBox()
    InUserSpecifiedScale:float=Get In User Specified Scale() -> (In User Specified Scale:float)
   ) -> (then)
   n32 = Get ShouldStarsAnimateComplete() -> (ShouldStarsAnimateComplete:bool)
   > Set ShowCompleteAnim(ShowCompleteAnim=@n32, self=Get RadialStar1() -> (RadialStar1:WBP Pilgrim Star Object Reference))
   > Set ShowCompleteAnim(ShowCompleteAnim=@n32, self=Get RadialStar2() -> (RadialStar2:WBP Pilgrim Star Object Reference))
   > Set ShowCompleteAnim(ShowCompleteAnim=@n32, self=Get RadialStar3() -> (RadialStar3:WBP Pilgrim Star Object Reference))
   > Set ShowCompleteAnim(ShowCompleteAnim=@n32, self=Get RadialStar4() -> (RadialStar4:WBP Pilgrim Star Object Reference))
   > Set ShowCompleteAnim(ShowCompleteAnim=@n32, self=Get RadialStar5() -> (RadialStar5:WBP Pilgrim Star Object Reference))
   > Set Stars(Stars:Array of WBP Pilgrim Star Object References=Make Array!(
     "[0]":WBP Pilgrim Star Object Reference=Get RadialStar1()
     "[1]":WBP Pilgrim Star Object Reference=Get RadialStar2()
     "[2]":WBP Pilgrim Star Object Reference=Get RadialStar3()
     "[3]":WBP Pilgrim Star Object Reference=Get RadialStar4()
     "[4]":WBP Pilgrim Star Object Reference=Get RadialStar5()
    ) -> (Array:Array of WBP Pilgrim Star Object References)) -> (Output_Get:Array of WBP Pilgrim Star Object References)
   Event Construct()
   > SubscribeForChangingVisibilityByTag!(
    execute
    self:Pilgrim Game Widget Base Object Reference
    GameplayTag:"Gameplay Tag Structure (by ref)"=MakeLiteralGameplayTag!(Value:Gameplay Tag Structure=(TagName="Pilgrim.Widget.LocalStarBar")) -> (ReturnValue:Gameplay Tag Structure)
   ) -> (then)
   Event OnMusicBattleInstanceInitializeEvent() {is=title K2Node_Event}
   > Set isPausingUpdates(isPausingUpdates=true)
   > n33 = PilgrimWaitForEventRouter!(execute:Exec, ContextActor:Actor Object Reference=GetOwningPlayer()) -> (then:Exec, OnEventRouterAvailable:Exec, EventRouter:Gameplay Event Router Component Object Reference) {is=title K2Node_AsyncAction}
   Set Event Router(execute=@n33.OnEventRouterAvailable, Event Router:Gameplay Event Router Component Object Reference=@n33) -> (Output_Get:Gameplay Event Router Component Object Reference)
   > n34 = Cast To FortPlayerStateAthena!(execute, Object=Get PlayerState!(self:Controller Object Reference=GetOwningPlayer()) -> (PlayerState:Player State Object Reference)) -> (then, CastFailed, AsFort Player State Athena:Fort Player State Athena Object Reference) {is=title K2Node_DynamicCast}
   n35 = Get MusicBattleInstance() -> (MusicBattleInstance:Pilgrim Music Battle Instance Object Reference)
   > Set bAllowGoldStar(bAllowGoldStar:bool="Equal (Enum)"(
     A:EPilgrimSongDifficulty Enum=GetSongDifficultyForPlayer!(self:Pilgrim Music Battle Instance Object Reference=@n35, Player:Fort Player State Athena Object Reference=@n34) -> (ReturnValue:EPilgrimSongDifficulty Enum=DifficultyEasy)
     B:EPilgrimSongDifficulty Enum=DifficultyExpert
    )) -> (Output_Get:bool)
   > SetStarsToDefault()
   n36 = Event OnScoreDataReceivedLocalPlayer!() -> (OutputDelegate, then, GameProgress:"Pilgrim Game Progress Structure (by ref)") {is=title K2Node_Event}
   > n37 = Get()
   Set isPausingUpdates(execute=@n37.else, isPausingUpdates=false)
   > RefreshStars!(execute, self, PIlgrimGameProgress:"Pilgrim Game Progress Structure (by ref)"=@n36.GameProgress) -> (then) {is=title K2Node_CallFunction}
  ]}
 Get Star Widget {nodes [
   n1 = Get Star Widget!() -> (then, Index:int) {is=title K2Node_FunctionEntry}
   > Return Node!(
    execute
    StarWidget:WBP Pilgrim Star Object Reference=Select!(
     Option 0:WBP Pilgrim Star Object Reference=Get RadialStar1()
     Option 1:WBP Pilgrim Star Object Reference=Get RadialStar2()
     Option 2:WBP Pilgrim Star Object Reference=Get RadialStar3()
     Option 3:WBP Pilgrim Star Object Reference=Get RadialStar4()
     Option 4:WBP Pilgrim Star Object Reference=Get RadialStar5()
     Index:int=@n1
    ) -> (ReturnValue:WBP Pilgrim Star Object Reference)
   )
  ]}
 GetSpectatedPlayerState {nodes [
   GetSpectatedPlayerState() {is=title K2Node_FunctionEntry}
   > Return Node!(execute, _SpectatedPlayerState:Fort Player State Object Reference=Get _SpectatedPlayerState())
  ]}
 Handle Star Update {nodes [
   n1 = Handle Star Update!() -> (then, CurrentStar:int, StarProgress:double=~) {is=title K2Node_FunctionEntry}
   > n2 = Check Is Star Changed!(execute, self, CurrentStarIndex:int=@n1, CurrentStarProgress:double=@n1.StarProgress) -> (then, IsChanged:bool) {is=title K2Node_CallFunction}
   > Branch(Condition=@n2)
   > CheckStarCompletion!(execute, self, CurrentStar:int=@n1) -> (then) {is=title K2Node_CallFunction}
   > n3 = Branch(Condition="integer <= integer"(A=Get MaxStars() -> (MaxStars=0), B=@n1))
   > Update Gold Star Progress Widgets(CurrentStar=@n1, GoldStarProgress=@n1.StarProgress)
   Update Current Star!(execute=@n3.else, self, CurrentStar:int=@n1, NextStarProgress:double=@n1.StarProgress) -> (then)
  ]}
 ResetCompletedVisuals {nodes [
   ResetCompletedVisuals() {is=title K2Node_FunctionEntry}
   > n1 = For Each Loop(Array=Get Stars())
   > ResetCompletedVisuals(self:WBP Pilgrim Star Object Reference=@n1) {is=title K2Node_CallFunction}
  ]}
 SetStarsToDefault {nodes [
   SetStarsToDefault() {is=title K2Node_FunctionEntry}
   > n1 = For Each Loop(Array=Get Stars())
   Return Node!(execute=@n1.Completed)
   > ResetCompletedVisuals(self=@n1)
  ]}
 SetupStarCount {nodes [
   n1 = SetupStarCount!() -> (then, StarCount:int) {is=title K2Node_FunctionEntry}
   > n2 = For Loop(LastIndex="Min (Integer)"(A=int - int(A=@n1), B=4) -> (ReturnValue=0))
   Branch(execute=@n2.Completed, Condition=Get bIsSequencedAnimation() -> (bIsSequencedAnimation:bool))
   > ResetCompletedVisuals(self:Self Object Reference)
   n3 = Get Star Widget(execute=@n2.LoopBody, Index=@n2) {is=title K2Node_CallFunction}
   > n4 = Set bIsGoldStars(bIsGoldStars:bool="integer > integer"(A=@n1, B=5)) -> (Output_Get:bool)
   > n5 = Branch(Condition=@n4)
   > SetGoldStarComplete(self=@n3)
   Branch(execute=@n5.else, Condition="integer > integer"(A=@n1, B=@n2))
   > SetComplete(self=@n3)
  ]}
 SetupStarsResults_DEPRECATED {nodes [
   n1 = SetupStarsResults_DEPRECATED!() -> (then, Instrument:EPilgrimTrackType Enum=TrackGuitar, StarCount:int) {is=title K2Node_FunctionEntry}
   > SetupStarCount!(execute, self, StarCount:int=@n1.StarCount) -> (then) {is=title K2Node_CallFunction}
  ]}
 ShowGoldStarIndicator {nodes [
   ShowGoldStarIndicator() {is=title K2Node_FunctionEntry}
   > SetVisibility(self=Get GoldStarIndicator() -> (GoldStarIndicator:Border Object Reference), InVisibility=Visible)
  ]}
 Update Current Star {nodes [
   n1 = Update Current Star!() -> (then, CurrentStar:int, NextStarProgress:double=~) {is=title K2Node_FunctionEntry}
   > Branch(Condition="float > float"(A=@n1.NextStarProgress))
   > n2 = Get Star Widget(Index=@n1)
   > SetStarProgress(self=@n2, SetStarProgress=@n1.NextStarProgress)
   Reroute Node(InputPin:Wildcard) -> (OutputPin:Wildcard)
   Reroute Node()
  ]}
 Update Gold Star Progress Widgets {nodes [
   n1 = Update Gold Star Progress Widgets!() -> (then, CurrentStar:int, GoldStarProgress:double=~) {is=title K2Node_FunctionEntry}
   > Branch(Condition=Get bAllowGoldStar() -> (bAllowGoldStar:bool))
   > n2 = For Loop(LastIndex=4)
   > n3 = Get Star Widget(Index=@n2)
   n4 = "integer > integer"(A=@n1, B=Get MaxStars() -> (MaxStars=0))
   > n5 = Branch(Condition=@n4)
   > SetGoldStarComplete(self=@n3)
   Set Gold Star Progress!(execute=@n5.else, self:WBP Pilgrim Star Object Reference=@n3, Progress:double=@n1.GoldStarProgress) -> (then) {is=title K2Node_CallFunction}
   n6 = Sequence!(execute=@n2.Completed) -> (then_0, then_1)
   > Branch(Condition=@n4)
   > ShowGoldStarIndicator() {is=title K2Node_CallFunction}
   Return Node(execute=@n6.then_1)
  ]}
 Update Icons Individualy {nodes [
   n1 = Update Icons Individualy!() -> (then, InstrumentsPlayed:Array of EPilgrimTrackType Enums=TrackGuitar) {is=title K2Node_FunctionEntry}
   > n2 = For Loop(LastIndex=3)
   > n3 = Get Star Widget(Index=@n2)
   n4 = Length(TargetArray:Array of EPilgrimTrackType Enums=@n1)
   > SetVisibility(
    self=@n3
    InVisibility=Select!(Option 0:ESlateVisibility Enum=Collapsed, Option 1:ESlateVisibility Enum=SelfHitTestInvisible, Index:bool="integer < integer"(A=@n2, B=@n4)) -> (ReturnValue:ESlateVisibility Enum)
   )
   > Branch(Condition=Is Valid Index!(TargetArray:Array of EPilgrimTrackType Enums=@n1, IndexToTest:int=@n2) -> (ReturnValue:bool))
   > SetStarProgress(self=@n3, SetStarProgress=0)
   > SetVisibility(self=Get RadialStar5(), InVisibility=Collapsed)
  ]}
]
variables [
 _SpectatedPlayerState {type=FortPlayerState, default=null}
 bAllowGoldStar {type=boolean}
 bIsGoldStars {type=boolean}
 bIsSequencedAnimation {type=boolean}
 Current Star {type=integer, default=1}
 Event Router {type=GameplayEventRouterComponent, default=null}
 In User Specified Scale {type=number, default=1}
 isPausingUpdates {type=boolean}
 LastStarIndex {type=integer}
 LastStarProgress {type=number}
 LocalPlayerPreviousStarValue {type=integer}
 MaxStars {type=integer, default=5}
 OD Deploy Buzz {type=AudioComponent, default=null}
 Prev Hits {type=integer}
 Prev Misses {type=integer}
 Prev Passes {type=integer}
 PrintDebugInfo {type=boolean}
 Scale Stretch {type=EStretch, default=ScaleToFit}
 ShouldStarsAnimateComplete {type=boolean}
 Star Count {type=integer}
 Star Progress {type=number}
 StarCompleteSound {type=SoundBase, default=/PilgrimQuickplay/Audio/Results/MSS_FM_ResultsVictory_Tally1pStar}
 Stars {type="array<WBP_PilgrimStar_C>", default []}
 StarSFXPitchIncrement {type=number, default=0.2}
 StartingStarSFXPitch {type=number, default=0.08}
]
widgets {
 info {inheritedWidgetCount=0, parentClass=/Script/PilgrimCoreUI.PilgrimGameWidgetBase, rootWidgetClass=&Overlay, widgetCount=13}
 widgets [
  StarsPanel {widgetClassPath=&Overlay, widget=StarsPanel}
  GoldStarIndicator {
   widgetClassPath=&Border
   bIsVariable=true
   parent=StarsPanel
   widget=GoldStarIndicator
   properties {
    visibility=Hidden
    background {
     bIsDynamicallyLoaded=false
     drawAs=Box
     imageType=NoImage
     mirroring=NoMirror
     resourceName=None
     resourceObject=null
     tiling=NoTile
     imageSize=<128 128>
     margin=<ltrb 1 1 1 1>
     outlineSettings {
      bUseBrushTransparency=false
      roundingType=HalfHeightRadius
      width=0
      color {colorUseRule=UseColor_Specified, specifiedColor=<rgba 0 0 0 0>}
      cornerRadii=<0 0 0 0>
     }
     tintColor {colorUseRule=UseColor_Specified, specifiedColor=<rgba 1 1 0 0>}
     uVRegion {bIsValid=false, max=<0 0>, min=<0 0>}
    }
    desiredSizeScale=<2 2>
    padding=<ltrb 0 0 0 0>
   }
   slot {class=OverlaySlot, properties {horizontalAlignment=HAlign_Fill, verticalAlignment=VAlign_Center}}
  }
  ScaleBox {
   widgetClassPath=&ScaleBox
   bIsVariable=true
   parent=StarsPanel
   widget=ScaleBox
   properties {stretch=UserSpecified}
   slot {class=OverlaySlot, properties {horizontalAlignment=HAlign_Fill, verticalAlignment=VAlign_Fill}}
  }
  StarContainer {
   widgetClassPath=&HorizontalBox
   bIsVariable=true
   parent=ScaleBox
   widget=StarContainer
   slot {class=ScaleBoxSlot, properties {horizontalAlignment=HAlign_Fill}}
  }
  RadialStar1 {
   widgetClassPath=&/WBP_PilgrimStar
   bIsVariable=true
   parent=StarContainer
   widget=RadialStar1
   slot {class=HorizontalBoxSlot, properties {horizontalAlignment=HAlign_Left, verticalAlignment=VAlign_Center}}
  }
  Spacer_103 {widgetClassPath=&Spacer, parent=StarContainer, widget=Spacer_103, properties {size=<4 1>}, slot {class=HorizontalBoxSlot}}
  RadialStar2 {
   widgetClassPath=&WBP_PilgrimStar
   bIsVariable=true
   parent=StarContainer
   widget=RadialStar2
   slot {class=HorizontalBoxSlot, properties {horizontalAlignment=HAlign_Left, verticalAlignment=VAlign_Center}}
  }
  Spacer {widgetClassPath=&Spacer, parent=StarContainer, widget=Spacer, properties {size=<4 1>}, slot {class=HorizontalBoxSlot}}
  RadialStar3 {
   widgetClassPath=&WBP_PilgrimStar
   bIsVariable=true
   parent=StarContainer
   widget=RadialStar3
   slot {class=HorizontalBoxSlot, properties {horizontalAlignment=HAlign_Left, verticalAlignment=VAlign_Center}}
  }
  Spacer_1 {widgetClassPath=&Spacer, parent=StarContainer, widget=Spacer_1, properties {size=<4 1>}, slot {class=HorizontalBoxSlot}}
  RadialStar4 {
   widgetClassPath=&WBP_PilgrimStar
   bIsVariable=true
   parent=StarContainer
   widget=RadialStar4
   slot {class=HorizontalBoxSlot, properties {horizontalAlignment=HAlign_Center, verticalAlignment=VAlign_Center}}
  }
  Spacer_2 {widgetClassPath=&Spacer, parent=StarContainer, widget=Spacer_2, properties {size=<4 1>}, slot {class=HorizontalBoxSlot}}
  RadialStar5 {
   widgetClassPath=&WBP_PilgrimStar
   bIsVariable=true
   parent=StarContainer
   widget=RadialStar5
   slot {class=HorizontalBoxSlot, properties {verticalAlignment=VAlign_Center}}
  }
 ]
}
