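# Quickplay UI widget that shows a rounded "SWAP DEVICE" prompt, with a guitar icon visible only while the player uses a peripheral.
# On Construct it waits for the player state's event router, starts listening for peripheral-change events, then reads the initial state from PilgrimPlayerControllerComponent.
# Each event updates the icon only for a local controller, through SetCurrentVisibility, which toggles PlasticImage between SelfHitTestInvisible and Collapsed.
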WidgetBlueprint /PilgrimQuickplay/Widgets/WBP_PilgrimQuickplay_SwapDevice
graphs [
 EventGraph {nodes [
   Event Construct()
   > n1 = PilgrimWaitForEventRouter!(
    execute:Exec
    ContextActor:Actor Object Reference=Get PlayerState!(self:Controller Object Reference=GetOwningPlayer()) -> (PlayerState:Player State Object Reference)
   ) -> (then:Exec, OnEventRouterAvailable:Exec, EventRouter:Gameplay Event Router Component Object Reference) {is=title K2Node_AsyncAction}
   Set Event Router(execute=@n1.OnEventRouterAvailable, Event Router:Gameplay Event Router Component Object Reference=@n1) -> (Output_Get:Gameplay Event Router Component Object Reference)
   > n2 = StartListeningToEvent!(
    execute:Exec
    Target:Gameplay Event Router Component Object Reference=Get Event Router() -> (Event Router:Gameplay Event Router Component Object Reference)
    EventType:Script Struct Object Reference=/Script/PilgrimCoreRuntime.PilgrimEvent_PlayerUsingPeripheralChanged
    Context:Object Reference
    EventBubblingRule:EEventBubblingRule Enum
   ) -> (
    then:Exec
    AsyncTaskProxy:Async Action Start Listening To Event Object Reference
    OnEventReceived:Exec
    Payload:Pilgrim Event Player Using Peripheral Changed Structure
   )
   n3 = GetControllerFromActor!(Actor:Actor Object Reference=GetOwningPlayer()) -> (ReturnValue:Controller Object Reference)
   > Is Valid(InputObject=@n3)
   n4 = GetComponentByClass!(self:Actor Object Reference=@n3, ComponentClass:Actor Component Class Reference=/Script/PilgrimCoreRuntime.PilgrimPlayerControllerComponent) -> (ReturnValue:Pilgrim Player Controller Component Object Reference)
   > Is Valid(InputObject=@n4)
   > SetCurrentVisibility!(
    execute
    self
    IsVisible:bool=GetIsCurrentDevicePeripheral!(self:Pilgrim Player Controller Component Object Reference=@n4) -> (ReturnValue:bool)
   ) -> (then)
   n5 = Break Pilgrim Event Player Using Peripheral Changed!(PilgrimEvent_PlayerUsingPeripheralChanged:"Pilgrim Event Player Using Peripheral Changed Structure (by ref)"=@n2.Payload) -> (
    ForPlayer:Fort Player State Object Reference
    bIsUsingPeripheral:bool=~
    bIsInputDeviceValidForTrackType:bool=~
    CurrentPeripheralDeviceType:EFMPeripheralInputMode Enum
   ) {is=title K2Node_BreakStruct}
   n6 = GetPlayerController!(self:Player State Object Reference=@n5) -> (ReturnValue:Player Controller Object Reference)
   Is Valid(exec=@n2.OnEventReceived, InputObject=@n6)
   > Branch(Condition=IsLocalController!(self:Controller Object Reference=@n6) -> (ReturnValue:bool))
   > SetCurrentVisibility(IsVisible=@n5.bIsUsingPeripheral)
   Event Tick()
  ]}
 SetCurrentVisibility {nodes [
   n1 = SetCurrentVisibility!() -> (then, IsVisible:bool=~) {is=title K2Node_FunctionEntry}
   > n2 = Branch(Condition=@n1)
   n3 = Get PlasticImage() -> (PlasticImage:Image Object Reference)
   > SetVisibility(self=@n3, InVisibility=SelfHitTestInvisible)
   SetVisibility(execute=@n2.else, self=@n3, InVisibility=Collapsed)
  ]}
]
variables [
 Event Router {type=GameplayEventRouterComponent, default=null}
 Start at Time {type=number, default=2}
]
widgets {
 info {inheritedWidgetCount=0, parentClass=&UserWidget, rootWidgetClass=&Overlay, widgetCount=6}
 widgets [
  Overlay_0 {widgetClassPath=&Overlay, widget=Overlay_0}
  SizeBox_0 {
   widgetClassPath=&SizeBox
   parent=Overlay_0
   widget=SizeBox_0
   properties {bOverride_HeightOverride=true, bOverride_WidthOverride=true, heightOverride=114, widthOverride=644}
   slot {class=OverlaySlot, properties {horizontalAlignment=HAlign_Center, verticalAlignment=VAlign_Center}}
  }
  Border_18 {
   widgetClassPath=&Border
   parent=SizeBox_0
   widget=Border_18
   properties {
    horizontalAlignment=HAlign_Center
    verticalAlignment=VAlign_Center
    background {
     bIsDynamicallyLoaded=false
     drawAs=RoundedBox
     imageType=NoImage
     mirroring=NoMirror
     resourceName=None
     resourceObject=null
     tiling=NoTile
     imageSize=<666 97>
     margin=<ltrb 0 0 0 0>
     outlineSettings {
      bUseBrushTransparency=false
      roundingType=HalfHeightRadius
      width=0
      color {colorUseRule=UseColor_Specified, specifiedColor=<rgba 0 0 0 0>}
      cornerRadii=<0 0 0 0>
     }
     tintColor {colorUseRule=UseColor_Specified, specifiedColor=<rgba 0 0 0 0.6>}
     uVRegion {bIsValid=false, max=<0 0>, min=<0 0>}
    }
    brushColor=<rgba 0 1 1 1>
    padding=<ltrb 0 0 0 0>
   }
   slot {class=SizeBoxSlot}
  }
  HorizontalBox_0 {
   widgetClassPath=&HorizontalBox
   parent=Border_18
   widget=HorizontalBox_0
   slot {class=BorderSlot, properties {horizontalAlignment=HAlign_Center, verticalAlignment=VAlign_Center, padding=<ltrb 0 0 0 0>}}
  }
  PlasticImage {
   widgetClassPath=&Image
   bIsVariable=true
   parent=HorizontalBox_0
   widget=PlasticImage
   properties {
    visibility=Collapsed
    brush {
     bIsDynamicallyLoaded=false
     drawAs=Image
     imageType=FullColor
     mirroring=NoMirror
     resourceName=None
     resourceObject=/PilgrimQuickplay/Widgets/Materials/MI_UI_Icon_Instrument_Type_Guitar
     tiling=NoTile
     imageSize=<64 64>
     margin=<ltrb 0 0 0 0>
     outlineSettings {
      bUseBrushTransparency=false
      roundingType=HalfHeightRadius
      width=0
      color {colorUseRule=UseColor_Specified, specifiedColor=<rgba 0 0 0 0>}
      cornerRadii=<0 0 0 0>
     }
     tintColor {colorUseRule=UseColor_Specified, specifiedColor=<rgba 1 1 1 1>}
     uVRegion {bIsValid=false, max=<0 0>, min=<0 0>}
    }
   }
   slot {class=HorizontalBoxSlot, properties {padding=<ltrb 0 0 10 0>}}
  }
  HelperText {
   widgetClassPath=&CommonTextBlock
   bIsVariable=true
   parent=HorizontalBox_0
   widget=HelperText
   properties {
    text=SWAP DEVICE
    textTransformPolicy=ToUpper
    font {
     bForceMonospaced=false
     bMaterialIsStencil=false
     fontMaterial=null
     fontObject=&HeadingNow
     letterSpacing=0
     monospacedWidth=1
     size=15
     skewAmount=0
     typefaceFontName=75 Medium
     outlineSettings {
      bApplyOutlineToDropShadows=false
      bMiteredCorners=false
      bSeparateFillAlpha=false
      outlineMaterial=null
      outlineSize=0
      outlineColor=<rgba 0 0 0 1>
     }
    }
    shadowOffset=<0 0>
   }
   slot {class=HorizontalBoxSlot, properties {verticalAlignment=VAlign_Center}}
  }
 ]
}
