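# Quickplay HUD widget for Pilgrim that lays out the local player's gameplay overlays.
# A bottom-left safe zone stacks star bar and score, with centered instrument-variety multiplier, collapsed swap-device widget and full-screen post-buffer capture.
# On initialization it plays Anim_Intro, sliding and fading in score and star bar and animating swap-device opacity.
# The multiplier fade-out animation and router, track and gem-map variables are unused in this graph.
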WidgetBlueprint /PilgrimQuickplay/Widgets/WBP_PilgrimQuickplay_HUD
animations [
 InstrumentVarietyMultiplier_FadeOut {
  binding_count=1
  length_seconds=5
  object=&self:InstrumentVarietyMultiplier_FadeOut
  sequence {bindings [WBP_PilgrimQuickplay_InstrumentVarietyMultiplier {
     id=#1
     possessed_class=WBP_PilgrimQuickplay_InstrumentVarietyMultiplier_C
     tracks [{
       type=MovieSceneFloatTrack
       display_name=Render Opacity
       sections [{
         type=MovieSceneFloatSection
         end_frame=100
         start_frame=0
         channels [None {type=MovieSceneScriptingFloatChannel, keys [[0,1],[80,1],[100,0]]}]
        }]
      }]
    }]}
 }
 Anim_Intro {
  binding_count=3
  length_seconds=2.75
  object=&self:Anim_Intro
  sequence {bindings [
    WBP_Pilgrim_Quickplay_SwapDevice {
     id=#2
     possessed_class=WBP_PilgrimQuickplay_SwapDevice_C
     tracks [{
       type=MovieSceneColorTrack
       display_name=Color and Opacity
       sections [{
         type=MovieSceneColorSection
         end_frame=60
         start_frame=5
         channels [
          Color.R {type=MovieSceneScriptingFloatChannel, const [1,1]}
          Color.G {type=MovieSceneScriptingFloatChannel, const [1,1]}
          Color.B {type=MovieSceneScriptingFloatChannel, const [1,1]}
          Color.A {type=MovieSceneScriptingFloatChannel, keys [[5,0],[15,0],[20,1]]}
         ]
        }]
      }]
    }
    WBP_PilgrimQuickplay_LocalScore {
     id=#3
     possessed_class=WBP_PilgrimQuickplay_LocalScore_C
     tracks [
      {
       type=MovieSceneFloatTrack
       display_name=Render Opacity
       sections [{type=MovieSceneFloatSection, channels [None {type=MovieSceneScriptingFloatChannel, keys [[11,0],[15,1]]}]}]
      }
      {
       type=MovieScene2DTransformTrack
       sections [{type=MovieScene2DTransformSection, channels [Translation.X {type=MovieSceneScriptingFloatChannel, keys [[9.6667,-384],[15,0]]}]}]
      }
     ]
    }
    WBP_PilgrimQuickplay_LocalStarBar {
     id=#4
     possessed_class=WBP_PilgrimQuickplay_LocalStarBar_C
     tracks [
      {
       type=MovieSceneFloatTrack
       display_name=Render Opacity
       sections [{type=MovieSceneFloatSection, channels [None {type=MovieSceneScriptingFloatChannel, keys [[9.6667,0],[13.6667,1]]}]}]
      }
      {
       type=MovieScene2DTransformTrack
       sections [{
         type=MovieScene2DTransformSection
         channels [Translation.X {type=MovieSceneScriptingFloatChannel, keys [[5,-384],[8.3333,-384],[13.6667,0]]}]
        }]
      }
     ]
    }
   ]}
 }
]
graphs [EventGraph {nodes [
   Event Construct()
   > Parent: Construct()
   Event OnInitialized()
   > PlayAnimation(InAnimation=Get Anim_Intro() -> (Anim_Intro:Widget Animation Object Reference))
  ]}]
variables [
 Event Router {type=GameplayEventRouterComponent, default=null}
 EventRouter {type=GameplayEventRouterComponent, default=null}
 LabelGemMap {default {}}
 Local Music Track {type=BP_PilgrimTrack_C, default=null}
]
widgets {
 info {inheritedWidgetCount=0, parentClass=&UserWidget, rootWidgetClass=&Overlay, widgetCount=9}
 widgets [
  Overlay_44 {widgetClassPath=&Overlay, widget=Overlay_44}
  WBP_CaptureForPostBufferUpdate {
   widgetClassPath=&WBP_CaptureForPostBufferUpdate
   bIsVariable=true
   parent=Overlay_44
   slot {class=OverlaySlot, properties {horizontalAlignment=HAlign_Fill, verticalAlignment=VAlign_Fill}}
  }
  SZ_LeftCornerStack {
   widgetClassPath=&SafeZone
   parent=Overlay_44
   widget=SZ_LeftCornerStack
   properties {padRight=false, padTop=false}
   slot {class=OverlaySlot, properties {verticalAlignment=VAlign_Bottom, padding=<ltrb 22 0 0 38>}}
  }
  VerticalBox_28 {widgetClassPath=&VerticalBox, parent=SZ_LeftCornerStack, widget=VerticalBox_28, slot {class=SafeZoneSlot}}
  VBox_LeftContent {
   widgetClassPath=&VerticalBox
   parent=VerticalBox_28
   widget=VBox_LeftContent
   properties {renderTransformPivot=<1 0.5>}
   slot {class=VerticalBoxSlot, properties {horizontalAlignment=HAlign_Left, padding=<ltrb 0 8 0 0>}}
  }
  WBP_PilgrimQuickplay_LocalStarBar {
   widgetClassPath=&/WBP_PilgrimQuickplay_LocalStarBar
   bIsVariable=true
   parent=VBox_LeftContent
   widget=WBP_PilgrimQuickplay_LocalStarBar
   properties {pixelSnapping=Disabled, shouldStarsAnimateComplete=true}
   slot {class=VerticalBoxSlot, properties {padding=<ltrb 0 0 0 8>}}
  }
  WBP_PilgrimQuickplay_LocalScore {
   widgetClassPath=&/WBP_PilgrimQuickplay_LocalScore
   bIsVariable=true
   parent=VBox_LeftContent
   widget=WBP_PilgrimQuickplay_LocalScore
   properties {pixelSnapping=Disabled}
   slot {class=VerticalBoxSlot}
  }
  WBP_Pilgrim_Quickplay_SwapDevice {
   widgetClassPath=&/WBP_PilgrimQuickplay_SwapDevice
   bIsVariable=true
   parent=Overlay_44
   widget=WBP_Pilgrim_Quickplay_SwapDevice
   properties {visibility=Collapsed}
   slot {class=OverlaySlot, properties {horizontalAlignment=HAlign_Center, verticalAlignment=VAlign_Center, padding=<ltrb 0 0 0 160>}}
  }
  WBP_PilgrimQuickplay_InstrumentVarietyMultiplier {
   widgetClassPath=&/WBP_PilgrimQuickplay_InstrumentVarietyMultiplier
   bIsVariable=true
   parent=Overlay_44
   widget=WBP_PilgrimQuickplay_InstrumentVarietyMultiplier
   slot {class=OverlaySlot, properties {horizontalAlignment=HAlign_Center, verticalAlignment=VAlign_Center, padding=<ltrb 0 0 0 300>}}
  }
 ]
}
