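# UI material that displays the Pilgrim track projection render target with a fade applied.
# The RenderTarget texture is sampled with UVs scaled horizontally about the center by TrackWidth.
# MF_PilgrimTrackFade multiplies both the color and the inverted alpha, producing faded color and coverage suited to alpha-composite blending.
# Separate radial-gradient, step and cosine mask branches, some labeled old or previously plugged in, remain outside that chain.
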Material /PilgrimCore/Cameras/M_PilgrimTrackProjection
graph {
 properties {
  blendMode=BLEND_AlphaComposite
  materialDomain=MD_UI
  opacityMaskClipValue=1
  pixelDepthOffsetMode=PDOM_Legacy
  usedShadingModels=Default Lit
 }
 nodes [
  MaterialFunctionCall(CenterPosition=Constant2Vector() {g=0.75, r=0.48}, Radius=Constant() {r=0.45}, Density=Constant() {r=6}) {materialFunction=&RadialGradientExponential}
  n1 = ScalarParameter() {defaultValue=1, parameterName=TrackWidth}
  n2 = Constant() {r=1}
  n3 = TextureSample(
   UVs=MaterialFunctionCall(Texture Scale=AppendVector(A=@n1, B=@n2)) {materialFunction=&ScaleUVsByCenter}.UVs
   Tex=TextureObjectParameter() {parameterName=RenderTarget, texture=&/RT_PilgrimTrackProjection}
  ) {texture=&RT_PilgrimTrackProjection}
  n4 = MaterialFunctionCall!(UV) {materialFunction=&/MF_PilgrimTrackFade}
  Multiply(A=@n3.RGB, B=@n4.Fade)
  Multiply(A=OneMinus(None=@n3.A), B=@n4.Fade)
  n5 = Cosine(None=ComponentMask(None=TextureCoordinate() {vTiling=0.5}) {g=true}) {period=1.2}
  n6 = ComponentMask(None=TextureCoordinate() {vTiling=0.66}) {g=true}
  n7 = Step(X=@n6) {constY=0.25}
  n8 = Multiply(A=SmoothStep(
    Min=@n6
    Max=ScalarParameter() {defaultValue=0.275, parameterName=Param_1}
    Value=ScalarParameter() {defaultValue=0.29, parameterName=Param_2}
   ) {constMax=0.25, constValue=0.275}) {constB=2}
  Multiply(A=Saturate(None=Multiply(A=@n7, B=@n8)) {desc=new mask}, B=OneMinus(None=Multiply(A=@n5) {constB=0.9})) {constB=2, desc=previously was plugged in}
  OneMinus(None=Multiply(A=Cosine(None=@n6) {period=1.2}) {constB=2}) {desc=old mask}
  n9 = ComponentMask(None=TextureCoordinate() {vTiling=0.3}) {g=true}
  Saturate(None=Multiply(A=Cosine(None=@n9) {period=1.2}) {constB=10})
 ]
}
