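# Pilgrim track background widget: a single dark Image whose opacity follows the player's track background opacity setting.
# On Construct it caches the client setting record from the Pilgrim player controller component and applies the opacity immediately.
# It also waits for the gameplay event router, then listens for opacity-changed events, re-reading the setting rather than using the payload.
# Destruct stops the stored listener handle and clears it.
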WidgetBlueprint /PilgrimCore/Widgets/WBP_Pilgrim_TrackBackground
graphs [EventGraph {nodes [
   Event Destruct()
   > Sequence()
   n1 = Get OpacityChangedEventHandle() -> (OpacityChangedEventHandle:Async Action Start Listening To Event Object Reference)
   > Is Valid(InputObject=@n1)
   > StopListeningToEvent(self:Async Action Start Listening To Event Object Reference=@n1)
   > Set OpacityChangedEventHandle(OpacityChangedEventHandle:Async Action Start Listening To Event Object Reference) -> (Output_Get:Async Action Start Listening To Event Object Reference)
   Event Construct()
   > n2 = Cast To FortPlayerControllerGameplay!(execute, Object=GetOwningPlayer()) -> (then, CastFailed, AsFort Player Controller Gameplay:Fort Player Controller Gameplay Object Reference) {is=title K2Node_DynamicCast}
   > Set PilgrimClientSettingRecordPartition(PilgrimClientSettingRecordPartition:Pilgrim Client Setting Record Partition Object Reference=GetPilgrimClientSettingRecordPartition!(self:Pilgrim Player Controller Component Object Reference=GetComponentByClass!(
      self:Actor Object Reference=@n2
      ComponentClass:Actor Component Class Reference=/PilgrimCore/Player/Components/BP_PilgrimPlayerControllerComponent.BP_PilgrimPlayerControllerComponent_C
     ) -> (ReturnValue:BP Pilgrim Player Controller Component Object Reference)) -> (ReturnValue:Pilgrim Client Setting Record Partition Object Reference)) -> (Output_Get:Pilgrim Client Setting Record Partition Object Reference)
   > n3 = Sequence()
   n4 = PilgrimWaitForEventRouter!(execute:Exec=@n3.then_1, ContextActor:Actor Object Reference=GetOwningPlayer()) -> (then:Exec, OnEventRouterAvailable:Exec, EventRouter:Gameplay Event Router Component Object Reference) {is=title K2Node_AsyncAction}
   n5 = StartListeningToEvent!(
    execute:Exec=@n4.OnEventRouterAvailable
    Target:Gameplay Event Router Component Object Reference=@n4
    EventType:Script Struct Object Reference=/Script/PilgrimSettingsRuntime.PilgrimEvent_PilgrimSettings_TrackBackgroundOpacityChanged
    Context:Object Reference
    EventBubblingRule:EEventBubblingRule Enum
   ) -> (
    then:Exec
    AsyncTaskProxy:Async Action Start Listening To Event Object Reference
    OnEventReceived:Exec
    Payload:Pilgrim Event Pilgrim Settings Track Background Opacity Changed Structure
   )
   > Set OpacityChangedEventHandle(OpacityChangedEventHandle=@n5)
   SetOpacity!(
    execute=[@n3.then_0, @n5.OnEventReceived]
    self:Image Object Reference=Get Background() -> (Background:Image Object Reference)
    InOpacity:float=GetTrackBackgroundOpacity!(self:Pilgrim Client Setting Record Partition Object Reference=Get PilgrimClientSettingRecordPartition() -> (PilgrimClientSettingRecordPartition:Pilgrim Client Setting Record Partition Object Reference)) -> (ReturnValue:float)
   ) -> (then)
  ]}]
variables [
 OpacityChangedEventHandle {type=AsyncAction_StartListeningToEvent, default=null}
 PilgrimClientSettingRecordPartition {type=PilgrimClientSettingRecordPartition, default=null}
]
widgets {
 info {inheritedWidgetCount=0, parentClass=&UserWidget, rootWidgetClass=&Image, widgetCount=1}
 widgets [Background {
   widgetClassPath=&Image
   bIsVariable=true
   widget=Background
   properties {visibility=SelfHitTestInvisible, colorAndOpacity=<rgba 0 0 0 0.1>}
  }]
}
