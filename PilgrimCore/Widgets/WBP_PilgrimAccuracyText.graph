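# Pilgrim rhythm-game HUD widget that shows Perfect, Early or Late hit banners and hosts a streak counter.
# It listens for GemScored events, filters to the owning player, and classifies hits by tier-name text match and timing sign.
# Without detailed feedback only Perfects show; SingleSpawnPoint routes every result to the Perfect banner, otherwise fixed banners get nudged.
# Banner shear follows lane count, and visuals receive a local-multiplayer flag.
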WidgetBlueprint /PilgrimCore/Widgets/WBP_PilgrimAccuracyText
graphs [
 CallNextAvailableAccuracyBanner {nodes [
   n1 = CallNextAvailableAccuracyBanner!() -> (then, Accuracy:Pilgrim Gem Hit Accuracy Structure) {is=title K2Node_FunctionEntry}
   > n2 = Sequence!(execute) -> (then_0, then_1, then_2:Exec, then_3:Exec)
   n3 = GetOwningPlayer()
   > Set IsLocalMP(IsLocalMP:bool=Is Local Player in Local MP Session!(PlayerState:Player State Object Reference=Get PlayerState!(self:Controller Object Reference=@n3) -> (PlayerState:Player State Object Reference)) -> (ReturnValue:bool)) -> (Output_Get:bool)
   n4 = GetLocalPilgrimGameForPlayer!(
    self:Pilgrim Music Battle Instance Object Reference=Get MusicBattleInstance() -> (MusicBattleInstance:Pilgrim Music Battle Instance Object Reference)
    InPC:Player Controller Object Reference=@n3
   ) -> (ReturnValue:Pilgrim Game Object Reference)
   Set NumberOfLanes(execute=@n2.then_2, NumberOfLanes:int=GetNumLanes!(self:Pilgrim Game Object Reference=@n4) -> (ReturnValue:int=0)) -> (Output_Get:int=0)
   Set IsUsingDetailedFeedback(
    execute=@n2.then_1
    IsUsingDetailedFeedback:bool=GetShowDetailedHitFeedback!(self:Pilgrim Track Object Reference=GetMainTrack!(self:Pilgrim Game Object Reference=@n4) -> (ReturnValue:Pilgrim Track Object Reference)) -> (ReturnValue:bool)
   ) -> (Output_Get:bool)
   n5 = GetEnumAccuracy!(execute=@n2.then_3, self, PilgrimGemAccuracy:Pilgrim Gem Hit Accuracy Structure=@n1) -> (then, ReturnValue:E_GemHitAccuracy Enum=NewEnumerator0)
   > n6 = Branch(Condition=Get IsUsingDetailedFeedback() -> (IsUsingDetailedFeedback:bool))
   n7 = GetAccuracyBannerWidget!(execute=[@n8.then, @n6.then], self, PilgrimGemHitAccuracy:"Pilgrim Gem Hit Accuracy Structure (by ref)"=@n1) -> (then, AccuracyBannerWidget:WBP Pilgrim Accuracy Text Instance Object Reference)
   > SetAccuracyBannerVisuals!(
    execute
    self
    BannerInstance:WBP Pilgrim Accuracy Text Instance Object Reference=@n7
    AccuracyStruct:Pilgrim Gem Hit Accuracy Structure=@n1
    IsLocalMP:bool=Get IsLocalMP() -> (IsLocalMP:bool)
    IsDetailedFeedback:bool=Get IsUsingDetailedFeedback()
   ) -> (then)
   n8 = Branch(execute=@n6.else, Condition="Equal (Enum)"(A:E_GemHitAccuracy Enum=@n5, B:E_GemHitAccuracy Enum=NewEnumerator1))
   Return Node(execute=@n8.else)
  ]}
 EventGraph {nodes [
   Event Construct()
   > Parent: Construct()
   > n1 = PilgrimWaitForEventRouter!(execute:Exec, ContextActor:Actor Object Reference=GetOwningPlayer()) -> (then:Exec, OnEventRouterAvailable:Exec, EventRouter:Gameplay Event Router Component Object Reference) {is=title K2Node_AsyncAction}
   n2 = StartListeningToEvent!(
    execute:Exec=@n1.OnEventRouterAvailable
    Target:Gameplay Event Router Component Object Reference=@n1
    EventType:Script Struct Object Reference=/Script/PilgrimCoreRuntime.PilgrimEvent_Scoring_GemScored
    Context:Object Reference=Get PlayerState(self=GetOwningPlayer())
    EventBubblingRule:EEventBubblingRule Enum
   ) -> (
    then:Exec
    AsyncTaskProxy:Async Action Start Listening To Event Object Reference
    OnEventReceived:Exec
    Payload:Pilgrim Event Scoring Gem Scored Structure
   )
   > Add(
    TargetArray:Array of Async Action Start Listening To Event Object References=Get GameplayEventHandles() -> (GameplayEventHandles:Array of Async Action Start Listening To Event Object References)
    NewItem:"Async Action Start Listening To Event Object Reference (by ref)"=@n2
   ) {is=title K2Node_CallArrayFunction}
   n3 = Break Pilgrim Event Scoring Gem Scored!(PilgrimEvent_Scoring_GemScored:"Pilgrim Event Scoring Gem Scored Structure (by ref)"=@n2.Payload) -> (
    ForPlayer:Fort Player State Object Reference
    Accuracy:Pilgrim Gem Hit Accuracy Structure
    Gem:Track Item Info Structure
    BaseScore:float=~
    BonusScore:float=~
   ) {is=title K2Node_BreakStruct}
   Branch(
    execute=@n2.OnEventReceived
    Condition="Equal (Object)"(A=GetPlayerController!(self:Player State Object Reference=@n3) -> (ReturnValue:Player Controller Object Reference), B=GetOwningPlayer())
   )
   > CallNextAvailableAccuracyBanner!(execute, self, Accuracy:Pilgrim Gem Hit Accuracy Structure=@n3.Accuracy) -> (then) {is=title K2Node_CallFunction}
   Event Destruct()
   n4 = Get GameplayEventHandles()
   > n5 = For Each Loop(Array:Array of Async Action Start Listening To Event Object References=@n4) -> (Array Element:Async Action Start Listening To Event Object Reference)
   Clear(execute=@n5.Completed, TargetArray:Array of Async Action Start Listening To Event Object References=@n4)
   Is Valid(exec=@n5.LoopBody, InputObject=@n5)
   > StopListeningToEvent(self:Async Action Start Listening To Event Object Reference=@n5)
  ]}
 Get Accuracy Text {nodes [
   n1 = Get Accuracy Text!() -> (then, PilgrimGemHitAccuracy:"Pilgrim Gem Hit Accuracy Structure (by ref)") {is=title K2Node_FunctionEntry}
   > Return Node!(
    execute
    ReturnValue:Text=Break Pilgrim Accuracy Tier!(PilgrimAccuracyTier:"Pilgrim Accuracy Tier Structure (by ref)"=Break Pilgrim Gem Hit Accuracy!(PilgrimGemHitAccuracy:"Pilgrim Gem Hit Accuracy Structure (by ref)"=@n1) -> (AccuracyMs:float=~, AccuracyTier:Pilgrim Accuracy Tier Structure, TierIndex:int) {is=title K2Node_BreakStruct}.AccuracyTier) -> (ThresholdMs:float=~, ScoreMultiplier:float=~, Name:Text) {is=title K2Node_BreakStruct}.Name
   )
  ]}
 GetAccuracyBannerWidget {nodes [
   n1 = GetAccuracyBannerWidget!() -> (then, PilgrimGemHitAccuracy:"Pilgrim Gem Hit Accuracy Structure (by ref)") {is=title K2Node_FunctionEntry}
   > n2 = GetEnumAccuracy(PilgrimGemAccuracy=@n1) -> (ReturnValue=NewEnumerator0)
   n3 = Break Pilgrim Gem Hit Accuracy(PilgrimGemHitAccuracy=@n1) -> (AccuracyMs=~)
   n4 = Select(
    Option 0:E_GemHitAccuracy Enum=Select(Option 0=NewEnumerator2, Option 1=NewEnumerator0, Index="float < float"(A=@n3))
    Option 1:E_GemHitAccuracy Enum=NewEnumerator1
    Index="Equal (Enum)"(A=@n2, B=NewEnumerator1)
   ) -> (ReturnValue:E_GemHitAccuracy Enum)
   n5 = Get WBP_PilgrimAccuracyText_Late() -> (WBP_PilgrimAccuracyText_Late:WBP Pilgrim Accuracy Text Instance Object Reference)
   n6 = Get WBP_PilgrimAccuracyText_Perfect() -> (WBP_PilgrimAccuracyText_Perfect:WBP Pilgrim Accuracy Text Instance Object Reference)
   n7 = Get WBP_PilgrimAccuracyText_Early() -> (WBP_PilgrimAccuracyText_Early:WBP Pilgrim Accuracy Text Instance Object Reference)
   > Return Node!(
    execute
    AccuracyBannerWidget:WBP Pilgrim Accuracy Text Instance Object Reference=Select(
     Option 0:WBP Pilgrim Accuracy Text Instance Object Reference=Select!(
      NewEnumerator0:WBP Pilgrim Accuracy Text Instance Object Reference=@n7
      NewEnumerator1:WBP Pilgrim Accuracy Text Instance Object Reference=@n6
      NewEnumerator2:WBP Pilgrim Accuracy Text Instance Object Reference=@n5
      Index:E_GemHitAccuracy Enum=@n4
     ) -> (ReturnValue)
     Option 1:WBP Pilgrim Accuracy Text Instance Object Reference=Get WBP_PilgrimAccuracyText_Perfect()
     Index=Get SingleSpawnPoint() -> (SingleSpawnPoint:bool)
    ) -> (ReturnValue:WBP Pilgrim Accuracy Text Instance Object Reference)
   )
  ]}
 GetEnumAccuracy {nodes [
   n1 = GetEnumAccuracy!() -> (then, PilgrimGemAccuracy:Pilgrim Gem Hit Accuracy Structure) {is=title K2Node_FunctionEntry}
   n2 = Break Pilgrim Gem Hit Accuracy(PilgrimGemHitAccuracy=@n1) -> (AccuracyMs=~)
   n3 = Break Pilgrim Accuracy Tier(PilgrimAccuracyTier=@n2.AccuracyTier) -> (ThresholdMs=~, ScoreMultiplier=~)
   n4 = Get PerfectText() -> (PerfectText:Text)
   n5 = "float < float"(A=@n2)
   > Return Node!(
    execute
    ReturnValue:E_GemHitAccuracy Enum=Select!(
     Option 0:E_GemHitAccuracy Enum=Select(Option 0=NewEnumerator2, Option 1=NewEnumerator0, Index=@n5)
     Option 1:E_GemHitAccuracy Enum=NewEnumerator1
     Index:bool="Equal Exactly (Text)"!(A:"Text (by ref)"=@n4, B:"Text (by ref)"=@n3.Name) -> (ReturnValue:bool)
    ) -> (ReturnValue:E_GemHitAccuracy Enum)
   )
  ]}
 GetMaximumDisplayedAccuracyTier {nodes [
   GetMaximumDisplayedAccuracyTier() {is=title K2Node_FunctionEntry}
   n1 = Get MusicBattleInstance()
   > n2 = Branch(Condition=IsValid(Object=@n1))
   n3 = GetConfig!(self:Pilgrim Music Battle Instance Object Reference=@n1) -> (ReturnValue:Pilgrim Settings Object Reference)
   > Return Node!(
    execute
    TierIndex:int=Break Pilgrim Game Config Settings!(PilgrimGameConfigSettings:"Pilgrim Game Config Settings Structure (by ref)"=Get ConfigSettings!(self:Pilgrim Settings Object Reference=@n3) -> (ConfigSettings:Pilgrim Game Config Settings Structure)) -> (
     SlopMs:float=~
     GemBreakEndInputUnlockMs:float=~
     EndOfTrackBuffer:float=~
     MsToNowBar:float=~
     BackgroundTrackDb:float=~
     BackgroundTrackGain:float=~
     SongAudioResponceCueSettingsMap:Map of EPilgrimSongAudioResponseCue Enums to Pilgrim Song Audio Response Cue Settings Structures
     FireMidiEventTicksEarly:float=~
     BaseGemScoreValue:"Map of EPilgrimSongDifficulty Enums to Float (single-precision)s"
     BaseCymbalGemScoreValue:"Map of EPilgrimSongDifficulty Enums to Float (single-precision)s"
     SustainScorePerBeat:int
     bNormalizeScores:bool=~
     NormalizedScaleTowardsHighestInDifficulty:Scalable Float Structure
     NormalizedScaleTowardsHighestInExpert:Scalable Float Structure
     OutroPreAlertSeconds:float=~
     ScoreMultiplierDefinitions:Array of Pilgrim Score Multiplier Info Structures
     PlayerHealthModifiers:Map of EPilgrimSongDifficulty Enums to Player Health Modifiers Structures
     PilgrimHealthCrowdResponseTiers:"Array of Float (single-precision)s"
     IndividualStarPercentageThresholds:Map of EPilgrimSongDifficulty Enums to Star Percentage Thresholds Structures
     BandStarPercentageThresholds:"Array of Float (single-precision)s"
     IsBandMultiplierBasedOnUniqueInstruments:bool=~
     IsBandVarietyBonusScoredDuringPlay:bool=~
     BandOverdriveScoreMultipliers:"Array of Float (single-precision)s"
     BandVarietyScoreMultipliers:"Array of Float (single-precision)s"
     AccuracyTiers:Array of Pilgrim Accuracy Tier Structures
     MaximumDisplayedAccuracyTierIndex:int
     TeamQuickplayInitialHealth:float=~
     TeamQuickplayOverdriveDecayRate:float=~
     TeamQuickplayInitialOverdrive:float=~
     TeamQuickplayOverdrive:float=~
     TeamQuickplayOverdriveScoreMultiplier:float=~
     TeamQuickplayOverdriveHealthGainMultiplier:float=~
     TeamQuickplayRequiredOverdriveToActivate:float=~
     TeamQuickplayOverdriveAddedOnSuccesfulSection:float=~
     bOverdriveClearsGems:bool=~
     bOverdriveClearAutoHoldSustain:bool=~
     GemBreakPendingBars:float=~
     bEnableInputSpecificGemsAndSmashers:bool=~
     InputConfig:Pilgrim Input Settings Object Reference
     bEnableOverdriveDeploys:bool=~
     bEnableOverdrivePhrases:bool=~
     bEnableCaptureSystem:bool=~
     bEnableBandPerformanceScoring:bool=~
     ArenaInitialShields:float=~
     ArenaInitialHealth:float=~
     ArenaInitialAmmo:bool=~
     DifficultyCaptureFactors:"Map of EPilgrimSongDifficulty Enums to Float (single-precision)s"
     PerInstrumentConfigs:Map of EPilgrimTrackType Enums to Pilgrim Settings Instrument Overrides Object References
     StreakToastThresholds:Array of Integers
     InstrumentTypeMissCues:Map of ESparksInstrumentType Enums to Sound Cue Object References
     DefaultInstrumentTypeMissCue:Sound Cue Object Reference
     CreateStandinBandmembers:bool=~
     SongIntroEmbargoMs:float=~
     DeployOverdriveGem_BarsUntilGemSpawn:float=~
     DeployOverdriveGem_TailDurationBar:float=~
     RollsMinInputForgivenessMs:float=~
     SwellsMinInputForgivenessMs:float=~
     ProVocalPitchThresholds:Map of EPilgrimSongDifficulty Enums to Pro Vocal Scoring Pitch Range Structures
     ProVocalSectionCompletionThresholds:"Map of EPilgrimSongDifficulty Enums to Float (single-precision)s"
     ProVocalsSectionMaxValue:"Map of EPilgrimSongDifficulty Enums to Float (single-precision)s"
     ProVocalsSectionRatings:Array of Pro Vocals Section Rating Structures
     ProVocalsScoreMultiplierDefinitions:Array of Pilgrim Score Multiplier Info Structures
     ScoreWeightInSection:Curve Float Object Reference
    ) {is=title K2Node_BreakStruct}.MaximumDisplayedAccuracyTierIndex
   )
   Return Node(execute=@n2.else, TierIndex=-1)
  ]}
 GetStreakCounterVisibility {nodes [
   GetStreakCounterVisibility() {is=title K2Node_FunctionEntry}
   > Return Node!(
    execute
    Visibility:ESlateVisibility Enum=GetVisibility!(self:Widget Object Reference=Get WBP_PilgrimCore_StreakCounter() -> (WBP_PilgrimCore_StreakCounter:WBP Pllgrim Core Streak Counter Object Reference)) -> (ReturnValue:ESlateVisibility Enum=Visible)
   )
  ]}
 SetAccuracyBannerVisuals {nodes [
   n1 = SetAccuracyBannerVisuals!() -> (
    then
    BannerInstance:WBP Pilgrim Accuracy Text Instance Object Reference
    AccuracyStruct:Pilgrim Gem Hit Accuracy Structure
    IsLocalMP:bool=~
    IsDetailedFeedback:bool=~
   ) {is=title K2Node_FunctionEntry}
   > n2 = GetAccuracyEnum!(
    execute
    self:WBP Pilgrim Accuracy Text Instance Object Reference=@n1
    PilgrimGemHitAccuracy:"Pilgrim Gem Hit Accuracy Structure (by ref)"=@n1.AccuracyStruct
   ) -> (then, ReturnValue:E_GemHitAccuracy Enum=NewEnumerator0)
   > n3 = Cast To GridSlot!(execute, Object=Get Slot!(self:Widget Object Reference=@n1) -> (Slot:Panel Slot Object Reference)) -> (then, CastFailed, AsGrid Slot:Grid Slot Object Reference) {is=title K2Node_DynamicCast}
   > Set VerticalAlignment!(execute, VerticalAlignment:EVerticalAlignment Enum=VAlign_Center, self:Grid Slot Object Reference=@n3) -> (then, Output_Get:EVerticalAlignment Enum=VAlign_Fill)
   > Set HorizontalAlignment!(execute, HorizontalAlignment:EHorizontalAlignment Enum=HAlign_Right, self:Grid Slot Object Reference=@n3) -> (then, Output_Get:EHorizontalAlignment Enum=HAlign_Fill)
   > n4 = Branch(Condition=Get SingleSpawnPoint())
   Set GridSlot!(execute=[@n4.then, @n9.then], GridSlot:Grid Slot Object Reference=@n3, self:WBP Pilgrim Accuracy Text Instance Object Reference=@n1) -> (then, Output_Get:Grid Slot Object Reference)
   n5 = Get NumberOfLanes() -> (NumberOfLanes:int=0)
   > SetAccuracyVisuals!(
    execute
    self:WBP Pilgrim Accuracy Text Instance Object Reference=@n1
    ShearX:double=Select(
     Option 0:double=Get 5LaneShearXAxis() -> (5LaneShearXAxis:double)
     Option 1:double=Get 4LaneShearXAxis() -> (4LaneShearXAxis:double)
     Index="Equal (Integer)"(A=@n5, B=4)
    ) -> (ReturnValue:double)
    IsDesignTime:bool=false
    IsLocalMP:bool=@n1.IsLocalMP
    IsDetailedFeedback:bool=@n1.IsDetailedFeedback
    InAccuracy:E_GemHitAccuracy Enum=@n2
   ) -> (then, Accuracy:E_GemHitAccuracy Enum=NewEnumerator0)
   n6 = Get OffsetY() -> (OffsetY:double)
   n7 = Get 5LaneShearXAxis()
   n8 = Get SingleSpawnPoint()
   n9 = Set Nudge!(
    execute=@n4.else
    Nudge:Vector 2D Structure=Select!(
     NewEnumerator0_X:double="Absolute (Float)"(A=@n7)
     NewEnumerator0_Y:double=@n6
     NewEnumerator1_X:double=0
     NewEnumerator1_Y:double=0
     NewEnumerator2_X:double=@n7
     NewEnumerator2_Y:double="Absolute (Float)"(A=@n6)
     Index:E_GemHitAccuracy Enum=Select!(Option 0:E_GemHitAccuracy Enum=@n2, Option 1:E_GemHitAccuracy Enum=NewEnumerator1, Index:bool=@n8) -> (ReturnValue:E_GemHitAccuracy Enum)
    ) -> (ReturnValue:Vector 2D Structure)
    self:Grid Slot Object Reference=@n3
   ) -> (then, Output_Get:Vector 2D Structure)
  ]}
 SetStreakCounterVisibility {nodes [
   n1 = SetStreakCounterVisibility!() -> (then, New Visibility:ESlateVisibility Enum=Visible) {is=title K2Node_FunctionEntry}
   > Set Visibility(Visibility=@n1, self=Get WBP_PilgrimCore_StreakCounter())
  ]}
 SetupRenderTransform {nodes [
   SetupRenderTransform() {is=title K2Node_FunctionEntry}
   > SetRenderTranslation(Translation=Select(
     Option 0:Vector 2D Structure=Get GemHitFeedbackPosition01_Render() -> (GemHitFeedbackPosition01_Render:Vector 2D Structure)
     Option 1:Vector 2D Structure=Get GemHitFeedbackPosition02_Render() -> (GemHitFeedbackPosition02_Render:Vector 2D Structure)
     Index=GetUseHitFeedbackUpTrack!() -> (ReturnValue:bool)
    ) -> (ReturnValue:Vector 2D Structure))
  ]}
]
variables [
 4LaneShearXAxis {type=number, default=-16}
 5LaneShearXAxis {type=number, default=-24}
 GemHitFeedbackPosition01_Render {type=Vector2D, default=<5 0>}
 GemHitFeedbackPosition02_Render {type=Vector2D, default=<80 -200>}
 NumberOfLanes {type=integer}
 OffsetY {type=number, default=-60}
 PilgrimGameConfig {
  type=PilgrimGameConfigSettings
  default {
   arenaInitialAmmo=false
   arenaInitialHealth=100
   arenaInitialShields=100
   bEnableBandPerformanceScoring=true
   bEnableCaptureSystem=false
   bEnableInputSpecificGemsAndSmashers=false
   bEnableOverdriveDeploys=true
   bEnableOverdrivePhrases=true
   bNormalizeScores=false
   bOverdriveClearAutoHoldSustain=false
   bOverdriveClearsGems=false
   backgroundTrackDb=-5
   backgroundTrackGain=0.8
   createStandinBandmembers=false
   defaultInstrumentTypeMissCue=null
   deployOverdriveGem_BarsUntilGemSpawn=2.5
   deployOverdriveGem_TailDurationBar=1
   endOfTrackBuffer=20
   fireMidiEventTicksEarly=960
   gemBreakEndInputUnlockMs=150
   gemBreakPendingBars=4
   inputConfig=null
   isBandMultiplierBasedOnUniqueInstruments=false
   isBandVarietyBonusScoredDuringPlay=true
   maximumDisplayedAccuracyTierIndex=0
   msToNowBar=3000
   outroPreAlertSeconds=5
   rollsMinInputForgivenessMs=160
   scoreWeightInSection=null
   slopMs=150
   songIntroEmbargoMs=2000
   staticLyricsAspectRatioTimeWindow=null
   sustainScorePerBeat=12
   swellsMinInputForgivenessMs=500
   teamQuickplayInitialHealth=0.7
   teamQuickplayInitialOverdrive=0
   teamQuickplayOverdrive=0.25
   teamQuickplayOverdriveAddedOnSuccesfulSection=0.25
   teamQuickplayOverdriveDecayRate=0.0312
   teamQuickplayOverdriveHealthGainMultiplier=1.5
   teamQuickplayOverdriveScoreMultiplier=2
   teamQuickplayRequiredOverdriveToActivate=0.5
   accuracyTiers []
   bandOverdriveScoreMultipliers [1,2,4,6,8]
   bandStarPercentageThresholds [0.1,0.2,0.4,0.7,1.1]
   bandVarietyScoreMultipliers [1,1.1,1.2,1.3]
   baseCymbalGemScoreValue {difficultyEasy=30, difficultyExpert=35, difficultyHard=30, difficultyMedium=30}
   baseGemScoreValue {difficultyEasy=25, difficultyExpert=25, difficultyHard=25, difficultyMedium=25}
   difficultyCaptureFactors {difficultyEasy=0.8, difficultyExpert=1.4, difficultyHard=1.2, difficultyMedium=1}
   individualStarPercentageThresholds {
    difficultyEasy {thresholds [0.1,0.2,0.4,0.7,1.1]}
    difficultyExpert {thresholds [0.1,0.2,0.4,0.7,1.1,1.75]}
    difficultyHard {thresholds [0.1,0.2,0.4,0.7,1.1]}
    difficultyMedium {thresholds [0.1,0.2,0.4,0.7,1.1]}
   }
   instrumentTypeMissCues {}
   normalizedScaleTowardsHighestInDifficulty {registryType=None, value=0.8, curve {curveTable=null, rowName=None}}
   normalizedScaleTowardsHighestInExpert {registryType=None, value=0.85, curve {curveTable=null, rowName=None}}
   perInstrumentConfigs {}
   pilgrimHealthCrowdResponseTiers [0.2,0.8]
   playerHealthModifiers {
    difficultyEasy {gemHit=0.1, gemMiss=-0.01, gemPass=0}
    difficultyExpert {gemHit=0.01, gemMiss=-0.1, gemPass=-0.1}
    difficultyHard {gemHit=0.03, gemMiss=-0.07, gemPass=-0.05}
    difficultyMedium {gemHit=0.07, gemMiss=-0.03, gemPass=-0.02}
   }
   proVocalPitchThresholds {
    difficultyEasy {outerThresholdInSemitones=4, perfectThresholdInSemitones=2}
    difficultyExpert {outerThresholdInSemitones=2, perfectThresholdInSemitones=1}
    difficultyHard {outerThresholdInSemitones=2.5, perfectThresholdInSemitones=1.25}
    difficultyMedium {outerThresholdInSemitones=3, perfectThresholdInSemitones=1.5}
   }
   proVocalSectionCompletionThresholds {difficultyEasy=0.4, difficultyExpert=0.8, difficultyHard=0.7, difficultyMedium=0.5}
   proVocalsScoreMultiplierDefinitions [{notesCount=1, scoreMultiplier=2}, {notesCount=2, scoreMultiplier=3}, {notesCount=3, scoreMultiplier=4}]
   proVocalsSectionMaxValue {difficultyEasy=200, difficultyExpert=1000, difficultyHard=800, difficultyMedium=400}
   proVocalsSectionRatings []
   proVocalsTrackDensityTiers []
   scoreMultiplierDefinitions []
   songAudioResponceCueSettingsMap {gemHit {fadeDurationMs=50, trackGain=1}, gemPassed {fadeDurationMs=50, trackGain=0.3}, userSwingAndMiss {fadeDurationMs=50, trackGain=0}}
   streakToastThresholds [10,20]
  }
 }
 SingleSpawnPoint {type=boolean, default=true}
]
widgets {
 info {inheritedWidgetCount=0, parentClass=/Script/PilgrimCoreUI.PilgrimGameWidgetBase, rootWidgetClass=&SizeBox, widgetCount=18}
 widgets [
  SizeBox_155 {
   widgetClassPath=&SizeBox
   widget=SizeBox_155
   properties {bOverride_HeightOverride=true, bOverride_WidthOverride=true, heightOverride=200, widthOverride=205}
  }
  Overlay_0 {widgetClassPath=&Overlay, parent=SizeBox_155, widget=Overlay_0, slot {class=SizeBoxSlot}}
  Container {
   widgetClassPath=&GridPanel
   bIsVariable=true
   parent=Overlay_0
   widget=Container
   properties {bIsEnabled=false, bOverride_Cursor=true, visibility=Hidden}
   slot {class=OverlaySlot}
  }
  WBP_PilgrimAccuracyText_Instance {
   widgetClassPath=&/WBP_PilgrimAccuracyText_Instance
   bIsVariable=true
   parent=Container
   widget=WBP_PilgrimAccuracyText_Instance
   properties {renderOpacity=0, visibility=Hidden}
   slot {class=GridSlot, properties {horizontalAlignment=HAlign_Right, verticalAlignment=VAlign_Center}}
  }
  WBP_PilgrimAccuracyText_Instance_1 {
   widgetClassPath=&WBP_PilgrimAccuracyText_Instance
   bIsVariable=true
   parent=Container
   widget=WBP_PilgrimAccuracyText_Instance_1
   properties {visibility=Hidden}
   slot {class=GridSlot, properties {horizontalAlignment=HAlign_Right, verticalAlignment=VAlign_Center}}
  }
  WBP_PilgrimAccuracyText_Instance_2 {
   widgetClassPath=&WBP_PilgrimAccuracyText_Instance
   bIsVariable=true
   parent=Container
   widget=WBP_PilgrimAccuracyText_Instance_2
   properties {visibility=Hidden}
   slot {class=GridSlot, properties {horizontalAlignment=HAlign_Right, verticalAlignment=VAlign_Center}}
  }
  WBP_PilgrimAccuracyText_Instance_3 {
   widgetClassPath=&WBP_PilgrimAccuracyText_Instance
   bIsVariable=true
   parent=Container
   widget=WBP_PilgrimAccuracyText_Instance_3
   properties {visibility=Hidden}
   slot {class=GridSlot, properties {horizontalAlignment=HAlign_Right, verticalAlignment=VAlign_Center}}
  }
  WBP_PilgrimAccuracyText_Instance_4 {
   widgetClassPath=&WBP_PilgrimAccuracyText_Instance
   bIsVariable=true
   parent=Container
   widget=WBP_PilgrimAccuracyText_Instance_4
   properties {visibility=Hidden}
   slot {class=GridSlot, properties {horizontalAlignment=HAlign_Right, verticalAlignment=VAlign_Center}}
  }
  WBP_PilgrimAccuracyText_Instance_5 {
   widgetClassPath=&WBP_PilgrimAccuracyText_Instance
   bIsVariable=true
   parent=Container
   widget=WBP_PilgrimAccuracyText_Instance_5
   properties {visibility=Hidden}
   slot {class=GridSlot, properties {horizontalAlignment=HAlign_Right, verticalAlignment=VAlign_Center}}
  }
  WBP_PilgrimAccuracyText_Instance_6 {
   widgetClassPath=&WBP_PilgrimAccuracyText_Instance
   bIsVariable=true
   parent=Container
   widget=WBP_PilgrimAccuracyText_Instance_6
   properties {visibility=Hidden}
   slot {class=GridSlot, properties {horizontalAlignment=HAlign_Right, verticalAlignment=VAlign_Center}}
  }
  WBP_PilgrimAccuracyText_Instance_7 {
   widgetClassPath=&WBP_PilgrimAccuracyText_Instance
   bIsVariable=true
   parent=Container
   widget=WBP_PilgrimAccuracyText_Instance_7
   properties {visibility=Hidden}
   slot {class=GridSlot, properties {horizontalAlignment=HAlign_Right, verticalAlignment=VAlign_Center}}
  }
  WBP_PilgrimAccuracyText_Instance_8 {
   widgetClassPath=&WBP_PilgrimAccuracyText_Instance
   bIsVariable=true
   parent=Container
   widget=WBP_PilgrimAccuracyText_Instance_8
   properties {visibility=Hidden}
   slot {class=GridSlot, properties {horizontalAlignment=HAlign_Right, verticalAlignment=VAlign_Center}}
  }
  WBP_PilgrimAccuracyText_Instance_9 {
   widgetClassPath=&WBP_PilgrimAccuracyText_Instance
   bIsVariable=true
   parent=Container
   widget=WBP_PilgrimAccuracyText_Instance_9
   properties {visibility=Hidden}
   slot {class=GridSlot, properties {horizontalAlignment=HAlign_Right, verticalAlignment=VAlign_Center}}
  }
  FixedAccuracies {
   widgetClassPath=&GridPanel
   bIsVariable=true
   parent=Overlay_0
   widget=FixedAccuracies
   properties {renderTransformPivot=<1 0.5>}
   slot {class=OverlaySlot, properties {horizontalAlignment=HAlign_Right, verticalAlignment=VAlign_Center}}
  }
  WBP_PilgrimAccuracyText_Early {
   widgetClassPath=&WBP_PilgrimAccuracyText_Instance
   bIsVariable=true
   parent=FixedAccuracies
   widget=WBP_PilgrimAccuracyText_Early
   properties {hitAccuracy=Early, renderOpacity=0, visibility=Visible}
   slot {class=GridSlot, properties {horizontalAlignment=HAlign_Right, verticalAlignment=VAlign_Center}}
  }
  WBP_PilgrimAccuracyText_Perfect {
   widgetClassPath=&WBP_PilgrimAccuracyText_Instance
   bIsVariable=true
   parent=FixedAccuracies
   widget=WBP_PilgrimAccuracyText_Perfect
   properties {visibility=Visible}
   slot {class=GridSlot, properties {horizontalAlignment=HAlign_Right, verticalAlignment=VAlign_Center}}
  }
  WBP_PilgrimAccuracyText_Late {
   widgetClassPath=&WBP_PilgrimAccuracyText_Instance
   bIsVariable=true
   parent=FixedAccuracies
   widget=WBP_PilgrimAccuracyText_Late
   properties {hitAccuracy=Late, renderOpacity=0, visibility=Visible}
   slot {class=GridSlot, properties {horizontalAlignment=HAlign_Right, verticalAlignment=VAlign_Center}}
  }
  WBP_PilgrimCore_StreakCounter {
   widgetClassPath=&/WBP_PllgrimCore_StreakCounter
   bIsVariable=true
   parent=Overlay_0
   widget=WBP_PilgrimCore_StreakCounter
   properties {visibility=Visible}
   slot {class=OverlaySlot, properties {horizontalAlignment=HAlign_Right, padding=<ltrb 0 -25 -45 0>}}
  }
 ]
}
