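# Pilgrim rhythm-track widget that labels each lane with its keybind, or a touch icon on touch input.
# A switcher picks 4-lane, 5-lane or Pro Drums layouts from track type and lane count, while lefty flip rebinds labels.
# Event router listeners, pawn tag waits and delayed input-method changes drive one forward/reverse fade, with visibility settings and peripheral checks gating display.
# Mobile expanded track widens the lane canvases, and a gem-break panel wraps everything.
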WidgetBlueprint /PilgrimCore/Widgets/WBP_PilgrimTrackInputLabels
animations [A_ShowHideKeyBinds {
  binding_count=26
  length_seconds=0.5
  object=&self:A_ShowHideKeyBinds
  sequence {bindings [
    TouchIcon_Lane3 {
     id=#1
     possessed_class=PilgrimMobileBrush_C
     tracks [
      {
       type=MovieSceneFloatTrack
       display_name=Render Opacity
       sections [{type=MovieSceneFloatSection, end_frame=10, start_frame=0, channels [None {type=MovieSceneScriptingFloatChannel, keys [[0,0],[10,1]]}]}]
      }
      {
       type=MovieSceneByteTrack
       display_name=Visibility
       sections [{type=MovieSceneByteSection, end_frame=10, start_frame=0, channels [None {type=MovieSceneScriptingByteChannel, keys [[0,1],[1,3],[10,4]]}]}]
      }
     ]
    }
    TouchIcon_Lane4_5Lane {
     id=#2
     possessed_class=PilgrimMobileBrush_C
     tracks [
      {
       type=MovieSceneFloatTrack
       display_name=Render Opacity
       sections [{type=MovieSceneFloatSection, end_frame=10, start_frame=0, channels [None {type=MovieSceneScriptingFloatChannel, keys [[0,0],[10,1]]}]}]
      }
      {
       type=MovieSceneByteTrack
       display_name=Visibility
       sections [{type=MovieSceneByteSection, end_frame=10, start_frame=0, channels [None {type=MovieSceneScriptingByteChannel, keys [[0,1],[1,3],[10,4]]}]}]
      }
     ]
    }
    TouchIcon_ProDrums_Lane3 {
     id=#3
     possessed_class=PilgrimMobileBrush_C
     tracks [
      {
       type=MovieSceneFloatTrack
       display_name=Render Opacity
       sections [{type=MovieSceneFloatSection, end_frame=10, start_frame=0, channels [None {type=MovieSceneScriptingFloatChannel, keys [[0,0],[10,1]]}]}]
      }
      {
       type=MovieSceneByteTrack
       display_name=Visibility
       sections [{type=MovieSceneByteSection, end_frame=10, start_frame=0, channels [None {type=MovieSceneScriptingByteChannel, keys [[0,1],[1,3],[10,4]]}]}]
      }
     ]
    }
    TouchIcon_Lane4 {
     id=#4
     possessed_class=PilgrimMobileBrush_C
     tracks [
      {
       type=MovieSceneFloatTrack
       display_name=Render Opacity
       sections [{type=MovieSceneFloatSection, end_frame=10, start_frame=0, channels [None {type=MovieSceneScriptingFloatChannel, keys [[0,0],[10,1]]}]}]
      }
      {
       type=MovieSceneByteTrack
       display_name=Visibility
       sections [{type=MovieSceneByteSection, end_frame=10, start_frame=0, channels [None {type=MovieSceneScriptingByteChannel, keys [[0,1],[1,3],[10,4]]}]}]
      }
     ]
    }
    KeybindWidget_PilgrimLane2 {
     id=#5
     possessed_class=KeybindWidget_C
     tracks [
      {
       type=MovieSceneFloatTrack
       display_name=Render Opacity
       sections [{type=MovieSceneFloatSection, end_frame=10, start_frame=0, channels [None {type=MovieSceneScriptingFloatChannel, keys [[0,0],[10,1]]}]}]
      }
      {
       type=MovieSceneByteTrack
       display_name=Visibility
       sections [{type=MovieSceneByteSection, end_frame=10, start_frame=0, channels [None {type=MovieSceneScriptingByteChannel, keys [[0,1],[1,3],[10,4]]}]}]
      }
     ]
    }
    KeybindWidget_PilgrimLane3_5Lane {
     id=#6
     possessed_class=KeybindWidget_C
     tracks [
      {
       type=MovieSceneFloatTrack
       display_name=Render Opacity
       sections [{type=MovieSceneFloatSection, end_frame=10, start_frame=0, channels [None {type=MovieSceneScriptingFloatChannel, keys [[0,0],[10,1]]}]}]
      }
      {
       type=MovieSceneByteTrack
       display_name=Visibility
       sections [{type=MovieSceneByteSection, end_frame=10, start_frame=0, channels [None {type=MovieSceneScriptingByteChannel, keys [[0,1],[1,3],[10,4]]}]}]
      }
     ]
    }
    KeybindWidget_PilgrimLane4_5Lane {
     id=#7
     possessed_class=KeybindWidget_C
     tracks [
      {
       type=MovieSceneFloatTrack
       display_name=Render Opacity
       sections [{type=MovieSceneFloatSection, end_frame=10, start_frame=0, channels [None {type=MovieSceneScriptingFloatChannel, keys [[0,0],[10,1]]}]}]
      }
      {
       type=MovieSceneByteTrack
       display_name=Visibility
       sections [{type=MovieSceneByteSection, end_frame=10, start_frame=0, channels [None {type=MovieSceneScriptingByteChannel, keys [[0,1],[1,3],[10,4]]}]}]
      }
     ]
    }
    KeybindWidget_PilgrimLane2_5Lane {
     id=#8
     possessed_class=KeybindWidget_C
     tracks [
      {
       type=MovieSceneFloatTrack
       display_name=Render Opacity
       sections [{type=MovieSceneFloatSection, end_frame=10, start_frame=0, channels [None {type=MovieSceneScriptingFloatChannel, keys [[0,0],[10,1]]}]}]
      }
      {
       type=MovieSceneByteTrack
       display_name=Visibility
       sections [{type=MovieSceneByteSection, end_frame=10, start_frame=0, channels [None {type=MovieSceneScriptingByteChannel, keys [[0,1],[1,3],[10,4]]}]}]
      }
     ]
    }
    TouchIcon_Lane3_5Lane {
     id=#9
     possessed_class=PilgrimMobileBrush_C
     tracks [
      {
       type=MovieSceneFloatTrack
       display_name=Render Opacity
       sections [{type=MovieSceneFloatSection, end_frame=10, start_frame=0, channels [None {type=MovieSceneScriptingFloatChannel, keys [[0,0],[10,1]]}]}]
      }
      {
       type=MovieSceneByteTrack
       display_name=Visibility
       sections [{type=MovieSceneByteSection, end_frame=10, start_frame=0, channels [None {type=MovieSceneScriptingByteChannel, keys [[0,1],[1,3],[10,4]]}]}]
      }
     ]
    }
    TouchIcon_ProDrums_Lane4 {
     id=#10
     possessed_class=PilgrimMobileBrush_C
     tracks [
      {
       type=MovieSceneFloatTrack
       display_name=Render Opacity
       sections [{type=MovieSceneFloatSection, end_frame=10, start_frame=0, channels [None {type=MovieSceneScriptingFloatChannel, keys [[0,0],[10,1]]}]}]
      }
      {
       type=MovieSceneByteTrack
       display_name=Visibility
       sections [{type=MovieSceneByteSection, end_frame=10, start_frame=0, channels [None {type=MovieSceneScriptingByteChannel, keys [[0,1],[1,3],[10,4]]}]}]
      }
     ]
    }
    KeybindWidget_ProDrums_PilgrimLane3 {
     id=#11
     possessed_class=KeybindWidget_C
     tracks [
      {
       type=MovieSceneFloatTrack
       display_name=Render Opacity
       sections [{type=MovieSceneFloatSection, end_frame=10, start_frame=0, channels [None {type=MovieSceneScriptingFloatChannel, keys [[0,0],[10,1]]}]}]
      }
      {
       type=MovieSceneByteTrack
       display_name=Visibility
       sections [{type=MovieSceneByteSection, end_frame=10, start_frame=0, channels [None {type=MovieSceneScriptingByteChannel, keys [[0,1],[1,3],[10,4]]}]}]
      }
     ]
    }
    TouchIcon_ProDrums_Lane1 {
     id=#12
     possessed_class=PilgrimMobileBrush_C
     tracks [
      {
       type=MovieSceneFloatTrack
       display_name=Render Opacity
       sections [{type=MovieSceneFloatSection, end_frame=10, start_frame=0, channels [None {type=MovieSceneScriptingFloatChannel, keys [[0,0],[10,1]]}]}]
      }
      {
       type=MovieSceneByteTrack
       display_name=Visibility
       sections [{type=MovieSceneByteSection, end_frame=10, start_frame=0, channels [None {type=MovieSceneScriptingByteChannel, keys [[0,1],[1,3],[10,4]]}]}]
      }
     ]
    }
    KeybindWidget_ProDrums_PilgrimLane1 {
     id=#13
     possessed_class=KeybindWidget_C
     tracks [
      {
       type=MovieSceneFloatTrack
       display_name=Render Opacity
       sections [{type=MovieSceneFloatSection, end_frame=10, start_frame=0, channels [None {type=MovieSceneScriptingFloatChannel, keys [[0,0],[10,1]]}]}]
      }
      {
       type=MovieSceneByteTrack
       display_name=Visibility
       sections [{type=MovieSceneByteSection, end_frame=10, start_frame=0, channels [None {type=MovieSceneScriptingByteChannel, keys [[0,1],[1,3],[10,4]]}]}]
      }
     ]
    }
    TouchIcon_Lane2 {
     id=#14
     possessed_class=PilgrimMobileBrush_C
     tracks [
      {
       type=MovieSceneFloatTrack
       display_name=Render Opacity
       sections [{type=MovieSceneFloatSection, end_frame=10, start_frame=0, channels [None {type=MovieSceneScriptingFloatChannel, keys [[0,0],[10,1]]}]}]
      }
      {
       type=MovieSceneByteTrack
       display_name=Visibility
       sections [{type=MovieSceneByteSection, end_frame=10, start_frame=0, channels [None {type=MovieSceneScriptingByteChannel, keys [[0,1],[1,3],[10,4]]}]}]
      }
     ]
    }
    KeybindWidget_PilgrimLane1 {
     id=#15
     possessed_class=KeybindWidget_C
     tracks [
      {
       type=MovieSceneFloatTrack
       display_name=Render Opacity
       sections [{type=MovieSceneFloatSection, end_frame=10, start_frame=0, channels [None {type=MovieSceneScriptingFloatChannel, keys [[0,0],[10,1]]}]}]
      }
      {
       type=MovieSceneByteTrack
       display_name=Visibility
       sections [{type=MovieSceneByteSection, end_frame=10, start_frame=0, channels [None {type=MovieSceneScriptingByteChannel, keys [[0,1],[1,3],[10,4]]}]}]
      }
     ]
    }
    TouchIcon_Lane1_5Lane {
     id=#16
     possessed_class=PilgrimMobileBrush_C
     tracks [
      {
       type=MovieSceneFloatTrack
       display_name=Render Opacity
       sections [{type=MovieSceneFloatSection, end_frame=10, start_frame=0, channels [None {type=MovieSceneScriptingFloatChannel, keys [[0,0],[10,1]]}]}]
      }
      {
       type=MovieSceneByteTrack
       display_name=Visibility
       sections [{type=MovieSceneByteSection, end_frame=10, start_frame=0, channels [None {type=MovieSceneScriptingByteChannel, keys [[0,1],[1,3],[10,4]]}]}]
      }
     ]
    }
    KeybindWidget_ProDrums_PilgrimLane2 {
     id=#17
     possessed_class=KeybindWidget_C
     tracks [
      {
       type=MovieSceneFloatTrack
       display_name=Render Opacity
       sections [{type=MovieSceneFloatSection, end_frame=10, start_frame=0, channels [None {type=MovieSceneScriptingFloatChannel, keys [[0,0],[10,1]]}]}]
      }
      {
       type=MovieSceneByteTrack
       display_name=Visibility
       sections [{type=MovieSceneByteSection, end_frame=10, start_frame=0, channels [None {type=MovieSceneScriptingByteChannel, keys [[0,1],[1,3],[10,4]]}]}]
      }
     ]
    }
    TouchIcon_Lane5_5Lane {
     id=#18
     possessed_class=PilgrimMobileBrush_C
     tracks [
      {
       type=MovieSceneFloatTrack
       display_name=Render Opacity
       sections [{type=MovieSceneFloatSection, end_frame=10, start_frame=0, channels [None {type=MovieSceneScriptingFloatChannel, keys [[0,0],[10,1]]}]}]
      }
      {
       type=MovieSceneByteTrack
       display_name=Visibility
       sections [{type=MovieSceneByteSection, end_frame=10, start_frame=0, channels [None {type=MovieSceneScriptingByteChannel, keys [[0,1],[1,3],[10,4]]}]}]
      }
     ]
    }
    KeybindWidget_PilgrimLane4 {
     id=#19
     possessed_class=KeybindWidget_C
     tracks [
      {
       type=MovieSceneFloatTrack
       display_name=Render Opacity
       sections [{type=MovieSceneFloatSection, end_frame=10, start_frame=0, channels [None {type=MovieSceneScriptingFloatChannel, keys [[0,0],[10,1]]}]}]
      }
      {
       type=MovieSceneByteTrack
       display_name=Visibility
       sections [{type=MovieSceneByteSection, end_frame=10, start_frame=0, channels [None {type=MovieSceneScriptingByteChannel, keys [[0,1],[1,3],[10,4]]}]}]
      }
     ]
    }
    KeybindWidget_PilgrimLane3 {
     id=#20
     possessed_class=KeybindWidget_C
     tracks [
      {
       type=MovieSceneFloatTrack
       display_name=Render Opacity
       sections [{type=MovieSceneFloatSection, end_frame=10, start_frame=0, channels [None {type=MovieSceneScriptingFloatChannel, keys [[0,0],[10,1]]}]}]
      }
      {
       type=MovieSceneByteTrack
       display_name=Visibility
       sections [{type=MovieSceneByteSection, end_frame=10, start_frame=0, channels [None {type=MovieSceneScriptingByteChannel, keys [[0,1],[1,3],[10,4]]}]}]
      }
     ]
    }
    TouchIcon_Lane1 {
     id=#21
     possessed_class=PilgrimMobileBrush_C
     tracks [
      {
       type=MovieSceneFloatTrack
       display_name=Render Opacity
       sections [{type=MovieSceneFloatSection, end_frame=10, start_frame=0, channels [None {type=MovieSceneScriptingFloatChannel, keys [[0,0],[10,1]]}]}]
      }
      {
       type=MovieSceneByteTrack
       display_name=Visibility
       sections [{type=MovieSceneByteSection, end_frame=10, start_frame=0, channels [None {type=MovieSceneScriptingByteChannel, keys [[0,1],[1,3],[10,4]]}]}]
      }
     ]
    }
    TouchIcon_Lane2_5Lane {
     id=#22
     possessed_class=PilgrimMobileBrush_C
     tracks [
      {
       type=MovieSceneFloatTrack
       display_name=Render Opacity
       sections [{type=MovieSceneFloatSection, end_frame=10, start_frame=0, channels [None {type=MovieSceneScriptingFloatChannel, keys [[0,0],[10,1]]}]}]
      }
      {
       type=MovieSceneByteTrack
       display_name=Visibility
       sections [{type=MovieSceneByteSection, end_frame=10, start_frame=0, channels [None {type=MovieSceneScriptingByteChannel, keys [[0,1],[1,3],[10,4]]}]}]
      }
     ]
    }
    KeybindWidget_PilgrimLane5_5Lane {
     id=#23
     possessed_class=KeybindWidget_C
     tracks [
      {
       type=MovieSceneFloatTrack
       display_name=Render Opacity
       sections [{type=MovieSceneFloatSection, end_frame=10, start_frame=0, channels [None {type=MovieSceneScriptingFloatChannel, keys [[0,0],[10,1]]}]}]
      }
      {
       type=MovieSceneByteTrack
       display_name=Visibility
       sections [{type=MovieSceneByteSection, end_frame=10, start_frame=0, channels [None {type=MovieSceneScriptingByteChannel, keys [[0,1],[1,3],[10,4]]}]}]
      }
     ]
    }
    TouchIcon_ProDrums_Lane2 {
     id=#24
     possessed_class=PilgrimMobileBrush_C
     tracks [
      {
       type=MovieSceneFloatTrack
       display_name=Render Opacity
       sections [{type=MovieSceneFloatSection, end_frame=10, start_frame=0, channels [None {type=MovieSceneScriptingFloatChannel, keys [[0,0],[10,1]]}]}]
      }
      {
       type=MovieSceneByteTrack
       display_name=Visibility
       sections [{type=MovieSceneByteSection, end_frame=10, start_frame=0, channels [None {type=MovieSceneScriptingByteChannel, keys [[0,1],[1,3],[10,4]]}]}]
      }
     ]
    }
    KeybindWidget_PilgrimLane1_5Lane {
     id=#25
     possessed_class=KeybindWidget_C
     tracks [
      {
       type=MovieSceneFloatTrack
       display_name=Render Opacity
       sections [{type=MovieSceneFloatSection, end_frame=10, start_frame=0, channels [None {type=MovieSceneScriptingFloatChannel, keys [[0,0],[10,1]]}]}]
      }
      {
       type=MovieSceneByteTrack
       display_name=Visibility
       sections [{type=MovieSceneByteSection, end_frame=10, start_frame=0, channels [None {type=MovieSceneScriptingByteChannel, keys [[0,1],[1,3],[10,4]]}]}]
      }
     ]
    }
    KeybindWidget_ProDrums_PilgrimLane4 {
     id=#26
     possessed_class=KeybindWidget_C
     tracks [
      {
       type=MovieSceneFloatTrack
       display_name=Render Opacity
       sections [{type=MovieSceneFloatSection, end_frame=10, start_frame=0, channels [None {type=MovieSceneScriptingFloatChannel, keys [[0,0],[10,1]]}]}]
      }
      {
       type=MovieSceneByteTrack
       display_name=Visibility
       sections [{type=MovieSceneByteSection, end_frame=10, start_frame=0, channels [None {type=MovieSceneScriptingByteChannel, keys [[0,1],[1,3],[10,4]]}]}]
      }
     ]
    }
   ]}
 }]
graphs [
 Check If Should Collapse on Input Type Changed {nodes [
   Check If Should Collapse on Input Type Changed() {is=title K2Node_FunctionEntry}
   > n1 = Branch(Condition=NOT Boolean(A=IsMobileAndShouldUseExpandedTrack!() -> (ReturnValue:bool)))
   > SetVisibility(InVisibility=SelfHitTestInvisible)
   n2 = Get NewInputType() -> (NewInputType:ECommonInputType Enum)
   SetVisibility(execute=@n1.else, InVisibility=Select(Option 0=SelfHitTestInvisible, Option 1=Collapsed, Index="Equal (Enum)"(A=@n2, B=Touch)))
  ]}
 Check for Lefty Flip {nodes [
   Check for Lefty Flip() {is=title K2Node_FunctionEntry}
   > n1 = Get!(execute:Exec) -> (then:Exec, else:Exec, MusicBattleInstance:Pilgrim Music Battle Instance Object Reference)
   > n2 = GetLocalPilgrimGameForPlayer!(execute, self:Pilgrim Music Battle Instance Object Reference=@n1, InPC:Player Controller Object Reference=GetOwningPlayer()) -> (then, ReturnValue:Pilgrim Game Object Reference)
   > n3 = Is Valid(InputObject=@n2)
   Return Node!(execute=@n3.Is Not Valid, IsLeftyFlip:bool=false)
   > Return Node(IsLeftyFlip=IsUsingLeftyFlip!(self:Pilgrim Game Object Reference=@n2) -> (ReturnValue:bool))
   Return Node(execute=@n1.else, IsLeftyFlip=false)
  ]}
 EventGraph {nodes [
   UpdateVisibility() {is=title K2Node_CustomEvent}
   > SetVisibility(InVisibility=Select(Option 0=HitTestInvisible, Option 1=Collapsed, Index=Is Track Type Peripheral!() -> (IsTrackTypePeripheral:bool)))
   n1 = On Input Method Changed!() -> (OutputDelegate, then, bNewInputType:ECommonInputType Enum) {is=title K2Node_CustomEvent}
   > Set NewInputType(NewInputType:ECommonInputType Enum=@n1.bNewInputType) -> (Output_Get:ECommonInputType Enum)
   > Set Timer by Event(Delegate=Create Event(), Time=Get InputSwitchDelay() -> (InputSwitchDelay:double), bLooping=false, bMaxOncePerFrame=true)
   > n2 = Branch(Condition=ShouldShowIcons!() -> (IsVisible:bool))
   > ShowIcons()
   HideIcons(execute=@n2.else)
   ShowIcons() {is=title K2Node_CustomEvent}
   > Branch(Condition=ShouldShowIcons())
   > PlayAnimationForward(InAnimation=Get A_ShowHideKeyBinds() -> (A_ShowHideKeyBinds:Widget Animation Object Reference))
   > Set IconsAreHidden(IconsAreHidden:bool=false) -> (Output_Get:bool)
   > SetupNumLanes!(execute, self, NumLanes:int=Get NumLanes() -> (NumLanes:int=0)) -> (then)
   HideIcons() {is=title K2Node_CustomEvent}
   > n3 = Branch(Condition=Get IconsAreHidden() -> (IconsAreHidden:bool))
   PlayAnimationReverse(execute=@n3.else, InAnimation=Get A_ShowHideKeyBinds())
   > Set IconsAreHidden(IconsAreHidden=true)
   Event Destruct()
   n4 = Get GameplayEventHandles() -> (GameplayEventHandles:Array of Async Action Start Listening To Event Object References)
   > n5 = For Each Loop(Array:Array of Async Action Start Listening To Event Object References=@n4) -> (Array Element:Async Action Start Listening To Event Object Reference)
   Clear(execute=@n5.Completed, TargetArray:Array of Async Action Start Listening To Event Object References=@n4)
   StopListeningToEvent(execute=@n5.LoopBody, self:Async Action Start Listening To Event Object Reference=@n5)
   Event Construct()
   > Parent: Construct()
   > n6 = Is Valid(InputObject=Get Event Router() -> (Event Router:Gameplay Event Router Component Object Reference))
   n7 = PilgrimWaitForEventRouter!(execute:Exec=@n6.Is Not Valid, ContextActor:Actor Object Reference=GetOwningPlayer()) -> (then:Exec, OnEventRouterAvailable:Exec, EventRouter:Gameplay Event Router Component Object Reference) {is=title K2Node_AsyncAction}
   n8 = Set Event Router(execute=@n7.OnEventRouterAvailable, Event Router:Gameplay Event Router Component Object Reference=@n7) -> (Output_Get:Gameplay Event Router Component Object Reference)
   n9 = Sequence!(execute=[@n6.Is Valid, @n8.then]) -> (then_0, then_1, then_2:Exec, then_3:Exec, then_4:Exec, then_5:Exec)
   > n10 = StartListeningToEvent!(
    execute:Exec
    Target:Gameplay Event Router Component Object Reference=Get Event Router()
    EventType:Script Struct Object Reference=/Script/PilgrimCoreRuntime.PilgrimEvent_PilgrimGame_PlayingSong
    Context:Object Reference
    EventBubblingRule:EEventBubblingRule Enum
   ) -> (
    then:Exec
    AsyncTaskProxy:Async Action Start Listening To Event Object Reference
    OnEventReceived:Exec
    Payload:Pilgrim Event Pilgrim Game Playing Song Structure
   )
   > Add(
    TargetArray:Array of Async Action Start Listening To Event Object References=Get GameplayEventHandles()
    NewItem:"Async Action Start Listening To Event Object Reference (by ref)"=@n10
   ) {is=title K2Node_CallArrayFunction}
   Set NumLanes(
    execute=@n10.OnEventReceived
    NumLanes:int=GetNumLanes!(self:Pilgrim Game Object Reference=Break Pilgrim Event Pilgrim Game Playing Song!(PilgrimEvent_PilgrimGame_PlayingSong:"Pilgrim Event Pilgrim Game Playing Song Structure (by ref)"=@n10.Payload) -> (PilgrimGame:Pilgrim Game Object Reference) {is=title K2Node_BreakStruct}) -> (ReturnValue:int=0)
   ) -> (Output_Get:int=0)
   > ShowIcons() {is=title K2Node_CallFunction}
   n11 = StartListeningToEvent(
    execute=@n9.then_1
    Target=Get Event Router()
    EventType=/Script/PilgrimSettingsRuntime.PilgrimEvent_PilgrimSettings_NonProTrackInputButtonVisibilityChanged
   ) -> (Payload:Pilgrim Event Pilgrim Settings Non Pro Track Input Button Visibility Changed Structure)
   > Add(TargetArray=Get GameplayEventHandles(), NewItem=@n11)
   n12 = Branch(execute=[@n13.OnEventReceived, @n11.OnEventReceived], Condition=ShouldShowIcons())
   > ShowIcons()
   HideIcons(execute=@n12.else) {is=title K2Node_CallFunction}
   n13 = StartListeningToEvent(
    execute=@n9.then_2
    Target=Get Event Router()
    EventType=/Script/PilgrimSettingsRuntime.PilgrimEvent_PilgrimSettings_ProTrackInputButtonVisibilityChanged
   ) -> (Payload:Pilgrim Event Pilgrim Settings Pro Track Input Button Visibility Changed Structure)
   > Add(TargetArray=Get GameplayEventHandles(), NewItem=@n13)
   n14 = Cast To FortPlayerPawn(
    execute=@n9.then_3
    Object=Get Controlled Pawn!(self:Controller Object Reference=GetOwningPlayer()) -> (ReturnValue:Pawn Object Reference) {is=title K2Node_CallFunction}
   )
   > n15 = Sequence!(execute) -> (then_0, then_1)
   > n16 = WaitGameplayTagAddToActor!(
    execute:Exec
    TargetActor:Actor Object Reference=@n14
    Tag:Gameplay Tag Structure=(TagName="Pilgrim.Widget.TouchInputDisplay")
    OnlyTriggerOnce:bool=false
   ) -> (then:Exec, AsyncTaskProxy:Ability Async Wait Gameplay Tag Added Object Reference, Added:Exec) {is=title K2Node_AsyncAction}
   > Add(
    TargetArray:Array of Cancellable Async Action Object References=Get AsyncEventHandles() -> (AsyncEventHandles:Array of Cancellable Async Action Object References)
    NewItem:"Cancellable Async Action Object Reference (by ref)"=@n16
   )
   HideIcons(execute=@n16.Added)
   n17 = WaitGameplayTagRemoveFromActor!(
    execute:Exec=@n15.then_1
    TargetActor:Actor Object Reference=@n14
    Tag:Gameplay Tag Structure=(TagName="Pilgrim.Widget.TouchInputDisplay")
    OnlyTriggerOnce:bool=false
   ) -> (then:Exec, AsyncTaskProxy:Ability Async Wait Gameplay Tag Removed Object Reference, Removed:Exec) {is=title K2Node_AsyncAction}
   > Add(TargetArray=Get AsyncEventHandles(), NewItem=@n17)
   ShowIcons(execute=@n17.Removed)
   n18 = StartListeningToEvent(
    execute=@n9.then_4
    Target=Get Event Router()
    EventType=/Script/PilgrimSettingsRuntime.PilgrimEvent_PilgrimSettings_UseGuitarLeftyFlipChanged
   ) -> (Payload:Pilgrim Event Pilgrim Settings Use Guitar Lefty Flip Changed Structure)
   > Add(
    TargetArray:Array of Async Action Start Listening To Event Object References=Get GameplayEventHandles()
    NewItem:"Async Action Start Listening To Event Object Reference (by ref)"=@n18
   )
   SetupLeftyFlip!(
    execute=@n18.OnEventReceived
    self
    IsUsingLeftyFlip:bool=Break Pilgrim Event Pilgrim Settings Use Guitar Lefty Flip Changed!(PilgrimEvent_PilgrimSettings_UseGuitarLeftyFlipChanged:"Pilgrim Event Pilgrim Settings Use Guitar Lefty Flip Changed Structure (by ref)"=@n18.Payload) -> (bUseGuitarLeftyFlip:bool=~) {is=title K2Node_BreakStruct}
   ) -> (then)
   Bind Event to On Input Method Changed(execute=@n9.then_5, self=Get CommonInputSubsystem(), Delegate=@n1)
   > On Input Method Changed!(execute, self, bNewInputType:ECommonInputType Enum=GetCurrentInputType(self=Get CommonInputSubsystem())) -> (then) {is=title K2Node_CallFunction}
   > n19 = StartListeningToEvent(Target=Get Event Router(), EventType=/Script/PilgrimSettingsRuntime.PilgrimEvent_PilgrimSettings_EnableMobileExpandedTrackChanged) -> (Payload:Pilgrim Event Pilgrim Settings Enable Mobile Expanded Track Changed Structure)
   > n20 = Add(TargetArray=Get GameplayEventHandles(), NewItem=@n19)
   n21 = Select(Option 0:Vector 2D Structure=(X=900,Y=64), Option 1:Vector 2D Structure=(X=1500,Y=64), Index=IsMobileAndShouldUseExpandedTrack()) -> (ReturnValue:Vector 2D Structure)
   SetCanvasContainerSize!(
    execute=[@n20.then, @n19.OnEventReceived]
    self
    CanvasPanel:Canvas Panel Object Reference=Get KeybindWidgetCanvasContainer_4Lane() -> (KeybindWidgetCanvasContainer_4Lane:Canvas Panel Object Reference)
    Size:Vector 2D Structure=@n21
   ) -> (then)
   > SetCanvasContainerSize(CanvasPanel=Get KeybindWidgetCanvasContainer_5Lane() -> (KeybindWidgetCanvasContainer_5Lane:Canvas Panel Object Reference), Size=@n21)
   Event OnMusicBattleInstanceInitializeEvent() {is=title K2Node_Event}
   n22 = GetLocalPilgrimGameForPlayer!(self=Get MusicBattleInstance() -> (MusicBattleInstance:Pilgrim Music Battle Instance Object Reference), InPC=GetOwningPlayer()) -> (ReturnValue)
   > Is Valid(InputObject=@n22)
   > n23 = Branch(Condition=IsDrumPeripheralTrack!(TrackType:EPilgrimTrackType Enum=GetSelectedTrackType!(self:Pilgrim Game Object Reference=@n22) -> (ReturnValue:EPilgrimTrackType Enum=TrackGuitar)) -> (ReturnValue:bool))
   n24 = Set NumLanes(execute=[@n26.then, @n23.then], NumLanes=GetNumLanes(self=@n22) -> (ReturnValue=0)) -> (Output_Get=0)
   > SetupNumLanes(NumLanes=@n24)
   > n25 = Branch(Condition=ShouldShowIcons())
   > ShowIcons()
   HideIcons(execute=@n25.else)
   n26 = SetupLeftyFlip(execute=@n23.else, IsUsingLeftyFlip=Check for Lefty Flip!() -> (IsLeftyFlip:bool) {is=title K2Node_CallFunction})
  ]}
 Is Track Type Drums Peripheral {nodes [
   Is Track Type Drums Peripheral() {is=title K2Node_FunctionEntry}
   n1 = Get MusicBattleInstance()
   > n2 = Is Valid(InputObject=@n1)
   n3 = GetLocalPilgrimGameForPlayer(self=@n1, InPC=GetOwningPlayer())
   > n4 = Is Valid(InputObject=@n3)
   > Return Node!(execute, IsTrackTypePeripheral:bool=IsDrumPeripheralTrack(TrackType=GetSelectedTrackType(self=@n3) -> (ReturnValue=TrackGuitar)))
   Return Node(execute=[@n4.Is Not Valid, @n2.Is Not Valid], IsTrackTypePeripheral=false)
  ]}
 Is Track Type Peripheral {nodes [
   Is Track Type Peripheral() {is=title K2Node_FunctionEntry}
   n1 = Get MusicBattleInstance()
   > n2 = Is Valid(InputObject=@n1)
   n3 = GetLocalPilgrimGameForPlayer(self=@n1, InPC=GetOwningPlayer())
   > n4 = Is Valid(InputObject=@n3)
   n5 = GetSelectedTrackType(self=@n3) -> (ReturnValue=TrackGuitar)
   > Return Node(IsTrackTypePeripheral=OR Boolean(A=IsGuitarPeripheralTrack!(TrackType:EPilgrimTrackType Enum=@n5) -> (ReturnValue:bool), B=IsDrumPeripheralTrack(TrackType=@n5)))
   Return Node(execute=[@n4.Is Not Valid, @n2.Is Not Valid], IsTrackTypePeripheral=false)
  ]}
 SetCanvasContainerSize {nodes [
   n1 = SetCanvasContainerSize!() -> (then, CanvasPanel:Canvas Panel Object Reference, Size:Vector 2D Structure) {is=title K2Node_FunctionEntry}
   > SetSize!(
    execute
    self:Canvas Panel Slot Object Reference=SlotAsCanvasSlot!(Widget:Widget Object Reference=GetChildAt!(self:Panel Widget Object Reference=@n1, Index:int=0) -> (ReturnValue:Widget Object Reference)) -> (ReturnValue:Canvas Panel Slot Object Reference)
    InSize:Vector 2D Structure=@n1.Size
   ) -> (then)
  ]}
 SetupLeftyFlip {nodes [
   n1 = SetupLeftyFlip!() -> (then, IsUsingLeftyFlip:bool=~) {is=title K2Node_FunctionEntry}
   > n2 = Sequence()
   > n3 = Branch(Condition=@n1)
   n4 = Get KeybindWidget_PilgrimLane1() -> (KeybindWidget_PilgrimLane1:Keybind Widget Object Reference)
   > SetBoundInputAction!(
    execute
    self:Fort Keybind Widget Object Reference=@n4
    NewAction:Input Action Object Reference=/PilgrimSettings/Input/InputActions/IA_PilgrimLane4Press
   ) -> (then)
   n5 = Get KeybindWidget_PilgrimLane2() -> (KeybindWidget_PilgrimLane2:Keybind Widget Object Reference)
   > SetBoundInputAction(self=@n5, NewAction=/PilgrimSettings/Input/InputActions/IA_PilgrimLane3Press)
   n6 = Get KeybindWidget_PilgrimLane3() -> (KeybindWidget_PilgrimLane3:Keybind Widget Object Reference)
   > SetBoundInputAction(self=@n6, NewAction=/PilgrimSettings/Input/InputActions/IA_PilgrimLane2Press)
   n7 = Get KeybindWidget_PilgrimLane4() -> (KeybindWidget_PilgrimLane4:Keybind Widget Object Reference)
   > SetBoundInputAction(self=@n7, NewAction=/PilgrimSettings/Input/InputActions/IA_PilgrimLane1Press)
   n8 = Get KeybindWidget_PilgrimLane1_5Lane() -> (KeybindWidget_PilgrimLane1_5Lane:Keybind Widget Object Reference)
   > SetBoundInputAction(self=@n8, NewAction=/PilgrimSettings/Input/InputActions/IA_PilgrimLane5Press_5Lane)
   n9 = Get KeybindWidget_PilgrimLane2_5Lane() -> (KeybindWidget_PilgrimLane2_5Lane:Keybind Widget Object Reference)
   > SetBoundInputAction(self=@n9, NewAction=/PilgrimSettings/Input/InputActions/IA_PilgrimLane4Press_5Lane)
   n10 = Get KeybindWidget_PilgrimLane4_5Lane() -> (KeybindWidget_PilgrimLane4_5Lane:Keybind Widget Object Reference)
   > SetBoundInputAction(self=@n10, NewAction=/PilgrimSettings/Input/InputActions/IA_PilgrimLane2Press_5Lane)
   n11 = Get KeybindWidget_PilgrimLane5_5Lane() -> (KeybindWidget_PilgrimLane5_5Lane:Keybind Widget Object Reference)
   > SetBoundInputAction(self=@n11, NewAction=/PilgrimSettings/Input/InputActions/IA_PilgrimLane1Press_5Lane)
   SetBoundInputAction(execute=@n3.else, self=@n4, NewAction=&IA_PilgrimLane1Press)
   > SetBoundInputAction(self=@n5, NewAction=&IA_PilgrimLane2Press)
   > SetBoundInputAction(self=@n6, NewAction=&IA_PilgrimLane3Press)
   > SetBoundInputAction(self=@n7, NewAction=&IA_PilgrimLane4Press)
   > SetBoundInputAction(self=@n8, NewAction=&IA_PilgrimLane1Press_5Lane)
   > SetBoundInputAction(self=@n9, NewAction=&IA_PilgrimLane2Press_5Lane)
   > SetBoundInputAction(self=@n10, NewAction=&IA_PilgrimLane4Press_5Lane)
   > SetBoundInputAction(self=@n11, NewAction=&IA_PilgrimLane5Press_5Lane)
   UpdateKeybindWidget(
    execute=@n2.then_1
    self:Fort Keybind Widget Object Reference=[
     Get KeybindWidget_PilgrimLane1()
     Get KeybindWidget_PilgrimLane2()
     Get KeybindWidget_PilgrimLane3()
     Get KeybindWidget_PilgrimLane4()
     Get KeybindWidget_PilgrimLane1_5Lane()
     Get KeybindWidget_PilgrimLane2_5Lane()
     Get KeybindWidget_PilgrimLane4_5Lane()
     Get KeybindWidget_PilgrimLane5_5Lane()
    ]
   )
  ]}
 SetupNumLanes {nodes [
   n1 = SetupNumLanes!() -> (then, NumLanes:int) {is=title K2Node_FunctionEntry}
   > n2 = Branch(Condition=Is Track Type Drums Peripheral!() -> (IsTrackTypePeripheral:bool) {is=title K2Node_CallFunction})
   > SetActiveWidget(
    self:Widget Switcher Object Reference=Get KeybindsSwitcher() -> (KeybindsSwitcher:Common Animated Switcher Object Reference)
    Widget=Get KeybindWidgetCanvasContainer_ProDrums() -> (KeybindWidgetCanvasContainer_ProDrums:Canvas Panel Object Reference)
   )
   n3 = Branch(execute=@n2.else, Condition="Equal (Integer)"(A=@n1, B=5))
   > SetActiveWidget(self=Get KeybindsSwitcher(), Widget=Get KeybindWidgetCanvasContainer_5Lane())
   SetActiveWidget(execute=@n3.else, self=Get KeybindsSwitcher(), Widget=Get KeybindWidgetCanvasContainer_4Lane())
   n4 = Branch(Condition=Check for Lefty Flip())
   > SetVisibility(self=Get LeftyCymbalsInputHorizontalBox() -> (LeftyCymbalsInputHorizontalBox:Horizontal Box Object Reference), InVisibility=HitTestInvisible)
   > SetVisibility(self=Get CymbalsInputHorizontalBox() -> (CymbalsInputHorizontalBox:Horizontal Box Object Reference), InVisibility=Collapsed)
   SetVisibility(execute=@n4.else, self=Get LeftyCymbalsInputHorizontalBox(), InVisibility=Collapsed)
   > SetVisibility(self=Get CymbalsInputHorizontalBox(), InVisibility=HitTestInvisible)
  ]}
 ShouldShowIcons {nodes [
   ShouldShowIcons() {is=title K2Node_FunctionEntry}
   n1 = GetControllerFromActor!(Actor:Actor Object Reference=GetOwningPlayer()) -> (ReturnValue:Controller Object Reference)
   > n2 = Is Valid(InputObject=@n1)
   > n3 = GetComponentByClass!(
    execute
    self:Actor Object Reference=@n1
    ComponentClass:Actor Component Class Reference=/Script/PilgrimCoreRuntime.PilgrimPlayerControllerComponent
   ) -> (then, ReturnValue:Pilgrim Player Controller Component Object Reference)
   > n4 = Is Valid(InputObject=@n3)
   Return Node!(execute=[@n4.Is Not Valid, @n2.Is Not Valid], IsVisible:bool=true)
   n5 = GetPilgrimClientSettingRecordPartition!(self:Pilgrim Player Controller Component Object Reference=@n3) -> (ReturnValue:Pilgrim Client Setting Record Partition Object Reference)
   n6 = Is Track Type Peripheral() {is=title K2Node_CallFunction}
   n7 = AND Boolean(A=GetProTrackInputButtonVisibility!(self:Pilgrim Client Setting Record Partition Object Reference=@n5) -> (ReturnValue:bool), B=@n6)
   n8 = AND Boolean(
    A=NOT Boolean(A=@n6)
    B=GetNonProTrackInputButtonVisibility!(self:Pilgrim Client Setting Record Partition Object Reference=@n5) -> (ReturnValue:bool)
   )
   n9 = GetCurrentDevicePeripheralType!(self:Pilgrim Player Controller Component Object Reference=@n3) -> (ReturnValue:EFMPeripheralInputMode Enum=None)
   > Return Node(IsVisible=AND Boolean(
     A="Not Equal (Enum)"!(A:EFMPeripheralInputMode Enum=@n9, B:EFMPeripheralInputMode Enum=Drums) -> (ReturnValue:bool)
     B=OR Boolean(A=@n7, B=@n8)
    ))
  ]}
]
variables [
 AsyncEventHandles {type="array<CancellableAsyncAction>", default []}
 Event Router {type=GameplayEventRouterComponent, default=null}
 HideAfterShowInterval {type=number, default=8}
 IconsAreHidden {type=boolean}
 InputSwitchDelay {type=number, default=0.25}
 Is Ending Break {type=boolean}
 IsUsingPeripheral {type=boolean}
 KeyShowingDuration_TimerHandle {type=TimerHandle}
 LabelGemMap {default {}}
 NewInputType {type=ECommonInputType, default=MouseAndKeyboard}
 NumLanes {type=integer}
 SuppressKeyIconsForDuration_TimerHandle {type=TimerHandle}
 SuppressReshowAfterHideDuration {type=number, default=6}
]
widgets {
 info {
  inheritedWidgetCount=0
  namedSlotCount=1
  parentClass=/Script/PilgrimCoreUI.PilgrimGameWidgetBase
  rootWidgetClass=&SizeBox
  widgetCount=169
 }
 widgets [
  WidgetSize {
   widgetClassPath=&SizeBox
   bIsVariable=true
   widget=WidgetSize
   properties {bOverride_WidthOverride=true, heightOverride=120, widthOverride=900}
  }
  GemBreakPanel_InputLabels {
   widgetClassPath=&/WBP_PilgrimGemBreakPanel
   parent=WidgetSize
   widget=GemBreakPanel_InputLabels
   properties {breakEndAnimation=0.5sFadeIn, breakStartAnimation=0.5sFadeOut}
   slot {class=SizeBoxSlot, properties {verticalAlignment=VAlign_Center}}
  }
  KeybindsSwitcher {
   widgetClassPath=/Script/CommonUI.CommonAnimatedSwitcher
   bIsVariable=true
   namedSlotHost=GemBreakPanel_InputLabels
   widget=KeybindsSwitcher
   properties {transitionDuration=0.1}
  }
  KeybindWidgetCanvasContainer_4Lane {
   widgetClassPath=&CanvasPanel
   bIsVariable=true
   parent=KeybindsSwitcher
   widget=KeybindWidgetCanvasContainer_4Lane
   slot {class=WidgetSwitcherSlot}
  }
  VerticalBox_6 {
   widgetClassPath=&VerticalBox
   parent=KeybindWidgetCanvasContainer_4Lane
   widget=VerticalBox_6
   slot {class=CanvasPanelSlot, properties {layoutData {alignment=<0.5 0>, anchors {maximum=<0.5 0>, minimum=<0.5 0>}, offsets=<ltrb 0 20 900 64>}}}
  }
  HorizontalBox_1 {
   widgetClassPath=&HorizontalBox
   parent=VerticalBox_6
   widget=HorizontalBox_1
   slot {class=VerticalBoxSlot, properties {padding=<ltrb 150 0 150 0>}}
  }
  PilgrimTrackUI_0 {
   widgetClassPath=/Script/PilgrimCoreUI.PilgrimTrackUI
   parent=HorizontalBox_1
   widget=PilgrimTrackUI_0
   properties {slateTransformMode=Translation}
   slot {class=HorizontalBoxSlot, properties {size {sizeRule=Fill, value=1}}}
  }
  VerticalBox_7 {
   widgetClassPath=&VerticalBox
   parent=PilgrimTrackUI_0
   widget=VerticalBox_7
   slot {class=OverlaySlot, properties {horizontalAlignment=HAlign_Center}}
  }
  Overlay_9 {
   widgetClassPath=&Overlay
   parent=VerticalBox_7
   widget=Overlay_9
   slot {class=VerticalBoxSlot, properties {padding=<ltrb -10 0 0 0>}}
  }
  CommonVisibilityWidgetLegacy_0 {
   widgetClassPath=&CommonVisibilityWidgetLegacy
   parent=Overlay_9
   widget=CommonVisibilityWidgetLegacy_0
   properties {bShowForTouch=false, padding=<ltrb 0 0 0 0>}
   slot {class=OverlaySlot, properties {horizontalAlignment=HAlign_Fill, verticalAlignment=VAlign_Fill}}
  }
  KeybindWidget_PilgrimLane1 {
   widgetClassPath=/FortUILibrary/Input/KeybindWidget
   bIsVariable=true
   parent=CommonVisibilityWidgetLegacy_0
   widget=KeybindWidget_PilgrimLane1
   properties {
    bUseKeyboardTextureForKeys=true
    boundAction=PilgrimLane1
    boundInputAction=&IA_PilgrimLane1Press
    visibility=HitTestInvisible
    foregroundColor {colorUseRule=UseColor_Specified, specifiedColor=<rgba 0.0609 0 0.6979 1>}
    renderTransform {angle=0, scale=<0.85 0.85>, shear=<0 0>, translation=<0 0>}
   }
   slot {class=BorderSlot, properties {padding=<ltrb 0 0 0 0>}}
  }
  CommonVisibilityWidgetLegacy {
   widgetClassPath=&CommonVisibilityWidgetLegacy
   parent=Overlay_9
   widget=CommonVisibilityWidgetLegacy
   properties {
    bShowForErebus=false
    bShowForGamepad=false
    bShowForMac=false
    bShowForMouseAndKeyboard=false
    bShowForPS4=false
    bShowForPS5=false
    bShowForSwitch2=false
    bShowForXBox=false
    bShowForXSX=false
    padding=<ltrb 0 0 0 0>
   }
   slot {class=OverlaySlot, properties {horizontalAlignment=HAlign_Fill, verticalAlignment=VAlign_Fill}}
  }
  TouchIcon_Lane1 {
   widgetClassPath=/PilgrimCore/Widgets/Mobile/PilgrimMobileBrush
   bIsVariable=true
   parent=CommonVisibilityWidgetLegacy
   widget=TouchIcon_Lane1
   properties {boundInputAction=&IA_PilgrimLane1Press}
   slot {class=BorderSlot, properties {padding=<ltrb 0 0 0 0>}}
  }
  PilgrimTrackUI_1 {
   widgetClassPath=&PilgrimTrackUI
   parent=HorizontalBox_1
   widget=PilgrimTrackUI_1
   properties {slateTransformMode=Translation}
   slot {class=HorizontalBoxSlot, properties {size {sizeRule=Fill, value=1}}}
  }
  VerticalBox_8 {
   widgetClassPath=&VerticalBox
   parent=PilgrimTrackUI_1
   widget=VerticalBox_8
   slot {class=OverlaySlot, properties {horizontalAlignment=HAlign_Center}}
  }
  Overlay_10 {widgetClassPath=&Overlay, parent=VerticalBox_8, widget=Overlay_10, slot {class=VerticalBoxSlot}}
  CommonVisibilityWidgetLegacy_1 {
   widgetClassPath=&CommonVisibilityWidgetLegacy
   parent=Overlay_10
   widget=CommonVisibilityWidgetLegacy_1
   properties {bShowForTouch=false, padding=<ltrb 0 0 0 0>}
   slot {class=OverlaySlot, properties {horizontalAlignment=HAlign_Fill, verticalAlignment=VAlign_Fill}}
  }
  KeybindWidget_PilgrimLane2 {
   widgetClassPath=&KeybindWidget
   bIsVariable=true
   parent=CommonVisibilityWidgetLegacy_1
   widget=KeybindWidget_PilgrimLane2
   properties {
    bUseKeyboardTextureForKeys=true
    boundAction=PilgrimLane2
    boundInputAction=&IA_PilgrimLane2Press
    visibility=Collapsed
    foregroundColor {colorUseRule=UseColor_Specified, specifiedColor=<rgba 0.0609 0 0.6979 1>}
    renderTransform {angle=0, scale=<0.85 0.85>, shear=<0 0>, translation=<0 0>}
   }
   slot {class=BorderSlot, properties {padding=<ltrb 0 0 0 0>}}
  }
  CommonVisibilityWidgetLegacy_9 {
   widgetClassPath=&CommonVisibilityWidgetLegacy
   parent=Overlay_10
   widget=CommonVisibilityWidgetLegacy_9
   properties {
    bShowForErebus=false
    bShowForGamepad=false
    bShowForMac=false
    bShowForMouseAndKeyboard=false
    bShowForPS4=false
    bShowForPS5=false
    bShowForSwitch2=false
    bShowForXBox=false
    bShowForXSX=false
    padding=<ltrb 0 0 0 0>
   }
   slot {class=OverlaySlot, properties {horizontalAlignment=HAlign_Fill, verticalAlignment=VAlign_Fill}}
  }
  TouchIcon_Lane2 {
   widgetClassPath=&PilgrimMobileBrush
   bIsVariable=true
   parent=CommonVisibilityWidgetLegacy_9
   widget=TouchIcon_Lane2
   properties {boundInputAction=&IA_PilgrimLane2Press}
   slot {class=BorderSlot, properties {padding=<ltrb 0 0 0 0>}}
  }
  Spacer {
   widgetClassPath=&Spacer
   parent=HorizontalBox_1
   widget=Spacer
   properties {size=<15 1>}
   slot {class=HorizontalBoxSlot, properties {horizontalAlignment=HAlign_Center}}
  }
  PilgrimTrackUI_2 {
   widgetClassPath=&PilgrimTrackUI
   parent=HorizontalBox_1
   widget=PilgrimTrackUI_2
   properties {slateTransformMode=Translation}
   slot {class=HorizontalBoxSlot, properties {size {sizeRule=Fill, value=1}}}
  }
  VerticalBox_9 {
   widgetClassPath=&VerticalBox
   parent=PilgrimTrackUI_2
   widget=VerticalBox_9
   slot {class=OverlaySlot, properties {horizontalAlignment=HAlign_Center}}
  }
  Overlay_11 {widgetClassPath=&Overlay, parent=VerticalBox_9, widget=Overlay_11, slot {class=VerticalBoxSlot}}
  CommonVisibilityWidgetLegacy_2 {
   widgetClassPath=&CommonVisibilityWidgetLegacy
   parent=Overlay_11
   widget=CommonVisibilityWidgetLegacy_2
   properties {bShowForTouch=false, padding=<ltrb 0 0 0 0>}
   slot {class=OverlaySlot, properties {horizontalAlignment=HAlign_Fill, verticalAlignment=VAlign_Fill}}
  }
  KeybindWidget_PilgrimLane3 {
   widgetClassPath=&KeybindWidget
   bIsVariable=true
   parent=CommonVisibilityWidgetLegacy_2
   widget=KeybindWidget_PilgrimLane3
   properties {
    bUseKeyboardTextureForKeys=true
    boundAction=PilgrimLane3
    boundInputAction=&IA_PilgrimLane3Press
    foregroundColor {colorUseRule=UseColor_Specified, specifiedColor=<rgba 0.0609 0 0.6979 1>}
    renderTransform {angle=0, scale=<0.85 0.85>, shear=<0 0>, translation=<0 0>}
   }
   slot {class=BorderSlot, properties {padding=<ltrb 0 0 0 0>}}
  }
  CommonVisibilityWidgetLegacy_10 {
   widgetClassPath=&CommonVisibilityWidgetLegacy
   parent=Overlay_11
   widget=CommonVisibilityWidgetLegacy_10
   properties {
    bShowForErebus=false
    bShowForGamepad=false
    bShowForMac=false
    bShowForMouseAndKeyboard=false
    bShowForPS4=false
    bShowForPS5=false
    bShowForSwitch2=false
    bShowForXBox=false
    bShowForXSX=false
    padding=<ltrb 0 0 0 0>
   }
   slot {class=OverlaySlot, properties {horizontalAlignment=HAlign_Fill, verticalAlignment=VAlign_Fill}}
  }
  TouchIcon_Lane3 {
   widgetClassPath=&PilgrimMobileBrush
   bIsVariable=true
   parent=CommonVisibilityWidgetLegacy_10
   widget=TouchIcon_Lane3
   properties {boundInputAction=&IA_PilgrimLane3Press}
   slot {class=BorderSlot, properties {padding=<ltrb 0 0 0 0>}}
  }
  PilgrimTrackUI_3 {
   widgetClassPath=&PilgrimTrackUI
   parent=HorizontalBox_1
   widget=PilgrimTrackUI_3
   properties {slateTransformMode=Translation}
   slot {class=HorizontalBoxSlot, properties {size {sizeRule=Fill, value=1}}}
  }
  VerticalBox_10 {
   widgetClassPath=&VerticalBox
   parent=PilgrimTrackUI_3
   widget=VerticalBox_10
   slot {class=OverlaySlot, properties {horizontalAlignment=HAlign_Center, padding=<ltrb -10 0 0 0>}}
  }
  Overlay_12 {
   widgetClassPath=&Overlay
   parent=VerticalBox_10
   widget=Overlay_12
   slot {class=VerticalBoxSlot, properties {padding=<ltrb 20 0 0 0>}}
  }
  CommonVisibilityWidgetLegacy_3 {
   widgetClassPath=&CommonVisibilityWidgetLegacy
   parent=Overlay_12
   widget=CommonVisibilityWidgetLegacy_3
   properties {bShowForTouch=false, padding=<ltrb 0 0 0 0>}
   slot {class=OverlaySlot, properties {horizontalAlignment=HAlign_Fill, verticalAlignment=VAlign_Fill}}
  }
  KeybindWidget_PilgrimLane4 {
   widgetClassPath=&KeybindWidget
   bIsVariable=true
   parent=CommonVisibilityWidgetLegacy_3
   widget=KeybindWidget_PilgrimLane4
   properties {
    bUseKeyboardTextureForKeys=true
    boundAction=PilgrimLane4
    boundInputAction=&IA_PilgrimLane4Press
    foregroundColor {colorUseRule=UseColor_Foreground, specifiedColor=<rgba 0.0609 0 0.6979 1>}
    renderTransform {angle=0, scale=<0.85 0.85>, shear=<0 0>, translation=<0 0>}
   }
   slot {class=BorderSlot, properties {padding=<ltrb 0 0 0 0>}}
  }
  CommonVisibilityWidgetLegacy_11 {
   widgetClassPath=&CommonVisibilityWidgetLegacy
   parent=Overlay_12
   widget=CommonVisibilityWidgetLegacy_11
   properties {
    bShowForErebus=false
    bShowForGamepad=false
    bShowForMac=false
    bShowForMouseAndKeyboard=false
    bShowForPS4=false
    bShowForPS5=false
    bShowForSwitch2=false
    bShowForXBox=false
    bShowForXSX=false
    padding=<ltrb 0 0 0 0>
   }
   slot {class=OverlaySlot, properties {horizontalAlignment=HAlign_Fill, verticalAlignment=VAlign_Fill}}
  }
  TouchIcon_Lane4 {
   widgetClassPath=&PilgrimMobileBrush
   bIsVariable=true
   parent=CommonVisibilityWidgetLegacy_11
   widget=TouchIcon_Lane4
   properties {boundInputAction=&IA_PilgrimLane4Press}
   slot {class=BorderSlot, properties {padding=<ltrb 0 0 0 0>}}
  }
  KeybindWidgetCanvasContainer_ProDrums {
   widgetClassPath=&CanvasPanel
   bIsVariable=true
   parent=KeybindsSwitcher
   widget=KeybindWidgetCanvasContainer_ProDrums
   slot {class=WidgetSwitcherSlot}
  }
  VerticalBox_11 {
   widgetClassPath=&VerticalBox
   parent=KeybindWidgetCanvasContainer_ProDrums
   widget=VerticalBox_11
   slot {class=CanvasPanelSlot, properties {layoutData {alignment=<0.5 0>, anchors {maximum=<0.5 0>, minimum=<0.5 0>}, offsets=<ltrb 0 20 900 64>}}}
  }
  HorizontalBox {
   widgetClassPath=&HorizontalBox
   parent=VerticalBox_11
   widget=HorizontalBox
   slot {class=VerticalBoxSlot, properties {padding=<ltrb 150 0 150 0>}}
  }
  PilgrimTrackUI {
   widgetClassPath=&PilgrimTrackUI
   parent=HorizontalBox
   widget=PilgrimTrackUI
   properties {slateTransformMode=Translation}
   slot {class=HorizontalBoxSlot, properties {size {sizeRule=Fill, value=1}}}
  }
  VerticalBox_12 {
   widgetClassPath=&VerticalBox
   parent=PilgrimTrackUI
   widget=VerticalBox_12
   slot {class=OverlaySlot, properties {horizontalAlignment=HAlign_Center}}
  }
  Overlay {
   widgetClassPath=&Overlay
   parent=VerticalBox_12
   widget=Overlay
   slot {class=VerticalBoxSlot, properties {padding=<ltrb -10 0 0 0>}}
  }
  CommonVisibilityWidgetLegacy_17 {
   widgetClassPath=&CommonVisibilityWidgetLegacy
   parent=Overlay
   widget=CommonVisibilityWidgetLegacy_17
   properties {bShowForTouch=false, padding=<ltrb 0 0 0 0>}
   slot {class=OverlaySlot, properties {horizontalAlignment=HAlign_Fill, verticalAlignment=VAlign_Fill}}
  }
  KeybindWidget_ProDrums_PilgrimLane1 {
   widgetClassPath=&KeybindWidget
   bIsVariable=true
   parent=CommonVisibilityWidgetLegacy_17
   widget=KeybindWidget_ProDrums_PilgrimLane1
   properties {
    bUseKeyboardTextureForKeys=true
    boundAction=PilgrimLane1
    boundInputAction=&IA_PilgrimLane2Press_5Lane
    visibility=HitTestInvisible
    foregroundColor {colorUseRule=UseColor_Specified, specifiedColor=<rgba 0.0609 0 0.6979 1>}
    renderTransform {angle=0, scale=<0.85 0.85>, shear=<0 0>, translation=<0 0>}
   }
   slot {class=BorderSlot, properties {padding=<ltrb 0 0 0 0>}}
  }
  CommonVisibilityWidgetLegacy_18 {
   widgetClassPath=&CommonVisibilityWidgetLegacy
   parent=Overlay
   widget=CommonVisibilityWidgetLegacy_18
   properties {
    bShowForErebus=false
    bShowForGamepad=false
    bShowForMac=false
    bShowForMouseAndKeyboard=false
    bShowForPS4=false
    bShowForPS5=false
    bShowForSwitch2=false
    bShowForXBox=false
    bShowForXSX=false
    padding=<ltrb 0 0 0 0>
   }
   slot {class=OverlaySlot, properties {horizontalAlignment=HAlign_Fill, verticalAlignment=VAlign_Fill}}
  }
  TouchIcon_ProDrums_Lane1 {
   widgetClassPath=&PilgrimMobileBrush
   bIsVariable=true
   parent=CommonVisibilityWidgetLegacy_18
   widget=TouchIcon_ProDrums_Lane1
   properties {boundInputAction=&IA_PilgrimLane2Press_5Lane}
   slot {class=BorderSlot, properties {padding=<ltrb 0 0 0 0>}}
  }
  PilgrimTrackUI_9 {
   widgetClassPath=&PilgrimTrackUI
   parent=HorizontalBox
   widget=PilgrimTrackUI_9
   properties {slateTransformMode=Translation}
   slot {class=HorizontalBoxSlot, properties {size {sizeRule=Fill, value=1}}}
  }
  VerticalBox_13 {
   widgetClassPath=&VerticalBox
   parent=PilgrimTrackUI_9
   widget=VerticalBox_13
   slot {class=OverlaySlot, properties {horizontalAlignment=HAlign_Center}}
  }
  Overlay_1 {widgetClassPath=&Overlay, parent=VerticalBox_13, widget=Overlay_1, slot {class=VerticalBoxSlot}}
  CommonVisibilityWidgetLegacy_19 {
   widgetClassPath=&CommonVisibilityWidgetLegacy
   parent=Overlay_1
   widget=CommonVisibilityWidgetLegacy_19
   properties {bShowForTouch=false, padding=<ltrb 0 0 0 0>}
   slot {class=OverlaySlot, properties {horizontalAlignment=HAlign_Fill, verticalAlignment=VAlign_Fill}}
  }
  KeybindWidget_ProDrums_PilgrimLane2 {
   widgetClassPath=&KeybindWidget
   bIsVariable=true
   parent=CommonVisibilityWidgetLegacy_19
   widget=KeybindWidget_ProDrums_PilgrimLane2
   properties {
    bUseKeyboardTextureForKeys=true
    boundAction=PilgrimLane2
    boundInputAction=/PilgrimSettings/Input/InputActions/IA_PilgrimLane3Press_5Lane
    visibility=Collapsed
    foregroundColor {colorUseRule=UseColor_Specified, specifiedColor=<rgba 0.0609 0 0.6979 1>}
    renderTransform {angle=0, scale=<0.85 0.85>, shear=<0 0>, translation=<0 0>}
   }
   slot {class=BorderSlot, properties {padding=<ltrb 0 0 0 0>}}
  }
  CommonVisibilityWidgetLegacy_20 {
   widgetClassPath=&CommonVisibilityWidgetLegacy
   parent=Overlay_1
   widget=CommonVisibilityWidgetLegacy_20
   properties {
    bShowForErebus=false
    bShowForGamepad=false
    bShowForMac=false
    bShowForMouseAndKeyboard=false
    bShowForPS4=false
    bShowForPS5=false
    bShowForSwitch2=false
    bShowForXBox=false
    bShowForXSX=false
    padding=<ltrb 0 0 0 0>
   }
   slot {class=OverlaySlot, properties {horizontalAlignment=HAlign_Fill, verticalAlignment=VAlign_Fill}}
  }
  TouchIcon_ProDrums_Lane2 {
   widgetClassPath=&PilgrimMobileBrush
   bIsVariable=true
   parent=CommonVisibilityWidgetLegacy_20
   widget=TouchIcon_ProDrums_Lane2
   properties {boundInputAction=&IA_PilgrimLane3Press_5Lane}
   slot {class=BorderSlot, properties {padding=<ltrb 0 0 0 0>}}
  }
  Spacer_1 {
   widgetClassPath=&Spacer
   parent=HorizontalBox
   widget=Spacer_1
   properties {size=<15 1>}
   slot {class=HorizontalBoxSlot, properties {horizontalAlignment=HAlign_Center}}
  }
  PilgrimTrackUI_10 {
   widgetClassPath=&PilgrimTrackUI
   parent=HorizontalBox
   widget=PilgrimTrackUI_10
   properties {slateTransformMode=Translation}
   slot {class=HorizontalBoxSlot, properties {size {sizeRule=Fill, value=1}}}
  }
  VerticalBox_14 {
   widgetClassPath=&VerticalBox
   parent=PilgrimTrackUI_10
   widget=VerticalBox_14
   slot {class=OverlaySlot, properties {horizontalAlignment=HAlign_Center}}
  }
  Overlay_2 {widgetClassPath=&Overlay, parent=VerticalBox_14, widget=Overlay_2, slot {class=VerticalBoxSlot}}
  CommonVisibilityWidgetLegacy_21 {
   widgetClassPath=&CommonVisibilityWidgetLegacy
   parent=Overlay_2
   widget=CommonVisibilityWidgetLegacy_21
   properties {bShowForTouch=false, padding=<ltrb 0 0 0 0>}
   slot {class=OverlaySlot, properties {horizontalAlignment=HAlign_Fill, verticalAlignment=VAlign_Fill}}
  }
  KeybindWidget_ProDrums_PilgrimLane3 {
   widgetClassPath=&KeybindWidget
   bIsVariable=true
   parent=CommonVisibilityWidgetLegacy_21
   widget=KeybindWidget_ProDrums_PilgrimLane3
   properties {
    bUseKeyboardTextureForKeys=true
    boundAction=PilgrimLane3
    boundInputAction=&IA_PilgrimLane4Press_5Lane
    foregroundColor {colorUseRule=UseColor_Specified, specifiedColor=<rgba 0.0609 0 0.6979 1>}
    renderTransform {angle=0, scale=<0.85 0.85>, shear=<0 0>, translation=<0 0>}
   }
   slot {class=BorderSlot, properties {padding=<ltrb 0 0 0 0>}}
  }
  CommonVisibilityWidgetLegacy_22 {
   widgetClassPath=&CommonVisibilityWidgetLegacy
   parent=Overlay_2
   widget=CommonVisibilityWidgetLegacy_22
   properties {
    bShowForErebus=false
    bShowForGamepad=false
    bShowForMac=false
    bShowForMouseAndKeyboard=false
    bShowForPS4=false
    bShowForPS5=false
    bShowForSwitch2=false
    bShowForXBox=false
    bShowForXSX=false
    padding=<ltrb 0 0 0 0>
   }
   slot {class=OverlaySlot, properties {horizontalAlignment=HAlign_Fill, verticalAlignment=VAlign_Fill}}
  }
  TouchIcon_ProDrums_Lane3 {
   widgetClassPath=&PilgrimMobileBrush
   bIsVariable=true
   parent=CommonVisibilityWidgetLegacy_22
   widget=TouchIcon_ProDrums_Lane3
   properties {boundInputAction=&IA_PilgrimLane4Press_5Lane}
   slot {class=BorderSlot, properties {padding=<ltrb 0 0 0 0>}}
  }
  PilgrimTrackUI_11 {
   widgetClassPath=&PilgrimTrackUI
   parent=HorizontalBox
   widget=PilgrimTrackUI_11
   properties {slateTransformMode=Translation}
   slot {class=HorizontalBoxSlot, properties {size {sizeRule=Fill, value=1}}}
  }
  VerticalBox_15 {
   widgetClassPath=&VerticalBox
   parent=PilgrimTrackUI_11
   widget=VerticalBox_15
   slot {class=OverlaySlot, properties {horizontalAlignment=HAlign_Center, padding=<ltrb -10 0 0 0>}}
  }
  Overlay_3 {
   widgetClassPath=&Overlay
   parent=VerticalBox_15
   widget=Overlay_3
   slot {class=VerticalBoxSlot, properties {padding=<ltrb 20 0 0 0>}}
  }
  CommonVisibilityWidgetLegacy_23 {
   widgetClassPath=&CommonVisibilityWidgetLegacy
   parent=Overlay_3
   widget=CommonVisibilityWidgetLegacy_23
   properties {bShowForTouch=false, padding=<ltrb 0 0 0 0>}
   slot {class=OverlaySlot, properties {horizontalAlignment=HAlign_Fill, verticalAlignment=VAlign_Fill}}
  }
  KeybindWidget_ProDrums_PilgrimLane4 {
   widgetClassPath=&KeybindWidget
   bIsVariable=true
   parent=CommonVisibilityWidgetLegacy_23
   widget=KeybindWidget_ProDrums_PilgrimLane4
   properties {
    bUseKeyboardTextureForKeys=true
    boundAction=PilgrimLane4
    boundInputAction=&IA_PilgrimLane1Press_5Lane
    foregroundColor {colorUseRule=UseColor_Foreground, specifiedColor=<rgba 0.0609 0 0.6979 1>}
    renderTransform {angle=0, scale=<0.85 0.85>, shear=<0 0>, translation=<0 0>}
   }
   slot {class=BorderSlot, properties {padding=<ltrb 0 0 0 0>}}
  }
  CommonVisibilityWidgetLegacy_24 {
   widgetClassPath=&CommonVisibilityWidgetLegacy
   parent=Overlay_3
   widget=CommonVisibilityWidgetLegacy_24
   properties {
    bShowForErebus=false
    bShowForGamepad=false
    bShowForMac=false
    bShowForMouseAndKeyboard=false
    bShowForPS4=false
    bShowForPS5=false
    bShowForSwitch2=false
    bShowForXBox=false
    bShowForXSX=false
    padding=<ltrb 0 0 0 0>
   }
   slot {class=OverlaySlot, properties {horizontalAlignment=HAlign_Fill, verticalAlignment=VAlign_Fill}}
  }
  TouchIcon_ProDrums_Lane4 {
   widgetClassPath=&PilgrimMobileBrush
   bIsVariable=true
   parent=CommonVisibilityWidgetLegacy_24
   widget=TouchIcon_ProDrums_Lane4
   properties {boundInputAction=&IA_PilgrimLane1Press_5Lane}
   slot {class=BorderSlot, properties {padding=<ltrb 0 0 0 0>}}
  }
  CymbalsInputHorizontalBox {
   widgetClassPath=&HorizontalBox
   bIsVariable=true
   parent=VerticalBox_11
   widget=CymbalsInputHorizontalBox
   properties {visibility=Collapsed}
   slot {class=VerticalBoxSlot, properties {padding=<ltrb 150 0 150 0>}}
  }
  PilgrimTrackUI_12 {
   widgetClassPath=&PilgrimTrackUI
   parent=CymbalsInputHorizontalBox
   widget=PilgrimTrackUI_12
   properties {slateTransformMode=Translation}
   slot {class=HorizontalBoxSlot, properties {size {sizeRule=Fill, value=1}}}
  }
  VerticalBox_16 {
   widgetClassPath=&VerticalBox
   parent=PilgrimTrackUI_12
   widget=VerticalBox_16
   slot {class=OverlaySlot, properties {horizontalAlignment=HAlign_Center}}
  }
  Overlay_4 {
   widgetClassPath=&Overlay
   parent=VerticalBox_16
   widget=Overlay_4
   properties {visibility=Collapsed}
   slot {class=VerticalBoxSlot, properties {padding=<ltrb -10 0 0 0>}}
  }
  CommonVisibilityWidgetLegacy_25 {
   widgetClassPath=&CommonVisibilityWidgetLegacy
   parent=Overlay_4
   widget=CommonVisibilityWidgetLegacy_25
   properties {bShowForTouch=false, padding=<ltrb 0 0 0 0>}
   slot {class=OverlaySlot, properties {horizontalAlignment=HAlign_Fill, verticalAlignment=VAlign_Fill}}
  }
  KeybindWidget_PilgrimLane1_2 {
   widgetClassPath=&KeybindWidget
   bIsVariable=true
   parent=CommonVisibilityWidgetLegacy_25
   widget=KeybindWidget_PilgrimLane1_2
   properties {
    bUseKeyboardTextureForKeys=true
    boundAction=PilgrimLane1
    visibility=HitTestInvisible
    foregroundColor {colorUseRule=UseColor_Specified, specifiedColor=<rgba 0.0609 0 0.6979 1>}
    renderTransform {angle=0, scale=<0.85 0.85>, shear=<0 0>, translation=<0 0>}
   }
   slot {class=BorderSlot, properties {padding=<ltrb 0 0 0 0>}}
  }
  CommonVisibilityWidgetLegacy_26 {
   widgetClassPath=&CommonVisibilityWidgetLegacy
   parent=Overlay_4
   widget=CommonVisibilityWidgetLegacy_26
   properties {
    bShowForErebus=false
    bShowForGamepad=false
    bShowForMac=false
    bShowForMouseAndKeyboard=false
    bShowForPS4=false
    bShowForPS5=false
    bShowForSwitch2=false
    bShowForXBox=false
    bShowForXSX=false
    padding=<ltrb 0 0 0 0>
   }
   slot {class=OverlaySlot, properties {horizontalAlignment=HAlign_Fill, verticalAlignment=VAlign_Fill}}
  }
  TouchIcon_Lane1_2 {
   widgetClassPath=&PilgrimMobileBrush
   bIsVariable=true
   parent=CommonVisibilityWidgetLegacy_26
   widget=TouchIcon_Lane1_2
   slot {class=BorderSlot, properties {padding=<ltrb 0 0 0 0>}}
  }
  PilgrimTrackUI_13 {
   widgetClassPath=&PilgrimTrackUI
   parent=CymbalsInputHorizontalBox
   widget=PilgrimTrackUI_13
   properties {slateTransformMode=Translation}
   slot {class=HorizontalBoxSlot, properties {size {sizeRule=Fill, value=1}}}
  }
  VerticalBox_17 {
   widgetClassPath=&VerticalBox
   parent=PilgrimTrackUI_13
   widget=VerticalBox_17
   slot {class=OverlaySlot, properties {horizontalAlignment=HAlign_Center}}
  }
  Overlay_5 {widgetClassPath=&Overlay, parent=VerticalBox_17, widget=Overlay_5, slot {class=VerticalBoxSlot}}
  CommonVisibilityWidgetLegacy_27 {
   widgetClassPath=&CommonVisibilityWidgetLegacy
   parent=Overlay_5
   widget=CommonVisibilityWidgetLegacy_27
   properties {bShowForTouch=false, padding=<ltrb 0 0 0 0>}
   slot {class=OverlaySlot, properties {horizontalAlignment=HAlign_Fill, verticalAlignment=VAlign_Fill}}
  }
  KeybindWidget_PilgrimLane2_2 {
   widgetClassPath=&KeybindWidget
   bIsVariable=true
   parent=CommonVisibilityWidgetLegacy_27
   widget=KeybindWidget_PilgrimLane2_2
   properties {
    bUseKeyboardTextureForKeys=true
    boundAction=PilgrimLane2
    boundInputAction=/PilgrimSettings/Input/InputActions/IA_PilgrimLane1Press_5Lane3_Special
    visibility=Collapsed
    foregroundColor {colorUseRule=UseColor_Specified, specifiedColor=<rgba 0.0609 0 0.6979 1>}
    renderTransform {angle=0, scale=<0.85 0.85>, shear=<0 0>, translation=<0 0>}
   }
   slot {class=BorderSlot, properties {padding=<ltrb 0 0 0 0>}}
  }
  CommonVisibilityWidgetLegacy_28 {
   widgetClassPath=&CommonVisibilityWidgetLegacy
   parent=Overlay_5
   widget=CommonVisibilityWidgetLegacy_28
   properties {
    bShowForErebus=false
    bShowForGamepad=false
    bShowForMac=false
    bShowForMouseAndKeyboard=false
    bShowForPS4=false
    bShowForPS5=false
    bShowForSwitch2=false
    bShowForXBox=false
    bShowForXSX=false
    padding=<ltrb 0 0 0 0>
   }
   slot {class=OverlaySlot, properties {horizontalAlignment=HAlign_Fill, verticalAlignment=VAlign_Fill}}
  }
  TouchIcon_Lane2_2 {
   widgetClassPath=&PilgrimMobileBrush
   bIsVariable=true
   parent=CommonVisibilityWidgetLegacy_28
   widget=TouchIcon_Lane2_2
   properties {boundInputAction=&IA_PilgrimLane1Press_5Lane3_Special}
   slot {class=BorderSlot, properties {padding=<ltrb 0 0 0 0>}}
  }
  Spacer_2 {
   widgetClassPath=&Spacer
   parent=CymbalsInputHorizontalBox
   widget=Spacer_2
   properties {size=<15 1>}
   slot {class=HorizontalBoxSlot, properties {horizontalAlignment=HAlign_Center}}
  }
  PilgrimTrackUI_14 {
   widgetClassPath=&PilgrimTrackUI
   parent=CymbalsInputHorizontalBox
   widget=PilgrimTrackUI_14
   properties {slateTransformMode=Translation}
   slot {class=HorizontalBoxSlot, properties {size {sizeRule=Fill, value=1}}}
  }
  VerticalBox_18 {
   widgetClassPath=&VerticalBox
   parent=PilgrimTrackUI_14
   widget=VerticalBox_18
   slot {class=OverlaySlot, properties {horizontalAlignment=HAlign_Center}}
  }
  Overlay_6 {widgetClassPath=&Overlay, parent=VerticalBox_18, widget=Overlay_6, slot {class=VerticalBoxSlot}}
  CommonVisibilityWidgetLegacy_29 {
   widgetClassPath=&CommonVisibilityWidgetLegacy
   parent=Overlay_6
   widget=CommonVisibilityWidgetLegacy_29
   properties {bShowForTouch=false, padding=<ltrb 0 0 0 0>}
   slot {class=OverlaySlot, properties {horizontalAlignment=HAlign_Fill, verticalAlignment=VAlign_Fill}}
  }
  KeybindWidget_PilgrimLane3_2 {
   widgetClassPath=&KeybindWidget
   bIsVariable=true
   parent=CommonVisibilityWidgetLegacy_29
   widget=KeybindWidget_PilgrimLane3_2
   properties {
    bUseKeyboardTextureForKeys=true
    boundAction=PilgrimLane3
    boundInputAction=/PilgrimSettings/Input/InputActions/IA_PilgrimLane1Press_5Lane4_Special
    foregroundColor {colorUseRule=UseColor_Specified, specifiedColor=<rgba 0.0609 0 0.6979 1>}
    renderTransform {angle=0, scale=<0.85 0.85>, shear=<0 0>, translation=<0 0>}
   }
   slot {class=BorderSlot, properties {padding=<ltrb 0 0 0 0>}}
  }
  CommonVisibilityWidgetLegacy_30 {
   widgetClassPath=&CommonVisibilityWidgetLegacy
   parent=Overlay_6
   widget=CommonVisibilityWidgetLegacy_30
   properties {
    bShowForErebus=false
    bShowForGamepad=false
    bShowForMac=false
    bShowForMouseAndKeyboard=false
    bShowForPS4=false
    bShowForPS5=false
    bShowForSwitch2=false
    bShowForXBox=false
    bShowForXSX=false
    padding=<ltrb 0 0 0 0>
   }
   slot {class=OverlaySlot, properties {horizontalAlignment=HAlign_Fill, verticalAlignment=VAlign_Fill}}
  }
  TouchIcon_Lane3_2 {
   widgetClassPath=&PilgrimMobileBrush
   bIsVariable=true
   parent=CommonVisibilityWidgetLegacy_30
   widget=TouchIcon_Lane3_2
   properties {boundInputAction=&IA_PilgrimLane1Press_5Lane4_Special}
   slot {class=BorderSlot, properties {padding=<ltrb 0 0 0 0>}}
  }
  PilgrimTrackUI_15 {
   widgetClassPath=&PilgrimTrackUI
   parent=CymbalsInputHorizontalBox
   widget=PilgrimTrackUI_15
   properties {slateTransformMode=Translation}
   slot {class=HorizontalBoxSlot, properties {size {sizeRule=Fill, value=1}}}
  }
  VerticalBox_19 {
   widgetClassPath=&VerticalBox
   parent=PilgrimTrackUI_15
   widget=VerticalBox_19
   slot {class=OverlaySlot, properties {horizontalAlignment=HAlign_Center, padding=<ltrb -10 0 0 0>}}
  }
  Overlay_7 {
   widgetClassPath=&Overlay
   parent=VerticalBox_19
   widget=Overlay_7
   slot {class=VerticalBoxSlot, properties {padding=<ltrb 20 0 0 0>}}
  }
  CommonVisibilityWidgetLegacy_31 {
   widgetClassPath=&CommonVisibilityWidgetLegacy
   parent=Overlay_7
   widget=CommonVisibilityWidgetLegacy_31
   properties {bShowForTouch=false, padding=<ltrb 0 0 0 0>}
   slot {class=OverlaySlot, properties {horizontalAlignment=HAlign_Fill, verticalAlignment=VAlign_Fill}}
  }
  KeybindWidget_PilgrimLane4_2 {
   widgetClassPath=&KeybindWidget
   bIsVariable=true
   parent=CommonVisibilityWidgetLegacy_31
   widget=KeybindWidget_PilgrimLane4_2
   properties {
    bUseKeyboardTextureForKeys=true
    boundAction=PilgrimLane4
    boundInputAction=/PilgrimSettings/Input/InputActions/IA_PilgrimLane1Press_5Lane1_Special
    foregroundColor {colorUseRule=UseColor_Foreground, specifiedColor=<rgba 0.0609 0 0.6979 1>}
    renderTransform {angle=0, scale=<0.85 0.85>, shear=<0 0>, translation=<0 0>}
   }
   slot {class=BorderSlot, properties {padding=<ltrb 0 0 0 0>}}
  }
  CommonVisibilityWidgetLegacy_32 {
   widgetClassPath=&CommonVisibilityWidgetLegacy
   parent=Overlay_7
   widget=CommonVisibilityWidgetLegacy_32
   properties {
    bShowForErebus=false
    bShowForGamepad=false
    bShowForMac=false
    bShowForMouseAndKeyboard=false
    bShowForPS4=false
    bShowForPS5=false
    bShowForSwitch2=false
    bShowForXBox=false
    bShowForXSX=false
    padding=<ltrb 0 0 0 0>
   }
   slot {class=OverlaySlot, properties {horizontalAlignment=HAlign_Fill, verticalAlignment=VAlign_Fill}}
  }
  TouchIcon_Lane4_2 {
   widgetClassPath=&PilgrimMobileBrush
   bIsVariable=true
   parent=CommonVisibilityWidgetLegacy_32
   widget=TouchIcon_Lane4_2
   properties {boundInputAction=&IA_PilgrimLane1Press_5Lane1_Special}
   slot {class=BorderSlot, properties {padding=<ltrb 0 0 0 0>}}
  }
  LeftyCymbalsInputHorizontalBox {
   widgetClassPath=&HorizontalBox
   bIsVariable=true
   parent=VerticalBox_11
   widget=LeftyCymbalsInputHorizontalBox
   properties {visibility=Collapsed}
   slot {class=VerticalBoxSlot, properties {padding=<ltrb 150 0 150 0>}}
  }
  PilgrimTrackUI_16 {
   widgetClassPath=&PilgrimTrackUI
   parent=LeftyCymbalsInputHorizontalBox
   widget=PilgrimTrackUI_16
   properties {slateTransformMode=Translation}
   slot {class=HorizontalBoxSlot, properties {size {sizeRule=Fill, value=1}}}
  }
  VerticalBox_20 {
   widgetClassPath=&VerticalBox
   parent=PilgrimTrackUI_16
   widget=VerticalBox_20
   slot {class=OverlaySlot, properties {horizontalAlignment=HAlign_Center}}
  }
  Overlay_8 {
   widgetClassPath=&Overlay
   parent=VerticalBox_20
   widget=Overlay_8
   slot {class=VerticalBoxSlot, properties {padding=<ltrb -10 0 0 0>}}
  }
  CommonVisibilityWidgetLegacy_33 {
   widgetClassPath=&CommonVisibilityWidgetLegacy
   parent=Overlay_8
   widget=CommonVisibilityWidgetLegacy_33
   properties {bShowForTouch=false, padding=<ltrb 0 0 0 0>}
   slot {class=OverlaySlot, properties {horizontalAlignment=HAlign_Fill, verticalAlignment=VAlign_Fill}}
  }
  KeybindWidget_PilgrimLane1_3 {
   widgetClassPath=&KeybindWidget
   bIsVariable=true
   parent=CommonVisibilityWidgetLegacy_33
   widget=KeybindWidget_PilgrimLane1_3
   properties {
    bUseKeyboardTextureForKeys=true
    boundAction=PilgrimLane1
    boundInputAction=&IA_PilgrimLane1Press_5Lane3_Special
    visibility=HitTestInvisible
    foregroundColor {colorUseRule=UseColor_Specified, specifiedColor=<rgba 0.0609 0 0.6979 1>}
    renderTransform {angle=0, scale=<0.85 0.85>, shear=<0 0>, translation=<0 0>}
   }
   slot {class=BorderSlot, properties {padding=<ltrb 0 0 0 0>}}
  }
  CommonVisibilityWidgetLegacy_34 {
   widgetClassPath=&CommonVisibilityWidgetLegacy
   parent=Overlay_8
   widget=CommonVisibilityWidgetLegacy_34
   properties {
    bShowForErebus=false
    bShowForGamepad=false
    bShowForMac=false
    bShowForMouseAndKeyboard=false
    bShowForPS4=false
    bShowForPS5=false
    bShowForSwitch2=false
    bShowForXBox=false
    bShowForXSX=false
    padding=<ltrb 0 0 0 0>
   }
   slot {class=OverlaySlot, properties {horizontalAlignment=HAlign_Fill, verticalAlignment=VAlign_Fill}}
  }
  TouchIcon_Lane1_3 {
   widgetClassPath=&PilgrimMobileBrush
   bIsVariable=true
   parent=CommonVisibilityWidgetLegacy_34
   widget=TouchIcon_Lane1_3
   properties {boundInputAction=&IA_PilgrimLane1Press_5Lane3_Special}
   slot {class=BorderSlot, properties {padding=<ltrb 0 0 0 0>}}
  }
  PilgrimTrackUI_17 {
   widgetClassPath=&PilgrimTrackUI
   parent=LeftyCymbalsInputHorizontalBox
   widget=PilgrimTrackUI_17
   properties {slateTransformMode=Translation}
   slot {class=HorizontalBoxSlot, properties {size {sizeRule=Fill, value=1}}}
  }
  VerticalBox_21 {
   widgetClassPath=&VerticalBox
   parent=PilgrimTrackUI_17
   widget=VerticalBox_21
   slot {class=OverlaySlot, properties {horizontalAlignment=HAlign_Center}}
  }
  Overlay_18 {widgetClassPath=&Overlay, parent=VerticalBox_21, widget=Overlay_18, slot {class=VerticalBoxSlot}}
  CommonVisibilityWidgetLegacy_35 {
   widgetClassPath=&CommonVisibilityWidgetLegacy
   parent=Overlay_18
   widget=CommonVisibilityWidgetLegacy_35
   properties {bShowForTouch=false, padding=<ltrb 0 0 0 0>}
   slot {class=OverlaySlot, properties {horizontalAlignment=HAlign_Fill, verticalAlignment=VAlign_Fill}}
  }
  KeybindWidget_PilgrimLane2_3 {
   widgetClassPath=&KeybindWidget
   bIsVariable=true
   parent=CommonVisibilityWidgetLegacy_35
   widget=KeybindWidget_PilgrimLane2_3
   properties {
    bUseKeyboardTextureForKeys=true
    boundAction=PilgrimLane2
    boundInputAction=&IA_PilgrimLane1Press_5Lane4_Special
    visibility=Collapsed
    foregroundColor {colorUseRule=UseColor_Specified, specifiedColor=<rgba 0.0609 0 0.6979 1>}
    renderTransform {angle=0, scale=<0.85 0.85>, shear=<0 0>, translation=<0 0>}
   }
   slot {class=BorderSlot, properties {padding=<ltrb 0 0 0 0>}}
  }
  CommonVisibilityWidgetLegacy_36 {
   widgetClassPath=&CommonVisibilityWidgetLegacy
   parent=Overlay_18
   widget=CommonVisibilityWidgetLegacy_36
   properties {
    bShowForErebus=false
    bShowForGamepad=false
    bShowForMac=false
    bShowForMouseAndKeyboard=false
    bShowForPS4=false
    bShowForPS5=false
    bShowForSwitch2=false
    bShowForXBox=false
    bShowForXSX=false
    padding=<ltrb 0 0 0 0>
   }
   slot {class=OverlaySlot, properties {horizontalAlignment=HAlign_Fill, verticalAlignment=VAlign_Fill}}
  }
  TouchIcon_Lane2_3 {
   widgetClassPath=&PilgrimMobileBrush
   bIsVariable=true
   parent=CommonVisibilityWidgetLegacy_36
   widget=TouchIcon_Lane2_3
   properties {boundInputAction=&IA_PilgrimLane1Press_5Lane4_Special}
   slot {class=BorderSlot, properties {padding=<ltrb 0 0 0 0>}}
  }
  Spacer_3 {
   widgetClassPath=&Spacer
   parent=LeftyCymbalsInputHorizontalBox
   widget=Spacer_3
   properties {size=<15 1>}
   slot {class=HorizontalBoxSlot, properties {horizontalAlignment=HAlign_Center}}
  }
  PilgrimTrackUI_18 {
   widgetClassPath=&PilgrimTrackUI
   parent=LeftyCymbalsInputHorizontalBox
   widget=PilgrimTrackUI_18
   properties {slateTransformMode=Translation}
   slot {class=HorizontalBoxSlot, properties {size {sizeRule=Fill, value=1}}}
  }
  VerticalBox_22 {
   widgetClassPath=&VerticalBox
   parent=PilgrimTrackUI_18
   widget=VerticalBox_22
   slot {class=OverlaySlot, properties {horizontalAlignment=HAlign_Center}}
  }
  Overlay_19 {widgetClassPath=&Overlay, parent=VerticalBox_22, widget=Overlay_19, slot {class=VerticalBoxSlot}}
  CommonVisibilityWidgetLegacy_37 {
   widgetClassPath=&CommonVisibilityWidgetLegacy
   parent=Overlay_19
   widget=CommonVisibilityWidgetLegacy_37
   properties {bShowForTouch=false, padding=<ltrb 0 0 0 0>}
   slot {class=OverlaySlot, properties {horizontalAlignment=HAlign_Fill, verticalAlignment=VAlign_Fill}}
  }
  KeybindWidget_PilgrimLane3_3 {
   widgetClassPath=&KeybindWidget
   bIsVariable=true
   parent=CommonVisibilityWidgetLegacy_37
   widget=KeybindWidget_PilgrimLane3_3
   properties {
    bUseKeyboardTextureForKeys=true
    boundAction=PilgrimLane3
    boundInputAction=&IA_PilgrimLane1Press_5Lane1_Special
    foregroundColor {colorUseRule=UseColor_Specified, specifiedColor=<rgba 0.0609 0 0.6979 1>}
    renderTransform {angle=0, scale=<0.85 0.85>, shear=<0 0>, translation=<0 0>}
   }
   slot {class=BorderSlot, properties {padding=<ltrb 0 0 0 0>}}
  }
  CommonVisibilityWidgetLegacy_38 {
   widgetClassPath=&CommonVisibilityWidgetLegacy
   parent=Overlay_19
   widget=CommonVisibilityWidgetLegacy_38
   properties {
    bShowForErebus=false
    bShowForGamepad=false
    bShowForMac=false
    bShowForMouseAndKeyboard=false
    bShowForPS4=false
    bShowForPS5=false
    bShowForSwitch2=false
    bShowForXBox=false
    bShowForXSX=false
    padding=<ltrb 0 0 0 0>
   }
   slot {class=OverlaySlot, properties {horizontalAlignment=HAlign_Fill, verticalAlignment=VAlign_Fill}}
  }
  TouchIcon_Lane3_3 {
   widgetClassPath=&PilgrimMobileBrush
   bIsVariable=true
   parent=CommonVisibilityWidgetLegacy_38
   widget=TouchIcon_Lane3_3
   properties {boundInputAction=&IA_PilgrimLane1Press_5Lane1_Special}
   slot {class=BorderSlot, properties {padding=<ltrb 0 0 0 0>}}
  }
  PilgrimTrackUI_19 {
   widgetClassPath=&PilgrimTrackUI
   parent=LeftyCymbalsInputHorizontalBox
   widget=PilgrimTrackUI_19
   properties {slateTransformMode=Translation}
   slot {class=HorizontalBoxSlot, properties {size {sizeRule=Fill, value=1}}}
  }
  VerticalBox_23 {
   widgetClassPath=&VerticalBox
   parent=PilgrimTrackUI_19
   widget=VerticalBox_23
   slot {class=OverlaySlot, properties {horizontalAlignment=HAlign_Center, padding=<ltrb -10 0 0 0>}}
  }
  Overlay_20 {
   widgetClassPath=&Overlay
   parent=VerticalBox_23
   widget=Overlay_20
   properties {visibility=Collapsed}
   slot {class=VerticalBoxSlot, properties {padding=<ltrb 20 0 0 0>}}
  }
  CommonVisibilityWidgetLegacy_39 {
   widgetClassPath=&CommonVisibilityWidgetLegacy
   parent=Overlay_20
   widget=CommonVisibilityWidgetLegacy_39
   properties {bShowForTouch=false, padding=<ltrb 0 0 0 0>}
   slot {class=OverlaySlot, properties {horizontalAlignment=HAlign_Fill, verticalAlignment=VAlign_Fill}}
  }
  KeybindWidget_PilgrimLane4_3 {
   widgetClassPath=&KeybindWidget
   bIsVariable=true
   parent=CommonVisibilityWidgetLegacy_39
   widget=KeybindWidget_PilgrimLane4_3
   properties {
    bUseKeyboardTextureForKeys=true
    boundAction=PilgrimLane4
    foregroundColor {colorUseRule=UseColor_Foreground, specifiedColor=<rgba 0.0609 0 0.6979 1>}
    renderTransform {angle=0, scale=<0.85 0.85>, shear=<0 0>, translation=<0 0>}
   }
   slot {class=BorderSlot, properties {padding=<ltrb 0 0 0 0>}}
  }
  CommonVisibilityWidgetLegacy_40 {
   widgetClassPath=&CommonVisibilityWidgetLegacy
   parent=Overlay_20
   widget=CommonVisibilityWidgetLegacy_40
   properties {
    bShowForErebus=false
    bShowForGamepad=false
    bShowForMac=false
    bShowForMouseAndKeyboard=false
    bShowForPS4=false
    bShowForPS5=false
    bShowForSwitch2=false
    bShowForXBox=false
    bShowForXSX=false
    padding=<ltrb 0 0 0 0>
   }
   slot {class=OverlaySlot, properties {horizontalAlignment=HAlign_Fill, verticalAlignment=VAlign_Fill}}
  }
  TouchIcon_Lane4_3 {
   widgetClassPath=&PilgrimMobileBrush
   bIsVariable=true
   parent=CommonVisibilityWidgetLegacy_40
   widget=TouchIcon_Lane4_3
   slot {class=BorderSlot, properties {padding=<ltrb 0 0 0 0>}}
  }
  KeybindWidgetCanvasContainer_5Lane {
   widgetClassPath=&CanvasPanel
   bIsVariable=true
   parent=KeybindsSwitcher
   widget=KeybindWidgetCanvasContainer_5Lane
   slot {class=WidgetSwitcherSlot}
  }
  VerticalBox_5 {
   widgetClassPath=&VerticalBox
   parent=KeybindWidgetCanvasContainer_5Lane
   widget=VerticalBox_5
   slot {class=CanvasPanelSlot, properties {layoutData {alignment=<0.5 0>, anchors {maximum=<0.5 0>, minimum=<0.5 0>}, offsets=<ltrb 0 20 900 64>}}}
  }
  HorizontalBox_2 {widgetClassPath=&HorizontalBox, parent=VerticalBox_5, widget=HorizontalBox_2, slot {class=VerticalBoxSlot}}
  5Lane_LeftSpacing {
   widgetClassPath=&Spacer
   parent=HorizontalBox_2
   widget=5Lane_LeftSpacing
   properties {size=<132 1>}
   slot {class=HorizontalBoxSlot}
  }
  5LaneHbox {
   widgetClassPath=&HorizontalBox
   bIsVariable=true
   parent=HorizontalBox_2
   widget=5LaneHbox
   properties {renderTransform {angle=0, scale=<1 1>, shear=<0 0>, translation=<8 0>}}
   slot {class=HorizontalBoxSlot, properties {size {sizeRule=Fill, value=1}}}
  }
  PilgrimTrackUI_4 {
   widgetClassPath=&PilgrimTrackUI
   parent=5LaneHbox
   widget=PilgrimTrackUI_4
   properties {slateTransformMode=Translation}
   slot {class=HorizontalBoxSlot, properties {horizontalAlignment=HAlign_Left, size {sizeRule=Fill, value=1}}}
  }
  VerticalBox_142 {widgetClassPath=&VerticalBox, parent=PilgrimTrackUI_4, widget=VerticalBox_142, slot {class=OverlaySlot}}
  Overlay_13 {widgetClassPath=&Overlay, parent=VerticalBox_142, widget=Overlay_13, slot {class=VerticalBoxSlot}}
  CommonVisibilityWidgetLegacy_4 {
   widgetClassPath=&CommonVisibilityWidgetLegacy
   parent=Overlay_13
   widget=CommonVisibilityWidgetLegacy_4
   properties {bShowForTouch=false, padding=<ltrb 0 0 0 0>}
   slot {class=OverlaySlot, properties {horizontalAlignment=HAlign_Fill, verticalAlignment=VAlign_Fill}}
  }
  KeybindWidget_PilgrimLane1_5Lane {
   widgetClassPath=&KeybindWidget
   bIsVariable=true
   parent=CommonVisibilityWidgetLegacy_4
   widget=KeybindWidget_PilgrimLane1_5Lane
   properties {
    bUseKeyboardTextureForKeys=true
    boundAction=PilgrimLane1
    boundInputAction=&IA_PilgrimLane1Press_5Lane
    foregroundColor {colorUseRule=UseColor_Specified, specifiedColor=<rgba 0.0609 0 0.6979 1>}
    renderTransform {angle=0, scale=<0.85 0.85>, shear=<0 0>, translation=<0 0>}
   }
   slot {class=BorderSlot, properties {padding=<ltrb 0 0 0 0>}}
  }
  CommonVisibilityWidgetLegacy_12 {
   widgetClassPath=&CommonVisibilityWidgetLegacy
   parent=Overlay_13
   widget=CommonVisibilityWidgetLegacy_12
   properties {
    bShowForErebus=false
    bShowForGamepad=false
    bShowForMac=false
    bShowForMouseAndKeyboard=false
    bShowForPS4=false
    bShowForPS5=false
    bShowForSwitch2=false
    bShowForXBox=false
    bShowForXSX=false
    padding=<ltrb 0 0 0 0>
   }
   slot {class=OverlaySlot, properties {horizontalAlignment=HAlign_Fill, verticalAlignment=VAlign_Fill}}
  }
  TouchIcon_Lane1_5Lane {
   widgetClassPath=&PilgrimMobileBrush
   bIsVariable=true
   parent=CommonVisibilityWidgetLegacy_12
   widget=TouchIcon_Lane1_5Lane
   properties {boundInputAction=&IA_PilgrimLane1Press_5Lane}
   slot {class=BorderSlot, properties {padding=<ltrb 0 0 0 0>}}
  }
  PilgrimTrackUI_5 {
   widgetClassPath=&PilgrimTrackUI
   parent=5LaneHbox
   widget=PilgrimTrackUI_5
   properties {slateTransformMode=Translation}
   slot {class=HorizontalBoxSlot, properties {horizontalAlignment=HAlign_Left, size {sizeRule=Fill, value=1}}}
  }
  VerticalBox_1 {
   widgetClassPath=&VerticalBox
   parent=PilgrimTrackUI_5
   widget=VerticalBox_1
   slot {class=OverlaySlot, properties {horizontalAlignment=HAlign_Center}}
  }
  Overlay_14 {widgetClassPath=&Overlay, parent=VerticalBox_1, widget=Overlay_14, slot {class=VerticalBoxSlot}}
  CommonVisibilityWidgetLegacy_5 {
   widgetClassPath=&CommonVisibilityWidgetLegacy
   parent=Overlay_14
   widget=CommonVisibilityWidgetLegacy_5
   properties {bShowForTouch=false, padding=<ltrb 0 0 0 0>}
   slot {class=OverlaySlot, properties {horizontalAlignment=HAlign_Fill, verticalAlignment=VAlign_Fill}}
  }
  KeybindWidget_PilgrimLane2_5Lane {
   widgetClassPath=&KeybindWidget
   bIsVariable=true
   parent=CommonVisibilityWidgetLegacy_5
   widget=KeybindWidget_PilgrimLane2_5Lane
   properties {
    bUseKeyboardTextureForKeys=true
    boundAction=PilgrimLane2
    boundInputAction=&IA_PilgrimLane2Press_5Lane
    visibility=Collapsed
    foregroundColor {colorUseRule=UseColor_Specified, specifiedColor=<rgba 0.0609 0 0.6979 1>}
    renderTransform {angle=0, scale=<0.85 0.85>, shear=<0 0>, translation=<0 0>}
   }
   slot {class=BorderSlot, properties {padding=<ltrb 0 0 0 0>}}
  }
  CommonVisibilityWidgetLegacy_13 {
   widgetClassPath=&CommonVisibilityWidgetLegacy
   parent=Overlay_14
   widget=CommonVisibilityWidgetLegacy_13
   properties {
    bShowForErebus=false
    bShowForGamepad=false
    bShowForMac=false
    bShowForMouseAndKeyboard=false
    bShowForPS4=false
    bShowForPS5=false
    bShowForSwitch2=false
    bShowForXBox=false
    bShowForXSX=false
    padding=<ltrb 0 0 0 0>
   }
   slot {class=OverlaySlot, properties {horizontalAlignment=HAlign_Fill, verticalAlignment=VAlign_Fill}}
  }
  TouchIcon_Lane2_5Lane {
   widgetClassPath=&PilgrimMobileBrush
   bIsVariable=true
   parent=CommonVisibilityWidgetLegacy_13
   widget=TouchIcon_Lane2_5Lane
   properties {boundInputAction=&IA_PilgrimLane2Press_5Lane}
   slot {class=BorderSlot, properties {padding=<ltrb 0 0 0 0>}}
  }
  PilgrimTrackUI_6 {
   widgetClassPath=&PilgrimTrackUI
   parent=5LaneHbox
   widget=PilgrimTrackUI_6
   properties {slateTransformMode=Translation}
   slot {class=HorizontalBoxSlot, properties {horizontalAlignment=HAlign_Left, size {sizeRule=Fill, value=1}}}
  }
  VerticalBox_2 {
   widgetClassPath=&VerticalBox
   parent=PilgrimTrackUI_6
   widget=VerticalBox_2
   slot {class=OverlaySlot, properties {horizontalAlignment=HAlign_Center}}
  }
  Overlay_15 {widgetClassPath=&Overlay, parent=VerticalBox_2, widget=Overlay_15, slot {class=VerticalBoxSlot}}
  CommonVisibilityWidgetLegacy_6 {
   widgetClassPath=&CommonVisibilityWidgetLegacy
   parent=Overlay_15
   widget=CommonVisibilityWidgetLegacy_6
   properties {bShowForTouch=false, padding=<ltrb 0 0 0 0>}
   slot {class=OverlaySlot, properties {horizontalAlignment=HAlign_Fill, verticalAlignment=VAlign_Fill}}
  }
  KeybindWidget_PilgrimLane3_5Lane {
   widgetClassPath=&KeybindWidget
   bIsVariable=true
   parent=CommonVisibilityWidgetLegacy_6
   widget=KeybindWidget_PilgrimLane3_5Lane
   properties {
    bUseKeyboardTextureForKeys=true
    boundAction=PilgrimLane3
    boundInputAction=&IA_PilgrimLane3Press_5Lane
    foregroundColor {colorUseRule=UseColor_Specified, specifiedColor=<rgba 0.0609 0 0.6979 1>}
    renderTransform {angle=0, scale=<0.85 0.85>, shear=<0 0>, translation=<0 0>}
   }
   slot {class=BorderSlot, properties {padding=<ltrb 0 0 0 0>}}
  }
  CommonVisibilityWidgetLegacy_14 {
   widgetClassPath=&CommonVisibilityWidgetLegacy
   parent=Overlay_15
   widget=CommonVisibilityWidgetLegacy_14
   properties {
    bShowForErebus=false
    bShowForGamepad=false
    bShowForMac=false
    bShowForMouseAndKeyboard=false
    bShowForPS4=false
    bShowForPS5=false
    bShowForSwitch2=false
    bShowForXBox=false
    bShowForXSX=false
    padding=<ltrb 0 0 0 0>
   }
   slot {class=OverlaySlot, properties {horizontalAlignment=HAlign_Fill, verticalAlignment=VAlign_Fill}}
  }
  TouchIcon_Lane3_5Lane {
   widgetClassPath=&PilgrimMobileBrush
   bIsVariable=true
   parent=CommonVisibilityWidgetLegacy_14
   widget=TouchIcon_Lane3_5Lane
   properties {boundInputAction=&IA_PilgrimLane3Press_5Lane}
   slot {class=BorderSlot, properties {padding=<ltrb 0 0 0 0>}}
  }
  PilgrimTrackUI_7 {
   widgetClassPath=&PilgrimTrackUI
   parent=5LaneHbox
   widget=PilgrimTrackUI_7
   properties {slateTransformMode=Translation}
   slot {class=HorizontalBoxSlot, properties {horizontalAlignment=HAlign_Left, size {sizeRule=Fill, value=1}}}
  }
  VerticalBox_3 {
   widgetClassPath=&VerticalBox
   parent=PilgrimTrackUI_7
   widget=VerticalBox_3
   slot {class=OverlaySlot, properties {horizontalAlignment=HAlign_Center}}
  }
  Overlay_16 {widgetClassPath=&Overlay, parent=VerticalBox_3, widget=Overlay_16, slot {class=VerticalBoxSlot}}
  CommonVisibilityWidgetLegacy_7 {
   widgetClassPath=&CommonVisibilityWidgetLegacy
   parent=Overlay_16
   widget=CommonVisibilityWidgetLegacy_7
   properties {bShowForTouch=false, padding=<ltrb 0 0 0 0>}
   slot {class=OverlaySlot, properties {horizontalAlignment=HAlign_Fill, verticalAlignment=VAlign_Fill}}
  }
  KeybindWidget_PilgrimLane4_5Lane {
   widgetClassPath=&KeybindWidget
   bIsVariable=true
   parent=CommonVisibilityWidgetLegacy_7
   widget=KeybindWidget_PilgrimLane4_5Lane
   properties {
    bUseKeyboardTextureForKeys=true
    boundAction=PilgrimLane4
    boundInputAction=&IA_PilgrimLane4Press_5Lane
    foregroundColor {colorUseRule=UseColor_Foreground, specifiedColor=<rgba 0.0609 0 0.6979 1>}
    renderTransform {angle=0, scale=<0.85 0.85>, shear=<0 0>, translation=<0 0>}
   }
   slot {class=BorderSlot, properties {padding=<ltrb 0 0 0 0>}}
  }
  CommonVisibilityWidgetLegacy_15 {
   widgetClassPath=&CommonVisibilityWidgetLegacy
   parent=Overlay_16
   widget=CommonVisibilityWidgetLegacy_15
   properties {
    bShowForErebus=false
    bShowForGamepad=false
    bShowForMac=false
    bShowForMouseAndKeyboard=false
    bShowForPS4=false
    bShowForPS5=false
    bShowForSwitch2=false
    bShowForXBox=false
    bShowForXSX=false
    padding=<ltrb 0 0 0 0>
   }
   slot {class=OverlaySlot, properties {horizontalAlignment=HAlign_Fill, verticalAlignment=VAlign_Fill}}
  }
  TouchIcon_Lane4_5Lane {
   widgetClassPath=&PilgrimMobileBrush
   bIsVariable=true
   parent=CommonVisibilityWidgetLegacy_15
   widget=TouchIcon_Lane4_5Lane
   properties {boundInputAction=&IA_PilgrimLane4Press_5Lane}
   slot {class=BorderSlot, properties {padding=<ltrb 0 0 0 0>}}
  }
  PilgrimTrackUI_8 {
   widgetClassPath=&PilgrimTrackUI
   parent=5LaneHbox
   widget=PilgrimTrackUI_8
   properties {slateTransformMode=Translation}
   slot {class=HorizontalBoxSlot}
  }
  VerticalBox_4 {
   widgetClassPath=&VerticalBox
   parent=PilgrimTrackUI_8
   widget=VerticalBox_4
   slot {class=OverlaySlot, properties {horizontalAlignment=HAlign_Center}}
  }
  VerticalBox {
   widgetClassPath=&VerticalBox
   parent=VerticalBox_4
   widget=VerticalBox
   slot {class=VerticalBoxSlot, properties {horizontalAlignment=HAlign_Center, verticalAlignment=VAlign_Top}}
  }
  Overlay_17 {widgetClassPath=&Overlay, parent=VerticalBox, widget=Overlay_17, slot {class=VerticalBoxSlot}}
  CommonVisibilityWidgetLegacy_8 {
   widgetClassPath=&CommonVisibilityWidgetLegacy
   parent=Overlay_17
   widget=CommonVisibilityWidgetLegacy_8
   properties {bShowForTouch=false, padding=<ltrb 0 0 0 0>}
   slot {class=OverlaySlot, properties {horizontalAlignment=HAlign_Fill, verticalAlignment=VAlign_Fill}}
  }
  KeybindWidget_PilgrimLane5_5Lane {
   widgetClassPath=&KeybindWidget
   bIsVariable=true
   parent=CommonVisibilityWidgetLegacy_8
   widget=KeybindWidget_PilgrimLane5_5Lane
   properties {
    bUseKeyboardTextureForKeys=true
    boundAction=PilgrimLane5
    boundInputAction=&IA_PilgrimLane5Press_5Lane
    visibility=Collapsed
    foregroundColor {colorUseRule=UseColor_Foreground, specifiedColor=<rgba 0.0609 0 0.6979 1>}
    renderTransform {angle=0, scale=<0.85 0.85>, shear=<0 0>, translation=<0 0>}
   }
   slot {class=BorderSlot, properties {padding=<ltrb 0 0 0 0>}}
  }
  CommonVisibilityWidgetLegacy_16 {
   widgetClassPath=&CommonVisibilityWidgetLegacy
   parent=Overlay_17
   widget=CommonVisibilityWidgetLegacy_16
   properties {
    bShowForErebus=false
    bShowForGamepad=false
    bShowForMac=false
    bShowForMouseAndKeyboard=false
    bShowForPS4=false
    bShowForPS5=false
    bShowForSwitch2=false
    bShowForXBox=false
    bShowForXSX=false
    padding=<ltrb 0 0 0 0>
   }
   slot {class=OverlaySlot, properties {horizontalAlignment=HAlign_Fill, verticalAlignment=VAlign_Fill}}
  }
  TouchIcon_Lane5_5Lane {
   widgetClassPath=&PilgrimMobileBrush
   bIsVariable=true
   parent=CommonVisibilityWidgetLegacy_16
   widget=TouchIcon_Lane5_5Lane
   properties {boundInputAction=&IA_PilgrimLane5Press_5Lane}
   slot {class=BorderSlot, properties {padding=<ltrb 0 0 0 0>}}
  }
  5Lane_RightSpacing {
   widgetClassPath=&Spacer
   parent=HorizontalBox_2
   widget=5Lane_RightSpacing
   properties {size=<132 1>}
   slot {class=HorizontalBoxSlot}
  }
 ]
}
