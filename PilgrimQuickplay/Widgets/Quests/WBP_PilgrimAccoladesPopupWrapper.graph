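# Pilgrim HUD wrapper widget that hosts the XP popup (HUD_XP_Pop_Widget) in an overlay and removes it for LMP sessions.
# The CheckAndRemoveIfLMP macro removes the popup when IsHamSandwichSessionActive is true.
# The check runs on Construct and again on each PlayerAdded playspace event, heard through the player controller's async-acquired event router.
# Destruct stops the stored listener.
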WidgetBlueprint /PilgrimQuickplay/Widgets/Quests/WBP_PilgrimAccoladesPopupWrapper
graphs [
 CheckAndRemoveIfLMP {nodes [
   Inputs()
   > Branch(Condition=IsHamSandwichSessionActive!() -> (ReturnValue:bool))
   n1 = Get HUD_XP_Pop_Widget() -> (HUD_XP_Pop_Widget:HUD XP Pop Widget Object Reference)
   > Is Valid(InputObject=@n1)
   > RemoveFromParent(self=@n1)
   > Outputs!(then:Exec)
  ]}
 EventGraph {nodes [
   Event Construct()
   > Parent: Construct()
   > n1 = Sequence()
   > Check and Remove if LMP!(execute:Exec) -> (then:Exec) {is=title K2Node_MacroInstance}
   n2 = GetPlayerController!(PlayerIndex:int=0) -> (ReturnValue:Player Controller Object Reference)
   Is Valid(exec=@n1.then_1, InputObject=@n2)
   > n3 = PilgrimWaitForEventRouter!(execute:Exec, ContextActor:Actor Object Reference=@n2) -> (then:Exec, OnEventRouterAvailable:Exec, EventRouter:Gameplay Event Router Component Object Reference) {is=title K2Node_AsyncAction}
   n4 = StartListeningToEvent!(
    execute:Exec=@n3.OnEventRouterAvailable
    Target:Gameplay Event Router Component Object Reference=@n3
    EventType:Script Struct Object Reference=/Script/PilgrimCoreRuntime.PilgrimEvent_Playspace_PlayerAdded
    Context:Object Reference
    EventBubblingRule:EEventBubblingRule Enum
   ) -> (
    then:Exec
    AsyncTaskProxy:Async Action Start Listening To Event Object Reference
    OnEventReceived:Exec
    Payload_PlayerInfo:Pilgrim Player Info Structure
   )
   > Set PlayerAddedEventRef(PlayerAddedEventRef:Async Action Start Listening To Event Object Reference=@n4) -> (Output_Get:Async Action Start Listening To Event Object Reference)
   Check and Remove if LMP(execute=@n4.OnEventReceived)
   Event Destruct()
   > StopListeningToEvent(self:Async Action Start Listening To Event Object Reference=Get PlayerAddedEventRef() -> (PlayerAddedEventRef:Async Action Start Listening To Event Object Reference))
   > Parent: Destruct()
  ]}
]
variables [PlayerAddedEventRef {type=AsyncAction_StartListeningToEvent, default=null}]
widgets {
 info {inheritedWidgetCount=0, parentClass=/Script/PilgrimCoreUI.PilgrimGameWidgetBase, rootWidgetClass=&Overlay, widgetCount=2}
 widgets [
  Overlay_0 {widgetClassPath=&Overlay, widget=Overlay_0}
  HUD_XP_Pop_Widget {widgetClassPath=/XpEverywhereUI/HUD_XP_Pop_Widget, bIsVariable=true, parent=Overlay_0, widget=HUD_XP_Pop_Widget, slot {class=OverlaySlot}}
 ]
}
