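# Pilgrim Quickplay HUD widget showing a circular song-break countdown prompt with layered material backgrounds, glows and timer text.
# PlayFadeAnimation plays A_Show forward or in reverse when the requested show state changes, starting from a transparent container set on Construct.
# When BarPulse starts, a bound event loops BeatPulse endlessly at BeatAnimPlaybackSpeedMultiplier; nothing here plays BarPulse or updates the text.
# BarsToVisibility returns Hidden at zero bars remaining, otherwise SelfHitTestInvisible.
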WidgetBlueprint /PilgrimQuickplay/Widgets/WBP_PilgrimQuickplay_BreakPrompt
animations [
 A_Show {
  binding_count=2
  length_seconds=0.2
  object=&self:A_Show
  sequence {bindings [
    Overlay_FullContainer {
     id=#1
     display_name=Overlay_0
     possessed_class=Overlay
     tracks [
      {
       type=MovieScene2DTransformTrack
       sections [{
         type=MovieScene2DTransformSection
         end_frame=4
         start_frame=0
         channels [
          Translation.X {type=MovieSceneScriptingFloatChannel, const [0,3]}
          Translation.Y {type=MovieSceneScriptingFloatChannel, keys [[0,30],[2,-15],[4,0]]}
          Angle {type=MovieSceneScriptingFloatChannel}
          Scale.X {type=MovieSceneScriptingFloatChannel, keys [[0,0.25],[4,1]]}
          Scale.Y {type=MovieSceneScriptingFloatChannel, keys [[0,0.25],[4,1]]}
          Shear.X {type=MovieSceneScriptingFloatChannel}
          Shear.Y {type=MovieSceneScriptingFloatChannel}
         ]
        }]
      }
      {
       type=MovieSceneFloatTrack
       display_name=Render Opacity
       sections [{type=MovieSceneFloatSection, end_frame=2, start_frame=0, channels [None {type=MovieSceneScriptingFloatChannel, keys [[0,0],[2,1]]}]}]
      }
     ]
    }
    TimerText {
     id=#2
     possessed_class=CommonTextBlock
     tracks [{
       type=MovieSceneFloatTrack
       display_name=Render Opacity
       sections [{type=MovieSceneFloatSection, end_frame=4, start_frame=0, channels [None {type=MovieSceneScriptingFloatChannel, keys [[0,0],[4,1]]}]}]
      }]
    }
   ]}
 }
 BarPulse {
  binding_count=4
  length_seconds=1
  object=&self:BarPulse
  sequence {bindings [
    Pulse_Glow {
     id=#3
     possessed_class=Image
     tracks [
      {
       type=MovieSceneFloatTrack
       display_name=Render Opacity
       sections [{type=MovieSceneFloatSection, channels [None {type=MovieSceneScriptingFloatChannel, keys [[0,0],[1,1],[10,0]]}]}]
      }
      {
       type=MovieScene2DTransformTrack
       sections [{
         type=MovieScene2DTransformSection
         channels [
          Scale.X {type=MovieSceneScriptingFloatChannel, keys [[-0.0003,1.25],[9.9997,3]]}
          Scale.Y {type=MovieSceneScriptingFloatChannel, keys [[-0.0003,1.25],[9.9997,3]]}
         ]
        }]
      }
      {
       type=MovieSceneWidgetMaterialTrack
       sections [{
         type=MovieSceneParameterSection
         end_frame=10
         start_frame=-1
         channels [TaperGlow {type=MovieSceneScriptingFloatChannel, keys [[2,1],[10,3]]}]
        }]
      }
      {
       type=MovieSceneByteTrack
       display_name=Visibility
       sections [{type=MovieSceneByteSection, end_frame=10, start_frame=-1, channels [None {type=MovieSceneScriptingByteChannel, keys [[-0.0003,4],[10,1]]}]}]
      }
     ]
    }
    Pulse_Glow_1 {
     id=#4
     possessed_class=Image
     tracks [
      {
       type=MovieSceneByteTrack
       display_name=Visibility
       sections [{type=MovieSceneByteSection, end_frame=20, start_frame=-1, channels [None {type=MovieSceneScriptingByteChannel, keys [[-0.0003,4],[20,2]]}]}]
      }
      {
       type=MovieSceneFloatTrack
       display_name=Render Opacity
       sections [{type=MovieSceneFloatSection, channels [None {type=MovieSceneScriptingFloatChannel, keys [[0,0],[1,1],[4,1],[20,0]]}]}]
      }
     ]
    }
    bg_2 {
     id=#5
     possessed_class=Image
     tracks [{
       type=MovieSceneWidgetMaterialTrack
       sections [{
         type=MovieSceneParameterSection
         end_frame=20
         start_frame=-1
         channels [Animation {type=MovieSceneScriptingFloatChannel, keys [[0,0],[2,1],[15,0]]}]
        }]
      }]
    }
    TimerText {
     id=#6
     possessed_class=CommonTextBlock
     tracks [
      {
       type=MovieScene2DTransformTrack
       sections [{
         type=MovieScene2DTransformSection
         channels [
          Scale.X {type=MovieSceneScriptingFloatChannel, keys [[0,1],[1,1.1],[5,1]]}
          Scale.Y {type=MovieSceneScriptingFloatChannel, keys [[0,1],[1,1.1],[5,1]]}
         ]
        }]
      }
      {
       type=MovieSceneWidgetMaterialTrack
       display_name=Font.FontMaterial
       sections [{type=MovieSceneParameterSection, channels [Anim_Color {type=MovieSceneScriptingFloatChannel, keys [[0,0],[1,1],[4,1],[15,0]]}]}]
      }
     ]
    }
   ]}
 }
 BeatPulse {
  binding_count=4
  length_seconds=1
  object=&self:BeatPulse
  sequence {
   bindings [
    bg_2 {
     id=#7
     possessed_class=Image
     tracks [{
       type=MovieSceneWidgetMaterialTrack
       sections [{
         type=MovieSceneParameterSection
         end_frame=60
         start_frame=0
         channels [Animation {type=MovieSceneScriptingFloatChannel, keys [[0,0],[3,0.5],[60,0]]}]
        }]
      }]
    }
    bg_3 {
     id=#8
     possessed_class=Image
     tracks [{
       type=MovieScene2DTransformTrack
       sections [{
         type=MovieScene2DTransformSection
         channels [
          Scale.X {type=MovieSceneScriptingFloatChannel, keys [[0,1],[3,1.125],[60,1]]}
          Scale.Y {type=MovieSceneScriptingFloatChannel, keys [[0,1],[3,1.125],[60,1]]}
         ]
        }]
      }]
    }
    bg_1 {
     id=#9
     possessed_class=Image
     tracks [{
       type=MovieSceneWidgetMaterialTrack
       sections [{
         type=MovieSceneParameterSection
         channels [
          Outline ColorR {type=MovieSceneScriptingFloatChannel, keys [[0,0.0101],[2,0.04],[60,0.0101]]}
          Outline ColorG {type=MovieSceneScriptingFloatChannel, const [0.0057,1]}
          Outline ColorB {type=MovieSceneScriptingFloatChannel, keys [[0,0.0437],[2,0.2],[60,0.0437]]}
          Outline ColorA {type=MovieSceneScriptingFloatChannel, const [1,1]}
          Middle ColorR {type=MovieSceneScriptingFloatChannel, keys [[0,0.0044],[2,0.0352],[60,0.0044]]}
          Middle ColorG {type=MovieSceneScriptingFloatChannel, const [0,2]}
          Middle ColorB {type=MovieSceneScriptingFloatChannel, keys [[0,0.0365],[2,0.2917],[60,0.0365]]}
          Middle ColorA {type=MovieSceneScriptingFloatChannel, const [1,1]}
         ]
        }]
      }]
    }
    bg {
     id=#10
     possessed_class=Image
     tracks [
      {
       type=MovieScene2DTransformTrack
       sections [{
         type=MovieScene2DTransformSection
         channels [
          Scale.X {type=MovieSceneScriptingFloatChannel, keys [[0,1],[2,1.125],[60,1]]}
          Scale.Y {type=MovieSceneScriptingFloatChannel, keys [[0,1],[3,1.125],[60,1]]}
         ]
        }]
      }
      {
       type=MovieSceneWidgetMaterialTrack
       sections [{
         type=MovieSceneParameterSection
         channels [
          Alpha_Final {type=MovieSceneScriptingFloatChannel, keys [[0,0.75],[2,1],[60,0.75]]}
          ColorR {type=MovieSceneScriptingFloatChannel, keys [[0,0.0008],[2,0.02],[60,0.0008]]}
          ColorG {type=MovieSceneScriptingFloatChannel, const [0,3]}
          ColorB {type=MovieSceneScriptingFloatChannel, keys [[0,0.005],[2,0.1],[60,0.005]]}
          ColorA {type=MovieSceneScriptingFloatChannel, const [1,1]}
         ]
        }]
      }
     ]
    }
   ]
   display_rate {numerator=60}
  }
 }
]
graphs [
 BarsToVisibility {nodes [
   n1 = BarsToVisibility!() -> (then, BarsRemaining:int) {is=title K2Node_FunctionEntry}
   > Return Node!(execute, NewParam:ESlateVisibility Enum=Select(Option 0=SelfHitTestInvisible, Option 1=Hidden, Index="Equal (Integer)"(A=@n1)))
  ]}
 EventGraph {nodes [
   n1 = BeatLoopAnim() {is=title K2Node_CustomEvent}
   > PlayAnimation(
    InAnimation=Get BeatPulse() -> (BeatPulse:Widget Animation Object Reference)
    NumLoopsToPlay=0
    PlaybackSpeed=Get BeatAnimPlaybackSpeedMultiplier() -> (BeatAnimPlaybackSpeedMultiplier:double)
   )
   Event Construct()
   > SetRenderOpacity(self=Get Overlay_FullContainer() -> (Overlay_FullContainer:Overlay Object Reference), InOpacity=0)
   > BindToAnimationEvent!(
    execute
    self:User Widget Object Reference
    Animation:Widget Animation Object Reference=Get BarPulse() -> (BarPulse:Widget Animation Object Reference)
    Delegate:Delegate=@n1
    AnimationEvent:EWidgetAnimationEvent Enum=Started
    UserTag:Name=None
   ) -> (then)
  ]}
 PlayFadeAnimation {nodes [
   n1 = PlayFadeAnimation!() -> (then, CanShowBreak:bool=~) {is=title K2Node_FunctionEntry}
   > Branch(Condition="Not Equal (Boolean)"(A=@n1, B=Get IsShowing() -> (IsShowing:bool)) -> (ReturnValue=~))
   > n2 = Set IsShowing(IsShowing:bool=@n1) -> (Output_Get:bool)
   > n3 = Branch(Condition=@n2)
   n4 = Get A_Show() -> (A_Show:Widget Animation Object Reference)
   > PlayAnimationForward(InAnimation=@n4)
   PlayAnimationReverse(execute=@n3.else, InAnimation=@n4)
  ]}
]
variables [
 BeatAnimPlaybackSpeedMultiplier {type=number}
 IsShowing {type=boolean}
]
widgets {
 info {inheritedWidgetCount=0, parentClass=/Script/PilgrimCoreUI.PilgrimGameWidgetBase, rootWidgetClass=&Overlay, widgetCount=10}
 widgets [
  Overlay_FullContainer {widgetClassPath=&Overlay, bIsVariable=true, widget=Overlay_FullContainer}
  Shadow {
   widgetClassPath=&Image
   parent=Overlay_FullContainer
   widget=Shadow
   properties {
    visibility=SelfHitTestInvisible
    brush {
     bIsDynamicallyLoaded=false
     drawAs=Image
     imageType=NoImage
     mirroring=NoMirror
     resourceName=None
     resourceObject=/PilgrimCore/Widgets/Materials/M_UI_Pilgrim_RoundedShadow_Round
     tiling=NoTile
     imageSize=<139 139>
     margin=<ltrb 0 0 0 0>
     outlineSettings {
      bUseBrushTransparency=false
      roundingType=HalfHeightRadius
      width=0
      color {colorUseRule=UseColor_Specified, specifiedColor=<rgba 0 0 0 0>}
      cornerRadii=<0 0 0 0>
     }
     tintColor {colorUseRule=UseColor_Specified, specifiedColor=<rgba 1 1 1 1>}
     uVRegion {bIsValid=false, max=<0 0>, min=<0 0>}
    }
    renderTransform {angle=0, scale=<1.75 1.75>, shear=<0 0>, translation=<0 0>}
   }
   slot {class=OverlaySlot, properties {horizontalAlignment=HAlign_Center, verticalAlignment=VAlign_Center}}
  }
  Spacer_55 {widgetClassPath=&Spacer, parent=Overlay_FullContainer, widget=Spacer_55, properties {size=<80 80>}, slot {class=OverlaySlot}}
  Pulse_Glow_1 {
   widgetClassPath=&Image
   parent=Overlay_FullContainer
   widget=Pulse_Glow_1
   properties {
    visibility=Collapsed
    brush {
     bIsDynamicallyLoaded=false
     drawAs=Image
     imageType=NoImage
     mirroring=NoMirror
     resourceName=None
     resourceObject=/PilgrimCore/Widgets/Materials/MI_UI_Piligrim_GemBreak_Pulse
     tiling=NoTile
     imageSize=<220 220>
     margin=<ltrb 0 0 0 0>
     outlineSettings {
      bUseBrushTransparency=false
      roundingType=HalfHeightRadius
      width=0
      color {colorUseRule=UseColor_Specified, specifiedColor=<rgba 0 0 0 0>}
      cornerRadii=<0 0 0 0>
     }
     tintColor {colorUseRule=UseColor_Specified, specifiedColor=<rgba 1 1 1 1>}
     uVRegion {bIsValid=false, max=<0 0>, min=<0 0>}
    }
    renderTransform {angle=0, scale=<1.25 1.25>, shear=<0 0>, translation=<0 0>}
   }
   slot {class=OverlaySlot, properties {horizontalAlignment=HAlign_Center, verticalAlignment=VAlign_Center}}
  }
  Pulse_Glow {
   widgetClassPath=&Image
   parent=Overlay_FullContainer
   widget=Pulse_Glow
   properties {
    visibility=Collapsed
    brush {
     bIsDynamicallyLoaded=false
     drawAs=Image
     imageType=FullColor
     mirroring=NoMirror
     resourceName=None
     resourceObject=&MI_UI_Piligrim_GemBreak_Pulse
     tiling=NoTile
     imageSize=<82 82>
     margin=<ltrb 0 0 0 0>
     outlineSettings {
      bUseBrushTransparency=false
      roundingType=HalfHeightRadius
      width=0
      color {colorUseRule=UseColor_Specified, specifiedColor=<rgba 0 0 0 0>}
      cornerRadii=<0 0 0 0>
     }
     tintColor {colorUseRule=UseColor_Specified, specifiedColor=<rgba 1 1 1 1>}
     uVRegion {bIsValid=false, max=<0 0>, min=<0 0>}
    }
    renderTransform {angle=0, scale=<1.2 1.2>, shear=<0 0>, translation=<0 0>}
   }
   slot {class=OverlaySlot, properties {horizontalAlignment=HAlign_Fill, verticalAlignment=VAlign_Fill}}
  }
  bg {
   widgetClassPath=&Image
   parent=Overlay_FullContainer
   widget=bg
   properties {
    pixelSnapping=Disabled
    visibility=HitTestInvisible
    brush {
     bIsDynamicallyLoaded=false
     drawAs=Image
     imageType=NoImage
     mirroring=NoMirror
     resourceName=None
     resourceObject=/PilgrimCore/Widgets/Materials/M_UI_Pilgrim_HUD_TimerBacking
     tiling=NoTile
     imageSize=<192 192>
     margin=<ltrb 0 0 0 0>
     outlineSettings {
      bUseBrushTransparency=false
      roundingType=HalfHeightRadius
      width=0
      color {colorUseRule=UseColor_Specified, specifiedColor=<rgba 0 0 0 0>}
      cornerRadii=<0 0 0 0>
     }
     tintColor {colorUseRule=UseColor_Specified, specifiedColor=<rgba 1 1 1 1>}
     uVRegion {bIsValid=false, max=<0 0>, min=<0 0>}
    }
   }
   slot {class=OverlaySlot, properties {horizontalAlignment=HAlign_Center, verticalAlignment=VAlign_Center}}
  }
  bg_3 {
   widgetClassPath=&Image
   parent=Overlay_FullContainer
   widget=bg_3
   properties {
    pixelSnapping=Disabled
    visibility=HitTestInvisible
    brush {
     bIsDynamicallyLoaded=false
     drawAs=Image
     imageType=NoImage
     mirroring=NoMirror
     resourceName=None
     resourceObject=&M_UI_Pilgrim_HUD_TimerBacking
     tiling=NoTile
     imageSize=<207 207>
     margin=<ltrb 0 0 0 0>
     outlineSettings {
      bUseBrushTransparency=false
      roundingType=HalfHeightRadius
      width=0
      color {colorUseRule=UseColor_Specified, specifiedColor=<rgba 0 0 0 0>}
      cornerRadii=<0 0 0 0>
     }
     tintColor {colorUseRule=UseColor_Specified, specifiedColor=<rgba 1 1 1 1>}
     uVRegion {bIsValid=false, max=<0 0>, min=<0 0>}
    }
   }
   slot {class=OverlaySlot, properties {horizontalAlignment=HAlign_Center, verticalAlignment=VAlign_Center}}
  }
  bg_1 {
   widgetClassPath=&Image
   parent=Overlay_FullContainer
   widget=bg_1
   properties {
    visibility=HitTestInvisible
    brush {
     bIsDynamicallyLoaded=false
     drawAs=Image
     imageType=NoImage
     mirroring=NoMirror
     resourceName=None
     resourceObject=/PilgrimCore/Widgets/Materials/MI_UI_Pilgrim_MultiplierNumberBG
     tiling=NoTile
     imageSize=<192 192>
     margin=<ltrb 0 0 0 0>
     outlineSettings {
      bUseBrushTransparency=false
      roundingType=HalfHeightRadius
      width=0
      color {colorUseRule=UseColor_Specified, specifiedColor=<rgba 0 0 0 0>}
      cornerRadii=<0 0 0 0>
     }
     tintColor {colorUseRule=UseColor_Specified, specifiedColor=<rgba 1 1 1 1>}
     uVRegion {bIsValid=false, max=<0 0>, min=<0 0>}
    }
   }
   slot {class=OverlaySlot, properties {horizontalAlignment=HAlign_Center, verticalAlignment=VAlign_Center, padding=<ltrb 2 2 2 2>}}
  }
  bg_2 {
   widgetClassPath=&Image
   parent=Overlay_FullContainer
   widget=bg_2
   properties {
    visibility=HitTestInvisible
    brush {
     bIsDynamicallyLoaded=false
     drawAs=Image
     imageType=FullColor
     mirroring=NoMirror
     resourceName=None
     resourceObject=/PilgrimCore/Widgets/Materials/MI_UI_Pilgrim_CenterGlow_SongBreakTimer
     tiling=NoTile
     imageSize=<192 192>
     margin=<ltrb 0 0 0 0>
     outlineSettings {
      bUseBrushTransparency=false
      roundingType=HalfHeightRadius
      width=0
      color {colorUseRule=UseColor_Specified, specifiedColor=<rgba 0 0 0 0>}
      cornerRadii=<0 0 0 0>
     }
     tintColor {colorUseRule=UseColor_Specified, specifiedColor=<rgba 1 1 1 1>}
     uVRegion {bIsValid=false, max=<0 0>, min=<0 0>}
    }
   }
   slot {class=OverlaySlot, properties {horizontalAlignment=HAlign_Center, verticalAlignment=VAlign_Center, padding=<ltrb 2 2 2 2>}}
  }
  TimerText {
   widgetClassPath=&CommonTextBlock
   bIsVariable=true
   parent=Overlay_FullContainer
   widget=TimerText
   properties {
    justification=Center
    mobileFontSizeMultiplier=2
    pixelSnapping=Disabled
    text=10
    font {
     bForceMonospaced=false
     bMaterialIsStencil=false
     fontMaterial=/PilgrimCore/Widgets/Materials/MI_UI_Piligrim_HUDTextColor
     fontObject=&HeadingNow
     letterSpacing=0
     monospacedWidth=1
     size=25.5
     skewAmount=0
     typefaceFontName=76 Bold Italic
     outlineSettings {
      bApplyOutlineToDropShadows=false
      bMiteredCorners=false
      bSeparateFillAlpha=true
      outlineMaterial=null
      outlineSize=0
      outlineColor=<rgba 0 0 0 1>
     }
    }
    renderTransform {angle=0, scale=<1 1>, shear=<0 0>, translation=<-1 0>}
    shadowOffset=<0 0>
   }
   slot {class=OverlaySlot, properties {horizontalAlignment=HAlign_Center, verticalAlignment=VAlign_Center, padding=<ltrb 0 0 0 1>}}
  }
 ]
}
