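# Pilgrim Quickplay notice widget that displays a "Controller Disconnected" prompt.
# A GridPanel layers a centered row over a dark material backing, with a rounded shadow bleeding past the edges through negative padding.
# The row pairs an Attention-type UIKit status indicator, set over an enlarged StreakMeter rings FX material, with a CommonTextBlock message.
# It is layout only, with an empty event graph, so it does no disconnect detection or visibility handling itself.
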WidgetBlueprint /PilgrimQuickplay/Widgets/WBP_PilgrimQuickplay_ControllerDisconnectPrompt
graphs [EventGraph {links [], nodes []}]
variables []
widgets {
 info {inheritedWidgetCount=0, parentClass=&UserWidget, rootWidgetClass=&GridPanel, widgetCount=9}
 widgets [
  Notice {widgetClassPath=&GridPanel, widget=Notice, properties {columnFill [1], rowFill [1]}}
  BackingElements {widgetClassPath=&Overlay, parent=Notice, widget=BackingElements, slot {class=GridSlot, properties {columnSpan=1, rowSpan=1}}}
  Shadow {
   widgetClassPath=&Image
   parent=BackingElements
   widget=Shadow
   properties {
    visibility=SelfHitTestInvisible
    brush {
     bIsDynamicallyLoaded=false
     drawAs=Image
     imageType=NoImage
     mirroring=NoMirror
     resourceName=None
     resourceObject=/PilgrimCore/Widgets/Materials/M_UI_Pilgrim_RoundedShadow_Box
     tiling=NoTile
     imageSize=<32 32>
     margin=<ltrb 0 0 0 0>
     outlineSettings {
      bUseBrushTransparency=false
      roundingType=HalfHeightRadius
      width=0
      color {colorUseRule=UseColor_Specified, specifiedColor=<rgba 0 0 0 0>}
      cornerRadii=<0 0 0 0>
     }
     tintColor {colorUseRule=UseColor_Specified, specifiedColor=<rgba 1 1 1 1>}
     uVRegion {bIsValid=false, max=<0 0>, min=<0 0>}
    }
   }
   slot {class=OverlaySlot, properties {horizontalAlignment=HAlign_Fill, verticalAlignment=VAlign_Fill, padding=<ltrb -16 -16 -16 -16>}}
  }
  Backing {
   widgetClassPath=&Image
   parent=BackingElements
   widget=Backing
   properties {
    visibility=SelfHitTestInvisible
    brush {
     bIsDynamicallyLoaded=false
     drawAs=Image
     imageType=FullColor
     mirroring=NoMirror
     resourceName=None
     resourceObject=/PilgrimCore/Widgets/Materials/M_UI_Pilgrim_HUD_SimpleBacking_Darker
     tiling=NoTile
     imageSize=<32 32>
     margin=<ltrb 0 0 0 0>
     outlineSettings {
      bUseBrushTransparency=false
      roundingType=HalfHeightRadius
      width=0
      color {colorUseRule=UseColor_Specified, specifiedColor=<rgba 0 0 0 0>}
      cornerRadii=<0 0 0 0>
     }
     tintColor {colorUseRule=UseColor_Specified, specifiedColor=<rgba 1 1 1 1>}
     uVRegion {bIsValid=false, max=<0 0>, min=<0 0>}
    }
   }
   slot {class=OverlaySlot, properties {horizontalAlignment=HAlign_Fill, verticalAlignment=VAlign_Fill}}
  }
  Contents {
   widgetClassPath=&HorizontalBox
   parent=Notice
   widget=Contents
   slot {class=GridSlot, properties {columnSpan=1, horizontalAlignment=HAlign_Center, layer=1, rowSpan=1, padding=<ltrb 20 20 20 20>}}
  }
  Overlay_4 {
   widgetClassPath=&Overlay
   parent=Contents
   widget=Overlay_4
   slot {class=HorizontalBoxSlot, properties {verticalAlignment=VAlign_Center}}
  }
  Image_1 {
   widgetClassPath=&Image
   parent=Overlay_4
   widget=Image_1
   properties {
    visibility=SelfHitTestInvisible
    brush {
     bIsDynamicallyLoaded=false
     drawAs=Image
     imageType=NoImage
     mirroring=NoMirror
     resourceName=None
     resourceObject=/PilgrimCore/Widgets/Materials/M_UI_StreakMeter_FX_Rings
     tiling=NoTile
     imageSize=<40 40>
     margin=<ltrb 0 0 0 0>
     outlineSettings {
      bUseBrushTransparency=false
      roundingType=HalfHeightRadius
      width=0
      color {colorUseRule=UseColor_Specified, specifiedColor=<rgba 0 0 0 0>}
      cornerRadii=<0 0 0 0>
     }
     tintColor {colorUseRule=UseColor_Specified, specifiedColor=<rgba 1 1 1 1>}
     uVRegion {bIsValid=false, max=<0 0>, min=<0 0>}
    }
    renderTransform {angle=0, scale=<2.5 2.5>, shear=<0 0>, translation=<0 0>}
   }
   slot {class=OverlaySlot}
  }
  StatusIndicator {
   widgetClassPath=/FortUILibrary/UIKit/StatusIndicator/WBP_UIKit_StatusIndicator
   parent=Overlay_4
   widget=StatusIndicator
   properties {type=Attention, overrideStatusIndicatorVisibility=true}
   slot {class=OverlaySlot, properties {horizontalAlignment=HAlign_Center, verticalAlignment=VAlign_Center}}
  }
  Text_NoticeBody {
   widgetClassPath=&CommonTextBlock
   bIsVariable=true
   parent=Contents
   widget=Text_NoticeBody
   properties {
    scrollStyle=&TextScrollStyle-Base_C
    text=Controller Disconnected
    font {
     bForceMonospaced=false
     bMaterialIsStencil=false
     fontMaterial=null
     fontObject=&HeadingNow
     letterSpacing=0
     monospacedWidth=1
     size=13.5
     skewAmount=0
     typefaceFontName=74 Regular
     outlineSettings {
      bApplyOutlineToDropShadows=false
      bMiteredCorners=false
      bSeparateFillAlpha=false
      outlineMaterial=null
      outlineSize=0
      outlineColor=<rgba 0 0 0 1>
     }
    }
    shadowOffset=<0 0>
    strikeBrush {
     bIsDynamicallyLoaded=false
     drawAs=NoDrawType
     imageType=NoImage
     mirroring=NoMirror
     resourceName=None
     resourceObject=null
     tiling=NoTile
     imageSize=<32 32>
     margin=<ltrb 0 0 0 0>
     outlineSettings {
      bUseBrushTransparency=false
      roundingType=HalfHeightRadius
      width=0
      color {colorUseRule=UseColor_Specified, specifiedColor=<rgba 0 0 0 0>}
      cornerRadii=<0 0 0 0>
     }
     tintColor {colorUseRule=UseColor_Specified, specifiedColor=<rgba 1 1 1 1>}
     uVRegion {bIsValid=false, max=<0 0>, min=<0 0>}
    }
   }
   slot {class=HorizontalBoxSlot, properties {verticalAlignment=VAlign_Center, padding=<ltrb 8 0 0 0>, size {sizeRule=Fill, value=1}}}
  }
 ]
}
