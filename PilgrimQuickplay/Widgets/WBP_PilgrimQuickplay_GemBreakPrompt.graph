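# Pilgrim Quickplay HUD widget that shows an emote keybind prompt during gem breaks.
# On construct it starts transparent, subscribes to tag-driven visibility, and labels itself "Quick Emote" in HamSandwich sessions or "Emote" otherwise.
# Tick polls the gem break view model during breaks, skipping HamSandwich sessions without song breaks enabled.
# Changes in the countdown flag play one opacity animation forward to show or reversed to hide.
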WidgetBlueprint /PilgrimQuickplay/Widgets/WBP_PilgrimQuickplay_GemBreakPrompt
animations [A_ShowEmotePropmt {
  binding_count=1
  length_seconds=0.25
  object=&self:A_ShowEmotePropmt
  sequence {bindings [WBP_PilgrimQuickplay_GemBreakPrompt_C_4 {
     id=#1
     possessed_class=WBP_PilgrimQuickplay_GemBreakPrompt_C
     tracks [{
       type=MovieSceneFloatTrack
       display_name=Render Opacity
       sections [{type=MovieSceneFloatSection, channels [None {type=MovieSceneScriptingFloatChannel, keys [[0,0],[5,1]]}]}]
      }]
    }]}
 }]
graphs [EventGraph {nodes [
   Event Construct()
   > SetRenderOpacity(InOpacity=0)
   > SubscribeForChangingVisibilityByTag!(
    execute
    self:Pilgrim Game Widget Base Object Reference
    GameplayTag:"Gameplay Tag Structure (by ref)"=MakeLiteralGameplayTag!(Value:Gameplay Tag Structure=(TagName="Pilgrim.Widget.GemBreakPrompt")) -> (ReturnValue:Gameplay Tag Structure)
   ) -> (then)
   > "SetText (Text)"(
    self=Get PromptText() -> (PromptText:Common Text Object Reference)
    InText=SelectText(
     A=Get PromptTextHamSandwich() -> (PromptTextHamSandwich:Text)
     B=Get PromptTextDefault() -> (PromptTextDefault:Text)
     bPickA=IsHamSandwichSessionActive!() -> (ReturnValue:bool)
    )
   )
   Event Tick()
   > Can Execute Cosmetic Events!(execute:Exec) -> (True:Exec, False:Exec) {is=title K2Node_MacroInstance}
   > Branch(Condition=Get bIsInGemBreak!(self:Pilgrim Gem Break View Model Object Reference=Get PilgrimGemBreakVM() -> (PilgrimGemBreakVM:Pilgrim Gem Break View Model Object Reference)) -> (bIsInGemBreak:bool))
   > n1 = Branch(Condition=IsHamSandwichSessionActive())
   > n2 = Branch(Condition=AreHamSandwichSongBreaksEnabled!() -> (ReturnValue:bool))
   n3 = Sequence(execute=[@n2.then, @n1.else])
   Set LastFrameWasCountingDown(
    execute=@n3.then_1
    LastFrameWasCountingDown:bool=Get bIsCountingDown!(self:Pilgrim Gem Break View Model Object Reference=Get PilgrimGemBreakVM()) -> (bIsCountingDown:bool)
   ) -> (Output_Get:bool)
   n4 = Get PilgrimGemBreakVM()
   Branch(
    execute=@n3.then_0
    Condition="Not Equal (Boolean)"(A=Get LastFrameWasCountingDown() -> (LastFrameWasCountingDown:bool), B=Get bIsCountingDown(self=@n4)) -> (ReturnValue=~)
   )
   n5 = Get PilgrimGemBreakVM()
   > "Play Animation with Finished event (legacy)"!(
    execute:Exec
    Widget:User Widget Object Reference=Self-Reference()
    InAnimation:Widget Animation Object Reference=Get A_ShowEmotePropmt() -> (A_ShowEmotePropmt:Widget Animation Object Reference)
    StartAtTime:float=0
    NumLoopsToPlay:int=1
    PlayMode:EUMGSequencePlayMode Enum=Select(Option 0:EUMGSequencePlayMode Enum=Reverse, Option 1:EUMGSequencePlayMode Enum, Index=Get bIsCountingDown(self=@n5)) -> (ReturnValue:EUMGSequencePlayMode Enum)
    PlaybackSpeed:float=1
   ) -> (then:Exec, Result:UMGSequence Player Object Reference, Finished:Exec) {is=title K2Node_PlayAnimation}
  ]}]
variables [
 LastFrameWasCountingDown {type=boolean}
 PromptTextDefault {type=string, default=Emote}
 PromptTextHamSandwich {type=string, default=Quick Emote}
]
widgets {
 info {inheritedWidgetCount=0, parentClass=/Script/PilgrimCoreUI.PilgrimGameWidgetBase, rootWidgetClass=&Overlay, widgetCount=6}
 widgets [
  Overlay_44 {widgetClassPath=&Overlay, widget=Overlay_44}
  UIKit_Backplate {
   widgetClassPath=&WBP_UIKit_Backplate
   parent=Overlay_44
   widget=UIKit_Backplate
   properties {brightness=Medium}
   slot {class=OverlaySlot, properties {horizontalAlignment=HAlign_Fill, verticalAlignment=VAlign_Fill}}
  }
  HorizontalBox_0 {
   widgetClassPath=&HorizontalBox
   bIsVariable=true
   parent=Overlay_44
   widget=HorizontalBox_0
   slot {class=OverlaySlot, properties {horizontalAlignment=HAlign_Center, verticalAlignment=VAlign_Center, padding=<ltrb 20 10 20 10>}}
  }
  KeybindWidget_Emote {
   widgetClassPath=/FortUILibrary/Input/KeybindWidget
   bIsVariable=true
   parent=HorizontalBox_0
   widget=KeybindWidget_Emote
   properties {
    bUseKeyboardTextureForKeys=true
    boundAction=EmotePicker
    boundInputAction=/PilgrimCore/Input/InputActions/IA_PilgrimEmote
    boundActionFallbacks [GamepadEmotePickerOrReplayLastEmoteHold]
   }
   slot {class=HorizontalBoxSlot, properties {horizontalAlignment=HAlign_Center, verticalAlignment=VAlign_Center}}
  }
  Spacer_143 {widgetClassPath=&Spacer, parent=HorizontalBox_0, widget=Spacer_143, properties {size=<8 1>}, slot {class=HorizontalBoxSlot}}
  PromptText {
   widgetClassPath=&CommonTextBlock
   bIsVariable=true
   parent=HorizontalBox_0
   widget=PromptText
   properties {
    text=Emote
    textTransformPolicy=ToUpper
    font {
     bForceMonospaced=false
     bMaterialIsStencil=false
     fontMaterial=null
     fontObject=&HeadingNow
     letterSpacing=-1
     monospacedWidth=1
     size=21
     skewAmount=0
     typefaceFontName=76 Bold
     outlineSettings {
      bApplyOutlineToDropShadows=false
      bMiteredCorners=false
      bSeparateFillAlpha=false
      outlineMaterial=&M-FontShadow
      outlineSize=2
      outlineColor=<rgba 0 0 0 1>
     }
    }
    shadowOffset=<0 0>
   }
   slot {class=HorizontalBoxSlot}
  }
 ]
}
