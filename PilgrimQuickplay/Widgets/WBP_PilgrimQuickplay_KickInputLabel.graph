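# Pilgrim HUD label showing the Kick keybind for drum-track players not using a drum peripheral.
# RefreshVisibility also requires the pro-track button setting and no gem break, rerunning on battle init, setting-change events and AllowEmoteDuringGemBreak tag changes.
# Battle finish collapses it, while input-method changes set visibility from peripheral type alone.
# Both event-router listeners share one task variable, so Destruct stops only the last stored.
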WidgetBlueprint /PilgrimQuickplay/Widgets/WBP_PilgrimQuickplay_KickInputLabel
graphs [
 EventGraph {nodes [
   n1 = On Input Method Changed!() -> (OutputDelegate, then, bNewInputType:ECommonInputType Enum) {is=title K2Node_CustomEvent}
   n2 = Get OwningPlayer() -> (OwningPlayer:Player Controller Object Reference)
   Is Valid(exec=[@n13.then, @n1.then], InputObject=@n2)
   n3 = GetControllerFromActor!(Actor:Actor Object Reference=@n2) -> (ReturnValue:Controller Object Reference)
   > Is Valid(InputObject=@n3)
   n4 = GetComponentByClass!(self:Actor Object Reference=@n3, ComponentClass:Actor Component Class Reference=/Script/PilgrimCoreRuntime.PilgrimPlayerControllerComponent) -> (ReturnValue:Pilgrim Player Controller Component Object Reference)
   > Is Valid(InputObject=@n4)
   n5 = GetCurrentDevicePeripheralType!(self:Pilgrim Player Controller Component Object Reference=@n4) -> (ReturnValue:EFMPeripheralInputMode Enum=None)
   > SetVisibility(
    self=Get KickPromptUp() -> (KickPromptUp:Border Object Reference)
    InVisibility=Select(
     Option 0=Collapsed
     Option 1=HitTestInvisible
     Index="Not Equal (Enum)"!(A:EFMPeripheralInputMode Enum=@n5, B:EFMPeripheralInputMode Enum=Drums) -> (ReturnValue:bool)
    )
   )
   Event Destruct()
   > n6 = Sequence()
   > StopListeningToEvent(self:Async Action Start Listening To Event Object Reference=Get Pro Track Async Task Ref() -> (Pro Track Async Task Ref:Async Action Start Listening To Event Object Reference))
   n7 = For Each Loop(
    Exec=@n6.then_1
    Array:Array of Cancellable Async Action Object References=Get AsyncEventHandles() -> (AsyncEventHandles:Array of Cancellable Async Action Object References)
   ) -> (Array Element:Cancellable Async Action Object Reference)
   Clear(execute=@n7.Completed, TargetArray:Array of Cancellable Async Action Object References=Get AsyncEventHandles())
   Cancel(execute=@n7.LoopBody, self:Cancellable Async Action Object Reference=@n7)
   Event Construct()
   > Set OwningPlayer(OwningPlayer:Player Controller Object Reference=GetOwningPlayer()) -> (Output_Get:Player Controller Object Reference)
   > n8 = Sequence!(execute) -> (then_0, then_1, then_2:Exec, then_3:Exec)
   > n9 = PilgrimWaitForEventRouter!(execute:Exec, ContextActor:Actor Object Reference=Get OwningPlayer()) -> (then:Exec, OnEventRouterAvailable:Exec, EventRouter:Gameplay Event Router Component Object Reference) {is=title K2Node_AsyncAction}
   n10 = Sequence!(execute=@n9.OnEventRouterAvailable) -> (then_0, then_1)
   > n11 = StartListeningToEvent!(
    execute:Exec
    Target:Gameplay Event Router Component Object Reference=@n9
    EventType:Script Struct Object Reference=/Script/PilgrimSettingsRuntime.PilgrimEvent_PilgrimSettings_ProTrackInputButtonVisibilityChanged
    Context:Object Reference
    EventBubblingRule:EEventBubblingRule Enum
   ) -> (
    then:Exec
    AsyncTaskProxy:Async Action Start Listening To Event Object Reference
    OnEventReceived:Exec
    Payload:Pilgrim Event Pilgrim Settings Pro Track Input Button Visibility Changed Structure
   )
   > Set Pro Track Async Task Ref(Pro Track Async Task Ref:Async Action Start Listening To Event Object Reference=@n11) -> (Output_Get:Async Action Start Listening To Event Object Reference)
   RefreshVisibility(execute=@n11.OnEventReceived)
   n12 = StartListeningToEvent(execute=@n10.then_1, Target=@n9, EventType=/Script/PilgrimCoreRuntime.PilgrimEvent_MBI_Finished) -> (Payload:Pilgrim Event MBI Finished Structure)
   > Set Pro Track Async Task Ref(Pro Track Async Task Ref=@n12)
   SetVisibility(execute=@n12.OnEventReceived, self=Get KickPromptUp(), InVisibility=Collapsed)
   n13 = Bind Event to On Input Method Changed(execute=@n8.then_3, self=Get CommonInputSubsystem(), Delegate=@n1)
   n14 = WaitGameplayTagRemoveFromActor!(
    execute:Exec=@n8.then_2
    TargetActor:Actor Object Reference=Get OwningPlayer()
    Tag:Gameplay Tag Structure=(TagName="Pilgrim.Player.AllowEmoteDuringGemBreak")
    OnlyTriggerOnce:bool=false
   ) -> (then:Exec, AsyncTaskProxy:Ability Async Wait Gameplay Tag Removed Object Reference, Removed:Exec) {is=title K2Node_AsyncAction}
   > Add(
    TargetArray:Array of Cancellable Async Action Object References=Get AsyncEventHandles()
    NewItem:"Cancellable Async Action Object Reference (by ref)"=@n14
   ) {is=title K2Node_CallArrayFunction}
   Set isGemBreak(execute=@n14.Removed, isGemBreak:bool=false) -> (Output_Get:bool)
   > RefreshVisibility()
   n15 = WaitGameplayTagAddToActor!(
    execute:Exec=@n8.then_1
    TargetActor:Actor Object Reference=Get OwningPlayer()
    Tag:Gameplay Tag Structure=(TagName="Pilgrim.Player.AllowEmoteDuringGemBreak")
    OnlyTriggerOnce:bool=false
   ) -> (then:Exec, AsyncTaskProxy:Ability Async Wait Gameplay Tag Added Object Reference, Added:Exec) {is=title K2Node_AsyncAction}
   > Add(TargetArray=Get AsyncEventHandles(), NewItem=@n15)
   Set isGemBreak(execute=@n15.Added, isGemBreak=true)
   > RefreshVisibility()
   Event OnMusicBattleInstanceInitializeEvent() {is=title K2Node_Event}
   > RefreshVisibility()
  ]}
 RefreshVisibility {nodes [
   RefreshVisibility() {is=title K2Node_FunctionEntry}
   > n1 = GetPilgrimCorePlayerStateComponent!(
    execute
    PlayerState:Player State Object Reference=Get PlayerState!(self:Controller Object Reference=Get OwningPlayer()) -> (PlayerState:Player State Object Reference)
   ) -> (then, ReturnValue:Pilgrim Core Player State Component Object Reference)
   > Is Valid(InputObject=@n1)
   n2 = IsDrumPeripheralTrack!(TrackType:EPilgrimTrackType Enum=GetTrackTypeForGameplay!(self:Pilgrim Core Player State Component Object Reference=@n1) -> (ReturnValue:EPilgrimTrackType Enum=TrackGuitar)) -> (ReturnValue:bool)
   n3 = GetComponentByClass(self=GetControllerFromActor(Actor=Get OwningPlayer()), ComponentClass=&PilgrimPlayerControllerComponent)
   n4 = AND Boolean(
    A=GetProTrackInputButtonVisibility!(self:Pilgrim Client Setting Record Partition Object Reference=GetPilgrimClientSettingRecordPartition!(self:Pilgrim Player Controller Component Object Reference=@n3) -> (ReturnValue:Pilgrim Client Setting Record Partition Object Reference)) -> (ReturnValue:bool)
    B=NOT Boolean(A=Get isGemBreak() -> (isGemBreak:bool))
   )
   n5 = "Not Equal (Enum)"(A=GetCurrentDevicePeripheralType(self=@n3) -> (ReturnValue=None), B=Drums)
   > SetVisibility(
    self=Get KickPromptUp()
    InVisibility=Select(Option 0=Collapsed, Option 1=HitTestInvisible, Index=AND Boolean!(A=@n2, B=@n5, C:bool=@n4) -> (ReturnValue))
   )
  ]}
]
variables [
 AsyncEventHandles {type="array<CancellableAsyncAction>", default []}
 isGemBreak {type=boolean}
 OwningPlayer {type=PlayerController, default=null}
 Pro Track Async Task Ref {type=AsyncAction_StartListeningToEvent, default=null}
 ProTrackInputVisibility {type=boolean}
]
widgets {
 info {inheritedWidgetCount=0, parentClass=/Script/PilgrimCoreUI.PilgrimGameWidgetBase, rootWidgetClass=&Border, widgetCount=12}
 widgets [
  KickPromptUp {
   widgetClassPath=&Border
   bIsVariable=true
   widget=KickPromptUp
   properties {
    verticalAlignment=VAlign_Center
    visibility=HitTestInvisible
    background {
     bIsDynamicallyLoaded=false
     drawAs=NoDrawType
     imageType=NoImage
     mirroring=NoMirror
     resourceName=None
     resourceObject=null
     tiling=NoTile
     imageSize=<32 32>
     margin=<ltrb 0 0 0 0>
     outlineSettings {
      bUseBrushTransparency=false
      roundingType=HalfHeightRadius
      width=0
      color {colorUseRule=UseColor_Specified, specifiedColor=<rgba 0 0 0 0>}
      cornerRadii=<0 0 0 0>
     }
     tintColor {colorUseRule=UseColor_Specified, specifiedColor=<rgba 1 1 1 1>}
     uVRegion {bIsValid=false, max=<0 0>, min=<0 0>}
    }
    brushColor=<rgba 1 1 1 0.25>
    padding=<ltrb 0 0 0 0>
   }
  }
  KickPrompts {
   widgetClassPath=&HorizontalBox
   bIsVariable=true
   parent=KickPromptUp
   widget=KickPrompts
   properties {visibility=HitTestInvisible}
   slot {class=BorderSlot, properties {verticalAlignment=VAlign_Center, padding=<ltrb 0 0 0 0>}}
  }
  CommonVisibilityWidgetLegacy {
   widgetClassPath=&CommonVisibilityWidgetLegacy
   parent=KickPrompts
   widget=CommonVisibilityWidgetLegacy
   properties {bShowForTouch=false, padding=<ltrb 0 0 4 0>}
   slot {class=HorizontalBoxSlot}
  }
  KeybindWidget_PilgrimKick_Overlay {
   widgetClassPath=&Overlay
   bIsVariable=true
   parent=CommonVisibilityWidgetLegacy
   widget=KeybindWidget_PilgrimKick_Overlay
   slot {class=BorderSlot, properties {padding=<ltrb 0 0 4 0>}}
  }
  HorizontalBox_0 {
   widgetClassPath=&HorizontalBox
   parent=KeybindWidget_PilgrimKick_Overlay
   widget=HorizontalBox_0
   slot {class=OverlaySlot, properties {horizontalAlignment=HAlign_Right, verticalAlignment=VAlign_Fill}}
  }
  PilgrimTrackUI_0 {
   widgetClassPath=/Script/PilgrimCoreUI.PilgrimTrackUI
   parent=HorizontalBox_0
   widget=PilgrimTrackUI_0
   properties {slateTransformMode=Translation}
   slot {class=HorizontalBoxSlot, properties {horizontalAlignment=HAlign_Right}}
  }
  KeybindWidget_PilgrimKick {
   widgetClassPath=/FortUILibrary/Input/KeybindWidget
   bIsVariable=true
   parent=PilgrimTrackUI_0
   widget=KeybindWidget_PilgrimKick
   properties {
    bShowUnboundStatus=true
    bUseKeyboardTextureForKeys=true
    boundInputAction=/PilgrimSettings/Input/InputActions/IA_PilgrimLane5Press_5Lane
    foregroundColor {colorUseRule=UseColor_Specified, specifiedColor=<rgba 0.0609 0 0.6979 1>}
    renderTransform {angle=0, scale=<0.8 0.8>, shear=<0 0>, translation=<0 0>}
   }
   slot {class=OverlaySlot}
  }
  SizeBox {
   widgetClassPath=&SizeBox
   parent=KickPrompts
   widget=SizeBox
   properties {bOverride_MaxDesiredWidth=true, maxDesiredWidth=142}
   slot {class=HorizontalBoxSlot, properties {horizontalAlignment=HAlign_Left, verticalAlignment=VAlign_Center}}
  }
  ScaleBox {widgetClassPath=&ScaleBox, parent=SizeBox, widget=ScaleBox, properties {stretchDirection=DownOnly}, slot {class=SizeBoxSlot}}
  Overlay_108 {
   widgetClassPath=&Overlay
   parent=ScaleBox
   widget=Overlay_108
   slot {class=ScaleBoxSlot, properties {horizontalAlignment=HAlign_Left}}
  }
  Image {
   widgetClassPath=&Image
   bIsVariable=true
   parent=Overlay_108
   widget=Image
   properties {
    renderOpacity=0.5
    visibility=SelfHitTestInvisible
    brush {
     bIsDynamicallyLoaded=false
     drawAs=Image
     imageType=NoImage
     mirroring=NoMirror
     resourceName=None
     resourceObject=/PilgrimCore/Widgets/Materials/MI_UI_Pilgrim_ScoreGlow
     tiling=NoTile
     imageSize=<32 32>
     margin=<ltrb 0 0 0 0>
     outlineSettings {
      bUseBrushTransparency=false
      roundingType=HalfHeightRadius
      width=0
      color {colorUseRule=UseColor_Specified, specifiedColor=<rgba 0 0 0 0>}
      cornerRadii=<0 0 0 0>
     }
     tintColor {colorUseRule=UseColor_Specified, specifiedColor=<rgba 1 1 1 1>}
     uVRegion {bIsValid=false, max=<0 0>, min=<0 0>}
    }
    renderTransform {angle=0, scale=<1 1>, shear=<-8 0>, translation=<0 0>}
   }
   slot {class=OverlaySlot, properties {horizontalAlignment=HAlign_Fill, verticalAlignment=VAlign_Fill, padding=<ltrb -12 -8 -12 -8>}}
  }
  Strum {
   widgetClassPath=&UEFN_TextBlock
   bIsVariable=true
   parent=Overlay_108
   widget=Strum
   properties {
    lineHeightPercentage=0.95
    text=Kick
    textTransformPolicy=ToUpper
    visibility=SelfHitTestInvisible
    colorAndOpacity {colorUseRule=UseColor_Specified, specifiedColor=<rgba 1 1 1 0.9>}
    font {
     bForceMonospaced=false
     bMaterialIsStencil=false
     fontMaterial=null
     fontObject=&HeadingNow
     letterSpacing=0
     monospacedWidth=1
     size=13.5
     skewAmount=0
     typefaceFontName=86 Bold Italic
     outlineSettings {
      bApplyOutlineToDropShadows=false
      bMiteredCorners=false
      bSeparateFillAlpha=false
      outlineMaterial=&M-FontShadow
      outlineSize=2
      outlineColor=<rgba 0.667 0.334 1 0.75>
     }
    }
   }
   slot {class=OverlaySlot, properties {verticalAlignment=VAlign_Center}}
  }
 ]
}
