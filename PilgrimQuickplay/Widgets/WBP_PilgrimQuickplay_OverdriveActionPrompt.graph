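# Pilgrim Quickplay HUD widget prompting the player to deploy Overdrive with a keybind icon and glowing label.
# On music battle initialization, it detects guitar-peripheral tracks and binds the deploy input for five lanes or the default.
# Overdrive data, input method changes and settings events from the playspace's event router refresh visibility.
# It shows only for non-touch input with Overdrive charged but inactive, gated by a console variable and track-type-specific setting.
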WidgetBlueprint /PilgrimQuickplay/Widgets/WBP_PilgrimQuickplay_OverdriveActionPrompt
graphs [
 EventGraph {nodes [
   n1 = Handle Input Method Changed!() -> (OutputDelegate, then, bNewInputType:ECommonInputType Enum) {is=title K2Node_CustomEvent}
   > SetOverdrivePromptVisibility!(
    execute
    self
    IsOverdriveActive:bool=Get IsOverdriveActive() -> (IsOverdriveActive:bool)
    CurrentOverdrive:double=Get CurrentOverdrive() -> (CurrentOverdrive:double)
   ) -> (then)
   Event OnInitialized()
   > Parent: OnInitialized() {is=title K2Node_CallParentFunction}
   > n2 = Sequence()
   n3 = Get CommonInputSubsystem()
   > Bind Event to On Input Method Changed(self=@n3, Delegate=@n1)
   > Handle Input Method Changed!(execute, self, bNewInputType:ECommonInputType Enum=GetCurrentInputType(self=@n3)) -> (then) {is=title K2Node_CallFunction}
   n4 = WaitForDiscoverablePlayspaceOfTypeForActor!(
    execute:Exec=@n2.then_1
    PlayspaceClass:Playspace Soft Class Reference=/Script/PilgrimQuickplayRuntime.PilgrimQuickplayPlayspace
    Actor:Actor Object Reference=GetOwningPlayer()
    Requester:Object Reference
   ) -> (
    then:Exec
    AsyncTaskProxy:Wait For Discoverable Playspace With Filter Action Object Reference
    OnFoundPlayspace:Exec
    Playspace:Playspace Object Reference
   ) {is=title K2Node_AsyncAction}
   n5 = PilgrimWaitForEventRouter!(execute:Exec=@n4.OnFoundPlayspace, ContextActor:Actor Object Reference=GetOwningPlayer()) -> (then:Exec, OnEventRouterAvailable:Exec, EventRouter:Gameplay Event Router Component Object Reference) {is=title K2Node_AsyncAction}
   n6 = Sequence(execute=@n5.OnEventRouterAvailable)
   > n7 = StartListeningToEvent!(
    execute:Exec
    Target:Gameplay Event Router Component Object Reference=@n5
    EventType:Script Struct Object Reference=/Script/PilgrimSettingsRuntime.PilgrimEvent_PilgrimSettings_NonProTrackInputButtonVisibilityChanged
    Context:Object Reference
    EventBubblingRule:EEventBubblingRule Enum
   ) -> (
    then:Exec
    AsyncTaskProxy:Async Action Start Listening To Event Object Reference
    OnEventReceived:Exec
    Payload:Pilgrim Event Pilgrim Settings Non Pro Track Input Button Visibility Changed Structure
   )
   n8 = Branch(execute=@n7.OnEventReceived, Condition=NOT Boolean(A=Get IsUsingPeripheral() -> (IsUsingPeripheral:bool)))
   SetOverdrivePromptVisibility(execute=[@n10.then, @n8.then], IsOverdriveActive=Get IsOverdriveActive(), CurrentOverdrive=Get CurrentOverdrive())
   n9 = StartListeningToEvent(execute=@n6.then_1, Target=@n5, EventType=/Script/PilgrimSettingsRuntime.PilgrimEvent_PilgrimSettings_ProTrackInputButtonVisibilityChanged) -> (Payload:Pilgrim Event Pilgrim Settings Pro Track Input Button Visibility Changed Structure)
   n10 = Branch(execute=@n9.OnEventReceived, Condition=Get IsUsingPeripheral())
   Event Construct()
   > n11 = Branch(Condition=Get isMBIInitialized() -> (isMBIInitialized:bool))
   SetVisibility(execute=@n11.else, InVisibility=Hidden)
   n12 = Event OnOverdriveDataReceivedLocalPlayer!() -> (OutputDelegate, then, GameProgress:"Pilgrim Game Progress Structure (by ref)") {is=title K2Node_Event}
   n13 = Break Pilgrim Game Progress!(PilgrimGameProgress:"Pilgrim Game Progress Structure (by ref)"=@n12.GameProgress) -> (
    NumStreakBreakingMisses:int
    bHasPerfectCombo:bool=~
    RawCurrentScore:float=~
    IsOverdriveActive:bool=~
    CurrentOverdrive:float=~
    RawMaxScore:float=~
    MaxStreakLength:int
    AverageStreakMultiplierPerHitOrPass:float=~
    CurrentSectionCompletionPercentage:float=~
   ) {is=title K2Node_BreakStruct}
   > Set IsOverdriveActive(IsOverdriveActive:bool=@n13.IsOverdriveActive) -> (Output_Get:bool)
   > Set CurrentOverdrive(CurrentOverdrive:double=@n13.CurrentOverdrive) -> (Output_Get:double)
   > SetOverdrivePromptVisibility(IsOverdriveActive=@n13.IsOverdriveActive, CurrentOverdrive=@n13.CurrentOverdrive)
   Event OnMusicBattleInstanceInitializeEvent() {is=title K2Node_Event}
   > Parent: OnMusicBattleInstanceInitializeEvent() {is=title K2Node_CallParentFunction}
   > n14 = GetPilgrimCorePlayerStateComponent!(
    execute
    PlayerState:Player State Object Reference=Get PlayerState!(self:Controller Object Reference=GetOwningPlayer()) -> (PlayerState:Player State Object Reference)
   ) -> (then, ReturnValue:Pilgrim Core Player State Component Object Reference)
   > n15 = Is Valid(InputObject=@n14)
   Set isMBIInitialized(execute=[@n15.Is Not Valid, @n17.then], isMBIInitialized:bool=true) -> (Output_Get:bool)
   > SetVisibility(
    self=[
     Get OverdrivePrompts() -> (OverdrivePrompts:Horizontal Box Object Reference)
     Get KeybindWidget_PilgrimOverdrive_Overlay() -> (KeybindWidget_PilgrimOverdrive_Overlay:Overlay Object Reference)
    ]
    InVisibility=Collapsed
   )
   > n16 = Is Valid(InputObject=GetLocalPilgrimGameForPlayer!(
     self:Pilgrim Music Battle Instance Object Reference=Get MusicBattleInstance() -> (MusicBattleInstance:Pilgrim Music Battle Instance Object Reference)
     InPC:Player Controller Object Reference=GetOwningPlayer()
    ) -> (ReturnValue:Pilgrim Game Object Reference))
   FortShippingLog!(
    execute=@n16.Is Not Valid
    WorldContextObject:Object Reference=GetOwningPlayer()
    InString:String="Can't find Pilgrim Game Instance (Overdrive Meter Icons)"
    bLogAsWarning:bool=false
   ) -> (then)
   SetupForLanes(execute=@n16.Is Valid)
   > SetVisibility(InVisibility=SelfHitTestInvisible)
   n17 = Set IsUsingPeripheral(
    execute=@n15.Is Valid
    IsUsingPeripheral:bool=IsGuitarPeripheralTrack!(TrackType:EPilgrimTrackType Enum=GetTrackTypeForGameplay!(self:Pilgrim Core Player State Component Object Reference=@n14) -> (ReturnValue:EPilgrimTrackType Enum=TrackGuitar)) -> (ReturnValue:bool)
   ) -> (Output_Get:bool)
  ]}
 Set Action Text {nodes [
   n1 = Set Action Text!() -> (then, InText:Text) {is=title K2Node_FunctionEntry}
   > "SetText (Text)"(self=Get DeployOverdrive() -> (DeployOverdrive:Text Block Object Reference), InText=@n1)
  ]}
 SetOverdrivePromptVisibility {nodes [
   SetOverdrivePromptVisibility!() -> (then, IsOverdriveActive:bool=~, CurrentOverdrive:double=~) {is=title K2Node_FunctionEntry}
   > n1 = GetConsoleVariableBoolValue!(execute, VariableName:String=Pilgrim.ShowOverdrivePrompt) -> (then, ReturnValue:bool)
   n2 = NOT Boolean(A=Get IsOverdriveActive())
   n3 = "Not Equal (Enum)"!(A:ECommonInputType Enum=GetCurrentInputType(self=Get CommonInputSubsystem()), B:ECommonInputType Enum=Touch) -> (ReturnValue:bool)
   n4 = "float >= float"(A=Get CurrentOverdrive(), B=0.01)
   n5 = GetOwningPlayer()
   n6 = GetPilgrimClientSettingRecordPartition!(self:Pilgrim Player Controller Component Object Reference=GetComponentByClass!(
     self:Actor Object Reference=GetControllerFromActor!(Actor:Actor Object Reference=@n5) -> (ReturnValue:Controller Object Reference)
     ComponentClass:Actor Component Class Reference=/Script/PilgrimCoreRuntime.PilgrimPlayerControllerComponent
    ) -> (ReturnValue:Pilgrim Player Controller Component Object Reference)) -> (ReturnValue:Pilgrim Client Setting Record Partition Object Reference)
   n7 = Get IsUsingPeripheral()
   n8 = OR Boolean(
    A=AND Boolean(
     A=Get IsUsingPeripheral()
     B=GetProTrackInputButtonVisibility!(self:Pilgrim Client Setting Record Partition Object Reference=@n6) -> (ReturnValue:bool)
    )
    B=AND Boolean(
     A=NOT Boolean(A=@n7)
     B=GetNonProTrackInputButtonVisibility!(self:Pilgrim Client Setting Record Partition Object Reference=@n6) -> (ReturnValue:bool)
    )
   )
   > SetVisibility(
    self=[Get OverdrivePrompts(), Get KeybindWidget_PilgrimOverdrive_Overlay()]
    InVisibility=Select(Option 0=Collapsed, Option 1=HitTestInvisible, Index=AND Boolean!(A=@n1, B=@n2, C:bool=@n4, D:bool=@n3, E:bool=@n8) -> (ReturnValue))
   )
  ]}
 SetupForLanes {nodes [
   SetupForLanes() {is=title K2Node_FunctionEntry}
   n1 = GetOwningPlayer()
   n2 = Get MusicBattleInstance()
   > n3 = Set NumLanes(NumLanes:int=GetNumLanes!(self:Pilgrim Game Object Reference=GetLocalPilgrimGameForPlayer(self=@n2, InPC=@n1)) -> (ReturnValue:int=0)) -> (Output_Get:int=0)
   n4 = Get KeybindWidget_PilgrimOverdrive() -> (KeybindWidget_PilgrimOverdrive:Keybind Widget Object Reference)
   > SetBoundInputAction!(
    execute
    self:Fort Keybind Widget Object Reference=@n4
    NewAction:Input Action Object Reference=Select(
     Option 0:Input Action Object Reference=/PilgrimSettings/Input/InputActions/IA_PilgrimDeployOverdrive
     Option 1:Input Action Object Reference=/PilgrimSettings/Input/InputActions/IA_PilgrimDeployOverdrive_5Lanes
     Index="Equal (Integer)"(A=@n3, B=5)
    ) -> (ReturnValue:Input Action Object Reference)
   ) -> (then)
   > UpdateKeybindWidget(self:Fort Keybind Widget Object Reference=@n4)
  ]}
]
variables [
 CurrentOverdrive {type=number}
 isMBIInitialized {type=boolean}
 IsOverdriveActive {type=boolean}
 IsUsingPeripheral {type=boolean}
]
widgets {
 info {inheritedWidgetCount=0, parentClass=/Script/PilgrimCoreUI.PilgrimGameWidgetBase, rootWidgetClass=&Border, widgetCount=11}
 widgets [
  OverdrivePromptContainer {
   widgetClassPath=&Border
   bIsVariable=true
   widget=OverdrivePromptContainer
   properties {
    verticalAlignment=VAlign_Center
    visibility=SelfHitTestInvisible
    background {
     bIsDynamicallyLoaded=false
     drawAs=NoDrawType
     imageType=NoImage
     mirroring=NoMirror
     resourceName=None
     resourceObject=null
     tiling=NoTile
     imageSize=<32 32>
     margin=<ltrb 0 0 0 0>
     outlineSettings {
      bUseBrushTransparency=false
      roundingType=HalfHeightRadius
      width=0
      color {colorUseRule=UseColor_Specified, specifiedColor=<rgba 0 0 0 0>}
      cornerRadii=<0 0 0 0>
     }
     tintColor {colorUseRule=UseColor_Specified, specifiedColor=<rgba 1 1 1 1>}
     uVRegion {bIsValid=false, max=<0 0>, min=<0 0>}
    }
    brushColor=<rgba 1 1 1 0.25>
    padding=<ltrb 0 0 0 0>
   }
  }
  OverdrivePrompts {
   widgetClassPath=&HorizontalBox
   bIsVariable=true
   parent=OverdrivePromptContainer
   widget=OverdrivePrompts
   properties {visibility=Collapsed}
   slot {class=BorderSlot, properties {verticalAlignment=VAlign_Center, padding=<ltrb 0 0 0 0>}}
  }
  CommonVisibilityWidgetLegacy_0 {
   widgetClassPath=&CommonVisibilityWidgetLegacy
   parent=OverdrivePrompts
   widget=CommonVisibilityWidgetLegacy_0
   properties {bShowForTouch=false, padding=<ltrb 0 0 4 0>}
   slot {class=HorizontalBoxSlot}
  }
  KeybindWidget_PilgrimOverdrive_Overlay {
   widgetClassPath=&Overlay
   bIsVariable=true
   parent=CommonVisibilityWidgetLegacy_0
   widget=KeybindWidget_PilgrimOverdrive_Overlay
   slot {class=BorderSlot, properties {padding=<ltrb 0 0 4 0>}}
  }
  PilgrimTrackUI_0 {
   widgetClassPath=/Script/PilgrimCoreUI.PilgrimTrackUI
   parent=KeybindWidget_PilgrimOverdrive_Overlay
   widget=PilgrimTrackUI_0
   properties {slateTransformMode=Translation}
   slot {class=OverlaySlot, properties {horizontalAlignment=HAlign_Right, verticalAlignment=VAlign_Fill}}
  }
  KeybindWidget_PilgrimOverdrive {
   widgetClassPath=/FortUILibrary/Input/KeybindWidget
   bIsVariable=true
   parent=PilgrimTrackUI_0
   widget=KeybindWidget_PilgrimOverdrive
   properties {
    bUseKeyboardTextureForKeys=true
    boundAction=ActivateOverdrive
    boundInputAction=&IA_PilgrimDeployOverdrive
    foregroundColor {colorUseRule=UseColor_Specified, specifiedColor=<rgba 0.0609 0 0.6979 1>}
    renderTransform {angle=0, scale=<0.8 0.8>, shear=<0 0>, translation=<0 0>}
   }
   slot {class=OverlaySlot}
  }
  SizeBox_0 {
   widgetClassPath=&SizeBox
   parent=OverdrivePrompts
   widget=SizeBox_0
   properties {bOverride_MaxDesiredWidth=true, maxDesiredWidth=142, widthOverride=142}
   slot {class=HorizontalBoxSlot, properties {horizontalAlignment=HAlign_Left, verticalAlignment=VAlign_Center}}
  }
  ScaleBox_2 {widgetClassPath=&ScaleBox, parent=SizeBox_0, widget=ScaleBox_2, properties {stretchDirection=DownOnly}, slot {class=SizeBoxSlot}}
  Overlay_1 {
   widgetClassPath=&Overlay
   parent=ScaleBox_2
   widget=Overlay_1
   slot {class=ScaleBoxSlot, properties {horizontalAlignment=HAlign_Left}}
  }
  Image_115 {
   widgetClassPath=&Image
   bIsVariable=true
   parent=Overlay_1
   widget=Image_115
   properties {
    renderOpacity=0.5
    visibility=SelfHitTestInvisible
    brush {
     bIsDynamicallyLoaded=false
     drawAs=Image
     imageType=NoImage
     mirroring=NoMirror
     resourceName=None
     resourceObject=/PilgrimCore/Widgets/Materials/MI_UI_Pilgrim_ScoreGlow
     tiling=NoTile
     imageSize=<32 32>
     margin=<ltrb 0 0 0 0>
     outlineSettings {
      bUseBrushTransparency=false
      roundingType=HalfHeightRadius
      width=0
      color {colorUseRule=UseColor_Specified, specifiedColor=<rgba 0 0 0 0>}
      cornerRadii=<0 0 0 0>
     }
     tintColor {colorUseRule=UseColor_Specified, specifiedColor=<rgba 1 1 1 1>}
     uVRegion {bIsValid=false, max=<0 0>, min=<0 0>}
    }
    renderTransform {angle=0, scale=<1 1>, shear=<-8 0>, translation=<0 0>}
   }
   slot {class=OverlaySlot, properties {horizontalAlignment=HAlign_Fill, verticalAlignment=VAlign_Fill, padding=<ltrb -12 -8 -12 -8>}}
  }
  DeployOverdrive {
   widgetClassPath=&UEFN_TextBlock
   bIsVariable=true
   parent=Overlay_1
   widget=DeployOverdrive
   properties {
    lineHeightPercentage=0.95
    text=Overdrive
    textTransformPolicy=ToUpper
    visibility=SelfHitTestInvisible
    colorAndOpacity {colorUseRule=UseColor_Specified, specifiedColor=<rgba 1 1 1 0.9>}
    font {
     bForceMonospaced=false
     bMaterialIsStencil=false
     fontMaterial=null
     fontObject=&HeadingNow
     letterSpacing=0
     monospacedWidth=1
     size=13.5
     skewAmount=0
     typefaceFontName=86 Bold Italic
     outlineSettings {
      bApplyOutlineToDropShadows=false
      bMiteredCorners=false
      bSeparateFillAlpha=false
      outlineMaterial=&M-FontShadow
      outlineSize=2
      outlineColor=<rgba 0.667 0.334 1 0.75>
     }
    }
   }
   slot {class=OverlaySlot, properties {verticalAlignment=VAlign_Center}}
  }
 ]
}
