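# Per-track HUD widget for Pilgrim Quickplay, holding input labels, bandmates, accuracy/streak text, player identity, and overdrive, break and controller-disconnect prompts.
# Nested PilgrimTrackUI layers transform elements with the track, and a gem break panel hosts those that fade with it.
# Layout defaults to four lanes, then follows the game's lane count, with per-widget offsets and mobile padding.
# Event router listeners update expanded-track padding and streak visibility; break-end triggers a delayed identity fade-out.
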WidgetBlueprint /PilgrimQuickplay/Widgets/WBP_PilgrimQuickplay_PerTrackWidgets
animations [
 AfterFadeIn {
  binding_count=1
  length_seconds=6.0185
  object=&self:AfterFadeIn
  sequence {bindings [LMP_IdentityOverlay {
     id=#1
     display_name=WBP_Pilgrim_PlayerIdentity_Overlay
     possessed_class=WBP_Pilgrim_PlayerIdentity_Overlay_C
     tracks [{
       type=MovieSceneFloatTrack
       display_name=Render Opacity
       sections [{
         type=MovieSceneFloatSection
         end_frame=120
         start_frame=0
         channels [None {type=MovieSceneScriptingFloatChannel, keys [[0,1],[111.8613,1],[120,0]]}]
        }]
      }]
    }]}
 }
 ControllerDisconnect {
  binding_count=1
  length_seconds=0.25
  object=&self:ControllerDisconnect
  sequence {bindings [WBP_ControllerDisconnectPrompt {
     id=#2
     possessed_class=WBP_PilgrimQuickplay_ControllerDisconnectPrompt_C
     tracks [
      {
       type=MovieSceneFloatTrack
       display_name=Render Opacity
       sections [{type=MovieSceneFloatSection, end_frame=5, start_frame=0, channels [None {type=MovieSceneScriptingFloatChannel, keys [[0,0],[1,0],[5,1]]}]}]
      }
      {
       type=MovieSceneByteTrack
       display_name=Visibility
       sections [{type=MovieSceneByteSection, end_frame=5, start_frame=0, channels [None {type=MovieSceneScriptingByteChannel, keys [[0,1],[1,4],[5,4]]}]}]
      }
     ]
    }]}
 }
]
graphs [
 Adjust Elements For Local MP {nodes [
   n1 = Adjust Elements For Local MP!() -> (then, Is Local MP View:bool=~) {is=title K2Node_FunctionEntry}
   > n2 = Sequence()
   > SetVisibility(
    self=Get WBP_Bandmates() -> (WBP_Bandmates:WBP Bandmates Object Reference)
    InVisibility=Select(Option 0=SelfHitTestInvisible, Option 1=Collapsed, Index=@n1)
   )
   > SetVisibility(
    self=Get LMP_IdentityOverlay() -> (LMP_IdentityOverlay:WBP Pilgrim Player Identity Overlay Object Reference)
    InVisibility=Select(Option 0=Collapsed, Option 1=SelfHitTestInvisible, Index=@n1)
   )
   > SetVisibility(
    self=Get LMP_Nameplate() -> (LMP_Nameplate:WBP Pilgrim Player Identity Nameplate Object Reference)
    InVisibility=Select(Option 0=Collapsed, Option 1=SelfHitTestInvisible, Index=@n1)
   )
   SetPadding!(
    execute=@n2.then_1
    self:User Widget Object Reference=Get LMP_IdentityOverlay()
    InPadding_Left:float=0
    InPadding_Top:float=0
    InPadding_Right:float=0
    InPadding_Bottom:float=Get LocalMPIdentityVerticalOffset() -> (LocalMPIdentityVerticalOffset:double)
   ) -> (then)
  ]}
 Adjust Elements For Track Width {nodes [
   n1 = Adjust Elements For Track Width!() -> (then, isTrackFourLane:bool=~) {is=title K2Node_FunctionEntry}
   > n2 = For Each Loop(Array:Array of Widget Five Lane Offset Structures=Get WidgetsWithFiveLaneOffsets() -> (WidgetsWithFiveLaneOffsets:Array of Widget Five Lane Offset Structures)) -> (Array Element:Widget Five Lane Offset Structure)
   n3 = Break Widget Five Lane Offset!(WidgetFiveLaneOffset:"Widget Five Lane Offset Structure (by ref)"=@n2) -> (
    Widget_2_#3:Widget Object Reference
    FiveLaneTransformOffset_5_#4:Vector 2D Structure
   ) {is=title K2Node_BreakStruct}
   > SetRenderTranslation(
    self=@n3
    Translation=Select(Option 0:Vector 2D Structure=@n3.FiveLaneTransformOffset_5_#4, Option 1:Vector 2D Structure, Index=@n1) -> (ReturnValue:Vector 2D Structure)
   )
  ]}
 Adjust Elements Padding For Mobile {nodes [
   Adjust Elements Padding For Mobile() {is=title K2Node_FunctionEntry}
   > Branch(Condition=IsMobileGame())
   > n1 = For Each Loop(Array:Array of Widget Mobile Margin Offset Structures=Get WidgetsMobilePaddingOffsets() -> (WidgetsMobilePaddingOffsets:Array of Widget Mobile Margin Offset Structures)) -> (Array Element:Widget Mobile Margin Offset Structure)
   n2 = Break Widget Mobile Margin Offset!(WidgetMobileMarginOffset:"Widget Mobile Margin Offset Structure (by ref)"=@n1) -> (Widget_2_#3:Widget Object Reference, Margin_7_#4:Margin Structure) {is=title K2Node_BreakStruct}
   > SetPadding!(
    execute
    self:Overlay Slot Object Reference=SlotAsOverlaySlot(Widget=@n2)
    InPadding:Margin Structure=@n2.Margin_7_#4
   ) -> (then)
  ]}
 EventGraph {nodes [
   n1 = "On Break Animation Finished (GemBreakPanel_FadeWithTrack)"!() -> (OutputDelegate, then, Name:Name=None) {is=title K2Node_ComponentBoundEvent}
   n2 = Get GemBreakPanel_FadeWithTrack() -> (GemBreakPanel_FadeWithTrack:WBP Pilgrim Gem Break Panel Object Reference)
   > Branch(Condition="Equal (Name)"(A=@n1.Name, B=Get BreakEndAnimation!(self:WBP Pilgrim Gem Break Panel Object Reference=@n2) -> (BreakEndAnimation:Name=None)))
   > Do Once(Reset=@n3.then)
   > PlayAnimationForward(InAnimation=Get AfterFadeIn() -> (AfterFadeIn:Widget Animation Object Reference))
   n3 = ResetAfterFadeInGate() {is=title K2Node_CustomEvent}
   Event Destruct()
   n4 = Get GameplayEventHandles() -> (GameplayEventHandles:Array of Async Action Start Listening To Event Object References)
   > n5 = For Each Loop(Array:Array of Async Action Start Listening To Event Object References=@n4) -> (Array Element:Async Action Start Listening To Event Object Reference)
   Clear(execute=@n5.Completed, TargetArray:Array of Async Action Start Listening To Event Object References=@n4)
   StopListeningToEvent(execute=@n5.LoopBody, self:Async Action Start Listening To Event Object Reference=@n5)
   Event Construct()
   > Adjust Elements Padding For Mobile() {is=title K2Node_CallFunction}
   > ResetAfterFadeInGate() {is=title K2Node_CallFunction}
   > n6 = Is Valid(InputObject=Get EventRouter() -> (EventRouter:Gameplay Event Router Component Object Reference))
   n7 = PilgrimWaitForEventRouter!(execute:Exec=@n6.Is Not Valid, ContextActor:Actor Object Reference=GetOwningPlayer()) -> (then:Exec, OnEventRouterAvailable:Exec, EventRouter:Gameplay Event Router Component Object Reference) {is=title K2Node_AsyncAction}
   n8 = Set EventRouter(execute=@n7.OnEventRouterAvailable, EventRouter:Gameplay Event Router Component Object Reference=@n7) -> (Output_Get:Gameplay Event Router Component Object Reference)
   n9 = StartListeningToEvent!(
    execute:Exec=[@n6.Is Valid, @n8.then]
    Target:Gameplay Event Router Component Object Reference=Get EventRouter()
    EventType:Script Struct Object Reference=/Script/PilgrimSettingsRuntime.PilgrimEvent_PilgrimSettings_EnableMobileExpandedTrackChanged
    Context:Object Reference
    EventBubblingRule:EEventBubblingRule Enum
   ) -> (
    then:Exec
    AsyncTaskProxy:Async Action Start Listening To Event Object Reference
    OnEventReceived:Exec
    Payload:Pilgrim Event Pilgrim Settings Enable Mobile Expanded Track Changed Structure
   )
   > Add(
    TargetArray:Array of Async Action Start Listening To Event Object References=Get GameplayEventHandles()
    NewItem:"Async Action Start Listening To Event Object Reference (by ref)"=@n9
   ) {is=title K2Node_CallArrayFunction}
   > n10 = StartListeningToEvent(
    Target=Get EventRouter()
    EventType=/Script/PilgrimSettingsRuntime.PilgrimEvent_PilgrimSettings_ShowStreakCounterChanged
    Context=GetOwningPlayer()
   ) -> (Payload:Pilgrim Event Pilgrim Settings Show Streak Counter Changed Structure)
   > Add(TargetArray=Get GameplayEventHandles(), NewItem=@n10)
   > DelayUntilNextTick!(execute) -> (then)
   n11 = Get IsStreakCounterSetup() -> (IsStreakCounterSetup:bool)
   n12 = GetShowStreakCounter!(self:Pilgrim Game Widget Base Object Reference) -> (ReturnValue:bool)
   n13 = "Equal (Enum)"(
    A:ESlateVisibility Enum=GetStreakCounterVisibility!(self:WBP Pilgrim Accuracy Text Object Reference=Get WBP_PilgrimAccuracy_Enhanced() -> (WBP_PilgrimAccuracy_Enhanced:WBP Pilgrim Accuracy Text Object Reference)) -> (Visibility:ESlateVisibility Enum=Visible)
    B:ESlateVisibility Enum=Visible
   )
   > Branch(Condition=AND Boolean(A="Not Equal (Boolean)"(A=@n13, B=@n12) -> (ReturnValue=~), B=NOT Boolean(A=@n11)))
   > SetupStreakCounter()
   SetStreakCounterVisibility!(
    execute=@n10.OnEventReceived
    self
    Index:bool=Break Pilgrim Event Pilgrim Settings Show Streak Counter Changed!(PilgrimEvent_PilgrimSettings_ShowStreakCounterChanged:"Pilgrim Event Pilgrim Settings Show Streak Counter Changed Structure (by ref)"=@n10.Payload) -> (bShowStreakCounter:bool=~) {is=title K2Node_BreakStruct}
   ) -> (then)
   SetWidgetPaddingOffsetsMobile(execute=@n9.OnEventReceived)
   > Adjust Elements Padding For Mobile()
   Event OnMusicBattleInstanceInitializeEvent() {is=title K2Node_Event}
   n14 = GetLocalPilgrimGameForPlayer!(
    self:Pilgrim Music Battle Instance Object Reference=Get MusicBattleInstance() -> (MusicBattleInstance:Pilgrim Music Battle Instance Object Reference)
    InPC:Player Controller Object Reference=GetOwningPlayer()
   ) -> (ReturnValue:Pilgrim Game Object Reference)
   > Is Valid(InputObject=@n14)
   > ProcessNumOfTrackLanes!(execute, self, IsTrackFourLanes:bool="Equal (Integer)"(A=GetNumLanes!(self:Pilgrim Game Object Reference=@n14) -> (ReturnValue:int=0), B=4)) -> (then)
   n15 = Event PreConstruct()
   > Set TrackIndex!(
    execute
    TrackIndex:int=Get Track Index() -> (Track Index:int=0)
    self:Pilgrim Track UI Object Reference=Get PilgrimTrackUI_AllTransforms() -> (PilgrimTrackUI_AllTransforms:Pilgrim Track UI Object Reference)
   ) -> (then, Output_Get:int=0)
   > SetWidgetFiveLaneOffsets()
   > SetWidgetPaddingOffsetsMobile()
   > n16 = Branch(Condition=@n15.IsDesignTime)
   > ProcessNumOfTrackLanes(IsTrackFourLanes=Get PreviewFourLaneUI() -> (PreviewFourLaneUI:bool))
   > Adjust Elements For Local MP!(execute, self, Is Local MP View:bool=Get PreviewLocalMP() -> (PreviewLocalMP:bool)) -> (then) {is=title K2Node_CallFunction}
   > HideHUDForScreenCapture!(execute, self, Hide:bool=Get Preview HideHUDForScreenCapture() -> (Preview HideHUDForScreenCapture:bool)) -> (then)
   Adjust Elements For Track Width!(execute=@n16.else, self, isTrackFourLane:bool=true) -> (then) {is=title K2Node_CallFunction}
   > SetupNumLanes!(
    execute
    self:WBP Pilgrim Track Input Labels Object Reference=Get WBP_PilgrimTrackInputLabels() -> (WBP_PilgrimTrackInputLabels:WBP Pilgrim Track Input Labels Object Reference)
    NumLanes:int=4
   ) -> (then)
   n17 = Get WBP_Bandmates()
   > SheerForTrack!(execute, self:WBP Bandmates Object Reference=@n17, NumLanes:int=4) -> (then)
   > PivotForTrack(self:WBP Bandmates Object Reference=@n17)
   > Set Visibility(Visibility=Hidden, self=Get TrackRenderTarget_Proxy() -> (TrackRenderTarget_Proxy:Image Object Reference))
   > SetupGemHitFeedbackPosition!(execute, self, IsFourLane:bool=true) -> (then)
  ]}
 HideHUDForScreenCapture {nodes [
   n1 = HideHUDForScreenCapture!() -> (then, Hide:bool=~) {is=title K2Node_FunctionEntry}
   > SetVisibility(
    self=[
     Get WBP_PilgrimAccoladesPopupWrapper() -> (WBP_PilgrimAccoladesPopupWrapper:WBP Pilgrim Accolades Popup Wrapper Object Reference)
     Get WBP_PilgrimQuickplay_BreakPrompt() -> (WBP_PilgrimQuickplay_BreakPrompt:WBP Pilgrim Quickplay Break Prompt Object Reference)
     Get WBP_PilgrimQuickplay_GemBreakPrompt() -> (WBP_PilgrimQuickplay_GemBreakPrompt:WBP Pilgrim Quickplay Gem Break Prompt Object Reference)
     Get WBP_Bandmates()
     Get WBP_PilgrimQuickplay_KickInputLabel() -> (WBP_PilgrimQuickplay_KickInputLabel:WBP Pilgrim Quickplay Kick Input Label Object Reference)
     Get WBP_PilgrimAccuracy_Enhanced()
     Get WBP_PilgrimQuickplay_StrumInputLabel() -> (WBP_PilgrimQuickplay_StrumInputLabel:WBP Pilgrim Quickplay Strum Input Label Object Reference)
     Get WBP_PilgrimQuickplay_OverdriveActionPrompt() -> (WBP_PilgrimQuickplay_OverdriveActionPrompt:WBP Pilgrim Quickplay Overdrive Action Prompt Object Reference)
     Get WBP_PilgrimTrackInputLabels()
    ]
    InVisibility=Select(Option 0:ESlateVisibility Enum=SelfHitTestInvisible, Option 1:ESlateVisibility Enum=Hidden, Index=@n1) -> (ReturnValue:ESlateVisibility Enum)
   )
  ]}
 ProcessNumOfTrackLanes {nodes [
   n1 = ProcessNumOfTrackLanes!() -> (then, IsTrackFourLanes:bool=~) {is=title K2Node_FunctionEntry}
   > SetupGemHitFeedbackPosition(IsFourLane=@n1)
   > Adjust Elements For Track Width(isTrackFourLane=@n1)
   n2 = SelectInt(A=4, B=5, bPickA=@n1)
   > SetupNumLanes(self=Get WBP_PilgrimTrackInputLabels(), NumLanes=@n2)
   n3 = Get WBP_Bandmates()
   > SheerForTrack(self=@n3, NumLanes=@n2)
   > PivotForTrack(self=@n3)
   > SetBrushFromTexture!(
    execute
    self:Image Object Reference=Get TrackRenderTarget_Proxy()
    Texture:Texture 2D Object Reference=Select(
     Option 0:Texture 2D Object Reference=/PilgrimQuickplay/Widgets/Textures/DevOnly/track_capture_5Lane
     Option 1:Texture 2D Object Reference=/PilgrimQuickplay/Widgets/Textures/DevOnly/track_capture_4Lane
     Index=@n1
    ) -> (ReturnValue:Texture 2D Object Reference)
    bMatchSize:bool=false
   ) -> (then)
  ]}
 Set ControllerDisconnected Visibility {nodes [
   n1 = Set ControllerDisconnected Visibility!() -> (then, IsDisconnected:bool=~) {is=title K2Node_FunctionEntry}
   n2 = IsPlayingReplay!() -> (ReturnValue:bool)
   n3 = IsMobileGame()
   > n4 = Branch(Condition=AND Boolean(A=NOT Boolean(A=@n3), B=NOT Boolean(A=@n2)))
   > n5 = Branch(Condition=@n1)
   n6 = IsVisible(self=Get WBP_ControllerDisconnectPrompt() -> (WBP_ControllerDisconnectPrompt:WBP Pilgrim Quickplay Controller Disconnect Prompt Object Reference))
   > n7 = Branch(Condition=@n6)
   n8 = Get ControllerDisconnect() -> (ControllerDisconnect:Widget Animation Object Reference)
   PlayAnimationForward(execute=@n7.else, InAnimation=@n8)
   Branch(execute=[@n5.else, @n4.else], Condition=@n6)
   > PlayAnimationReverse(InAnimation=@n8)
  ]}
 SetPilgrimTrackCapture {nodes [
   n1 = SetPilgrimTrackCapture!() -> (then, Pilgrim Track Capture:Pilgrim Track Capture Component Object Reference) {is=title K2Node_FunctionEntry}
   > Set PilgrimTrackCapture!(
    execute
    PilgrimTrackCapture:Pilgrim Track Capture Component Object Reference=@n1
    self:Pilgrim Track UI Object Reference=Get PilgrimTrackUI_AllTransforms()
   ) -> (then, Output_Get:Pilgrim Track Capture Component Object Reference)
  ]}
 SetStreakCounterVisibility {nodes [
   n1 = SetStreakCounterVisibility!() -> (then, Index:bool=~) {is=title K2Node_FunctionEntry}
   > SetStreakCounterVisibility!(
    execute
    self:WBP Pilgrim Accuracy Text Object Reference=Get WBP_PilgrimAccuracy_Enhanced()
    New Visibility:ESlateVisibility Enum=Select(Option 0:ESlateVisibility Enum=Hidden, Option 1:ESlateVisibility Enum=Visible, Index=@n1) -> (ReturnValue:ESlateVisibility Enum)
   ) -> (then) {is=title K2Node_CallFunction}
  ]}
 SetWidgetFiveLaneOffsets {nodes [
   SetWidgetFiveLaneOffsets() {is=title K2Node_FunctionEntry}
   > Clear(TargetArray:Array of Widget Five Lane Offset Structures=Get WidgetsWithFiveLaneOffsets())
   n1 = Get WBP_PilgrimQuickplay_StrumInputLabel()
   n2 = Select(
    Option 0:Vector 2D Structure=(X=60,Y=0)
    Option 1:Vector 2D Structure=(X=0,Y=0)
    Index=IsMobileAndShouldUseExpandedTrack!() -> (ReturnValue:bool)
   ) -> (ReturnValue:Vector 2D Structure)
   n3 = Get WBP_Bandmates()
   n4 = Get WBP_PilgrimQuickplay_OverdriveActionPrompt()
   > Append Array!(
    execute
    TargetArray:Array of Widget Five Lane Offset Structures=Get WidgetsWithFiveLaneOffsets()
    SourceArray:Array of Widget Five Lane Offset Structures=Make Array!(
     "[0]":Widget Five Lane Offset Structure=Make Widget Five Lane Offset!(
      Widget_2_#3:Widget Object Reference=@n3
      FiveLaneTransformOffset_5_#4:Vector 2D Structure=@n2
     ) -> (WidgetFiveLaneOffset:Widget Five Lane Offset Structure) {is=title K2Node_MakeStruct}
     "[1]":Widget Five Lane Offset Structure=Make Widget Five Lane Offset(Widget_2_#3=@n4, FiveLaneTransformOffset_5_#4=(X=68,Y=0))
     "[2]":Widget Five Lane Offset Structure=Make Widget Five Lane Offset(Widget_2_#3=@n1, FiveLaneTransformOffset_5_#4=(X=-56,Y=0))
    ) -> (Array:Array of Widget Five Lane Offset Structures)
   ) -> (then)
   Make Widget Five Lane Offset(
    Widget_2_#3=Get WBP_PilgrimAccuracy_Enhanced()
    FiveLaneTransformOffset_5_#4=(X=-56,Y=0)
   )
  ]}
 SetWidgetPaddingOffsetsMobile {nodes [
   SetWidgetPaddingOffsetsMobile() {is=title K2Node_FunctionEntry}
   > Clear(TargetArray:Array of Widget Mobile Margin Offset Structures=Get WidgetsMobilePaddingOffsets())
   n1 = IsMobileAndShouldUseExpandedTrack()
   n2 = Make Margin(Left=Select(Option 0:float=0, Option 1:float=0, Index=@n1) -> (ReturnValue:float), Top=140, Right=0, Bottom=232)
   n3 = Get WBP_Bandmates()
   n4 = Make Margin(Left=Select(Option 0=-450, Option 1=-740, Index=@n1), Top=0, Right=0, Bottom=70)
   n5 = Get WBP_PilgrimAccuracy_Enhanced()
   > Append Array(
    TargetArray:Array of Widget Mobile Margin Offset Structures=Get WidgetsMobilePaddingOffsets()
    SourceArray:Array of Widget Mobile Margin Offset Structures=Make Array!(
     "[0]":Widget Mobile Margin Offset Structure=Make Widget Mobile Margin Offset!(Widget_2_#3:Widget Object Reference=@n3, Margin_7_#4:Margin Structure=@n2) -> (WidgetMobileMarginOffset:Widget Mobile Margin Offset Structure) {is=title K2Node_MakeStruct}
     "[1]":Widget Mobile Margin Offset Structure=Make Widget Mobile Margin Offset(Widget_2_#3=@n5, Margin_7_#4=@n4)
    ) -> (Array:Array of Widget Mobile Margin Offset Structures)
   )
  ]}
 SetupGemHitFeedbackPosition {nodes [
   n1 = SetupGemHitFeedbackPosition!() -> (then, IsFourLane:bool=~) {is=title K2Node_FunctionEntry}
   n2 = OR Boolean(A=GetUseHitFeedbackUpTrack!() -> (ReturnValue:bool), B=Get TestHitFeedbackUpTrack() -> (TestHitFeedbackUpTrack:bool))
   n3 = Select(
    Option 0:Vector 2D Structure=Get HitFeedbackFourLaneOffset() -> (HitFeedbackFourLaneOffset:Vector 2D Structure)
    Option 1:Vector 2D Structure=Get HitFeedbackFourLaneOffsetUptrack() -> (HitFeedbackFourLaneOffsetUptrack:Vector 2D Structure)
    Index=@n2
   ) -> (ReturnValue:Vector 2D Structure)
   n4 = Select(
    Option 0=Get GemHitFeedbackPosition01_Render() -> (GemHitFeedbackPosition01_Render:Vector 2D Structure)
    Option 1=Get GemHitFeedbackPosition02_Render() -> (GemHitFeedbackPosition02_Render:Vector 2D Structure)
    Index=@n2
   )
   > SetRenderTranslation(
    self=Get WBP_PilgrimAccuracy_Enhanced()
    Translation=Select(
     Option 0=@n4
     Option 1=vector2d + vector2d(A:Vector 2D Structure=@n4, B:Vector 2D Structure=@n3) -> (ReturnValue:Vector 2D Structure) {is=title K2Node_PromotableOperator}
     Index=@n1
    )
   )
  ]}
 SetupStreakCounter {nodes [
   SetupStreakCounter() {is=title K2Node_FunctionEntry}
   > SetStreakCounterVisibility!(execute, self:Self Object Reference, Index:bool=GetShowStreakCounter()) -> (then)
   > Set IsStreakCounterSetup(IsStreakCounterSetup:bool=true) -> (Output_Get:bool)
  ]}
]
variables [
 EventRouter {type=GameplayEventRouterComponent, default=null}
 GemHitFeedbackPosition01_Render {type=Vector2D, default=<0 0>}
 GemHitFeedbackPosition02_Render {type=Vector2D, default=<0 0>}
 HitFeedbackFourLaneOffset {type=Vector2D, default=<60 0>}
 HitFeedbackFourLaneOffsetUptrack {type=Vector2D, default=<0 0>}
 IsStreakCounterSetup {type=boolean}
 LocalMPIdentityVerticalOffset {type=number, default=-60}
 Preview HideHUDForScreenCapture {type=boolean}
 PreviewFourLaneUI {type=boolean, default=true}
 PreviewLocalMP {type=boolean}
 TestHitFeedbackUpTrack {type=boolean}
 Track Index {type=integer}
 WidgetsMobilePaddingOffsets {type="array<WidgetMobileMarginOffset>", default [{widget=null, margin=<ltrb 320 140 0 232>}, {widget=null, margin=<ltrb -415 0 0 138>}]}
 WidgetsWithFiveLaneOffsets {
  type="array<WidgetFiveLaneOffset>"
  default [{widget=null, fiveLaneTransformOffset=<60 0>}, {widget=null, fiveLaneTransformOffset=<68 0>}, {widget=null, fiveLaneTransformOffset=<-56 0>}]
 }
]
widgets {
 info {
  inheritedWidgetCount=0
  namedSlotCount=1
  parentClass=/Script/PilgrimCoreUI.PilgrimGameWidgetBase
  rootWidgetClass=/Script/PilgrimCoreUI.PilgrimTrackUI
  widgetCount=24
 }
 widgets [
  PilgrimTrackUI_AllTransforms {
   widgetClassPath=&PilgrimTrackUI
   bIsVariable=true
   widget=PilgrimTrackUI_AllTransforms
   properties {trackPlaneOffset=-1.3092}
  }
  TrackRenderTarget_Proxy {
   widgetClassPath=&Image
   bIsVariable=true
   parent=PilgrimTrackUI_AllTransforms
   widget=TrackRenderTarget_Proxy
   properties {
    visibility=Hidden
    brush {
     bIsDynamicallyLoaded=false
     drawAs=Image
     imageType=FullColor
     mirroring=NoMirror
     resourceName=None
     resourceObject=&track_capture_4Lane
     tiling=NoTile
     imageSize=<610 440>
     margin=<ltrb 0 0 0 0>
     outlineSettings {
      bUseBrushTransparency=false
      roundingType=HalfHeightRadius
      width=0
      color {colorUseRule=UseColor_Specified, specifiedColor=<rgba 0 0 0 0>}
      cornerRadii=<0 0 0 0>
     }
     tintColor {colorUseRule=UseColor_Specified, specifiedColor=<rgba 1 1 1 1>}
     uVRegion {bIsValid=false, max=<0 0>, min=<0 0>}
    }
    renderTransform {angle=0, scale=<0.846 0.62>, shear=<0 0>, translation=<7.9405 194.9999>}
   }
   slot {class=OverlaySlot, properties {horizontalAlignment=HAlign_Fill, verticalAlignment=VAlign_Fill}}
  }
  GemBreakPanel_FadeWithTrack {
   widgetClassPath=/PilgrimCore/Widgets/WBP_PilgrimGemBreakPanel
   bIsVariable=true
   parent=PilgrimTrackUI_AllTransforms
   widget=GemBreakPanel_FadeWithTrack
   properties {applyLocalMultiplayerExceptions=true}
   slot {class=OverlaySlot, properties {horizontalAlignment=HAlign_Fill, verticalAlignment=VAlign_Fill}}
  }
  Overlay_FadeWithTrack {widgetClassPath=&Overlay, namedSlotHost=GemBreakPanel_FadeWithTrack, widget=Overlay_FadeWithTrack}
  SizeBox_1 {
   widgetClassPath=&SizeBox
   parent=Overlay_FadeWithTrack
   widget=SizeBox_1
   properties {bOverride_HeightOverride=true, bOverride_WidthOverride=true, heightOverride=200, widthOverride=200}
   slot {class=OverlaySlot, properties {horizontalAlignment=HAlign_Center, verticalAlignment=VAlign_Bottom, padding=<ltrb 380 0 0 232>}}
  }
  WBP_Bandmates {
   widgetClassPath=/PilgrimCore/Widgets/WBP_Bandmates
   bIsVariable=true
   parent=SizeBox_1
   widget=WBP_Bandmates
   properties {visibility=Visible, renderPivot=<0 1>, renderTransform {angle=0, scale=<1 1>, shear=<19 0>, translation=<0 0>}, renderTransformPivot=<0 1>}
   slot {class=SizeBoxSlot, properties {horizontalAlignment=HAlign_Left, verticalAlignment=VAlign_Bottom}}
  }
  WBP_PilgrimTrackInputLabels {
   widgetClassPath=/PilgrimCore/Widgets/WBP_PilgrimTrackInputLabels
   bIsVariable=true
   parent=Overlay_FadeWithTrack
   widget=WBP_PilgrimTrackInputLabels
   slot {class=OverlaySlot, properties {horizontalAlignment=HAlign_Center, verticalAlignment=VAlign_Bottom, padding=<ltrb 0 0 0 -36>}}
  }
  PilgrimTrackUI_NoShear {
   widgetClassPath=&PilgrimTrackUI
   parent=Overlay_FadeWithTrack
   widget=PilgrimTrackUI_NoShear
   properties {slateTransformMode=ScaleTranslation}
   slot {class=OverlaySlot, properties {horizontalAlignment=HAlign_Fill, verticalAlignment=VAlign_Fill}}
  }
  WBP_PilgrimQuickplay_StrumInputLabel {
   widgetClassPath=&/WBP_PilgrimQuickplay_StrumInputLabel
   bIsVariable=true
   parent=PilgrimTrackUI_NoShear
   widget=WBP_PilgrimQuickplay_StrumInputLabel
   slot {class=OverlaySlot, properties {verticalAlignment=VAlign_Bottom, padding=<ltrb 242 0 0 82>}}
  }
  WBP_PilgrimQuickplay_KickInputLabel {
   widgetClassPath=&/WBP_PilgrimQuickplay_KickInputLabel
   bIsVariable=true
   parent=PilgrimTrackUI_NoShear
   widget=WBP_PilgrimQuickplay_KickInputLabel
   slot {class=OverlaySlot, properties {verticalAlignment=VAlign_Bottom, padding=<ltrb 242 0 0 82>}}
  }
  WBP_PilgrimAccuracy_Enhanced {
   widgetClassPath=/PilgrimCore/Widgets/WBP_PilgrimAccuracyText
   bIsVariable=true
   parent=PilgrimTrackUI_NoShear
   widget=WBP_PilgrimAccuracy_Enhanced
   properties {renderTransform {angle=0, scale=<0.7 0.7>, shear=<0 0>, translation=<0 0>}, renderTransformPivot=<1 0.5>}
   slot {class=OverlaySlot, properties {horizontalAlignment=HAlign_Center, verticalAlignment=VAlign_Bottom, padding=<ltrb -450 0 0 70>}}
  }
  LMP_IdentityOverlay {
   widgetClassPath=/PilgrimQuickplay/Widgets/Players/WBP_Pilgrim_PlayerIdentity_Overlay
   bIsVariable=true
   parent=PilgrimTrackUI_NoShear
   widget=LMP_IdentityOverlay
   properties {visibility=Collapsed}
   slot {class=OverlaySlot, properties {horizontalAlignment=HAlign_Center, verticalAlignment=VAlign_Bottom, padding=<ltrb 0 0 0 520>}}
  }
  WBP_PilgrimQuickplay_OverdriveActionPrompt {
   widgetClassPath=&/WBP_PilgrimQuickplay_OverdriveActionPrompt
   bIsVariable=true
   parent=PilgrimTrackUI_NoShear
   widget=WBP_PilgrimQuickplay_OverdriveActionPrompt
   properties {renderTransform {angle=0, scale=<1 1>, shear=<0 0>, translation=<41 0>}}
   slot {class=OverlaySlot, properties {horizontalAlignment=HAlign_Right, verticalAlignment=VAlign_Bottom, padding=<ltrb 0 0 240 82>}}
  }
  LMP_Nameplate {
   widgetClassPath=/PilgrimQuickplay/Widgets/Players/WBP_Pilgrim_PlayerIdentity_Nameplate
   bIsVariable=true
   parent=PilgrimTrackUI_NoShear
   widget=LMP_Nameplate
   slot {class=OverlaySlot, properties {horizontalAlignment=HAlign_Center, verticalAlignment=VAlign_Bottom, padding=<ltrb 0 0 0 -70>}}
  }
  PilgrimTrackUI_NoShear2 {
   widgetClassPath=&PilgrimTrackUI
   parent=PilgrimTrackUI_AllTransforms
   widget=PilgrimTrackUI_NoShear2
   properties {slateTransformMode=ScaleTranslation}
   slot {class=OverlaySlot, properties {horizontalAlignment=HAlign_Center, verticalAlignment=VAlign_Bottom, padding=<ltrb 0 0 0 130>}}
  }
  SizeBox_0 {
   widgetClassPath=&SizeBox
   parent=PilgrimTrackUI_NoShear2
   widget=SizeBox_0
   properties {bOverride_WidthOverride=true, widthOverride=600}
   slot {class=OverlaySlot, properties {horizontalAlignment=HAlign_Fill}}
  }
  WBP_ControllerDisconnectPrompt {
   widgetClassPath=&/WBP_PilgrimQuickplay_ControllerDisconnectPrompt
   bIsVariable=true
   parent=SizeBox_0
   widget=WBP_ControllerDisconnectPrompt
   properties {renderOpacity=0, visibility=Collapsed}
   slot {class=SizeBoxSlot}
  }
  PilgrimTrackUI_NotFadingWithTrack {
   widgetClassPath=&PilgrimTrackUI
   parent=PilgrimTrackUI_AllTransforms
   widget=PilgrimTrackUI_NotFadingWithTrack
   properties {
    slateTransformMode=ScaleTranslation
    widgetClassesToTransform [
     &Border
     &Image
     &TextBlock
     /Script/FortniteUI.FortKeybindWidget
     /PilgrimQuickplay/Widgets/Quests/WBP_PilgrimAccoladesPopupWrapper.WBP_PilgrimAccoladesPopupWrapper_C
    ]
   }
   slot {class=OverlaySlot, properties {horizontalAlignment=HAlign_Fill, verticalAlignment=VAlign_Fill}}
  }
  CommonVisualAttachment_79 {
   widgetClassPath=&CommonVisualAttachment
   parent=PilgrimTrackUI_NotFadingWithTrack
   widget=CommonVisualAttachment_79
   properties {contentAnchor=<0.35 0.035>}
   slot {class=OverlaySlot, properties {verticalAlignment=VAlign_Center}}
  }
  WBP_PilgrimAccoladesPopupWrapper {
   widgetClassPath=/PilgrimQuickplay/Widgets/Quests/WBP_PilgrimAccoladesPopupWrapper
   bIsVariable=true
   parent=CommonVisualAttachment_79
   widget=WBP_PilgrimAccoladesPopupWrapper
   slot {class=SizeBoxSlot}
  }
  CommonVisibilityWidgetLegacy_0 {
   widgetClassPath=&CommonVisibilityWidgetLegacy
   parent=PilgrimTrackUI_NotFadingWithTrack
   widget=CommonVisibilityWidgetLegacy_0
   properties {bShowForTouch=false}
   slot {class=OverlaySlot, properties {horizontalAlignment=HAlign_Center, verticalAlignment=VAlign_Bottom, padding=<ltrb 0 0 0 220>}}
  }
  WBP_PilgrimQuickplay_GemBreakPrompt {
   widgetClassPath=&/WBP_PilgrimQuickplay_GemBreakPrompt
   bIsVariable=true
   parent=CommonVisibilityWidgetLegacy_0
   widget=WBP_PilgrimQuickplay_GemBreakPrompt
   slot {class=BorderSlot}
  }
  SizeBox_2 {
   widgetClassPath=&SizeBox
   parent=PilgrimTrackUI_NotFadingWithTrack
   widget=SizeBox_2
   properties {bOverride_HeightOverride=true, bOverride_WidthOverride=true, heightOverride=80, widthOverride=80}
   slot {class=OverlaySlot, properties {horizontalAlignment=HAlign_Center, verticalAlignment=VAlign_Bottom, padding=<ltrb 0 0 0 330>}}
  }
  WBP_PilgrimQuickplay_BreakPrompt {
   widgetClassPath=&/WBP_PilgrimQuickplay_BreakPrompt
   bIsVariable=true
   parent=SizeBox_2
   widget=WBP_PilgrimQuickplay_BreakPrompt
   slot {class=SizeBoxSlot, properties {horizontalAlignment=HAlign_Center, verticalAlignment=VAlign_Center}}
  }
 ]
}
