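# Quickplay gameplay widget that displays Pilgrim's captured note track with per-track UI and mobile lane touch input.
# On music battle initialization, it creates widgets for non-pro-vocal tracks and feeds the capture render target, fade bounds and UV transform into a projection material.
# Mobile shows 4- or 5-lane polygonal input zones, transformed to match the track and reversible for lefty flip.
# Gem-break tags and song-stop events drive the background fade.
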WidgetBlueprint /PilgrimQuickplay/Widgets/WBP_PilgrimQuickplay_TrackAndUI
animations [FadeInGameplayBackground {
  binding_count=1
  length_seconds=2.0003
  object=&self:FadeInGameplayBackground
  sequence {bindings [WBP_Pilgrim_TrackBackground {
     id=#1
     possessed_class=WBP_Pilgrim_TrackBackground_C
     tracks [{
       type=MovieSceneFloatTrack
       display_name=Render Opacity
       sections [{type=MovieSceneFloatSection, end_frame=40, start_frame=0, channels [None {type=MovieSceneScriptingFloatChannel, keys [[0,0],[20,0],[40,1]]}]}]
      }]
    }]}
 }]
graphs [
 AdjustElementsPaddingForMobile {nodes [
   AdjustElementsPaddingForMobile() {is=title K2Node_FunctionEntry}
   > n1 = For Each Loop(Array:Array of Widget Mobile Margin Offset Structures=Get WidgetsMobilePaddingOffsets() -> (WidgetsMobilePaddingOffsets:Array of Widget Mobile Margin Offset Structures)) -> (Array Element:Widget Mobile Margin Offset Structure)
   n2 = Break Widget Mobile Margin Offset!(WidgetMobileMarginOffset:"Widget Mobile Margin Offset Structure (by ref)"=@n1) -> (Widget_2_#2:Widget Object Reference, Margin_7_#3:Margin Structure) {is=title K2Node_BreakStruct}
   > SetPadding!(
    execute
    self:Overlay Slot Object Reference=SlotAsOverlaySlot(Widget=@n2)
    InPadding:Margin Structure=@n2.Margin_7_#3
   ) -> (then)
  ]}
 Can Fade Track Elements {nodes [
   Can Fade Track Elements() {is=title K2Node_FunctionEntry}
   n1 = Get MusicBattleInstance() -> (MusicBattleInstance:Pilgrim Music Battle Instance Object Reference)
   > n2 = Is Valid(InputObject=@n1)
   > n3 = Set LocalPilgrimGame_Local(LocalPilgrimGame_Local:Pilgrim Game Object Reference=GetLocalPilgrimGameForPlayer!(self:Pilgrim Music Battle Instance Object Reference=@n1, InPC:Player Controller Object Reference=GetOwningPlayer()) -> (ReturnValue:Pilgrim Game Object Reference)) -> (Output_Get:Pilgrim Game Object Reference)
   > n4 = Is Valid(InputObject=@n3)
   n5 = PrintString!(
    execute=@n4.Is Not Valid
    InString="No valid pilgrim game, defaulting to fading in the track"
    bPrintToScreen
    bPrintToLog
    TextColor
    Duration
    Key
   ) -> (then)
   Return Node!(execute=[@n7.else, @n5.then, @n2.Is Not Valid], CanFadeTrackElements:bool=true)
   n6 = Branch(execute=@n4.Is Valid, Condition=IsHamSandwichSessionActive!() -> (ReturnValue:bool))
   > n7 = Branch(Condition=AreHamSandwichSongBreaksEnabled!() -> (ReturnValue:bool))
   n8 = Break Pilgrim Gem Break!(PilgrimGemBreak:"Pilgrim Gem Break Structure (by ref)"=GetGemBreakForTrackDifficulty!(
     self:Pilgrim Game Object Reference=@n3
     Track:EPilgrimTrackType Enum=GetSelectedTrackType!(self:Pilgrim Game Object Reference=@n3) -> (ReturnValue:EPilgrimTrackType Enum=TrackGuitar)
     Difficulty:EPilgrimSongDifficulty Enum=GetSelectedDifficulty!(self:Pilgrim Game Object Reference=@n3) -> (ReturnValue:EPilgrimSongDifficulty Enum=DifficultyEasy)
    ) -> (
     FoundBreak:Pilgrim Gem Break Structure
     BarsLeftInBreak:float
     NextBreak:Pilgrim Gem Break Structure
     BarsTillNextBreak:float
     ReturnValue:bool
    )) -> (
    Id:int
    bIsValidGemBreak:bool=~
    StartBreakAtBar:float=~
    StartBreakAtMs:float=~
    BreakBarDuration:float=~
    BreakMsDuration:float=~
    EndBreakAtBar:float=~
    EndBreakAtMs:float=~
    bIsStartingBreak:bool=~
    bIsEndingBreak:bool=~
   ) {is=title K2Node_BreakStruct}
   Return Node(
    execute=[@n7.then, @n6.else]
    CanFadeTrackElements=NAND Boolean(A=@n8.bIsValidGemBreak, B=@n8.bIsStartingBreak) {is=title K2Node_CommutativeAssociativeBinaryOperator}
   )
  ]}
 ConstructPerTrackWidget {nodes [
   n1 = ConstructPerTrackWidget!() -> (then, OwningPlayer:Player Controller Object Reference, TrackIndex:int) {is=title K2Node_FunctionEntry}
   > n2 = Create Widget!(
    execute:Exec
    Class:User Widget Class Reference=&/WBP_PilgrimQuickplay_PerTrackWidgets.WBP_PilgrimQuickplay_PerTrackWidgets_C
    OwningPlayer:Player Controller Object Reference=@n1
    Track Index:int=@n1.TrackIndex
   ) -> (then:Exec, ReturnValue:WBP Pilgrim Quickplay Per Track Widgets Object Reference)
   > n3 = AddChildToOverlay!(execute, self:Overlay Object Reference=Get Track_Overlay() -> (Track_Overlay:Overlay Object Reference), Content:Widget Object Reference=@n2) -> (then, ReturnValue:Overlay Slot Object Reference)
   > Set HorizontalAlignment!(execute, HorizontalAlignment:EHorizontalAlignment Enum=HAlign_Fill, self:Overlay Slot Object Reference=@n3) -> (then, Output_Get:EHorizontalAlignment Enum=HAlign_Fill)
   > Set VerticalAlignment!(execute, VerticalAlignment:EVerticalAlignment Enum=VAlign_Fill, self:Overlay Slot Object Reference=@n3) -> (then, Output_Get:EVerticalAlignment Enum=VAlign_Fill)
   > Return Node!(execute, Widget:WBP Pilgrim Quickplay Per Track Widgets Object Reference=@n2)
  ]}
 CreatePerTrackWidgets {nodes [
   CreatePerTrackWidgets() {is=title K2Node_FunctionEntry}
   > n1 = Get!(execute:Exec) -> (then:Exec, else:Exec, MusicBattleInstance:Pilgrim Music Battle Instance Object Reference)
   > n2 = Set TrackAndGameManager(TrackAndGameManager:Pilgrim Track And Game Manager Component Object Reference=GetTrackAndGameManager!(self:Pilgrim Music Battle Instance Object Reference=@n1) -> (ReturnValue:Pilgrim Track And Game Manager Component Object Reference)) -> (Output_Get:Pilgrim Track And Game Manager Component Object Reference)
   > n3 = Is Valid(InputObject=@n2)
   LogUIEvent!(
    execute=@n3.Is Not Valid
    self:Pilgrim Music Battle Instance Object Reference=Get MusicBattleInstance()
    LogString:String=unable to find track and game manager.
   ) -> (then)
   > Return Node!(execute)
   n4 = Get TrackAndGameManager() -> (TrackAndGameManager:Pilgrim Track And Game Manager Component Object Reference)
   n5 = For Loop(
    execute=@n3.Is Valid
    LastIndex=Last Index!(TargetArray:Array of Pilgrim Track Object References=Get Tracks!(self:Pilgrim Track And Game Manager Component Object Reference=@n4) -> (Tracks:Array of Pilgrim Track Object References)) -> (ReturnValue=0)
   )
   n6 = Get TrackAndGameManager()
   > Branch(Condition=ShouldAddPerTrackWidgets!(TrackIndex:int=@n5, GameManager:"Pilgrim Track And Game Manager Component Object Reference (by ref)"=@n6) -> (ShouldAdd:bool))
   > n7 = GetOwningPlayer!(
    execute:Exec
    self:Pilgrim Track Object Reference="Get (a copy)"(Array:Array of Pilgrim Track Object References=Get Tracks(self=@n6), Dimension 1=@n5) -> (Output:Pilgrim Track Object Reference)
   ) -> (then:Exec, ReturnValue)
   > n8 = ConstructPerTrackWidget!(
    execute
    self
    OwningPlayer:Player Controller Object Reference=Select(
     Option 0:Player Controller Object Reference=GetOwningPlayer!(self:Widget Object Reference) -> (ReturnValue)
     Option 1:Player Controller Object Reference=@n7
     Index=IsValid(Object=@n7)
    ) -> (ReturnValue:Player Controller Object Reference)
    TrackIndex:int=@n5
   ) -> (then, Widget:WBP Pilgrim Quickplay Per Track Widgets Object Reference) {is=title K2Node_CallFunction}
   > Add(
    TargetArray:Array of WBP Pilgrim Quickplay Per Track Widgets Object References=Get AllTrackWidgets() -> (AllTrackWidgets:Array of WBP Pilgrim Quickplay Per Track Widgets Object References)
    NewItem:"WBP Pilgrim Quickplay Per Track Widgets Object Reference (by ref)"=@n8
   ) {is=title K2Node_CallArrayFunction}
  ]}
 EventGraph {nodes [
   n1 = Event Setup Render Target() {is=title K2Node_CustomEvent}
   > SetupRenderTarget()
   n2 = Event PreConstruct()
   > Parent: PreConstruct(IsDesignTime=@n2.IsDesignTime)
   > SetBrushFromMaterial(
    self=Get TrackRenderTarget() -> (TrackRenderTarget:Image Object Reference)
    Material=Get RenderTargetMaterial() -> (RenderTargetMaterial:Material Interface Object Reference)
   )
   > Branch(Condition=@n2.IsDesignTime)
   > n3 = Branch(Condition=Get PreviewPerTrackWidgets() -> (PreviewPerTrackWidgets:bool))
   > n4 = ConstructPerTrackWidget(TrackIndex=0)
   > n5 = Set PerTrackWidgetPreview(PerTrackWidgetPreview:WBP Pilgrim Quickplay Per Track Widgets Object Reference=@n4) -> (Output_Get:WBP Pilgrim Quickplay Per Track Widgets Object Reference)
   > n6 = Set PreviewFourLaneUI!(
    execute
    PreviewFourLaneUI:bool=Get Preview4LaneUI() -> (Preview4LaneUI:bool)
    self:WBP Pilgrim Quickplay Per Track Widgets Object Reference=@n5
   ) -> (then, Output_Get:bool)
   > ProcessNumOfTrackLanes!(
    execute
    self:WBP Pilgrim Quickplay Per Track Widgets Object Reference=Get PerTrackWidgetPreview() -> (PerTrackWidgetPreview:WBP Pilgrim Quickplay Per Track Widgets Object Reference)
    IsTrackFourLanes:bool=@n6
   ) -> (then)
   n7 = Get!(execute=@n3.else) -> (then, else, PerTrackWidgetPreview:WBP Pilgrim Quickplay Per Track Widgets Object Reference)
   > RemoveFromParent(self=@n7)
   > Set PerTrackWidgetPreview()
   Event Construct()
   > SetSidesToPad!(
    execute
    self:Safe Zone Object Reference=Get SZ_MobileOnly() -> (SZ_MobileOnly:Safe Zone Object Reference)
    InPadLeft:bool=false
    InPadRight:bool=false
    InPadTop:bool=false
    InPadBottom:bool=IsMobileGame()
   ) -> (then)
   > n8 = Branch(Condition=Get isMBIInitialized() -> (isMBIInitialized:bool))
   SetVisibility(execute=@n8.else, InVisibility=Hidden)
   Event OnMusicBattleInstanceInitializeEvent() {is=title K2Node_Event}
   > n9 = Sequence!(execute) -> (then_0, then_1, then_2:Exec, then_3:Exec, then_4:Exec, then_5:Exec, then_6:Exec, then_7:Exec, then_8:Exec)
   > n10 = Get!(execute) -> (then, else, MusicBattleInstance:Pilgrim Music Battle Instance Object Reference)
   > LogUIEvent(self=@n10, LogString=TrackAndUI widget being Initialized)
   Set isMBIInitialized(execute=@n9.then_1, isMBIInitialized:bool=true) -> (Output_Get:bool)
   > SetVisibility(InVisibility=SelfHitTestInvisible)
   CreatePerTrackWidgets(execute=@n9.then_2) {is=title K2Node_CallFunction}
   n11 = CreateDynamicMaterialInstance!(execute=@n9.then_3, Parent=Get RenderTargetMaterial(), OptionalName=None, CreationFlags=None) -> (then, ReturnValue)
   > Set Material(Material:Material Instance Dynamic Object Reference=@n11) -> (Output_Get:Material Instance Dynamic Object Reference)
   > SetupRenderTarget()
   > Bind Event to On Render Target Setup(self:Pilgrim Track And Game Manager Component Object Reference=GetTrackAndGameManager(self=Get MusicBattleInstance()), Delegate=@n1) {is=title K2Node_AddDelegate}
   n12 = Can Fade Track Elements!(execute=@n9.then_5, self) -> (then, CanFadeTrackElements:bool) {is=title K2Node_CallFunction}
   > Branch(Condition=@n12)
   > PlayAnimationForward(InAnimation=Get FadeInGameplayBackground() -> (FadeInGameplayBackground:Widget Animation Object Reference))
   n13 = StartListeningToEvent!(
    execute:Exec=@n9.then_7
    Target:Gameplay Event Router Component Object Reference=Get Event Router() -> (Event Router:Gameplay Event Router Component Object Reference)
    EventType:Script Struct Object Reference=/Script/PilgrimSettingsRuntime.PilgrimEvent_PilgrimSettings_EnableMobileExpandedTrackChanged
    Context:Object Reference
    EventBubblingRule:EEventBubblingRule Enum
   ) -> (
    then:Exec
    AsyncTaskProxy:Async Action Start Listening To Event Object Reference
    OnEventReceived:Exec
    Payload:Pilgrim Event Pilgrim Settings Enable Mobile Expanded Track Changed Structure
   )
   > Add(
    TargetArray:Array of Async Action Start Listening To Event Object References=Get GameplayEventHandles() -> (GameplayEventHandles:Array of Async Action Start Listening To Event Object References)
    NewItem:"Async Action Start Listening To Event Object Reference (by ref)"=@n13
   )
   > n14 = StartListeningToEvent(Target=Get Event Router(), EventType=/Script/PilgrimSettingsRuntime.PilgrimEvent_PilgrimSettings_UseGuitarLeftyFlipChanged) -> (Payload:Pilgrim Event Pilgrim Settings Use Guitar Lefty Flip Changed Structure)
   > Add(TargetArray=Get GameplayEventHandles(), NewItem=@n14)
   > SetLetfyFlipInputZones!(execute, self, ShouldReverse:bool=GetIsUsingLeftyFlip!() -> (IsUsingLeftyFlip:bool)) -> (then)
   SetLetfyFlipInputZones(
    execute=@n14.OnEventReceived
    ShouldReverse=Break Pilgrim Event Pilgrim Settings Use Guitar Lefty Flip Changed!(PilgrimEvent_PilgrimSettings_UseGuitarLeftyFlipChanged:"Pilgrim Event Pilgrim Settings Use Guitar Lefty Flip Changed Structure (by ref)"=@n14.Payload) -> (bUseGuitarLeftyFlip:bool=~) {is=title K2Node_BreakStruct}
   )
   SetupRenderTarget(execute=@n13.OnEventReceived)
   Branch(execute=@n9.then_8, Condition=IsMobileGame())
   > SetVisibility(self=Get MobileTouchPanel() -> (MobileTouchPanel:Overlay Object Reference), InVisibility=SelfHitTestInvisible)
   n15 = GetLocalPilgrimGameForPlayer(self=Get MusicBattleInstance(), InPC=GetOwningPlayer())
   > Set IsUsing5Lanes(IsUsing5Lanes:bool="Equal (Integer)"(A=GetNumLanes!(self:Pilgrim Game Object Reference=@n15) -> (ReturnValue:int=0), B=5)) -> (Output_Get:bool)
   > SetVisibility(
    self=Get WBP_PolygonalInputZoneController_4Lanes() -> (WBP_PolygonalInputZoneController_4Lanes:WBP Polygonal Input Zone Controller Object Reference)
    InVisibility=Select(Option 0:ESlateVisibility Enum=Visible, Option 1:ESlateVisibility Enum=Collapsed, Index=Get IsUsing5Lanes() -> (IsUsing5Lanes:bool)) -> (ReturnValue:ESlateVisibility Enum)
   )
   > SetVisibility(
    self=Get WBP_PolygonalInputZoneController_5Lanes() -> (WBP_PolygonalInputZoneController_5Lanes:WBP Polygonal Input Zone Controller Object Reference)
    InVisibility=Select(Option 0=Collapsed, Option 1=Visible, Index=Get IsUsing5Lanes())
   )
   Set IsInitializing(execute=@n9.then_4, IsInitializing:bool=true) -> (Output_Get:bool)
   > n16 = Sequence!(execute) -> (then_0, then_1, then_2)
   > n17 = WaitGameplayTagAddToActor!(
    execute:Exec
    TargetActor:Actor Object Reference=GetOwningPlayer()
    Tag:Gameplay Tag Structure=(TagName="Pilgrim.Player.AllowEmoteDuringGemBreak")
    OnlyTriggerOnce:bool=false
   ) -> (then:Exec, AsyncTaskProxy:Ability Async Wait Gameplay Tag Added Object Reference, Added:Exec) {is=title K2Node_AsyncAction}
   > Add(
    TargetArray:Array of Cancellable Async Action Object References=Get AsyncEventHandles() -> (AsyncEventHandles:Array of Cancellable Async Action Object References)
    NewItem:"Cancellable Async Action Object Reference (by ref)"=@n17
   )
   n18 = Can Fade Track Elements(execute=@n17.Added)
   n19 = NOT Boolean(A=Get IsInitializing() -> (IsInitializing:bool))
   > Branch(Condition=AND Boolean(A=@n18, B=@n19))
   > PlayAnimationReverse(InAnimation=Get FadeInGameplayBackground())
   Set IsInitializing(execute=@n16.then_2, IsInitializing=false)
   n20 = WaitGameplayTagRemoveFromActor!(
    execute:Exec=@n16.then_1
    TargetActor:Actor Object Reference=GetOwningPlayer()
    Tag:Gameplay Tag Structure=(TagName="Pilgrim.Player.AllowEmoteDuringGemBreak")
    OnlyTriggerOnce:bool=false
   ) -> (then:Exec, AsyncTaskProxy:Ability Async Wait Gameplay Tag Removed Object Reference, Removed:Exec) {is=title K2Node_AsyncAction}
   > Add(TargetArray=Get AsyncEventHandles(), NewItem=@n20)
   Branch(execute=@n20.Removed, Condition=@n19)
   > PlayAnimationForward(InAnimation=Get FadeInGameplayBackground())
   n21 = PilgrimWaitForEventRouter!(execute:Exec=@n9.then_6, ContextActor:Actor Object Reference=GetOwningPlayer()) -> (then:Exec, OnEventRouterAvailable:Exec, EventRouter:Gameplay Event Router Component Object Reference) {is=title K2Node_AsyncAction}
   n22 = Set Event Router(execute=@n21.OnEventRouterAvailable, Event Router:Gameplay Event Router Component Object Reference=@n21) -> (Output_Get:Gameplay Event Router Component Object Reference)
   > n23 = StartListeningToEvent(Target=@n22, EventType=/Script/PilgrimCoreRuntime.PilgrimEvent_PilgrimGame_SongStopped) -> (Payload:Pilgrim Event Pilgrim Game Song Stopped Structure)
   > Add(
    TargetArray:Array of Async Action Start Listening To Event Object References=Get GameplayEventHandles()
    NewItem:"Async Action Start Listening To Event Object Reference (by ref)"=@n23
   )
   PlayAnimationReverse(execute=@n23.OnEventReceived, InAnimation=Get FadeInGameplayBackground())
  ]}
 GetIsUsingLeftyFlip {nodes [
   GetIsUsingLeftyFlip() {is=title K2Node_FunctionEntry}
   n1 = GetComponentByClass!(
    self:Actor Object Reference=GetOwningPlayer()
    ComponentClass:Actor Component Class Reference=/PilgrimCore/Player/Components/BP_PilgrimPlayerControllerComponent.BP_PilgrimPlayerControllerComponent_C
   ) -> (ReturnValue:BP Pilgrim Player Controller Component Object Reference)
   > Is Valid(InputObject=@n1)
   > Return Node!(
    execute
    IsUsingLeftyFlip:bool=GetIsGuitarLeftyFlip!(self:Pilgrim Client Setting Record Partition Object Reference=GetPilgrimClientSettingRecordPartition!(self:Pilgrim Player Controller Component Object Reference=@n1) -> (ReturnValue:Pilgrim Client Setting Record Partition Object Reference)) -> (ReturnValue:bool)
   )
  ]}
 SetLetfyFlipInputZones {nodes [
   n1 = SetLetfyFlipInputZones!() -> (then, ShouldReverse:bool=~) {is=title K2Node_FunctionEntry}
   > SetReverseActionsInInputZones!(
    execute
    self:Polygonal Input Zone Controller Base Object Reference=[Get WBP_PolygonalInputZoneController_4Lanes(), Get WBP_PolygonalInputZoneController_5Lanes()]
    ShouldReverse:bool=@n1
   ) -> (then)
  ]}
 SetOverdriveActionTextForPlayer {nodes [
   SetOverdriveActionTextForPlayer!() -> (then, PlayerController:Player Controller Object Reference, Text:Text) {is=title K2Node_FunctionEntry}
   > n1 = For Each Loop with Break!(Exec:Exec, Array:Array of WBP Pilgrim Quickplay Per Track Widgets Object References=Get AllTrackWidgets(), Break:Exec=@n2.then) -> (LoopBody:Exec, Array Element:WBP Pilgrim Quickplay Per Track Widgets Object Reference, Array Index:int, Completed:Exec)
   > Branch(Condition="Equal (Object)"(A=GetOwningPlayer(self=@n1), B=Get PlayerController() -> (PlayerController:Player Controller Object Reference)))
   > n2 = Set Action Text!(
    execute
    self:WBP Pilgrim Quickplay Overdrive Action Prompt Object Reference=Get WBP_PilgrimQuickplay_OverdriveActionPrompt!(self:WBP Pilgrim Quickplay Per Track Widgets Object Reference=@n1) -> (WBP_PilgrimQuickplay_OverdriveActionPrompt:WBP Pilgrim Quickplay Overdrive Action Prompt Object Reference)
    InText:Text=Get Text!() -> (Text)
   ) -> (then) {is=title K2Node_CallFunction}
  ]}
 SetupRenderTarget {nodes [
   SetupRenderTarget() {is=title K2Node_FunctionEntry}
   > n1 = Set TrackAndGameManager(TrackAndGameManager=GetTrackAndGameManager(self=Get MusicBattleInstance()))
   > n2 = Is Valid(InputObject=@n1)
   n3 = Get TrackAndGameManager()
   > n4 = GetTrackCaptureFade!(
    execute
    self:Pilgrim Track Capture Component Object Reference=Get CaptureComp!(self:Pilgrim Track And Game Manager Component Object Reference=@n3) -> (CaptureComp:Pilgrim Track Capture Component Object Reference)
    InNumTracks:int=Length(TargetArray:Array of Pilgrim Track Object References=Get Tracks(self=@n3))
   ) -> (then, OutFade:Pilgrim Track Capture Fade Structure)
   > n5 = GetRenderTarget!(execute, self:Pilgrim Track And Game Manager Component Object Reference=Get TrackAndGameManager()) -> (then, ReturnValue:Texture Render Target 2D Object Reference)
   > n6 = Is Valid(InputObject=@n5)
   PrintString(execute=[@n6.Is Not Valid, @n2.Is Not Valid], InString=Unable to set up Pilgrim Render Target)
   SetTextureParameterValue(execute=@n6.Is Valid, self=Get Material() -> (Material:Material Instance Dynamic Object Reference), ParameterName=RenderTarget, Value=@n5)
   > n7 = Sequence!(execute) -> (then_0, then_1, then_2, then_3:Exec)
   n8 = Break Pilgrim Track Capture Fade!(PilgrimTrackCaptureFade:"Pilgrim Track Capture Fade Structure (by ref)"=@n4) -> (
    SafeZoneMargin:Pilgrim Track Capture Safe Zone Margin Structure
    Crop:Pilgrim Track Capture Box Structure
    bHasCropAspectRatioOverride:bool=~
    CropAspectRatioOverride:float=~
    EdgeLengths:Pilgrim Track Capture Box Structure
    InnerBounds:Pilgrim Track Capture Box Structure
    OuterBounds:Pilgrim Track Capture Box Structure
   ) {is=title K2Node_BreakStruct}
   n9 = Break Pilgrim Track Capture Box!(PilgrimTrackCaptureBox:"Pilgrim Track Capture Box Structure (by ref)"=@n8.EdgeLengths) -> (LeftTop:Vector 2D Structure, RightBottom:Vector 2D Structure) {is=title K2Node_BreakStruct}
   n10 = BreakVector2D(InVec=@n9)
   n11 = BreakVector2D(InVec=@n9.RightBottom)
   > SetVectorParameterValue(
    self=Get Material()
    ParameterName=FadeLength
    Value=Make Linear Color!(R:float=@n10, G:float=@n10.Y, B:float=@n11, A:float=@n11.Y) -> (LinearColor:Linear Color Structure) {is=title K2Node_MakeStruct}
   )
   n12 = Break Pilgrim Track Capture Box(PilgrimTrackCaptureBox=@n8.InnerBounds)
   n13 = BreakVector2D(InVec=@n12)
   n14 = BreakVector2D(InVec=@n12.RightBottom)
   SetVectorParameterValue(execute=@n7.then_1, self=Get Material(), ParameterName=InnerFadeBounds, Value=Make Linear Color(R=@n13, G=@n13.Y, B=@n14, A=@n14.Y))
   n15 = Break Pilgrim Track Capture Box(PilgrimTrackCaptureBox=@n8.OuterBounds)
   n16 = BreakVector2D(InVec=@n15)
   n17 = BreakVector2D(InVec=@n15.RightBottom)
   SetVectorParameterValue(execute=@n7.then_2, self=Get Material(), ParameterName=OuterFadeBounds, Value=Make Linear Color(R=@n16, G=@n16.Y, B=@n17, A=@n17.Y))
   n18 = BreakVector4!(InVec:"Vector 4 Structure (by ref)"=GetRenderTargetToSafeZoneUVScaleOffset!(self:Pilgrim Track And Game Manager Component Object Reference=Get TrackAndGameManager()) -> (ReturnValue:Vector 4f Structure)) -> (X:double, Y:double, Z:double, W:double)
   SetVectorParameterValue(execute=@n7.then_3, self=Get Material(), ParameterName=UVScaleOffset, Value=Make Linear Color(R=@n18, G=@n18.Y, B=@n18.Z, A=@n18.W))
   > SetBrushFromMaterial(self=Get TrackRenderTarget(), Material=Get Material())
   n19 = BreakVector4(InVec=GetRenderTargetToSquareUVScaleOffset!(self:Pilgrim Track And Game Manager Component Object Reference=Get TrackAndGameManager()) -> (ReturnValue:Vector 4f Structure))
   > SetRenderScale!(
    execute
    self:Widget Object Reference=[Get TrackRenderTarget(), Get MobileTouchPanel()]
    Scale:Vector 2D Structure=MakeVector2D(X=@n19, Y=@n19.Y)
   ) -> (then)
   n20 = Get Brush!(self:Image Object Reference=Get TrackRenderTarget()) -> (Brush:Slate Brush Structure)
   > SetRenderTranslation(
    self=[Get MobileTouchPanel(), Get TrackRenderTarget()]
    Translation=vector2d * vector2d(
     A:Vector 2D Structure=Break Slate Brush!(SlateBrush:"Slate Brush Structure (by ref)"=@n20) -> (
      TintColor:Slate Color Structure
      DrawAs:ESlateBrushDrawType Enum
      Tiling:ESlateBrushTileType Enum
      Mirroring:ESlateBrushMirrorType Enum
      ImageSize:Vector 2D Structure
      Margin:Margin Structure
      ResourceObject:Object Reference
      OutlineSettings:Slate Brush Outline Settings Structure
     ) {is=title K2Node_BreakStruct}.ImageSize
     B:Vector 2D Structure=MakeVector2D(X=@n19.Z, Y=@n19.W)
    ) -> (ReturnValue:Vector 2D Structure) {is=title K2Node_PromotableOperator}
   )
   > n21 = For Each Loop(Array:Array of WBP Pilgrim Quickplay Per Track Widgets Object References=Get AllTrackWidgets()) -> (Array Element:WBP Pilgrim Quickplay Per Track Widgets Object Reference)
   > SetPilgrimTrackCapture!(
    execute
    self:WBP Pilgrim Quickplay Per Track Widgets Object Reference=@n21
    Pilgrim Track Capture:Pilgrim Track Capture Component Object Reference=Get CaptureComp(self=Get TrackAndGameManager())
   ) -> (then)
  ]}
 ShouldAddPerTrackWidgets {nodes [
   n1 = ShouldAddPerTrackWidgets!() -> (then, TrackIndex:int, GameManager:"Pilgrim Track And Game Manager Component Object Reference (by ref)") {is=title K2Node_FunctionEntry}
   n2 = "Get (a copy)"(
    Array:Array of Pilgrim Game Object References=Get PilgrimGames!(self:Pilgrim Track And Game Manager Component Object Reference=@n1.GameManager) -> (PilgrimGames:Array of Pilgrim Game Object References)
    Dimension 1=@n1
   ) -> (Output:Pilgrim Game Object Reference)
   > Return Node!(
    execute
    ShouldAdd:bool="Not Equal (Enum)"!(A:EPilgrimTrackType Enum=GetSelectedTrackType(self=@n2) -> (ReturnValue=TrackGuitar), B:EPilgrimTrackType Enum=TrackProVocals) -> (ReturnValue:bool)
   )
  ]}
]
variables [
 AllTrackWidgets {type="array<WBP_PilgrimQuickplay_PerTrackWidgets_C>", default []}
 AsyncEventHandles {type="array<CancellableAsyncAction>", default []}
 Event Router {type=GameplayEventRouterComponent, default=null}
 FadeLength {type=LinearColor, default=<rgba 0.0157 0 0.0157 0>}
 IsInitializing {type=boolean}
 isMBIInitialized {type=boolean}
 IsUsing5Lanes {type=boolean}
 LayoutDefinitions {type=PilgrimTrackCaptureLayoutDefinitions, default=/PilgrimCore/GemAndTrack/Track/PilgrimTrackCaptureLayoutDefinitions}
 Material {type=MaterialInstanceDynamic}
 PerTrackWidgetPreview {type=WBP_PilgrimQuickplay_PerTrackWidgets_C, default=null}
 Preview4LaneUI {type=boolean}
 PreviewPerTrackWidgets {type=boolean, default=true}
 RenderTargetMaterial {type=MaterialInterface, default=/PilgrimCore/Cameras/M_PilgrimTrackProjection}
 WidgetsMobilePaddingOffsets {
  type="array<WidgetMobileMarginOffset>"
  default [
   {widget=null, margin=<ltrb 0 0 0 0>}
   {widget=null, margin=<ltrb 0 0 0 0>}
   {widget=null, margin=<ltrb 0 0 0 0>}
   {widget=null, margin=<ltrb 0 0 0 0>}
  ]
 }
 WidgetsWithFiveLaneOffsets {
  type="array<WidgetFiveLaneOffset>"
  default [
   {widget=null, fiveLaneTransformOffset=<22 0>}
   {widget=null, fiveLaneTransformOffset=<22 0>}
   {widget=null, fiveLaneTransformOffset=<54 0>}
   {widget=null, fiveLaneTransformOffset=<41 0>}
  ]
 }
]
widgets {
 info {inheritedWidgetCount=0, parentClass=/Script/PilgrimCoreUI.PilgrimGameWidgetBase, rootWidgetClass=&Overlay, widgetCount=11}
 widgets [
  Overlay_0 {widgetClassPath=&Overlay, widget=Overlay_0}
  WBP_Pilgrim_TrackBackground {
   widgetClassPath=/PilgrimCore/Widgets/WBP_Pilgrim_TrackBackground
   parent=Overlay_0
   widget=WBP_Pilgrim_TrackBackground
   slot {class=OverlaySlot, properties {horizontalAlignment=HAlign_Fill, verticalAlignment=VAlign_Fill}}
  }
  ScaleBox_0 {
   widgetClassPath=&ScaleBox
   parent=Overlay_0
   widget=ScaleBox_0
   slot {class=OverlaySlot, properties {horizontalAlignment=HAlign_Fill, verticalAlignment=VAlign_Fill}}
  }
  SZ_MobileOnly {
   widgetClassPath=&SafeZone
   bIsVariable=true
   parent=ScaleBox_0
   widget=SZ_MobileOnly
   properties {padBottom=false, padLeft=false, padRight=false, padTop=false}
   slot {class=ScaleBoxSlot, properties {verticalAlignment=VAlign_Bottom}}
  }
  Track_Overlay {widgetClassPath=&Overlay, bIsVariable=true, parent=SZ_MobileOnly, widget=Track_Overlay, slot {class=SafeZoneSlot}}
  TrackRenderTarget {
   widgetClassPath=&Image
   bIsVariable=true
   parent=Track_Overlay
   widget=TrackRenderTarget
   properties {
    visibility=SelfHitTestInvisible
    brush {
     bIsDynamicallyLoaded=false
     drawAs=Image
     imageType=NoImage
     mirroring=NoMirror
     resourceName=None
     resourceObject=&M_PilgrimTrackProjection
     tiling=NoTile
     imageSize=<1000 1000>
     margin=<ltrb 0 0 0 0>
     outlineSettings {
      bUseBrushTransparency=false
      roundingType=HalfHeightRadius
      width=0
      color {colorUseRule=UseColor_Specified, specifiedColor=<rgba 0 0 0 0>}
      cornerRadii=<0 0 0 0>
     }
     tintColor {colorUseRule=UseColor_Specified, specifiedColor=<rgba 1 1 1 1>}
     uVRegion {bIsValid=false, max=<0 0>, min=<0 0>}
    }
   }
   slot {class=OverlaySlot, properties {horizontalAlignment=HAlign_Fill, verticalAlignment=VAlign_Fill}}
  }
  TrackRenderTarget_Proxy {
   widgetClassPath=&Image
   bIsVariable=true
   parent=Track_Overlay
   widget=TrackRenderTarget_Proxy
   properties {
    visibility=Hidden
    brush {
     bIsDynamicallyLoaded=false
     drawAs=Image
     imageType=FullColor
     mirroring=NoMirror
     resourceName=None
     resourceObject=/PilgrimQuickplay/Widgets/Textures/DevOnly/track_capture_4Lane
     tiling=NoTile
     imageSize=<610 440>
     margin=<ltrb 0 0 0 0>
     outlineSettings {
      bUseBrushTransparency=false
      roundingType=HalfHeightRadius
      width=0
      color {colorUseRule=UseColor_Specified, specifiedColor=<rgba 0 0 0 0>}
      cornerRadii=<0 0 0 0>
     }
     tintColor {colorUseRule=UseColor_Specified, specifiedColor=<rgba 1 1 1 1>}
     uVRegion {bIsValid=false, max=<0 0>, min=<0 0>}
    }
    renderTransform {angle=0, scale=<0.987 0.7>, shear=<0 0>, translation=<7.9405 150>}
   }
   slot {class=OverlaySlot, properties {horizontalAlignment=HAlign_Fill, verticalAlignment=VAlign_Fill}}
  }
  Expansion_NamedSlot {
   widgetClassPath=&NamedSlot
   bIsVariable=true
   parent=Track_Overlay
   widget=Expansion_NamedSlot
   slot {class=OverlaySlot, properties {horizontalAlignment=HAlign_Fill, verticalAlignment=VAlign_Fill}}
  }
  MobileTouchPanel {
   widgetClassPath=&Overlay
   bIsVariable=true
   parent=Track_Overlay
   widget=MobileTouchPanel
   properties {visibility=Collapsed, renderTransform {angle=0, scale=<0.987 0.7>, shear=<0 0>, translation=<7.9405 150>}}
   slot {class=OverlaySlot, properties {horizontalAlignment=HAlign_Fill, verticalAlignment=VAlign_Fill}}
  }
  WBP_PolygonalInputZoneController_4Lanes {
   widgetClassPath=/PolygonalInputZones/Assets/WBP_PolygonalInputZoneController
   bIsVariable=true
   parent=MobileTouchPanel
   widget=WBP_PolygonalInputZoneController_4Lanes
   properties {
    bUseContinuousInputInjectionForAction=true
    interactionMode=InputActions
    inputZones [
     {
      activated=true
      deactivateOnRelease=false
      priority=0
      zoneAction=/PilgrimSettings/Input/InputActions/IA_PilgrimLane1Press
      zoneId=""
      zoneWidget=null
      zoneVertices [<0.095 1>, <0.335 0>, <0.4133 0>, <0.2943 1>]
     }
     {
      activated=true
      deactivateOnRelease=false
      priority=0
      zoneAction=/PilgrimSettings/Input/InputActions/IA_PilgrimLane2Press
      zoneId=""
      zoneWidget=null
      zoneVertices [<0.2943 1>, <0.4133 0>, <0.4915 0>, <0.4935 1>]
     }
     {
      activated=true
      deactivateOnRelease=false
      priority=0
      zoneAction=/PilgrimSettings/Input/InputActions/IA_PilgrimLane3Press
      zoneId=""
      zoneWidget=null
      zoneVertices [<0.4935 1>, <0.4915 0>, <0.5697 0>, <0.6927 1>]
     }
     {
      activated=true
      deactivateOnRelease=false
      priority=0
      zoneAction=/PilgrimSettings/Input/InputActions/IA_PilgrimLane4Press
      zoneId=""
      zoneWidget=null
      zoneVertices [<0.6927 1>, <0.5697 0>, <0.648 0>, <0.892 1>]
     }
    ]
   }
   slot {class=OverlaySlot, properties {horizontalAlignment=HAlign_Fill, verticalAlignment=VAlign_Fill}}
  }
  WBP_PolygonalInputZoneController_5Lanes {
   widgetClassPath=&WBP_PolygonalInputZoneController
   bIsVariable=true
   parent=MobileTouchPanel
   widget=WBP_PolygonalInputZoneController_5Lanes
   properties {
    bUseContinuousInputInjectionForAction=true
    interactionMode=InputActions
    inputZones [
     {
      activated=true
      deactivateOnRelease=false
      priority=0
      zoneAction=&IA_PilgrimLane1Press
      zoneId=""
      zoneWidget=null
      zoneVertices [<0.012 1>, <0.305 0>, <0.381 0>, <0.2066 1>]
     }
     {
      activated=true
      deactivateOnRelease=false
      priority=0
      zoneAction=&IA_PilgrimLane2Press
      zoneId=""
      zoneWidget=null
      zoneVertices [<0.2066 1>, <0.381 0>, <0.457 0>, <0.4012 1>]
     }
     {
      activated=true
      deactivateOnRelease=false
      priority=0
      zoneAction=&IA_PilgrimLane3Press
      zoneId=""
      zoneWidget=null
      zoneVertices [<0.4012 1>, <0.457 0>, <0.533 0>, <0.5958 1>]
     }
     {
      activated=true
      deactivateOnRelease=false
      priority=0
      zoneAction=&IA_PilgrimLane4Press
      zoneId=""
      zoneWidget=null
      zoneVertices [<0.5958 1>, <0.533 0>, <0.609 0>, <0.7904 1>]
     }
     {
      activated=true
      deactivateOnRelease=false
      priority=0
      zoneAction=/PilgrimSettings/Input/InputActions/IA_PilgrimLane5Press_5Lane
      zoneId=""
      zoneWidget=null
      zoneVertices [<0.7904 1>, <0.609 0>, <0.685 0>, <0.985 1>]
     }
    ]
   }
   slot {class=OverlaySlot, properties {horizontalAlignment=HAlign_Fill, verticalAlignment=VAlign_Fill}}
  }
 ]
}
