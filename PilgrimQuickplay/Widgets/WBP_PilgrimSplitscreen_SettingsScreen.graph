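# Pilgrim Quickplay splitscreen settings screen widget, wrapping the Fortnite SettingScreen with a player identity pill in its named slot.
# PushSettingScreen casts the owning player to FortPlayerController and pushes the settings screen as that player's content widget.
# It then navigates to the four-lane gamepad preset under controller bindings.
# Finally it sets the pill from that player's PlayerState and shows it, identifying whose settings are open.
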WidgetBlueprint /PilgrimQuickplay/Widgets/WBP_PilgrimSplitscreen_SettingsScreen
graphs [EventGraph {nodes [
   PushSettingScreen() {is=title K2Node_CustomEvent}
   > n1 = Cast To FortPlayerController(Object=GetOwningPlayer())
   n2 = Get SettingScreen() -> (SettingScreen:Setting Screen Object Reference)
   > PushContentWidgetForPlayer!(execute, LocalPlayerController:Fort Player Controller Object Reference=@n1, PushWidget:Widget Object Reference=@n2) -> (then)
   > NavigateToSetting!(
    execute
    self:Fort Settings Screen Object Reference=@n2
    SettingCategory:Name=ControllerBindingsCollection
    SettingToSelect:String=Pilgrim_Preset_4Lanes_Gamepad
   ) -> (then)
   n3 = Get WBP_Pilgrim_PlayerIdentity_Pill() -> (WBP_Pilgrim_PlayerIdentity_Pill:WBP Pilgrim Player Identity Pill Object Reference)
   > OverridePlayerInfo!(
    execute
    self:WBP Pilgrim Player Identity Pill Object Reference=@n3
    PlayerState:Player State Object Reference=Get PlayerState!(self:Controller Object Reference=GetOwningPlayer()) -> (PlayerState:Player State Object Reference)
   ) -> (then)
   > ShowPlayerDisplay(self:WBP Pilgrim Player Identity Pill Object Reference=@n3)
   Event PreConstruct()
   Event Tick()
  ]}]
variables []
widgets {
 info {
  inheritedWidgetCount=0
  namedSlotCount=1
  parentClass=/Script/PilgrimQuickplayUI.PilgrimSplitscreenSettingScreen
  rootWidgetClass=&Overlay
  widgetCount=4
 }
 widgets [
  Overlay_274 {widgetClassPath=&Overlay, widget=Overlay_274}
  SettingScreen {
   widgetClassPath=/FortniteCoreUI/Settings/SettingScreen
   bIsVariable=true
   parent=Overlay_274
   widget=SettingScreen
   slot {class=OverlaySlot, properties {horizontalAlignment=HAlign_Fill, verticalAlignment=VAlign_Fill}}
  }
  Overlay_72 {widgetClassPath=&Overlay, namedSlotHost=SettingScreen, widget=Overlay_72}
  WBP_Pilgrim_PlayerIdentity_Pill {
   widgetClassPath=/PilgrimCore/Widgets/Player/WBP_Pilgrim_PlayerIdentity_Pill
   bIsVariable=true
   parent=Overlay_72
   widget=WBP_Pilgrim_PlayerIdentity_Pill
   properties {visibility=Visible}
   slot {class=OverlaySlot, properties {horizontalAlignment=HAlign_Right, padding=<ltrb 0 24 80 0>}}
  }
 ]
}
