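# Translucent UI material for the Pilgrim instrument intensity widget, drawing a circular indicator shape.
# MF_UI_Circle supplies fill and outline signed distance fields, which are each inverted and saturated.
# A static switch named Filled, on by default, picks the solid disc or the ring.
# A narrow SmoothStep then sharpens the chosen field into a crisp-edged mask.
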Material /PilgrimCore/Widgets/Materials/M_UI_Pilgrim_InstrumentIntensity
graph {
 properties {
  bUsedWithStaticMesh=false
  blendMode=BLEND_Translucent
  materialDomain=MD_UI
  pixelDepthOffsetMode=PDOM_Legacy
  usedShadingModels=Default Lit
 }
 nodes [
  Constant() {r=1}
  n1 = MaterialFunctionCall(Size=Constant() {r=0.9}) {materialFunction=&MF_UI_Circle}
  n2 = OneMinus(None=@n1.SDF)
  n3 = OneMinus(None=@n1.SDFOutline)
  SmoothStep(Value=StaticSwitchParameter(True=Saturate(None=@n2), False=Saturate(None=@n3)) {defaultValue=true, parameterName=Filled}) {constMin=0.95}
 ]
}
