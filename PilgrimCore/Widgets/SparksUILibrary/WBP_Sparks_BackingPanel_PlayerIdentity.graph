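# Player identity variant of the Sparks UI backing panel, a child widget of WBP_Sparks_BackingPanel that adds no widgets or variables of its own.
# Its SetBackingColor function writes the given color to the Color parameter on the Fill image's dynamic material.
# The same function sets Alpha_Final on the Darken image's dynamic material to a fixed partial value, independent of the input color.
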WidgetBlueprint /PilgrimCore/Widgets/SparksUILibrary/WBP_Sparks_BackingPanel_PlayerIdentity
graphs [
 EventGraph {links [], nodes []}
 SetBackingColor {nodes [
   n1 = SetBackingColor!() -> (then, Linear Color:Linear Color Structure) {is=title K2Node_FunctionEntry}
   > n2 = GetDynamicMaterial(self=Get Fill() -> (Fill:Image Object Reference))
   > SetVectorParameterValue(self=@n2, ParameterName=Color, Value=@n1)
   > n3 = GetDynamicMaterial(self=Get Darken() -> (Darken:Image Object Reference))
   > SetScalarParameterValue(self=@n3, ParameterName=Alpha_Final, Value=0.25)
  ]}
]
variables []
widgets {
 info {
  inheritedWidgetCount=0
  parentClass=&/WBP_Sparks_BackingPanel.WBP_Sparks_BackingPanel_C
  rootWidgetClass=null
  widgetCount=0
 }
 widgets []
}
