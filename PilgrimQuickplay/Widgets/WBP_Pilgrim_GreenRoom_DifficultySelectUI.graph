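# Green room difficulty selector button for Pilgrim Quickplay that shows a difficulty label and pips, with hover and press animations.
# Clicks forward swap-difficulty input to the player controller component.
# On Construct, async listeners drive keybind visibility from a gameplay tag, widget visibility from timer and quickplay state, and pips from the playspace's supported difficulties.
# Exposed functions set player color, highlight and widen the selected pip, or cycle difficulty with wraparound.
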WidgetBlueprint /PilgrimQuickplay/Widgets/WBP_Pilgrim_GreenRoom_DifficultySelectUI
animations [
 Hovered {
  binding_count=2
  length_seconds=0.2
  object=&self:Hovered
  sequence {bindings [
    SelectionBackingIndicator {
     id=#1
     possessed_class=Image
     tracks [{
       type=MovieSceneFloatTrack
       display_name=Render Opacity
       sections [{type=MovieSceneFloatSection, end_frame=4, start_frame=0, channels [None {type=MovieSceneScriptingFloatChannel, keys [[0,0],[4,1]]}]}]
      }]
    }
    Overlay_1 {
     id=#2
     possessed_class=Overlay
     tracks [{
       type=MovieScene2DTransformTrack
       sections [{
         type=MovieScene2DTransformSection
         channels [
          Scale.X {type=MovieSceneScriptingFloatChannel, keys [[0,1],[4,1.05]]}
          Scale.Y {type=MovieSceneScriptingFloatChannel, keys [[0,1],[4,1.05]]}
         ]
        }]
      }]
    }
   ]}
 }
 Pressed {
  binding_count=1
  length_seconds=0.2
  object=&self:Pressed
  sequence {bindings [WBP_Pilgrim_GreenRoom_DifficultySelectUI_C_51 {
     id=#3
     possessed_class=WBP_Pilgrim_GreenRoom_DifficultySelectUI_C
     tracks [{
       type=MovieScene2DTransformTrack
       sections [{
         type=MovieScene2DTransformSection
         channels [
          Scale.X {type=MovieSceneScriptingFloatChannel, keys [[0,1],[2,0.975],[4,1]]}
          Scale.Y {type=MovieSceneScriptingFloatChannel, keys [[0,1],[2,0.975],[4,1]]}
         ]
        }]
      }]
    }]}
 }
]
graphs [
 EventGraph {nodes [
   "On Clicked (WBP_UIKit_Button_Regular)"!() -> (OutputDelegate, then, Button:Common Button Base Object Reference) {is=title K2Node_ComponentBoundEvent}
   n1 = Get PlayerState!(self:Controller Object Reference=GetOwningPlayer()) -> (PlayerState:Player State Object Reference)
   > Swap Difficulty Input Pressed(self:BP Pilgrim Player Controller Component Object Reference=GetComponentByClass!(
     self:Actor Object Reference=GetPlayerController!(self:Player State Object Reference=@n1) -> (ReturnValue:Player Controller Object Reference)
     ComponentClass:Actor Component Class Reference=/PilgrimCore/Player/Components/BP_PilgrimPlayerControllerComponent.BP_PilgrimPlayerControllerComponent_C
    ) -> (ReturnValue:BP Pilgrim Player Controller Component Object Reference))
   > PlayAnimation(InAnimation=Get Pressed() -> (Pressed:Widget Animation Object Reference))
   "On Hovered (WBP_UIKit_Button_Regular)"!() -> (OutputDelegate, then, Button:Common Button Base Object Reference) {is=title K2Node_ComponentBoundEvent}
   > OnHoverVisuals()
   "On Unhovered (WBP_UIKit_Button_Regular)"!() -> (OutputDelegate, then, Button:Common Button Base Object Reference) {is=title K2Node_ComponentBoundEvent}
   > OnUnhoverVisuals()
   Event Construct()
   > SetVisibility(
    self=Get WBP_Pilgrim_Results_Difficulty() -> (WBP_Pilgrim_Results_Difficulty:WBP Pilgrim Results Difficulty Object Reference)
    InVisibility=Select(Option 0=Collapsed, Option 1=SelfHitTestInvisible, Index=Get ShowDifficultyIcon() -> (ShowDifficultyIcon:bool))
   )
   > n2 = Sequence!(execute) -> (then_0, then_1, then_2:Exec, then_3:Exec, then_4:Exec)
   > n3 = SetVisibility(
    self=[
     Get Image_medium() -> (Image_medium:Image Object Reference)
     Get Image_easy() -> (Image_easy:Image Object Reference)
     Get Image_expert() -> (Image_expert:Image Object Reference)
     Get Image_hard() -> (Image_hard:Image Object Reference)
    ]
    InVisibility=Collapsed
   )
   n4 = GetOwningPlayer()
   n5 = HasMatchingGameplayTag!(
    execute:Exec=[@n8.TagCountChanged, @n3.then]
    self:Object Reference=GetAbilitySystemComponent!(Actor:Actor Object Reference=Get PlayerState(self=@n4)) -> (ReturnValue:Ability System Component Object Reference)
    TagToCheck:Gameplay Tag Structure=(TagName="Pilgrim.GreenRoom.Input.SwapDifficulty")
   ) -> (then:Exec, ReturnValue:bool) {is=title K2Node_Message}
   n6 = Get Keybind_GridPanel() -> (Keybind_GridPanel:Grid Panel Object Reference)
   n7 = AND Boolean(A=Get IsOnGreenRoom() -> (IsOnGreenRoom:bool), B=@n5)
   > SetIsEnabled(self=@n6, bInIsEnabled=@n7)
   > SetVisibility(self=@n6, InVisibility=Select(Option 0=Collapsed, Option 1=SelfHitTestInvisible, Index=@n7))
   n8 = WaitGameplayTagCountChangedOnActor!(
    execute:Exec=@n2.then_1
    TargetActor:Actor Object Reference=Get PlayerState(self=GetOwningPlayer())
    Tag:Gameplay Tag Structure=(TagName="Pilgrim.GreenRoom.Input.SwapDifficulty")
   ) -> (then:Exec, AsyncTaskProxy:Ability Async Wait Gameplay Tag Count Changed Object Reference, TagCountChanged:Exec, TagCount:int) {is=title K2Node_AsyncAction}
   > Add(
    TargetArray:Array of Cancellable Async Action Object References=Get AsyncEventHandles() -> (AsyncEventHandles:Array of Cancellable Async Action Object References)
    NewItem:"Cancellable Async Action Object Reference (by ref)"=@n8
   ) {is=title K2Node_CallArrayFunction}
   n9 = PilgrimWaitForEventRouter!(execute:Exec=@n2.then_2, ContextActor:Actor Object Reference=GetOwningPlayer()) -> (then:Exec, OnEventRouterAvailable:Exec, EventRouter:Gameplay Event Router Component Object Reference) {is=title K2Node_AsyncAction}
   n10 = StartListeningToEvent!(
    execute:Exec=@n9.OnEventRouterAvailable
    Target:Gameplay Event Router Component Object Reference=@n9
    EventType:Script Struct Object Reference=/PilgrimQuickplay/Items/PilgrimEvent_GreenRoomTimerStateChanged
    Context:Object Reference
    EventBubblingRule:EEventBubblingRule Enum
   ) -> (
    then:Exec
    AsyncTaskProxy:Async Action Start Listening To Event Object Reference
    OnEventReceived:Exec
    Payload:Pilgrim Event Green Room Timer State Changed Structure
   )
   > Add(
    TargetArray:Array of Async Action Start Listening To Event Object References=Get GameplayEventHandles() -> (GameplayEventHandles:Array of Async Action Start Listening To Event Object References)
    NewItem:"Async Action Start Listening To Event Object Reference (by ref)"=@n10
   )
   n11 = Branch(
    execute=@n10.OnEventReceived
    Condition="Equal (Enum)"(
     A:GreenRoomTimerState Enum=Break Pilgrim Event Green Room Timer State Changed!(PilgrimEvent_GreenRoomTimerStateChanged:"Pilgrim Event Green Room Timer State Changed Structure (by ref)"=@n10.Payload) -> (
      GreenRoomTimerState_2_#4:GreenRoomTimerState Enum
      FinalTimerLength_5_#5:double=~
     ) {is=title K2Node_BreakStruct}
     B:GreenRoomTimerState Enum=NewEnumerator2
    )
   )
   > SetVisibility(InVisibility=Collapsed)
   SetVisibility(execute=@n11.else, InVisibility=SelfHitTestInvisible)
   n12 = PlaylistDataReadyAsync!(execute:Exec=@n2.then_3) -> (
    then:Exec
    OnReady:Exec
    GameState:Fort Game State Athena Object Reference
    Playlist:Fort Playlist Object Reference
    PlaylistContextTags:"Gameplay Tag Container Structure (by ref)"
   ) {is=title K2Node_AsyncAction}
   n13 = PilgrimWaitForEventRouter(execute=@n12.OnReady, ContextActor=GetOwningPlayer())
   n14 = StartListeningToEvent!(
    execute=@n13.OnEventRouterAvailable
    Target=@n13
    EventType=/Script/PilgrimQPSharedRuntime.PilgrimEvent_QuickplayState_Began
    Context
    EventBubblingRule
    None:%=@n13
   ) -> (then, AsyncTaskProxy, OnEventReceived, Payload_StateId:EPilgrimQuickplayState Enum)
   > Add(TargetArray=Get GameplayEventHandles(), NewItem=@n14)
   n15 = Branch(
    execute=@n14.OnEventReceived
    Condition=OR Boolean!(
     A="Equal (Enum)"(A:EPilgrimQuickplayState Enum=@n14.Payload_StateId, B:EPilgrimQuickplayState Enum=Pregame)
     B="Equal (Enum)"(A=@n14.Payload_StateId, B=SongResults)
     C:bool="Equal (Enum)"(A=@n14.Payload_StateId, B=SetResults)
    ) -> (ReturnValue)
   )
   > SetVisibility(InVisibility=Visible)
   SetVisibility(execute=@n15.else, InVisibility=Hidden)
   n16 = WaitForDiscoverablePlayspaceOfTypeForActor!(
    execute:Exec=@n2.then_4
    PlayspaceClass:Playspace Soft Class Reference=/Script/PilgrimCoreRuntime.PilgrimPlayspace
    Actor:Actor Object Reference=GetOwningPlayer()
    Requester:Object Reference
   ) -> (
    then:Exec
    AsyncTaskProxy:Wait For Discoverable Playspace With Filter Action Object Reference
    OnFoundPlayspace:Exec
    Playspace:Playspace Object Reference
   ) {is=title K2Node_AsyncAction}
   n17 = Cast To PilgrimPlayspace!(execute=@n16.OnFoundPlayspace, Object=@n16.Playspace) -> (then, CastFailed, AsPilgrim Playspace:Pilgrim Playspace Object Reference) {is=title K2Node_DynamicCast}
   > n18 = Set PilgrimPlayspace(PilgrimPlayspace:Pilgrim Playspace Object Reference=@n17) -> (Output_Get:Pilgrim Playspace Object Reference)
   > n19 = For Each Loop(Array:Array of EPilgrimSongDifficulty Enums=Get SupportedDifficulties!(self:Pilgrim Supported Gameplay Object Reference=Get PilgrimSupportedGameplayConfig!(self:Pilgrim Playspace Object Reference=@n18) -> (PilgrimSupportedGameplayConfig:Pilgrim Supported Gameplay Object Reference)) -> (SupportedDifficulties:Array of EPilgrimSongDifficulty Enums=DifficultyEasy)) -> (Array Element:EPilgrimSongDifficulty Enum)
   > n20 = Switch on EPilgrimSongDifficulty!(execute, Selection:EPilgrimSongDifficulty Enum=@n19) -> (DifficultyEasy:Exec, DifficultyMedium:Exec, DifficultyHard:Exec, DifficultyExpert:Exec, None:Exec) {is=title K2Node_SwitchEnum}
   > SetVisibility(self=Get Image_easy(), InVisibility=SelfHitTestInvisible)
   SetVisibility(execute=@n20.DifficultyExpert, self=Get Image_expert(), InVisibility=SelfHitTestInvisible)
   SetVisibility(execute=@n20.DifficultyHard, self=Get Image_hard(), InVisibility=SelfHitTestInvisible)
   SetVisibility(execute=@n20.DifficultyMedium, self=Get Image_medium(), InVisibility=SelfHitTestInvisible)
   FortShippingLog!(
    execute=@n17.CastFailed
    WorldContextObject:Object Reference=Self-Reference()
    InString:String=Append!(A="WaitForPlayspace: Given playspace ", B=GetDisplayName(Object=@n16.Playspace), C:String=" which doesn't derive from SparksMusicPlayspace.") -> (ReturnValue)
    bLogAsWarning:bool=true
   ) -> (then)
  ]}
 OnHoverVisuals {nodes [
   OnHoverVisuals() {is=title K2Node_FunctionEntry}
   > PlayAnimationForward(InAnimation=Get Hovered() -> (Hovered:Widget Animation Object Reference))
   > RequestTransition!(
    execute
    self:UIKit Block Object Reference=Get WBP_UIKit_Block_Outline() -> (WBP_UIKit_Block_Outline:WBP UIKit Block Outline Object Reference)
    State:EUIKitBlockVisualState Enum=Focused
   ) -> (then)
   > PlayHover(self:WBP Sparks Backing Panel Object Reference=Get WBP_Sparks_BackingPanel() -> (WBP_Sparks_BackingPanel:WBP Sparks Backing Panel Player Identity Object Reference))
   > SetVisibility(self=Get InputContainer_GridPanel() -> (InputContainer_GridPanel:Grid Panel Object Reference), InVisibility=SelfHitTestInvisible)
  ]}
 OnUnhoverVisuals {nodes [
   OnUnhoverVisuals() {is=title K2Node_FunctionEntry}
   > PlayAnimationReverse(InAnimation=Get Hovered())
   > RequestTransition(self=Get WBP_UIKit_Block_Outline(), State=Unfocused)
   > PlayUnhover(self:WBP Sparks Backing Panel Object Reference=Get WBP_Sparks_BackingPanel())
   > SetVisibility(self=Get InputContainer_GridPanel(), InVisibility=Collapsed)
  ]}
 SetDifficultyDisplay {nodes [
   n1 = SetDifficultyDisplay!() -> (then, Song Difficulty:EPilgrimSongDifficulty Enum=DifficultyEasy) {is=title K2Node_FunctionEntry}
   > n2 = Sequence()
   > n3 = Branch(Condition=Get ShowDifficultyIcon())
   > n4 = SetDifficulty!(
    execute
    self:WBP Pilgrim Results Difficulty Object Reference=Get WBP_Pilgrim_Results_Difficulty()
    Difficulty:EPilgrimSongDifficulty Enum=@n1
   ) -> (then)
   "SetText (Text)"(
    execute=[@n4.then, @n3.else]
    self=Get Common_Text_Difficulty() -> (Common_Text_Difficulty:Common Text Object Reference)
    InText=Select!(
     DifficultyEasy:Text="LOCTEXT(\"Easy\")"
     DifficultyMedium:Text="LOCTEXT(\"Medium\")"
     DifficultyHard:Text="LOCTEXT(\"Hard\")"
     DifficultyExpert:Text="LOCTEXT(\"Expert\")"
     None:Text="LOCTEXT(\"Easy\")"
     Index:EPilgrimSongDifficulty Enum=@n1
    ) -> (ReturnValue:Text)
   )
   n5 = "Equal (Enum)"(A:EPilgrimSongDifficulty Enum=@n1, B:EPilgrimSongDifficulty Enum=DifficultyExpert)
   SetRenderOpacity(execute=@n2.then_4, self=Get Image_expert(), InOpacity=Select!(Option 0:float=0.5, Option 1:float=1, Index:bool=@n5) -> (ReturnValue:float))
   > SetSize!(
    execute
    self:Spacer Object Reference=Get ExpertSpacer() -> (ExpertSpacer:Spacer Object Reference)
    InSize:Vector 2D Structure=Select(
     Option 0:Vector 2D Structure=(X=16,Y=16)
     Option 1:Vector 2D Structure=Get SelectedPipWidth() -> (SelectedPipWidth:Vector 2D Structure)
     Index=@n5
    ) -> (ReturnValue:Vector 2D Structure)
   ) -> (then)
   n6 = "Equal (Enum)"(A=@n1, B=DifficultyHard)
   SetRenderOpacity(execute=@n2.then_3, self=Get Image_hard(), InOpacity=Select(Option 0:float=0.5, Option 1:float=1, Index=@n6) -> (ReturnValue:float))
   > SetSize(
    self=Get HardSpacer() -> (HardSpacer:Spacer Object Reference)
    InSize=Select(Option 0:Vector 2D Structure=(X=16,Y=16), Option 1:Vector 2D Structure=Get SelectedPipWidth(), Index=@n6) -> (ReturnValue:Vector 2D Structure)
   )
   n7 = "Equal (Enum)"(A=@n1, B=DifficultyEasy)
   SetRenderOpacity(execute=@n2.then_1, self=Get Image_easy(), InOpacity=Select(Option 0:float=0.5, Option 1:float=1, Index=@n7) -> (ReturnValue:float))
   > SetSize(
    self=Get EasySpacer() -> (EasySpacer:Spacer Object Reference)
    InSize=Select(Option 0:Vector 2D Structure=(X=16,Y=16), Option 1:Vector 2D Structure=Get SelectedPipWidth(), Index=@n7) -> (ReturnValue:Vector 2D Structure)
   )
   n8 = "Equal (Enum)"(A=@n1, B=DifficultyMedium)
   SetRenderOpacity(execute=@n2.then_2, self=Get Image_medium(), InOpacity=Select(Option 0:float=0.5, Option 1:float=1, Index=@n8) -> (ReturnValue:float))
   > SetSize(
    self=Get MediumSpacer() -> (MediumSpacer:Spacer Object Reference)
    InSize=Select(Option 0:Vector 2D Structure=(X=16,Y=16), Option 1:Vector 2D Structure=Get SelectedPipWidth(), Index=@n8) -> (ReturnValue:Vector 2D Structure)
   )
  ]}
 SetDifficultyModeToNextAvailable {nodes [
   n1 = SetDifficultyModeToNextAvailable!() -> (
    then
    PilgrimCorePlayerStateComponent:Pilgrim Core Player State Component Object Reference
    InCurrentDifficulty:EPilgrimSongDifficulty Enum=DifficultyEasy
   ) {is=title K2Node_FunctionEntry}
   n2 = Get SupportedDifficulties(self=Get PilgrimSupportedGameplayConfig(self=Get PilgrimPlayspace() -> (PilgrimPlayspace:Pilgrim Playspace Object Reference))) -> (SupportedDifficulties=DifficultyEasy)
   n3 = int + int(A=Find Item!(TargetArray:Array of EPilgrimSongDifficulty Enums=@n2, ItemToFind:"EPilgrimSongDifficulty Enum (by ref)"=@n1.InCurrentDifficulty) -> (ReturnValue))
   > n4 = Branch(Condition=Is Valid Index!(TargetArray:Array of EPilgrimSongDifficulty Enums=@n2, IndexToTest:int=@n3) -> (ReturnValue:bool))
   > SetSongDifficulty!(
    execute
    self:Pilgrim Core Player State Component Object Reference=@n1
    Difficulty:EPilgrimSongDifficulty Enum="Get (a ref)"(Array:Array of EPilgrimSongDifficulty Enums=@n2, Dimension 1=@n3) -> (Output:"EPilgrimSongDifficulty Enum (by ref)")
   ) -> (then)
   n5 = Branch(execute=@n4.else, Condition=Is Valid Index(TargetArray=@n2))
   > SetSongDifficulty(self=@n1, Difficulty="Get (a ref)"(Array=@n2, Dimension 1=0))
   PrintString!(
    execute=@n5.else
    InString=ERROR: Supported Difficulties Array is empty. Default will be Easy.
    bPrintToScreen
    bPrintToLog
    TextColor=(R=1,G=0.0289,B=0,A=1)
    Duration=200
    Key
   ) -> (then)
  ]}
 SetPlayerColor {nodes [
   n1 = SetPlayerColor!() -> (then, PlayerColor:Linear Color Structure) {is=title K2Node_FunctionEntry}
   > SetBackingColor!(
    execute
    self:WBP Sparks Backing Panel Player Identity Object Reference=Get WBP_Sparks_BackingPanel()
    Linear Color:Linear Color Structure=@n1
   ) -> (then)
  ]}
]
variables [
 AsyncEventHandles {type="array<CancellableAsyncAction>", default []}
 EventRouter {type=GameplayEventRouterComponent, default=null}
 GameplayEventHandles {type="array<AsyncAction_StartListeningToEvent>", default []}
 IsOnGreenRoom {type=boolean, default=true}
 PilgrimPlayspace {type=PilgrimPlayspace, default=null}
 PlayerController {type=PilgrimPlayerControllerComponent, default=null}
 SelectedPipWidth {type=Vector2D, default=<48 16>}
 ShowDifficultyIcon {type=boolean}
]
widgets {
 info {inheritedWidgetCount=0, parentClass=&UserWidget, rootWidgetClass=&Overlay, widgetCount=34}
 widgets [
  GridPanel {widgetClassPath=&Overlay, bIsVariable=true, widget=GridPanel}
  Backing {
   widgetClassPath=&Overlay
   bIsVariable=true
   parent=GridPanel
   widget=Backing
   slot {class=OverlaySlot, properties {horizontalAlignment=HAlign_Fill, verticalAlignment=VAlign_Fill}}
  }
  Spacer_161 {
   widgetClassPath=&Spacer
   parent=Backing
   widget=Spacer_161
   properties {size=<1 48>}
   slot {class=OverlaySlot, properties {verticalAlignment=VAlign_Fill}}
  }
  SelectionBackingIndicator {
   widgetClassPath=&Image
   bIsVariable=true
   parent=Backing
   widget=SelectionBackingIndicator
   properties {
    renderOpacity=0
    visibility=SelfHitTestInvisible
    brush {
     bIsDynamicallyLoaded=false
     drawAs=Image
     imageType=NoImage
     mirroring=NoMirror
     resourceName=None
     resourceObject=/SparksUICommon/UILibrary/MI_FMCore_ButtonBacking_DifficultySelector
     tiling=NoTile
     imageSize=<32 32>
     margin=<ltrb 0 0 0 0>
     outlineSettings {
      bUseBrushTransparency=false
      roundingType=HalfHeightRadius
      width=0
      color {colorUseRule=UseColor_Specified, specifiedColor=<rgba 0 0 0 0>}
      cornerRadii=<0 0 0 0>
     }
     tintColor {colorUseRule=UseColor_Specified, specifiedColor=<rgba 1 1 1 1>}
     uVRegion {bIsValid=false, max=<0 0>, min=<0 0>}
    }
   }
   slot {class=OverlaySlot, properties {horizontalAlignment=HAlign_Fill, verticalAlignment=VAlign_Fill, padding=<ltrb 1 1 1 1>}}
  }
  WBP_UIKit_Block_Outline {
   widgetClassPath=/FortUILibrary/UIKit/Buttons/Blocks/WBP_UIKit_Block_Outline
   bIsVariable=true
   parent=Backing
   widget=WBP_UIKit_Block_Outline
   slot {class=OverlaySlot, properties {horizontalAlignment=HAlign_Fill, verticalAlignment=VAlign_Fill}}
  }
  WBP_Sparks_BackingPanel {
   widgetClassPath=/PilgrimCore/Widgets/SparksUILibrary/WBP_Sparks_BackingPanel_PlayerIdentity
   bIsVariable=true
   parent=Backing
   widget=WBP_Sparks_BackingPanel
   properties {cornerRadius=Large}
   slot {class=OverlaySlot, properties {horizontalAlignment=HAlign_Fill, verticalAlignment=VAlign_Fill}}
  }
  WBP_UIKit_Button_Regular {
   widgetClassPath=&WBP_UIKit_Button_Regular
   bIsVariable=true
   parent=Backing
   widget=WBP_UIKit_Button_Regular
   properties {
    bHideInputActionWithKeyboard=true
    bUseNaturalNavigation=true
    buttonSize=Small
    buttonType=Quiet
    inputModeOverride=All
    renderOpacity=0
    text=""
    visibility=Visible
   }
   slot {class=OverlaySlot, properties {horizontalAlignment=HAlign_Fill, verticalAlignment=VAlign_Fill}}
  }
  GridPanel_0 {
   widgetClassPath=&GridPanel
   parent=GridPanel
   widget=GridPanel_0
   properties {columnFill [1], rowFill [1]}
   slot {class=OverlaySlot, properties {horizontalAlignment=HAlign_Fill, verticalAlignment=VAlign_Center}}
  }
  HorizontalBox_0 {
   widgetClassPath=&HorizontalBox
   parent=GridPanel_0
   widget=HorizontalBox_0
   slot {class=GridSlot, properties {verticalAlignment=VAlign_Center, padding=<ltrb 23 0 15 0>}}
  }
  Keybind_GridPanel {
   widgetClassPath=&GridPanel
   bIsVariable=true
   parent=HorizontalBox_0
   widget=Keybind_GridPanel
   slot {class=HorizontalBoxSlot, properties {horizontalAlignment=HAlign_Left, verticalAlignment=VAlign_Center, padding=<ltrb 0 0 20 0>}}
  }
  CommonVisibilityWidgetLegacy_0 {
   widgetClassPath=&CommonVisibilityWidgetLegacy
   parent=Keybind_GridPanel
   widget=CommonVisibilityWidgetLegacy_0
   properties {bShowForTouch=false, verticalAlignment=VAlign_Center, padding=<ltrb 0 0 0 0>}
   slot {class=GridSlot}
  }
  ScaleBox_1 {
   widgetClassPath=&ScaleBox
   parent=CommonVisibilityWidgetLegacy_0
   widget=ScaleBox_1
   properties {stretch=UserSpecified, userSpecifiedScale=0.75}
   slot {class=BorderSlot, properties {verticalAlignment=VAlign_Center, padding=<ltrb 0 0 0 0>}}
  }
  KeybindWidget_ChangeBindings_1 {
   widgetClassPath=/FortUILibrary/Input/KeybindWidget
   bIsVariable=true
   parent=ScaleBox_1
   widget=KeybindWidget_ChangeBindings_1
   properties {
    bUseKeyboardTextureForKeys=true
    boundAction=Pilgrim_SwapDifficulty
    boundInputAction=/PilgrimSettings/Player/Abilities/IA_PilgrimSwapDifficulty
    boundKeyFallback=F2
    boundActionFallbacks [GamepadSwitchMode, GamepadSwitchModeOrEditBuilding, GamepadHoldToBuild]
   }
   slot {class=ScaleBoxSlot}
  }
  ScaleBox_0 {
   widgetClassPath=&ScaleBox
   parent=HorizontalBox_0
   widget=ScaleBox_0
   properties {stretch=UserSpecified, userSpecifiedScale=0.5}
   slot {class=HorizontalBoxSlot}
  }
  WBP_Pilgrim_Results_Difficulty {
   widgetClassPath=/PilgrimQuickplay/Widgets/Results/WBP_Pilgrim_Results_Difficulty
   bIsVariable=true
   parent=ScaleBox_0
   widget=WBP_Pilgrim_Results_Difficulty
   properties {visibility=Collapsed, renderTransform {angle=0, scale=<1.5 1.5>, shear=<0 0>, translation=<0 0>}}
   slot {class=ScaleBoxSlot}
  }
  FortResponsivePanel_0 {
   widgetClassPath=/Script/FortniteUI.FortResponsivePanel
   parent=HorizontalBox_0
   widget=FortResponsivePanel_0
   slot {class=HorizontalBoxSlot, properties {verticalAlignment=VAlign_Center, padding=<ltrb 18 0 0 0>}}
  }
  Common_Text_Difficulty {
   widgetClassPath=&CommonTextBlock
   bIsVariable=true
   parent=FortResponsivePanel_0
   widget=Common_Text_Difficulty
   properties {
    text=Easy
    textTransformPolicy=ToUpper
    visibility=HitTestInvisible
    colorAndOpacity {colorUseRule=UseColor_Specified, specifiedColor=<rgba 0.5 0.8333 1 1>}
    font {
     bForceMonospaced=false
     bMaterialIsStencil=false
     fontMaterial=null
     fontObject=&HeadingNow
     letterSpacing=0
     monospacedWidth=1
     size=13.5
     skewAmount=0
     typefaceFontName=76 Bold Italic
     outlineSettings {
      bApplyOutlineToDropShadows=false
      bMiteredCorners=false
      bSeparateFillAlpha=false
      outlineMaterial=null
      outlineSize=0
      outlineColor=<rgba 0 0 0 1>
     }
    }
    shadowOffset=<0 0>
   }
   slot {class=FortResponsivePanelSlot}
  }
  Overlay_1 {
   widgetClassPath=&Overlay
   parent=HorizontalBox_0
   widget=Overlay_1
   properties {renderTransformPivot=<1 0.5>}
   slot {class=HorizontalBoxSlot, properties {horizontalAlignment=HAlign_Right, verticalAlignment=VAlign_Center, size {sizeRule=Fill, value=1}}}
  }
  PipBacking {
   widgetClassPath=&Image
   parent=Overlay_1
   widget=PipBacking
   properties {
    renderOpacity=0.5
    visibility=Collapsed
    brush {
     bIsDynamicallyLoaded=false
     drawAs=Box
     imageType=NoImage
     mirroring=NoMirror
     resourceName=None
     resourceObject=/PilgrimCore/Widgets/Materials/M_UI_Pilgrim_InstrumentIntensity
     tiling=NoTile
     imageSize=<24 24>
     margin=<ltrb 0.5 0.5 0.5 0.5>
     outlineSettings {
      bUseBrushTransparency=false
      roundingType=HalfHeightRadius
      width=0
      color {colorUseRule=UseColor_Specified, specifiedColor=<rgba 0 0 0 0>}
      cornerRadii=<0 0 0 0>
     }
     tintColor {colorUseRule=UseColor_Specified, specifiedColor=<rgba 1 1 1 1>}
     uVRegion {bIsValid=false, max=<0 0>, min=<0 0>}
    }
    colorAndOpacity=<rgba 0.0033 0.003 0.0497 1>
   }
   slot {class=OverlaySlot, properties {horizontalAlignment=HAlign_Fill, verticalAlignment=VAlign_Center}}
  }
  HorizontalBox_116 {
   widgetClassPath=&HorizontalBox
   parent=Overlay_1
   widget=HorizontalBox_116
   slot {class=OverlaySlot, properties {horizontalAlignment=HAlign_Center, verticalAlignment=VAlign_Center, padding=<ltrb 4 0 4 0>}}
  }
  Overlay_2 {
   widgetClassPath=&Overlay
   parent=HorizontalBox_116
   widget=Overlay_2
   slot {class=HorizontalBoxSlot, properties {verticalAlignment=VAlign_Bottom}}
  }
  EasySpacer {
   widgetClassPath=&Spacer
   bIsVariable=true
   parent=Overlay_2
   widget=EasySpacer
   properties {size=<40 1>}
   slot {class=OverlaySlot, properties {horizontalAlignment=HAlign_Fill, verticalAlignment=VAlign_Fill}}
  }
  Image_easy {
   widgetClassPath=&Image
   bIsVariable=true
   parent=Overlay_2
   widget=Image_easy
   properties {
    visibility=SelfHitTestInvisible
    brush {
     bIsDynamicallyLoaded=false
     drawAs=Box
     imageType=NoImage
     mirroring=NoMirror
     resourceName=None
     resourceObject=&M_UI_Pilgrim_InstrumentIntensity
     tiling=NoTile
     imageSize=<16 16>
     margin=<ltrb 0.5 0.5 0.5 0.5>
     outlineSettings {
      bUseBrushTransparency=false
      roundingType=HalfHeightRadius
      width=0
      color {colorUseRule=UseColor_Specified, specifiedColor=<rgba 0 0 0 0>}
      cornerRadii=<0 0 0 0>
     }
     tintColor {colorUseRule=UseColor_Specified, specifiedColor=<rgba 1 1 1 1>}
     uVRegion {bIsValid=false, max=<0 0>, min=<0 0>}
    }
    colorAndOpacity=<rgba 0.5 0.8333 1 1>
   }
   slot {class=OverlaySlot, properties {horizontalAlignment=HAlign_Fill, verticalAlignment=VAlign_Fill}}
  }
  Overlay_4 {
   widgetClassPath=&Overlay
   parent=HorizontalBox_116
   widget=Overlay_4
   slot {class=HorizontalBoxSlot, properties {verticalAlignment=VAlign_Bottom, padding=<ltrb 4 0 0 0>}}
  }
  MediumSpacer {
   widgetClassPath=&Spacer
   bIsVariable=true
   parent=Overlay_4
   widget=MediumSpacer
   slot {class=OverlaySlot, properties {horizontalAlignment=HAlign_Fill, verticalAlignment=VAlign_Fill}}
  }
  Image_medium {
   widgetClassPath=&Image
   bIsVariable=true
   parent=Overlay_4
   widget=Image_medium
   properties {
    renderOpacity=0.5
    visibility=SelfHitTestInvisible
    brush {
     bIsDynamicallyLoaded=false
     drawAs=Box
     imageType=NoImage
     mirroring=NoMirror
     resourceName=None
     resourceObject=&M_UI_Pilgrim_InstrumentIntensity
     tiling=NoTile
     imageSize=<16 16>
     margin=<ltrb 0.5 0.5 0.5 0.5>
     outlineSettings {
      bUseBrushTransparency=false
      roundingType=HalfHeightRadius
      width=0
      color {colorUseRule=UseColor_Specified, specifiedColor=<rgba 0 0 0 0>}
      cornerRadii=<0 0 0 0>
     }
     tintColor {colorUseRule=UseColor_Specified, specifiedColor=<rgba 1 1 1 1>}
     uVRegion {bIsValid=false, max=<0 0>, min=<0 0>}
    }
    colorAndOpacity=<rgba 0.5 0.8333 1 1>
   }
   slot {class=OverlaySlot, properties {horizontalAlignment=HAlign_Fill, verticalAlignment=VAlign_Fill}}
  }
  Overlay_3 {
   widgetClassPath=&Overlay
   parent=HorizontalBox_116
   widget=Overlay_3
   slot {class=HorizontalBoxSlot, properties {verticalAlignment=VAlign_Bottom, padding=<ltrb 4 0 0 0>}}
  }
  HardSpacer {
   widgetClassPath=&Spacer
   bIsVariable=true
   parent=Overlay_3
   widget=HardSpacer
   slot {class=OverlaySlot, properties {horizontalAlignment=HAlign_Fill, verticalAlignment=VAlign_Fill}}
  }
  Image_hard {
   widgetClassPath=&Image
   bIsVariable=true
   parent=Overlay_3
   widget=Image_hard
   properties {
    renderOpacity=0.5
    visibility=SelfHitTestInvisible
    brush {
     bIsDynamicallyLoaded=false
     drawAs=Box
     imageType=NoImage
     mirroring=NoMirror
     resourceName=None
     resourceObject=&M_UI_Pilgrim_InstrumentIntensity
     tiling=NoTile
     imageSize=<16 16>
     margin=<ltrb 0.5 0.5 0.5 0.5>
     outlineSettings {
      bUseBrushTransparency=false
      roundingType=HalfHeightRadius
      width=0
      color {colorUseRule=UseColor_Specified, specifiedColor=<rgba 0 0 0 0>}
      cornerRadii=<0 0 0 0>
     }
     tintColor {colorUseRule=UseColor_Specified, specifiedColor=<rgba 1 1 1 1>}
     uVRegion {bIsValid=false, max=<0 0>, min=<0 0>}
    }
    colorAndOpacity=<rgba 0.5 0.8333 1 1>
   }
   slot {class=OverlaySlot, properties {horizontalAlignment=HAlign_Fill, verticalAlignment=VAlign_Fill}}
  }
  Overlay_5 {
   widgetClassPath=&Overlay
   parent=HorizontalBox_116
   widget=Overlay_5
   slot {class=HorizontalBoxSlot, properties {verticalAlignment=VAlign_Bottom, padding=<ltrb 4 0 0 0>}}
  }
  ExpertSpacer {
   widgetClassPath=&Spacer
   bIsVariable=true
   parent=Overlay_5
   widget=ExpertSpacer
   slot {class=OverlaySlot, properties {horizontalAlignment=HAlign_Fill, verticalAlignment=VAlign_Fill}}
  }
  Image_expert {
   widgetClassPath=&Image
   bIsVariable=true
   parent=Overlay_5
   widget=Image_expert
   properties {
    renderOpacity=0.5
    visibility=SelfHitTestInvisible
    brush {
     bIsDynamicallyLoaded=false
     drawAs=Box
     imageType=NoImage
     mirroring=NoMirror
     resourceName=None
     resourceObject=&M_UI_Pilgrim_InstrumentIntensity
     tiling=NoTile
     imageSize=<16 16>
     margin=<ltrb 0.5 0.5 0.5 0.5>
     outlineSettings {
      bUseBrushTransparency=false
      roundingType=HalfHeightRadius
      width=0
      color {colorUseRule=UseColor_Specified, specifiedColor=<rgba 0 0 0 0>}
      cornerRadii=<0 0 0 0>
     }
     tintColor {colorUseRule=UseColor_Specified, specifiedColor=<rgba 1 1 1 1>}
     uVRegion {bIsValid=false, max=<0 0>, min=<0 0>}
    }
    colorAndOpacity=<rgba 0.5 0.8333 1 1>
   }
   slot {class=OverlaySlot, properties {horizontalAlignment=HAlign_Fill, verticalAlignment=VAlign_Fill}}
  }
  InputContainer_GridPanel {
   widgetClassPath=&GridPanel
   bIsVariable=true
   parent=GridPanel
   widget=InputContainer_GridPanel
   properties {visibility=Collapsed}
   slot {class=OverlaySlot, properties {verticalAlignment=VAlign_Center, padding=<ltrb -14 0 0 0>}}
  }
  InputActionWidget {
   widgetClassPath=&CommonActionWidget
   bIsVariable=true
   parent=InputContainer_GridPanel
   widget=InputActionWidget
   properties {
    progressMaterialParam=Percentage
    visibility=HitTestInvisible
    inputActions [{dataTable=/FortUILibrary/Foundation/ActivityBrowserInputActionTable, rowName=Select}]
    progressMaterialBrush {
     bIsDynamicallyLoaded=false
     drawAs=Image
     imageType=NoImage
     mirroring=NoMirror
     resourceName=None
     resourceObject=/FortUILibrary/UIKit/Buttons/MaterialInstances/CTAButtons/MI_UIKit_EmptyProgress
     tiling=NoTile
     imageSize=<48 48>
     margin=<ltrb 0 0 0 0>
     outlineSettings {
      bUseBrushTransparency=false
      roundingType=HalfHeightRadius
      width=0
      color {colorUseRule=UseColor_Specified, specifiedColor=<rgba 0 0 0 0>}
      cornerRadii=<0 0 0 0>
     }
     tintColor {colorUseRule=UseColor_Specified, specifiedColor=<rgba 1 1 1 1>}
     uVRegion {bIsValid=false, max=<0 0>, min=<0 0>}
    }
   }
   slot {class=GridSlot, properties {padding=<ltrb -10 0 0 0>}}
  }
 ]
}
