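# Pilgrim Quickplay green room HUD widget that shows a keybind prompt labeled "Return to Results".
# On init it gets the owning player's event router once playlist data is ready, then stores the player state.
# Quickplay state events show it in song or set results and hide it otherwise; one green room timer state also hides it.
# A macro on the music battle instance's router makes the canvas respond to calibration modal events.
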WidgetBlueprint /PilgrimQuickplay/Widgets/WBP_Pilgrim_GreenRoom_ReturnToResultsUI
graphs [
 EventGraph {nodes [
   Event OnInitialized()
   > n1 = Sequence!(execute) -> (then_0, then_1, then_2:Exec)
   > n2 = PilgrimWaitForEventRouter!(
    execute:Exec
    ContextActor:Actor Object Reference=Get MusicBattleInstance() -> (MusicBattleInstance:Pilgrim Music Battle Instance Object Reference)
   ) -> (then:Exec, OnEventRouterAvailable:Exec, EventRouter:Gameplay Event Router Component Object Reference) {is=title K2Node_AsyncAction}
   Respond to Calibration Modal State Events!(
    execute:Exec=@n2.OnEventRouterAvailable
    EventRouter:Gameplay Event Router Component Object Reference=@n2
    Widget:Widget Object Reference=Get CanvasPanel() -> (CanvasPanel:Canvas Panel Object Reference)
   ) -> (exec:Exec) {is=title K2Node_MacroInstance}
   n3 = PlaylistDataReadyAsync!(execute:Exec=@n1.then_1) -> (
    then:Exec
    OnReady:Exec
    GameState:Fort Game State Athena Object Reference
    Playlist:Fort Playlist Object Reference
    PlaylistContextTags:"Gameplay Tag Container Structure (by ref)"
   ) {is=title K2Node_AsyncAction}
   n4 = GetOwningPlayer()
   n5 = PilgrimWaitForEventRouter(execute=@n3.OnReady, ContextActor=@n4)
   Set EventRouter(execute=@n5.OnEventRouterAvailable, EventRouter:Gameplay Event Router Component Object Reference=@n5) -> (Output_Get:Gameplay Event Router Component Object Reference)
   > Set Up Player!(
    execute
    self
    PlayerState:Player State Object Reference=Get PlayerState!(self:Controller Object Reference=@n4) -> (PlayerState:Player State Object Reference)
   ) -> (then) {is=title K2Node_CallFunction}
   > n6 = Sequence!(execute) -> (then_0, then_1)
   n7 = Get EventRouter() -> (EventRouter:Gameplay Event Router Component Object Reference)
   > n8 = StartListeningToEvent!(
    execute:Exec
    Target:Gameplay Event Router Component Object Reference=@n7
    EventType:Script Struct Object Reference=/PilgrimQuickplay/Items/PilgrimEvent_GreenRoomTimerStateChanged
    Context:Object Reference
    EventBubblingRule:EEventBubblingRule Enum
    None:%=@n7
   ) -> (
    then:Exec
    AsyncTaskProxy:Async Action Start Listening To Event Object Reference
    OnEventReceived:Exec
    Payload_GreenRoomTimerState_2_#1:GreenRoomTimerState Enum
    Payload_FinalTimerLength_5_#2:double=~
   )
   > Add(
    TargetArray:Array of Async Action Start Listening To Event Object References=Get GameplayEventHandles() -> (GameplayEventHandles:Array of Async Action Start Listening To Event Object References)
    NewItem:"Async Action Start Listening To Event Object Reference (by ref)"=@n8
   ) {is=title K2Node_CallArrayFunction}
   n9 = Switch on GreenRoomTimerState!(execute=@n8.OnEventReceived, Selection:GreenRoomTimerState Enum=@n8.Payload_GreenRoomTimerState_2_#1) -> (NewEnumerator0:Exec, NewEnumerator1:Exec, NewEnumerator2:Exec, NewEnumerator3:Exec, NewEnumerator4:Exec) {is=title K2Node_SwitchEnum}
   SetVisibility(execute=@n9.NewEnumerator2, InVisibility=Hidden)
   n10 = Get EventRouter()
   n11 = StartListeningToEvent!(
    execute=@n6.then_1
    Target=@n10
    EventType=/Script/PilgrimQPSharedRuntime.PilgrimEvent_QuickplayState_Began
    Context
    EventBubblingRule
    None:%=@n10
   ) -> (then, AsyncTaskProxy, OnEventReceived, Payload_StateId:EPilgrimQuickplayState Enum)
   > Add(TargetArray=Get GameplayEventHandles(), NewItem=@n11)
   n12 = Branch(
    execute=@n11.OnEventReceived
    Condition=OR Boolean(
     A="Equal (Enum)"(A:EPilgrimQuickplayState Enum=@n11.Payload_StateId, B:EPilgrimQuickplayState Enum=SongResults)
     B="Equal (Enum)"(A=@n11.Payload_StateId, B=SetResults)
    )
   )
   > SetVisibility(InVisibility=SelfHitTestInvisible)
   SetVisibility(execute=@n12.else, InVisibility=Hidden)
  ]}
 Set Up Player {nodes [
   n1 = Set Up Player!() -> (then, PlayerState:Player State Object Reference) {is=title K2Node_FunctionEntry}
   > n2 = Set MyPlayerState(MyPlayerState:Player State Object Reference=@n1) -> (Output_Get:Player State Object Reference)
   > n3 = Cast To FortPlayerStateAthena!(execute, Object=@n2) -> (then, CastFailed, AsFort Player State Athena:Fort Player State Athena Object Reference) {is=title K2Node_DynamicCast}
   PrintString!(execute=@n3.CastFailed, InString="CAST FAILED!", bPrintToScreen, bPrintToLog, TextColor=(R=1,G=0.078,B=0,A=1), Duration=8, Key) -> (then)
  ]}
]
variables [
 CountDown {type=WidgetAnimation, default=null}
 EventRouter {type=GameplayEventRouterComponent, default=null}
 InitialCountdownState {type=WidgetAnimation, default=null}
 MyPlayerState {type=PlayerState, default=null}
]
widgets {
 info {inheritedWidgetCount=0, parentClass=/Script/PilgrimCoreUI.PilgrimGameWidgetBase, rootWidgetClass=&CanvasPanel, widgetCount=4}
 widgets [
  CanvasPanel {widgetClassPath=&CanvasPanel, bIsVariable=true, widget=CanvasPanel}
  HorizontalBox_1 {
   widgetClassPath=&HorizontalBox
   parent=CanvasPanel
   widget=HorizontalBox_1
   properties {renderTransformPivot=<1 1>}
   slot {
    class=CanvasPanelSlot
    properties {bAutoSize=true, layoutData {alignment=<0 0>, anchors {maximum=<0 0>, minimum=<0 0>}, offsets=<ltrb 0 0 206.4 93.3333>}}
   }
  }
  KeybindWidget_ChangeBindings_1 {
   widgetClassPath=/FortUILibrary/Input/KeybindWidget
   bIsVariable=true
   parent=HorizontalBox_1
   widget=KeybindWidget_ChangeBindings_1
   properties {
    boundAction=Use
    boundInputAction=/PilgrimSettings/Player/Abilities/IA_PilgrimReturnToResults
    boundKeyFallback=F2
    inputContextOverride=/PilgrimSettings/Player/Abilities/IMC_GreenRoomMappingContext
    boundActionFallbacks [UseOrReload]
   }
   slot {class=HorizontalBoxSlot, properties {horizontalAlignment=HAlign_Left, verticalAlignment=VAlign_Center}}
  }
  CurrentInputBindingNumber {
   widgetClassPath=&CommonTextBlock
   bIsVariable=true
   parent=HorizontalBox_1
   widget=CurrentInputBindingNumber
   properties {
    style=/FortUILibrary/Styles/Typography/Foundation/TextStyle_BurbankSmall_S_White.TextStyle_BurbankSmall_S_White_C
    text=Return to Results
    font {
     bForceMonospaced=false
     bMaterialIsStencil=false
     fontMaterial=null
     fontObject=&BurbankSmall
     letterSpacing=0
     monospacedWidth=1
     size=18
     skewAmount=0
     typefaceFontName=Black
     outlineSettings {
      bApplyOutlineToDropShadows=false
      bMiteredCorners=false
      bSeparateFillAlpha=false
      outlineMaterial=null
      outlineSize=0
      outlineColor=<rgba 0 0 0 1>
     }
    }
    shadowOffset=<0 0>
   }
   slot {class=HorizontalBoxSlot, properties {horizontalAlignment=HAlign_Center, verticalAlignment=VAlign_Center}}
  }
 ]
}
