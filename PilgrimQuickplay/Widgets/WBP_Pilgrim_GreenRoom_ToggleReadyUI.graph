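# Hold-to-toggle Ready Up/Unready button widget for the Pilgrim Quickplay green room.
# On construct it waits for playlist data and the event router, then listens for readiness, setlist, timer and quickplay state events to drive visibility and display.
# Set Ready State Display updates text, materials and the checkmark animation, disabling the prompt when no songs are selected or externally disabled.
# The input action only plays hold sounds here; a gameplay tag gates the keybind.
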WidgetBlueprint /PilgrimQuickplay/Widgets/WBP_Pilgrim_GreenRoom_ToggleReadyUI
animations [ReadyUp {
  binding_count=4
  length_seconds=1
  object=&self:ReadyUp
  sequence {bindings [
    StatusIndicator {
     id=#1
     possessed_class=WBP_UIKit_StatusIndicator_C
     tracks [
      {
       type=MovieSceneFloatTrack
       display_name=Render Opacity
       sections [{type=MovieSceneFloatSection, channels [None {type=MovieSceneScriptingFloatChannel, keys [[0,0],[4,1]]}]}]
      }
      {
       type=MovieScene2DTransformTrack
       sections [{
         type=MovieScene2DTransformSection
         channels [
          Translation.X {type=MovieSceneScriptingFloatChannel, const [0,1]}
          Translation.Y {type=MovieSceneScriptingFloatChannel, const [0,1]}
          Scale.X {type=MovieSceneScriptingFloatChannel, keys [[0,0.5],[3,1.1],[8,1]]}
          Scale.Y {type=MovieSceneScriptingFloatChannel, keys [[0,0.5],[3,1.1],[8,1]]}
         ]
        }]
      }
     ]
    }
    Check_Pulse {
     id=#2
     possessed_class=Image
     tracks [{
       type=MovieSceneWidgetMaterialTrack
       sections [{type=MovieSceneParameterSection, channels [Alpha_Final {type=MovieSceneScriptingFloatChannel, keys [[2,0],[6,1],[20,0.5]]}]}]
      }]
    }
    ReadyIndicator {
     id=#3
     possessed_class=Overlay
     tracks [{
       type=MovieSceneFloatTrack
       display_name=Render Opacity
       sections [{type=MovieSceneFloatSection, channels [None {type=MovieSceneScriptingFloatChannel, keys [[2,0],[4,1]]}]}]
      }]
    }
    Check_Glow {
     id=#4
     possessed_class=Image
     tracks [
      {
       type=MovieSceneWidgetMaterialTrack
       sections [{
         type=MovieSceneParameterSection
         end_frame=18
         start_frame=0
         channels [Alpha {type=MovieSceneScriptingFloatChannel, keys [[2,0],[7.2,1],[18,0]]}]
        }]
      }
      {
       type=MovieSceneFloatTrack
       display_name=Render Opacity
       sections [{type=MovieSceneFloatSection, channels [None {type=MovieSceneScriptingFloatChannel, keys [[2,0],[4.5,1]]}]}]
      }
     ]
    }
   ]}
 }]
graphs [
 ChangeButtonBorder {nodes [
   n1 = ChangeButtonBorder!() -> (then, CornerRadius:float=~) {is=title K2Node_FunctionEntry}
   > n2 = GetDynamicMaterial(self=Get ReadyLayer() -> (ReadyLayer:Image Object Reference))
   > n3 = GetDynamicMaterial(self=Get ButtonImage() -> (ButtonImage:Image Object Reference))
   > SetScalarParameterValue(self=[@n2, @n3], ParameterName=Corner Radius, Value=@n1)
  ]}
 EventGraph {nodes [
   n1 = EnhancedInputAction IA_PilgrimToggleReady!() -> (
    Triggered:Exec
    Started:Exec
    Ongoing:Exec
    Canceled:Exec
    Completed:Exec
    ActionValue:bool
    ElapsedSeconds:double=~
    TriggeredSeconds:double=~
    InputAction:Input Action Object Reference=/PilgrimSettings/Player/Abilities/IA_PilgrimToggleReady
   ) {is=title K2Node_EnhancedInputAction}
   n2 = Branch(execute=@n1.Canceled, Condition=Get IsReadyUpDisabled() -> (IsReadyUpDisabled:bool))
   n3 = SpawnSound2D!(
    execute=@n2.else
    Sound:Sound Base Object Reference=Get ReadyUpRelease Sound() -> (ReadyUpRelease Sound:Sound Base Object Reference)
    VolumeMultiplier:float=1
    PitchMultiplier:float=1
    StartTime:float=0
    ConcurrencySettings:Sound Concurrency Object Reference
    bPersistAcrossLevelTransition:bool=false
    bAutoDestroy:bool=true
   ) -> (then, ReturnValue:Audio Component Object Reference)
   > n4 = Set ReadyUpRelease Component(ReadyUpRelease Component:Audio Component Object Reference=@n3) -> (Output_Get:Audio Component Object Reference)
   n5 = Get!(execute:Exec=[@n4.then, @n9.then]) -> (then:Exec, else:Exec, ReadyUpPress Component:Audio Component Object Reference)
   > n6 = FadeOut!(
    execute
    self:Audio Component Object Reference=@n5
    FadeOutDuration:float=0.2
    FadeVolumeLevel:float=0
    FadeCurve:EAudioFaderCurve Enum=Linear
   ) -> (then)
   n7 = Get!(execute=[@n6.then, @n5.else]) -> (then, else, ReadyUpHold Component:Audio Component Object Reference)
   > FadeOut(self=@n7, FadeOutDuration=0.2, FadeVolumeLevel=0, FadeCurve=Linear)
   n8 = Branch(execute=@n1.Completed, Condition=Get IsReadyUpDisabled())
   n9 = PlaySound2D(execute=@n8.else, Sound=Get ReadyUpComplete Sound() -> (ReadyUpComplete Sound:Sound Base Object Reference))
   n10 = Branch(execute=@n1.Started, Condition=Get IsReadyUpDisabled())
   n11 = SpawnSound2D(
    execute=@n10.else
    Sound=Get ReadyUpPress Sound() -> (ReadyUpPress Sound:Sound Base Object Reference)
    VolumeMultiplier=1
    PitchMultiplier=1
    StartTime=0
    bPersistAcrossLevelTransition=false
    bAutoDestroy=true
   )
   > Set ReadyUpPress Component(ReadyUpPress Component:Audio Component Object Reference=@n11) -> (Output_Get:Audio Component Object Reference)
   > n12 = SpawnSound2D(
    Sound=Get ReadyUpHold Sound() -> (ReadyUpHold Sound:Sound Base Object Reference)
    VolumeMultiplier=1
    PitchMultiplier=1
    StartTime=0
    bPersistAcrossLevelTransition=false
    bAutoDestroy=true
   )
   > Set ReadyUpHold Component(ReadyUpHold Component:Audio Component Object Reference=@n12) -> (Output_Get:Audio Component Object Reference)
   > n13 = Get!(execute) -> (then, else, ReadyUpRelease Component:Audio Component Object Reference)
   > FadeOut(self=@n13, FadeOutDuration=0.2, FadeVolumeLevel=0, FadeCurve=Linear)
   Event Construct()
   > FortShippingLog!(execute, WorldContextObject:Object Reference=Self-Reference(), InString:String="[FORT-838371] Event Construct", bLogAsWarning:bool=false) -> (then)
   > n14 = Sequence()
   > n15 = WaitGameplayTagCountChangedOnActor!(
    execute:Exec
    TargetActor:Actor Object Reference=Get PlayerState!(self:Controller Object Reference=GetOwningPlayer()) -> (PlayerState:Player State Object Reference)
    Tag:Gameplay Tag Structure=(TagName="Pilgrim.GreenRoom.Input.ToggleReadyUp")
   ) -> (then:Exec, AsyncTaskProxy:Ability Async Wait Gameplay Tag Count Changed Object Reference, TagCountChanged:Exec, TagCount:int) {is=title K2Node_AsyncAction}
   > Add(
    TargetArray:Array of Cancellable Async Action Object References=Get AsyncEventHandles() -> (AsyncEventHandles:Array of Cancellable Async Action Object References)
    NewItem:"Cancellable Async Action Object Reference (by ref)"=@n15
   ) {is=title K2Node_CallArrayFunction}
   n16 = GetOwningPlayer()
   n17 = HasMatchingGameplayTag!(
    execute:Exec=@n15.TagCountChanged
    self:Object Reference=GetAbilitySystemComponent!(Actor:Actor Object Reference=Get PlayerState(self=@n16)) -> (ReturnValue:Ability System Component Object Reference)
    TagToCheck:Gameplay Tag Structure=(TagName="Pilgrim.GreenRoom.Input.ToggleReadyUp")
   ) -> (then:Exec, ReturnValue:bool) {is=title K2Node_Message}
   n18 = Get KeybindWidget_ChangeBindings_1() -> (KeybindWidget_ChangeBindings_1:Keybind Widget Object Reference)
   > SetIsEnabled(self=@n18, bInIsEnabled=@n17)
   > SetVisibility(self=@n18, InVisibility=Select(Option 0=Collapsed, Option 1=SelfHitTestInvisible, Index=@n17))
   n19 = PlaylistDataReadyAsync!(execute:Exec=@n14.then_1) -> (
    then:Exec
    OnReady:Exec
    GameState:Fort Game State Athena Object Reference
    Playlist:Fort Playlist Object Reference
    PlaylistContextTags:"Gameplay Tag Container Structure (by ref)"
   ) {is=title K2Node_AsyncAction}
   n20 = GetOwningPlayer()
   n21 = PilgrimWaitForEventRouter!(execute:Exec=@n19.OnReady, ContextActor:Actor Object Reference=@n20) -> (then:Exec, OnEventRouterAvailable:Exec, EventRouter:Gameplay Event Router Component Object Reference) {is=title K2Node_AsyncAction}
   Set EventRouter(execute=@n21.OnEventRouterAvailable, EventRouter:Gameplay Event Router Component Object Reference=@n21) -> (Output_Get:Gameplay Event Router Component Object Reference)
   > Set Up Player!(execute, self, PlayerState:Player State Object Reference=Get PlayerState(self=@n20)) -> (then) {is=title K2Node_CallFunction}
   > n22 = Sequence!(execute) -> (then_0, then_1, then_2:Exec, then_3:Exec)
   n23 = Get EventRouter() -> (EventRouter:Gameplay Event Router Component Object Reference)
   > n24 = StartListeningToEvent!(
    execute:Exec
    Target:Gameplay Event Router Component Object Reference=@n23
    EventType:Script Struct Object Reference=/Script/PilgrimCoreRuntime.PilgrimEvent_PlayerState_ReadinessChanged
    Context:Object Reference
    EventBubblingRule:EEventBubblingRule Enum
    None:%=@n23
   ) -> (
    then:Exec
    AsyncTaskProxy:Async Action Start Listening To Event Object Reference
    OnEventReceived:Exec
    Payload_ForPlayer:Fort Player Pawn Object Reference
    Payload_bIsReady:bool=~
   )
   > Add(
    TargetArray:Array of Async Action Start Listening To Event Object References=Get GameplayEventHandles() -> (GameplayEventHandles:Array of Async Action Start Listening To Event Object References)
    NewItem:"Async Action Start Listening To Event Object Reference (by ref)"=@n24
   )
   Branch(
    execute=@n24.OnEventReceived
    Condition="Equal (Object)"(A=Get PlayerState(self:Pawn Object Reference=@n24.Payload_ForPlayer), B=Get MyPlayerState() -> (MyPlayerState:Player State Object Reference))
   )
   > Set Ready State Display!(execute, self, Is Ready:bool=@n24.Payload_bIsReady, InNumSongs:int=-1) -> (then) {is=title K2Node_CallFunction}
   n25 = Get EventRouter()
   n26 = StartListeningToEvent!(
    execute=@n22.then_1
    Target=@n25
    EventType=/PilgrimQuickplay/Items/PilgrimEvent_GreenRoomTimerStateChanged
    Context
    EventBubblingRule
    None:%=@n25
   ) -> (
    then
    AsyncTaskProxy
    OnEventReceived
    Payload_GreenRoomTimerState_2_#5:GreenRoomTimerState Enum
    Payload_FinalTimerLength_5_#6:double=~
   )
   > Add(TargetArray=Get GameplayEventHandles(), NewItem=@n26)
   n27 = Switch on GreenRoomTimerState!(execute=@n26.OnEventReceived, Selection:GreenRoomTimerState Enum=@n26.Payload_GreenRoomTimerState_2_#5) -> (NewEnumerator0:Exec, NewEnumerator1:Exec, NewEnumerator2:Exec, NewEnumerator3:Exec, NewEnumerator4:Exec) {is=title K2Node_SwitchEnum}
   SetVisibility(
    execute=@n27.NewEnumerator2
    self=Get HorizontalBoxReadyUnreadyWithInput() -> (HorizontalBoxReadyUnreadyWithInput:Overlay Object Reference)
    InVisibility=Collapsed
   )
   FortShippingLog(
    execute=@n22.then_2
    WorldContextObject=Self-Reference()
    InString="[FORT-838371] Listening to event QuickplayState_Began"
    bLogAsWarning=false
   )
   n28 = Get EventRouter()
   > n29 = StartListeningToEvent!(execute, Target=@n28, EventType=/Script/PilgrimQPSharedRuntime.PilgrimEvent_QuickplayState_Began, Context, EventBubblingRule, None:%=@n28) -> (then, AsyncTaskProxy, OnEventReceived, Payload_StateId:EPilgrimQuickplayState Enum)
   > Add(TargetArray=Get GameplayEventHandles(), NewItem=@n29)
   n30 = Branch(
    execute=@n29.OnEventReceived
    Condition=OR Boolean!(
     A="Equal (Enum)"(A:EPilgrimQuickplayState Enum=@n29.Payload_StateId, B:EPilgrimQuickplayState Enum=Pregame)
     B="Equal (Enum)"(A=@n29.Payload_StateId, B=SongResults)
     C:bool="Equal (Enum)"(A=@n29.Payload_StateId, B=SetResults)
    ) -> (ReturnValue)
   )
   > SetVisibility(InVisibility=Visible)
   > SetVisibility(self=Get HorizontalBoxReadyUnreadyWithInput(), InVisibility=SelfHitTestInvisible)
   > "SetText (Text)"(self=Get Block_Text() -> (Block_Text:Common Text Object Reference), InText="LOCTEXT(\"Ready Up\")")
   > "SetText (Text)"(self=Get Block_SecondaryText() -> (Block_SecondaryText:Common Text Object Reference), InText="LOCTEXT(\"(Hold)\")")
   > FortShippingLog(
    WorldContextObject=Self-Reference()
    InString=Append!(
     A="[FORT-838371] Showing ToggleReadyUI"
     B=" - PilgrimQuickplayState: "
     C:String=Enum to String!(Enumerator:EPilgrimQuickplayState Enum=@n29.Payload_StateId) -> (ReturnValue:String)
    ) -> (ReturnValue)
    bLogAsWarning=false
   )
   SetVisibility(execute=@n30.else, InVisibility=Hidden)
   > FortShippingLog(
    WorldContextObject=Self-Reference()
    InString=Append(A="[FORT-838371] Hidding ToggleReadyUI", B=" - PilgrimQuickplayState: ", C=Enum to String(Enumerator=@n29.Payload_StateId))
    bLogAsWarning=false
   )
   n31 = StartListeningToEvent!(
    execute=@n22.then_3
    Target=Get EventRouter()
    EventType=/Script/PilgrimCoreRuntime.PilgrimEvent_BandState_CustomSetlistChanged
    Context
    EventBubblingRule
   ) -> (then, AsyncTaskProxy, OnEventReceived, Payload_ChangedSetlist:Pilgrim Custom Setlist Object Reference)
   > Add(TargetArray=Get GameplayEventHandles(), NewItem=@n31)
   n32 = GetPilgrimQuickplayPlayerStateComponent!(execute=@n31.OnEventReceived, PlayerState:Player State Object Reference=Get MyPlayerState()) -> (then, ReturnValue:Pilgrim Quickplay Player State Component Object Reference)
   > Set Ready State Display(
    Is Ready=GetIsReady!(self:Pilgrim Core Player State Component Object Reference=@n32) -> (ReturnValue:bool)
    InNumSongs=GetNumSongs!(self:Pilgrim Setlist Interface Interface=@n31.Payload_ChangedSetlist) -> (ReturnValue:int=0)
   )
   K2Node_Event_1 = CustomEvent_2() {is=title K2Node_CustomEvent}
   > PlayAnimation(InAnimation=Get CountDown() -> (CountDown:Widget Animation Object Reference))
   > PlaySound2D(Sound=/Game/Sounds/Fort_UI_Sounds/FrontEnd/SC_UI_MatchMaking_TimeTick_Cue)
   K2Node_Event_2 = CustomEvent() {is=title K2Node_CustomEvent}
   > PlaySound2D(Sound=/Game/Sounds/Fort_UI_Sounds/FrontEnd/SC_UI_MatchMaking_ReadyUp_Cue)
   > PlayAnimation(InAnimation=Get InitialCountdownState() -> (InitialCountdownState:Widget Animation Object Reference))
   K2Node_Event_3 = CustomEvent_1() {is=title K2Node_CustomEvent}
   > PlaySound2D(Sound=/Game/Sounds/Fort_UI_Sounds/FrontEnd/SC_UI_MatchMaking_UnReady_Cue)
   Event Destruct()
   > Parent: Destruct()
   > FortShippingLog(WorldContextObject=Self-Reference(), InString="[FORT-838371] Event Destruct", bLogAsWarning=false)
   K2Node_Event_7 = CustomEvent_3() {is=title K2Node_CustomEvent}
   > PlayAnimation(InAnimation=Get InitialCountdownState())
  ]}
 GetDisabledByExternalSource {nodes [
   GetDisabledByExternalSource() {is=title K2Node_FunctionEntry}
   > Return Node!(execute, _DisabledByExternalSource:bool=Get _DisabledByExternalSource() -> (_DisabledByExternalSource:bool))
  ]}
 HasSelectedSongs {nodes [
   HasSelectedSongs() {is=title K2Node_FunctionEntry}
   > n1 = Get!(execute) -> (then, else, MyPlayerState:Player State Object Reference)
   n2 = GetSparksMusicPlayspaceOfClass!(
    PlayspaceType:Sparks Music Playspace Class Reference=/PilgrimQuickplay/Playspaces/BP_PilgrimQuickplayPlayspace.BP_PilgrimQuickplayPlayspace_C
    Actor:Actor Object Reference=@n1
    bMatchChildTypes:bool=false
   ) -> (ReturnValue:BP Pilgrim Quickplay Playspace Object Reference)
   > n3 = Is Valid(InputObject=@n2)
   Return Node!(execute=@n3.Is Not Valid, Has Songs Selected:bool=false)
   n4 = Get PilgrimCustomSetlist!(self:BP Pilgrim Quickplay Playspace Object Reference=@n2) -> (PilgrimCustomSetlist:Pilgrim Custom Setlist Object Reference)
   > Return Node(Has Songs Selected="integer > integer"(A=GetNumSongs(self=@n4) -> (ReturnValue=0)))
   Return Node(execute=@n1.else, Has Songs Selected=false)
  ]}
 Set Ready State Display {nodes [
   n1 = Set Ready State Display!() -> (then, Is Ready:bool=~, InNumSongs:int=-1) {is=title K2Node_FunctionEntry}
   > Set LocalIsReady(LocalIsReady:bool=@n1) -> (Output_Get:bool)
   > n2 = Sequence!(execute) -> (then_0, then_1)
   SetVisibility(execute=@n2.then_1, self=Get Block_SecondaryText(), InVisibility=Collapsed)
   > SetVisibility(
    self=[Get ReadyLayer(), Get ReadyIndicator() -> (ReadyIndicator:Overlay Object Reference)]
    InVisibility=Select(Option 0=Collapsed, Option 1=HitTestInvisible, Index=Get LocalIsReady() -> (LocalIsReady:bool))
   )
   > n3 = Branch(Condition=Get LocalIsReady())
   > "SetText (Text)"(self=Get Block_Text(), InText="LOCTEXT(\"Unready\")")
   > "SetText (Text)"(self=Get Block_SecondaryText(), InText="LOCTEXT(\"(Hold)\")")
   > n4 = GetDynamicMaterial(self=Get ButtonImage())
   > SetVectorParameterValue(self=@n4, ParameterName=Color_Idle_A, Value=(R=0.525,G=0.6,B=0.75,A=1))
   > SetScalarParameterValue(self=@n4, ParameterName=PostProcessMixAmount, Value=0.8)
   > PlayAnimation(InAnimation=Get ReadyUp() -> (ReadyUp:Widget Animation Object Reference))
   "SetText (Text)"(execute=@n3.else, self=Get Block_Text(), InText="LOCTEXT(\"Ready Up\")")
   > "SetText (Text)"(self=Get Block_SecondaryText(), InText="LOCTEXT(\"(Hold)\")")
   > n5 = GetDynamicMaterial(self=Get ButtonImage())
   > n6 = Branch(Condition=Get IsReadyUpDisabled())
   > SetVectorParameterValue(self=@n5, ParameterName=Color_Idle_A, Value=(R=0.525,G=0.6,B=0.75,A=1))
   > n7 = SetScalarParameterValue(self=@n5, ParameterName=PostProcessMixAmount, Value=0.8)
   StopAnimation(execute=[@n8.then, @n7.then], InAnimation=Get ReadyUp())
   SetVectorParameterValue(execute=@n6.else, self=@n5, ParameterName=Color_Idle_A, Value=(R=0.9301,G=1,B=0.0103,A=1))
   > n8 = SetScalarParameterValue(self=@n5, ParameterName=PostProcessMixAmount, Value=0.1)
   n9 = Get InNumSongs() -> (InNumSongs:int=0)
   n10 = Branch(execute=@n2.then_0, Condition="integer >= integer"(A=@n9))
   > n11 = Set LocalNumSongs(LocalNumSongs:int=@n9) -> (Output_Get:int=0)
   n12 = Get LocalNumSongs() -> (LocalNumSongs:int=0)
   n13 = Branch(
    execute=[@n11.then, @n18.then, @n16.else]
    Condition=OR Boolean!(A="Equal (Integer)"(A=@n12), B=Get _DisabledByExternalSource()) -> (ReturnValue)
   )
   > Set IsReadyUpDisabled(IsReadyUpDisabled:bool=true) -> (Output_Get:bool)
   n14 = Get KeybindWidget_ChangeBindings_1()
   > SetForcedHoldKeybindStatus!(execute, self:Fort Keybind Widget Object Reference=@n14, InForcedHoldKeybindStatus:EFortKeybindForcedHoldStatus Enum=NeverShowHold) -> (then)
   > SetRenderOpacity(self=@n14, InOpacity=0.4)
   > SetRenderTranslation(self=@n14, Translation=(X=0,Y=0))
   > Set LocalIsReady(LocalIsReady=false)
   > SetVisibility(self=Get Temp_HoldToFill_Underlayer() -> (Temp_HoldToFill_Underlayer:Image Object Reference), InVisibility=Hidden)
   Set IsReadyUpDisabled(execute=@n13.else, IsReadyUpDisabled=false)
   n15 = Get KeybindWidget_ChangeBindings_1()
   > SetForcedHoldKeybindStatus(self=@n15, InForcedHoldKeybindStatus=ForcedHold)
   > SetRenderOpacity(self=@n15, InOpacity=1)
   > SetRenderTranslation(self=@n15, Translation=(X=0,Y=0))
   n16 = Get(execute=@n10.else)
   n17 = GetSparksMusicPlayspaceOfClass(PlayspaceType=&BP_PilgrimQuickplayPlayspace_C, Actor=@n16, bMatchChildTypes=false)
   > n18 = Set LocalNumSongs(LocalNumSongs=GetNumSongs(self=Get PilgrimCustomSetlist(self=@n17)) -> (ReturnValue=0)) -> (Output_Get=0)
  ]}
 Set Up Player {nodes [
   n1 = Set Up Player!() -> (then, PlayerState:Player State Object Reference) {is=title K2Node_FunctionEntry}
   > n2 = Set MyPlayerState(MyPlayerState:Player State Object Reference=@n1) -> (Output_Get:Player State Object Reference)
   > n3 = Cast To FortPlayerStateAthena!(execute, Object=@n2) -> (then, CastFailed, AsFort Player State Athena:Fort Player State Athena Object Reference) {is=title K2Node_DynamicCast}
   n4 = Get MyPlayerState()
   > n5 = GetPilgrimQuickplayPlayerStateComponent(PlayerState=@n4)
   n6 = GetSparksMusicPlayspaceOfClass(PlayspaceType=&BP_PilgrimQuickplayPlayspace_C, Actor=@n4, bMatchChildTypes=false)
   > Set Ready State Display(Is Ready=GetIsReady(self=@n5), InNumSongs=GetNumSongs(self=Get PilgrimCustomSetlist(self=@n6)) -> (ReturnValue=0)) {is=title K2Node_CallFunction}
   PrintString!(execute=@n3.CastFailed, InString="CAST FAILED!", bPrintToScreen, bPrintToLog, TextColor=(R=1,G=0.078,B=0,A=1), Duration=8, Key) -> (then)
  ]}
 SetDisabledByExternalSource {nodes [
   n1 = SetDisabledByExternalSource!() -> (then, DisabledByExternalSource:bool=~) {is=title K2Node_FunctionEntry}
   > Set _DisabledByExternalSource(_DisabledByExternalSource:bool=@n1) -> (Output_Get:bool)
   > Set Ready State Display(Is Ready=IsVisible(self=Get ReadyIndicator()), InNumSongs=-1)
  ]}
 SetPlayerColor {nodes [
   n1 = SetPlayerColor!() -> (then, PlayerColor:Linear Color Structure) {is=title K2Node_FunctionEntry}
   n2 = "To SlateColor (Color)"!(Value:Linear Color Structure=@n1) -> (ReturnValue:Slate Color Structure)
   > SetBrushTintColor!(execute, self:Image Object Reference=Get ReadyLayer(), TintColor:Slate Color Structure=@n2) -> (then)
   > SetBrushTintColor(self=Get ButtonImage(), TintColor=@n2)
  ]}
]
variables [
 _DisabledByExternalSource {type=boolean}
 AsyncEventHandles {type="array<CancellableAsyncAction>", default []}
 CountDown {type=WidgetAnimation, default=null}
 EventRouter {type=GameplayEventRouterComponent, default=null}
 InitialCountdownState {type=WidgetAnimation, default=null}
 IsReadyUpDisabled {type=boolean}
 MyPlayerState {type=PlayerState, default=null}
 ReadyText {type=string, default=Ready Up}
 ReadyText_0 {type=string, default=Ready Up}
 ReadyUpComplete Sound {
  type=SoundBase
  default=/Game/Sounds/Fort_UI_Sounds/FrontEnd/DefaultSounds/HoldableButtons/Holdable_Loud/MSS_UI_Button_Holdable_Loud_Complete
 }
 ReadyUpHold Component {type=AudioComponent, default=null}
 ReadyUpHold Sound {type=SoundBase, default=/Game/Sounds/Fort_UI_Sounds/FrontEnd/DefaultSounds/HoldableButtons/Holdable_Loud/MSS_UI_Button_Holdable_Loud_Hold_lp}
 ReadyUpPress Component {type=AudioComponent, default=null}
 ReadyUpPress Sound {type=SoundBase, default=/Game/Sounds/Fort_UI_Sounds/FrontEnd/DefaultSounds/HoldableButtons/Holdable_Loud/MSS_UI_Button_Holdable_Loud_Press}
 ReadyUpRelease Component {type=AudioComponent, default=null}
 ReadyUpRelease Sound {type=SoundBase, default=/Game/Sounds/Fort_UI_Sounds/FrontEnd/DefaultSounds/HoldableButtons/Holdable_Loud/MSS_UI_Button_Holdable_Loud_Release}
]
widgets {
 info {inheritedWidgetCount=0, parentClass=/Script/PilgrimCoreUI.PilgrimGameWidgetBase, rootWidgetClass=&Overlay, widgetCount=21}
 widgets [
  PromptParent {widgetClassPath=&Overlay, bIsVariable=true, widget=PromptParent}
  HorizontalBoxReadyUnreadyWithInput {
   widgetClassPath=&Overlay
   bIsVariable=true
   parent=PromptParent
   widget=HorizontalBoxReadyUnreadyWithInput
   slot {class=OverlaySlot, properties {horizontalAlignment=HAlign_Fill, verticalAlignment=VAlign_Fill}}
  }
  ReadyUpOverlay {
   widgetClassPath=&Overlay
   bIsVariable=true
   parent=HorizontalBoxReadyUnreadyWithInput
   widget=ReadyUpOverlay
   slot {class=OverlaySlot, properties {horizontalAlignment=HAlign_Fill, verticalAlignment=VAlign_Fill}}
  }
  Spacer_130 {
   widgetClassPath=&Spacer
   parent=ReadyUpOverlay
   widget=Spacer_130
   properties {size=<1 64>}
   slot {class=OverlaySlot, properties {verticalAlignment=VAlign_Fill}}
  }
  ReadyLayer {
   widgetClassPath=&Image
   bIsVariable=true
   parent=ReadyUpOverlay
   widget=ReadyLayer
   properties {
    visibility=Collapsed
    brush {
     bIsDynamicallyLoaded=false
     drawAs=Image
     imageType=NoImage
     mirroring=NoMirror
     resourceName=None
     resourceObject=/PilgrimCore/Widgets/Materials/MI_UI_Pilgrim_ReadyButtonLayer
     tiling=NoTile
     imageSize=<32 32>
     margin=<ltrb 0 0 0 0>
     outlineSettings {
      bUseBrushTransparency=false
      roundingType=HalfHeightRadius
      width=0
      color {colorUseRule=UseColor_Specified, specifiedColor=<rgba 0 0 0 0>}
      cornerRadii=<0 0 0 0>
     }
     tintColor {colorUseRule=UseColor_Specified, specifiedColor=<rgba 1 1 1 1>}
     uVRegion {bIsValid=false, max=<0 0>, min=<0 0>}
    }
   }
   slot {class=OverlaySlot, properties {horizontalAlignment=HAlign_Fill, verticalAlignment=VAlign_Fill}}
  }
  ButtonImage {
   widgetClassPath=&Image
   bIsVariable=true
   parent=ReadyUpOverlay
   widget=ButtonImage
   properties {brush {
     bIsDynamicallyLoaded=false
     drawAs=Image
     imageType=FullColor
     mirroring=NoMirror
     resourceName=None
     resourceObject=/SparksUICommon/UILibrary/MI_FMCore_Button_Regular_ReadyUp
     tiling=NoTile
     imageSize=<32 32>
     margin=<ltrb 0 0 0 0>
     outlineSettings {
      bUseBrushTransparency=false
      roundingType=HalfHeightRadius
      width=0
      color {colorUseRule=UseColor_Specified, specifiedColor=<rgba 0 0 0 0>}
      cornerRadii=<0 0 0 0>
     }
     tintColor {colorUseRule=UseColor_Specified, specifiedColor=<rgba 1 1 1 1>}
     uVRegion {bIsValid=false, max=<0 0>, min=<0 0>}
    }}
   slot {class=OverlaySlot, properties {horizontalAlignment=HAlign_Fill, verticalAlignment=VAlign_Fill}}
  }
  VerticalBox_Text {
   widgetClassPath=&VerticalBox
   bIsVariable=true
   parent=ReadyUpOverlay
   widget=VerticalBox_Text
   slot {class=OverlaySlot, properties {horizontalAlignment=HAlign_Center, verticalAlignment=VAlign_Center, padding=<ltrb 64 8 64 8>}}
  }
  CommonVisibilityWidgetLegacy_120 {
   widgetClassPath=&CommonVisibilityWidgetLegacy
   parent=VerticalBox_Text
   widget=CommonVisibilityWidgetLegacy_120
   properties {bShowForTouch=false}
   slot {class=VerticalBoxSlot, properties {horizontalAlignment=HAlign_Center}}
  }
  Block_Text {
   widgetClassPath=&CommonTextBlock
   bIsVariable=true
   parent=CommonVisibilityWidgetLegacy_120
   widget=Block_Text
   properties {
    justification=Center
    text=Ready Up
    textTransformPolicy=ToUpper
    colorAndOpacity {colorUseRule=UseColor_Specified, specifiedColor=<rgba 0.0048 0.004 0.0116 1>}
    font {
     bForceMonospaced=false
     bMaterialIsStencil=false
     fontMaterial=null
     fontObject=&HeadingNow
     letterSpacing=0
     monospacedWidth=1
     size=21
     skewAmount=0
     typefaceFontName=86 Bold Italic
     outlineSettings {
      bApplyOutlineToDropShadows=false
      bMiteredCorners=false
      bSeparateFillAlpha=false
      outlineMaterial=null
      outlineSize=0
      outlineColor=<rgba 0 0 0 1>
     }
    }
    shadowOffset=<0 0>
   }
   slot {class=BorderSlot}
  }
  CommonVisibilityWidgetLegacy {
   widgetClassPath=&CommonVisibilityWidgetLegacy
   parent=VerticalBox_Text
   widget=CommonVisibilityWidgetLegacy
   properties {bShowForGamepad=false, bShowForMouseAndKeyboard=false}
   slot {class=VerticalBoxSlot, properties {horizontalAlignment=HAlign_Center}}
  }
  Block_Text_1 {
   widgetClassPath=&CommonTextBlock
   bIsVariable=true
   parent=CommonVisibilityWidgetLegacy
   widget=Block_Text_1
   properties {
    justification=Center
    text=Ready Up
    textTransformPolicy=ToUpper
    colorAndOpacity {colorUseRule=UseColor_Specified, specifiedColor=<rgba 0.0048 0.004 0.0116 1>}
    font {
     bForceMonospaced=false
     bMaterialIsStencil=false
     fontMaterial=null
     fontObject=&HeadingNow
     letterSpacing=0
     monospacedWidth=1
     size=24
     skewAmount=0
     typefaceFontName=86 Bold Italic
     outlineSettings {
      bApplyOutlineToDropShadows=false
      bMiteredCorners=false
      bSeparateFillAlpha=false
      outlineMaterial=null
      outlineSize=0
      outlineColor=<rgba 0 0 0 1>
     }
    }
    shadowOffset=<0 0>
   }
   slot {class=BorderSlot}
  }
  Block_SecondaryText {
   widgetClassPath=&CommonTextBlock
   bIsVariable=true
   parent=VerticalBox_Text
   widget=Block_SecondaryText
   properties {
    text=(Hold)
    colorAndOpacity {colorUseRule=UseColor_Specified, specifiedColor=<rgba 0.0048 0.004 0.0116 0.5>}
    font {
     bForceMonospaced=false
     bMaterialIsStencil=false
     fontMaterial=null
     fontObject=&HeadingNow
     letterSpacing=0
     monospacedWidth=1
     size=15
     skewAmount=0
     typefaceFontName=75 Medium
     outlineSettings {
      bApplyOutlineToDropShadows=false
      bMiteredCorners=false
      bSeparateFillAlpha=false
      outlineMaterial=null
      outlineSize=0
      outlineColor=<rgba 0 0 0 1>
     }
    }
    renderTransform {angle=0, scale=<1 1>, shear=<0 0>, translation=<0 -2>}
    shadowOffset=<0 0>
   }
   slot {class=VerticalBoxSlot, properties {horizontalAlignment=HAlign_Center}}
  }
  ReadyIndicator {
   widgetClassPath=&Overlay
   bIsVariable=true
   parent=ReadyUpOverlay
   widget=ReadyIndicator
   slot {class=OverlaySlot, properties {horizontalAlignment=HAlign_Right, padding=<ltrb 0 11 11 0>}}
  }
  Check_Glow {
   widgetClassPath=&Image
   parent=ReadyIndicator
   widget=Check_Glow
   properties {
    visibility=SelfHitTestInvisible
    brush {
     bIsDynamicallyLoaded=false
     drawAs=Image
     imageType=NoImage
     mirroring=NoMirror
     resourceName=None
     resourceObject=/PilgrimCore/Widgets/Materials/MI_UI_PilgrimReadyGlow
     tiling=NoTile
     imageSize=<32 32>
     margin=<ltrb 0 0 0 0>
     outlineSettings {
      bUseBrushTransparency=false
      roundingType=HalfHeightRadius
      width=0
      color {colorUseRule=UseColor_Specified, specifiedColor=<rgba 0 0 0 0>}
      cornerRadii=<0 0 0 0>
     }
     tintColor {colorUseRule=UseColor_Specified, specifiedColor=<rgba 1 1 1 1>}
     uVRegion {bIsValid=false, max=<0 0>, min=<0 0>}
    }
    renderTransform {angle=0, scale=<2.2 2.2>, shear=<0 0>, translation=<0 0>}
   }
   slot {class=OverlaySlot, properties {horizontalAlignment=HAlign_Fill, verticalAlignment=VAlign_Fill}}
  }
  Check_Pulse {
   widgetClassPath=&Image
   parent=ReadyIndicator
   widget=Check_Pulse
   properties {
    visibility=SelfHitTestInvisible
    brush {
     bIsDynamicallyLoaded=false
     drawAs=Image
     imageType=NoImage
     mirroring=NoMirror
     resourceName=None
     resourceObject=/PilgrimCore/Widgets/Materials/MI_UI_PilgrimReadyStatePulse
     tiling=NoTile
     imageSize=<32 32>
     margin=<ltrb 0 0 0 0>
     outlineSettings {
      bUseBrushTransparency=false
      roundingType=HalfHeightRadius
      width=0
      color {colorUseRule=UseColor_Specified, specifiedColor=<rgba 0 0 0 0>}
      cornerRadii=<0 0 0 0>
     }
     tintColor {colorUseRule=UseColor_Specified, specifiedColor=<rgba 1 1 1 1>}
     uVRegion {bIsValid=false, max=<0 0>, min=<0 0>}
    }
    renderTransform {angle=0, scale=<2.2 2.2>, shear=<0 0>, translation=<0 0>}
   }
   slot {class=OverlaySlot, properties {horizontalAlignment=HAlign_Fill, verticalAlignment=VAlign_Fill}}
  }
  StatusIndicator {
   widgetClassPath=/FortUILibrary/UIKit/StatusIndicator/WBP_UIKit_StatusIndicator
   bIsVariable=true
   parent=ReadyIndicator
   widget=StatusIndicator
   properties {icon=/FortUILibrary/Textures/MaterialInstances/MI_UIKit_Icon_Check, overrideStatusIndicatorVisibility=true}
   slot {class=OverlaySlot, properties {horizontalAlignment=HAlign_Center, verticalAlignment=VAlign_Center}}
  }
  CommonVisibilityWidgetLegacy_0 {
   widgetClassPath=&CommonVisibilityWidgetLegacy
   parent=HorizontalBoxReadyUnreadyWithInput
   widget=CommonVisibilityWidgetLegacy_0
   properties {bShowForTouch=false, padding=<ltrb 0 0 0 0>}
   slot {class=OverlaySlot, properties {verticalAlignment=VAlign_Center, padding=<ltrb 23 0 0 0>}}
  }
  Overlay_0 {
   widgetClassPath=&Overlay
   parent=CommonVisibilityWidgetLegacy_0
   widget=Overlay_0
   slot {class=BorderSlot, properties {padding=<ltrb 0 0 0 0>}}
  }
  Temp_HoldToFill_Underlayer {
   widgetClassPath=&Image
   bIsVariable=true
   parent=Overlay_0
   widget=Temp_HoldToFill_Underlayer
   properties {
    brush {
     bIsDynamicallyLoaded=false
     drawAs=Image
     imageType=FullColor
     mirroring=NoMirror
     resourceName=None
     resourceObject=/PilgrimCore/Widgets/Materials/MI_UI_Pilgrim_Panel_Underlayer_ColorFill_PromtUnderlayer
     tiling=NoTile
     imageSize=<34 34>
     margin=<ltrb 0 0 0 0>
     outlineSettings {
      bUseBrushTransparency=false
      roundingType=HalfHeightRadius
      width=0
      color {colorUseRule=UseColor_Specified, specifiedColor=<rgba 0 0 0 0>}
      cornerRadii=<0 0 0 0>
     }
     tintColor {colorUseRule=UseColor_Specified, specifiedColor=<rgba 1 1 1 0.45>}
     uVRegion {bIsValid=false, max=<0 0>, min=<0 0>}
    }
    renderTransform {angle=0, scale=<1.5 1.5>, shear=<0 0>, translation=<0 0>}
   }
   slot {class=OverlaySlot, properties {horizontalAlignment=HAlign_Center, verticalAlignment=VAlign_Center}}
  }
  KeybindContainer {
   widgetClassPath=&ScaleBox
   bIsVariable=true
   parent=Overlay_0
   widget=KeybindContainer
   properties {stretch=UserSpecified, userSpecifiedScale=0.75}
   slot {class=OverlaySlot, properties {horizontalAlignment=HAlign_Center, verticalAlignment=VAlign_Center}}
  }
  KeybindWidget_ChangeBindings_1 {
   widgetClassPath=/FortUILibrary/Input/KeybindWidget
   bIsVariable=true
   parent=KeybindContainer
   widget=KeybindWidget_ChangeBindings_1
   properties {
    bUseKeyboardTextureForKeys=true
    boundAction=Pilgrim_ToggleReady
    boundInputAction=&IA_PilgrimToggleReady
    forcedHoldKeybindStatus=ForcedHold
    visibility=Visible
   }
   slot {class=ScaleBoxSlot}
  }
 ]
}
