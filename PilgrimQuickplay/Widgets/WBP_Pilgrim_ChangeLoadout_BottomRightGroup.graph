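# Widget Blueprint that groups Pilgrim quickplay green room controls in the bottom-right corner, named for the change-loadout screen.
# A root canvas holds three child widgets anchored bottom-right and stacked at different heights: difficulty select, toggle ready, and return to results.
# It is layout only, with an empty event graph; the EventRouter and PlayerState variables are declared but unused here.
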WidgetBlueprint /PilgrimQuickplay/Widgets/WBP_Pilgrim_ChangeLoadout_BottomRightGroup
graphs [EventGraph {links [], nodes []}]
variables [
 EventRouter {type=GameplayEventRouterComponent, default=null}
 MyPlayerState {type=PlayerState, default=null}
]
widgets {
 info {inheritedWidgetCount=0, parentClass=&UserWidget, rootWidgetClass=&CanvasPanel, widgetCount=4}
 widgets [
  CanvasPanel {widgetClassPath=&CanvasPanel, bIsVariable=true, widget=CanvasPanel}
  WBP_Pilgrim_GreenRoom_DifficultySelectUI {
   widgetClassPath=&/WBP_Pilgrim_GreenRoom_DifficultySelectUI
   bIsVariable=true
   parent=CanvasPanel
   widget=WBP_Pilgrim_GreenRoom_DifficultySelectUI
   slot {class=CanvasPanelSlot, properties {layoutData {alignment=<0 0>, anchors {maximum=<1 1>, minimum=<1 1>}, offsets=<ltrb -24 -550 0 0>}}}
  }
  WBP_Pilgrim_GreenRoom_ToggleReadyUI {
   widgetClassPath=&/WBP_Pilgrim_GreenRoom_ToggleReadyUI
   bIsVariable=true
   parent=CanvasPanel
   widget=WBP_Pilgrim_GreenRoom_ToggleReadyUI
   slot {class=CanvasPanelSlot, properties {layoutData {alignment=<0 0>, anchors {maximum=<1 1>, minimum=<1 1>}, offsets=<ltrb -24 -600 0 0>}}}
  }
  WBP_Pilgrim_GreenRoom_ReturnToResultsUI {
   widgetClassPath=&/WBP_Pilgrim_GreenRoom_ReturnToResultsUI
   bIsVariable=true
   parent=CanvasPanel
   widget=WBP_Pilgrim_GreenRoom_ReturnToResultsUI
   slot {class=CanvasPanelSlot, properties {layoutData {alignment=<0 0>, anchors {maximum=<1 1>, minimum=<1 1>}, offsets=<ltrb -260 -656 236.4 40>}}}
  }
 ]
}
