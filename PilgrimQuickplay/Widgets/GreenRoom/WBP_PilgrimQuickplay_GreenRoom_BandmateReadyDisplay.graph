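# Green-room status panel for Pilgrim Quickplay that prompts players to choose a song or ready up, or shows a curated setlist's name.
# For non-curated setlists, event-router listeners track readiness and setlist size to update the text, and quickplay state to toggle visibility.
# Pressing ready-up while disabled replays a scale/fade emphasis on the text for that player.
# Listener handles are stored and stopped on Destruct; the nested BandmateReadyDisplay starts collapsed.
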WidgetBlueprint /PilgrimQuickplay/Widgets/GreenRoom/WBP_PilgrimQuickplay_GreenRoom_BandmateReadyDisplay
animations [
 Default {
  binding_count=0
  length_seconds=5
  object=&self:Default
  sequence {bindings []}
 }
 ReaffirmMessage {
  binding_count=1
  length_seconds=0.5
  object=&self:ReaffirmMessage
  sequence {bindings [HelperText {
     id=#1
     possessed_class=CommonTextBlock
     tracks [
      {
       type=MovieScene2DTransformTrack
       sections [{
         type=MovieScene2DTransformSection
         end_frame=10
         start_frame=0
         channels [
          Scale.X {type=MovieSceneScriptingFloatChannel, keys [[0,1.2],[5,1]]}
          Scale.Y {type=MovieSceneScriptingFloatChannel, keys [[0,1.2],[5,1]]}
         ]
        }]
      }
      {
       type=MovieSceneColorTrack
       display_name=Color and Opacity
       sections [{
         type=MovieSceneColorSection
         end_frame=6
         start_frame=0
         channels [
          Color.R {type=MovieSceneScriptingFloatChannel, const [1,1]}
          Color.G {type=MovieSceneScriptingFloatChannel, const [1,1]}
          Color.B {type=MovieSceneScriptingFloatChannel, const [1,1]}
          Color.A {type=MovieSceneScriptingFloatChannel, keys [[0,0.5],[5,1]]}
         ]
        }]
      }
     ]
    }]}
 }
]
graphs [
 EventGraph {nodes [
   n1 = UpdateHelperText!() -> (OutputDelegate, then, OptInNumSongs:int=-1) {is=title K2Node_CustomEvent}
   > TryUpdateVisibility()
   > n2 = Branch(Condition="integer < integer"(A=@n1.OptInNumSongs))
   n3 = Get PilgrimCustomSetlist!(self:BP Pilgrim Quickplay Playspace Object Reference=GetSparksMusicPlayspaceOfClass!(
     PlayspaceType:Sparks Music Playspace Class Reference=/PilgrimQuickplay/Playspaces/BP_PilgrimQuickplayPlayspace.BP_PilgrimQuickplayPlayspace_C
     Actor:Actor Object Reference=GetOwningPlayer()
     bMatchChildTypes:bool=false
    ) -> (ReturnValue:BP Pilgrim Quickplay Playspace Object Reference)) -> (PilgrimCustomSetlist:Pilgrim Custom Setlist Object Reference)
   > n4 = Branch(Condition="integer > integer"(A=GetNumSongs!(self:Pilgrim Setlist Interface Interface=@n3) -> (ReturnValue:int=0)))
   n5 = Get NumPlayersReady() -> (NumPlayersReady:int=0)
   n6 = Get HelperText() -> (HelperText:Common Text Object Reference)
   "SetText (Text)"(
    execute=[@n7.then, @n4.then]
    self=@n6
    InText=Select(
     Option 0:Text=Get NoOneIsReadyText() -> (NoOneIsReadyText:Text)
     Option 1:Text=Get SomeoneIsReadyText() -> (SomeoneIsReadyText:Text)
     Index="integer > integer"(A=@n5)
    ) -> (ReturnValue:Text)
   )
   > Branch(Condition=Get AllPlayersAreReady() -> (AllPlayersAreReady:bool))
   > "SetText (Text)"(self=@n6, InText=Get HittingStageText() -> (HittingStageText:Text))
   "SetText (Text)"(execute=[@n7.else, @n4.else], self=Get HelperText(), InText="LOCTEXT(\"Choose a song\")")
   n7 = Branch(execute=@n2.else, Condition="integer > integer"(A=@n1.OptInNumSongs))
   n8 = SetupForCuratedSetlistVM!() -> (OutputDelegate, then, isCuratedSetlist:bool=~) {is=title K2Node_CustomEvent}
   > Branch(Condition=@n8.isCuratedSetlist)
   > "SetText (Text)"(
    self=Get HelperText()
    InText=Get CuratedSetlistName!(self:Pilgrim Curated Setlist VM Object Reference=Get PilgrimCuratedSetlistVM() -> (PilgrimCuratedSetlistVM:Pilgrim Curated Setlist VM Object Reference)) -> (CuratedSetlistName:Text)
   )
   Event Destruct()
   > Parent: Destruct()
   n9 = Get PlayerStateEventHandles() -> (PlayerStateEventHandles:Array of Async Action Start Listening To Event Object References)
   > n10 = For Each Loop(Array:Array of Async Action Start Listening To Event Object References=@n9) -> (Array Element:Async Action Start Listening To Event Object Reference)
   Clear(execute=@n10.Completed, TargetArray:Array of Async Action Start Listening To Event Object References=@n9)
   Is Valid(exec=@n10.LoopBody, InputObject=@n10)
   > StopListeningToEvent(self:Async Action Start Listening To Event Object Reference=@n10)
   Event Construct()
   > n11 = PlaylistDataReadyAsync!(execute:Exec) -> (
    then:Exec
    OnReady:Exec
    GameState:Fort Game State Athena Object Reference
    Playlist:Fort Playlist Object Reference
    PlaylistContextTags:"Gameplay Tag Container Structure (by ref)"
   ) {is=title K2Node_AsyncAction}
   n12 = PilgrimWaitForEventRouter!(execute:Exec=@n11.OnReady, ContextActor:Actor Object Reference=GetOwningPlayer()) -> (then:Exec, OnEventRouterAvailable:Exec, EventRouter:Gameplay Event Router Component Object Reference) {is=title K2Node_AsyncAction}
   Set Event Router(execute=@n12.OnEventRouterAvailable, Event Router:Gameplay Event Router Component Object Reference=@n12) -> (Output_Get:Gameplay Event Router Component Object Reference)
   > n13 = Branch(Condition=Get bIsCuratedSetlist!(self:Pilgrim Curated Setlist VM Object Reference=Get PilgrimCuratedSetlistVM()) -> (bIsCuratedSetlist:bool))
   TryUpdateVisibility(execute=@n13.else)
   > n14 = Sequence!(execute) -> (then_0, then_1, then_2:Exec, then_3:Exec, then_4:Exec, then_5:Exec)
   > n15 = StartListeningToEvent!(
    execute:Exec
    Target:Gameplay Event Router Component Object Reference=Get Event Router() -> (Event Router:Gameplay Event Router Component Object Reference)
    EventType:Script Struct Object Reference=/Script/PilgrimCoreRuntime.PilgrimEvent_PlayerState_ReadinessChanged
    Context:Object Reference
    EventBubblingRule:EEventBubblingRule Enum
   ) -> (
    then:Exec
    AsyncTaskProxy:Async Action Start Listening To Event Object Reference
    OnEventReceived:Exec
    Payload_ForPlayer:Fort Player Pawn Object Reference
    Payload_bIsReady:bool=~
   )
   > Add(
    TargetArray:Array of Async Action Start Listening To Event Object References=Get PlayerStateEventHandles()
    NewItem:"Async Action Start Listening To Event Object Reference (by ref)"=@n15
   ) {is=title K2Node_CallArrayFunction}
   n16 = Branch(execute=@n15.OnEventReceived, Condition=@n15.Payload_bIsReady)
   n17 = Get NumPlayersReady() -> (NumPlayersReady=0)
   > n18 = Increment Int(Value=@n17)
   UpdateHelperText!(execute=[@n18."  ", @n19."  "], self, OptInNumSongs:int=-1) -> (then) {is=title K2Node_CallFunction}
   n19 = Decrement Int(" "=@n16.else, Value=@n17)
   n20 = StartListeningToEvent!(
    execute=@n14.then_4
    Target=Get Event Router()
    EventType=/PilgrimCore/Player/Components/PilgrimEvent_ReadyUpPressedWhenDisabled
    Context
    EventBubblingRule
   ) -> (
    then
    AsyncTaskProxy
    OnEventReceived
    Payload_FortPlayerController_11_#2:Fort Player Controller Object Reference
   )
   > Add(TargetArray=Get PlayerStateEventHandles(), NewItem=@n20)
   Branch(
    execute=@n20.OnEventReceived
    Condition="Equal (Object)"(A=@n20.Payload_FortPlayerController_11_#2, B=GetOwningPlayer())
   )
   > PlayAnimation(InAnimation=Get ReaffirmMessage() -> (ReaffirmMessage:Widget Animation Object Reference))
   n21 = StartListeningToEvent!(
    execute=@n14.then_3
    Target=Get Event Router()
    EventType=/Script/PilgrimCoreRuntime.PilgrimEvent_BandState_CustomSetlistChanged
    Context
    EventBubblingRule
   ) -> (then, AsyncTaskProxy, OnEventReceived, Payload_ChangedSetlist:Pilgrim Custom Setlist Object Reference)
   > Add(TargetArray=Get PlayerStateEventHandles(), NewItem=@n21)
   UpdateHelperText(execute=@n21.OnEventReceived, OptInNumSongs=GetNumSongs(self=@n21.Payload_ChangedSetlist) -> (ReturnValue=0))
   n22 = StartListeningToEvent!(
    execute=@n14.then_2
    Target=Get Event Router()
    EventType=/Script/PilgrimQPSharedRuntime.PilgrimEvent_QuickplayState_Began
    Context
    EventBubblingRule
   ) -> (then, AsyncTaskProxy, OnEventReceived, Payload_StateId:EPilgrimQuickplayState Enum)
   > Add(TargetArray=Get PlayerStateEventHandles(), NewItem=@n22)
   n23 = Branch(
    execute=@n22.OnEventReceived
    Condition=OR Boolean!(
     A="Equal (Enum)"(A:EPilgrimQuickplayState Enum=@n22.Payload_StateId, B:EPilgrimQuickplayState Enum=Pregame)
     B="Equal (Enum)"(A=@n22.Payload_StateId, B=SongResults)
     C:bool="Equal (Enum)"(A=@n22.Payload_StateId, B=SetResults)
    ) -> (ReturnValue)
   )
   > SetVisibility(InVisibility=SelfHitTestInvisible)
   SetVisibility(execute=@n23.else, InVisibility=Hidden)
   n24 = StartListeningToEvent!(
    execute=@n14.then_1
    Target=Get Event Router()
    EventType=/Script/PilgrimCoreRuntime.PilgrimEvent_BandState_ReadinessChanged
    Context
    EventBubblingRule
   ) -> (then, AsyncTaskProxy, OnEventReceived, Payload_bAreAllPlayersReady:bool=~)
   > Add(TargetArray=Get PlayerStateEventHandles(), NewItem=@n24)
   Set AllPlayersAreReady(execute=@n24.OnEventReceived, AllPlayersAreReady:bool=@n24.Payload_bAreAllPlayersReady) -> (Output_Get:bool)
   > UpdateHelperText(OptInNumSongs=-1)
   n25 = GetSparksMusicPlayspaceOfClass(
    PlayspaceType=&BP_PilgrimQuickplayPlayspace_C
    Actor=GetOwningPlayer()
    bMatchChildTypes=false
   )
   UpdateHelperText(execute=@n14.then_5, OptInNumSongs=GetNumSongs(self=Get PilgrimCustomSetlist(self=@n25)) -> (ReturnValue=0))
  ]}
 IsReadyUpAvailable {nodes [
   IsReadyUpAvailable() {is=title K2Node_FunctionEntry}
   n1 = Get PilgrimCustomSetlist(self=GetSparksMusicPlayspaceOfClass(
     PlayspaceType=&BP_PilgrimQuickplayPlayspace_C
     Actor=GetOwningPlayer()
     bMatchChildTypes=false
    ))
   > Return Node!(execute, IsReadyUpAvailable:bool="integer > integer"(A=GetNumSongs(self=@n1) -> (ReturnValue=0)))
  ]}
 TryUpdateVisibility {nodes [
   TryUpdateVisibility() {is=title K2Node_FunctionEntry}
   > Get!(execute:Exec) -> (then:Exec, else:Exec, HiddenWhenReadyUpIsUnavailable:bool)
   > SetVisibility(InVisibility=Select(
     Option 0:ESlateVisibility Enum=Hidden
     Option 1:ESlateVisibility Enum=HitTestInvisible
     Index=IsReadyUpAvailable!() -> (IsReadyUpAvailable:bool) {is=title K2Node_CallFunction}
    ) -> (ReturnValue:ESlateVisibility Enum))
  ]}
]
variables [
 AllPlayersAreReady {type=boolean}
 Event Router {type=GameplayEventRouterComponent, default=null}
 HiddenWhenReadyUpIsUnavailable {type=boolean}
 HittingStageText {type=string, default=Hitting the Stage}
 NoOneIsReadyText {type=string, default=Ready Up to start playing}
 NumPlayersReady {type=integer}
 PlayerStateEventHandles {type="array<AsyncAction_StartListeningToEvent>", default []}
 SomeoneIsReadyText {type=string, default=Waiting for players...}
]
widgets {
 info {inheritedWidgetCount=0, parentClass=&CommonUserWidget, rootWidgetClass=&Overlay, widgetCount=5}
 widgets [
  Overlay_1 {widgetClassPath=&Overlay, widget=Overlay_1}
  WBP_Sparks_BackingPanel {
   widgetClassPath=/PilgrimCore/Widgets/SparksUILibrary/WBP_Sparks_BackingPanel
   bIsVariable=true
   parent=Overlay_1
   widget=WBP_Sparks_BackingPanel
   properties {cornerRadius=Large}
   slot {class=OverlaySlot, properties {horizontalAlignment=HAlign_Fill, verticalAlignment=VAlign_Fill}}
  }
  VerticalBox_0 {
   widgetClassPath=&VerticalBox
   parent=Overlay_1
   widget=VerticalBox_0
   slot {class=OverlaySlot, properties {horizontalAlignment=HAlign_Center, verticalAlignment=VAlign_Center, padding=<ltrb 20 8 20 8>}}
  }
  HelperText {
   widgetClassPath=&CommonTextBlock
   bIsVariable=true
   parent=VerticalBox_0
   widget=HelperText
   properties {
    text=Choose a song
    textTransformPolicy=ToUpper
    colorAndOpacity {colorUseRule=UseColor_Specified, specifiedColor=<rgba 0.5 0.8333 1 1>}
    font {
     bForceMonospaced=false
     bMaterialIsStencil=false
     fontMaterial=null
     fontObject=&HeadingNow
     letterSpacing=0
     monospacedWidth=1
     size=27
     skewAmount=0
     typefaceFontName=76 Bold Italic
     outlineSettings {
      bApplyOutlineToDropShadows=false
      bMiteredCorners=false
      bSeparateFillAlpha=false
      outlineMaterial=null
      outlineSize=0
      outlineColor=<rgba 0 0 0 1>
     }
    }
    shadowOffset=<0 0>
   }
   slot {class=VerticalBoxSlot}
  }
  WBP_Pilgrim_Quickplay_BandmateReadyDisplay_60 {
   widgetClassPath=/PilgrimQuickplay/Widgets/Results/WBP_Pilgrim_Quickplay_BandmateReadyDisplay
   bIsVariable=true
   parent=VerticalBox_0
   widget=WBP_Pilgrim_Quickplay_BandmateReadyDisplay_60
   properties {visibility=Collapsed}
   slot {class=VerticalBoxSlot, properties {horizontalAlignment=HAlign_Center, verticalAlignment=VAlign_Center, padding=<ltrb 0 12 0 12>}}
  }
 ]
}
