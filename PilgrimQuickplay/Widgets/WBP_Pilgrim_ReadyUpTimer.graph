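# Pilgrim Quickplay widget for the green room ready-up countdown, switching between a "Starting in" timer and a "Sending Band to Stage" panel.
# Construct seeds state and count from the player manager playspace component's replicated values.
# OnInitialized awaits playlist data and the event router, then listens for timer, quickplay state and all-players-ready events to update text and visibility.
# Waiting and matchmaking-error panels exist in the layout, but these graphs never activate them.
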WidgetBlueprint /PilgrimQuickplay/Widgets/WBP_Pilgrim_ReadyUpTimer
graphs [
 EventGraph {nodes [
   n1 = UpdateTimerByState!() -> (OutputDelegate, then, GreenRoomTimerState:GreenRoomTimerState Enum=NewEnumerator0) {is=title K2Node_CustomEvent}
   > n2 = Switch on GreenRoomTimerState!(execute, Selection:GreenRoomTimerState Enum=@n1.GreenRoomTimerState) -> (NewEnumerator0:Exec, NewEnumerator1:Exec, NewEnumerator2:Exec, NewEnumerator3:Exec, NewEnumerator4:Exec) {is=title K2Node_SwitchEnum}
   SetActiveWidget(
    execute=@n2.NewEnumerator1
    self=Get ReadyAndStartSwitcher() -> (ReadyAndStartSwitcher:Common Visibility Switcher Object Reference)
    Widget=Get Container_CountdownStatus() -> (Container_CountdownStatus:Overlay Object Reference)
   )
   > SetVisibility(self=Get Container_CountdownStatus(), InVisibility=Visible)
   SetActiveWidget(execute=[@n2.NewEnumerator0, @n2.NewEnumerator3], self=Get ReadyAndStartSwitcher(), Widget=Get Container_CountdownStatus())
   > SetVisibility(self=Get Container_CountdownStatus(), InVisibility=Collapsed)
   SetActiveWidget(
    execute=@n2.NewEnumerator2
    self=Get ReadyAndStartSwitcher()
    Widget=Get Container_Starting() -> (Container_Starting:Overlay Object Reference)
   )
   n3 = Event PreConstruct()
   > Parent: PreConstruct(IsDesignTime=@n3.IsDesignTime)
   > Set Text(
    Text=Get Starting In Text() -> (Starting In Text:Text)
    self=Get Text_CountdownStatus() -> (Text_CountdownStatus:Common Text Object Reference)
   )
   K2Node_Event_1 = CustomEvent_2() {is=title K2Node_CustomEvent}
   > PlayAnimation(InAnimation=Get CountDown() -> (CountDown:Widget Animation Object Reference))
   > PlaySound2D(Sound=/Game/Sounds/Fort_UI_Sounds/FrontEnd/SC_UI_MatchMaking_TimeTick_Cue)
   K2Node_Event_2 = CustomEvent() {is=title K2Node_CustomEvent}
   > PlaySound2D(Sound=/Game/Sounds/Fort_UI_Sounds/FrontEnd/SC_UI_MatchMaking_ReadyUp_Cue)
   > PlayAnimation(InAnimation=Get InitialCountdownState() -> (InitialCountdownState:Widget Animation Object Reference))
   K2Node_Event_3 = CustomEvent_1() {is=title K2Node_CustomEvent}
   > PlaySound2D(Sound=/Game/Sounds/Fort_UI_Sounds/FrontEnd/SC_UI_MatchMaking_UnReady_Cue)
   Event Construct()
   n4 = GetSparksMusicPlayspaceOfClass!(
    PlayspaceType:Sparks Music Playspace Class Reference=/Script/PilgrimQuickplayRuntime.PilgrimQuickplayPlayspace
    Actor:Actor Object Reference=GetOwningPlayer()
    bMatchChildTypes:bool=false
   ) -> (ReturnValue:Pilgrim Quickplay Playspace Object Reference)
   > Is Valid(InputObject=@n4)
   n5 = GetComponentByClass!(
    self:Actor Object Reference=@n4
    ComponentClass:Actor Component Class Reference=/PilgrimQuickplay/Playspaces/BP_PilgrimQuickplay_PlayerManager_PlayspaceComponent.BP_PilgrimQuickplay_PlayerManager_PlayspaceComponent_C
   ) -> (ReturnValue:BP Pilgrim Quickplay Player Manager Playspace Component Object Reference)
   > Is Valid(InputObject=@n5)
   > UpdateTimerByState!(
    execute
    self
    GreenRoomTimerState:GreenRoomTimerState Enum=Get!(self:BP Pilgrim Quickplay Player Manager Playspace Component Object Reference=@n5) -> (
     PilgrimTimerState_Replicated_GreenRoomTimerState_2_#1:GreenRoomTimerState Enum
     PilgrimTimerState_Replicated_FinalTimerLength_5_#2:double=~
    )
   ) -> (then) {is=title K2Node_CallFunction}
   > "SetText (Text)"(
    self=Get Text_CountdownCurrent() -> (Text_CountdownCurrent:Common Text Object Reference)
    InText="To Text (Integer)"(Value=Get!(self=@n5) -> (Pulse Timer Count_TimerValue_7_#3:int))
   )
   Event OnInitialized()
   > n6 = PlaylistDataReadyAsync!(execute:Exec) -> (
    then:Exec
    OnReady:Exec
    GameState:Fort Game State Athena Object Reference
    Playlist:Fort Playlist Object Reference
    PlaylistContextTags:"Gameplay Tag Container Structure (by ref)"
   ) {is=title K2Node_AsyncAction}
   n7 = PilgrimWaitForEventRouter!(execute:Exec=@n6.OnReady, ContextActor:Actor Object Reference=GetOwningPlayer()) -> (then:Exec, OnEventRouterAvailable:Exec, EventRouter:Gameplay Event Router Component Object Reference) {is=title K2Node_AsyncAction}
   Set EventRouter(execute=@n7.OnEventRouterAvailable, EventRouter:Gameplay Event Router Component Object Reference=@n7) -> (Output_Get:Gameplay Event Router Component Object Reference)
   > n8 = Sequence!(execute) -> (then_0, then_1, then_2:Exec, then_3:Exec, then_4:Exec)
   n9 = StartListeningToEvent!(
    execute:Exec=@n8.then_1
    Target:Gameplay Event Router Component Object Reference=Get EventRouter() -> (EventRouter:Gameplay Event Router Component Object Reference)
    EventType:Script Struct Object Reference=/PilgrimQuickplay/Items/PilgrimEvent_GreenRoomTimerStateChanged
    Context:Object Reference
    EventBubblingRule:EEventBubblingRule Enum
   ) -> (
    then:Exec
    AsyncTaskProxy:Async Action Start Listening To Event Object Reference
    OnEventReceived:Exec
    Payload_GreenRoomTimerState_2_#1:GreenRoomTimerState Enum
    Payload_FinalTimerLength_5_#2:double=~
   )
   > Add(
    TargetArray:Array of Async Action Start Listening To Event Object References=Get GameplayEventHandles() -> (GameplayEventHandles:Array of Async Action Start Listening To Event Object References)
    NewItem:"Async Action Start Listening To Event Object Reference (by ref)"=@n9
   ) {is=title K2Node_CallArrayFunction}
   UpdateTimerByState(execute=@n9.OnEventReceived, GreenRoomTimerState=@n9.Payload_GreenRoomTimerState_2_#1)
   n10 = StartListeningToEvent!(
    execute=@n8.then_2
    Target=Get EventRouter()
    EventType=/PilgrimQuickplay/Actors/GreenRoomPlayerDisplay/PilgrimEvent_GreenRoomTimerChanged
    Context
    EventBubblingRule
   ) -> (then, AsyncTaskProxy, OnEventReceived, Payload_TimerValue_7_#3:int)
   > Add(TargetArray=Get GameplayEventHandles(), NewItem=@n10)
   n11 = Branch(execute=@n10.OnEventReceived, Condition="integer <= integer"(A=@n10.Payload_TimerValue_7_#3))
   "SetText (Text)"(
    execute=@n11.else
    self=Get Text_CountdownCurrent()
    InText="To Text (Integer)"(Value=@n10.Payload_TimerValue_7_#3)
   )
   n12 = Get EventRouter()
   n13 = StartListeningToEvent!(
    execute=@n8.then_3
    Target=@n12
    EventType=/Script/PilgrimQPSharedRuntime.PilgrimEvent_QuickplayState_Began
    Context
    EventBubblingRule
    None:%=@n12
   ) -> (then, AsyncTaskProxy, OnEventReceived, Payload_StateId:EPilgrimQuickplayState Enum)
   > Add(TargetArray=Get GameplayEventHandles(), NewItem=@n13)
   SetActiveWidget(execute=@n13.OnEventReceived, self=Get ReadyAndStartSwitcher(), Widget=Get Container_CountdownStatus())
   > n14 = Branch(Condition=OR Boolean!(
     A="Equal (Enum)"(A:EPilgrimQuickplayState Enum=@n13.Payload_StateId, B:EPilgrimQuickplayState Enum=Pregame)
     B="Equal (Enum)"(A=@n13.Payload_StateId, B=SetResults)
     C:bool="Equal (Enum)"(A=@n13.Payload_StateId, B=SongResults)
    ) -> (ReturnValue))
   > SetVisibility(InVisibility=Visible)
   > SetVisibility(self=Get Container_CountdownStatus(), InVisibility=Collapsed)
   SetVisibility(execute=@n14.else, InVisibility=Collapsed)
   n15 = Get EventRouter()
   n16 = StartListeningToEvent!(
    execute=@n8.then_4
    Target=@n15
    EventType=/Script/PilgrimQuickplayRuntime.PilgrimQuickplayAllPlayersReadyEvent
    Context
    EventBubblingRule
    None:%=@n15
   ) -> (then, AsyncTaskProxy, OnEventReceived, Payload:Pilgrim Quickplay All Players Ready Event Structure)
   > Add(TargetArray=Get GameplayEventHandles(), NewItem=@n16)
   SetActiveWidget(execute=@n16.OnEventReceived, self=Get ReadyAndStartSwitcher(), Widget=Get Container_Starting())
   K2Node_Event_7 = CustomEvent_3() {is=title K2Node_CustomEvent}
   > PlayAnimation(InAnimation=Get InitialCountdownState())
  ]}
 Set Ready State Display {nodes [
   n1 = Set Ready State Display!() -> (then, Is Ready:bool=~) {is=title K2Node_FunctionEntry}
   SetVisibility(
    self=Get Button_UnReady() -> (Button_UnReady:WBP UIKit Button Regular Object Reference)
    InVisibility=Select(Option 0=Collapsed, Option 1=Visible, Index=@n1)
   )
   > SetVisibility(
    self=Get Button_ReadyUp() -> (Button_ReadyUp:WBP UIKit Button Loud Object Reference)
    InVisibility=Select(Option 0=Visible, Option 1=Collapsed, Index=@n1)
   )
  ]}
]
variables [
 CountDown {type=WidgetAnimation, default=null}
 EventRouter {type=GameplayEventRouterComponent, default=null}
 InitialCountdownState {type=WidgetAnimation, default=null}
 MyPlayerState {type=PlayerState, default=null}
 Starting In Text {type=string, default=Starting in}
]
widgets {
 info {inheritedWidgetCount=0, parentClass=/Script/PilgrimCoreUI.PilgrimGameWidgetBase, rootWidgetClass=&Overlay, widgetCount=39}
 widgets [
  Timer {widgetClassPath=&Overlay, bIsVariable=true, widget=Timer}
  Switcher_MatchmakingStatus {
   widgetClassPath=&CommonWidgetSwitcherLegacy
   bIsVariable=true
   parent=Timer
   widget=Switcher_MatchmakingStatus
   slot {class=OverlaySlot}
  }
  ReadyAndStartSwitcher {
   widgetClassPath=&CommonVisibilitySwitcher
   bIsVariable=true
   parent=Switcher_MatchmakingStatus
   widget=ReadyAndStartSwitcher
   slot {class=WidgetSwitcherSlot}
  }
  Container_CountdownStatus {
   widgetClassPath=&Overlay
   bIsVariable=true
   parent=ReadyAndStartSwitcher
   widget=Container_CountdownStatus
   properties {visibility=Hidden}
   slot {class=CommonVisibilitySwitcherSlot}
  }
  BackingElements_0 {
   widgetClassPath=&Overlay
   parent=Container_CountdownStatus
   widget=BackingElements_0
   slot {class=OverlaySlot, properties {horizontalAlignment=HAlign_Fill, verticalAlignment=VAlign_Fill}}
  }
  Shadow {
   widgetClassPath=&Image
   parent=BackingElements_0
   widget=Shadow
   properties {
    visibility=SelfHitTestInvisible
    brush {
     bIsDynamicallyLoaded=false
     drawAs=Image
     imageType=NoImage
     mirroring=NoMirror
     resourceName=None
     resourceObject=/PilgrimCore/Widgets/Materials/M_UI_Pilgrim_RoundedShadow_Box
     tiling=NoTile
     imageSize=<32 32>
     margin=<ltrb 0 0 0 0>
     outlineSettings {
      bUseBrushTransparency=false
      roundingType=HalfHeightRadius
      width=0
      color {colorUseRule=UseColor_Specified, specifiedColor=<rgba 0 0 0 0>}
      cornerRadii=<0 0 0 0>
     }
     tintColor {colorUseRule=UseColor_Specified, specifiedColor=<rgba 1 1 1 1>}
     uVRegion {bIsValid=false, max=<0 0>, min=<0 0>}
    }
   }
   slot {class=OverlaySlot, properties {horizontalAlignment=HAlign_Fill, verticalAlignment=VAlign_Fill, padding=<ltrb -16 -16 -16 -16>}}
  }
  WBP_Sparks_BackingPanel {
   widgetClassPath=/PilgrimCore/Widgets/SparksUILibrary/WBP_Sparks_BackingPanel
   bIsVariable=true
   parent=BackingElements_0
   widget=WBP_Sparks_BackingPanel
   properties {blurFrame=false, cornerRadius=Large}
   slot {class=OverlaySlot, properties {horizontalAlignment=HAlign_Fill, verticalAlignment=VAlign_Fill}}
  }
  Container_CountdownStatus_hbox {
   widgetClassPath=&HorizontalBox
   parent=Container_CountdownStatus
   widget=Container_CountdownStatus_hbox
   slot {class=OverlaySlot, properties {horizontalAlignment=HAlign_Center, verticalAlignment=VAlign_Fill, padding=<ltrb 16 8 16 8>}}
  }
  Text_CountdownStatus {
   widgetClassPath=&CommonTextBlock
   bIsVariable=true
   parent=Container_CountdownStatus_hbox
   widget=Text_CountdownStatus
   properties {
    justification=Right
    text=Starting in
    textTransformPolicy=ToUpper
    colorAndOpacity {colorUseRule=UseColor_Specified, specifiedColor=<rgba 0.5 0.8333 1 1>}
    font {
     bForceMonospaced=false
     bMaterialIsStencil=false
     fontMaterial=null
     fontObject=&HeadingNow
     letterSpacing=0
     monospacedWidth=1
     size=18
     skewAmount=0
     typefaceFontName=86 Bold Italic
     outlineSettings {
      bApplyOutlineToDropShadows=false
      bMiteredCorners=false
      bSeparateFillAlpha=false
      outlineMaterial=null
      outlineSize=0
      outlineColor=<rgba 0 0 0 1>
     }
    }
    shadowOffset=<0 0>
    strikeBrush {
     bIsDynamicallyLoaded=false
     drawAs=NoDrawType
     imageType=NoImage
     mirroring=NoMirror
     resourceName=None
     resourceObject=null
     tiling=NoTile
     imageSize=<32 32>
     margin=<ltrb 0 0 0 0>
     outlineSettings {
      bUseBrushTransparency=false
      roundingType=HalfHeightRadius
      width=0
      color {colorUseRule=UseColor_Specified, specifiedColor=<rgba 0 0 0 0>}
      cornerRadii=<0 0 0 0>
     }
     tintColor {colorUseRule=UseColor_Specified, specifiedColor=<rgba 1 1 1 1>}
     uVRegion {bIsValid=false, max=<0 0>, min=<0 0>}
    }
   }
   slot {class=HorizontalBoxSlot, properties {horizontalAlignment=HAlign_Right, verticalAlignment=VAlign_Center}}
  }
  Image_72 {
   widgetClassPath=&Image
   bIsVariable=true
   parent=Container_CountdownStatus_hbox
   widget=Image_72
   properties {
    visibility=SelfHitTestInvisible
    brush {
     bIsDynamicallyLoaded=false
     drawAs=Image
     imageType=NoImage
     mirroring=NoMirror
     resourceName=None
     resourceObject=/PilgrimCore/Widgets/Materials/MI_Pilgrim_GreenRoom_DividerLine
     tiling=NoTile
     imageSize=<6 8>
     margin=<ltrb 0.5 0.5 0.5 0.5>
     outlineSettings {
      bUseBrushTransparency=false
      roundingType=HalfHeightRadius
      width=0
      color {colorUseRule=UseColor_Specified, specifiedColor=<rgba 0 0 0 0>}
      cornerRadii=<0 0 0 0>
     }
     tintColor {colorUseRule=UseColor_Specified, specifiedColor=<rgba 1 1 1 1>}
     uVRegion {bIsValid=false, max=<0 0>, min=<0 0>}
    }
    renderTransform {angle=0, scale=<1 1>, shear=<-10 0>, translation=<0 0>}
   }
   slot {class=HorizontalBoxSlot, properties {horizontalAlignment=HAlign_Center, padding=<ltrb 10 8 0 8>}}
  }
  Overlay_3 {
   widgetClassPath=&Overlay
   parent=Container_CountdownStatus_hbox
   widget=Overlay_3
   slot {class=HorizontalBoxSlot, properties {verticalAlignment=VAlign_Center, padding=<ltrb 16 0 0 0>}}
  }
  Text_CountdownCurrent {
   widgetClassPath=&CommonTextBlock
   bIsVariable=true
   parent=Overlay_3
   widget=Text_CountdownCurrent
   properties {
    justification=Center
    text=15
    textTransformPolicy=ToUpper
    font {
     bForceMonospaced=false
     bMaterialIsStencil=false
     fontMaterial=null
     fontObject=&HeadingNow
     letterSpacing=0
     monospacedWidth=1
     size=30
     skewAmount=0
     typefaceFontName=86 Bold Italic
     outlineSettings {
      bApplyOutlineToDropShadows=false
      bMiteredCorners=false
      bSeparateFillAlpha=false
      outlineMaterial=&M-FontShadow
      outlineSize=0
      outlineColor=<rgba 0.175 0.4871 1 1>
     }
    }
    shadowOffset=<0 0>
    strikeBrush {
     bIsDynamicallyLoaded=false
     drawAs=NoDrawType
     imageType=NoImage
     mirroring=NoMirror
     resourceName=None
     resourceObject=null
     tiling=NoTile
     imageSize=<32 32>
     margin=<ltrb 0 0 0 0>
     outlineSettings {
      bUseBrushTransparency=false
      roundingType=HalfHeightRadius
      width=0
      color {colorUseRule=UseColor_Specified, specifiedColor=<rgba 0 0 0 0>}
      cornerRadii=<0 0 0 0>
     }
     tintColor {colorUseRule=UseColor_Specified, specifiedColor=<rgba 1 1 1 1>}
     uVRegion {bIsValid=false, max=<0 0>, min=<0 0>}
    }
   }
   slot {class=OverlaySlot}
  }
  Glow {
   widgetClassPath=&Image
   parent=Overlay_3
   widget=Glow
   properties {
    visibility=SelfHitTestInvisible
    brush {
     bIsDynamicallyLoaded=false
     drawAs=Image
     imageType=NoImage
     mirroring=NoMirror
     resourceName=None
     resourceObject=/PilgrimCore/Widgets/Materials/MI_UI_Pilgrim_ScoreGlow
     tiling=NoTile
     imageSize=<32 32>
     margin=<ltrb 0 0 0 0>
     outlineSettings {
      bUseBrushTransparency=false
      roundingType=HalfHeightRadius
      width=0
      color {colorUseRule=UseColor_Specified, specifiedColor=<rgba 0 0 0 0>}
      cornerRadii=<0 0 0 0>
     }
     tintColor {colorUseRule=UseColor_Specified, specifiedColor=<rgba 1 1 1 1>}
     uVRegion {bIsValid=false, max=<0 0>, min=<0 0>}
    }
   }
   slot {class=OverlaySlot, properties {horizontalAlignment=HAlign_Fill, verticalAlignment=VAlign_Fill, padding=<ltrb -10 -2 -10 -2>}}
  }
  Container_Starting {
   widgetClassPath=&Overlay
   bIsVariable=true
   parent=ReadyAndStartSwitcher
   widget=Container_Starting
   properties {visibility=Hidden}
   slot {class=CommonVisibilitySwitcherSlot}
  }
  BackingElements {
   widgetClassPath=&Overlay
   parent=Container_Starting
   widget=BackingElements
   slot {class=OverlaySlot, properties {horizontalAlignment=HAlign_Fill, verticalAlignment=VAlign_Fill}}
  }
  Shadow_1 {
   widgetClassPath=&Image
   parent=BackingElements
   widget=Shadow_1
   properties {
    visibility=SelfHitTestInvisible
    brush {
     bIsDynamicallyLoaded=false
     drawAs=Image
     imageType=NoImage
     mirroring=NoMirror
     resourceName=None
     resourceObject=&M_UI_Pilgrim_RoundedShadow_Box
     tiling=NoTile
     imageSize=<32 32>
     margin=<ltrb 0 0 0 0>
     outlineSettings {
      bUseBrushTransparency=false
      roundingType=HalfHeightRadius
      width=0
      color {colorUseRule=UseColor_Specified, specifiedColor=<rgba 0 0 0 0>}
      cornerRadii=<0 0 0 0>
     }
     tintColor {colorUseRule=UseColor_Specified, specifiedColor=<rgba 1 1 1 1>}
     uVRegion {bIsValid=false, max=<0 0>, min=<0 0>}
    }
   }
   slot {class=OverlaySlot, properties {horizontalAlignment=HAlign_Fill, verticalAlignment=VAlign_Fill, padding=<ltrb -16 -16 -16 -16>}}
  }
  WBP_Sparks_BackingPanel_1 {
   widgetClassPath=&WBP_Sparks_BackingPanel
   bIsVariable=true
   parent=BackingElements
   widget=WBP_Sparks_BackingPanel_1
   properties {blurFrame=false}
   slot {class=OverlaySlot, properties {horizontalAlignment=HAlign_Fill, verticalAlignment=VAlign_Fill}}
  }
  Container_Starting_Hbox {
   widgetClassPath=&HorizontalBox
   parent=Container_Starting
   widget=Container_Starting_Hbox
   slot {class=OverlaySlot, properties {padding=<ltrb 16 8 16 8>}}
  }
  Text_Starting {
   widgetClassPath=&CommonTextBlock
   parent=Container_Starting_Hbox
   widget=Text_Starting
   properties {
    justification=Center
    text=Sending Band to Stage.
    textTransformPolicy=ToUpper
    colorAndOpacity {colorUseRule=UseColor_Specified, specifiedColor=<rgba 0.5 0.8333 1 1>}
    font {
     bForceMonospaced=false
     bMaterialIsStencil=false
     fontMaterial=null
     fontObject=&HeadingNow
     letterSpacing=0
     monospacedWidth=1
     size=18
     skewAmount=0
     typefaceFontName=86 Bold Italic
     outlineSettings {
      bApplyOutlineToDropShadows=false
      bMiteredCorners=false
      bSeparateFillAlpha=false
      outlineMaterial=null
      outlineSize=0
      outlineColor=<rgba 0 0 0 1>
     }
    }
    shadowOffset=<0 0>
    strikeBrush {
     bIsDynamicallyLoaded=false
     drawAs=NoDrawType
     imageType=NoImage
     mirroring=NoMirror
     resourceName=None
     resourceObject=null
     tiling=NoTile
     imageSize=<32 32>
     margin=<ltrb 0 0 0 0>
     outlineSettings {
      bUseBrushTransparency=false
      roundingType=HalfHeightRadius
      width=0
      color {colorUseRule=UseColor_Specified, specifiedColor=<rgba 0 0 0 0>}
      cornerRadii=<0 0 0 0>
     }
     tintColor {colorUseRule=UseColor_Specified, specifiedColor=<rgba 1 1 1 1>}
     uVRegion {bIsValid=false, max=<0 0>, min=<0 0>}
    }
   }
   slot {class=HorizontalBoxSlot, properties {horizontalAlignment=HAlign_Center, verticalAlignment=VAlign_Center}}
  }
  Container_Waiting {
   widgetClassPath=&Overlay
   bIsVariable=true
   parent=Switcher_MatchmakingStatus
   widget=Container_Waiting
   properties {visibility=Hidden}
   slot {class=WidgetSwitcherSlot}
  }
  BackingElements_1 {
   widgetClassPath=&Overlay
   parent=Container_Waiting
   widget=BackingElements_1
   slot {class=OverlaySlot, properties {horizontalAlignment=HAlign_Fill, verticalAlignment=VAlign_Fill}}
  }
  Shadow_2 {
   widgetClassPath=&Image
   parent=BackingElements_1
   widget=Shadow_2
   properties {
    visibility=SelfHitTestInvisible
    brush {
     bIsDynamicallyLoaded=false
     drawAs=Image
     imageType=NoImage
     mirroring=NoMirror
     resourceName=None
     resourceObject=&M_UI_Pilgrim_RoundedShadow_Box
     tiling=NoTile
     imageSize=<32 32>
     margin=<ltrb 0 0 0 0>
     outlineSettings {
      bUseBrushTransparency=false
      roundingType=HalfHeightRadius
      width=0
      color {colorUseRule=UseColor_Specified, specifiedColor=<rgba 0 0 0 0>}
      cornerRadii=<0 0 0 0>
     }
     tintColor {colorUseRule=UseColor_Specified, specifiedColor=<rgba 1 1 1 1>}
     uVRegion {bIsValid=false, max=<0 0>, min=<0 0>}
    }
   }
   slot {class=OverlaySlot, properties {horizontalAlignment=HAlign_Fill, verticalAlignment=VAlign_Fill, padding=<ltrb -16 -16 -16 -16>}}
  }
  WBP_Sparks_BackingPanel_2 {
   widgetClassPath=&WBP_Sparks_BackingPanel
   bIsVariable=true
   parent=BackingElements_1
   widget=WBP_Sparks_BackingPanel_2
   properties {blurFrame=false}
   slot {class=OverlaySlot, properties {horizontalAlignment=HAlign_Fill, verticalAlignment=VAlign_Fill}}
  }
  Container_Waiting_Vbox {
   widgetClassPath=&VerticalBox
   parent=Container_Waiting
   widget=Container_Waiting_Vbox
   slot {class=OverlaySlot, properties {padding=<ltrb 16 8 16 8>}}
  }
  HorizontalBox_0 {
   widgetClassPath=&HorizontalBox
   parent=Container_Waiting_Vbox
   widget=HorizontalBox_0
   slot {class=VerticalBoxSlot, properties {horizontalAlignment=HAlign_Right, verticalAlignment=VAlign_Center}}
  }
  Overlay_2 {widgetClassPath=&Overlay, parent=HorizontalBox_0, widget=Overlay_2, slot {class=HorizontalBoxSlot}}
  Image {
   widgetClassPath=&Image
   bIsVariable=true
   parent=Overlay_2
   widget=Image
   properties {
    visibility=SelfHitTestInvisible
    brush {
     bIsDynamicallyLoaded=false
     drawAs=Image
     imageType=NoImage
     mirroring=NoMirror
     resourceName=None
     resourceObject=/PilgrimQuickplay/Widgets/Materials/MI_UI_Throbber_Sparks
     tiling=NoTile
     imageSize=<40 40>
     margin=<ltrb 0 0 0 0>
     outlineSettings {
      bUseBrushTransparency=false
      roundingType=HalfHeightRadius
      width=0
      color {colorUseRule=UseColor_Specified, specifiedColor=<rgba 0 0 0 0>}
      cornerRadii=<0 0 0 0>
     }
     tintColor {colorUseRule=UseColor_Specified, specifiedColor=<rgba 1 1 1 1>}
     uVRegion {bIsValid=false, max=<0 0>, min=<0 0>}
    }
   }
   slot {class=OverlaySlot}
  }
  Text_Waiting {
   widgetClassPath=&CommonTextBlock
   bIsVariable=true
   parent=HorizontalBox_0
   widget=Text_Waiting
   properties {
    justification=Center
    text=Waiting for Party Leader...
    textTransformPolicy=ToUpper
    colorAndOpacity {colorUseRule=UseColor_Specified, specifiedColor=<rgba 0.5 0.8333 1 1>}
    font {
     bForceMonospaced=false
     bMaterialIsStencil=false
     fontMaterial=null
     fontObject=&HeadingNow
     letterSpacing=10
     monospacedWidth=1
     size=18
     skewAmount=0
     typefaceFontName=75 Medium Italic
     outlineSettings {
      bApplyOutlineToDropShadows=false
      bMiteredCorners=false
      bSeparateFillAlpha=false
      outlineMaterial=null
      outlineSize=0
      outlineColor=<rgba 0 0 0 1>
     }
    }
    shadowOffset=<0 0>
    strikeBrush {
     bIsDynamicallyLoaded=false
     drawAs=NoDrawType
     imageType=NoImage
     mirroring=NoMirror
     resourceName=None
     resourceObject=null
     tiling=NoTile
     imageSize=<32 32>
     margin=<ltrb 0 0 0 0>
     outlineSettings {
      bUseBrushTransparency=false
      roundingType=HalfHeightRadius
      width=0
      color {colorUseRule=UseColor_Specified, specifiedColor=<rgba 0 0 0 0>}
      cornerRadii=<0 0 0 0>
     }
     tintColor {colorUseRule=UseColor_Specified, specifiedColor=<rgba 1 1 1 1>}
     uVRegion {bIsValid=false, max=<0 0>, min=<0 0>}
    }
   }
   slot {class=HorizontalBoxSlot, properties {verticalAlignment=VAlign_Center, padding=<ltrb 12 0 0 0>}}
  }
  Container_Error {
   widgetClassPath=&Overlay
   parent=Switcher_MatchmakingStatus
   widget=Container_Error
   properties {visibility=Hidden}
   slot {class=WidgetSwitcherSlot, properties {horizontalAlignment=HAlign_Center}}
  }
  BackingElements_2 {
   widgetClassPath=&Overlay
   parent=Container_Error
   widget=BackingElements_2
   slot {class=OverlaySlot, properties {horizontalAlignment=HAlign_Fill, verticalAlignment=VAlign_Fill}}
  }
  Shadow_3 {
   widgetClassPath=&Image
   parent=BackingElements_2
   widget=Shadow_3
   properties {
    visibility=SelfHitTestInvisible
    brush {
     bIsDynamicallyLoaded=false
     drawAs=Image
     imageType=NoImage
     mirroring=NoMirror
     resourceName=None
     resourceObject=&M_UI_Pilgrim_RoundedShadow_Box
     tiling=NoTile
     imageSize=<32 32>
     margin=<ltrb 0 0 0 0>
     outlineSettings {
      bUseBrushTransparency=false
      roundingType=HalfHeightRadius
      width=0
      color {colorUseRule=UseColor_Specified, specifiedColor=<rgba 0 0 0 0>}
      cornerRadii=<0 0 0 0>
     }
     tintColor {colorUseRule=UseColor_Specified, specifiedColor=<rgba 1 1 1 1>}
     uVRegion {bIsValid=false, max=<0 0>, min=<0 0>}
    }
   }
   slot {class=OverlaySlot, properties {horizontalAlignment=HAlign_Fill, verticalAlignment=VAlign_Fill, padding=<ltrb -16 -16 -16 -16>}}
  }
  Backing_3 {
   widgetClassPath=&Image
   parent=BackingElements_2
   widget=Backing_3
   properties {
    visibility=SelfHitTestInvisible
    brush {
     bIsDynamicallyLoaded=false
     drawAs=Image
     imageType=NoImage
     mirroring=NoMirror
     resourceName=None
     resourceObject=/PilgrimCore/Widgets/Materials/M_UI_Pilgrim_HUD_SimpleBacking
     tiling=NoTile
     imageSize=<32 32>
     margin=<ltrb 0 0 0 0>
     outlineSettings {
      bUseBrushTransparency=false
      roundingType=HalfHeightRadius
      width=0
      color {colorUseRule=UseColor_Specified, specifiedColor=<rgba 0 0 0 0>}
      cornerRadii=<0 0 0 0>
     }
     tintColor {colorUseRule=UseColor_Specified, specifiedColor=<rgba 1 1 1 1>}
     uVRegion {bIsValid=false, max=<0 0>, min=<0 0>}
    }
   }
   slot {class=OverlaySlot, properties {horizontalAlignment=HAlign_Fill, verticalAlignment=VAlign_Fill}}
  }
  Container_Error_VBox {
   widgetClassPath=&VerticalBox
   parent=Container_Error
   widget=Container_Error_VBox
   slot {class=OverlaySlot, properties {padding=<ltrb 20 8 20 20>}}
  }
  Overlay_4 {
   widgetClassPath=&Overlay
   parent=Container_Error_VBox
   widget=Overlay_4
   slot {class=VerticalBoxSlot, properties {horizontalAlignment=HAlign_Center}}
  }
  Image_1 {
   widgetClassPath=&Image
   parent=Overlay_4
   widget=Image_1
   properties {
    visibility=SelfHitTestInvisible
    brush {
     bIsDynamicallyLoaded=false
     drawAs=Image
     imageType=NoImage
     mirroring=NoMirror
     resourceName=None
     resourceObject=/PilgrimCore/Widgets/Materials/M_UI_StreakMeter_FX_Rings
     tiling=NoTile
     imageSize=<40 40>
     margin=<ltrb 0 0 0 0>
     outlineSettings {
      bUseBrushTransparency=false
      roundingType=HalfHeightRadius
      width=0
      color {colorUseRule=UseColor_Specified, specifiedColor=<rgba 0 0 0 0>}
      cornerRadii=<0 0 0 0>
     }
     tintColor {colorUseRule=UseColor_Specified, specifiedColor=<rgba 1 1 1 1>}
     uVRegion {bIsValid=false, max=<0 0>, min=<0 0>}
    }
    renderTransform {angle=0, scale=<2.5 2.5>, shear=<0 0>, translation=<0 0>}
   }
   slot {class=OverlaySlot}
  }
  StatusIndicator {
   widgetClassPath=/FortUILibrary/UIKit/StatusIndicator/WBP_UIKit_StatusIndicator
   bIsVariable=true
   parent=Overlay_4
   widget=StatusIndicator
   properties {type=Attention, overrideStatusIndicatorVisibility=true}
   slot {class=OverlaySlot, properties {horizontalAlignment=HAlign_Center, verticalAlignment=VAlign_Center}}
  }
  Text_ErrorHeader {
   widgetClassPath=&CommonTextBlock
   bIsVariable=true
   parent=Container_Error_VBox
   widget=Text_ErrorHeader
   properties {
    justification=Center
    text=Matchmaking Failed
    textTransformPolicy=ToUpper
    font {
     bForceMonospaced=false
     bMaterialIsStencil=false
     fontMaterial=null
     fontObject=&HeadingNow
     letterSpacing=0
     monospacedWidth=1
     size=18
     skewAmount=0
     typefaceFontName=86 Bold Italic
     outlineSettings {
      bApplyOutlineToDropShadows=false
      bMiteredCorners=false
      bSeparateFillAlpha=false
      outlineMaterial=null
      outlineSize=0
      outlineColor=<rgba 0 0 0 1>
     }
    }
    shadowOffset=<0 0>
    strikeBrush {
     bIsDynamicallyLoaded=false
     drawAs=NoDrawType
     imageType=NoImage
     mirroring=NoMirror
     resourceName=None
     resourceObject=null
     tiling=NoTile
     imageSize=<32 32>
     margin=<ltrb 0 0 0 0>
     outlineSettings {
      bUseBrushTransparency=false
      roundingType=HalfHeightRadius
      width=0
      color {colorUseRule=UseColor_Specified, specifiedColor=<rgba 0 0 0 0>}
      cornerRadii=<0 0 0 0>
     }
     tintColor {colorUseRule=UseColor_Specified, specifiedColor=<rgba 1 1 1 1>}
     uVRegion {bIsValid=false, max=<0 0>, min=<0 0>}
    }
   }
   slot {class=VerticalBoxSlot, properties {horizontalAlignment=HAlign_Center, verticalAlignment=VAlign_Center, padding=<ltrb 0 4 0 0>}}
  }
  Text_ErrorHeader_1 {
   widgetClassPath=&CommonTextBlock
   bIsVariable=true
   parent=Container_Error_VBox
   widget=Text_ErrorHeader_1
   properties {
    justification=Center
    text="Returning to the lobby in {NUMBEROFSECONDS} seconds."
    wrapTextAt=384
    font {
     bForceMonospaced=false
     bMaterialIsStencil=false
     fontMaterial=/PilgrimCore/Widgets/Materials/MI_UI_Piligrim_HUDTextColor
     fontObject=&HeadingNow
     letterSpacing=0
     monospacedWidth=1
     size=13.5
     skewAmount=0
     typefaceFontName=74 Regular
     outlineSettings {
      bApplyOutlineToDropShadows=false
      bMiteredCorners=false
      bSeparateFillAlpha=false
      outlineMaterial=null
      outlineSize=0
      outlineColor=<rgba 0 0 0 1>
     }
    }
    shadowOffset=<0 0>
    strikeBrush {
     bIsDynamicallyLoaded=false
     drawAs=NoDrawType
     imageType=NoImage
     mirroring=NoMirror
     resourceName=None
     resourceObject=null
     tiling=NoTile
     imageSize=<32 32>
     margin=<ltrb 0 0 0 0>
     outlineSettings {
      bUseBrushTransparency=false
      roundingType=HalfHeightRadius
      width=0
      color {colorUseRule=UseColor_Specified, specifiedColor=<rgba 0 0 0 0>}
      cornerRadii=<0 0 0 0>
     }
     tintColor {colorUseRule=UseColor_Specified, specifiedColor=<rgba 1 1 1 1>}
     uVRegion {bIsValid=false, max=<0 0>, min=<0 0>}
    }
   }
   slot {class=VerticalBoxSlot, properties {horizontalAlignment=HAlign_Center, verticalAlignment=VAlign_Center, padding=<ltrb 0 4 0 0>}}
  }
  Container_NoStatus {
   widgetClassPath=&HorizontalBox
   parent=Switcher_MatchmakingStatus
   widget=Container_NoStatus
   slot {class=WidgetSwitcherSlot, properties {horizontalAlignment=HAlign_Center, verticalAlignment=VAlign_Center}}
  }
 ]
}
