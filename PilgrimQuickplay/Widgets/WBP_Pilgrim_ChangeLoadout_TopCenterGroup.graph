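# Top-center group widget for the Pilgrim Quickplay change-loadout UI, showing a ready-up timer and bandmate ready status.
# A Canvas Panel anchors the ready-up timer and the GreenRoom bandmate ready display at top center, with the display offset lower.
# The event graph is empty, so this is layout only; the declared EventRouter and MyPlayerState variables are not used here.
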WidgetBlueprint /PilgrimQuickplay/Widgets/WBP_Pilgrim_ChangeLoadout_TopCenterGroup
graphs [EventGraph {links [], nodes []}]
variables [
 EventRouter {type=GameplayEventRouterComponent, default=null}
 MyPlayerState {type=PlayerState, default=null}
]
widgets {
 info {inheritedWidgetCount=0, parentClass=&UserWidget, rootWidgetClass=&CanvasPanel, widgetCount=3}
 widgets [
  CanvasPanel {widgetClassPath=&CanvasPanel, bIsVariable=true, widget=CanvasPanel}
  WBP_Pilgrim_ReadyUpTimer {
   widgetClassPath=&/WBP_Pilgrim_ReadyUpTimer
   bIsVariable=true
   parent=CanvasPanel
   widget=WBP_Pilgrim_ReadyUpTimer
   slot {
    class=CanvasPanelSlot
    properties {layoutData {alignment=<0 0>, anchors {maximum=<0.5 0>, minimum=<0.5 0>}, offsets=<ltrb -180 -28 412.5352 145.3725>}}
   }
  }
  WBP_PilgrimQuickplay_GreenRoom_BandmateReadyDisplay {
   widgetClassPath=/PilgrimQuickplay/Widgets/GreenRoom/WBP_PilgrimQuickplay_GreenRoom_BandmateReadyDisplay
   bIsVariable=true
   parent=CanvasPanel
   widget=WBP_PilgrimQuickplay_GreenRoom_BandmateReadyDisplay
   slot {class=CanvasPanelSlot, properties {layoutData {alignment=<0 0>, anchors {maximum=<0.5 0>, minimum=<0.5 0>}, offsets=<ltrb -392 120 682 80>}}}
  }
 ]
}
