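# Pilgrim Quickplay prompt widget pairing a keybind icon with "Open Settings (Hold)" text.
# It shows only during a HamSandwich session for a non-primary player using a peripheral or gamepad.
# On Construct it listens through the Pilgrim event router for peripheral changes, refreshing icon and visibility, and stops on Destruct.
# The icon refresh reapplies the current action; its peripheral-based action Select is left unconnected.
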WidgetBlueprint /PilgrimQuickplay/Widgets/WBP_PilgrimQuickplay_OpenSettings
graphs [
 EventGraph {nodes [
   Event Destruct()
   n1 = Get PlayerUsingPeripheralChanged Async Ref() -> (PlayerUsingPeripheralChanged Async Ref:Async Action Start Listening To Event Object Reference)
   > Is Valid(InputObject=@n1)
   > StopListeningToEvent(self:Async Action Start Listening To Event Object Reference=@n1)
   > Set PlayerUsingPeripheralChanged Async Ref(PlayerUsingPeripheralChanged Async Ref:Async Action Start Listening To Event Object Reference) -> (Output_Get:Async Action Start Listening To Event Object Reference)
   Event Construct()
   > UpdateInputActionIcon()
   > n2 = PilgrimWaitForEventRouter!(execute:Exec, ContextActor:Actor Object Reference=GetOwningPlayer()) -> (then:Exec, OnEventRouterAvailable:Exec, EventRouter:Gameplay Event Router Component Object Reference) {is=title K2Node_AsyncAction}
   n3 = StartListeningToEvent!(
    execute:Exec=@n2.OnEventRouterAvailable
    Target:Gameplay Event Router Component Object Reference=@n2
    EventType:Script Struct Object Reference=/Script/PilgrimCoreRuntime.PilgrimEvent_PlayerUsingPeripheralChanged
    Context:Object Reference
    EventBubblingRule:EEventBubblingRule Enum
    None:%=@n2
   ) -> (
    then:Exec
    AsyncTaskProxy:Async Action Start Listening To Event Object Reference
    OnEventReceived:Exec
    Payload:Pilgrim Event Player Using Peripheral Changed Structure
   )
   > n4 = Set PlayerUsingPeripheralChanged Async Ref(PlayerUsingPeripheralChanged Async Ref=@n3)
   SetPromptVisibility(execute=[@n5.then, @n4.then])
   n5 = UpdateInputActionIcon(execute=@n3.OnEventReceived)
  ]}
 SetPromptVisibility {nodes [
   SetPromptVisibility() {is=title K2Node_FunctionEntry}
   n1 = IsHamSandwichSessionActive!() -> (ReturnValue:bool)
   n2 = GetOwningPlayer()
   n3 = GetComponentByClass!(
    self:Actor Object Reference=GetControllerFromActor!(Actor:Actor Object Reference=@n2) -> (ReturnValue:Controller Object Reference)
    ComponentClass:Actor Component Class Reference=/Script/PilgrimCoreRuntime.PilgrimPlayerControllerComponent
   ) -> (ReturnValue:Pilgrim Player Controller Component Object Reference)
   n4 = AND Boolean(
    A=OR Boolean(
     A=GetIsCurrentDevicePeripheral!(self:Pilgrim Player Controller Component Object Reference=@n3) -> (ReturnValue:bool)
     B=IsUsingGamepad!(self:Pilgrim Player Controller Component Object Reference=@n3) -> (ReturnValue:bool)
    )
    B=NOT Boolean(A=IsPrimaryPlayer!(self:Player Controller Object Reference=@n2) -> (ReturnValue:bool))
   )
   > SetVisibility(InVisibility=Select(Option 0=Collapsed, Option 1=SelfHitTestInvisible, Index=AND Boolean(A=@n1, B=@n4)))
  ]}
 UpdateInputActionIcon {nodes [
   UpdateInputActionIcon() {is=title K2Node_FunctionEntry}
   n1 = Get KeybindWidget_OpenSettings() -> (KeybindWidget_OpenSettings:Keybind Widget Object Reference)
   n2 = Get BoundInputAction!(self:Fort Keybind Widget Object Reference=@n1) -> (BoundInputAction:Input Action Object Reference)
   > SetBoundInputAction!(execute, self:Fort Keybind Widget Object Reference=@n1, NewAction:Input Action Object Reference=@n2) -> (then)
   n3 = GetControllerFromActor(Actor=GetOwningPlayer())
   Select(
    Option 0:Input Action Object Reference=/PilgrimCore/Player/Abilities/IA_PilgrimQuickEmote_Stop
    Option 1:Input Action Object Reference=@n2
    Index=GetIsCurrentDevicePeripheral(self=GetComponentByClass(self=@n3, ComponentClass=&PilgrimPlayerControllerComponent))
   ) -> (ReturnValue:Input Action Object Reference)
  ]}
]
variables [PlayerUsingPeripheralChanged Async Ref {type=AsyncAction_StartListeningToEvent, default=null}]
widgets {
 info {inheritedWidgetCount=0, parentClass=&CommonUserWidget, rootWidgetClass=&HorizontalBox, widgetCount=4}
 widgets [
  HorizontalBox {widgetClassPath=&HorizontalBox, widget=HorizontalBox}
  KeybindWidget_OpenSettings {
   widgetClassPath=/FortUILibrary/Input/KeybindWidget
   bIsVariable=true
   parent=HorizontalBox
   widget=KeybindWidget_OpenSettings
   properties {bUseKeyboardTextureForKeys=true, boundInputAction=/PilgrimCore/Input/InputActions/IA_PilgrimSplitscreenTakeControl}
   slot {class=HorizontalBoxSlot, properties {horizontalAlignment=HAlign_Center, verticalAlignment=VAlign_Center}}
  }
  Spacer_143 {widgetClassPath=&Spacer, parent=HorizontalBox, widget=Spacer_143, properties {size=<8 1>}, slot {class=HorizontalBoxSlot}}
  PromptText {
   widgetClassPath=&CommonTextBlock
   bIsVariable=true
   parent=HorizontalBox
   widget=PromptText
   properties {
    text="Open Settings (Hold)"
    font {
     bForceMonospaced=false
     bMaterialIsStencil=false
     fontMaterial=null
     fontObject=&HeadingNow
     letterSpacing=-1
     monospacedWidth=1
     size=15
     skewAmount=0
     typefaceFontName=76 Bold
     outlineSettings {
      bApplyOutlineToDropShadows=false
      bMiteredCorners=false
      bSeparateFillAlpha=false
      outlineMaterial=&M-FontShadow
      outlineSize=2
      outlineColor=<rgba 0 0 0 1>
     }
    }
    shadowOffset=<0 0>
   }
   slot {class=HorizontalBoxSlot, properties {verticalAlignment=VAlign_Center}}
  }
 ]
}
