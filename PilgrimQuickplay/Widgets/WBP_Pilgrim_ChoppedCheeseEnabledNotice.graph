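# Pilgrim quickplay HUD notice widget that warns when the Chopped Cheese override part is enabled, showing an attention indicator and message over a backing.
# On construct, it reads the current state from the owning player's Pilgrim player controller component.
# On initialize, it waits for the event router and listens for the override cvar-changed event.
# Both paths call RefreshDisplay, which toggles the widget between Visible and Collapsed.
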WidgetBlueprint /PilgrimQuickplay/Widgets/WBP_Pilgrim_ChoppedCheeseEnabledNotice
graphs [
 EventGraph {nodes [
   Event OnInitialized()
   > n1 = PilgrimWaitForEventRouter!(execute:Exec, ContextActor:Actor Object Reference=GetOwningPlayer()) -> (then:Exec, OnEventRouterAvailable:Exec, EventRouter:Gameplay Event Router Component Object Reference) {is=title K2Node_AsyncAction}
   Sequence(execute=@n1.OnEventRouterAvailable)
   > n2 = StartListeningToEvent!(
    execute:Exec
    Target:Gameplay Event Router Component Object Reference=@n1
    EventType:Script Struct Object Reference=/Script/PilgrimCoreRuntime.PilgrimEvent_EnableChoppedCheeseOverridePartCvarChanged
    Context:Object Reference
    EventBubblingRule:EEventBubblingRule Enum
   ) -> (
    then:Exec
    AsyncTaskProxy:Async Action Start Listening To Event Object Reference
    OnEventReceived:Exec
    Payload:Pilgrim Event Enable Chopped Cheese Override Part Cvar Changed Structure
   )
   RefreshDisplay!(
    execute=@n2.OnEventReceived
    self
    IsVisible:bool=Break Pilgrim Event Enable Chopped Cheese Override Part Cvar Changed!(PilgrimEvent_EnableChoppedCheeseOverridePartCvarChanged:"Pilgrim Event Enable Chopped Cheese Override Part Cvar Changed Structure (by ref)"=@n2.Payload) -> (bIsChoppedCheeseOverridePartCvarEnabled:bool=~) {is=title K2Node_BreakStruct}
   ) -> (then)
   Event Construct()
   n3 = GetComponentByClass!(
    self:Actor Object Reference=GetOwningPlayer()
    ComponentClass:Actor Component Class Reference=/PilgrimCore/Player/Components/BP_PilgrimPlayerControllerComponent.BP_PilgrimPlayerControllerComponent_C
   ) -> (ReturnValue:BP Pilgrim Player Controller Component Object Reference)
   > Is Valid(InputObject=@n3)
   > RefreshDisplay(IsVisible=IsChoppedCheeseOverridePartEnabled!(self:Pilgrim Player Controller Component Object Reference=@n3) -> (ReturnValue:bool))
  ]}
 RefreshDisplay {nodes [
   n1 = RefreshDisplay!() -> (then, IsVisible:bool=~) {is=title K2Node_FunctionEntry}
   > SetVisibility(InVisibility=Select(Option 0=Collapsed, Option 1=Visible, Index=@n1))
  ]}
]
variables []
widgets {
 info {inheritedWidgetCount=0, parentClass=&UserWidget, rootWidgetClass=&Overlay, widgetCount=9}
 widgets [
  Notice {widgetClassPath=&Overlay, bIsVariable=true, widget=Notice, properties {visibility=HitTestInvisible}}
  BackingElements {
   widgetClassPath=&Overlay
   parent=Notice
   widget=BackingElements
   slot {class=OverlaySlot, properties {horizontalAlignment=HAlign_Fill, verticalAlignment=VAlign_Fill}}
  }
  Shadow {
   widgetClassPath=&Image
   parent=BackingElements
   widget=Shadow
   properties {
    visibility=SelfHitTestInvisible
    brush {
     bIsDynamicallyLoaded=false
     drawAs=Image
     imageType=NoImage
     mirroring=NoMirror
     resourceName=None
     resourceObject=/PilgrimCore/Widgets/Materials/M_UI_Pilgrim_RoundedShadow_Box
     tiling=NoTile
     imageSize=<32 32>
     margin=<ltrb 0 0 0 0>
     outlineSettings {
      bUseBrushTransparency=false
      roundingType=HalfHeightRadius
      width=0
      color {colorUseRule=UseColor_Specified, specifiedColor=<rgba 0 0 0 0>}
      cornerRadii=<0 0 0 0>
     }
     tintColor {colorUseRule=UseColor_Specified, specifiedColor=<rgba 1 1 1 1>}
     uVRegion {bIsValid=false, max=<0 0>, min=<0 0>}
    }
   }
   slot {class=OverlaySlot, properties {horizontalAlignment=HAlign_Fill, verticalAlignment=VAlign_Fill, padding=<ltrb -16 -16 -16 -16>}}
  }
  Backing {
   widgetClassPath=&Image
   parent=BackingElements
   widget=Backing
   properties {
    visibility=SelfHitTestInvisible
    brush {
     bIsDynamicallyLoaded=false
     drawAs=Image
     imageType=NoImage
     mirroring=NoMirror
     resourceName=None
     resourceObject=/PilgrimCore/Widgets/Materials/M_UI_Pilgrim_HUD_SimpleBacking
     tiling=NoTile
     imageSize=<32 32>
     margin=<ltrb 0 0 0 0>
     outlineSettings {
      bUseBrushTransparency=false
      roundingType=HalfHeightRadius
      width=0
      color {colorUseRule=UseColor_Specified, specifiedColor=<rgba 0 0 0 0>}
      cornerRadii=<0 0 0 0>
     }
     tintColor {colorUseRule=UseColor_Specified, specifiedColor=<rgba 1 1 1 1>}
     uVRegion {bIsValid=false, max=<0 0>, min=<0 0>}
    }
   }
   slot {class=OverlaySlot, properties {horizontalAlignment=HAlign_Fill, verticalAlignment=VAlign_Fill}}
  }
  Contents {
   widgetClassPath=&HorizontalBox
   parent=Notice
   widget=Contents
   slot {class=OverlaySlot, properties {horizontalAlignment=HAlign_Center, verticalAlignment=VAlign_Center, padding=<ltrb 20 20 20 20>}}
  }
  Overlay_4 {
   widgetClassPath=&Overlay
   parent=Contents
   widget=Overlay_4
   slot {class=HorizontalBoxSlot, properties {verticalAlignment=VAlign_Center}}
  }
  Image_1 {
   widgetClassPath=&Image
   parent=Overlay_4
   widget=Image_1
   properties {
    visibility=SelfHitTestInvisible
    brush {
     bIsDynamicallyLoaded=false
     drawAs=Image
     imageType=NoImage
     mirroring=NoMirror
     resourceName=None
     resourceObject=/PilgrimCore/Widgets/Materials/M_UI_StreakMeter_FX_Rings
     tiling=NoTile
     imageSize=<40 40>
     margin=<ltrb 0 0 0 0>
     outlineSettings {
      bUseBrushTransparency=false
      roundingType=HalfHeightRadius
      width=0
      color {colorUseRule=UseColor_Specified, specifiedColor=<rgba 0 0 0 0>}
      cornerRadii=<0 0 0 0>
     }
     tintColor {colorUseRule=UseColor_Specified, specifiedColor=<rgba 1 1 1 1>}
     uVRegion {bIsValid=false, max=<0 0>, min=<0 0>}
    }
    renderTransform {angle=0, scale=<2.5 2.5>, shear=<0 0>, translation=<0 0>}
   }
   slot {class=OverlaySlot}
  }
  StatusIndicator {
   widgetClassPath=/FortUILibrary/UIKit/StatusIndicator/WBP_UIKit_StatusIndicator
   bIsVariable=true
   parent=Overlay_4
   widget=StatusIndicator
   properties {type=Attention, overrideStatusIndicatorVisibility=true}
   slot {class=OverlaySlot, properties {horizontalAlignment=HAlign_Center, verticalAlignment=VAlign_Center}}
  }
  Text_NoticeBody {
   widgetClassPath=&CommonTextBlock
   bIsVariable=true
   parent=Contents
   widget=Text_NoticeBody
   properties {
    autoWrapText=true
    bIsScrollingEnabled=false
    text="Chopped Cheese Override Part is Enabled!"
    wrapTextAt=384
    font {
     bForceMonospaced=false
     bMaterialIsStencil=false
     fontMaterial=null
     fontObject=&HeadingNow
     letterSpacing=0
     monospacedWidth=1
     size=13.5
     skewAmount=0
     typefaceFontName=74 Regular
     outlineSettings {
      bApplyOutlineToDropShadows=false
      bMiteredCorners=false
      bSeparateFillAlpha=false
      outlineMaterial=null
      outlineSize=0
      outlineColor=<rgba 0 0 0 1>
     }
    }
    shadowOffset=<0 0>
    strikeBrush {
     bIsDynamicallyLoaded=false
     drawAs=NoDrawType
     imageType=NoImage
     mirroring=NoMirror
     resourceName=None
     resourceObject=null
     tiling=NoTile
     imageSize=<32 32>
     margin=<ltrb 0 0 0 0>
     outlineSettings {
      bUseBrushTransparency=false
      roundingType=HalfHeightRadius
      width=0
      color {colorUseRule=UseColor_Specified, specifiedColor=<rgba 0 0 0 0>}
      cornerRadii=<0 0 0 0>
     }
     tintColor {colorUseRule=UseColor_Specified, specifiedColor=<rgba 1 1 1 1>}
     uVRegion {bIsValid=false, max=<0 0>, min=<0 0>}
    }
   }
   slot {class=HorizontalBoxSlot, properties {verticalAlignment=VAlign_Center, padding=<ltrb 8 0 0 0>, size {sizeRule=Fill, value=1}}}
  }
 ]
}
