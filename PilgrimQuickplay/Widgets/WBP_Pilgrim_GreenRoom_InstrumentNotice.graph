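# Green room notice widget in Pilgrim that warns when the chosen instrument's strumming style mismatches the active controller.
# On input device change, it passes the player's instrument choice to RefreshDisplay, which checks for a Pro (guitar peripheral) track and guitar controller.
# Pro without guitar, with unified part picker off, shows "requires strumming"; non-Pro with guitar shows "played without strumming".
# Otherwise it collapses; PreConstruct fills default text using localized instrument names.
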WidgetBlueprint /PilgrimQuickplay/Widgets/WBP_Pilgrim_GreenRoom_InstrumentNotice
graphs [
 EventGraph {nodes [
   n1 = On Input Hardware Device Changed!() -> (OutputDelegate, then, UserId:Platform User Id Structure, DeviceId:Input Device Id Structure) {is=title K2Node_CustomEvent}
   > n2 = GetPilgrimCorePlayerStateComponent!(
    execute
    PlayerState:Player State Object Reference=Get PlayerState!(self:Controller Object Reference=GetOwningPlayer()) -> (PlayerState:Player State Object Reference)
   ) -> (then, ReturnValue:Pilgrim Core Player State Component Object Reference)
   > RefreshDisplay!(
    execute
    self
    TrackType:EPilgrimTrackType Enum=GetInstrumentChoice!(self:Pilgrim Core Player State Component Object Reference=@n2) -> (ReturnValue:EPilgrimTrackType Enum=TrackGuitar)
   ) -> (then)
   Event OnInitialized()
   > Bind Event to On Input Hardware Device Changed(
    self:Input Device Subsystem Object Reference=Get InputDeviceSubsystem() -> (ReturnValue:Input Device Subsystem Object Reference) {is=title K2Node_GetEngineSubsystem}
    Delegate=@n1
   ) {is=title K2Node_AddDelegate}
   Event PreConstruct()
   > "SetText (Text)"(
    self=Get Text_NoticeBody() -> (Text_NoticeBody:Common Text Object Reference)
    InText=Format Text!(
     Format="LOCTEXT(\"This instrument is played without strumming. To strum, choose {ProLead} or {ProBass}!\")"
     ProLead:Text=MakeLiteralText(Value="LOCTABLE(\"/SparksSettings/UI/ST_Sparks_InstrumentNames\", \"ProLead\")")
     ProBass:Text=MakeLiteralText(Value="LOCTABLE(\"&ST_Sparks_InstrumentNames\", \"ProBass\")")
    ) -> (Result)
   )
  ]}
 RefreshDisplay {nodes [
   n1 = RefreshDisplay!() -> (then, TrackType:EPilgrimTrackType Enum=TrackGuitar) {is=title K2Node_FunctionEntry}
   > Set IsProPart(IsProPart:bool=IsGuitarPeripheralTrack!(TrackType:EPilgrimTrackType Enum=@n1) -> (ReturnValue:bool)) -> (Output_Get:bool)
   n2 = GetComponentByClass!(
    self:Actor Object Reference=GetControllerFromActor!(Actor:Actor Object Reference=GetOwningPlayer()) -> (ReturnValue:Controller Object Reference)
    ComponentClass:Actor Component Class Reference=/Script/PilgrimCoreRuntime.PilgrimPlayerControllerComponent
   ) -> (ReturnValue:Pilgrim Player Controller Component Object Reference)
   > Set IsGuitarControllerActive(IsGuitarControllerActive:bool=GetIsCurrentDevicePeripheral!(self:Pilgrim Player Controller Component Object Reference=@n2) -> (ReturnValue:bool)) -> (Output_Get:bool)
   n3 = IsUnifiedPartPickerEnabled!() -> (ReturnValue:bool)
   n4 = Get IsGuitarControllerActive() -> (IsGuitarControllerActive:bool)
   > n5 = Branch(Condition=AND Boolean!(A=Get IsProPart() -> (IsProPart:bool), B=NOT Boolean(A=@n4), C:bool=NOT Boolean(A=@n3)) -> (ReturnValue))
   > n6 = Set Text(Text="LOCTEXT(\"This instrument requires strumming. While holding a lane input, press the strum input!\")", self=Get Text_NoticeBody())
   SetVisibility(execute=[@n9.then, @n6.then], InVisibility=SelfHitTestInvisible)
   n7 = Get IsProPart()
   n8 = Branch(execute=@n5.else, Condition=AND Boolean!(A=NOT Boolean(A=@n7), B=Get IsGuitarControllerActive()) -> (ReturnValue))
   > n9 = Set Text(Text="LOCTEXT(\"This instrument is played without strumming. To strum, choose Pro Lead or Pro Bass!\")", self=Get Text_NoticeBody())
   SetVisibility(execute=@n8.else, InVisibility=Collapsed)
   n10 = GetCurrentDevicePeripheralType!(self:Pilgrim Player Controller Component Object Reference=@n2) -> (ReturnValue:EFMPeripheralInputMode Enum=None)
   PrintText!(
    execute
    InText=Format Text!(
     Format="LOCTEXT(\"Current Device Peripheral Type: {type}\")"
     type:String=Enum to String!(Enumerator:EFMPeripheralInputMode Enum=@n10) -> (ReturnValue:String)
    ) -> (Result)
    bPrintToScreen
    bPrintToLog
    TextColor
    Duration
    Key
   ) -> (then)
  ]}
]
variables []
widgets {
 info {inheritedWidgetCount=0, parentClass=&UserWidget, rootWidgetClass=&Overlay, widgetCount=9}
 widgets [
  Notice {widgetClassPath=&Overlay, bIsVariable=true, widget=Notice, properties {visibility=HitTestInvisible}}
  BackingElements {
   widgetClassPath=&Overlay
   parent=Notice
   widget=BackingElements
   slot {class=OverlaySlot, properties {horizontalAlignment=HAlign_Fill, verticalAlignment=VAlign_Fill}}
  }
  Shadow {
   widgetClassPath=&Image
   parent=BackingElements
   widget=Shadow
   properties {
    visibility=SelfHitTestInvisible
    brush {
     bIsDynamicallyLoaded=false
     drawAs=Image
     imageType=NoImage
     mirroring=NoMirror
     resourceName=None
     resourceObject=/PilgrimCore/Widgets/Materials/M_UI_Pilgrim_RoundedShadow_Box
     tiling=NoTile
     imageSize=<32 32>
     margin=<ltrb 0 0 0 0>
     outlineSettings {
      bUseBrushTransparency=false
      roundingType=HalfHeightRadius
      width=0
      color {colorUseRule=UseColor_Specified, specifiedColor=<rgba 0 0 0 0>}
      cornerRadii=<0 0 0 0>
     }
     tintColor {colorUseRule=UseColor_Specified, specifiedColor=<rgba 1 1 1 1>}
     uVRegion {bIsValid=false, max=<0 0>, min=<0 0>}
    }
   }
   slot {class=OverlaySlot, properties {horizontalAlignment=HAlign_Fill, verticalAlignment=VAlign_Fill, padding=<ltrb -16 -16 -16 -16>}}
  }
  Backing {
   widgetClassPath=&Image
   parent=BackingElements
   widget=Backing
   properties {
    visibility=SelfHitTestInvisible
    brush {
     bIsDynamicallyLoaded=false
     drawAs=Image
     imageType=NoImage
     mirroring=NoMirror
     resourceName=None
     resourceObject=/PilgrimCore/Widgets/Materials/M_UI_Pilgrim_HUD_SimpleBacking
     tiling=NoTile
     imageSize=<32 32>
     margin=<ltrb 0 0 0 0>
     outlineSettings {
      bUseBrushTransparency=false
      roundingType=HalfHeightRadius
      width=0
      color {colorUseRule=UseColor_Specified, specifiedColor=<rgba 0 0 0 0>}
      cornerRadii=<0 0 0 0>
     }
     tintColor {colorUseRule=UseColor_Specified, specifiedColor=<rgba 1 1 1 1>}
     uVRegion {bIsValid=false, max=<0 0>, min=<0 0>}
    }
   }
   slot {class=OverlaySlot, properties {horizontalAlignment=HAlign_Fill, verticalAlignment=VAlign_Fill}}
  }
  Contents {
   widgetClassPath=&HorizontalBox
   parent=Notice
   widget=Contents
   slot {class=OverlaySlot, properties {horizontalAlignment=HAlign_Center, verticalAlignment=VAlign_Center, padding=<ltrb 20 20 20 20>}}
  }
  Overlay_4 {
   widgetClassPath=&Overlay
   parent=Contents
   widget=Overlay_4
   slot {class=HorizontalBoxSlot, properties {verticalAlignment=VAlign_Center}}
  }
  Image_1 {
   widgetClassPath=&Image
   parent=Overlay_4
   widget=Image_1
   properties {
    visibility=SelfHitTestInvisible
    brush {
     bIsDynamicallyLoaded=false
     drawAs=Image
     imageType=NoImage
     mirroring=NoMirror
     resourceName=None
     resourceObject=/PilgrimCore/Widgets/Materials/M_UI_StreakMeter_FX_Rings
     tiling=NoTile
     imageSize=<40 40>
     margin=<ltrb 0 0 0 0>
     outlineSettings {
      bUseBrushTransparency=false
      roundingType=HalfHeightRadius
      width=0
      color {colorUseRule=UseColor_Specified, specifiedColor=<rgba 0 0 0 0>}
      cornerRadii=<0 0 0 0>
     }
     tintColor {colorUseRule=UseColor_Specified, specifiedColor=<rgba 1 1 1 1>}
     uVRegion {bIsValid=false, max=<0 0>, min=<0 0>}
    }
    renderTransform {angle=0, scale=<2.5 2.5>, shear=<0 0>, translation=<0 0>}
   }
   slot {class=OverlaySlot}
  }
  StatusIndicator {
   widgetClassPath=/FortUILibrary/UIKit/StatusIndicator/WBP_UIKit_StatusIndicator
   bIsVariable=true
   parent=Overlay_4
   widget=StatusIndicator
   properties {type=Attention, overrideStatusIndicatorVisibility=true}
   slot {class=OverlaySlot, properties {horizontalAlignment=HAlign_Center, verticalAlignment=VAlign_Center}}
  }
  Text_NoticeBody {
   widgetClassPath=&CommonTextBlock
   bIsVariable=true
   parent=Contents
   widget=Text_NoticeBody
   properties {
    autoWrapText=true
    bIsScrollingEnabled=false
    wrapTextAt=384
    font {
     bForceMonospaced=false
     bMaterialIsStencil=false
     fontMaterial=null
     fontObject=&HeadingNow
     letterSpacing=0
     monospacedWidth=1
     size=13.5
     skewAmount=0
     typefaceFontName=74 Regular
     outlineSettings {
      bApplyOutlineToDropShadows=false
      bMiteredCorners=false
      bSeparateFillAlpha=false
      outlineMaterial=null
      outlineSize=0
      outlineColor=<rgba 0 0 0 1>
     }
    }
    shadowOffset=<0 0>
    strikeBrush {
     bIsDynamicallyLoaded=false
     drawAs=NoDrawType
     imageType=NoImage
     mirroring=NoMirror
     resourceName=None
     resourceObject=null
     tiling=NoTile
     imageSize=<32 32>
     margin=<ltrb 0 0 0 0>
     outlineSettings {
      bUseBrushTransparency=false
      roundingType=HalfHeightRadius
      width=0
      color {colorUseRule=UseColor_Specified, specifiedColor=<rgba 0 0 0 0>}
      cornerRadii=<0 0 0 0>
     }
     tintColor {colorUseRule=UseColor_Specified, specifiedColor=<rgba 1 1 1 1>}
     uVRegion {bIsValid=false, max=<0 0>, min=<0 0>}
    }
   }
   slot {class=HorizontalBoxSlot, properties {padding=<ltrb 8 0 0 0>, size {sizeRule=Fill, value=1}}}
  }
 ]
}
