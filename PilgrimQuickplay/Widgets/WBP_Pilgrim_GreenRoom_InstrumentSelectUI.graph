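# Green room widget for Pilgrim where the player picks an instrument: guitar, bass, drums, vocals or their pro/peripheral variants.
# Clicks cycle instruments through the player controller component, and OnInstrumentChanged updates the selected tab and name.
# ScoreChase playlist tags dim disallowed instruments.
# Event router listeners and input-method changes drive pro-option visibility, moving players off unavailable instruments, and the green room timer state toggles the whole widget.
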WidgetBlueprint /PilgrimQuickplay/Widgets/WBP_Pilgrim_GreenRoom_InstrumentSelectUI
animations [
 Hover {
  binding_count=1
  length_seconds=0.2
  object=&self:Hover
  sequence {bindings [SelectionBackingIndicator {
     id=#1
     possessed_class=Image
     tracks [{
       type=MovieSceneFloatTrack
       display_name=Render Opacity
       sections [{type=MovieSceneFloatSection, channels [None {type=MovieSceneScriptingFloatChannel, keys [[0,0],[4,1]]}]}]
      }]
    }]}
 }
 Pressed {
  binding_count=1
  length_seconds=0.2
  object=&self:Pressed
  sequence {bindings [WBP_Pilgrim_GreenRoom_InstrumentSelectUI_C_36 {
     id=#2
     possessed_class=WBP_Pilgrim_GreenRoom_InstrumentSelectUI_C
     tracks [{
       type=MovieScene2DTransformTrack
       sections [{
         type=MovieScene2DTransformSection
         channels [
          Scale.X {type=MovieSceneScriptingFloatChannel, keys [[0,1],[2,0.975],[4,1]]}
          Scale.Y {type=MovieSceneScriptingFloatChannel, keys [[0,1],[2,0.975],[4,1]]}
         ]
        }]
      }]
    }]}
 }
]
graphs [
 ActionGraph {nodes [
   n1 = "On Clicked (UIKit_Button_Guitar)"!() -> (OutputDelegate, then, Button:Common Button Base Object Reference) {is=title K2Node_ComponentBoundEvent}
   SwapInstrumentBackward(
    execute=[@n1.then, @n2.then, @n3.then, @n4.then, @n5.then, @n6.then, @n7.then, @n8.then]
    self:BP Pilgrim Player Controller Component Object Reference=Get BP_PilgrimPlayerControllerComponent() -> (BP_PilgrimPlayerControllerComponent:BP Pilgrim Player Controller Component Object Reference)
   )
   n2 = "On Clicked (UIKit_Button_Drums)"!() -> (OutputDelegate, then, Button:Common Button Base Object Reference) {is=title K2Node_ComponentBoundEvent}
   n3 = "On Clicked (UIKit_Button_Bass_Peripheral)"!() -> (OutputDelegate, then, Button:Common Button Base Object Reference) {is=title K2Node_ComponentBoundEvent}
   n4 = "On Clicked (UIKit_Button_Drums_Peripheral)"!() -> (OutputDelegate, then, Button:Common Button Base Object Reference) {is=title K2Node_ComponentBoundEvent}
   n5 = "On Clicked (UIKit_Button_Vocal)"!() -> (OutputDelegate, then, Button:Common Button Base Object Reference) {is=title K2Node_ComponentBoundEvent}
   n6 = "On Clicked (UIKit_Button_Bass)"!() -> (OutputDelegate, then, Button:Common Button Base Object Reference) {is=title K2Node_ComponentBoundEvent}
   n7 = "On Clicked (UIKit_Button_Vocals_Peripheral)"!() -> (OutputDelegate, then, Button:Common Button Base Object Reference) {is=title K2Node_ComponentBoundEvent}
   n8 = "On Clicked (UIKit_Button_Guitar_Peripheral)"!() -> (OutputDelegate, then, Button:Common Button Base Object Reference) {is=title K2Node_ComponentBoundEvent}
   SetSelectionByIndex!(execute, self, Index:int=3) -> (then)
   SetSelectionByIndex(Index=1)
   SetSelectionByIndex(Index=2)
   SetSelectionByIndex(Index=0)
  ]}
 CheckForEnableProInstrumentSelection {nodes [
   CheckForEnableProInstrumentSelection() {is=title K2Node_FunctionEntry}
   n1 = GetOwningPlayer()
   > n2 = Is Valid(InputObject=@n1)
   Return Node!(execute=[@n10.Is Not Valid, @n2.Is Not Valid, @n6.Is Not Valid, @n8.Is Not Valid, @n4.Is Not Valid], ReturnValue:bool=true)
   n3 = Get PlayerState!(self:Controller Object Reference=@n1) -> (PlayerState:Player State Object Reference)
   > n4 = Is Valid(InputObject=@n3)
   n5 = GetPlayerController!(self:Player State Object Reference=@n3) -> (ReturnValue:Player Controller Object Reference)
   > n6 = Is Valid(InputObject=@n5)
   n7 = GetComponentByClass!(
    self:Actor Object Reference=@n5
    ComponentClass:Actor Component Class Reference=/PilgrimCore/Player/Components/BP_PilgrimPlayerControllerComponent.BP_PilgrimPlayerControllerComponent_C
   ) -> (ReturnValue:BP Pilgrim Player Controller Component Object Reference)
   > n8 = Is Valid(InputObject=@n7)
   n9 = GetPilgrimClientSettingRecordPartition!(self:Pilgrim Player Controller Component Object Reference=@n7) -> (ReturnValue:Pilgrim Client Setting Record Partition Object Reference)
   > n10 = Is Valid(InputObject=@n9)
   > Return Node(ReturnValue=GetAlwaysShowProInstrumentTrackSelection!(self:Pilgrim Client Setting Record Partition Object Reference=@n9) -> (ReturnValue:bool))
  ]}
 Does Platform Support Guitar Peripheral {nodes [
   Does Platform Support Guitar Peripheral() {is=title K2Node_FunctionEntry}
   n1 = GetCurrentInputType(self=Get CommonInputSubsystem())
   > Return Node!(execute, PlatformSupport:bool=NOT Boolean(A="Equal (Enum)"(A=@n1, B=Touch)))
  ]}
 EventGraph {nodes [
   n1 = "On Visibility Changed (KeybindWidget_ChangeBindings_n3)"!() -> (OutputDelegate, then, InVisibility:ESlateVisibility Enum=Visible) {is=title K2Node_ComponentBoundEvent}
   > n2 = Branch(Condition="Not Equal (Enum)"!(A:ESlateVisibility Enum=@n1.InVisibility, B:ESlateVisibility Enum=Hidden) -> (ReturnValue:bool))
   Branch(execute=[@n8.then, @n12.then, @n2.then, @n29.then], Condition=NOT Boolean(A=Get bAllowGuitarSelect() -> (bAllowGuitarSelect:bool)))
   n3 = Get KeybindWidget_ChangeBindings_n5() -> (KeybindWidget_ChangeBindings_n5:Keybind Widget Object Reference)
   n4 = Get KeybindWidget_ChangeBindings_n3() -> (KeybindWidget_ChangeBindings_n3:Keybind Widget Object Reference)
   > SetColorAndOpacity(
    self:User Widget Object Reference=[
     Get UIKit_Button_Guitar_Peripheral() -> (UIKit_Button_Guitar_Peripheral:WBP UIKit Button Selectable Object Reference)
     @n3
     Get UIKit_Button_Guitar() -> (UIKit_Button_Guitar:WBP UIKit Button Selectable Object Reference)
     @n4
    ]
    InColorAndOpacity:Linear Color Structure=(R=1,G=1,B=1,A=0.2)
   )
   > SetVisibility(self=[@n3, @n4], InVisibility=Hidden)
   "On Clicked (WBP_UIKit_Button_Regular)"!() -> (OutputDelegate, then, Button:Common Button Base Object Reference) {is=title K2Node_ComponentBoundEvent}
   > SwapInstrumentForward!(
    execute
    self:BP Pilgrim Player Controller Component Object Reference=Get BP_PilgrimPlayerControllerComponent()
    IsTouchInputPlatform:bool=IsUsingTouchInput!() -> (IsUsingTouchInput:bool)
   ) -> (then)
   > PlayAnimation(InAnimation=Get Pressed() -> (Pressed:Widget Animation Object Reference))
   n5 = "On Visibility Changed (UIKit_Button_Bass)"!() -> (OutputDelegate, then, InVisibility:ESlateVisibility Enum=Visible) {is=title K2Node_ComponentBoundEvent}
   Branch(execute=[@n5.then, @n13.then, @n31.then, @n19.then], Condition=NOT Boolean(A=Get bAllowBassSelect() -> (bAllowBassSelect:bool)))
   n6 = Get KeybindWidget_ChangeBindings_n6() -> (KeybindWidget_ChangeBindings_n6:Keybind Widget Object Reference)
   n7 = Get KeybindWidget_ChangeBindings_n2() -> (KeybindWidget_ChangeBindings_n2:Keybind Widget Object Reference)
   > SetColorAndOpacity(
    self=[
     Get UIKit_Button_Bass() -> (UIKit_Button_Bass:WBP UIKit Button Selectable Object Reference)
     Get UIKit_Button_Bass_Peripheral() -> (UIKit_Button_Bass_Peripheral:WBP UIKit Button Selectable Object Reference)
     @n6
     @n7
    ]
    InColorAndOpacity=(R=1,G=1,B=1,A=0.2)
   )
   > SetVisibility(self=[@n6, @n7], InVisibility=Hidden)
   n8 = "On Visibility Changed (UIKit_Button_Guitar)"!() -> (OutputDelegate, then, InVisibility:ESlateVisibility Enum=Visible) {is=title K2Node_ComponentBoundEvent}
   n9 = "On Visibility Changed (UIKit_Button_Drums)"!() -> (OutputDelegate, then, InVisibility:ESlateVisibility Enum=Visible) {is=title K2Node_ComponentBoundEvent}
   Branch(execute=[@n9.then, @n25.then, @n27.then, @n24.then], Condition=NOT Boolean(A=Get bAllowDrumsSelect() -> (bAllowDrumsSelect:bool)))
   n10 = Get KeybindWidget_ChangeBindings_n7() -> (KeybindWidget_ChangeBindings_n7:Keybind Widget Object Reference)
   n11 = Get KeybindWidget_ChangeBindings_n4() -> (KeybindWidget_ChangeBindings_n4:Keybind Widget Object Reference)
   > SetColorAndOpacity(
    self=[
     Get UIKit_Button_Drums_Peripheral() -> (UIKit_Button_Drums_Peripheral:WBP UIKit Button Selectable Object Reference)
     @n10
     Get UIKit_Button_Drums() -> (UIKit_Button_Drums:WBP UIKit Button Selectable Object Reference)
     @n11
    ]
    InColorAndOpacity=(R=1,G=1,B=1,A=0.2)
   )
   > SetVisibility(self=[@n10, @n11], InVisibility=Hidden)
   n12 = "On Visibility Changed (UIKit_Button_Guitar_Peripheral)"!() -> (OutputDelegate, then, InVisibility:ESlateVisibility Enum=Visible) {is=title K2Node_ComponentBoundEvent}
   n13 = "On Visibility Changed (UIKit_Button_Bass_Peripheral)"!() -> (OutputDelegate, then, InVisibility:ESlateVisibility Enum=Visible) {is=title K2Node_ComponentBoundEvent}
   n14 = "On Visibility Changed (KeybindWidget_ChangeBindings_n1)"!() -> (OutputDelegate, then, InVisibility:ESlateVisibility Enum=Visible) {is=title K2Node_ComponentBoundEvent}
   > n15 = Branch(Condition="Not Equal (Enum)"(A=@n14.InVisibility, B=Hidden))
   Branch(execute=[@n22.then, @n32.then, @n15.then, @n21.then], Condition=NOT Boolean(A=Get bAllowVocalsSelect() -> (bAllowVocalsSelect:bool)))
   n16 = Get KeybindWidget_ChangeBindings_n8() -> (KeybindWidget_ChangeBindings_n8:Keybind Widget Object Reference)
   n17 = Get KeybindWidget_ChangeBindings_n1() -> (KeybindWidget_ChangeBindings_n1:Keybind Widget Object Reference)
   > SetColorAndOpacity(
    self=[
     Get UIKit_Button_Vocals_Peripheral() -> (UIKit_Button_Vocals_Peripheral:WBP UIKit Button Selectable Object Reference)
     @n16
     Get UIKit_Button_Vocal() -> (UIKit_Button_Vocal:WBP UIKit Button Selectable Object Reference)
     @n17
    ]
    InColorAndOpacity=(R=1,G=1,B=1,A=0.2)
   )
   > SetVisibility(self=[@n16, @n17], InVisibility=Hidden)
   n18 = "On Visibility Changed (KeybindWidget_ChangeBindings_n2)"!() -> (OutputDelegate, then, InVisibility:ESlateVisibility Enum=Visible) {is=title K2Node_ComponentBoundEvent}
   > n19 = Branch(Condition="Not Equal (Enum)"(A=@n18.InVisibility, B=Hidden))
   n20 = "On Visibility Changed (KeybindWidget_ChangeBindings_n8)"!() -> (OutputDelegate, then, InVisibility:ESlateVisibility Enum=Visible) {is=title K2Node_ComponentBoundEvent}
   > n21 = Branch(Condition="Not Equal (Enum)"(A=@n20.InVisibility, B=Hidden))
   n22 = "On Visibility Changed (UIKit_Button_Vocals_Peripheral)"!() -> (OutputDelegate, then, InVisibility:ESlateVisibility Enum=Visible) {is=title K2Node_ComponentBoundEvent}
   "On Hovered (WBP_UIKit_Button_Regular)"!() -> (OutputDelegate, then, Button:Common Button Base Object Reference) {is=title K2Node_ComponentBoundEvent}
   > OnHoverVisuals()
   "On Unhovered (WBP_UIKit_Button_Regular)"!() -> (OutputDelegate, then, Button:Common Button Base Object Reference) {is=title K2Node_ComponentBoundEvent}
   > OnUnhoverVisuals()
   n23 = "On Visibility Changed (KeybindWidget_ChangeBindings_n7)"!() -> (OutputDelegate, then, InVisibility:ESlateVisibility Enum=Visible) {is=title K2Node_ComponentBoundEvent}
   > n24 = Branch(Condition="Not Equal (Enum)"(A=@n23.InVisibility, B=Hidden))
   n25 = "On Visibility Changed (UIKit_Button_Drums_Peripheral)"!() -> (OutputDelegate, then, InVisibility:ESlateVisibility Enum=Visible) {is=title K2Node_ComponentBoundEvent}
   n26 = "On Visibility Changed (KeybindWidget_ChangeBindings_n4)"!() -> (OutputDelegate, then, InVisibility:ESlateVisibility Enum=Visible) {is=title K2Node_ComponentBoundEvent}
   > n27 = Branch(Condition="Not Equal (Enum)"(A=@n26.InVisibility, B=Hidden))
   n28 = "On Visibility Changed (KeybindWidget_ChangeBindings_n5)"!() -> (OutputDelegate, then, InVisibility:ESlateVisibility Enum=Visible) {is=title K2Node_ComponentBoundEvent}
   > n29 = Branch(Condition="Not Equal (Enum)"(A=@n28.InVisibility, B=Hidden))
   n30 = "On Visibility Changed (KeybindWidget_ChangeBindings_n6)"!() -> (OutputDelegate, then, InVisibility:ESlateVisibility Enum=Visible) {is=title K2Node_ComponentBoundEvent}
   > n31 = Branch(Condition="Not Equal (Enum)"(A=@n30.InVisibility, B=Hidden))
   n32 = "On Visibility Changed (UIKit_Button_Vocal)"!() -> (OutputDelegate, then, InVisibility:ESlateVisibility Enum=Visible) {is=title K2Node_ComponentBoundEvent}
   n33 = On Input Method Changed!() -> (OutputDelegate, then, bNewInputType:ECommonInputType Enum) {is=title K2Node_CustomEvent}
   > n34 = Swap to Valid Instrument()
   SetGuitarPeripheralAvailability!(execute=[@n44.then, @n34.then], self, IsGuitarAvailable:bool=true) -> (then)
   Event PreConstruct()
   > SetupIcons()
   > SetSelectionByIndex(Index=0)
   n35 = GetOwningPlayer()
   > Is Valid(InputObject=@n35)
   > Set BP_PilgrimPlayerControllerComponent(BP_PilgrimPlayerControllerComponent:BP Pilgrim Player Controller Component Object Reference=GetComponentByClass(self=@n35, ComponentClass=&BP_PilgrimPlayerControllerComponent_C)) -> (Output_Get:BP Pilgrim Player Controller Component Object Reference)
   Event Construct()
   > n36 = Sequence()
   n37 = Get BP_PilgrimPlayerControllerComponent()
   > Is Valid(InputObject=@n37)
   > SetProInstrumentUIVisibility!(
    execute
    self
    IsVisible:bool=CanProInstrumentsBeSelected!(self:Pilgrim Player Controller Component Object Reference=@n37) -> (ReturnValue:bool)
   ) -> (then)
   n38 = PlaylistDataReadyAsync!(execute:Exec=@n36.then_1) -> (
    then:Exec
    OnReady:Exec
    GameState:Fort Game State Athena Object Reference
    Playlist:Fort Playlist Object Reference
    PlaylistContextTags:"Gameplay Tag Container Structure (by ref)"
   ) {is=title K2Node_AsyncAction}
   n39 = Cast To FortPlaylistAthena!(execute=@n38.OnReady, Object=@n38.Playlist) -> (then, CastFailed, AsFort Playlist Athena:Fort Playlist Athena Object Reference) {is=title K2Node_DynamicCast}
   > SetupInstrumentFiltersFromPlaylistGameplayTags!(
    execute
    self
    InGameplayTags:Gameplay Tag Container Structure=Get GameplayTagContainer!(self:Fort Playlist Object Reference=@n39) -> (GameplayTagContainer:Gameplay Tag Container Structure)
   ) -> (then)
   > n40 = Sequence()
   > Set Inputs Visibility() {is=title K2Node_CallFunction}
   > n41 = PlaylistDataReadyAsync()
   n42 = PilgrimWaitForEventRouter!(execute:Exec=@n41.OnReady, ContextActor:Actor Object Reference=GetOwningPlayer()) -> (then:Exec, OnEventRouterAvailable:Exec, EventRouter:Gameplay Event Router Component Object Reference) {is=title K2Node_AsyncAction}
   n43 = Sequence!(execute=@n42.OnEventRouterAvailable) -> (then_0, then_1, then_2:Exec, then_3:Exec, then_4:Exec, then_5:Exec)
   > n44 = Bind Event to On Input Method Changed(self=Get CommonInputSubsystem(), Delegate=@n33)
   n45 = StartListeningToEvent!(
    execute:Exec=@n43.then_1
    Target:Gameplay Event Router Component Object Reference=@n42
    EventType:Script Struct Object Reference=/PilgrimQuickplay/Items/PilgrimEvent_GreenRoomTimerStateChanged
    Context:Object Reference
    EventBubblingRule:EEventBubblingRule Enum
   ) -> (
    then:Exec
    AsyncTaskProxy:Async Action Start Listening To Event Object Reference
    OnEventReceived:Exec
    Payload:Pilgrim Event Green Room Timer State Changed Structure
   )
   > Add(
    TargetArray:Array of Async Action Start Listening To Event Object References=Get GameplayEventHandles() -> (GameplayEventHandles:Array of Async Action Start Listening To Event Object References)
    NewItem:"Async Action Start Listening To Event Object Reference (by ref)"=@n45
   ) {is=title K2Node_CallArrayFunction}
   n46 = Branch(
    execute=@n45.OnEventReceived
    Condition="Equal (Enum)"(
     A:GreenRoomTimerState Enum=Break Pilgrim Event Green Room Timer State Changed!(PilgrimEvent_GreenRoomTimerStateChanged:"Pilgrim Event Green Room Timer State Changed Structure (by ref)"=@n45.Payload) -> (
      GreenRoomTimerState_2_#3:GreenRoomTimerState Enum
      FinalTimerLength_5_#4:double=~
     ) {is=title K2Node_BreakStruct}
     B:GreenRoomTimerState Enum=NewEnumerator2
    )
   )
   > SetVisibility(InVisibility=Collapsed)
   SetVisibility(execute=@n46.else, InVisibility=SelfHitTestInvisible)
   n47 = StartListeningToEvent(
    execute=@n43.then_2
    Target=@n42
    EventType=/Script/PilgrimSettingsRuntime.PilgrimEvent_PilgrimSettings_AlwaysShowProInstrumentTrackSelection
   ) -> (Payload:Pilgrim Event Pilgrim Settings Always Show Pro Instrument Track Selection Structure)
   SetProInstrumentUIVisibility(execute=@n47.OnEventReceived, IsVisible=CanProInstrumentsBeSelected(self=Get BP_PilgrimPlayerControllerComponent()))
   SetProInstrumentUIVisibility(execute=@n43.then_3, IsVisible=CanProInstrumentsBeSelected(self=Get BP_PilgrimPlayerControllerComponent()))
   n48 = StartListeningToEvent(execute=@n43.then_4, Target=@n42, EventType=/Script/PilgrimCoreRuntime.PilgrimEvent_PlayerState_HasConnectedPeripheralChanged) -> (Payload:Pilgrim Event Player State Has Connected Peripheral Changed Structure)
   SetProInstrumentUIVisibility(execute=[@n48.OnEventReceived, @n48.then], IsVisible=CanProInstrumentsBeSelected(self=Get BP_PilgrimPlayerControllerComponent()))
   n49 = StartListeningToEvent!(
    execute=@n43.then_5
    Target=@n42
    EventType=/Script/PilgrimCoreRuntime.PilgrimEvent_EnableChoppedCheeseCvarChanged
    Context
    EventBubblingRule
   ) -> (then, AsyncTaskProxy, OnEventReceived, Payload_bIsChoppedCheeseCvarEnabled:bool=~)
   SetProDumsVisibility!(execute=@n49.OnEventReceived, self, IsVisible:bool=@n49.Payload_bIsChoppedCheeseCvarEnabled) -> (then)
   SetDisabledInstrumentVisuals(execute=@n40.then_1)
  ]}
 GetButtonsByIndex {nodes [
   n1 = GetButtonsByIndex!() -> (then, Index:int) {is=title K2Node_FunctionEntry}
   n2 = Get UIKit_Button_Guitar()
   n3 = Get UIKit_Button_Guitar_Peripheral()
   n4 = Get UIKit_Button_Bass_Peripheral()
   n5 = Get UIKit_Button_Bass()
   n6 = Get UIKit_Button_Drums_Peripheral()
   n7 = Get UIKit_Button_Vocals_Peripheral()
   n8 = Get UIKit_Button_Drums()
   n9 = Get UIKit_Button_Vocal()
   > Return Node!(
    execute
    ButtonWidget:WBP UIKit Button Selectable Object Reference="Get (a copy)"(
     Array:Array of WBP UIKit Button Selectable Object References=Make Array!(
      "[0]":WBP UIKit Button Selectable Object Reference=@n2
      "[1]":WBP UIKit Button Selectable Object Reference=@n8
      "[2]":WBP UIKit Button Selectable Object Reference=@n9
      "[3]":WBP UIKit Button Selectable Object Reference=@n5
      "[4]":WBP UIKit Button Selectable Object Reference=@n3
      "[5]":WBP UIKit Button Selectable Object Reference=@n4
      "[6]":WBP UIKit Button Selectable Object Reference=@n6
      "[7]":WBP UIKit Button Selectable Object Reference=@n7
     ) -> (Array:Array of WBP UIKit Button Selectable Object References)
     Dimension 1=@n1
    ) -> (Output:WBP UIKit Button Selectable Object Reference)
   )
  ]}
 IsProInstrument {nodes [
   n1 = IsProInstrument!() -> (then, A:int) {is=title K2Node_FunctionEntry}
   > Return Node!(execute, ReturnValue:bool=AND Boolean(A="integer > integer"(A=@n1, B=3), B="integer < integer"(A=@n1, B=7)))
  ]}
 IsUsingTouchInput {nodes [
   IsUsingTouchInput() {is=title K2Node_FunctionEntry}
   n1 = Get CommonInputSubsystem()
   > Return Node!(execute, IsUsingTouchInput:bool="Equal (Enum)"(A:ECommonInputType Enum=GetCurrentInputType(self=@n1), B:ECommonInputType Enum=Touch))
  ]}
 OnHoverVisuals {nodes [
   OnHoverVisuals() {is=title K2Node_FunctionEntry}
   > PlayAnimationForward(InAnimation=Get Hover() -> (Hover:Widget Animation Object Reference))
   > RequestTransition!(
    execute
    self:UIKit Block Object Reference=Get WBP_UIKit_Block_Outline() -> (WBP_UIKit_Block_Outline:WBP UIKit Block Outline Object Reference)
    State:EUIKitBlockVisualState Enum=Focused
   ) -> (then)
   > PlayHover(self:WBP Sparks Backing Panel Object Reference=Get WBP_Sparks_BackingPanel() -> (WBP_Sparks_BackingPanel:WBP Sparks Backing Panel Player Identity Object Reference))
   > SetVisibility(self=Get InputContainer_GridPanel() -> (InputContainer_GridPanel:Grid Panel Object Reference), InVisibility=SelfHitTestInvisible)
   n1 = For Loop(LastIndex=3)
   > n2 = GetButtonsByIndex!(execute, self, Index:int=@n1) -> (then, ButtonWidget:WBP UIKit Button Selectable Object Reference) {is=title K2Node_CallFunction}
   > SetColorAndOpacity(self=@n2, InColorAndOpacity=(R=0,G=0,B=0,A=1))
  ]}
 OnInstrumentChanged {nodes [
   n1 = OnInstrumentChanged!() -> (then, New Instrument Type:EPilgrimTrackType Enum=TrackGuitar) {is=title K2Node_FunctionEntry}
   > SetSelectionByIndex(Index=Select!(
     TrackGuitar:int=0
     TrackBass:int=3
     TrackVocals:int=2
     TrackDrum:int=1
     TrackPlasticGuitar:int=4
     TrackPlasticBass:int=5
     TrackPlasticDrum:int=6
     TrackProVocals:int=7
     TrackProCymbals:int=0
     TrackEvents:int=0
     TrackNone:int=0
     Index:EPilgrimTrackType Enum=@n1
    ) -> (ReturnValue:int))
   > Set Text(
    Text=GetInstrumentNameFromPart!(TrackType:EPilgrimTrackType Enum=@n1) -> (InstrumentText:Text)
    self=Get Common_Text_Instrument() -> (Common_Text_Instrument:Common Text Object Reference)
   )
  ]}
 OnUnhoverVisuals {nodes [
   OnUnhoverVisuals() {is=title K2Node_FunctionEntry}
   > PlayAnimationReverse(InAnimation=Get Hover())
   > SetVisibility(self=Get InputContainer_GridPanel(), InVisibility=Collapsed)
   > RequestTransition(self=Get WBP_UIKit_Block_Outline(), State=Unfocused)
   > PlayUnhover(self:WBP Sparks Backing Panel Object Reference=Get WBP_Sparks_BackingPanel())
   > SetColorAndOpacity!(
    execute
    self:Text Object Reference=Get CurrentInputBindingNumber() -> (CurrentInputBindingNumber:Common Text Object Reference)
    InColorAndOpacity_SpecifiedColor:Linear Color Structure=(R=1,G=1,B=1,A=1)
    InColorAndOpacity_ColorUseRule:ESlateColorStylingMode Enum
   ) -> (then)
   > SetColorAndOpacity(self=Get Common_Text_Instrument(), InColorAndOpacity_SpecifiedColor=(R=0.3185,G=0.7231,B=1,A=1))
  ]}
 Set Inputs Visibility {nodes [
   Set Inputs Visibility() {is=title K2Node_FunctionEntry}
   n1 = Get IsOnGreenRoom() -> (IsOnGreenRoom:bool)
   n2 = Get Keybind_GridPanel() -> (Keybind_GridPanel:Grid Panel Object Reference)
   > SetIsEnabled(self=@n2, bInIsEnabled=@n1)
   > SetVisibility(
    self=[@n2, Get MainInputsContainer() -> (MainInputsContainer:Horizontal Box Object Reference)]
    InVisibility=Select!(Option 0:ESlateVisibility Enum=Collapsed, Option 1:ESlateVisibility Enum=SelfHitTestInvisible, Index:bool=@n1) -> (ReturnValue:ESlateVisibility Enum)
   )
  ]}
 SetDisabledInstrumentVisuals {nodes [
   SetDisabledInstrumentVisuals() {is=title K2Node_FunctionEntry}
   > n1 = Sequence!(execute) -> (then_0, then_1, then_2, then_3)
   > Branch(Condition=NOT Boolean(A=Get bAllowGuitarSelect()))
   n2 = Get KeybindWidget_ChangeBindings_n5()
   n3 = Get KeybindWidget_ChangeBindings_n3()
   > SetColorAndOpacity!(
    execute
    self:User Widget Object Reference=[Get UIKit_Button_Guitar_Peripheral(), Get UIKit_Button_Guitar(), @n2, @n3]
    InColorAndOpacity:Linear Color Structure=(R=1,G=1,B=1,A=0.2)
   ) -> (then)
   > SetVisibility(self=[@n2, @n3], InVisibility=Hidden)
   Branch(execute=@n1.then_1, Condition=NOT Boolean(A=Get bAllowBassSelect()))
   n4 = Get KeybindWidget_ChangeBindings_n6()
   n5 = Get KeybindWidget_ChangeBindings_n2()
   > SetColorAndOpacity(self=[Get UIKit_Button_Bass_Peripheral(), Get UIKit_Button_Bass(), @n4, @n5], InColorAndOpacity=(R=1,G=1,B=1,A=0.2))
   > SetVisibility(self=[@n4, @n5], InVisibility=Hidden)
   Branch(execute=@n1.then_2, Condition=NOT Boolean(A=Get bAllowVocalsSelect()))
   n6 = Get KeybindWidget_ChangeBindings_n1()
   n7 = Get KeybindWidget_ChangeBindings_n8()
   > SetColorAndOpacity(self=[Get UIKit_Button_Vocal(), @n6, Get UIKit_Button_Vocals_Peripheral(), @n7], InColorAndOpacity=(R=1,G=1,B=1,A=0.2))
   > SetVisibility(self=[@n6, @n7], InVisibility=Hidden)
   Branch(execute=@n1.then_3, Condition=NOT Boolean(A=Get bAllowDrumsSelect()))
   n8 = Get KeybindWidget_ChangeBindings_n7()
   n9 = Get KeybindWidget_ChangeBindings_n4()
   > SetColorAndOpacity(self=[Get UIKit_Button_Drums_Peripheral(), Get UIKit_Button_Drums(), @n8, @n9], InColorAndOpacity=(R=1,G=1,B=1,A=0.2))
   > SetVisibility(self=[@n8, @n9], InVisibility=Hidden)
  ]}
 SetGuitarPeripheralAvailability {nodes [
   n1 = SetGuitarPeripheralAvailability!() -> (then, IsGuitarAvailable:bool=~) {is=title K2Node_FunctionEntry}
   n2 = Does Platform Support Guitar Peripheral!() -> (PlatformSupport:bool) {is=title K2Node_CallFunction}
   n3 = CanProInstrumentsBeSelected(self=Get BP_PilgrimPlayerControllerComponent())
   > SetProInstrumentUIVisibility(IsVisible=AND Boolean(A=@n1, B=AND Boolean(A=@n2, B=@n3)))
  ]}
 SetIconImage {nodes [
   n1 = SetIconImage!() -> (then, Button:WBP UIKit Button Generic Base Object Reference, IconIndex:int) {is=title K2Node_FunctionEntry}
   n2 = Get Image!(self:WBP UIKit Block Image Object Reference=Get Block_Icon!(self:WBP UIKit Button Generic Base Object Reference=@n1) -> (Block_Icon:WBP UIKit Block Image Object Reference)) -> (Image:Fort Mobile Image Object Reference)
   > SetBrushFromMaterial(self=@n2, Material=/PilgrimQuickplay/Widgets/Materials/MI_UI_Icon_Instrument_Type_Tab)
   > n3 = GetDynamicMaterial(self=@n2)
   > SetScalarParameterValue(self=@n3, ParameterName=Icon, Value="To Float (Integer)"(InInt=@n1.IconIndex))
  ]}
 SetPlayerColor {nodes [
   n1 = SetPlayerColor!() -> (then, PlayerColor:Linear Color Structure) {is=title K2Node_FunctionEntry}
   > SetBackingColor!(
    execute
    self:WBP Sparks Backing Panel Player Identity Object Reference=Get WBP_Sparks_BackingPanel()
    Linear Color:Linear Color Structure=@n1
   ) -> (then)
   > n2 = GetDynamicMaterial(self:Border Object Reference=Get Instruments_Border() -> (Instruments_Border:Border Object Reference))
   > n3 = GetDynamicMaterial(self=Get Peripheral_Border() -> (Peripheral_Border:Border Object Reference))
   > SetScalarParameterValue(self=[@n3, @n2], ParameterName=IsPlayerColor, Value=1)
   > SetVectorParameterValue(self=[@n3, @n2], ParameterName=Color, Value=Get PlayerColor() -> (PlayerColor:Linear Color Structure))
   > n4 = For Each Loop(Array:Array of WBP UIKit Button Selectable Object References=Make Array(
     "[0]"=Get UIKit_Button_Vocal()
     "[1]"=Get UIKit_Button_Bass()
     "[2]"=Get UIKit_Button_Guitar()
     "[3]"=Get UIKit_Button_Drums()
     "[4]"=Get UIKit_Button_Guitar_Peripheral()
     "[5]"=Get UIKit_Button_Bass_Peripheral()
     "[6]"=Get UIKit_Button_Drums_Peripheral()
     "[7]"=Get UIKit_Button_Vocals_Peripheral()
    )) -> (Array Element:WBP UIKit Button Selectable Object Reference)
   > n5 = CreateDynamicMaterialInstance!(execute, Parent=/PilgrimCore/Widgets/Materials/MI_UI_Piligrim_IconTabBacking, OptionalName=None, CreationFlags=None) -> (then, ReturnValue)
   > SetScalarParameterValue(self=@n5, ParameterName=IsPlayerColor, Value=1)
   > SetVectorParameterValue(self=@n5, ParameterName=Color, Value=Get PlayerColor())
   > SetBackgroundMaterials!(
    execute
    self:WBP UIKit Button Generic Base Object Reference=@n4
    Background MI:Material Instance Object Reference=@n5
    Outline MI:Material Instance Object Reference
   ) -> (then)
  ]}
 SetProDumsVisibility {nodes [
   n1 = SetProDumsVisibility!() -> (then, IsVisible:bool=~) {is=title K2Node_FunctionEntry}
   > n2 = Branch(Condition=@n1)
   n3 = Get UIKit_Button_Drums_Peripheral()
   n4 = Get Overlay_13() -> (Overlay_13:Overlay Object Reference)
   n5 = Get Spacer_3() -> (Spacer_3:Spacer Object Reference)
   > SetVisibility(self=[@n3, @n4, @n5], InVisibility=Visible)
   SetVisibility(execute=@n2.else, self=[@n3, @n4, @n5], InVisibility=Collapsed)
  ]}
 SetProGuitarsVisibility {nodes [
   n1 = SetProGuitarsVisibility!() -> (then, Condition:bool=~) {is=title K2Node_FunctionEntry}
   > n2 = Branch(Condition=@n1)
   n3 = Get Overlay_11() -> (Overlay_11:Overlay Object Reference)
   n4 = Get Overlay_12() -> (Overlay_12:Overlay Object Reference)
   n5 = Get Spacer_2() -> (Spacer_2:Spacer Object Reference)
   n6 = Get Spacer_8() -> (Spacer_8:Spacer Object Reference)
   n7 = Get PeripheralParts+Prompts() -> (PeripheralParts+Prompts:Overlay Object Reference)
   > SetVisibility(self=[@n3, @n4, @n5, @n6, @n7], InVisibility=Visible)
   SetVisibility(execute=@n2.else, self=[@n3, @n4, @n5, @n6, @n7], InVisibility=Collapsed)
  ]}
 SetProInstrumentUIVisibility {nodes [
   n1 = SetProInstrumentUIVisibility!() -> (then, IsVisible:bool=~) {is=title K2Node_FunctionEntry}
   n2 = Get BP_PilgrimPlayerControllerComponent()
   > SetProDumsVisibility(IsVisible=AND Boolean!(A:bool=@n1, B:bool=IsChoppedCheeseEnabled!(self:Pilgrim Player Controller Component Object Reference=@n2) -> (ReturnValue:bool)) -> (ReturnValue:bool) {is=title K2Node_CallFunction}) {is=title K2Node_CallFunction}
   > SetProVocalsVisibility!(
    execute
    self
    IsVisible:bool=AND Boolean(A=@n1, B=IsChickenParmEnabled!(self:Pilgrim Player Controller Component Object Reference=@n2) -> (ReturnValue:bool))
   ) -> (then)
   > SetProGuitarsVisibility!(execute, self, Condition:bool=@n1) -> (then) {is=title K2Node_CallFunction}
   > n3 = Branch(Condition=@n1)
   n4 = Get PlayerState(self=GetOwningPlayer())
   n5 = GetPilgrimCorePlayerStateComponent!(execute=@n3.else, PlayerState:Player State Object Reference=@n4) -> (then, ReturnValue:Pilgrim Core Player State Component Object Reference)
   n6 = GetInstrumentChoice!(self:Pilgrim Core Player State Component Object Reference=@n5) -> (ReturnValue:EPilgrimTrackType Enum=TrackGuitar)
   > Branch(Condition=OR Boolean(A="Equal (Enum)"(A:EPilgrimTrackType Enum=@n6, B:EPilgrimTrackType Enum=TrackPlasticGuitar), B="Equal (Enum)"(A=@n6, B=TrackPlasticBass)))
   > SetInstrumentByChoice!(
    execute
    self:BP Pilgrim Player Controller Component Object Reference=Get BP_PilgrimPlayerControllerComponent()
    InstrumentSelected:EPilgrimTrackType Enum=TrackDrum
   ) -> (then)
   GetPlayerController(self=@n4)
  ]}
 SetProVocalsVisibility {nodes [
   n1 = SetProVocalsVisibility!() -> (then, IsVisible:bool=~) {is=title K2Node_FunctionEntry}
   > n2 = Branch(Condition=@n1)
   n3 = Get UIKit_Button_Vocals_Peripheral()
   n4 = Get Overlay_14() -> (Overlay_14:Overlay Object Reference)
   n5 = Get Spacer_4() -> (Spacer_4:Spacer Object Reference)
   > SetVisibility(self=[@n3, @n4, @n5], InVisibility=Visible)
   SetVisibility(execute=@n2.else, self=[@n3, @n4, @n5], InVisibility=Collapsed)
  ]}
 SetSelectionByIndex {nodes [
   n1 = SetSelectionByIndex!() -> (then, Index:int) {is=title K2Node_FunctionEntry}
   > n2 = GetButtonsByIndex(Index=@n1)
   > ClearSelection(self:Common Button Base Object Reference=@n2)
   > SetIsSelected!(execute, self:Common Button Base Object Reference=@n2, InSelected:bool=true, bGiveClickFeedback:bool=true) -> (then)
   > n3 = For Each Loop(Array=Get TabArray() -> (TabArray:Array of WBP UIKit Button Selectable Object References))
   n4 = "Not Equal (Integer)"(A=@n3.Array Index, B=@n1)
   > Branch(Condition=@n4)
   > n5 = GetButtonsByIndex(Index=@n3.Array Index)
   > Branch(Condition=GetSelected!(self:Common Button Base Object Reference=@n5) -> (ReturnValue:bool))
   > ClearSelection(self=@n5)
   > SetIsSelected(self=@n5, InSelected=false, bGiveClickFeedback=true)
   n6 = CanProInstrumentsBeSelected(self=Get BP_PilgrimPlayerControllerComponent())
   > Branch(Condition=AND Boolean(A=IsProInstrument!(A:int=@n1) -> (ReturnValue:bool) {is=title K2Node_CallFunction}, B=NOT Boolean(A=@n6)) {is=title K2Node_CommutativeAssociativeBinaryOperator})
   > SetInstrumentByChoice(self=Get BP_PilgrimPlayerControllerComponent(), InstrumentSelected=TrackDrum)
   AND Boolean(A=GetSelected(self=@n3), B=@n4)
  ]}
 SetupIcons {nodes [
   SetupIcons() {is=title K2Node_FunctionEntry}
   > n1 = Set TabArray(TabArray:Array of WBP UIKit Button Selectable Object References=Make Array(
     "[0]"=Get UIKit_Button_Guitar()
     "[1]"=Get UIKit_Button_Bass()
     "[2]"=Get UIKit_Button_Vocal()
     "[3]"=Get UIKit_Button_Drums()
     "[4]"=Get UIKit_Button_Guitar_Peripheral()
     "[5]"=Get UIKit_Button_Bass_Peripheral()
     "[6]"=Get UIKit_Button_Drums_Peripheral()
     "[7]"=Get UIKit_Button_Vocals_Peripheral()
    )) -> (Output_Get:Array of WBP UIKit Button Selectable Object References)
   > n2 = For Each Loop(Array=@n1)
   n3 = Select!(
    Option 0:EPilgrimTrackType Enum=TrackGuitar
    Option 1:EPilgrimTrackType Enum=TrackBass
    Option 2:EPilgrimTrackType Enum=TrackVocals
    Option 3:EPilgrimTrackType Enum=TrackDrum
    Option 4:EPilgrimTrackType Enum=TrackPlasticGuitar
    Option 5:EPilgrimTrackType Enum=TrackPlasticBass
    Option 6:EPilgrimTrackType Enum=TrackPlasticDrum
    Option 7:EPilgrimTrackType Enum=TrackProVocals
    Index:int=@n2.Array Index
   ) -> (ReturnValue:EPilgrimTrackType Enum)
   > SetIconImage!(
    execute
    self
    Button:WBP UIKit Button Generic Base Object Reference=@n2
    IconIndex:int=GetInstrumentIconIndex!(
     InstrumentType:EPilgrimTrackType Enum=@n3
     SongMetaData:Pilgrim Song Metadata Object Reference
     StartingInstrumentIndex:ESparksInstrumentType Enum=PilgrimTrackTypeToSparksInstrument!(PilgrimType:EPilgrimTrackType Enum=@n3) -> (ReturnValue:ESparksInstrumentType Enum=Guitar)
     bUseKeyTarIcon:bool=false
    ) -> (InstrumentIdexIcon:int=0)
   ) -> (then) {is=title K2Node_CallFunction}
  ]}
 SetupInstrumentFiltersFromPlaylistGameplayTags {nodes [
   n1 = SetupInstrumentFiltersFromPlaylistGameplayTags!() -> (then, InGameplayTags:Gameplay Tag Container Structure) {is=title K2Node_FunctionEntry}
   > n2 = Branch(Condition=HasTag!(
     TagContainer:"Gameplay Tag Container Structure (by ref)"=@n1
     Tag:Gameplay Tag Structure=(TagName="Sparks.Playlist.Pilgrim.BattleStage.ScoreChase.Bass")
     bExactMatch:bool=true
    ) -> (ReturnValue:bool))
   > Set bAllowBassSelect(bAllowBassSelect:bool=true) -> (Output_Get:bool)
   > Set bAllowGuitarSelect(bAllowGuitarSelect:bool=false) -> (Output_Get:bool)
   > Set bAllowVocalsSelect(bAllowVocalsSelect:bool=false) -> (Output_Get:bool)
   > Set bAllowDrumsSelect(bAllowDrumsSelect:bool=false) -> (Output_Get:bool)
   n3 = Branch(execute=@n2.else, Condition=HasTag(TagContainer=@n1, Tag=(TagName="Sparks.Playlist.Pilgrim.BattleStage.ScoreChase.Drums"), bExactMatch=true))
   > Set bAllowBassSelect(bAllowBassSelect=false)
   > Set bAllowGuitarSelect(bAllowGuitarSelect=false)
   > Set bAllowVocalsSelect(bAllowVocalsSelect=false)
   > Set bAllowDrumsSelect(bAllowDrumsSelect=true)
   n4 = Branch(execute=@n3.else, Condition=HasTag(TagContainer=@n1, Tag=(TagName="Sparks.Playlist.Pilgrim.BattleStage.ScoreChase.Lead"), bExactMatch=true))
   > Set bAllowBassSelect(bAllowBassSelect=false)
   > Set bAllowGuitarSelect(bAllowGuitarSelect=true)
   > Set bAllowVocalsSelect(bAllowVocalsSelect=false)
   > Set bAllowDrumsSelect(bAllowDrumsSelect=false)
   Branch(
    execute=@n4.else
    Condition=HasTag(TagContainer=@n1, Tag=(TagName="Sparks.Playlist.Pilgrim.BattleStage.ScoreChase.Vocals"), bExactMatch=true)
   )
   > Set bAllowBassSelect(bAllowBassSelect=false)
   > Set bAllowGuitarSelect(bAllowGuitarSelect=false)
   > Set bAllowVocalsSelect(bAllowVocalsSelect=true)
   > Set bAllowDrumsSelect(bAllowDrumsSelect=false)
  ]}
 Swap to Valid Instrument {nodes [
   Swap to Valid Instrument() {is=title K2Node_FunctionEntry}
   > SwitchInstrumentOnPlatformChange!(
    execute
    self:BP Pilgrim Player Controller Component Object Reference=Get BP_PilgrimPlayerControllerComponent()
    IsTouchInputPlatform:bool=IsUsingTouchInput() {is=title K2Node_CallFunction}
   ) -> (then)
  ]}
]
variables [
 bAllowBassSelect {type=boolean, default=true}
 bAllowDrumsSelect {type=boolean, default=true}
 bAllowGuitarSelect {type=boolean, default=true}
 bAllowVocalsSelect {type=boolean, default=true}
 BP_PilgrimPlayerControllerComponent {type=BP_PilgrimPlayerControllerComponent_C, default=null}
 CurrentlySelected {type=boolean}
 DynamicTabIcon {type=MaterialInstanceDynamic}
 GameplayEventHandles {type="array<AsyncAction_StartListeningToEvent>", default []}
 IsOnGreenRoom {type=boolean, default=true}
 MyPlayerState {type=PlayerState, default=null}
 PilgrimPlayspace {type=PilgrimPlayspace, default=null}
 PlayerController {type=PilgrimPlayerControllerComponent, default=null}
 PreviouslySelected {type=boolean}
 SelectedButtonIndex {type=WBP_UIKit_Button_Selectable_C, default=null}
 TabArray {type="array<WBP_UIKit_Button_Selectable_C>", default []}
]
widgets {
 info {inheritedWidgetCount=0, parentClass=&UserWidget, rootWidgetClass=&GridPanel, widgetCount=76}
 widgets [
  GridPanel {widgetClassPath=&GridPanel, bIsVariable=true, widget=GridPanel, properties {columnFill [1], rowFill [1]}}
  SizeBox_1 {
   widgetClassPath=&SizeBox
   parent=GridPanel
   widget=SizeBox_1
   properties {heightOverride=160, widthOverride=302.5919}
   slot {class=GridSlot}
  }
  Overlay_0 {widgetClassPath=&Overlay, parent=SizeBox_1, widget=Overlay_0, slot {class=SizeBoxSlot}}
  SelectionBackingIndicator {
   widgetClassPath=&Image
   bIsVariable=true
   parent=Overlay_0
   widget=SelectionBackingIndicator
   properties {
    renderOpacity=0
    visibility=SelfHitTestInvisible
    brush {
     bIsDynamicallyLoaded=false
     drawAs=Image
     imageType=NoImage
     mirroring=NoMirror
     resourceName=None
     resourceObject=/SparksUICommon/UILibrary/MI_FMCore_ButtonBacking_InstrumentSelector
     tiling=NoTile
     imageSize=<32 32>
     margin=<ltrb 0 0 0 0>
     outlineSettings {
      bUseBrushTransparency=false
      roundingType=HalfHeightRadius
      width=0
      color {colorUseRule=UseColor_Specified, specifiedColor=<rgba 0 0 0 0>}
      cornerRadii=<0 0 0 0>
     }
     tintColor {colorUseRule=UseColor_Specified, specifiedColor=<rgba 1 1 1 1>}
     uVRegion {bIsValid=false, max=<0 0>, min=<0 0>}
    }
   }
   slot {class=OverlaySlot, properties {horizontalAlignment=HAlign_Fill, verticalAlignment=VAlign_Fill, padding=<ltrb 1 1 1 1>}}
  }
  WBP_UIKit_Block_Outline {
   widgetClassPath=/FortUILibrary/UIKit/Buttons/Blocks/WBP_UIKit_Block_Outline
   bIsVariable=true
   parent=Overlay_0
   widget=WBP_UIKit_Block_Outline
   properties {bottomLeftCornerRadius=0.2, bottomRightCornerRadius=0.2, topLeftCornerRadius=0.2, topRightCornerRadius=0.2}
   slot {class=OverlaySlot, properties {horizontalAlignment=HAlign_Fill, verticalAlignment=VAlign_Fill}}
  }
  WBP_Sparks_BackingPanel {
   widgetClassPath=/PilgrimCore/Widgets/SparksUILibrary/WBP_Sparks_BackingPanel_PlayerIdentity
   bIsVariable=true
   parent=Overlay_0
   widget=WBP_Sparks_BackingPanel
   slot {class=OverlaySlot, properties {horizontalAlignment=HAlign_Fill, verticalAlignment=VAlign_Fill}}
  }
  WBP_UIKit_Button_Regular {
   widgetClassPath=&WBP_UIKit_Button_Regular
   bIsVariable=true
   parent=Overlay_0
   widget=WBP_UIKit_Button_Regular
   properties {
    bDisplayInputActionWhenNotInteractable=false
    bHideInputActionWithKeyboard=true
    bUseNaturalNavigation=true
    buttonSize=Small
    buttonType=Quiet
    inputModeOverride=All
    minHeight=88
    renderOpacity=0
    text=""
   }
   slot {class=OverlaySlot, properties {horizontalAlignment=HAlign_Fill, verticalAlignment=VAlign_Fill}}
  }
  InputContainer_GridPanel {
   widgetClassPath=&GridPanel
   bIsVariable=true
   parent=Overlay_0
   widget=InputContainer_GridPanel
   properties {visibility=Collapsed}
   slot {class=OverlaySlot, properties {verticalAlignment=VAlign_Center, padding=<ltrb -14 0 0 0>}}
  }
  InputActionWidget {
   widgetClassPath=&CommonActionWidget
   bIsVariable=true
   parent=InputContainer_GridPanel
   widget=InputActionWidget
   properties {
    progressMaterialParam=Percentage
    visibility=HitTestInvisible
    inputActions [{dataTable=/FortUILibrary/Foundation/ActivityBrowserInputActionTable, rowName=Select}]
    progressMaterialBrush {
     bIsDynamicallyLoaded=false
     drawAs=Image
     imageType=NoImage
     mirroring=NoMirror
     resourceName=None
     resourceObject=/FortUILibrary/UIKit/Buttons/MaterialInstances/CTAButtons/MI_UIKit_EmptyProgress
     tiling=NoTile
     imageSize=<48 48>
     margin=<ltrb 0 0 0 0>
     outlineSettings {
      bUseBrushTransparency=false
      roundingType=HalfHeightRadius
      width=0
      color {colorUseRule=UseColor_Specified, specifiedColor=<rgba 0 0 0 0>}
      cornerRadii=<0 0 0 0>
     }
     tintColor {colorUseRule=UseColor_Specified, specifiedColor=<rgba 1 1 1 1>}
     uVRegion {bIsValid=false, max=<0 0>, min=<0 0>}
    }
   }
   slot {class=GridSlot, properties {padding=<ltrb -10 0 0 0>}}
  }
  VerticalBox {
   widgetClassPath=&VerticalBox
   parent=Overlay_0
   widget=VerticalBox
   slot {class=OverlaySlot, properties {horizontalAlignment=HAlign_Fill, verticalAlignment=VAlign_Center, padding=<ltrb 15 5 15 5>}}
  }
  CurrentInputBindingNumber {
   widgetClassPath=&CommonTextBlock
   bIsVariable=true
   parent=VerticalBox
   widget=CurrentInputBindingNumber
   properties {
    style=/FortUILibrary/Styles/Typography/_TextStyle_SubH4._TextStyle_SubH4_C
    text=INSTRUMENT
    visibility=Collapsed
    font {
     bForceMonospaced=false
     bMaterialIsStencil=false
     fontMaterial=null
     fontObject=&BurbankSmall
     letterSpacing=0
     monospacedWidth=1
     size=24
     skewAmount=0
     typefaceFontName=Black
     outlineSettings {
      bApplyOutlineToDropShadows=false
      bMiteredCorners=false
      bSeparateFillAlpha=false
      outlineMaterial=null
      outlineSize=0
      outlineColor=<rgba 0 0 0 1>
     }
    }
    shadowOffset=<0 0>
    strikeBrush {
     bIsDynamicallyLoaded=false
     drawAs=NoDrawType
     imageType=NoImage
     mirroring=NoMirror
     resourceName=None
     resourceObject=null
     tiling=NoTile
     imageSize=<32 32>
     margin=<ltrb 0 0 0 0>
     outlineSettings {
      bUseBrushTransparency=false
      roundingType=HalfHeightRadius
      width=0
      color {colorUseRule=UseColor_Specified, specifiedColor=<rgba 0 0 0 0>}
      cornerRadii=<0 0 0 0>
     }
     tintColor {colorUseRule=UseColor_Specified, specifiedColor=<rgba 1 1 1 1>}
     uVRegion {bIsValid=false, max=<0 0>, min=<0 0>}
    }
   }
   slot {class=VerticalBoxSlot, properties {horizontalAlignment=HAlign_Center}}
  }
  PartTitle_PartSelectionInputPrompt {widgetClassPath=&Overlay, parent=VerticalBox, widget=PartTitle_PartSelectionInputPrompt, slot {class=VerticalBoxSlot}}
  Keybind_GridPanel {
   widgetClassPath=&GridPanel
   bIsVariable=true
   parent=PartTitle_PartSelectionInputPrompt
   widget=Keybind_GridPanel
   properties {columnFill [1], rowFill [1]}
   slot {class=OverlaySlot, properties {horizontalAlignment=HAlign_Fill, verticalAlignment=VAlign_Fill}}
  }
  TriggersVisibility {
   widgetClassPath=&CommonVisibilityWidgetLegacy
   parent=Keybind_GridPanel
   widget=TriggersVisibility
   properties {bShowForMouseAndKeyboard=false, bShowForTouch=false, verticalAlignment=VAlign_Center, padding=<ltrb 0 0 0 0>}
   slot {class=GridSlot, properties {verticalAlignment=VAlign_Top, padding=<ltrb 24 4 24 0>}}
  }
  GridPanel_1 {
   widgetClassPath=&GridPanel
   parent=TriggersVisibility
   widget=GridPanel_1
   properties {columnFill [1], rowFill [1]}
   slot {class=BorderSlot, properties {verticalAlignment=VAlign_Center, padding=<ltrb 0 0 0 0>}}
  }
  KeybindWidget_LeftTrigger {
   widgetClassPath=/FortUILibrary/Input/KeybindWidget
   bIsVariable=true
   parent=GridPanel_1
   widget=KeybindWidget_LeftTrigger
   properties {
    bUseKeyboardTextureForKeys=true
    boundAction=Pilgrim_SwapInstrumentLeft
    boundInputAction=/PilgrimSettings/Player/Abilities/IA_PilgrimSwapInstrument
    boundKeyFallback=F2
    boundActionFallbacks [GamepadSwitchMode, GamepadSwitchModeOrEditBuilding, GamepadHoldToBuild]
    renderTransform {angle=0, scale=<0.8 0.8>, shear=<0 0>, translation=<0 0>}
   }
   slot {class=GridSlot, properties {horizontalAlignment=HAlign_Left, verticalAlignment=VAlign_Center}}
  }
  KeybindWidget_RightTrigger {
   widgetClassPath=&KeybindWidget
   bIsVariable=true
   parent=GridPanel_1
   widget=KeybindWidget_RightTrigger
   properties {
    bUseKeyboardTextureForKeys=true
    boundAction=Pilgrim_SwapInstrumentRight
    boundInputAction=/PilgrimSettings/Player/Abilities/IA_PilgrimSwapInstrumentRight
    boundKeyFallback=F2
    boundActionFallbacks [GamepadSwitchMode, GamepadSwitchModeOrEditBuilding, GamepadHoldToBuild]
    renderTransform {angle=0, scale=<0.8 0.8>, shear=<0 0>, translation=<0 0>}
   }
   slot {class=GridSlot, properties {horizontalAlignment=HAlign_Right, verticalAlignment=VAlign_Center}}
  }
  FortResponsivePanel_0 {
   widgetClassPath=/Script/FortniteUI.FortResponsivePanel
   parent=PartTitle_PartSelectionInputPrompt
   widget=FortResponsivePanel_0
   slot {class=OverlaySlot, properties {horizontalAlignment=HAlign_Center, verticalAlignment=VAlign_Center, padding=<ltrb 0 5 0 5>}}
  }
  Common_Text_Instrument {
   widgetClassPath=&CommonTextBlock
   bIsVariable=true
   parent=FortResponsivePanel_0
   widget=Common_Text_Instrument
   properties {
    justification=Center
    text=DRUMS
    textTransformPolicy=ToUpper
    visibility=HitTestInvisible
    colorAndOpacity {colorUseRule=UseColor_Specified, specifiedColor=<rgba 0.5 0.8333 1 1>}
    font {
     bForceMonospaced=false
     bMaterialIsStencil=false
     fontMaterial=null
     fontObject=&HeadingNow
     letterSpacing=0
     monospacedWidth=1
     size=13.5
     skewAmount=0
     typefaceFontName=76 Bold Italic
     outlineSettings {
      bApplyOutlineToDropShadows=false
      bMiteredCorners=false
      bSeparateFillAlpha=false
      outlineMaterial=null
      outlineSize=0
      outlineColor=<rgba 0 0 0 1>
     }
    }
    shadowOffset=<0 0>
   }
   slot {class=FortResponsivePanelSlot}
  }
  PartIcons {
   widgetClassPath=&ScaleBox
   parent=VerticalBox
   widget=PartIcons
   slot {class=VerticalBoxSlot, properties {verticalAlignment=VAlign_Center, padding=<ltrb 0 0 0 3>}}
  }
  Parts {
   widgetClassPath=&HorizontalBox
   parent=PartIcons
   widget=Parts
   slot {class=ScaleBoxSlot, properties {horizontalAlignment=HAlign_Fill, verticalAlignment=VAlign_Fill}}
  }
  DefaultParts+Prompts {
   widgetClassPath=&Overlay
   parent=Parts
   widget=DefaultParts+Prompts
   slot {class=HorizontalBoxSlot, properties {size {sizeRule=Fill, value=1}}}
  }
  Instruments_Border {
   widgetClassPath=&Border
   bIsVariable=true
   parent=DefaultParts+Prompts
   widget=Instruments_Border
   properties {
    visibility=SelfHitTestInvisible
    background {
     bIsDynamicallyLoaded=false
     drawAs=Image
     imageType=FullColor
     mirroring=NoMirror
     resourceName=None
     resourceObject=/PilgrimCore/Widgets/Materials/MI_Pilgrim_TabBacking_Panel
     tiling=NoTile
     imageSize=<32 32>
     margin=<ltrb 0 0 0 0>
     outlineSettings {
      bUseBrushTransparency=false
      roundingType=HalfHeightRadius
      width=0
      color {colorUseRule=UseColor_Specified, specifiedColor=<rgba 0 0 0 0>}
      cornerRadii=<0 0 0 0>
     }
     tintColor {colorUseRule=UseColor_Specified, specifiedColor=<rgba 1 1 1 1>}
     uVRegion {bIsValid=false, max=<0 0>, min=<0 0>}
    }
    padding=<ltrb 8 8 8 8>
   }
   slot {class=OverlaySlot, properties {horizontalAlignment=HAlign_Fill, verticalAlignment=VAlign_Fill}}
  }
  HorizontalBox_116 {
   widgetClassPath=&HorizontalBox
   parent=Instruments_Border
   widget=HorizontalBox_116
   slot {class=BorderSlot, properties {padding=<ltrb 8 8 8 8>}}
  }
  UIKit_Button_Vocal {
   widgetClassPath=/FortUILibrary/UIKit/Buttons/WBP_UIKit_Button_Selectable
   bIsVariable=true
   parent=HorizontalBox_116
   widget=UIKit_Button_Vocal
   properties {
    bIsFocusable=false
    backgroundClass=/FortUILibrary/Systems/ModularButton/Blocks/WBP_UIKit_Block_BackgroundGeneric.WBP_UIKit_Block_BackgroundGeneric_C
    backgroundMaterial=&MI_UI_Piligrim_IconTabBacking
    iconHorizontalAlignment=HAlign_Fill
    iconMaterial=/PilgrimQuickplay/Widgets/Materials/MI_UI_Icon_Instrument_Type
    iconVerticalAlignment=VAlign_Fill
    minHeight=0
    outlineMaterial=&MI_UIKit_EmptyProgress
    text=0
    useTextBlock=false
    visibility=HitTestInvisible
    buttonPadding=<ltrb 9 9 9 9>
    iconPadding=<ltrb 5 5 5 5>
    iconSize=<48 48>
   }
   slot {class=HorizontalBoxSlot, properties {size {sizeRule=Fill, value=1}}}
  }
  UIKit_Button_Bass {
   widgetClassPath=&WBP_UIKit_Button_Selectable
   bIsVariable=true
   parent=HorizontalBox_116
   widget=UIKit_Button_Bass
   properties {
    bIsFocusable=false
    backgroundClass=&WBP_UIKit_Block_BackgroundGeneric_C
    backgroundMaterial=&MI_UI_Piligrim_IconTabBacking
    iconHorizontalAlignment=HAlign_Fill
    iconMaterial=&MI_UI_Icon_Instrument_Type
    iconVerticalAlignment=VAlign_Fill
    minHeight=0
    outlineMaterial=&MI_UIKit_EmptyProgress
    text=0
    useTextBlock=false
    visibility=HitTestInvisible
    buttonPadding=<ltrb 9 9 9 9>
    iconPadding=<ltrb 5 5 5 5>
    iconSize=<48 48>
   }
   slot {class=HorizontalBoxSlot, properties {size {sizeRule=Fill, value=1}}}
  }
  UIKit_Button_Guitar {
   widgetClassPath=&WBP_UIKit_Button_Selectable
   bIsVariable=true
   parent=HorizontalBox_116
   widget=UIKit_Button_Guitar
   properties {
    bIsFocusable=false
    backgroundClass=&WBP_UIKit_Block_BackgroundGeneric_C
    backgroundMaterial=&MI_UI_Piligrim_IconTabBacking
    iconHorizontalAlignment=HAlign_Fill
    iconMaterial=&MI_UI_Icon_Instrument_Type_Tab
    iconVerticalAlignment=VAlign_Fill
    minHeight=0
    outlineMaterial=&MI_UIKit_EmptyProgress
    text=0
    useTextBlock=false
    visibility=HitTestInvisible
    buttonPadding=<ltrb 9 9 9 9>
    iconPadding=<ltrb 5 5 5 5>
    iconSize=<48 48>
   }
   slot {class=HorizontalBoxSlot, properties {size {sizeRule=Fill, value=1}}}
  }
  UIKit_Button_Drums {
   widgetClassPath=&WBP_UIKit_Button_Selectable
   bIsVariable=true
   parent=HorizontalBox_116
   widget=UIKit_Button_Drums
   properties {
    bIsFocusable=false
    backgroundClass=&WBP_UIKit_Block_BackgroundGeneric_C
    backgroundMaterial=&MI_UI_Piligrim_IconTabBacking
    iconHorizontalAlignment=HAlign_Fill
    iconMaterial=&MI_UI_Icon_Instrument_Type
    iconVerticalAlignment=VAlign_Fill
    minHeight=0
    outlineMaterial=&MI_UIKit_EmptyProgress
    text=0
    useTextBlock=false
    visibility=HitTestInvisible
    buttonPadding=<ltrb 9 9 9 9>
    iconPadding=<ltrb 5 5 5 5>
    iconSize=<48 48>
   }
   slot {class=HorizontalBoxSlot, properties {size {sizeRule=Fill, value=1}}}
  }
  PeripheralParts+Prompts {
   widgetClassPath=&Overlay
   bIsVariable=true
   parent=Parts
   widget=PeripheralParts+Prompts
   properties {visibility=Collapsed}
   slot {class=HorizontalBoxSlot, properties {size {sizeRule=Automatic, value=0.5}}}
  }
  Peripheral_Border {
   widgetClassPath=&Border
   bIsVariable=true
   parent=PeripheralParts+Prompts
   widget=Peripheral_Border
   properties {
    visibility=SelfHitTestInvisible
    background {
     bIsDynamicallyLoaded=false
     drawAs=Image
     imageType=FullColor
     mirroring=NoMirror
     resourceName=None
     resourceObject=&MI_Pilgrim_TabBacking_Panel
     tiling=NoTile
     imageSize=<32 32>
     margin=<ltrb 0 0 0 0>
     outlineSettings {
      bUseBrushTransparency=false
      roundingType=HalfHeightRadius
      width=0
      color {colorUseRule=UseColor_Specified, specifiedColor=<rgba 0 0 0 0>}
      cornerRadii=<0 0 0 0>
     }
     tintColor {colorUseRule=UseColor_Specified, specifiedColor=<rgba 1 1 1 1>}
     uVRegion {bIsValid=false, max=<0 0>, min=<0 0>}
    }
    padding=<ltrb 8 8 8 8>
   }
   slot {class=OverlaySlot, properties {horizontalAlignment=HAlign_Fill, verticalAlignment=VAlign_Fill}}
  }
  HorizontalBox {
   widgetClassPath=&HorizontalBox
   parent=Peripheral_Border
   widget=HorizontalBox
   slot {class=BorderSlot, properties {padding=<ltrb 8 8 8 8>}}
  }
  UIKit_Button_Guitar_Peripheral {
   widgetClassPath=&WBP_UIKit_Button_Selectable
   bIsVariable=true
   parent=HorizontalBox
   widget=UIKit_Button_Guitar_Peripheral
   properties {
    bIsFocusable=false
    backgroundClass=&WBP_UIKit_Block_BackgroundGeneric_C
    backgroundMaterial=&MI_UI_Piligrim_IconTabBacking
    iconHorizontalAlignment=HAlign_Fill
    iconMaterial=&MI_UI_Icon_Instrument_Type
    iconVerticalAlignment=VAlign_Fill
    minHeight=0
    outlineMaterial=&MI_UIKit_EmptyProgress
    text=0
    useTextBlock=false
    visibility=HitTestInvisible
    buttonPadding=<ltrb 9 9 9 9>
    iconPadding=<ltrb 5 5 5 5>
    iconSize=<48 48>
   }
   slot {class=HorizontalBoxSlot, properties {size {sizeRule=Fill, value=1}}}
  }
  UIKit_Button_Bass_Peripheral {
   widgetClassPath=&WBP_UIKit_Button_Selectable
   bIsVariable=true
   parent=HorizontalBox
   widget=UIKit_Button_Bass_Peripheral
   properties {
    bIsFocusable=false
    backgroundClass=&WBP_UIKit_Block_BackgroundGeneric_C
    backgroundMaterial=&MI_UI_Piligrim_IconTabBacking
    iconHorizontalAlignment=HAlign_Fill
    iconMaterial=&MI_UI_Icon_Instrument_Type
    iconVerticalAlignment=VAlign_Fill
    minHeight=0
    outlineMaterial=&MI_UIKit_EmptyProgress
    text=0
    useTextBlock=false
    visibility=HitTestInvisible
    buttonPadding=<ltrb 9 9 9 9>
    iconPadding=<ltrb 5 5 5 5>
    iconSize=<48 48>
   }
   slot {class=HorizontalBoxSlot, properties {size {sizeRule=Fill, value=1}}}
  }
  UIKit_Button_Drums_Peripheral {
   widgetClassPath=&WBP_UIKit_Button_Selectable
   bIsVariable=true
   parent=HorizontalBox
   widget=UIKit_Button_Drums_Peripheral
   properties {
    bIsFocusable=false
    backgroundClass=&WBP_UIKit_Block_BackgroundGeneric_C
    backgroundMaterial=&MI_UI_Piligrim_IconTabBacking
    iconHorizontalAlignment=HAlign_Left
    iconMaterial=&MI_UI_Icon_Instrument_Type
    minHeight=0
    outlineMaterial=&MI_UIKit_EmptyProgress
    text=0
    useTextBlock=false
    visibility=Collapsed
    buttonPadding=<ltrb 5 5 5 5>
    iconPadding=<ltrb 8 0 8 0>
    iconSize=<48 48>
   }
   slot {class=HorizontalBoxSlot}
  }
  UIKit_Button_Vocals_Peripheral {
   widgetClassPath=&WBP_UIKit_Button_Selectable
   bIsVariable=true
   parent=HorizontalBox
   widget=UIKit_Button_Vocals_Peripheral
   properties {
    bIsFocusable=false
    backgroundClass=&WBP_UIKit_Block_BackgroundGeneric_C
    backgroundMaterial=&MI_UI_Piligrim_IconTabBacking
    iconHorizontalAlignment=HAlign_Fill
    iconMaterial=&MI_UI_Icon_Instrument_Type
    iconVerticalAlignment=VAlign_Fill
    minHeight=0
    outlineMaterial=&MI_UIKit_EmptyProgress
    text=0
    useTextBlock=false
    visibility=Collapsed
    buttonPadding=<ltrb 9 9 9 9>
    iconPadding=<ltrb 5 5 5 5>
    iconSize=<48 48>
   }
   slot {class=HorizontalBoxSlot, properties {size {sizeRule=Fill, value=1}}}
  }
  PartSelection_InputPrompts {
   widgetClassPath=&CommonVisibilityWidgetLegacy
   bIsVariable=true
   parent=VerticalBox
   widget=PartSelection_InputPrompts
   properties {bShowForGamepad=false, bShowForTouch=false, horizontalAlignment=HAlign_Center, verticalAlignment=VAlign_Top, padding=<ltrb 0 0 0 0>}
   slot {class=VerticalBoxSlot, properties {verticalAlignment=VAlign_Bottom}}
  }
  MainInputsContainer {
   widgetClassPath=&HorizontalBox
   bIsVariable=true
   parent=PartSelection_InputPrompts
   widget=MainInputsContainer
   slot {class=BorderSlot, properties {horizontalAlignment=HAlign_Center, verticalAlignment=VAlign_Top, padding=<ltrb 0 0 0 0>}}
  }
  Overlay_5 {
   widgetClassPath=&Overlay
   parent=MainInputsContainer
   widget=Overlay_5
   slot {class=HorizontalBoxSlot, properties {horizontalAlignment=HAlign_Left, verticalAlignment=VAlign_Center, size {sizeRule=Fill, value=1}}}
  }
  Key1Backing {
   widgetClassPath=&Image
   parent=Overlay_5
   widget=Key1Backing
   properties {
    renderOpacity=0.5
    visibility=SelfHitTestInvisible
    brush {
     bIsDynamicallyLoaded=false
     drawAs=Image
     imageType=NoImage
     mirroring=NoMirror
     resourceName=None
     resourceObject=/PilgrimCore/Widgets/Materials/M_UI_Pilgrim_HUD_SimpleBacking
     tiling=NoTile
     imageSize=<32 32>
     margin=<ltrb 0 0 0 0>
     outlineSettings {
      bUseBrushTransparency=false
      roundingType=HalfHeightRadius
      width=0
      color {colorUseRule=UseColor_Specified, specifiedColor=<rgba 0 0 0 0>}
      cornerRadii=<0 0 0 0>
     }
     tintColor {colorUseRule=UseColor_Specified, specifiedColor=<rgba 1 1 1 1>}
     uVRegion {bIsValid=false, max=<0 0>, min=<0 0>}
    }
   }
   slot {class=OverlaySlot, properties {horizontalAlignment=HAlign_Fill, verticalAlignment=VAlign_Fill}}
  }
  ScaleBox_0 {
   widgetClassPath=&ScaleBox
   parent=Overlay_5
   widget=ScaleBox_0
   properties {stretch=UserSpecified, userSpecifiedScale=0.75}
   slot {class=OverlaySlot, properties {horizontalAlignment=HAlign_Center, verticalAlignment=VAlign_Center}}
  }
  KeybindWidget_ChangeBindings_n1 {
   widgetClassPath=&KeybindWidget
   bIsVariable=true
   parent=ScaleBox_0
   widget=KeybindWidget_ChangeBindings_n1
   properties {
    bUseKeyboardTextureForKeys=true
    boundInputAction=/PilgrimSettings/Player/Abilities/IA_PilgrimSwapInstrumentNumber1
    boundActionFallbacks [GamepadSwitchMode, GamepadSwitchModeOrEditBuilding, GamepadHoldToBuild]
   }
   slot {class=ScaleBoxSlot}
  }
  Spacer_186 {
   widgetClassPath=&Spacer
   parent=MainInputsContainer
   widget=Spacer_186
   properties {size=<42 21>}
   slot {class=HorizontalBoxSlot, properties {verticalAlignment=VAlign_Center}}
  }
  Overlay_6 {
   widgetClassPath=&Overlay
   parent=MainInputsContainer
   widget=Overlay_6
   slot {class=HorizontalBoxSlot, properties {horizontalAlignment=HAlign_Left, verticalAlignment=VAlign_Center, size {sizeRule=Fill, value=1}}}
  }
  Key2Backing {
   widgetClassPath=&Image
   parent=Overlay_6
   widget=Key2Backing
   properties {
    renderOpacity=0.5
    visibility=SelfHitTestInvisible
    brush {
     bIsDynamicallyLoaded=false
     drawAs=Image
     imageType=NoImage
     mirroring=NoMirror
     resourceName=None
     resourceObject=&M_UI_Pilgrim_HUD_SimpleBacking
     tiling=NoTile
     imageSize=<32 32>
     margin=<ltrb 0 0 0 0>
     outlineSettings {
      bUseBrushTransparency=false
      roundingType=HalfHeightRadius
      width=0
      color {colorUseRule=UseColor_Specified, specifiedColor=<rgba 0 0 0 0>}
      cornerRadii=<0 0 0 0>
     }
     tintColor {colorUseRule=UseColor_Specified, specifiedColor=<rgba 1 1 1 1>}
     uVRegion {bIsValid=false, max=<0 0>, min=<0 0>}
    }
   }
   slot {class=OverlaySlot, properties {horizontalAlignment=HAlign_Fill, verticalAlignment=VAlign_Fill}}
  }
  ScaleBox_1 {
   widgetClassPath=&ScaleBox
   parent=Overlay_6
   widget=ScaleBox_1
   properties {stretch=UserSpecified, userSpecifiedScale=0.75}
   slot {class=OverlaySlot, properties {horizontalAlignment=HAlign_Center, verticalAlignment=VAlign_Center}}
  }
  KeybindWidget_ChangeBindings_n2 {
   widgetClassPath=&KeybindWidget
   bIsVariable=true
   parent=ScaleBox_1
   widget=KeybindWidget_ChangeBindings_n2
   properties {
    bUseKeyboardTextureForKeys=true
    boundInputAction=/PilgrimSettings/Player/Abilities/IA_PilgrimSwapInstrumentNumber2
    boundActionFallbacks [GamepadSwitchMode, GamepadSwitchModeOrEditBuilding, GamepadHoldToBuild]
   }
   slot {class=ScaleBoxSlot}
  }
  Spacer {widgetClassPath=&Spacer, parent=MainInputsContainer, widget=Spacer, properties {size=<42 1>}, slot {class=HorizontalBoxSlot}}
  Overlay_7 {
   widgetClassPath=&Overlay
   parent=MainInputsContainer
   widget=Overlay_7
   slot {class=HorizontalBoxSlot, properties {horizontalAlignment=HAlign_Center, verticalAlignment=VAlign_Center, size {sizeRule=Fill, value=1}}}
  }
  Key3Backing {
   widgetClassPath=&Image
   parent=Overlay_7
   widget=Key3Backing
   properties {
    renderOpacity=0.5
    visibility=SelfHitTestInvisible
    brush {
     bIsDynamicallyLoaded=false
     drawAs=Image
     imageType=NoImage
     mirroring=NoMirror
     resourceName=None
     resourceObject=&M_UI_Pilgrim_HUD_SimpleBacking
     tiling=NoTile
     imageSize=<32 32>
     margin=<ltrb 0 0 0 0>
     outlineSettings {
      bUseBrushTransparency=false
      roundingType=HalfHeightRadius
      width=0
      color {colorUseRule=UseColor_Specified, specifiedColor=<rgba 0 0 0 0>}
      cornerRadii=<0 0 0 0>
     }
     tintColor {colorUseRule=UseColor_Specified, specifiedColor=<rgba 1 1 1 1>}
     uVRegion {bIsValid=false, max=<0 0>, min=<0 0>}
    }
   }
   slot {class=OverlaySlot, properties {horizontalAlignment=HAlign_Fill, verticalAlignment=VAlign_Fill}}
  }
  ScaleBox_2 {
   widgetClassPath=&ScaleBox
   parent=Overlay_7
   widget=ScaleBox_2
   properties {stretch=UserSpecified, userSpecifiedScale=0.75}
   slot {class=OverlaySlot, properties {horizontalAlignment=HAlign_Center, verticalAlignment=VAlign_Center}}
  }
  KeybindWidget_ChangeBindings_n3 {
   widgetClassPath=&KeybindWidget
   bIsVariable=true
   parent=ScaleBox_2
   widget=KeybindWidget_ChangeBindings_n3
   properties {
    bUseKeyboardTextureForKeys=true
    boundInputAction=/PilgrimSettings/Player/Abilities/IA_PilgrimSwapInstrumentNumber3
    boundActionFallbacks [GamepadSwitchMode, GamepadSwitchModeOrEditBuilding, GamepadHoldToBuild]
   }
   slot {class=ScaleBoxSlot}
  }
  Spacer_1 {widgetClassPath=&Spacer, parent=MainInputsContainer, widget=Spacer_1, properties {size=<42 1>}, slot {class=HorizontalBoxSlot}}
  Overlay_8 {
   widgetClassPath=&Overlay
   parent=MainInputsContainer
   widget=Overlay_8
   slot {class=HorizontalBoxSlot, properties {horizontalAlignment=HAlign_Left, verticalAlignment=VAlign_Center, size {sizeRule=Fill, value=1}}}
  }
  Key4Backing {
   widgetClassPath=&Image
   parent=Overlay_8
   widget=Key4Backing
   properties {
    renderOpacity=0.5
    visibility=SelfHitTestInvisible
    brush {
     bIsDynamicallyLoaded=false
     drawAs=Image
     imageType=NoImage
     mirroring=NoMirror
     resourceName=None
     resourceObject=&M_UI_Pilgrim_HUD_SimpleBacking
     tiling=NoTile
     imageSize=<32 32>
     margin=<ltrb 0 0 0 0>
     outlineSettings {
      bUseBrushTransparency=false
      roundingType=HalfHeightRadius
      width=0
      color {colorUseRule=UseColor_Specified, specifiedColor=<rgba 0 0 0 0>}
      cornerRadii=<0 0 0 0>
     }
     tintColor {colorUseRule=UseColor_Specified, specifiedColor=<rgba 1 1 1 1>}
     uVRegion {bIsValid=false, max=<0 0>, min=<0 0>}
    }
   }
   slot {class=OverlaySlot, properties {horizontalAlignment=HAlign_Fill, verticalAlignment=VAlign_Fill}}
  }
  ScaleBox_3 {
   widgetClassPath=&ScaleBox
   parent=Overlay_8
   widget=ScaleBox_3
   properties {stretch=UserSpecified, userSpecifiedScale=0.75}
   slot {class=OverlaySlot, properties {horizontalAlignment=HAlign_Center, verticalAlignment=VAlign_Center}}
  }
  KeybindWidget_ChangeBindings_n4 {
   widgetClassPath=&KeybindWidget
   bIsVariable=true
   parent=ScaleBox_3
   widget=KeybindWidget_ChangeBindings_n4
   properties {
    bUseKeyboardTextureForKeys=true
    boundInputAction=/PilgrimSettings/Player/Abilities/IA_PilgrimSwapInstrumentNumber4
    boundActionFallbacks [GamepadSwitchMode, GamepadSwitchModeOrEditBuilding, GamepadHoldToBuild]
   }
   slot {class=ScaleBoxSlot}
  }
  Spacer_2 {
   widgetClassPath=&Spacer
   bIsVariable=true
   parent=MainInputsContainer
   widget=Spacer_2
   properties {visibility=Collapsed, size=<62.1507 1>}
   slot {class=HorizontalBoxSlot}
  }
  Overlay_11 {
   widgetClassPath=&Overlay
   bIsVariable=true
   parent=MainInputsContainer
   widget=Overlay_11
   properties {visibility=Collapsed}
   slot {class=HorizontalBoxSlot, properties {horizontalAlignment=HAlign_Left, verticalAlignment=VAlign_Center}}
  }
  Key5Backing_1 {
   widgetClassPath=&Image
   parent=Overlay_11
   widget=Key5Backing_1
   properties {
    renderOpacity=0.5
    visibility=SelfHitTestInvisible
    brush {
     bIsDynamicallyLoaded=false
     drawAs=Image
     imageType=NoImage
     mirroring=NoMirror
     resourceName=None
     resourceObject=&M_UI_Pilgrim_HUD_SimpleBacking
     tiling=NoTile
     imageSize=<32 32>
     margin=<ltrb 0 0 0 0>
     outlineSettings {
      bUseBrushTransparency=false
      roundingType=HalfHeightRadius
      width=0
      color {colorUseRule=UseColor_Specified, specifiedColor=<rgba 0 0 0 0>}
      cornerRadii=<0 0 0 0>
     }
     tintColor {colorUseRule=UseColor_Specified, specifiedColor=<rgba 1 1 1 1>}
     uVRegion {bIsValid=false, max=<0 0>, min=<0 0>}
    }
   }
   slot {class=OverlaySlot, properties {horizontalAlignment=HAlign_Fill, verticalAlignment=VAlign_Fill}}
  }
  ScaleBox_11 {
   widgetClassPath=&ScaleBox
   parent=Overlay_11
   widget=ScaleBox_11
   properties {stretch=UserSpecified, userSpecifiedScale=0.75}
   slot {class=OverlaySlot, properties {horizontalAlignment=HAlign_Center, verticalAlignment=VAlign_Center}}
  }
  KeybindWidget_ChangeBindings_n5 {
   widgetClassPath=&KeybindWidget
   bIsVariable=true
   parent=ScaleBox_11
   widget=KeybindWidget_ChangeBindings_n5
   properties {
    bUseKeyboardTextureForKeys=true
    boundInputAction=/PilgrimSettings/Player/Abilities/IA_PilgrimSwapInstrumentNumber5
    boundActionFallbacks [GamepadSwitchMode, GamepadSwitchModeOrEditBuilding, GamepadHoldToBuild]
   }
   slot {class=ScaleBoxSlot}
  }
  Spacer_8 {
   widgetClassPath=&Spacer
   bIsVariable=true
   parent=MainInputsContainer
   widget=Spacer_8
   properties {visibility=Collapsed, size=<39 1>}
   slot {class=HorizontalBoxSlot}
  }
  Overlay_12 {
   widgetClassPath=&Overlay
   bIsVariable=true
   parent=MainInputsContainer
   widget=Overlay_12
   properties {visibility=Collapsed}
   slot {class=HorizontalBoxSlot, properties {horizontalAlignment=HAlign_Left, verticalAlignment=VAlign_Center}}
  }
  Key6Backing_1 {
   widgetClassPath=&Image
   parent=Overlay_12
   widget=Key6Backing_1
   properties {
    renderOpacity=0.5
    visibility=SelfHitTestInvisible
    brush {
     bIsDynamicallyLoaded=false
     drawAs=Image
     imageType=NoImage
     mirroring=NoMirror
     resourceName=None
     resourceObject=&M_UI_Pilgrim_HUD_SimpleBacking
     tiling=NoTile
     imageSize=<32 32>
     margin=<ltrb 0 0 0 0>
     outlineSettings {
      bUseBrushTransparency=false
      roundingType=HalfHeightRadius
      width=0
      color {colorUseRule=UseColor_Specified, specifiedColor=<rgba 0 0 0 0>}
      cornerRadii=<0 0 0 0>
     }
     tintColor {colorUseRule=UseColor_Specified, specifiedColor=<rgba 1 1 1 1>}
     uVRegion {bIsValid=false, max=<0 0>, min=<0 0>}
    }
   }
   slot {class=OverlaySlot, properties {horizontalAlignment=HAlign_Fill, verticalAlignment=VAlign_Fill}}
  }
  ScaleBox_12 {
   widgetClassPath=&ScaleBox
   parent=Overlay_12
   widget=ScaleBox_12
   properties {stretch=UserSpecified, userSpecifiedScale=0.75}
   slot {class=OverlaySlot, properties {horizontalAlignment=HAlign_Center, verticalAlignment=VAlign_Center}}
  }
  KeybindWidget_ChangeBindings_n6 {
   widgetClassPath=&KeybindWidget
   bIsVariable=true
   parent=ScaleBox_12
   widget=KeybindWidget_ChangeBindings_n6
   properties {
    bUseKeyboardTextureForKeys=true
    boundInputAction=/PilgrimSettings/Player/Abilities/IA_PilgrimSwapInstrumentNumber6
    boundActionFallbacks [GamepadSwitchMode, GamepadSwitchModeOrEditBuilding, GamepadHoldToBuild]
   }
   slot {class=ScaleBoxSlot}
  }
  Spacer_3 {
   widgetClassPath=&Spacer
   bIsVariable=true
   parent=MainInputsContainer
   widget=Spacer_3
   properties {visibility=Collapsed, size=<39 1>}
   slot {class=HorizontalBoxSlot}
  }
  Overlay_13 {
   widgetClassPath=&Overlay
   bIsVariable=true
   parent=MainInputsContainer
   widget=Overlay_13
   properties {visibility=Collapsed}
   slot {class=HorizontalBoxSlot, properties {horizontalAlignment=HAlign_Left, verticalAlignment=VAlign_Center}}
  }
  Key6Backing {
   widgetClassPath=&Image
   parent=Overlay_13
   widget=Key6Backing
   properties {
    renderOpacity=0.5
    visibility=SelfHitTestInvisible
    brush {
     bIsDynamicallyLoaded=false
     drawAs=Image
     imageType=NoImage
     mirroring=NoMirror
     resourceName=None
     resourceObject=&M_UI_Pilgrim_HUD_SimpleBacking
     tiling=NoTile
     imageSize=<32 32>
     margin=<ltrb 0 0 0 0>
     outlineSettings {
      bUseBrushTransparency=false
      roundingType=HalfHeightRadius
      width=0
      color {colorUseRule=UseColor_Specified, specifiedColor=<rgba 0 0 0 0>}
      cornerRadii=<0 0 0 0>
     }
     tintColor {colorUseRule=UseColor_Specified, specifiedColor=<rgba 1 1 1 1>}
     uVRegion {bIsValid=false, max=<0 0>, min=<0 0>}
    }
   }
   slot {class=OverlaySlot, properties {horizontalAlignment=HAlign_Fill, verticalAlignment=VAlign_Fill}}
  }
  ScaleBox {
   widgetClassPath=&ScaleBox
   parent=Overlay_13
   widget=ScaleBox
   properties {stretch=UserSpecified, userSpecifiedScale=0.75}
   slot {class=OverlaySlot, properties {horizontalAlignment=HAlign_Center, verticalAlignment=VAlign_Center}}
  }
  KeybindWidget_ChangeBindings_n7 {
   widgetClassPath=&KeybindWidget
   bIsVariable=true
   parent=ScaleBox
   widget=KeybindWidget_ChangeBindings_n7
   properties {
    bUseKeyboardTextureForKeys=true
    boundInputAction=/PilgrimSettings/Player/Abilities/IA_PilgrimSwapInstrumentNumber7
    boundActionFallbacks [GamepadSwitchMode, GamepadSwitchModeOrEditBuilding, GamepadHoldToBuild]
   }
   slot {class=ScaleBoxSlot}
  }
  Spacer_4 {
   widgetClassPath=&Spacer
   bIsVariable=true
   parent=MainInputsContainer
   widget=Spacer_4
   properties {visibility=Collapsed, size=<39 1>}
   slot {class=HorizontalBoxSlot}
  }
  Overlay_14 {
   widgetClassPath=&Overlay
   bIsVariable=true
   parent=MainInputsContainer
   widget=Overlay_14
   properties {visibility=Collapsed}
   slot {class=HorizontalBoxSlot, properties {horizontalAlignment=HAlign_Left, verticalAlignment=VAlign_Center}}
  }
  Key8Backing {
   widgetClassPath=&Image
   parent=Overlay_14
   widget=Key8Backing
   properties {
    renderOpacity=0.5
    visibility=SelfHitTestInvisible
    brush {
     bIsDynamicallyLoaded=false
     drawAs=Image
     imageType=NoImage
     mirroring=NoMirror
     resourceName=None
     resourceObject=&M_UI_Pilgrim_HUD_SimpleBacking
     tiling=NoTile
     imageSize=<32 32>
     margin=<ltrb 0 0 0 0>
     outlineSettings {
      bUseBrushTransparency=false
      roundingType=HalfHeightRadius
      width=0
      color {colorUseRule=UseColor_Specified, specifiedColor=<rgba 0 0 0 0>}
      cornerRadii=<0 0 0 0>
     }
     tintColor {colorUseRule=UseColor_Specified, specifiedColor=<rgba 1 1 1 1>}
     uVRegion {bIsValid=false, max=<0 0>, min=<0 0>}
    }
   }
   slot {class=OverlaySlot, properties {horizontalAlignment=HAlign_Fill, verticalAlignment=VAlign_Fill}}
  }
  ScaleBox_4 {
   widgetClassPath=&ScaleBox
   parent=Overlay_14
   widget=ScaleBox_4
   properties {stretch=UserSpecified, userSpecifiedScale=0.75}
   slot {class=OverlaySlot, properties {horizontalAlignment=HAlign_Center, verticalAlignment=VAlign_Center}}
  }
  KeybindWidget_ChangeBindings_n8 {
   widgetClassPath=&KeybindWidget
   bIsVariable=true
   parent=ScaleBox_4
   widget=KeybindWidget_ChangeBindings_n8
   properties {
    bUseKeyboardTextureForKeys=true
    boundInputAction=/PilgrimSettings/Player/Abilities/IA_PilgrimSwapInstrumentNumber8
    boundActionFallbacks [GamepadSwitchMode, GamepadSwitchModeOrEditBuilding, GamepadHoldToBuild]
   }
   slot {class=ScaleBoxSlot}
  }
 ]
}
