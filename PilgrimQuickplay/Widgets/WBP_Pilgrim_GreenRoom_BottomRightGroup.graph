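# Bottom-right green room widget group for Pilgrim Quickplay: next-song panel, instrument and difficulty selection, ready-up, notices and player nameplate.
# Button presses forward to the player controller's Pilgrim component.
# When enabled, the Unified Part Picker, created directly or slotted via a tagged widget host, replaces instrument and difficulty controls.
# Emote tags toggle the stop-emote label, timer events toggle group visibility, and controller disconnects fade controls into a prompt.
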WidgetBlueprint /PilgrimQuickplay/Widgets/WBP_Pilgrim_GreenRoom_BottomRightGroup
animations [
 A_Show {
  binding_count=1
  length_seconds=0.3
  object=&self:A_Show
  sequence {bindings [Overlay_0 {
     id=#1
     possessed_class=Overlay
     tracks [
      {
       type=MovieSceneFloatTrack
       display_name=Render Opacity
       sections [{type=MovieSceneFloatSection, end_frame=6, start_frame=0, channels [None {type=MovieSceneScriptingFloatChannel, keys [[0,0],[6,1]]}]}]
      }
      {
       type=MovieScene2DTransformTrack
       sections [{
         type=MovieScene2DTransformSection
         end_frame=6
         start_frame=0
         channels [
          Translation.X {type=MovieSceneScriptingFloatChannel, const [0,2]}
          Translation.Y {type=MovieSceneScriptingFloatChannel, keys [[0,8],[6,0]]}
          Angle {type=MovieSceneScriptingFloatChannel, const [0,1]}
          Scale.X {type=MovieSceneScriptingFloatChannel, const [1,1]}
          Scale.Y {type=MovieSceneScriptingFloatChannel, const [1,1]}
          Shear.X {type=MovieSceneScriptingFloatChannel, const [0,1]}
          Shear.Y {type=MovieSceneScriptingFloatChannel, const [0,1]}
         ]
        }]
      }
     ]
    }]}
 }
 ControllerDisconnect {
  binding_count=9
  length_seconds=1
  object=&self:ControllerDisconnect
  sequence {bindings [
    Overlay_UpcomingSongWidget {
     id=#2
     possessed_class=Overlay
     tracks [{
       type=MovieSceneFloatTrack
       display_name=Render Opacity
       sections [{type=MovieSceneFloatSection, end_frame=15, start_frame=0, channels [None {type=MovieSceneScriptingFloatChannel, keys [[0,1],[15,0]]}]}]
      }]
    }
    Button_ChangeDifficulty {
     id=#3
     possessed_class=Button
     tracks [{
       type=MovieSceneFloatTrack
       display_name=Render Opacity
       sections [{type=MovieSceneFloatSection, end_frame=15, start_frame=0, channels [None {type=MovieSceneScriptingFloatChannel, keys [[0,1],[15,0]]}]}]
      }]
    }
    GreenRoom_InstrumentSelectUI {
     id=#4
     possessed_class=WBP_Pilgrim_GreenRoom_InstrumentSelectUI_C
     tracks [{
       type=MovieSceneFloatTrack
       display_name=Render Opacity
       sections [{type=MovieSceneFloatSection, end_frame=15, start_frame=0, channels [None {type=MovieSceneScriptingFloatChannel, keys [[0,1],[15,0]]}]}]
      }]
    }
    ControllerDisconnectPrompt {
     id=#5
     possessed_class=WBP_PilgrimQuickplay_ControllerDisconnectPrompt_C
     tracks [
      {
       type=MovieSceneFloatTrack
       display_name=Render Opacity
       sections [{type=MovieSceneFloatSection, end_frame=20, start_frame=0, channels [None {type=MovieSceneScriptingFloatChannel, keys [[0,0],[15,0],[20,1]]}]}]
      }
      {
       type=MovieSceneByteTrack
       display_name=Visibility
       sections [{type=MovieSceneByteSection, end_frame=20, start_frame=0, channels [None {type=MovieSceneScriptingByteChannel, keys [[0,1],[15,4],[20,4]]}]}]
      }
     ]
    }
    WBP_Pilgrim_ChoppedCheeseEnabledNotice {
     id=#6
     possessed_class=WBP_Pilgrim_ChoppedCheeseEnabledNotice_C
     tracks [{
       type=MovieSceneFloatTrack
       display_name=Render Opacity
       sections [{type=MovieSceneFloatSection, end_frame=15, start_frame=0, channels [None {type=MovieSceneScriptingFloatChannel, keys [[0,1],[15,0]]}]}]
      }]
    }
    StopEmoteLabel {
     id=#7
     possessed_class=WBP_PilgrimQuickplay_StopEmoteLabel_C
     tracks [{
       type=MovieSceneFloatTrack
       display_name=Render Opacity
       sections [{type=MovieSceneFloatSection, end_frame=15, start_frame=0, channels [None {type=MovieSceneScriptingFloatChannel, keys [[0,1],[15,0]]}]}]
      }]
    }
    Button_ToggleReadyUp {
     id=#8
     possessed_class=Button
     tracks [{
       type=MovieSceneFloatTrack
       display_name=Render Opacity
       sections [{type=MovieSceneFloatSection, end_frame=15, start_frame=0, channels [None {type=MovieSceneScriptingFloatChannel, keys [[0,1],[15,0]]}]}]
      }]
    }
    GreenRoom_InstrumentNotice {
     id=#9
     possessed_class=WBP_Pilgrim_GreenRoom_InstrumentNotice_C
     tracks [{
       type=MovieSceneFloatTrack
       display_name=Render Opacity
       sections [{type=MovieSceneFloatSection, end_frame=15, start_frame=0, channels [None {type=MovieSceneScriptingFloatChannel, keys [[0,1],[15,0]]}]}]
      }]
    }
    WBP_PilgrimQuickplay_OpenSettings {
     id=#10
     possessed_class=WBP_PilgrimQuickplay_OpenSettings_C
     tracks [{
       type=MovieSceneFloatTrack
       display_name=Render Opacity
       sections [{type=MovieSceneFloatSection, end_frame=15, start_frame=0, channels [None {type=MovieSceneScriptingFloatChannel, keys [[0,1],[15,0]]}]}]
      }]
    }
   ]}
 }
]
graphs [
 CheckAndAddUPPWidget {nodes [
   CheckAndAddUPPWidget() {is=title K2Node_FunctionEntry}
   > Branch(Condition=IsUnifiedPartPickerEnabled!() -> (ReturnValue:bool))
   > n1 = Branch(Condition=Get bIsUPPManualCreated() -> (bIsUPPManualCreated:bool))
   > n2 = Create Widget!(
    execute:Exec
    Class:User Widget Class Reference=/PilgrimQuickplay/Widgets/UniversalPartPicker/WBP_PilgrimQuickplay_PartPicker_MainControls.WBP_PilgrimQuickplay_PartPicker_MainControls_C
    OwningPlayer:Player Controller Object Reference=GetOwningPlayer()
   ) -> (then:Exec, ReturnValue:WBP Pilgrim Quickplay Part Picker Main Controls Object Reference)
   > AddChildToOverlay!(execute, self:Overlay Object Reference=Get UPP_OVerlay() -> (UPP_OVerlay:Overlay Object Reference), Content:Widget Object Reference=@n2) -> (then, ReturnValue:Overlay Slot Object Reference)
   > RemoveFromParent(self=Get Button_ChangeDifficulty() -> (Button_ChangeDifficulty:Button Object Reference))
   > RemoveFromParent(self=Get GreenRoom_InstrumentSelectUI() -> (GreenRoom_InstrumentSelectUI:WBP Pilgrim Green Room Instrument Select UI Object Reference))
   RegisterWidgetHost!(
    execute=@n1.else
    WidgetHost:Widget Object Reference
    WidgetHostTag:"Gameplay Tag Structure (by ref)"=Get UnivesarlPartPickerSlotTag() -> (UnivesarlPartPickerSlotTag:Gameplay Tag Structure)
    IsPlayerSpecific:bool=true
   ) -> (then)
  ]}
 EventGraph {nodes [
   "On Pressed (Button_ToggleReadyUp)"() {is=title K2Node_ComponentBoundEvent}
   > n1 = Branch(Condition=Get IsReadyUpDisabled!(self:WBP Pilgrim Green Room Toggle Ready UI Object Reference=Get GreenRoom_ToggleReadyUI() -> (GreenRoom_ToggleReadyUI:WBP Pilgrim Green Room Toggle Ready UI Object Reference)) -> (IsReadyUpDisabled:bool))
   n2 = Cast To FortPlayerControllerGameplay!(execute=@n1.else, Object=GetOwningPlayer()) -> (then, CastFailed, AsFort Player Controller Gameplay:Fort Player Controller Gameplay Object Reference) {is=title K2Node_DynamicCast}
   n3 = GetComponentByClass!(
    self:Actor Object Reference=@n2
    ComponentClass:Actor Component Class Reference=/PilgrimCore/Player/Components/BP_PilgrimPlayerControllerComponent.BP_PilgrimPlayerControllerComponent_C
   ) -> (ReturnValue:BP Pilgrim Player Controller Component Object Reference)
   > Is Valid(InputObject=@n3)
   > Request Server Toggle Player ReadyUp!(execute, self:BP Pilgrim Player Controller Component Object Reference=@n3, Controller:Fort Player Controller Object Reference=@n2) -> (then)
   "On Pressed (Button_ChangeDifficulty)"() {is=title K2Node_ComponentBoundEvent}
   n4 = GetComponentByClass(
    self=GetOwningPlayer()
    ComponentClass=&BP_PilgrimPlayerControllerComponent_C
   )
   > Is Valid(InputObject=@n4)
   > Swap Difficulty Input Pressed(self:BP Pilgrim Player Controller Component Object Reference=@n4)
   n5 = On Input Method Changed!() -> (OutputDelegate, then, bNewInputType:ECommonInputType Enum) {is=title K2Node_CustomEvent}
   > SetPromptVisibility(self:WBP Pilgrim Quickplay Open Settings Object Reference=Get WBP_PilgrimQuickplay_OpenSettings() -> (WBP_PilgrimQuickplay_OpenSettings:WBP Pilgrim Quickplay Open Settings Object Reference))
   Event PreConstruct()
   > n6 = Create Widget!(
    execute
    Class=Get NextSongDetailsPanel() -> (NextSongDetailsPanel:WBP Pilgrim Quickplay Next Song Details Panel Class Reference)
    OwningPlayer=GetOwningPlayer()
    isInGreenRoom:bool=true
   ) -> (then, ReturnValue:WBP Pilgrim Quickplay Next Song Details Panel Object Reference)
   > n7 = AddChildToOverlay(self=Get Overlay_UpcomingSongWidget() -> (Overlay_UpcomingSongWidget:Overlay Object Reference), Content=@n6)
   > SetHorizontalAlignment(self=@n7, InHorizontalAlignment=HAlign_Fill)
   Event Construct()
   > CheckAndAddUPPWidget() {is=title K2Node_CallFunction}
   > PlayAnimationForward(InAnimation=Get A_Show() -> (A_Show:Widget Animation Object Reference))
   > n8 = Sequence!(execute) -> (then_0, then_1, then_2:Exec)
   > Bind Event to On Input Method Changed(self=Get CommonInputSubsystem(), Delegate=@n5)
   n9 = Sequence!(execute=@n8.then_1) -> (then_0, then_1)
   n10 = GetOwningPlayer()
   > n11 = WaitGameplayTagAddToActor!(
    execute:Exec
    TargetActor:Actor Object Reference=@n10
    Tag:Gameplay Tag Structure=(TagName="Gameplay.Action.Player.Emote")
    OnlyTriggerOnce:bool=false
   ) -> (then:Exec, AsyncTaskProxy:Ability Async Wait Gameplay Tag Added Object Reference, Added:Exec) {is=title K2Node_AsyncAction}
   n12 = Get StopEmoteLabel() -> (StopEmoteLabel:WBP Pilgrim Quickplay Stop Emote Label Object Reference)
   SetVisibility(execute=@n11.Added, self=@n12, InVisibility=GetEmoteLabelVisibilityOnEmote!() -> (ShouldShow:ESlateVisibility Enum=Visible))
   n13 = WaitGameplayTagRemoveFromActor!(
    execute:Exec=@n9.then_1
    TargetActor:Actor Object Reference=@n10
    Tag:Gameplay Tag Structure=(TagName="Gameplay.Action.Player.Emote")
    OnlyTriggerOnce:bool=false
   ) -> (then:Exec, AsyncTaskProxy:Ability Async Wait Gameplay Tag Removed Object Reference, Removed:Exec) {is=title K2Node_AsyncAction}
   SetVisibility(execute=@n13.Removed, self=@n12, InVisibility=Collapsed)
   Branch(execute=@n8.then_2, Condition=IsVisible())
   > n14 = PilgrimWaitForEventRouter!(execute:Exec, ContextActor:Actor Object Reference=GetOwningPlayer()) -> (then:Exec, OnEventRouterAvailable:Exec, EventRouter:Gameplay Event Router Component Object Reference) {is=title K2Node_AsyncAction}
   n15 = StartListeningToEvent!(
    execute:Exec=@n14.OnEventRouterAvailable
    Target:Gameplay Event Router Component Object Reference=@n14
    EventType:Script Struct Object Reference=/PilgrimQuickplay/Items/PilgrimEvent_GreenRoomTimerStateChanged
    Context:Object Reference
    EventBubblingRule:EEventBubblingRule Enum
   ) -> (
    then:Exec
    AsyncTaskProxy:Async Action Start Listening To Event Object Reference
    OnEventReceived:Exec
    Payload:Pilgrim Event Green Room Timer State Changed Structure
   )
   > Add(
    TargetArray:Array of Async Action Start Listening To Event Object References=Get GameplayEventHandles() -> (GameplayEventHandles:Array of Async Action Start Listening To Event Object References)
    NewItem:"Async Action Start Listening To Event Object Reference (by ref)"=@n15
   ) {is=title K2Node_CallArrayFunction}
   n16 = Branch(
    execute=@n15.OnEventReceived
    Condition="Equal (Enum)"(
     A:GreenRoomTimerState Enum=Break Pilgrim Event Green Room Timer State Changed!(PilgrimEvent_GreenRoomTimerStateChanged:"Pilgrim Event Green Room Timer State Changed Structure (by ref)"=@n15.Payload) -> (
      GreenRoomTimerState_2_#11:GreenRoomTimerState Enum
      FinalTimerLength_5_#12:double=~
     ) {is=title K2Node_BreakStruct}
     B:GreenRoomTimerState Enum=NewEnumerator2
    )
   )
   > SetVisibility(InVisibility=Collapsed)
   SetVisibility(execute=@n16.else, InVisibility=SelfHitTestInvisible)
   n17 = Event Add Content To Host Widget!() -> (
    OutputDelegate
    then
    WidgetHostTag:"Gameplay Tag Structure (by ref)"
    ContentWidget:Widget Object Reference
    Padding:"Margin Structure (by ref)"
    bFill:bool
   ) {is=title K2Node_Event}
   > Branch(Condition="Equal (GameplayTag)"(A=@n17.WidgetHostTag, B=Get UnivesarlPartPickerSlotTag()))
   > AddChildToOverlay(self=Get UPP_OVerlay(), Content=@n17.ContentWidget)
   > RemoveFromParent(self=Get Button_ChangeDifficulty())
   > RemoveFromParent(self=Get GreenRoom_InstrumentSelectUI())
   Event Destruct()
   > UnregisterWidgetHost!(execute, WidgetHost:Widget Object Reference, WidgetHostTag:"Gameplay Tag Structure (by ref)"=Get UnivesarlPartPickerSlotTag()) -> (then)
  ]}
 GetEmoteLabelVisibilityOnEmote {nodes [
   GetEmoteLabelVisibilityOnEmote() {is=title K2Node_FunctionEntry}
   n1 = NOT Boolean(A=IsPrimaryPlayer!(self:Player Controller Object Reference=GetOwningPlayer()) -> (ReturnValue:bool))
   n2 = IsHamSandwichSessionActive!() -> (ReturnValue:bool)
   > Return Node!(execute, ShouldShow:ESlateVisibility Enum=Select(Option 0=Collapsed, Option 1=SelfHitTestInvisible, Index=AND Boolean(A=@n2, B=@n1)))
  ]}
 RemoveContentFromWidgetHost {nodes [
   Remove Content From Host Widget!() -> (
    then
    WidgetHostTag:"Gameplay Tag Structure (by ref)"
    ContentWidget:Widget Object Reference
    RemovalType:EWidgetSlottingRemovalType Enum=Unparent
   ) {is=title K2Node_FunctionEntry}
   > Return Node!(execute, ReturnValue:bool=false)
  ]}
 Set ControllerDisconnected Visibility {nodes [
   n1 = Set ControllerDisconnected Visibility!() -> (then, IsDisconnected:bool=~) {is=title K2Node_FunctionEntry}
   n2 = IsPlayingReplay!() -> (ReturnValue:bool)
   > n3 = Branch(Condition=AND Boolean(A=@n1, B=NOT Boolean(A=@n2)))
   n4 = IsVisible(self=Get ControllerDisconnectPrompt() -> (ControllerDisconnectPrompt:WBP Pilgrim Quickplay Controller Disconnect Prompt Object Reference))
   > n5 = Branch(Condition=@n4)
   n6 = Get ControllerDisconnect() -> (ControllerDisconnect:Widget Animation Object Reference)
   n7 = PlayAnimationForward(execute=@n5.else, InAnimation=@n6)
   SetPromptVisibility(execute=[@n8.then, @n7.then], self=Get WBP_PilgrimQuickplay_OpenSettings())
   Branch(execute=@n3.else, Condition=@n4)
   > n8 = PlayAnimationReverse(InAnimation=@n6)
  ]}
 SetLocalPlayerOwner {nodes [
   n1 = SetLocalPlayerOwner!() -> (then, LocalPlayerController:Player Controller Object Reference) {is=title K2Node_FunctionEntry}
   > Branch(Condition=IsLocalPlayerController!(self:Controller Object Reference=@n1) -> (ReturnValue:bool))
   > SetOwningPlayer!(
    execute
    self:User Widget Object Reference=[
     Get GreenRoom_DifficultySelectUI() -> (GreenRoom_DifficultySelectUI:WBP Pilgrim Green Room Difficulty Select UI Object Reference)
     Get GreenRoom_InstrumentSelectUI()
     Get GreenRoom_ToggleReadyUI()
     Get GreenRoom_InstrumentNotice() -> (GreenRoom_InstrumentNotice:WBP Pilgrim Green Room Instrument Notice Object Reference)
     Get PlayerIdentity() -> (PlayerIdentity:WBP Pilgrim Player Identity Nameplate Object Reference)
    ]
    LocalPlayerController:Player Controller Object Reference=@n1
   ) -> (then)
  ]}
]
variables [
 bIsUPPManualCreated {type=boolean}
 EventRouter {type=GameplayEventRouterComponent, default=null}
 MyPlayerState {type=PlayerState, default=null}
 NextSongDetailsPanel {
  type=Class
  default=/PilgrimQuickplay/Widgets/Results/SongResults/WBP_PilgrimQuickplay_NextSongDetailsPanel.WBP_PilgrimQuickplay_NextSongDetailsPanel_C
 }
 UnivesarlPartPickerSlotTag {type=GameplayTag, default {tagName=Pilgrim.Hud.PartPicker.Slot}}
]
widgets {
 info {inheritedWidgetCount=0, parentClass=/Script/PilgrimCoreUI.PilgrimGameWidgetBase, rootWidgetClass=&Overlay, widgetCount=15}
 widgets [
  Overlay_0 {widgetClassPath=&Overlay, widget=Overlay_0}
  Vbox {widgetClassPath=&VerticalBox, parent=Overlay_0, widget=Vbox, slot {class=OverlaySlot}}
  StopEmoteLabel {
   widgetClassPath=&/WBP_PilgrimQuickplay_StopEmoteLabel
   bIsVariable=true
   parent=Vbox
   widget=StopEmoteLabel
   properties {visibility=Collapsed}
   slot {class=VerticalBoxSlot, properties {horizontalAlignment=HAlign_Center}}
  }
  WBP_PilgrimQuickplay_OpenSettings {
   widgetClassPath=&/WBP_PilgrimQuickplay_OpenSettings
   bIsVariable=true
   parent=Vbox
   widget=WBP_PilgrimQuickplay_OpenSettings
   properties {visibility=Collapsed}
   slot {class=VerticalBoxSlot, properties {horizontalAlignment=HAlign_Center, padding=<ltrb 0 10 0 10>}}
  }
  Overlay_UpcomingSongWidget {widgetClassPath=&Overlay, bIsVariable=true, parent=Vbox, widget=Overlay_UpcomingSongWidget, slot {class=VerticalBoxSlot}}
  WBP_Pilgrim_ChoppedCheeseEnabledNotice {
   widgetClassPath=&/WBP_Pilgrim_ChoppedCheeseEnabledNotice
   bIsVariable=true
   parent=Vbox
   widget=WBP_Pilgrim_ChoppedCheeseEnabledNotice
   slot {class=VerticalBoxSlot}
  }
  GreenRoom_InstrumentNotice {
   widgetClassPath=&/WBP_Pilgrim_GreenRoom_InstrumentNotice
   bIsVariable=true
   parent=Vbox
   widget=GreenRoom_InstrumentNotice
   slot {class=VerticalBoxSlot, properties {padding=<ltrb 0 10 0 0>}}
  }
  UPP_OVerlay {widgetClassPath=&Overlay, bIsVariable=true, parent=Vbox, widget=UPP_OVerlay, slot {class=VerticalBoxSlot}}
  GreenRoom_InstrumentSelectUI {
   widgetClassPath=&/WBP_Pilgrim_GreenRoom_InstrumentSelectUI
   bIsVariable=true
   parent=Vbox
   widget=GreenRoom_InstrumentSelectUI
   slot {class=VerticalBoxSlot, properties {padding=<ltrb 0 10 0 10>}}
  }
  Button_ChangeDifficulty {
   widgetClassPath=&Button
   parent=Vbox
   widget=Button_ChangeDifficulty
   properties {
    isFocusable=false
    widgetStyle {
     clickedSlateSound {resourceObject=null}
     disabled {
      bIsDynamicallyLoaded=false
      drawAs=NoDrawType
      imageType=NoImage
      mirroring=NoMirror
      resourceName=None
      resourceObject=null
      tiling=NoTile
      imageSize=<0 0>
      margin=<ltrb 0 0 0 0>
      outlineSettings {
       bUseBrushTransparency=false
       roundingType=HalfHeightRadius
       width=0
       color {colorUseRule=UseColor_Specified, specifiedColor=<rgba 0 0 0 0>}
       cornerRadii=<0 0 0 0>
      }
      tintColor {colorUseRule=UseColor_Specified, specifiedColor=<rgba 1 1 1 1>}
      uVRegion {bIsValid=false, max=<0 0>, min=<0 0>}
     }
     disabledForeground {colorUseRule=UseColor_Specified, specifiedColor=<rgba 0.5271 0.5271 0.5271 1>}
     hovered {
      bIsDynamicallyLoaded=false
      drawAs=NoDrawType
      imageType=NoImage
      mirroring=NoMirror
      resourceName=None
      resourceObject=null
      tiling=NoTile
      imageSize=<32 32>
      margin=<ltrb 0 0 0 0>
      outlineSettings {
       bUseBrushTransparency=true
       roundingType=FixedRadius
       width=1
       color {colorUseRule=UseColor_Specified, specifiedColor=<rgba 0.7243 0.7243 0.7243 1>}
       cornerRadii=<4 4 4 4>
      }
      tintColor {colorUseRule=UseColor_Specified, specifiedColor=<rgba 0.7243 0.7243 0.7243 1>}
      uVRegion {bIsValid=false, max=<0 0>, min=<0 0>}
     }
     hoveredForeground {colorUseRule=UseColor_Specified, specifiedColor=<rgba 0 1 1 1>}
     hoveredSlateSound {resourceObject=null}
     normal {
      bIsDynamicallyLoaded=false
      drawAs=NoDrawType
      imageType=NoImage
      mirroring=NoMirror
      resourceName=None
      resourceObject=null
      tiling=NoTile
      imageSize=<32 32>
      margin=<ltrb 0 0 0 0>
      outlineSettings {
       bUseBrushTransparency=true
       roundingType=FixedRadius
       width=1
       color {colorUseRule=UseColor_Specified, specifiedColor=<rgba 0.6951 0.6951 0.6951 1>}
       cornerRadii=<4 4 4 4>
      }
      tintColor {colorUseRule=UseColor_Specified, specifiedColor=<rgba 0.4955 0.4955 0.4955 1>}
      uVRegion {bIsValid=false, max=<0 0>, min=<0 0>}
     }
     normalForeground {colorUseRule=UseColor_Specified, specifiedColor=<rgba 0 1 1 1>}
     normalPadding=<ltrb 0 0 0 0>
     pressed {
      bIsDynamicallyLoaded=false
      drawAs=NoDrawType
      imageType=NoImage
      mirroring=NoMirror
      resourceName=None
      resourceObject=null
      tiling=NoTile
      imageSize=<32 32>
      margin=<ltrb 0 0 0 0>
      outlineSettings {
       bUseBrushTransparency=true
       roundingType=FixedRadius
       width=1
       color {colorUseRule=UseColor_Specified, specifiedColor=<rgba 0.7243 0.7243 0.7243 1>}
       cornerRadii=<4 4 4 4>
      }
      tintColor {colorUseRule=UseColor_Specified, specifiedColor=<rgba 0.3843 0.3843 0.3843 1>}
      uVRegion {bIsValid=false, max=<0 0>, min=<0 0>}
     }
     pressedForeground {colorUseRule=UseColor_Specified, specifiedColor=<rgba 0 1 1 1>}
     pressedPadding=<ltrb 0 0 0 0>
     pressedSlateSound {resourceObject=null}
    }
   }
   slot {class=VerticalBoxSlot}
  }
  GreenRoom_DifficultySelectUI {
   widgetClassPath=&/WBP_Pilgrim_GreenRoom_DifficultySelectUI
   bIsVariable=true
   parent=Button_ChangeDifficulty
   widget=GreenRoom_DifficultySelectUI
   properties {visibility=HitTestInvisible}
   slot {class=ButtonSlot, properties {horizontalAlignment=HAlign_Fill, padding=<ltrb 0 0 0 0>}}
  }
  Button_ToggleReadyUp {
   widgetClassPath=&Button
   parent=Vbox
   widget=Button_ToggleReadyUp
   properties {
    isFocusable=false
    widgetStyle {
     clickedSlateSound {resourceObject=null}
     disabled {
      bIsDynamicallyLoaded=false
      drawAs=NoDrawType
      imageType=NoImage
      mirroring=NoMirror
      resourceName=None
      resourceObject=null
      tiling=NoTile
      imageSize=<0 0>
      margin=<ltrb 0 0 0 0>
      outlineSettings {
       bUseBrushTransparency=false
       roundingType=HalfHeightRadius
       width=0
       color {colorUseRule=UseColor_Specified, specifiedColor=<rgba 0 0 0 0>}
       cornerRadii=<0 0 0 0>
      }
      tintColor {colorUseRule=UseColor_Specified, specifiedColor=<rgba 1 1 1 1>}
      uVRegion {bIsValid=false, max=<0 0>, min=<0 0>}
     }
     disabledForeground {colorUseRule=UseColor_Specified, specifiedColor=<rgba 0.5271 0.5271 0.5271 1>}
     hovered {
      bIsDynamicallyLoaded=false
      drawAs=NoDrawType
      imageType=NoImage
      mirroring=NoMirror
      resourceName=None
      resourceObject=null
      tiling=NoTile
      imageSize=<32 32>
      margin=<ltrb 0 0 0 0>
      outlineSettings {
       bUseBrushTransparency=true
       roundingType=FixedRadius
       width=1
       color {colorUseRule=UseColor_Specified, specifiedColor=<rgba 0.7243 0.7243 0.7243 1>}
       cornerRadii=<4 4 4 4>
      }
      tintColor {colorUseRule=UseColor_Specified, specifiedColor=<rgba 0.7243 0.7243 0.7243 1>}
      uVRegion {bIsValid=false, max=<0 0>, min=<0 0>}
     }
     hoveredForeground {colorUseRule=UseColor_Specified, specifiedColor=<rgba 0 1 1 1>}
     hoveredSlateSound {resourceObject=null}
     normal {
      bIsDynamicallyLoaded=false
      drawAs=NoDrawType
      imageType=NoImage
      mirroring=NoMirror
      resourceName=None
      resourceObject=null
      tiling=NoTile
      imageSize=<32 32>
      margin=<ltrb 0 0 0 0>
      outlineSettings {
       bUseBrushTransparency=true
       roundingType=FixedRadius
       width=1
       color {colorUseRule=UseColor_Specified, specifiedColor=<rgba 0.6951 0.6951 0.6951 1>}
       cornerRadii=<4 4 4 4>
      }
      tintColor {colorUseRule=UseColor_Specified, specifiedColor=<rgba 0.4955 0.4955 0.4955 1>}
      uVRegion {bIsValid=false, max=<0 0>, min=<0 0>}
     }
     normalForeground {colorUseRule=UseColor_Specified, specifiedColor=<rgba 0 1 1 1>}
     normalPadding=<ltrb 0 0 0 0>
     pressed {
      bIsDynamicallyLoaded=false
      drawAs=NoDrawType
      imageType=NoImage
      mirroring=NoMirror
      resourceName=None
      resourceObject=null
      tiling=NoTile
      imageSize=<32 32>
      margin=<ltrb 0 0 0 0>
      outlineSettings {
       bUseBrushTransparency=true
       roundingType=FixedRadius
       width=1
       color {colorUseRule=UseColor_Specified, specifiedColor=<rgba 0.7243 0.7243 0.7243 1>}
       cornerRadii=<4 4 4 4>
      }
      tintColor {colorUseRule=UseColor_Specified, specifiedColor=<rgba 0.3843 0.3843 0.3843 1>}
      uVRegion {bIsValid=false, max=<0 0>, min=<0 0>}
     }
     pressedForeground {colorUseRule=UseColor_Specified, specifiedColor=<rgba 0 1 1 1>}
     pressedPadding=<ltrb 0 0 0 0>
     pressedSlateSound {resourceObject=null}
    }
   }
   slot {class=VerticalBoxSlot, properties {padding=<ltrb 0 10 0 0>}}
  }
  GreenRoom_ToggleReadyUI {
   widgetClassPath=&/WBP_Pilgrim_GreenRoom_ToggleReadyUI
   bIsVariable=true
   parent=Button_ToggleReadyUp
   widget=GreenRoom_ToggleReadyUI
   properties {visibility=HitTestInvisible}
   slot {class=ButtonSlot, properties {horizontalAlignment=HAlign_Fill, padding=<ltrb 0 0 0 0>}}
  }
  ControllerDisconnectPrompt {
   widgetClassPath=&/WBP_PilgrimQuickplay_ControllerDisconnectPrompt
   bIsVariable=true
   parent=Vbox
   widget=ControllerDisconnectPrompt
   properties {renderOpacity=0, visibility=Collapsed}
   slot {class=VerticalBoxSlot, properties {verticalAlignment=VAlign_Center}}
  }
  PlayerIdentity {
   widgetClassPath=/PilgrimQuickplay/Widgets/Players/WBP_Pilgrim_PlayerIdentity_Nameplate
   bIsVariable=true
   parent=Vbox
   widget=PlayerIdentity
   slot {class=VerticalBoxSlot, properties {verticalAlignment=VAlign_Bottom, padding=<ltrb 0 10 0 0>}}
  }
 ]
}
