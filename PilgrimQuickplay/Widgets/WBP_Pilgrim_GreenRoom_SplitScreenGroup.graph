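# Green Room split-screen HUD row that gives up to four local players their own bottom-right group widget.
# Construct builds the owner's group, adds other existing local players, then follows playspace player added/removed events via event routers.
# New groups fill the HBox matching their array index; removal only collapses that box, leaving the widget listed.
# Conversation entry collapses the whole row; leaving restores it if a HamSandwich session is active.
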WidgetBlueprint /PilgrimQuickplay/Widgets/WBP_Pilgrim_GreenRoom_SplitScreenGroup
graphs [
 Add Player UI {nodes [
   Add Player UI!() -> (then, PlayerController:Player Controller Object Reference) {is=title K2Node_FunctionEntry}
   > n1 = Create Widget!(
    execute:Exec
    Class:User Widget Class Reference=&/WBP_Pilgrim_GreenRoom_BottomRightGroup.WBP_Pilgrim_GreenRoom_BottomRightGroup_C
    OwningPlayer:Player Controller Object Reference=Get PlayerController() -> (PlayerController:Player Controller Object Reference)
    bIsUPPManualCreated:bool=true
   ) -> (then:Exec, ReturnValue:WBP Pilgrim Green Room Bottom Right Group Object Reference)
   > n2 = Set NewLocalPlayerWidget(NewLocalPlayerWidget:WBP Pilgrim Green Room Bottom Right Group Object Reference=@n1) -> (Output_Get:WBP Pilgrim Green Room Bottom Right Group Object Reference)
   > n3 = Add(
    TargetArray:Array of WBP Pilgrim Green Room Bottom Right Group Object References=Get ListOfPlayerBottomRightGroup() -> (ListOfPlayerBottomRightGroup:Array of WBP Pilgrim Green Room Bottom Right Group Object References)
    NewItem:"WBP Pilgrim Green Room Bottom Right Group Object Reference (by ref)"=@n2
   ) {is=title K2Node_CallArrayFunction}
   n4 = "Get (a copy)"(
    Array:Array of Horizontal Box Object References=Get ListOfPlayerUI() -> (ListOfPlayerUI:Array of Horizontal Box Object References)
    Dimension 1=@n3
   ) -> (Output:Horizontal Box Object Reference)
   > SetVisibility(self=@n4, InVisibility=Visible)
   > n5 = AddChildToHorizontalBox!(
    execute
    self:Horizontal Box Object Reference=@n4
    Content:Widget Object Reference=Get NewLocalPlayerWidget() -> (NewLocalPlayerWidget:WBP Pilgrim Green Room Bottom Right Group Object Reference)
   ) -> (then, ReturnValue:Horizontal Box Slot Object Reference)
   > SetSize!(execute, self:Horizontal Box Slot Object Reference=@n5, InSize_Value:float=1, InSize_SizeRule:ESlateSizeRule Enum=Fill) -> (then)
  ]}
 EventGraph {nodes [
   n1 = Event PreConstruct()
   > Parent: PreConstruct(IsDesignTime=@n1.IsDesignTime)
   Event Construct()
   > n2 = PilgrimWaitForEventRouter!(execute:Exec, ContextActor:Actor Object Reference=GetOwningPlayer()) -> (then:Exec, OnEventRouterAvailable:Exec, EventRouter:Gameplay Event Router Component Object Reference) {is=title K2Node_AsyncAction}
   > Set Event Router(Event Router:Gameplay Event Router Component Object Reference=@n2) -> (Output_Get:Gameplay Event Router Component Object Reference)
   > n3 = Sequence!(execute) -> (then_0, then_1, then_2:Exec, then_3:Exec, then_4:Exec)
   > n4 = Create Widget(
    Class=&WBP_Pilgrim_GreenRoom_BottomRightGroup_C
    OwningPlayer=GetOwningPlayer()
    bIsUPPManualCreated=true
   )
   > AddChild(self=Get VerticalBox_P1() -> (VerticalBox_P1:Vertical Box Object Reference), Content=@n4)
   > Add(TargetArray=Get ListOfPlayerBottomRightGroup(), NewItem=@n4)
   n5 = Get ListOfPlayerUI()
   > Add(
    TargetArray:Array of Horizontal Box Object References=@n5
    NewItem:"Horizontal Box Object Reference (by ref)"=Get LocalPlayer1_HBox() -> (LocalPlayer1_HBox:Horizontal Box Object Reference)
   )
   > Add(TargetArray=@n5, NewItem=Get LocalPlayer2_HBox() -> (LocalPlayer2_HBox:Horizontal Box Object Reference))
   > Add(TargetArray=@n5, NewItem=Get LocalPlayer3_HBox() -> (LocalPlayer3_HBox:Horizontal Box Object Reference))
   > Add(TargetArray=@n5, NewItem=Get LocalPlayer4_HBox() -> (LocalPlayer4_HBox:Horizontal Box Object Reference))
   n6 = StartListeningToEvent!(
    execute:Exec=@n3.then_1
    Target:Gameplay Event Router Component Object Reference=Get Event Router() -> (Event Router:Gameplay Event Router Component Object Reference)
    EventType:Script Struct Object Reference=/Script/PilgrimCoreRuntime.PilgrimEvent_Playspace_PlayerAdded
    Context:Object Reference
    EventBubblingRule:EEventBubblingRule Enum
   ) -> (
    then:Exec
    AsyncTaskProxy:Async Action Start Listening To Event Object Reference
    OnEventReceived:Exec
    Payload:Pilgrim Event Playspace Player Added Structure
   )
   n7 = Break Pilgrim Player Info!(PilgrimPlayerInfo:"Pilgrim Player Info Structure (by ref)"=Break Pilgrim Event Playspace Player Added!(PilgrimEvent_Playspace_PlayerAdded:"Pilgrim Event Playspace Player Added Structure (by ref)"=@n6.Payload) -> (PlayerInfo:Pilgrim Player Info Structure) {is=title K2Node_BreakStruct}) -> (PlayerState:Fort Player State Weak object, PlayerController:Fort Player Controller Weak object, bIsLocalPlayer:bool=~) {is=title K2Node_BreakStruct}
   Branch(execute=@n6.OnEventReceived, Condition=@n7.bIsLocalPlayer)
   n8 = GetFortPlayerControllerFromActor!(Actor:Actor Object Reference=@n7.PlayerController) -> (ReturnValue:Fort Player Controller Object Reference)
   > Is Valid(InputObject=@n8)
   n9 = GetOwningPlayer()
   > Branch(Condition=AND Boolean(
     A=IsLocalPlayerController!(self:Controller Object Reference=@n8) -> (ReturnValue:bool)
     B="Not Equal (Object)"(A:Object Reference=@n8, B:Object Reference=@n9) -> (ReturnValue:bool=~) {is=title K2Node_PromotableOperator}
    ))
   > Add Player UI!(execute, self, PlayerController:Player Controller Object Reference=@n8) -> (then) {is=title K2Node_CallFunction}
   n10 = StartListeningToEvent(execute=@n3.then_2, Target=Get Event Router(), EventType=/Script/PilgrimCoreRuntime.PilgrimEvent_Playspace_PlayerRemoved) -> (Payload:Pilgrim Event Playspace Player Removed Structure)
   n11 = Break Pilgrim Player Info(PilgrimPlayerInfo=Break Pilgrim Event Playspace Player Removed!(PilgrimEvent_Playspace_PlayerRemoved:"Pilgrim Event Playspace Player Removed Structure (by ref)"=@n10.Payload) -> (PlayerInfo:Pilgrim Player Info Structure) {is=title K2Node_BreakStruct}) -> (bIsLocalPlayer=~)
   Branch(execute=@n10.OnEventReceived, Condition=@n11.bIsLocalPlayer)
   > RemovePlayerUI!(execute, self, PlayerController:Controller Object Reference=@n11.PlayerController) -> (then)
   InitLocalPlayerUI(execute=@n3.then_3)
   n12 = Sequence!(execute=@n3.then_4) -> (then_0, then_1)
   n13 = GetGlobalEventRouter()
   > n14 = StartListeningToEvent(Target=@n13, EventType=/Script/FortniteConversationRuntime.FortConversationEnterEvent) -> (Payload:Fort Conversation Enter Event Structure)
   SetVisibility(execute=@n14.OnEventReceived, InVisibility=Collapsed)
   n15 = StartListeningToEvent(execute=@n12.then_1, Target=@n13, EventType=/Script/FortniteConversationRuntime.FortConversationLeaveEvent) -> (Payload:Fort Conversation Leave Event Structure)
   Branch(execute=@n15.OnEventReceived, Condition=IsHamSandwichSessionActive!() -> (ReturnValue:bool))
   > SetVisibility(InVisibility=SelfHitTestInvisible)
  ]}
 InitLocalPlayerUI {nodes [
   InitLocalPlayerUI() {is=title K2Node_FunctionEntry}
   > Branch(Condition="integer > integer"(A=GetNumLocalPlayerControllers!() -> (ReturnValue:int=0), B=1))
   n1 = GetOwningPlayer()
   > n2 = For Each Loop(Array:Array of Fort Player Controller Object References=GetAllLocalPlayerControllers!(self:Pilgrim Playspace Object Reference=GetSparksMusicPlayspaceOfClass!(
      PlayspaceType:Sparks Music Playspace Class Reference=/Script/PilgrimCoreRuntime.PilgrimPlayspace
      Actor:Actor Object Reference=@n1
      bMatchChildTypes:bool=false
     ) -> (ReturnValue:Pilgrim Playspace Object Reference)) -> (ReturnValue:Array of Fort Player Controller Object References)) -> (Array Element:Fort Player Controller Object Reference)
   > Branch(Condition="Not Equal (Object)"(A=@n2, B=GetOwningPlayer()) -> (ReturnValue=~))
   > Add Player UI(PlayerController=@n2)
  ]}
 PlayerUIVisibility {nodes [
   n1 = PlayerUIVisibility!() -> (then, PlayerIndex:int, IsVisibile:bool=~) {is=title K2Node_FunctionEntry}
   > SetVisibility(
    self="Get (a copy)"(Array=Get ListOfPlayerUI(), Dimension 1=@n1)
    InVisibility=Select(Option 0=Collapsed, Option 1=Visible, Index=@n1.IsVisibile)
   )
  ]}
 RemovePlayerUI {nodes [
   n1 = RemovePlayerUI!() -> (then, PlayerController:Controller Object Reference) {is=title K2Node_FunctionEntry}
   > n2 = For Each Loop with Break!(
    Exec:Exec
    Array:Array of WBP Pilgrim Green Room Bottom Right Group Object References=Get ListOfPlayerBottomRightGroup()
    Break:Exec=@n3.then
   ) -> (LoopBody:Exec, Array Element:WBP Pilgrim Green Room Bottom Right Group Object Reference, Array Index:int, Completed:Exec)
   > Branch(Condition="Equal (Object)"(A=@n1, B=GetOwningPlayer(self=@n2)))
   > n3 = PlayerUIVisibility!(execute, self, PlayerIndex:int=@n2.Array Index, IsVisibile:bool=false) -> (then) {is=title K2Node_CallFunction}
  ]}
]
variables [
 Event Router {type=GameplayEventRouterComponent, default=null}
 ListOfPlayerBottomRightGroup {type="array<WBP_Pilgrim_GreenRoom_BottomRightGroup_C>", default []}
 ListOfPlayerUI {type="array<HorizontalBox>", default []}
]
widgets {
 info {inheritedWidgetCount=0, parentClass=/Script/PilgrimCoreUI.PilgrimGameWidgetBase, rootWidgetClass=&HorizontalBox, widgetCount=7}
 widgets [
  HorizontalBox_17 {widgetClassPath=&HorizontalBox, widget=HorizontalBox_17}
  LocalPlayer1_HBox {
   widgetClassPath=&HorizontalBox
   bIsVariable=true
   parent=HorizontalBox_17
   widget=LocalPlayer1_HBox
   slot {class=HorizontalBoxSlot, properties {horizontalAlignment=HAlign_Left, verticalAlignment=VAlign_Bottom, size {sizeRule=Fill, value=1}}}
  }
  VerticalBox_P1 {
   widgetClassPath=&VerticalBox
   bIsVariable=true
   parent=LocalPlayer1_HBox
   widget=VerticalBox_P1
   slot {class=HorizontalBoxSlot, properties {horizontalAlignment=HAlign_Left, size {sizeRule=Fill, value=1}}}
  }
  TopSlotP1LocalMP {widgetClassPath=&NamedSlot, bIsVariable=true, parent=VerticalBox_P1, widget=TopSlotP1LocalMP, slot {class=VerticalBoxSlot}}
  LocalPlayer2_HBox {
   widgetClassPath=&HorizontalBox
   bIsVariable=true
   parent=HorizontalBox_17
   widget=LocalPlayer2_HBox
   properties {visibility=Collapsed}
   slot {
    class=HorizontalBoxSlot
    properties {horizontalAlignment=HAlign_Left, verticalAlignment=VAlign_Bottom, padding=<ltrb 64 0 0 0>, size {sizeRule=Fill, value=1}}
   }
  }
  LocalPlayer3_HBox {
   widgetClassPath=&HorizontalBox
   bIsVariable=true
   parent=HorizontalBox_17
   widget=LocalPlayer3_HBox
   properties {visibility=Collapsed}
   slot {
    class=HorizontalBoxSlot
    properties {horizontalAlignment=HAlign_Left, verticalAlignment=VAlign_Bottom, padding=<ltrb 64 0 0 0>, size {sizeRule=Fill, value=1}}
   }
  }
  LocalPlayer4_HBox {
   widgetClassPath=&HorizontalBox
   bIsVariable=true
   parent=HorizontalBox_17
   widget=LocalPlayer4_HBox
   properties {visibility=Collapsed}
   slot {
    class=HorizontalBoxSlot
    properties {horizontalAlignment=HAlign_Left, verticalAlignment=VAlign_Bottom, padding=<ltrb 64 0 0 0>, size {sizeRule=Fill, value=1}}
   }
  }
 ]
}
