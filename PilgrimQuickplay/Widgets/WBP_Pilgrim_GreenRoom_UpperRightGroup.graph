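# WBP_Pilgrim_GreenRoom_UpperRightGroup is a UserWidget blueprint for Pilgrim Quickplay, named as the upper-right group of the Green Room.
# Its widget tree holds only an empty root CanvasPanel, so it displays no content of its own.
# It declares EventRouter (GameplayEventRouterComponent) and MyPlayerState variables, both null by default, but its event graph is empty and uses neither.
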WidgetBlueprint /PilgrimQuickplay/Widgets/WBP_Pilgrim_GreenRoom_UpperRightGroup
graphs [EventGraph {links [], nodes []}]
variables [
 EventRouter {type=GameplayEventRouterComponent, default=null}
 MyPlayerState {type=PlayerState, default=null}
]
widgets {
 info {inheritedWidgetCount=0, parentClass=&UserWidget, rootWidgetClass=&CanvasPanel, widgetCount=1}
 widgets [CanvasPanel {widgetClassPath=&CanvasPanel, bIsVariable=true, widget=CanvasPanel}]
}
