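# Green room HUD for Pilgrim Quickplay, composing ready-up, player list, identity pill, pinned quests and corner widget groups.
# Playspace user events toggle solo and split-screen layouts through IsHamSandwichSessionActive, which also selects the Pro Parts coach mark.
# Onboarding chains Stage Tech and Pro Parts coach marks, saving seen state through the FTUE controller component.
# Async Jamming-tag waits hide and restore the active bottom safe zone around now-playing UI.
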WidgetBlueprint /PilgrimQuickplay/Widgets/WBP_Pilgrim_GreenRoom_AllGroups
graphs [
 ActivateProPartsCoachmark {nodes [
   ActivateProPartsCoachmark() {is=title K2Node_FunctionEntry}
   n1 = Get ProPartsCoachMarkToUse() -> (ProPartsCoachMarkToUse:WBP Pilgrim Coach Mark Pro Parts Object Reference)
   > Branch(Condition=ShouldShowOnboarding!(self:Fort Sidebar Onboard Tooltip Widget Object Reference=@n1) -> (ReturnValue:bool))
   > ActivateWidget(self=@n1)
   > Bind Event to On Pro Parts Tooltip Dismissed(self:WBP Pilgrim Coach Mark Pro Parts Object Reference=@n1, Delegate=Create Event()) {is=title K2Node_AddDelegate}
  ]}
 ActivateStageTechCoachmark {nodes [
   ActivateStageTechCoachmark() {is=title K2Node_FunctionEntry}
   n1 = Get WBP_PilgrimCoachMark_StageTech() -> (WBP_PilgrimCoachMark_StageTech:WBP Pilgrim Coach Mark Stage Tech Object Reference)
   > Branch(Condition=ShouldShowOnboarding(self=@n1))
   > ActivateWidget(self=@n1)
   > Bind Event to On Stage Tech Tooltip Dismissed(self:WBP Pilgrim Coach Mark Stage Tech Object Reference=@n1, Delegate=Create Event()) {is=title K2Node_AddDelegate}
  ]}
 EventGraph {nodes [
   n1 = UserRemovedEvent!() -> (OutputDelegate, then, RemovedUser:"Playspace User Structure (by ref)") {is=title K2Node_CustomEvent}
   > UpdateSplitScreenVisibility()
   n2 = CustomEvent!() -> (OutputDelegate, then, AddedUser:"Playspace User Structure (by ref)") {is=title K2Node_CustomEvent}
   > UpdateSplitScreenVisibility()
   OnFirstCoachMarkClosed() {is=title K2Node_CustomEvent}
   > ActivateProPartsCoachmark() {is=title K2Node_CallFunction}
   OnSecondCoachMarkClosed() {is=title K2Node_CustomEvent}
   n3 = GetComponentByClass!(
    self:Actor Object Reference=GetOwningPlayer()
    ComponentClass:Actor Component Class Reference=/Script/PilgrimCoreUI.PilgrimFTUEControllerComponent
   ) -> (ReturnValue:Pilgrim FTUEController Component Object Reference)
   > Is Valid(InputObject=@n3)
   > SetHasSeenRoadieCoachMark!(execute, self:Pilgrim FTUEController Component Object Reference=@n3, bHasSeenRoadieCoachMark:bool=true) -> (then)
   > SetInputMode_GameOnly!(execute, PlayerController:Player Controller Object Reference=GetOwningPlayer(), bFlushInput:bool=true) -> (then)
   Event OnInitialized()
   > n4 = Sequence!(execute) -> (then_0, then_1, then_2:Exec, then_3:Exec, then_4:Exec, then_5:Exec)
   > UpdateSplitScreenVisibility()
   Bind Event to Notify Playspace User Removed(
    execute=@n4.then_2
    self:Playspace Object Reference=GetSparksMusicPlayspaceOfClass!(
     PlayspaceType:Sparks Music Playspace Class Reference=/Script/PilgrimCoreRuntime.PilgrimPlayspace
     Actor:Actor Object Reference=GetOwningPlayer()
     bMatchChildTypes:bool=false
    ) -> (ReturnValue:Pilgrim Playspace Object Reference)
    Delegate=@n1
   ) {is=title K2Node_AddDelegate}
   SetInputMode_GameOnly(execute=@n4.then_3, PlayerController=GetOwningPlayer(), bFlushInput=true)
   Bind Event to Notify Playspace User Added(
    execute=@n4.then_1
    self:Playspace Object Reference=GetSparksMusicPlayspaceOfClass(PlayspaceType=&PilgrimPlayspace, Actor=GetOwningPlayer(), bMatchChildTypes=false)
    Delegate=@n2
   ) {is=title K2Node_AddDelegate}
   n5 = GetPlayerController!(PlayerIndex:int=0) -> (ReturnValue:Player Controller Object Reference)
   Is Valid(exec=@n4.then_5, InputObject=@n5)
   > n6 = PilgrimWaitForEventRouter!(execute:Exec, ContextActor:Actor Object Reference=@n5) -> (then:Exec, OnEventRouterAvailable:Exec, EventRouter:Gameplay Event Router Component Object Reference) {is=title K2Node_AsyncAction}
   n7 = StartListeningToEvent!(
    execute:Exec=@n6.OnEventRouterAvailable
    Target:Gameplay Event Router Component Object Reference=@n6
    EventType:Script Struct Object Reference=/Script/PilgrimCoreRuntime.PilgrimEvent_Playspace_PlayerAdded
    Context:Object Reference
    EventBubblingRule:EEventBubblingRule Enum
   ) -> (
    then:Exec
    AsyncTaskProxy:Async Action Start Listening To Event Object Reference
    OnEventReceived:Exec
    Payload_PlayerInfo:Pilgrim Player Info Structure
   )
   > Set AddingPlayerInPieEvent(AddingPlayerInPieEvent:Async Action Start Listening To Event Object Reference=@n7) -> (Output_Get:Async Action Start Listening To Event Object Reference)
   SetProPartsCoachmarkReference(execute=@n7.OnEventReceived)
   n8 = Sequence!(execute=@n4.then_4) -> (then_0, then_1)
   n9 = GetOwningPlayer()
   > n10 = WaitGameplayTagAddToActor!(
    execute:Exec
    TargetActor:Actor Object Reference=@n9
    Tag:Gameplay Tag Structure=(TagName="Gameplay.Action.Player.Jam.Jamming")
    OnlyTriggerOnce:bool=false
   ) -> (then:Exec, AsyncTaskProxy:Ability Async Wait Gameplay Tag Added Object Reference, Added:Exec) {is=title K2Node_AsyncAction}
   Branch(execute=@n10.Added, Condition=IsJamShowingNowPlayingUI!() -> (IsShowing:bool))
   > SetVisibility(
    self=Select(
     Option 0:Widget Object Reference=Get SafeZone_Solo() -> (SafeZone_Solo:Safe Zone Object Reference)
     Option 1:Widget Object Reference=Get SafeZone_SplitScreen() -> (SafeZone_SplitScreen:Safe Zone Object Reference)
     Index=IsHamSandwichSessionActive!() -> (ReturnValue:bool)
    ) -> (ReturnValue:Widget Object Reference)
    InVisibility=Hidden
   )
   n11 = WaitGameplayTagRemoveFromActor!(
    execute:Exec=@n8.then_1
    TargetActor:Actor Object Reference=@n9
    Tag:Gameplay Tag Structure=(TagName="Gameplay.Action.Player.Jam.Jamming")
    OnlyTriggerOnce:bool=false
   ) -> (then:Exec, AsyncTaskProxy:Ability Async Wait Gameplay Tag Removed Object Reference, Removed:Exec) {is=title K2Node_AsyncAction}
   SetVisibility(
    execute=@n11.Removed
    self=Select(Option 0=Get SafeZone_Solo(), Option 1=Get SafeZone_SplitScreen(), Index=IsHamSandwichSessionActive())
    InVisibility=SelfHitTestInvisible
   )
   Event Construct()
   n12 = Get PlayerIdentity() -> (PlayerIdentity:WBP Pilgrim Player Identity Pill Object Reference)
   > OverridePlayerInfo!(
    execute
    self:WBP Pilgrim Player Identity Pill Object Reference=@n12
    PlayerState:Player State Object Reference=Get PlayerState!(self:Controller Object Reference=GetOwningPlayer()) -> (PlayerState:Player State Object Reference)
   ) -> (then)
   n13 = GetOwningPlayer()
   n14 = GetComponentByClass(self=@n13, ComponentClass=&PilgrimFTUEControllerComponent)
   > Is Valid(InputObject=@n14)
   > Set OwnerPlayerReference(OwnerPlayerReference:Player Controller Object Reference=@n13) -> (Output_Get:Player Controller Object Reference)
   > SetProPartsCoachmarkReference()
   n15 = GetHasSeenRoadieCoachMark!(self:Pilgrim FTUEController Component Object Reference=@n14) -> (bOutHasValidSettings:bool, ReturnValue:bool)
   > n16 = Set bIsCoachmarkInfoValid(bIsCoachmarkInfoValid:bool=@n15) -> (Output_Get:bool)
   > Branch(Condition=@n16)
   n17 = NOT Boolean(A=@n15.ReturnValue)
   n18 = Get WBP_PilgrimCoachMark_StageTech()
   > n19 = Branch(Condition=AND Boolean(A=ShouldShowOnboarding(self=@n18), B=@n17))
   n20 = Get OwnerPlayerReference() -> (OwnerPlayerReference:Player Controller Object Reference)
   > Set Input Mode Game And UI!(
    execute
    PlayerController:Player Controller Object Reference=@n20
    InWidgetToFocus:Widget Object Reference
    InMouseLockMode:EMouseLockMode Enum=DoNotLock
    bHideCursorDuringCapture:bool=true
    bFlushInput:bool=false
   ) -> (then) {is=title K2Node_CallFunction}
   > ActivateStageTechCoachmark() {is=title K2Node_CallFunction}
   n21 = Get WBP_PilgrimCoachMark_ProParts_LocalMP() -> (WBP_PilgrimCoachMark_ProParts_LocalMP:WBP Pilgrim Coach Mark Pro Parts Object Reference)
   Branch(execute=@n19.else, Condition=AND Boolean(A=@n17, B=ShouldShowOnboarding(self=@n21)))
   > Set Input Mode Game And UI(PlayerController=@n20, InMouseLockMode=DoNotLock, bHideCursorDuringCapture=true, bFlushInput=false)
   > ActivateProPartsCoachmark()
   Event PreConstruct()
   > Set TextFormatOverride!(
    execute
    TextFormatOverride:Text="LOCTEXT(\"{PlayerID} has control\")"
    self:WBP Pilgrim Player Identity Local Identity Label Object Reference=Get PlayerIdentity_LocalIdentity_Label!(self:WBP Pilgrim Player Identity Pill Object Reference=@n12) -> (PlayerIdentity_LocalIdentity_Label:WBP Pilgrim Player Identity Local Identity Label Object Reference)
   ) -> (then, Output_Get:Text)
   Event Destruct()
   > Unbind all Events from On Stage Tech Tooltip Dismissed!(execute:Exec, self:WBP Pilgrim Coach Mark Stage Tech Object Reference=Get WBP_PilgrimCoachMark_StageTech()) -> (then:Exec) {is=title K2Node_ClearDelegate}
   > Unbind all Events from On Pro Parts Tooltip Dismissed!(execute:Exec, self:WBP Pilgrim Coach Mark Pro Parts Object Reference=Get ProPartsCoachMarkToUse()) -> (then:Exec) {is=title K2Node_ClearDelegate}
   > StopListeningToEvent(self:Async Action Start Listening To Event Object Reference=Get AddingPlayerInPieEvent() -> (AddingPlayerInPieEvent:Async Action Start Listening To Event Object Reference))
   > Parent: Destruct()
  ]}
 SetProPartsCoachmarkReference {nodes [
   SetProPartsCoachmarkReference() {is=title K2Node_FunctionEntry}
   > Set ProPartsCoachMarkToUse(ProPartsCoachMarkToUse:WBP Pilgrim Coach Mark Pro Parts Object Reference=Select(
     Option 0:WBP Pilgrim Coach Mark Pro Parts Object Reference=Get WBP_PilgrimCoachMark_ProParts_Solo() -> (WBP_PilgrimCoachMark_ProParts_Solo:WBP Pilgrim Coach Mark Pro Parts Object Reference)
     Option 1:WBP Pilgrim Coach Mark Pro Parts Object Reference=Get WBP_PilgrimCoachMark_ProParts_LocalMP()
     Index=IsHamSandwichSessionActive()
    ) -> (ReturnValue:WBP Pilgrim Coach Mark Pro Parts Object Reference)) -> (Output_Get:WBP Pilgrim Coach Mark Pro Parts Object Reference)
  ]}
 UpdateSplitScreenVisibility {nodes [
   UpdateSplitScreenVisibility() {is=title K2Node_FunctionEntry}
   > n1 = Branch(Condition=IsHamSandwichSessionActive())
   n2 = IsVisible(self=Get WBP_Pilgrim_GreenRoom_SplitScreenGroup() -> (WBP_Pilgrim_GreenRoom_SplitScreenGroup:WBP Pilgrim Green Room Split Screen Group Object Reference))
   > n3 = Branch(Condition=@n2)
   SetVisibility(execute=@n3.else, self=Get WBP_Pilgrim_GreenRoom_SplitScreenGroup(), InVisibility=Visible)
   > SetVisibility(self=Get SafeZone_Solo(), InVisibility=Collapsed)
   Branch(execute=@n1.else, Condition=@n2)
   > SetVisibility(self=Get WBP_Pilgrim_GreenRoom_SplitScreenGroup(), InVisibility=Collapsed)
   > SetVisibility(self=Get SafeZone_Solo(), InVisibility=Visible)
  ]}
]
variables [
 AddingPlayerInPieEvent {type=AsyncAction_StartListeningToEvent, default=null}
 bIsCoachmarkInfoValid {type=boolean}
 ConfirmationScene {type=DynamicUIScene, default=/PilgrimQuickplay/Widgets/Mobile/Pilgrim_Quickplay_GreenRoom_Confirmation}
 EventRouter {type=GameplayEventRouterComponent, default=null}
 MyPlayerState {type=PlayerState, default=null}
 OwnerPlayerReference {type=PlayerController, default=null}
 ProPartsCoachMarkToUse {type=WBP_PilgrimCoachMark_ProParts_C, default=null}
]
widgets {
 info {
  inheritedWidgetCount=0
  namedSlotCount=1
  parentClass=/Script/PilgrimCoreUI.PilgrimGameWidgetBase
  rootWidgetClass=&Overlay
  widgetCount=24
 }
 widgets [
  CanvasPanel {widgetClassPath=&Overlay, widget=CanvasPanel}
  WBP_CaptureForPostBufferUpdate {
   widgetClassPath=&WBP_CaptureForPostBufferUpdate
   bIsVariable=true
   parent=CanvasPanel
   slot {class=OverlaySlot, properties {horizontalAlignment=HAlign_Fill, verticalAlignment=VAlign_Fill}}
  }
  VerticalBox_1 {
   widgetClassPath=&VerticalBox
   parent=CanvasPanel
   widget=VerticalBox_1
   slot {class=OverlaySlot, properties {horizontalAlignment=HAlign_Center, padding=<ltrb 0 40 0 0>}}
  }
  WBP_PilgrimQuickplay_GreenRoom_BandmateReadyDisplay {
   widgetClassPath=/PilgrimQuickplay/Widgets/GreenRoom/WBP_PilgrimQuickplay_GreenRoom_BandmateReadyDisplay
   bIsVariable=true
   parent=VerticalBox_1
   widget=WBP_PilgrimQuickplay_GreenRoom_BandmateReadyDisplay
   slot {class=VerticalBoxSlot}
  }
  WBP_Pilgrim_ReadyUpTimer {
   widgetClassPath=&/WBP_Pilgrim_ReadyUpTimer
   bIsVariable=true
   parent=VerticalBox_1
   widget=WBP_Pilgrim_ReadyUpTimer
   slot {class=VerticalBoxSlot, properties {horizontalAlignment=HAlign_Center, verticalAlignment=VAlign_Top, padding=<ltrb 0 10 0 0>}}
  }
  PlayerOwnershipWarning {
   widgetClassPath=&GridPanel
   bIsVariable=true
   parent=VerticalBox_1
   widget=PlayerOwnershipWarning
   properties {columnFill [1], rowFill [1]}
   slot {class=VerticalBoxSlot, properties {horizontalAlignment=HAlign_Center, padding=<ltrb 0 10 0 0>}}
  }
  PlayerIdentity {
   widgetClassPath=/PilgrimCore/Widgets/Player/WBP_Pilgrim_PlayerIdentity_Pill
   bIsVariable=true
   parent=PlayerOwnershipWarning
   widget=PlayerIdentity
   properties {design Preview Local Player Index=0}
   slot {class=GridSlot, properties {columnSpan=1, horizontalAlignment=HAlign_Left, rowSpan=1}}
  }
  SafeZone_SplitScreen {
   widgetClassPath=&SafeZone
   bIsVariable=true
   parent=CanvasPanel
   widget=SafeZone_SplitScreen
   properties {padLeft=false, padRight=false, padTop=false}
   slot {class=OverlaySlot, properties {horizontalAlignment=HAlign_Center, verticalAlignment=VAlign_Bottom, padding=<ltrb 64 64 64 12>}}
  }
  WBP_Pilgrim_GreenRoom_SplitScreenGroup {
   widgetClassPath=&/WBP_Pilgrim_GreenRoom_SplitScreenGroup
   bIsVariable=true
   parent=SafeZone_SplitScreen
   widget=WBP_Pilgrim_GreenRoom_SplitScreenGroup
   properties {visibility=Collapsed}
   slot {class=SafeZoneSlot}
  }
  CommonVisualAttachment_2 {
   widgetClassPath=&CommonVisualAttachment
   namedSlotHost=WBP_Pilgrim_GreenRoom_SplitScreenGroup
   widget=CommonVisualAttachment_2
   properties {contentAnchor=<-0.3 1>}
  }
  WBP_PilgrimCoachMark_ProParts_LocalMP {
   widgetClassPath=/PilgrimCore/Widgets/FTUE/CoachMarks/WBP_PilgrimCoachMark_ProParts
   bIsVariable=true
   parent=CommonVisualAttachment_2
   widget=WBP_PilgrimCoachMark_ProParts_LocalMP
   properties {desiredFocusWidget {widgetName=None}}
   slot {class=SizeBoxSlot}
  }
  WBP_Pilgrim_GreenRoom_BandwidthPlayerList {
   widgetClassPath=/PilgrimQuickplay/Widgets/GreenRoom/WBP_Pilgrim_GreenRoom_PlayerInfoList
   bIsVariable=true
   parent=CanvasPanel
   widget=WBP_Pilgrim_GreenRoom_BandwidthPlayerList
   slot {class=OverlaySlot, properties {padding=<ltrb 0 112 0 0>}}
  }
  SafeZone_Solo {
   widgetClassPath=&SafeZone
   bIsVariable=true
   parent=CanvasPanel
   widget=SafeZone_Solo
   properties {padLeft=false, padRight=false, padTop=false}
   slot {class=OverlaySlot, properties {verticalAlignment=VAlign_Bottom, padding=<ltrb 64 64 64 12>}}
  }
  VerticalBox_2 {widgetClassPath=&VerticalBox, parent=SafeZone_Solo, widget=VerticalBox_2, slot {class=SafeZoneSlot}}
  CommonVisualAttachment_1 {
   widgetClassPath=&CommonVisualAttachment
   parent=VerticalBox_2
   widget=CommonVisualAttachment_1
   properties {contentAnchor=<-0.3 1>}
   slot {class=VerticalBoxSlot}
  }
  WBP_PilgrimCoachMark_ProParts_Solo {
   widgetClassPath=&WBP_PilgrimCoachMark_ProParts
   bIsVariable=true
   parent=CommonVisualAttachment_1
   widget=WBP_PilgrimCoachMark_ProParts_Solo
   properties {desiredFocusWidget {widgetName=None}}
   slot {class=SizeBoxSlot}
  }
  WBP_Pilgrim_GreenRoom_BottomRightGroup {
   widgetClassPath=&/WBP_Pilgrim_GreenRoom_BottomRightGroup
   parent=VerticalBox_2
   widget=WBP_Pilgrim_GreenRoom_BottomRightGroup
   slot {class=VerticalBoxSlot}
  }
  WBP_Pilgrim_GreenRoom_TopCenterGroup {
   widgetClassPath=&/WBP_Pilgrim_GreenRoom_TopCenterGroup
   bIsVariable=true
   parent=CanvasPanel
   widget=WBP_Pilgrim_GreenRoom_TopCenterGroup
   properties {renderOpacity=0, visibility=Collapsed}
   slot {class=OverlaySlot}
  }
  VerticalBox_0 {widgetClassPath=&VerticalBox, parent=CanvasPanel, widget=VerticalBox_0, slot {class=OverlaySlot}}
  WBP_Pilgrim_GreenRoom_UpperRightGroup {
   widgetClassPath=&/WBP_Pilgrim_GreenRoom_UpperRightGroup
   bIsVariable=true
   parent=VerticalBox_0
   widget=WBP_Pilgrim_GreenRoom_UpperRightGroup
   slot {class=VerticalBoxSlot}
  }
  CommonVisualAttachment_0 {
   widgetClassPath=&CommonVisualAttachment
   parent=CanvasPanel
   widget=CommonVisualAttachment_0
   slot {class=OverlaySlot, properties {padding=<ltrb 25 110 0 0>}}
  }
  WBP_PilgrimCoachMark_StageTech {
   widgetClassPath=/PilgrimCore/Widgets/FTUE/CoachMarks/WBP_PilgrimCoachMark_StageTech
   bIsVariable=true
   parent=CommonVisualAttachment_0
   widget=WBP_PilgrimCoachMark_StageTech
   properties {desiredFocusWidget {widgetName=None}}
   slot {class=SizeBoxSlot}
  }
  SZ_PinnedQuests {
   widgetClassPath=&SafeZone
   bIsVariable=true
   parent=CanvasPanel
   widget=SZ_PinnedQuests
   properties {padBottom=false, padLeft=false}
   slot {class=OverlaySlot, properties {horizontalAlignment=HAlign_Right, padding=<ltrb 0 20 10 0>}}
  }
  PinnedQuestList {
   widgetClassPath=/Game/Athena/HUD/PinnedContentList
   bIsVariable=true
   parent=SZ_PinnedQuests
   widget=PinnedQuestList
   slot {class=SafeZoneSlot}
  }
 ]
}
