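# Pilgrim Quickplay green room screen widget, a layout variant with no bottom-right group.
# A canvas anchors a scaled, fixed-width bandmate ready display and a ready-up timer at top center, plus an upper-right group.
# The top-center group starts collapsed and transparent, and the bandmate display is set to hide when ready-up is unavailable.
# The event graph is empty; EventRouter and MyPlayerState variables are declared but unused here.
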WidgetBlueprint /PilgrimQuickplay/Widgets/WBP_Pilgrim_Greenroom_NoBottomRight
graphs [EventGraph {links [], nodes []}]
variables [
 EventRouter {type=GameplayEventRouterComponent, default=null}
 MyPlayerState {type=PlayerState, default=null}
]
widgets {
 info {inheritedWidgetCount=0, parentClass=&UserWidget, rootWidgetClass=&CanvasPanel, widgetCount=8}
 widgets [
  CanvasPanel {widgetClassPath=&CanvasPanel, bIsVariable=true, widget=CanvasPanel}
  SizeBox {
   widgetClassPath=&SizeBox
   bIsVariable=true
   parent=CanvasPanel
   widget=SizeBox
   properties {bOverride_WidthOverride=true, widthOverride=720}
   slot {
    class=CanvasPanelSlot
    properties {bAutoSize=true, layoutData {alignment=<0.5 0>, anchors {maximum=<0.5 0>, minimum=<0.5 0>}, offsets=<ltrb 0 60 0 0>}}
   }
  }
  ScaleBox_0 {widgetClassPath=&ScaleBox, parent=SizeBox, widget=ScaleBox_0, slot {class=SizeBoxSlot}}
  WBP_PilgrimQuickplay_GreenRoom_BandmateReadyDisplay {
   widgetClassPath=/PilgrimQuickplay/Widgets/GreenRoom/WBP_PilgrimQuickplay_GreenRoom_BandmateReadyDisplay
   bIsVariable=true
   parent=ScaleBox_0
   widget=WBP_PilgrimQuickplay_GreenRoom_BandmateReadyDisplay
   properties {hiddenWhenReadyUpIsUnavailable=true}
   slot {class=ScaleBoxSlot, properties {horizontalAlignment=HAlign_Fill, verticalAlignment=VAlign_Fill}}
  }
  WBP_Pilgrim_ReadyUpTimer {
   widgetClassPath=&/WBP_Pilgrim_ReadyUpTimer
   bIsVariable=true
   parent=CanvasPanel
   widget=WBP_Pilgrim_ReadyUpTimer
   slot {
    class=CanvasPanelSlot
    properties {bAutoSize=true, layoutData {alignment=<0.5 0>, anchors {maximum=<0.5 0>, minimum=<0.5 0>}, offsets=<ltrb 0 289 412.5352 145.3725>}}
   }
  }
  WBP_Pilgrim_GreenRoom_TopCenterGroup {
   widgetClassPath=&/WBP_Pilgrim_GreenRoom_TopCenterGroup
   bIsVariable=true
   parent=CanvasPanel
   widget=WBP_Pilgrim_GreenRoom_TopCenterGroup
   properties {renderOpacity=0, visibility=Collapsed}
   slot {
    class=CanvasPanelSlot
    properties {layoutData {alignment=<0 0>, anchors {maximum=<0.5104 0>, minimum=<0.5104 0>}, offsets=<ltrb -25.4174 0 0 0>}}
   }
  }
  VerticalBox_0 {
   widgetClassPath=&VerticalBox
   parent=CanvasPanel
   widget=VerticalBox_0
   slot {
    class=CanvasPanelSlot
    properties {bAutoSize=true, layoutData {alignment=<1 0>, anchors {maximum=<1 0>, minimum=<1 0>}, offsets=<ltrb -24 100 100 30>}}
   }
  }
  WBP_Pilgrim_GreenRoom_UpperRightGroup {
   widgetClassPath=&/WBP_Pilgrim_GreenRoom_UpperRightGroup
   bIsVariable=true
   parent=VerticalBox_0
   widget=WBP_Pilgrim_GreenRoom_UpperRightGroup
   slot {class=VerticalBoxSlot}
  }
 ]
}
