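# WBP_Pilgrim_SongLoadingOverlay is a Pilgrim Quickplay overlay widget that shows a "Waiting for all players to load song..." message.
# It derives from WBP_Pilgrim_UI_Results_Base and adds no graph logic or variables of its own.
# An Overlay root holds a VerticalBox with top padding, containing one centered, outlined and drop-shadowed CommonTextBlock named RoundDisplay, whose text is static here.
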WidgetBlueprint /PilgrimQuickplay/Widgets/WBP_Pilgrim_SongLoadingOverlay
graphs [EventGraph {links [], nodes []}]
variables []
widgets {
 info {
  inheritedWidgetCount=0
  parentClass=/PilgrimQuickplay/Widgets/Results/WBP_Pilgrim_UI_Results_Base.WBP_Pilgrim_UI_Results_Base_C
  rootWidgetClass=&Overlay
  widgetCount=3
 }
 widgets [
  VS_Overlay {widgetClassPath=&Overlay, bIsVariable=true, widget=VS_Overlay}
  UpperContent {
   widgetClassPath=&VerticalBox
   parent=VS_Overlay
   widget=UpperContent
   slot {class=OverlaySlot, properties {horizontalAlignment=HAlign_Fill, padding=<ltrb 0 35 0 0>}}
  }
  RoundDisplay {
   widgetClassPath=&CommonTextBlock
   bIsVariable=true
   parent=UpperContent
   widget=RoundDisplay
   properties {
    justification=Center
    text=Waiting for all players to load song...
    font {
     bForceMonospaced=false
     bMaterialIsStencil=false
     fontMaterial=null
     fontObject=&HeadingNow
     letterSpacing=0
     monospacedWidth=1
     size=41.25
     skewAmount=0
     typefaceFontName=86 Bold Italic
     outlineSettings {
      bApplyOutlineToDropShadows=false
      bMiteredCorners=false
      bSeparateFillAlpha=false
      outlineMaterial=null
      outlineSize=1
      outlineColor=<rgba 0.0003 0.0595 0.2961 1>
     }
    }
    shadowColorAndOpacity=<rgba 0 0 0 0.25>
    shadowOffset=<0 2.5>
    strikeBrush {
     bIsDynamicallyLoaded=false
     drawAs=NoDrawType
     imageType=NoImage
     mirroring=NoMirror
     resourceName=None
     resourceObject=null
     tiling=NoTile
     imageSize=<32 32>
     margin=<ltrb 0 0 0 0>
     outlineSettings {
      bUseBrushTransparency=false
      roundingType=HalfHeightRadius
      width=0
      color {colorUseRule=UseColor_Specified, specifiedColor=<rgba 0 0 0 0>}
      cornerRadii=<0 0 0 0>
     }
     tintColor {colorUseRule=UseColor_Specified, specifiedColor=<rgba 1 1 1 1>}
     uVRegion {bIsValid=false, max=<0 0>, min=<0 0>}
    }
   }
   slot {class=VerticalBoxSlot}
  }
 ]
}
