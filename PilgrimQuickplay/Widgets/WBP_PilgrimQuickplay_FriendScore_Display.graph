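# Quickplay widget that alerts the player when they beat a friend's leaderboard score, using a reward notification child widget.
# On construct it binds to the owning controller's leaderboard component Friends Surpassed event.
# If the Pilgrim.NewHighscoreAlertVisibility console variable is true, it shows uppercase text naming the highest surpassed friend, plays intro then outro animations, and collapses.
# A dispatcher exists to relay the notification's visibility changes.
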WidgetBlueprint /PilgrimQuickplay/Widgets/WBP_PilgrimQuickplay_FriendScore_Display
graphs [
 Display New Highscore Alert {nodes [
   n1 = Display New Highscore Alert!() -> (then, Surpassed Friend Entry:"Pilgrim Player Feed Result Entry Structure (by ref)") {is=title K2Node_FunctionEntry}
   > n2 = GetConsoleVariableBoolValue!(execute, VariableName:String=Pilgrim.NewHighscoreAlertVisibility) -> (then, ReturnValue:bool)
   > Branch(Condition=@n2)
   n3 = Break Pilgrim Player Feed Result Entry!(PilgrimPlayerFeedResultEntry:"Pilgrim Player Feed Result Entry Structure (by ref)"=@n1) -> (
    FriendName:String
    SongShortName:Name
    LeaderboardType:EPilgrimLeaderboardType Enum
    InstrumentsPlayed:Array of EPilgrimTrackType Enums
    PlayerHighscore:int
    FriendHighscore:int
    FriendAchievedScoreDate:Date Time Structure
   ) {is=title K2Node_BreakStruct}
   > Set Notification Text!(
    execute
    self:WBP Pilgrim Results Reward Notif Object Reference=Get WBP_Pilgrim_Results_RewardNotif() -> (WBP_Pilgrim_Results_RewardNotif:WBP Pilgrim Results Reward Notif Object Reference)
    NotificationText:Text=TextToUpper!(InText:"Text (by ref)"=Format Text!(Format="LOCTEXT(\"YOU BEAT {FriendName}\\'s SCORE!\")", FriendName:String=@n3) -> (Result)) -> (ReturnValue:Text)
   ) -> (then) {is=title K2Node_CallFunction}
   > Animate Alert()
  ]}
 EventGraph {nodes [
   n1 = OnFriendsScoreSurpassed!() -> (OutputDelegate, then, HighestFriendEntry:Pilgrim Player Feed Result Entry Structure, TotalFriendsSurpassed:int) {is=title K2Node_CustomEvent}
   > Display New Highscore Alert!(execute, self, Surpassed Friend Entry:"Pilgrim Player Feed Result Entry Structure (by ref)"=@n1.HighestFriendEntry) -> (then) {is=title K2Node_CallFunction}
   n2 = NotificationVisibilityChanged!() -> (OutputDelegate, then, InVisibility:ESlateVisibility Enum=Visible) {is=title K2Node_CustomEvent}
   > Call On Notification Visibility Changed!(execute:Exec, self:WBP Pilgrim Quickplay Friend Score Display Object Reference, NewVisibility:ESlateVisibility Enum=@n2.InVisibility) -> (then:Exec) {is=title K2Node_CallDelegate}
   Animate Alert() {is=title K2Node_CustomEvent}
   > SetVisibility(self=Get WBP_Pilgrim_Results_RewardNotif(), InVisibility=SelfHitTestInvisible)
   n3 = Get WBP_Pilgrim_Results_RewardNotif()
   > n4 = "Play Animation with Finished event (legacy)"!(
    execute:Exec
    Widget:User Widget Object Reference=@n3
    InAnimation:Widget Animation Object Reference=Get anim_Intro!(self:WBP Pilgrim Results Reward Notif Object Reference=@n3) -> (anim_Intro:Widget Animation Object Reference)
    StartAtTime:float=0
    NumLoopsToPlay:int=1
    PlayMode:EUMGSequencePlayMode Enum
    PlaybackSpeed:float=1
   ) -> (then:Exec, Result:UMGSequence Player Object Reference, Finished:Exec) {is=title K2Node_PlayAnimation}
   n5 = Get WBP_Pilgrim_Results_RewardNotif()
   n6 = "Play Animation with Finished event (legacy)"(
    execute=@n4.Finished
    Widget=@n5
    InAnimation=Get anim_Outro!(self:WBP Pilgrim Results Reward Notif Object Reference=@n5) -> (anim_Outro:Widget Animation Object Reference)
    StartAtTime=0
    NumLoopsToPlay=1
    PlaybackSpeed=1
   )
   SetVisibility(execute=@n6.Finished, self=Get WBP_Pilgrim_Results_RewardNotif(), InVisibility=Collapsed)
   Event Destruct()
   n7 = Create Event()
   > Unbind Event from On Visibility Changed(self:User Widget Object Reference=Get WBP_Pilgrim_Results_RewardNotif(), Delegate=@n7) {is=title K2Node_RemoveDelegate}
   Event Construct()
   > Bind Event to On Visibility Changed(self:User Widget Object Reference=Get WBP_Pilgrim_Results_RewardNotif(), Delegate=@n7) {is=title K2Node_AddDelegate}
   > SetVisibility(self=Get WBP_Pilgrim_Results_RewardNotif(), InVisibility=Collapsed)
   n8 = GetComponentByClass!(
    self:Actor Object Reference=GetControllerFromActor!(Actor:Actor Object Reference=GetOwningPlayer()) -> (ReturnValue:Controller Object Reference)
    ComponentClass:Actor Component Class Reference=/Script/PilgrimCoreRuntime.PilgrimLeaderboardPlayerComponent
   ) -> (ReturnValue:Pilgrim Leaderboard Player Component Object Reference)
   > Bind Event to On Friends Surpassed(self:Pilgrim Leaderboard Player Component Object Reference=@n8, Delegate=@n1) {is=title K2Node_AddDelegate}
   > Set CachedLeaderboardComponent(CachedLeaderboardComponent:Pilgrim Leaderboard Player Component Object Reference=@n8) -> (Output_Get:Pilgrim Leaderboard Player Component Object Reference)
  ]}
 OnNotificationVisibilityChanged {nodes [OnNotificationVisibilityChanged!() -> (then, NewVisibility:ESlateVisibility Enum=Visible) {is=title K2Node_FunctionEntry}]}
]
variables [
 CachedLeaderboardComponent {type=PilgrimLeaderboardPlayerComponent, default=null}
 OnNotificationVisibilityChanged {}
]
widgets {
 info {inheritedWidgetCount=0, parentClass=/Script/PilgrimCoreUI.PilgrimGameWidgetBase, rootWidgetClass=&Overlay, widgetCount=2}
 widgets [
  Overlay_0 {widgetClassPath=&Overlay, widget=Overlay_0}
  WBP_Pilgrim_Results_RewardNotif {
   widgetClassPath=/PilgrimQuickplay/Widgets/Results/WBP_Pilgrim_Results_RewardNotif
   bIsVariable=true
   parent=Overlay_0
   widget=WBP_Pilgrim_Results_RewardNotif
   properties {notificationType=FriendScore}
   slot {class=OverlaySlot}
  }
 ]
}
