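# Prioritized-widget wrapper around the Pilgrim Quickplay friend score display, so the prioritized widget system knows when it has content to show.
# One tick after construct, it binds to the display's notification visibility delegate and starts as inactive.
# Each visibility change sets active content to false for Collapsed or Hidden and true for visible states.
# On destruct it unbinds and clears active content.
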WidgetBlueprint /PilgrimQuickplay/Widgets/PrioritizedWidgets/WBP_PilgrimQuickplay_FriendScore_Display_Prioritized
graphs [
 EventGraph {nodes [
   n1 = OnNotificationVisibilityChanged!() -> (OutputDelegate, then, InVisibility:ESlateVisibility Enum=Visible) {is=title K2Node_CustomEvent}
   > UpdateHasActiveContent!(execute, self, MashPhaseVisibility:ESlateVisibility Enum=@n1.InVisibility) -> (then)
   Event Construct()
   > DelayUntilNextTick!(execute) -> (then)
   n2 = Create Event()
   n3 = Get WBP_PilgrimQuickplay_FriendScore_Display() -> (WBP_PilgrimQuickplay_FriendScore_Display:WBP Pilgrim Quickplay Friend Score Display Object Reference)
   > Bind Event to On Notification Visibility Changed(self:WBP Pilgrim Quickplay Friend Score Display Object Reference=@n3, Delegate=@n2) {is=title K2Node_AddDelegate}
   > UpdateHasActiveContent(MashPhaseVisibility=Collapsed)
   Event Destruct()
   > Unbind Event from On Notification Visibility Changed(self:WBP Pilgrim Quickplay Friend Score Display Object Reference=@n3, Delegate=@n2) {is=title K2Node_RemoveDelegate}
   > SetHasActiveContent!(execute, self:Fort Prioritized Widget Interface, bNewActive:bool=false) -> (then)
  ]}
 UpdateHasActiveContent {nodes [
   n1 = UpdateHasActiveContent!() -> (then, MashPhaseVisibility:ESlateVisibility Enum=Visible) {is=title K2Node_FunctionEntry}
   > n2 = Switch on ESlateVisibility!(execute, Selection:ESlateVisibility Enum=@n1) -> (Visible:Exec, Collapsed:Exec, Hidden:Exec, HitTestInvisible:Exec, SelfHitTestInvisible:Exec) {is=title K2Node_SwitchEnum}
   SetHasActiveContent(execute=[@n2.Collapsed, @n2.Hidden], bNewActive=false)
   SetHasActiveContent(execute=[@n2.HitTestInvisible, @n2.SelfHitTestInvisible, @n2.Visible], bNewActive=true)
  ]}
]
variables []
widgets {
 info {
  inheritedWidgetCount=1
  parentClass=/Script/FortniteUI.AthenaCommonPrioritizedWidget
  rootWidgetClass=/PilgrimQuickplay/Widgets/WBP_PilgrimQuickplay_FriendScore_Display.WBP_PilgrimQuickplay_FriendScore_Display_C
  widgetCount=2
 }
 widgets [
  WBP_PilgrimQuickplay_FriendScore_Display {
   widgetClassPath=/PilgrimQuickplay/Widgets/WBP_PilgrimQuickplay_FriendScore_Display
   bIsVariable=true
   widget=WBP_PilgrimQuickplay_FriendScore_Display
  }
  TutorialHighlight {widgetClassPath=&TutorialHighlight}
 ]
}
