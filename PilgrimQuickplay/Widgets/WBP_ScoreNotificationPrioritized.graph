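# WBP_ScoreNotificationPrioritized is a CommonUserWidget for Pilgrim Quickplay HUD score notifications.
# Its root FortPrioritizedWidgetSwitcher holds a high-score display and a friend-score display, each identified by its own Pilgrim.Hud.Notification gameplay tag.
# The high-score display overrides its prioritization to queue behind when contesting, while the friend-score display keeps default prioritization.
# The event graph contains only unconnected PreConstruct, Construct and Tick stubs, so the widget adds no Blueprint logic.
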WidgetBlueprint /PilgrimQuickplay/Widgets/WBP_ScoreNotificationPrioritized
graphs [EventGraph {nodes [
   Event PreConstruct()
   Event Construct()
   Event Tick()
  ]}]
variables []
widgets {
 info {
  inheritedWidgetCount=0
  parentClass=&CommonUserWidget
  rootWidgetClass=/Script/FortniteUI.FortPrioritizedWidgetSwitcher
  widgetCount=3
 }
 widgets [
  PrioritizedSwitcher_Notifications {widgetClassPath=&FortPrioritizedWidgetSwitcher, bIsVariable=true, widget=PrioritizedSwitcher_Notifications}
  WBP_PilgrimQuickplay_Highscore_Display_Prioritized {
   widgetClassPath=/PilgrimQuickplay/Widgets/PrioritizedWidgets/WBP_PilgrimQuickplay_Highscore_Display_Prioritized
   bIsVariable=true
   parent=PrioritizedSwitcher_Notifications
   widget=WBP_PilgrimQuickplay_Highscore_Display_Prioritized
   properties {elementTag {tagName=Pilgrim.Hud.Notification.HighScore}, prioritizationData {priority=Priority_2, whenContesting=QueueBehind}}
   slot {class=WidgetSwitcherSlot}
  }
  WBP_PilgrimQuickplay_FriendScore_Display_Prioritized {
   widgetClassPath=/PilgrimQuickplay/Widgets/PrioritizedWidgets/WBP_PilgrimQuickplay_FriendScore_Display_Prioritized
   bIsVariable=true
   parent=PrioritizedSwitcher_Notifications
   widget=WBP_PilgrimQuickplay_FriendScore_Display_Prioritized
   properties {elementTag {tagName=Pilgrim.Hud.Notification.FriendScore}}
   slot {class=WidgetSwitcherSlot}
  }
 ]
}
