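# Pilgrim Quickplay widget that shows a "new high score" notification when the owning player sets a personal best.
# On initialization it waits for the gameplay event router, then listens for NewHighscore events.
# Events for the owning player set solo, band, spotlight or all-time text on the child reward notification.
# Animate Alert chains its intro and outro animations, then collapses it; a dispatcher exposes notification visibility changes.
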WidgetBlueprint /PilgrimQuickplay/Widgets/WBP_PilgrimQuickplay_Highscore_Display
graphs [
 Display New Highscore Alert {nodes [
   n1 = Display New Highscore Alert!() -> (then, Highscore Event:"Pilgrim Event Player State New Highscore Structure (by ref)") {is=title K2Node_FunctionEntry}
   n2 = Break Pilgrim Event Player State New Highscore!(PilgrimEvent_PlayerState_NewHighscore:"Pilgrim Event Player State New Highscore Structure (by ref)"=@n1) -> (ForPlayer:Fort Player Pawn Object Reference, Score:int, BestType:EPilgrimPersonalBestType Enum) {is=title K2Node_BreakStruct}
   > Branch(Condition="Equal (Object)"(
     A=GetFortPlayerControllerFromActor!(Actor:Actor Object Reference=@n2) -> (ReturnValue:Fort Player Controller Object Reference)
     B=GetOwningPlayer()
    ))
   > n3 = Switch on EPilgrimPersonalBestType!(execute, Selection:EPilgrimPersonalBestType Enum=@n2.BestType) -> (SoloScore:Exec, BandScore:Exec, SpotlightScore:Exec, AlltimeScore:Exec) {is=title K2Node_SwitchEnum}
   n4 = Set HighScoreType(execute=@n3.AlltimeScore, HighScoreType:Text="LOCTEXT(\"ALL-TIME\")") -> (Output_Get:Text)
   Set Notification Text!(
    execute=[@n7.then, @n4.then, @n6.then, @n5.then]
    self:WBP Pilgrim Results Reward Notif Object Reference=Get WBP_Pilgrim_Results_RewardNotif() -> (WBP_Pilgrim_Results_RewardNotif:WBP Pilgrim Results Reward Notif Object Reference)
    NotificationText:Text=Format Text!(Format="LOCTEXT(\"NEW {HighScoreType} HIGH SCORE!\")", HighScoreType:Text=Get HighScoreType() -> (HighScoreType:Text)) -> (Result)
   ) -> (then) {is=title K2Node_CallFunction}
   > Animate Alert()
   n5 = Set HighScoreType(execute=@n3.BandScore, HighScoreType="LOCTEXT(\"BAND\")")
   n6 = Set HighScoreType(execute=@n3.SoloScore, HighScoreType="LOCTEXT(\"SOLO\")")
   n7 = Set HighScoreType(execute=@n3.SpotlightScore, HighScoreType="LOCTEXT(\"SPOTLIGHT\")")
  ]}
 EventGraph {nodes [
   n1 = Get WBP_Pilgrim_Results_RewardNotif()
   n2 = PlayAnimation(
    self=@n1
    InAnimation=Get anim_Intro!(self:WBP Pilgrim Results Reward Notif Object Reference=@n1) -> (anim_Intro:Widget Animation Object Reference)
   )
   n3 = TriggerNotificationOutro() {is=title K2Node_CustomEvent}
   n4 = Get WBP_Pilgrim_Results_RewardNotif()
   BindToAnimationFinished(execute=@n2.then, self=@n4, Animation=Get anim_Intro(self=@n4), Delegate=@n3)
   Animate Alert() {is=title K2Node_CustomEvent}
   > SetVisibility(self=Get WBP_Pilgrim_Results_RewardNotif(), InVisibility=SelfHitTestInvisible)
   n5 = Get WBP_Pilgrim_Results_RewardNotif()
   > n6 = "Play Animation with Finished event (legacy)"!(
    execute:Exec
    Widget:User Widget Object Reference=@n5
    InAnimation:Widget Animation Object Reference=Get anim_Intro(self=@n5)
    StartAtTime:float=0
    NumLoopsToPlay:int=1
    PlayMode:EUMGSequencePlayMode Enum
    PlaybackSpeed:float=1
   ) -> (then:Exec, Result:UMGSequence Player Object Reference, Finished:Exec) {is=title K2Node_PlayAnimation}
   n7 = Get WBP_Pilgrim_Results_RewardNotif()
   n8 = "Play Animation with Finished event (legacy)"(
    execute=@n6.Finished
    Widget=@n7
    InAnimation=Get anim_Outro!(self:WBP Pilgrim Results Reward Notif Object Reference=@n7) -> (anim_Outro:Widget Animation Object Reference)
    StartAtTime=0
    NumLoopsToPlay=1
    PlaybackSpeed=1
   )
   SetVisibility(execute=@n8.Finished, self=Get WBP_Pilgrim_Results_RewardNotif(), InVisibility=Collapsed)
   n9 = NotificationVisibilityChanged!() -> (OutputDelegate, then, InVisibility:ESlateVisibility Enum=Visible) {is=title K2Node_CustomEvent}
   > Call On Notification Visibility Changed!(execute:Exec, self:WBP Pilgrim Quickplay Highscore Display Object Reference, NewVisibility:ESlateVisibility Enum=@n9.InVisibility) -> (then:Exec) {is=title K2Node_CallDelegate}
   n10 = OnOutroAnimComplete() {is=title K2Node_CustomEvent}
   > SetVisibility(self=Get WBP_Pilgrim_Results_RewardNotif(), InVisibility=Collapsed)
   Event Destruct()
   n11 = Create Event()
   > Unbind Event from On Visibility Changed(self:User Widget Object Reference=Get WBP_Pilgrim_Results_RewardNotif(), Delegate=@n11) {is=title K2Node_RemoveDelegate}
   Event Construct()
   > Bind Event to On Visibility Changed(self:User Widget Object Reference=Get WBP_Pilgrim_Results_RewardNotif(), Delegate=@n11) {is=title K2Node_AddDelegate}
   > SetVisibility(self=Get WBP_Pilgrim_Results_RewardNotif(), InVisibility=Collapsed)
   Event OnInitialized()
   > n12 = PilgrimWaitForEventRouter!(execute:Exec, ContextActor:Actor Object Reference=GetOwningPlayer()) -> (then:Exec, OnEventRouterAvailable:Exec, EventRouter:Gameplay Event Router Component Object Reference) {is=title K2Node_AsyncAction}
   n13 = StartListeningToEvent!(
    execute:Exec=@n12.OnEventRouterAvailable
    Target:Gameplay Event Router Component Object Reference=@n12
    EventType:Script Struct Object Reference=/Script/PilgrimCoreRuntime.PilgrimEvent_PlayerState_NewHighscore
    Context:Object Reference
    EventBubblingRule:EEventBubblingRule Enum
   ) -> (
    then:Exec
    AsyncTaskProxy:Async Action Start Listening To Event Object Reference
    OnEventReceived:Exec
    Payload:Pilgrim Event Player State New Highscore Structure
   )
   > Add(
    TargetArray:Array of Async Action Start Listening To Event Object References=Get GameplayEventHandles() -> (GameplayEventHandles:Array of Async Action Start Listening To Event Object References)
    NewItem:"Async Action Start Listening To Event Object Reference (by ref)"=@n13
   ) {is=title K2Node_CallArrayFunction}
   Display New Highscore Alert!(execute=@n13.OnEventReceived, self, Highscore Event:"Pilgrim Event Player State New Highscore Structure (by ref)"=@n13.Payload) -> (then) {is=title K2Node_CallFunction}
   n14 = Get WBP_Pilgrim_Results_RewardNotif()
   BindToAnimationFinished(execute=@n16.then, self=@n14, Animation=Get anim_Outro(self=@n14), Delegate=@n10)
   n15 = Get WBP_Pilgrim_Results_RewardNotif()
   n16 = PlayAnimation(execute=@n3.then, self=@n15, InAnimation=Get anim_Outro(self=@n15))
  ]}
 OnNotificationVisibilityChanged {nodes [OnNotificationVisibilityChanged!() -> (then, NewVisibility:ESlateVisibility Enum=Visible) {is=title K2Node_FunctionEntry}]}
]
variables [OnNotificationVisibilityChanged {}]
widgets {
 info {inheritedWidgetCount=0, parentClass=/Script/PilgrimCoreUI.PilgrimGameWidgetBase, rootWidgetClass=&Overlay, widgetCount=2}
 widgets [
  Overlay_0 {widgetClassPath=&Overlay, widget=Overlay_0}
  WBP_Pilgrim_Results_RewardNotif {
   widgetClassPath=/PilgrimQuickplay/Widgets/Results/WBP_Pilgrim_Results_RewardNotif
   bIsVariable=true
   parent=Overlay_0
   widget=WBP_Pilgrim_Results_RewardNotif
   properties {notificationType=HighScore}
   slot {class=OverlaySlot}
  }
 ]
}
