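# WBP_ScoreNotificationStack is a CommonUserWidget in Pilgrim Quickplay that groups the friend score and high score displays into one vertical stack.
# An Overlay root holds a bottom-aligned VerticalBox, with the friend score display above the high score display and padding between them.
# The widget is layout only: its PreConstruct, Construct and Tick events are empty and it declares no variables of its own, so any behavior comes from the child widgets.
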WidgetBlueprint /PilgrimQuickplay/Widgets/WBP_ScoreNotificationStack
graphs [EventGraph {nodes [
   Event PreConstruct()
   Event Construct()
   Event Tick()
  ]}]
variables []
widgets {
 info {inheritedWidgetCount=0, parentClass=&CommonUserWidget, rootWidgetClass=&Overlay, widgetCount=4}
 widgets [
  Overlay_26 {widgetClassPath=&Overlay, widget=Overlay_26}
  VerticalBox_32 {
   widgetClassPath=&VerticalBox
   parent=Overlay_26
   widget=VerticalBox_32
   slot {class=OverlaySlot, properties {verticalAlignment=VAlign_Bottom}}
  }
  WBP_PilgrimQuickplay_FriendScore_Display {
   widgetClassPath=&/WBP_PilgrimQuickplay_FriendScore_Display
   bIsVariable=true
   parent=VerticalBox_32
   widget=WBP_PilgrimQuickplay_FriendScore_Display
   slot {class=VerticalBoxSlot, properties {verticalAlignment=VAlign_Bottom}}
  }
  WBP_PilgrimQuickplay_Highscore_Display {
   widgetClassPath=&/WBP_PilgrimQuickplay_Highscore_Display
   bIsVariable=true
   parent=VerticalBox_32
   widget=WBP_PilgrimQuickplay_Highscore_Display
   slot {class=VerticalBoxSlot, properties {verticalAlignment=VAlign_Bottom, padding=<ltrb 0 24 0 0>}}
  }
 ]
}
