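# Additive UI material that draws a tinted, flickering star flare with optional rotating sub-flares.
# The main cross shape is procedural: the product of centered absolute UV axes is inverted and raised to a power, which an eased, time-driven ping-pong modulates for flicker.
# Rotating radial segments, masked by a radial gradient, blend in as sub-flares.
# The combined mask brightens the Color parameter and, faded radially, drives adjustable opacity.
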Material /PilgrimCore/Widgets/Materials/M_UI_Habanero_Flare
graph {
 properties {bUsedWithStaticMesh=false, blendMode=BLEND_Additive, materialDomain=MD_UI, pixelDepthOffsetMode=PDOM_Legacy, usedShadingModels=Default Lit}
 nodes [
  n1 = VectorParameter() {parameterName=Color, defaultValue=<rgba 0.4841 0.6094 0.0391 1>}
  n2 = MaterialFunctionCall() {materialFunction=&GetUserInterfaceUV}
  n3 = Multiply(A=Abs(None=Subtract(A=@n2.Normalized UV) {constB=0.5})) {constB=0.8}
  n4 = ComponentMask(None=@n3) {r=true}
  n5 = ComponentMask(None=@n3) {g=true}
  n6 = ScalarParameter() {defaultValue=60, parameterName=Flare Strength}
  n7 = Time()
  n8 = ScalarParameter() {defaultValue=1, parameterName=Flickering Speed}
  n9 = MaterialFunctionCall("Value (0-1)"=Frac(None=Multiply(A=@n7, B=@n8))) {materialFunction=&MF_UI_EaseCurves}
  n10 = Min(A=@n9.EaseInOutSine, B=OneMinus(None=@n9.EaseInOutSine))
  n11 = Power(Base=OneMinus(None=Multiply(A=@n4, B=@n5)), Exp=LinearInterpolate(A=@n6, B=Add(A=@n6) {constB=-16}, Alpha=Multiply(A=@n10) {constB=2})) {constExponent=120}
  n12 = MaterialFunctionCall(UVs=TextureCoordinate(), "Rotation Angle (0-1)"=Multiply(A=Time(), B=ScalarParameter() {defaultValue=0.1, parameterName=Rotation Speed})) {materialFunction=&CustomRotator}
  n13 = Constant() {r=12}
  n14 = Constant() {r=0.75}
  n15 = Multiply(
   A=Power(Base=MaterialFunctionCall!(UV=@n12.Rotated Values, Segment Count=@n13, Segment Spacing=@n14, Fixed Spacing) {materialFunction=/FortUILibrary/MaterialFunctions/Basics/MF_UI_RadialSegments}.Segmented Lines Gradient) {constExponent=4}
   B=MaterialFunctionCall(GradientStart=Constant() {r=0.25}) {materialFunction=&MF_UI_RadialGradient}.Smooth
  )
  n16 = Saturate(None=LinearInterpolate(A=@n11, B=Add(A=@n11, B=@n15), Alpha=ScalarParameter() {defaultValue=1, parameterName=Display Sub Flares, sliderMax=1}))
  LinearInterpolate(A=@n1.RGB, B=Add(A=@n1.RGB, B=@n16), Alpha=@n16)
  n17 = MaterialFunctionCall() {materialFunction=&MF_UI_RadialGradient}
  Multiply(A=Saturate(None=Multiply(A=@n16, B=@n17.Smooth)), B=ScalarParameter() {defaultValue=1, parameterName=Opacity, sliderMax=1})
 ]
}
