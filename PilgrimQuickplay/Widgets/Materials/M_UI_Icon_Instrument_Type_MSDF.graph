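# Translucent UI material that draws one plastic-instrument icon (bass, guitar, mic, keys, drums, Pro variants) for Quickplay widgets.
# The scalar Icon parameter indexes a chain of saturated lerps across MSDF samples, with Size scaling UVs about the center.
# The blended MSDF becomes an anti-aliased SDF mask after selection.
# Static switches add a circular background (Bg Toggle) or modular-button selectable state colors and opacity (IsSelectableTabIcon).
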Material /PilgrimQuickplay/Widgets/Materials/M_UI_Icon_Instrument_Type_MSDF
graph {
 properties {blendMode=BLEND_Translucent, materialDomain=MD_UI, pixelDepthOffsetMode=PDOM_Legacy, usedShadingModels=Default Lit}
 nodes [
  n1 = TextureSample(UVs=NamedRerouteUsage()) {texture=/PilgrimQuickplay/Widgets/Textures/T_Icon_Instrument_DF_PlasticInstruments_Bass}
  Multiply(A=MaterialFunctionCall(
    DistanceField=MaterialFunctionCall!(MSDF RGB=@n1.RGB) {materialFunction=/FortUILibrary/MaterialFunctions/Shapes/Modifiers/MF_UI_MSDFtoSDF}.SDF
    Threshold=Constant() {r=0.22}
   ) {materialFunction=&MF_UI_AntiAliasing}.Outer)
  n2 = TextureSample(UVs=NamedRerouteUsage()) {texture=/PilgrimQuickplay/Widgets/Textures/T_Icon_Instrument_DF_PlasticInstruments_GuitarPlus}
  Multiply(A=MaterialFunctionCall(
    DistanceField=MaterialFunctionCall(MSDF RGB=@n2.RGB) {materialFunction=&MF_UI_MSDFtoSDF}.SDF
    Threshold=Constant() {r=0.22}
   ) {materialFunction=&MF_UI_AntiAliasing}.Outer)
  n3 = TextureSample(UVs=NamedRerouteUsage()) {texture=/PilgrimQuickplay/Widgets/Textures/T_Icon_Instrument_DF_PlasticInstruments_MicPlus}
  Multiply(A=MaterialFunctionCall(
    DistanceField=MaterialFunctionCall(MSDF RGB=@n3.RGB) {materialFunction=&MF_UI_MSDFtoSDF}.SDF
    Threshold=Constant() {r=0.22}
   ) {materialFunction=&MF_UI_AntiAliasing}.Outer)
  n4 = TextureSample(UVs=NamedRerouteUsage()) {texture=/PilgrimQuickplay/Widgets/Textures/T_Icon_Instrument_DF_PlasticInstruments_PianoPlus}
  Multiply(A=MaterialFunctionCall(
    DistanceField=MaterialFunctionCall(MSDF RGB=@n4.RGB) {materialFunction=&MF_UI_MSDFtoSDF}.SDF
    Threshold=Constant() {r=0.22}
   ) {materialFunction=&MF_UI_AntiAliasing}.Outer)
  n5 = TextureSample(UVs=NamedRerouteUsage()) {texture=/PilgrimQuickplay/Widgets/Textures/T_Icon_Instrument_DF_PlasticInstruments_Mic1}
  Multiply(A=MaterialFunctionCall(
    DistanceField=MaterialFunctionCall(MSDF RGB=@n5.RGB) {materialFunction=&MF_UI_MSDFtoSDF}.SDF
    Threshold=Constant() {r=0.22}
   ) {materialFunction=&MF_UI_AntiAliasing}.Outer)
  n6 = TextureSample(UVs=NamedRerouteUsage()) {texture=/PilgrimQuickplay/Widgets/Textures/T_Icon_Instrument_DF_PlasticInstruments_Drums}
  Multiply(A=MaterialFunctionCall(
    DistanceField=MaterialFunctionCall(MSDF RGB=@n6.RGB) {materialFunction=&MF_UI_MSDFtoSDF}.SDF
    Threshold=Constant() {r=0.22}
   ) {materialFunction=&MF_UI_AntiAliasing}.Outer)
  n7 = MaterialFunctionCall(UVs=TextureCoordinate(), Texture Scale=ScalarParameter() {defaultValue=1, parameterName=Size}) {materialFunction=&ScaleUVsByCenter}
  Multiply(A=TextureSample(UVs=@n7.UVs) {samplerType=SAMPLERTYPE_LinearColor, texture=/SparksUICommon/Textures/T_Icon_Instrument_Type01_SDF}.A)
  n8 = TextureSample(UVs=NamedRerouteUsage()) {texture=/PilgrimQuickplay/Widgets/Textures/T_Icon_Instrument_DF_PlasticInstruments_Guitar}
  Multiply(A=MaterialFunctionCall(
    DistanceField=MaterialFunctionCall(MSDF RGB=@n8.RGB) {materialFunction=&MF_UI_MSDFtoSDF}.SDF
    Threshold=Constant() {r=0.22}
   ) {materialFunction=&MF_UI_AntiAliasing}.Outer)
  n9 = TextureSample(UVs=NamedRerouteUsage()) {texture=/PilgrimQuickplay/Widgets/Textures/T_Icon_Instrument_DF_PlasticInstruments_Piano}
  Multiply(A=MaterialFunctionCall(
    DistanceField=MaterialFunctionCall(MSDF RGB=@n9.RGB) {materialFunction=&MF_UI_MSDFtoSDF}.SDF
    Threshold=Constant() {r=0.22}
   ) {materialFunction=&MF_UI_AntiAliasing}.Outer)
  n10 = TextureSample(UVs=NamedRerouteUsage()) {texture=/PilgrimQuickplay/Widgets/Textures/T_Icon_Instrument_DF_PlasticInstruments_DrumsPlus}
  Multiply(A=MaterialFunctionCall(
    DistanceField=MaterialFunctionCall(MSDF RGB=@n10.RGB) {materialFunction=&MF_UI_MSDFtoSDF}.SDF
    Threshold=Constant() {r=0.22}
   ) {materialFunction=&MF_UI_AntiAliasing}.Outer)
  n11 = TextureSample(UVs=NamedRerouteUsage()) {texture=/PilgrimQuickplay/Widgets/Textures/T_Icon_Instrument_DF_PlasticInstruments_DrumsPlus1}
  Multiply(A=MaterialFunctionCall(
    DistanceField=MaterialFunctionCall(MSDF RGB=@n11.RGB) {materialFunction=&MF_UI_MSDFtoSDF}.SDF
    Threshold=Constant() {r=0.22}
   ) {materialFunction=&MF_UI_AntiAliasing}.Outer)
  n12 = TextureSample(UVs=NamedRerouteUsage()) {texture=/PilgrimQuickplay/Widgets/Textures/T_Icon_Instrument_DF_PlasticInstruments_BassPlus}
  Multiply(A=MaterialFunctionCall(
    DistanceField=MaterialFunctionCall(MSDF RGB=@n12.RGB) {materialFunction=&MF_UI_MSDFtoSDF}.SDF
    Threshold=Constant() {r=0.22}
   ) {materialFunction=&MF_UI_AntiAliasing}.Outer)
  NamedRerouteDeclaration(None=@n7.UVs) {name=TextCoord, nodeColor=<rgba 0.125 1 0 1>}
  NamedRerouteDeclaration(None=@n1.RGB) {name=BassIcon, nodeColor=<rgba 0.2891 1 0 1>}
  NamedRerouteDeclaration(None=@n3.RGB) {name=ProMicIcon, nodeColor=<rgba 0.8281 0 1 1>}
  NamedRerouteDeclaration(None=@n4.RGB) {name=ProPianoIcon, nodeColor=<rgba 0 1 0.5078 1>}
  NamedRerouteDeclaration(None=@n7.UVs) {name=ScaleUVsByCenter, nodeColor=<rgba 1 0.6563 0 1>}
  NamedRerouteDeclaration(None=@n5.RGB) {name=MicIcon, nodeColor=<rgba 1 0 0.6094 1>}
  NamedRerouteDeclaration(None=@n6.RGB) {name=DrumsIcon, nodeColor=<rgba 0 0.9297 1 1>}
  NamedRerouteDeclaration(None=@n8.RGB) {name=GuitarIcon, nodeColor=<rgba 1 0.75 0 1>}
  NamedRerouteDeclaration(None=@n9.RGB) {name=PianoIcon, nodeColor=<rgba 0.4297 0 1 1>}
  NamedRerouteDeclaration(None=@n10.RGB) {name=ProDrumsIcon, nodeColor=<rgba 0 1 0.1094 1>}
  NamedRerouteDeclaration(None=@n11.RGB) {name=ProDrumsIcon1, nodeColor=<rgba 1 0 0.2109 1>}
  NamedRerouteDeclaration(None=@n12.RGB) {name=ProBassIcon, nodeColor=<rgba 0 0.5312 1 1>}
  NamedRerouteDeclaration(None=@n2.RGB) {name=ProGuitarIcon, nodeColor=<rgba 0.8516 1 0 1>}
  n13 = TextureSample(UVs=NamedRerouteUsage()) {
   samplerSource=SSM_Clamp_WorldGroupSettings
   samplerType=SAMPLERTYPE_LinearColor
   texture=/PilgrimQuickplay/Widgets/Textures/T_Icon_Instrument_DF_PlasticInstruments
  }
  Saturate(None=@n13.G)
  Saturate(None=TextureSample(UVs=NamedRerouteUsage()) {samplerType=SAMPLERTYPE_Alpha, texture=/PilgrimQuickplay/Widgets/Textures/T_Icon_Instrument_ProDrums_DF}.RGB)
  Saturate(None=TextureSample(UVs=NamedRerouteUsage()) {samplerType=SAMPLERTYPE_Alpha, texture=/PilgrimQuickplay/Widgets/Textures/T_Icon_Instrument_ProSnare_DF}.RGB)
  Saturate(None=@n13.R)
  Saturate(None=@n13.B)
  n14 = MaterialFunctionCall!(Idle, Idle Opacity, Hovered, Hovered Opacity, Focused, Focused Opacity, Pressed, Pressed Opacity, Disabled, Disabled Opacity) {materialFunction=/FortUILibrary/Systems/ModularButton/MaterialFunctions/MF_UIKit_ButtonStatesBase_Lerp}
  n15 = MaterialFunctionCall!(
   Base States=@n14.Color
   Base States Opacity=@n14.Color
   Selected_Idle
   Selected_Idle Opacity
   Selected_Hovered
   Selected_Hovered Opacity
   Selected_Focused
   Selected_Focused Opacity
   Selected_Pressed
   Selected_Pressed Opacity
   Selected_Disabled
   Selected_Disabled Opacity
  ) {materialFunction=/FortUILibrary/Systems/ModularButton/MaterialFunctions/MF_UIKit_ButtonStatesSelectable_Lerp}
  n16 = MaterialFunctionCall(UVs=NamedRerouteUsage(), Size=Subtract() {constB=0.13, desc=Size circle to icon}) {materialFunction=&MF_UI_Circle}
  n17 = LinearInterpolate(
   A=VectorParameter() {parameterName=Icon Color, defaultValue=<rgba 1 1 1 1>}.RGB
   B=VectorParameter() {parameterName=Bg Color, defaultValue=<rgba 0 0 0 1>}.RGB
   Alpha=@n16.Fill
  )
  n18 = VectorParameter() {parameterName=Icon Color, defaultValue=<rgba 1 1 1 1>}
  n19 = NamedRerouteUsage()
  n20 = NamedRerouteUsage()
  n21 = ScalarParameter() {defaultValue=10, parameterName=Icon}
  n22 = Saturate(None=@n21)
  n23 = Subtract(A=@n21)
  n24 = LinearInterpolate(
   A=LinearInterpolate(A=LinearInterpolate(A=@n19, B=@n20, Alpha=@n22), B=NamedRerouteUsage(), Alpha=Saturate(None=@n23))
   B=NamedRerouteUsage()
   Alpha=Saturate(None=Subtract(A=@n21) {constB=2})
  )
  n25 = Saturate(None=Subtract(A=@n21) {constB=3})
  n26 = NamedRerouteUsage()
  n27 = Subtract(A=@n21) {constB=4}
  n28 = LinearInterpolate(
   A=LinearInterpolate(A=LinearInterpolate(A=@n24, B=@n26, Alpha=@n25), B=NamedRerouteUsage(), Alpha=Saturate(None=@n27))
   B=NamedRerouteUsage()
   Alpha=Saturate(None=Subtract(A=@n21) {constB=5})
  )
  n29 = NamedRerouteUsage()
  n30 = Saturate(None=Subtract(A=@n21) {constB=6})
  n31 = Subtract(A=@n21) {constB=7}
  n32 = LinearInterpolate(
   A=LinearInterpolate(A=LinearInterpolate(A=@n28, B=@n29, Alpha=@n30), B=NamedRerouteUsage(), Alpha=Saturate(None=@n31))
   B=NamedRerouteUsage()
   Alpha=Saturate(None=Subtract(A=@n21) {constB=8})
  )
  n33 = NamedRerouteUsage()
  n34 = Saturate(None=Subtract(A=@n21) {constB=9})
  n35 = MaterialFunctionCall(
   DistanceField=MaterialFunctionCall(MSDF RGB=LinearInterpolate(A=@n32, B=@n33, Alpha=@n34)) {materialFunction=&MF_UI_MSDFtoSDF}.SDF
   Threshold=Constant() {r=0.5}
  ) {materialFunction=&MF_UI_AntiAliasing}
  StaticSwitchParameter(
   True=@n15.Color
   False=StaticSwitchParameter(True=LinearInterpolate(A=@n17, B=@n18.RGB, Alpha=@n35.Outer), False=@n18.RGB) {parameterName=Bg Toggle}
  ) {parameterName=IsSelectableTabIcon}
  n36 = Saturate(None=StaticSwitchParameter(True=Max(A=@n16.Fill, B=@n35.Outer), False=@n35.Outer) {parameterName=Bg Toggle})
  StaticSwitchParameter(True=Multiply(A=@n15.Opacity, B=@n36), False=@n36) {parameterName=IsSelectableTabIcon}
  TextureSample(UVs=NamedRerouteUsage()) {samplerType=SAMPLERTYPE_Alpha, texture=/PilgrimQuickplay/Widgets/Textures/T_Icon_Instrument_Keytar_DF}
 ]
}
