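# Translucent UI material for Pilgrim Quickplay, apparently a band overdrive-active indicator, that draws a soft-edged box rotated 45 degrees by default.
# MF_UI_Box produces an SDF from rotated UVs, which is inverted and smoothstepped into a mask.
# The mask multiplies the BG Color and, separately, a time-driven sine pulse.
# The pulse lerp reads PulseConstraints.R for both ends, so its value stays constant.
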Material /PilgrimQuickplay/Widgets/Materials/M_UI_Pilgrim_BandODActive
graph {
 properties {blendMode=BLEND_Translucent, materialDomain=MD_UI, pixelDepthOffsetMode=PDOM_Legacy, usedShadingModels=Default Lit}
 nodes [
  n1 = ScalarParameter() {defaultValue=0.7156, parameterName=Icon Smooth Min}
  n2 = ScalarParameter() {defaultValue=0.3132, parameterName=Icon Smooth Max}
  n3 = MaterialFunctionCall(UVs=NamedRerouteUsage(), Size=NamedRerouteUsage(), CornerRoundness=Constant() {r=-1}, OutlineThickness=NamedRerouteUsage()) {materialFunction=&MF_UI_Box}
  n4 = Saturate(None=SmoothStep(Min=@n1, Max=Add(A=@n1, B=@n2) {constB=0.21}, Value=OneMinus(None=@n3.SDF)) {constValue=0.1})
  Multiply(A=VectorParameter() {parameterName=BG Color, defaultValue=<rgba 1 0.5795 0 1>}.RGB, B=@n4)
  n5 = VectorParameter() {group=Pulse, parameterName=PulseConstraints, defaultValue=<rgba 0.5573 1 0.5573 1>}
  n6 = Time()
  n7 = ScalarParameter() {defaultValue=0.6, group=Pulse, parameterName=PulseSpeed}
  n8 = Add(A=Fmod(A=Multiply(A=@n6, B=@n7), B=Constant() {r=1}), B=ScalarParameter() {parameterName=PulseTimeOffset})
  Multiply(
   A=@n4
   B=LinearInterpolate(
    A=@n5.R
    B=@n5.R
    Alpha=MaterialFunctionCall(Value=@n8) {materialFunction=&LinearSine}.Linear Sine
   )
  )
  NamedRerouteDeclaration(None=MaterialFunctionCall(UVs=TextureCoordinate(), "Rotation Angle (0-1)"=ScalarParameter() {defaultValue=0.125, parameterName=Rotation}) {materialFunction=&CustomRotator}.Rotated Values) {name=Rotation, nodeColor=<rgba 0 0 0 1>}
  NamedRerouteDeclaration(None=ScalarParameter() {defaultValue=0.494, parameterName=Size}) {name=Size, nodeColor=<rgba 0 0 0 1>}
  NamedRerouteDeclaration(None=ScalarParameter() {defaultValue=0.001, parameterName=Stroke Size}) {name=Outline Size, nodeColor=<rgba 0 0 0 1>}
 ]
}
