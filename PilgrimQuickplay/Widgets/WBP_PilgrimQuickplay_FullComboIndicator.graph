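# Pilgrim Quickplay widget for full or perfect combo, defaulting to an animated gold "Flawless!" callout among badge, icon and ring variants in a visibility switcher.
# Design-time PreConstruct previews "100%" or "Flawless!" with effects via IsPerfectComboIndicator; Construct applies only font size and padding.
# TriggerPerfectRevealAnim plays a one-second reveal of text, outline, glow, rings, flare and particles.
# Nodes selecting full-combo visibility and icon-only mode are unwired.
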WidgetBlueprint /PilgrimQuickplay/Widgets/WBP_PilgrimQuickplay_FullComboIndicator
animations [Anim_PerfectComboReveal {
  binding_count=6
  length_seconds=1
  object=&self:Anim_PerfectComboReveal
  sequence {
   bindings [
    PulseRings {
     id=#1
     possessed_class=Image
     tracks [{
       type=MovieSceneFloatTrack
       display_name=Render Opacity
       sections [{type=MovieSceneFloatSection, end_frame=27, start_frame=0, channels [None {type=MovieSceneScriptingFloatChannel, keys [[6,0],[27,1]]}]}]
      }]
    }
    Image_GlowOutline {
     id=#2
     possessed_class=Image
     tracks [
      {
       type=MovieSceneFloatTrack
       display_name=Render Opacity
       sections [{type=MovieSceneFloatSection, channels [None {type=MovieSceneScriptingFloatChannel, keys [[3,0],[12,1]]}]}]
      }
      {
       type=MovieScene2DTransformTrack
       sections [{
         type=MovieScene2DTransformSection
         end_frame=27
         start_frame=0
         channels [
          Scale.X {type=MovieSceneScriptingFloatChannel, keys [[3,1.5],[9,0.95],[15,1]]}
          Scale.Y {type=MovieSceneScriptingFloatChannel, keys [[3,2],[9,1]]}
         ]
        }]
      }
     ]
    }
    PerfectCombo {
     id=#3
     possessed_class=CommonTextBlock
     tracks [
      {
       type=MovieScene2DTransformTrack
       sections [{
         type=MovieScene2DTransformSection
         channels [
          Translation.X {type=MovieSceneScriptingFloatChannel, const [0,1]}
          Translation.Y {type=MovieSceneScriptingFloatChannel, const [0,1]}
          Angle {type=MovieSceneScriptingFloatChannel, const [0,1]}
          Scale.X {type=MovieSceneScriptingFloatChannel, keys [[0,2],[10,0.95],[21,1]]}
          Scale.Y {type=MovieSceneScriptingFloatChannel, keys [[0,2],[10,0.95],[21,1]]}
          Shear.X {type=MovieSceneScriptingFloatChannel, const [-12,1]}
          Shear.Y {type=MovieSceneScriptingFloatChannel, const [0,1]}
         ]
        }]
      }
      {
       type=MovieSceneFloatTrack
       display_name=Render Opacity
       sections [{type=MovieSceneFloatSection, channels [None {type=MovieSceneScriptingFloatChannel, keys [[0,0],[12,1]]}]}]
      }
     ]
    }
    Flare {
     id=#4
     possessed_class=Image
     tracks [
      {
       type=MovieScene2DTransformTrack
       sections [{
         type=MovieScene2DTransformSection
         end_frame=15
         start_frame=0
         channels [
          Scale.X {type=MovieSceneScriptingFloatChannel, keys [[9,3],[15,2]]}
          Scale.Y {type=MovieSceneScriptingFloatChannel, const [1,1]}
         ]
        }]
      }
      {
       type=MovieSceneFloatTrack
       display_name=Render Opacity
       sections [{type=MovieSceneFloatSection, channels [None {type=MovieSceneScriptingFloatChannel, keys [[5.546,0],[8.546,1]]}]}]
      }
      {
       type=MovieSceneWidgetMaterialTrack
       sections [{type=MovieSceneParameterSection, channels [Alpha_Final {type=MovieSceneScriptingFloatChannel, keys [[3,4],[6,4],[60,0.5]]}]}]
      }
     ]
    }
    Glow {
     id=#5
     possessed_class=Image
     tracks [{
       type=MovieSceneFloatTrack
       display_name=Render Opacity
       sections [{type=MovieSceneFloatSection, end_frame=27, start_frame=0, channels [None {type=MovieSceneScriptingFloatChannel, keys [[6,0],[27,1]]}]}]
      }]
    }
    Particles {
     id=#6
     possessed_class=HorizontalBox
     tracks [{
       type=MovieSceneFloatTrack
       display_name=Render Opacity
       sections [{type=MovieSceneFloatSection, channels [None {type=MovieSceneScriptingFloatChannel, keys [[6,0],[9,1]]}]}]
      }]
    }
   ]
   display_rate {numerator=60}
  }
 }]
graphs [
 EventGraph {nodes [
   n1 = Event PreConstruct()
   > Branch(Condition=@n1.IsDesignTime)
   > SetFontSize!(
    execute
    self:Text Object Reference=Get PerfectCombo() -> (PerfectCombo:Common Text Object Reference)
    DisplayFontSize:float=Get PerfectComboDisplayFontSize() -> (PerfectComboDisplayFontSize:float)
   ) -> (then)
   > SetPadding!(
    execute
    self:Widget Switcher Slot Object Reference=SlotAsWidgetSwitcherSlot!(Widget:Widget Object Reference=Get PerfectOverlay() -> (PerfectOverlay:Overlay Object Reference)) -> (ReturnValue:Widget Switcher Slot Object Reference)
    InPadding_Left:float=Get In Padding Left() -> (In Padding Left:float)
    InPadding_Top:float=Get In Padding Top() -> (In Padding Top:float)
    InPadding_Right:float=Get In Padding Right() -> (In Padding Right:float)
    InPadding_Bottom:float=Get In Padding Bottom() -> (In Padding Bottom:float)
   ) -> (then)
   > n2 = Branch(Condition=Get IsPerfectComboIndicator() -> (IsPerfectComboIndicator:bool))
   > Set ActiveWidgetIndex!(
    execute
    ActiveWidgetIndex:int=1
    self:Widget Switcher Object Reference=Get TextType() -> (TextType:Common Activatable Widget Switcher Object Reference)
   ) -> (then, Output_Get:int=0)
   > SetVisibility(self=Get FX() -> (FX:Overlay Object Reference), InVisibility=SelfHitTestInvisible)
   Set ActiveWidgetIndex(execute=@n2.else, ActiveWidgetIndex=0, self=Get TextType()) -> (Output_Get=0)
   > SetVisibility(self=Get FX(), InVisibility=Collapsed)
   Event Construct()
   > SetFontSize(self=Get PerfectCombo(), DisplayFontSize=Get PerfectComboDisplayFontSize())
   > SetPadding(
    self=SlotAsWidgetSwitcherSlot(Widget=Get PerfectOverlay())
    InPadding_Left=Get In Padding Left()
    InPadding_Top=Get In Padding Top()
    InPadding_Right=Get In Padding Right()
    InPadding_Bottom=Get In Padding Bottom()
   )
   SetVisibility(InVisibility=Select(
     Option 0=Hidden
     Option 1=SelfHitTestInvisible
     Index=Break Pilgrim Game Progress!(PilgrimGameProgress:"Pilgrim Game Progress Structure (by ref)") -> (
      NumHits:int
      NumMisses:int
      NumStreakBreakingMisses:int
      NumPasses:int
      bHasFullCombo:bool=~
      bHasPerfectCombo:bool=~
      CurrentTick:int
      NoteStreak:int
      CurrentScore:float=~
      RawCurrentScore:float=~
      CurrentModifierBonus:float=~
      CurrentStar:int
      NextStarProgress:float=~
      StreakScoreMultiplier:int
      NotesPastLastMultiplier:int
      NotesToNextMultiplier:int
      CurrentHealth:float=~
      IsOverdriveActive:bool=~
      CurrentOverdrive:float=~
      CurrentOverdriveSectionNum:int
      BandScoreContribution:float=~
      BandModifierBonusContribution:float=~
      OverdriveScoringState:EOverdriveScoringState Enum
      MaxScore:float=~
      RawMaxScore:float=~
      MaxStreakLength:int
      AverageStreakMultiplierPerHitOrPass:float=~
      bIsAtMaxStreak:bool=~
      CurrentSectionCompletionPercentage:float=~
     ) {is=title K2Node_BreakStruct}.bHasFullCombo
    ))
   SetActiveWidget(
    self=Get CommonVisibilitySwitcher() -> (CommonVisibilitySwitcher:Common Visibility Switcher Object Reference)
    Widget=Select(
     Option 0:Widget Object Reference=Get Notif_Full() -> (Notif_Full:Overlay Object Reference)
     Option 1:Widget Object Reference=Get Notif_IconOnly() -> (Notif_IconOnly:Image Object Reference)
     Index=Get ShowIconOnly() -> (ShowIconOnly:bool)
    ) -> (ReturnValue:Widget Object Reference)
   )
  ]}
 TriggerPerfectRevealAnim {nodes [
   TriggerPerfectRevealAnim() {is=title K2Node_FunctionEntry}
   > PlayAnimation(InAnimation=Get Anim_PerfectComboReveal() -> (Anim_PerfectComboReveal:Widget Animation Object Reference))
  ]}
]
variables [
 Active Widget Index {type=integer}
 In Padding {type=Margin, default=<ltrb 0 0 0 0>}
 In Padding Bottom {type=number}
 In Padding Left {type=number}
 In Padding Right {type=number}
 In Padding Top {type=number}
 IsPerfectComboIndicator {type=boolean}
 Padding_0 {type=Margin, default=<ltrb 0 0 0 0>}
 PerfectComboDisplayFontSize {type=number, default=20}
 ShowIconOnly {type=boolean}
]
widgets {
 info {
  inheritedWidgetCount=0
  parentClass=/Script/PilgrimCoreUI.PilgrimGameWidgetBase
  rootWidgetClass=&CommonVisibilitySwitcher
  widgetCount=26
 }
 widgets [
  CommonVisibilitySwitcher {
   widgetClassPath=&CommonVisibilitySwitcher
   bIsVariable=true
   widget=CommonVisibilitySwitcher
   properties {activeWidgetIndex=3}
  }
  Notif_Full {
   widgetClassPath=&Overlay
   bIsVariable=true
   parent=CommonVisibilitySwitcher
   widget=Notif_Full
   slot {class=CommonVisibilitySwitcherSlot}
  }
  Image_86 {
   widgetClassPath=&Image
   bIsVariable=true
   parent=Notif_Full
   widget=Image_86
   properties {brush {
     bIsDynamicallyLoaded=false
     drawAs=Image
     imageType=NoImage
     mirroring=NoMirror
     resourceName=None
     resourceObject=/PilgrimQuickplay/Widgets/Materials/MI_UI_Creative_CoreGamer_RoundedBox_FCBG
     tiling=NoTile
     imageSize=<16 16>
     margin=<ltrb 0 0 0 0>
     outlineSettings {
      bUseBrushTransparency=false
      roundingType=HalfHeightRadius
      width=0
      color {colorUseRule=UseColor_Specified, specifiedColor=<rgba 0 0 0 0>}
      cornerRadii=<0 0 0 0>
     }
     tintColor {colorUseRule=UseColor_Specified, specifiedColor=<rgba 1 1 1 1>}
     uVRegion {bIsValid=false, max=<0 0>, min=<0 0>}
    }}
   slot {class=OverlaySlot, properties {horizontalAlignment=HAlign_Fill, verticalAlignment=VAlign_Fill}}
  }
  SizeBox_1 {
   widgetClassPath=&SizeBox
   parent=Notif_Full
   widget=SizeBox_1
   properties {bOverride_HeightOverride=true, heightOverride=20}
   slot {class=OverlaySlot, properties {horizontalAlignment=HAlign_Fill, verticalAlignment=VAlign_Fill, padding=<ltrb 2 0 2 0>}}
  }
  HorizontalBox_0 {widgetClassPath=&HorizontalBox, parent=SizeBox_1, widget=HorizontalBox_0, slot {class=SizeBoxSlot}}
  FullComboIndicatorIcon {
   widgetClassPath=&Image
   bIsVariable=true
   parent=HorizontalBox_0
   widget=FullComboIndicatorIcon
   properties {brush {
     bIsDynamicallyLoaded=false
     drawAs=Image
     imageType=NoImage
     mirroring=NoMirror
     resourceName=None
     resourceObject=/PilgrimQuickplay/Widgets/Materials/Mi_UI_FullCombo_Dark
     tiling=NoTile
     imageSize=<20 20>
     margin=<ltrb 0 0 0 0>
     outlineSettings {
      bUseBrushTransparency=false
      roundingType=HalfHeightRadius
      width=0
      color {colorUseRule=UseColor_Specified, specifiedColor=<rgba 0 0 0 0>}
      cornerRadii=<0 0 0 0>
     }
     tintColor {colorUseRule=UseColor_Specified, specifiedColor=<rgba 1 1 1 1>}
     uVRegion {bIsValid=false, max=<0 0>, min=<0 0>}
    }}
   slot {class=HorizontalBoxSlot, properties {horizontalAlignment=HAlign_Center, verticalAlignment=VAlign_Center}}
  }
  UEFN_TextBlock_C_137 {
   widgetClassPath=&UEFN_TextBlock
   parent=HorizontalBox_0
   widget=UEFN_TextBlock_C_137
   properties {
    text="FLAWLESS!"
    colorAndOpacity {colorUseRule=UseColor_Specified, specifiedColor=<rgba 0 0 0 1>}
    font {
     bForceMonospaced=false
     bMaterialIsStencil=false
     fontMaterial=null
     fontObject=&BurbankSmall
     letterSpacing=0
     monospacedWidth=1
     size=12
     skewAmount=0.2
     typefaceFontName=Black
     outlineSettings {
      bApplyOutlineToDropShadows=false
      bMiteredCorners=false
      bSeparateFillAlpha=false
      outlineMaterial=null
      outlineSize=0
      outlineColor=<rgba 0 0 0 1>
     }
    }
   }
   slot {class=HorizontalBoxSlot, properties {verticalAlignment=VAlign_Center, padding=<ltrb 2 0 0 0>}}
  }
  Notif_IconOnly {
   widgetClassPath=&Image
   bIsVariable=true
   parent=CommonVisibilitySwitcher
   widget=Notif_IconOnly
   properties {brush {
     bIsDynamicallyLoaded=false
     drawAs=Image
     imageType=NoImage
     mirroring=NoMirror
     resourceName=None
     resourceObject=/PilgrimQuickplay/Widgets/Materials/MI_UI_FullCombo
     tiling=NoTile
     imageSize=<20 20>
     margin=<ltrb 0 0 0 0>
     outlineSettings {
      bUseBrushTransparency=false
      roundingType=HalfHeightRadius
      width=0
      color {colorUseRule=UseColor_Specified, specifiedColor=<rgba 0 0 0 0>}
      cornerRadii=<0 0 0 0>
     }
     tintColor {colorUseRule=UseColor_Specified, specifiedColor=<rgba 1 1 1 1>}
     uVRegion {bIsValid=false, max=<0 0>, min=<0 0>}
    }}
   slot {class=CommonVisibilitySwitcherSlot, properties {horizontalAlignment=HAlign_Center, verticalAlignment=VAlign_Center}}
  }
  FullComboRing {
   widgetClassPath=&Image
   bIsVariable=true
   parent=CommonVisibilitySwitcher
   widget=FullComboRing
   properties {brush {
     bIsDynamicallyLoaded=false
     drawAs=Image
     imageType=NoImage
     mirroring=NoMirror
     resourceName=None
     resourceObject=/PilgrimCore/Widgets/Materials/MI_UI_Pilgrim_MultiplierBackground_FullCombo
     tiling=NoTile
     imageSize=<80 80>
     margin=<ltrb 0 0 0 0>
     outlineSettings {
      bUseBrushTransparency=false
      roundingType=HalfHeightRadius
      width=0
      color {colorUseRule=UseColor_Specified, specifiedColor=<rgba 0 0 0 0>}
      cornerRadii=<0 0 0 0>
     }
     tintColor {colorUseRule=UseColor_Specified, specifiedColor=<rgba 1 1 1 1>}
     uVRegion {bIsValid=false, max=<0 0>, min=<0 0>}
    }}
   slot {class=CommonVisibilitySwitcherSlot, properties {horizontalAlignment=HAlign_Center, verticalAlignment=VAlign_Center}}
  }
  Flawless {widgetClassPath=&Overlay, parent=CommonVisibilitySwitcher, widget=Flawless, slot {class=CommonVisibilitySwitcherSlot}}
  FX {
   widgetClassPath=&Overlay
   bIsVariable=true
   parent=Flawless
   widget=FX
   properties {visibility=Collapsed}
   slot {class=OverlaySlot, properties {horizontalAlignment=HAlign_Fill, verticalAlignment=VAlign_Fill}}
  }
  FortVisualAttachment_0 {
   widgetClassPath=&FortVisualAttachment
   parent=FX
   widget=FortVisualAttachment_0
   properties {bOverride_WidthOverride=true, widthOverride=448, contentAnchor=<0.5 0.5>}
   slot {class=OverlaySlot, properties {horizontalAlignment=HAlign_Center, verticalAlignment=VAlign_Center}}
  }
  Overlay_1 {
   widgetClassPath=&Overlay
   parent=FortVisualAttachment_0
   widget=Overlay_1
   slot {class=SizeBoxSlot, properties {verticalAlignment=VAlign_Center}}
  }
  Flare {
   widgetClassPath=&Image
   parent=Overlay_1
   widget=Flare
   properties {
    visibility=SelfHitTestInvisible
    brush {
     bIsDynamicallyLoaded=false
     drawAs=Image
     imageType=NoImage
     mirroring=NoMirror
     resourceName=None
     resourceObject=/PilgrimCore/Widgets/Materials/MI_UI_Pilgrim_ODText_Flare
     tiling=NoTile
     imageSize=<32 32>
     margin=<ltrb 0 0 0 0>
     outlineSettings {
      bUseBrushTransparency=false
      roundingType=HalfHeightRadius
      width=0
      color {colorUseRule=UseColor_Specified, specifiedColor=<rgba 0 0 0 0>}
      cornerRadii=<0 0 0 0>
     }
     tintColor {colorUseRule=UseColor_Specified, specifiedColor=<rgba 1 1 1 1>}
     uVRegion {bIsValid=false, max=<0 0>, min=<0 0>}
    }
    renderTransform {angle=0, scale=<1.5 1>, shear=<0 0>, translation=<0 0>}
   }
   slot {class=OverlaySlot, properties {horizontalAlignment=HAlign_Fill, verticalAlignment=VAlign_Center}}
  }
  Particles {
   widgetClassPath=&HorizontalBox
   parent=Overlay_1
   widget=Particles
   slot {class=OverlaySlot, properties {horizontalAlignment=HAlign_Fill, verticalAlignment=VAlign_Center}}
  }
  Image_259 {
   widgetClassPath=&Image
   parent=Particles
   widget=Image_259
   properties {
    visibility=SelfHitTestInvisible
    brush {
     bIsDynamicallyLoaded=false
     drawAs=Image
     imageType=NoImage
     mirroring=NoMirror
     resourceName=None
     resourceObject=/Game/UI/Foundation/Materials/M_FX_TriangleScroll_tileadjustment
     tiling=NoTile
     imageSize=<24 24>
     margin=<ltrb 0 0 0 0>
     outlineSettings {
      bUseBrushTransparency=false
      roundingType=HalfHeightRadius
      width=0
      color {colorUseRule=UseColor_Specified, specifiedColor=<rgba 0 0 0 0>}
      cornerRadii=<0 0 0 0>
     }
     tintColor {colorUseRule=UseColor_Specified, specifiedColor=<rgba 1 1 1 1>}
     uVRegion {bIsValid=false, max=<0 0>, min=<0 0>}
    }
    renderTransform {angle=0, scale=<-1 1>, shear=<0 0>, translation=<0 0>}
   }
   slot {class=HorizontalBoxSlot, properties {verticalAlignment=VAlign_Center, size {sizeRule=Fill, value=1}}}
  }
  Image {
   widgetClassPath=&Image
   parent=Particles
   widget=Image
   properties {
    visibility=SelfHitTestInvisible
    brush {
     bIsDynamicallyLoaded=false
     drawAs=Image
     imageType=NoImage
     mirroring=NoMirror
     resourceName=None
     resourceObject=&M_FX_TriangleScroll_tileadjustment
     tiling=NoTile
     imageSize=<24 24>
     margin=<ltrb 0 0 0 0>
     outlineSettings {
      bUseBrushTransparency=false
      roundingType=HalfHeightRadius
      width=0
      color {colorUseRule=UseColor_Specified, specifiedColor=<rgba 0 0 0 0>}
      cornerRadii=<0 0 0 0>
     }
     tintColor {colorUseRule=UseColor_Specified, specifiedColor=<rgba 1 1 1 1>}
     uVRegion {bIsValid=false, max=<0 0>, min=<0 0>}
    }
    renderTransform {angle=0, scale=<1 -1>, shear=<0 0>, translation=<0 0>}
   }
   slot {class=HorizontalBoxSlot, properties {verticalAlignment=VAlign_Center, size {sizeRule=Fill, value=1}}}
  }
  TextType {
   widgetClassPath=&CommonActivatableWidgetSwitcher
   bIsVariable=true
   parent=Flawless
   widget=TextType
   properties {activeWidgetIndex=1}
   slot {class=OverlaySlot, properties {horizontalAlignment=HAlign_Center, verticalAlignment=VAlign_Center, padding=<ltrb 8 0 8 0>}}
  }
  HorizontalBox_1 {
   widgetClassPath=&HorizontalBox
   parent=TextType
   widget=HorizontalBox_1
   properties {flowDirectionPreference=Culture}
   slot {class=WidgetSwitcherSlot}
  }
  100 {
   widgetClassPath=&CommonTextBlock
   parent=HorizontalBox_1
   widget=100
   properties {
    justification=Center
    text=100
    colorAndOpacity {colorUseRule=UseColor_Specified, specifiedColor=<rgba 1 0.8235 0.0196 1>}
    font {
     bForceMonospaced=false
     bMaterialIsStencil=false
     fontMaterial=null
     fontObject=&HeadingNow
     letterSpacing=0
     monospacedWidth=1
     size=15
     skewAmount=0
     typefaceFontName=75 Medium
     outlineSettings {
      bApplyOutlineToDropShadows=false
      bMiteredCorners=false
      bSeparateFillAlpha=false
      outlineMaterial=null
      outlineSize=0
      outlineColor=<rgba 1 1 1 1>
     }
    }
    renderTransform {angle=0, scale=<1 1>, shear=<-12 0>, translation=<0 0>}
    shadowOffset=<0 0>
   }
   slot {class=HorizontalBoxSlot}
  }
  100_1 {
   widgetClassPath=&CommonTextBlock
   parent=HorizontalBox_1
   widget=100_1
   properties {
    justification=Center
    text=%
    colorAndOpacity {colorUseRule=UseColor_Specified, specifiedColor=<rgba 1 0.8235 0.0196 1>}
    font {
     bForceMonospaced=false
     bMaterialIsStencil=false
     fontMaterial=null
     fontObject=&HeadingNow
     letterSpacing=0
     monospacedWidth=1
     size=15
     skewAmount=0
     typefaceFontName=75 Medium
     outlineSettings {
      bApplyOutlineToDropShadows=false
      bMiteredCorners=false
      bSeparateFillAlpha=false
      outlineMaterial=null
      outlineSize=0
      outlineColor=<rgba 1 1 1 1>
     }
    }
    renderTransform {angle=0, scale=<1 1>, shear=<-12 0>, translation=<0 0>}
    shadowOffset=<0 0>
   }
   slot {class=HorizontalBoxSlot}
  }
  PerfectOverlay {widgetClassPath=&Overlay, bIsVariable=true, parent=TextType, widget=PerfectOverlay, slot {class=WidgetSwitcherSlot}}
  Glow {
   widgetClassPath=&Image
   parent=PerfectOverlay
   widget=Glow
   properties {
    visibility=SelfHitTestInvisible
    brush {
     bIsDynamicallyLoaded=false
     drawAs=Box
     imageType=NoImage
     mirroring=NoMirror
     resourceName=None
     resourceObject=/PilgrimCore/Widgets/Materials/M_UI_Overdrive_ProgressMeter_Glow_Activated
     tiling=NoTile
     imageSize=<32 32>
     margin=<ltrb 0.5 0.5 0.5 0.5>
     outlineSettings {
      bUseBrushTransparency=false
      roundingType=HalfHeightRadius
      width=0
      color {colorUseRule=UseColor_Specified, specifiedColor=<rgba 0 0 0 0>}
      cornerRadii=<0 0 0 0>
     }
     tintColor {colorUseRule=UseColor_Specified, specifiedColor=<rgba 1 1 1 1>}
     uVRegion {bIsValid=false, max=<0 0>, min=<0 0>}
    }
    renderTransform {angle=0, scale=<1 1>, shear=<-10 0>, translation=<0 0>}
   }
   slot {class=OverlaySlot, properties {horizontalAlignment=HAlign_Fill, verticalAlignment=VAlign_Fill, padding=<ltrb -8 -2 -8 -2>}}
  }
  PulseRings {
   widgetClassPath=&Image
   parent=PerfectOverlay
   widget=PulseRings
   properties {
    visibility=SelfHitTestInvisible
    brush {
     bIsDynamicallyLoaded=false
     drawAs=Box
     imageType=NoImage
     mirroring=NoMirror
     resourceName=None
     resourceObject=/PilgrimCore/Widgets/Materials/M_UI_StreakMeter_FX_Rings
     tiling=NoTile
     imageSize=<68 68>
     margin=<ltrb 0.5 0.5 0.5 0.5>
     outlineSettings {
      bUseBrushTransparency=false
      roundingType=HalfHeightRadius
      width=0
      color {colorUseRule=UseColor_Specified, specifiedColor=<rgba 0 0 0 0>}
      cornerRadii=<0 0 0 0>
     }
     tintColor {colorUseRule=UseColor_Specified, specifiedColor=<rgba 1 1 1 1>}
     uVRegion {bIsValid=false, max=<0 0>, min=<0 0>}
    }
    renderTransform {angle=0, scale=<1 1>, shear=<-10 0>, translation=<0 0>}
   }
   slot {class=OverlaySlot, properties {horizontalAlignment=HAlign_Fill, verticalAlignment=VAlign_Fill, padding=<ltrb -26 -24 -26 -24>}}
  }
  PerfectCombo {
   widgetClassPath=&CommonTextBlock
   bIsVariable=true
   parent=PerfectOverlay
   widget=PerfectCombo
   properties {
    justification=Center
    text="Flawless!"
    textTransformPolicy=ToUpper
    colorAndOpacity {colorUseRule=UseColor_Specified, specifiedColor=<rgba 1 0.8235 0.0196 1>}
    font {
     bForceMonospaced=false
     bMaterialIsStencil=false
     fontMaterial=null
     fontObject=&HeadingNow
     letterSpacing=0
     monospacedWidth=1
     size=15
     skewAmount=0
     typefaceFontName=75 Medium
     outlineSettings {
      bApplyOutlineToDropShadows=false
      bMiteredCorners=false
      bSeparateFillAlpha=true
      outlineMaterial=&M-FontShadow
      outlineSize=2
      outlineColor=<rgba 0.0125 0.0021 0 1>
     }
    }
    renderTransform {angle=0, scale=<1 1>, shear=<-12 0>, translation=<0 0>}
    shadowOffset=<0 0>
   }
   slot {class=OverlaySlot}
  }
  Image_GlowOutline {
   widgetClassPath=&Image
   bIsVariable=true
   parent=Flawless
   widget=Image_GlowOutline
   properties {
    brush {
     bIsDynamicallyLoaded=false
     drawAs=Image
     imageType=NoImage
     mirroring=NoMirror
     resourceName=None
     resourceObject=/PilgrimCore/Widgets/Materials/M_UI_Pilgrim_Rounded_Outline
     tiling=NoTile
     imageSize=<36 36>
     margin=<ltrb 0.5 0.5 0.5 0.5>
     outlineSettings {
      bUseBrushTransparency=false
      roundingType=HalfHeightRadius
      width=0
      color {colorUseRule=UseColor_Specified, specifiedColor=<rgba 0 0 0 0>}
      cornerRadii=<0 0 0 0>
     }
     tintColor {colorUseRule=UseColor_Specified, specifiedColor=<rgba 1 1 1 1>}
     uVRegion {bIsValid=false, max=<0 0>, min=<0 0>}
    }
    renderTransform {angle=0, scale=<1 1>, shear=<-8 0>, translation=<0 0>}
   }
   slot {class=OverlaySlot, properties {horizontalAlignment=HAlign_Fill, verticalAlignment=VAlign_Fill}}
  }
 ]
}
