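# Pilgrim Quickplay band Overdrive HUD widget showing which bandmates have Overdrive active, the band multiplier, and points earned during it.
# On battle init it fills avatar icon materials with player portraits and squad colors.
# Replicated score data, for bands of two or more, drives bandmate counting, then visibility, scoring text and pop animations.
# Pilgrim.ShowOverdriveUI gates display; MaxOverdrive stays collapsed because both Select options are Collapsed.
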WidgetBlueprint /PilgrimQuickplay/Widgets/WBP_PilgrimQuickplay_OverdriveUI_Display
animations [
 BandPointsEarnedWhileOverdriveActive_Visibility {
  binding_count=4
  length_seconds=4
  object=&self:BandPointsEarnedWhileOverdriveActive_Visibility
  sequence {bindings [
    Glow_1 {
     id=#1
     possessed_class=Image
     tracks [
      {
       type=MovieSceneWidgetMaterialTrack
       sections [{type=MovieSceneParameterSection, channels [Alpha {type=MovieSceneScriptingFloatChannel, keys [[0,0],[6,2],[15,0]]}]}]
      }
      {
       type=MovieSceneFloatTrack
       display_name=Render Opacity
       sections [{type=MovieSceneFloatSection, channels [None {type=MovieSceneScriptingFloatChannel, keys [[0,1],[10,1],[15,0]]}]}]
      }
      {
       type=MovieScene2DTransformTrack
       sections [{
         type=MovieScene2DTransformSection
         channels [
          Scale.X {type=MovieSceneScriptingFloatChannel, keys [[0,0.25],[4,1]]}
          Scale.Y {type=MovieSceneScriptingFloatChannel, keys [[0,1],[2,1.5],[10,1]]}
         ]
        }]
      }
     ]
    }
    BandPointsRecapOverlay {
     id=#2
     possessed_class=Overlay
     tracks [
      {
       type=MovieSceneFloatTrack
       display_name=Render Opacity
       sections [{type=MovieSceneFloatSection, channels [None {type=MovieSceneScriptingFloatChannel, keys [[0,0],[2,1],[40,1],[80,0]]}]}]
      }
      {
       type=MovieScene2DTransformTrack
       sections [{
         type=MovieScene2DTransformSection
         end_frame=6
         start_frame=0
         channels [
          Translation.X {type=MovieSceneScriptingFloatChannel, const [0,1]}
          Translation.Y {type=MovieSceneScriptingFloatChannel, keys [[0,64],[5,0]]}
          Angle {type=MovieSceneScriptingFloatChannel, const [0,1]}
          Scale.X {type=MovieSceneScriptingFloatChannel, keys [[0,1.25],[5,1]]}
          Scale.Y {type=MovieSceneScriptingFloatChannel, keys [[0,1.25],[5,1]]}
          Shear.X {type=MovieSceneScriptingFloatChannel, const [0,1]}
          Shear.Y {type=MovieSceneScriptingFloatChannel, const [0,1]}
         ]
        }]
      }
     ]
    }
    BandOverdriveBonus_Text_PointCounterLabel {
     id=#3
     possessed_class=CommonTextBlock
     tracks [{
       type=MovieSceneFloatTrack
       display_name=Render Opacity
       sections [{type=MovieSceneFloatSection, channels [None {type=MovieSceneScriptingFloatChannel, keys [[0,0],[4,1],[11,1]]}]}]
      }]
    }
    Flare_1 {
     id=#4
     possessed_class=Image
     tracks [
      {
       type=MovieSceneFloatTrack
       display_name=Render Opacity
       sections [{type=MovieSceneFloatSection, channels [None {type=MovieSceneScriptingFloatChannel, keys [[2,1],[9,0]]}]}]
      }
      {
       type=MovieScene2DTransformTrack
       sections [{
         type=MovieScene2DTransformSection
         channels [
          Scale.X {type=MovieSceneScriptingFloatChannel, keys [[2,5],[9,2]]}
          Scale.Y {type=MovieSceneScriptingFloatChannel, const [1,1]}
         ]
        }]
      }
     ]
    }
   ]}
 }
 BandOverdriveBonus_Visibility {
  binding_count=6
  length_seconds=1
  object=&self:BandOverdriveBonus_Visibility
  sequence {
   bindings [
    ValueAmountOverlay {
     id=#5
     possessed_class=Overlay
     tracks [
      {
       type=MovieScene2DTransformTrack
       sections [{
         type=MovieScene2DTransformSection
         channels [
          Angle {type=MovieSceneScriptingFloatChannel, const [0,1]}
          Scale.X {type=MovieSceneScriptingFloatChannel, keys [[11,1.25],[26,1]]}
          Scale.Y {type=MovieSceneScriptingFloatChannel, keys [[11,1.25],[26,1]]}
         ]
        }]
      }
      {
       type=MovieSceneFloatTrack
       display_name=Render Opacity
       sections [{type=MovieSceneFloatSection, channels [None {type=MovieSceneScriptingFloatChannel, keys [[11,0],[17,1]]}]}]
      }
     ]
    }
    AddLayer {
     id=#6
     possessed_class=Image
     tracks [{
       type=MovieScene2DTransformTrack
       sections [{
         type=MovieScene2DTransformSection
         channels [
          Scale.X {type=MovieSceneScriptingFloatChannel, const [1,2]}
          Scale.Y {type=MovieSceneScriptingFloatChannel}
         ]
        }]
      }]
    }
    Flare {
     id=#7
     possessed_class=Image
     tracks [
      {
       type=MovieScene2DTransformTrack
       sections [{
         type=MovieScene2DTransformSection
         channels [
          Scale.X {type=MovieSceneScriptingFloatChannel, keys [[6,5],[27,2]]}
          Scale.Y {type=MovieSceneScriptingFloatChannel, const [1,1]}
         ]
        }]
      }
      {
       type=MovieSceneFloatTrack
       display_name=Render Opacity
       sections [{type=MovieSceneFloatSection, channels [None {type=MovieSceneScriptingFloatChannel, keys [[6,1],[27,0]]}]}]
      }
     ]
    }
    Text_SoftShadow {
     id=#8
     possessed_class=Image
     tracks [{
       type=MovieSceneFloatTrack
       display_name=Render Opacity
       sections [{type=MovieSceneFloatSection, channels [None {type=MovieSceneScriptingFloatChannel, keys [[30,0],[60,1]]}]}]
      }]
    }
    ActiveText {
     id=#9
     possessed_class=Overlay
     tracks [
      {
       type=MovieSceneFloatTrack
       display_name=Render Opacity
       sections [{type=MovieSceneFloatSection, channels [None {type=MovieSceneScriptingFloatChannel, keys [[0,0],[3,1]]}]}]
      }
      {
       type=MovieScene2DTransformTrack
       sections [{
         type=MovieScene2DTransformSection
         channels [
          Translation.X {type=MovieSceneScriptingFloatChannel, const [0,1]}
          Translation.Y {type=MovieSceneScriptingFloatChannel, keys [[0,64],[15,0]]}
          Angle {type=MovieSceneScriptingFloatChannel, const [0,1]}
          Scale.X {type=MovieSceneScriptingFloatChannel, keys [[0,1.25],[15,1],[30,1],[60,0.8]]}
          Scale.Y {type=MovieSceneScriptingFloatChannel, keys [[0,1.25],[15,1],[30,1],[60,0.8]]}
          Shear.X {type=MovieSceneScriptingFloatChannel, const [0,1]}
          Shear.Y {type=MovieSceneScriptingFloatChannel, const [0,1]}
         ]
        }]
      }
     ]
    }
    Glow {
     id=#10
     possessed_class=Image
     tracks [
      {
       type=MovieSceneWidgetMaterialTrack
       sections [{type=MovieSceneParameterSection, channels [Alpha {type=MovieSceneScriptingFloatChannel, keys [[0,0],[18,2],[45,0]]}]}]
      }
      {
       type=MovieSceneFloatTrack
       display_name=Render Opacity
       sections [{type=MovieSceneFloatSection, channels [None {type=MovieSceneScriptingFloatChannel, keys [[0,1],[30,1],[45,0]]}]}]
      }
      {
       type=MovieScene2DTransformTrack
       sections [{
         type=MovieScene2DTransformSection
         channels [
          Scale.X {type=MovieSceneScriptingFloatChannel, keys [[0,0.25],[12,1]]}
          Scale.Y {type=MovieSceneScriptingFloatChannel, keys [[0,1],[6,1.5],[30,1]]}
         ]
        }]
      }
     ]
    }
   ]
   display_rate {numerator=60}
  }
 }
 BandInstrumentIcon1_Activated {
  binding_count=2
  length_seconds=0.5
  object=&self:BandInstrumentIcon1_Activated
  sequence {bindings [
    GridSlot_0 {
     id=#11
     parent_id=#12
     possessed_class=GridSlot
     tracks [{
       type=MovieSceneIntegerTrack
       display_name=Layer
       sections [{type=MovieSceneIntegerSection, channels [None {type=MovieSceneScriptingIntegerChannel, keys [[0,1],[9,0],[10,0]]}]}]
      }]
    }
    BandAvatarIcon1 {
     id=#12
     possessed_class=Image
     tracks [
      {
       type=MovieScene2DTransformTrack
       sections [{
         type=MovieScene2DTransformSection
         channels [
          Scale.X {type=MovieSceneScriptingFloatChannel, keys [[0,1],[5,1.5],[10,1]]}
          Scale.Y {type=MovieSceneScriptingFloatChannel, keys [[0,1],[5,1.5],[10,1]]}
         ]
        }]
      }
      {
       type=MovieSceneByteTrack
       display_name=Visibility
       sections [{type=MovieSceneByteSection, end_frame=0, start_frame=0, channels [None {type=MovieSceneScriptingByteChannel, const [0,1]}]}]
      }
     ]
    }
   ]}
 }
 BandInstrumentIcon2_Activated {
  binding_count=2
  length_seconds=0.5
  object=&self:BandInstrumentIcon2_Activated
  sequence {bindings [
    BandAvatarIcon2 {
     id=#13
     possessed_class=Image
     tracks [
      {
       type=MovieScene2DTransformTrack
       sections [{
         type=MovieScene2DTransformSection
         channels [
          Scale.X {type=MovieSceneScriptingFloatChannel, keys [[0,1],[5,1.5],[10,1]]}
          Scale.Y {type=MovieSceneScriptingFloatChannel, keys [[0,1],[5,1.5],[10,1]]}
         ]
        }]
      }
      {
       type=MovieSceneByteTrack
       display_name=Visibility
       sections [{type=MovieSceneByteSection, end_frame=0, start_frame=0, channels [None {type=MovieSceneScriptingByteChannel, const [0,1]}]}]
      }
     ]
    }
    GridSlot_1 {
     id=#14
     parent_id=#13
     possessed_class=GridSlot
     tracks [{
       type=MovieSceneIntegerTrack
       display_name=Layer
       sections [{type=MovieSceneIntegerSection, channels [None {type=MovieSceneScriptingIntegerChannel, keys [[0,1],[9,1],[10,0]]}]}]
      }]
    }
   ]}
 }
 BandInstrumentIcon3_Activated {
  binding_count=2
  length_seconds=0.5
  object=&self:BandInstrumentIcon3_Activated
  sequence {bindings [
    GridSlot_2 {
     id=#15
     parent_id=#16
     possessed_class=GridSlot
     tracks [{
       type=MovieSceneIntegerTrack
       display_name=Layer
       sections [{type=MovieSceneIntegerSection, channels [None {type=MovieSceneScriptingIntegerChannel, keys [[0,1],[9,1],[10,0]]}]}]
      }]
    }
    BandAvatarIcon3 {
     id=#16
     possessed_class=Image
     tracks [
      {
       type=MovieScene2DTransformTrack
       sections [{
         type=MovieScene2DTransformSection
         channels [
          Scale.X {type=MovieSceneScriptingFloatChannel, keys [[0,1],[5,1.5],[10,1]]}
          Scale.Y {type=MovieSceneScriptingFloatChannel, keys [[0,1],[5,1.5],[10,1]]}
         ]
        }]
      }
      {
       type=MovieSceneByteTrack
       display_name=Visibility
       sections [{type=MovieSceneByteSection, end_frame=0, start_frame=0, channels [None {type=MovieSceneScriptingByteChannel, const [0,1]}]}]
      }
     ]
    }
   ]}
 }
 BandInstrumentIcon4_Activated {
  binding_count=2
  length_seconds=0.5
  object=&self:BandInstrumentIcon4_Activated
  sequence {bindings [
    GridSlot_3 {
     id=#17
     parent_id=#18
     possessed_class=GridSlot
     tracks [{
       type=MovieSceneIntegerTrack
       display_name=Layer
       sections [{type=MovieSceneIntegerSection, channels [None {type=MovieSceneScriptingIntegerChannel, keys [[0,1],[9,0],[10,0]]}]}]
      }]
    }
    BandAvatarIcon4 {
     id=#18
     possessed_class=Image
     tracks [
      {
       type=MovieScene2DTransformTrack
       sections [{
         type=MovieScene2DTransformSection
         channels [
          Scale.X {type=MovieSceneScriptingFloatChannel, keys [[0,1],[5,1.5],[10,1]]}
          Scale.Y {type=MovieSceneScriptingFloatChannel, keys [[0,1],[5,1.5],[10,1]]}
         ]
        }]
      }
      {
       type=MovieSceneByteTrack
       display_name=Visibility
       sections [{type=MovieSceneByteSection, end_frame=0, start_frame=0, channels [None {type=MovieSceneScriptingByteChannel, const [0,1]}]}]
      }
     ]
    }
   ]}
 }
 A_MaxOverdrive_Activated {
  binding_count=3
  length_seconds=0.4
  object=&self:A_MaxOverdrive_Activated
  sequence {bindings [
    MaxOverdrive {
     id=#19
     possessed_class=Overlay
     tracks [
      {
       type=MovieSceneFloatTrack
       display_name=Render Opacity
       sections [{type=MovieSceneFloatSection, channels [None {type=MovieSceneScriptingFloatChannel, keys [[0,0],[8,1]]}]}]
      }
      {
       type=MovieScene2DTransformTrack
       sections [{
         type=MovieScene2DTransformSection
         channels [
          Translation.X {type=MovieSceneScriptingFloatChannel}
          Translation.Y {type=MovieSceneScriptingFloatChannel}
          Angle {type=MovieSceneScriptingFloatChannel}
          Scale.X {type=MovieSceneScriptingFloatChannel, keys [[0,1.2],[4,1]]}
          Scale.Y {type=MovieSceneScriptingFloatChannel, keys [[0,1.2],[4,1]]}
          Shear.X {type=MovieSceneScriptingFloatChannel}
          Shear.Y {type=MovieSceneScriptingFloatChannel}
         ]
        }]
      }
     ]
    }
    ScalablePortionOfBandmateOverdriveUI {
     id=#20
     possessed_class=Overlay
     tracks [{
       type=MovieScene2DTransformTrack
       sections [{
         type=MovieScene2DTransformSection
         channels [
          Translation.X {type=MovieSceneScriptingFloatChannel}
          Translation.Y {type=MovieSceneScriptingFloatChannel}
          Angle {type=MovieSceneScriptingFloatChannel}
          Scale.X {type=MovieSceneScriptingFloatChannel, keys [[0,1],[1,1.1],[5,1]]}
          Scale.Y {type=MovieSceneScriptingFloatChannel, keys [[0,1],[1,1.1],[5,1]]}
          Shear.X {type=MovieSceneScriptingFloatChannel}
          Shear.Y {type=MovieSceneScriptingFloatChannel}
         ]
        }]
      }]
    }
    MaxBandOverdrive_Gradient {
     id=#21
     possessed_class=Image
     tracks [{
       type=MovieScene2DTransformTrack
       sections [{
         type=MovieScene2DTransformSection
         channels [
          Translation.X {type=MovieSceneScriptingFloatChannel}
          Translation.Y {type=MovieSceneScriptingFloatChannel}
          Angle {type=MovieSceneScriptingFloatChannel}
          Scale.X {type=MovieSceneScriptingFloatChannel, keys [[0,0],[3,1.3],[6,1]]}
          Scale.Y {type=MovieSceneScriptingFloatChannel, keys [[0,0.3],[3,0.3],[6,1]]}
          Shear.X {type=MovieSceneScriptingFloatChannel}
          Shear.Y {type=MovieSceneScriptingFloatChannel}
         ]
        }]
      }]
    }
   ]}
 }
]
graphs [
 EventGraph {nodes [
   Event Construct()
   > Parent: Construct()
   > n1 = Branch(Condition=Get MBIInitialized() -> (MBIInitialized:bool))
   SetVisibility(execute=@n1.else, InVisibility=Collapsed)
   Event OnMusicBattleInstanceInitializeEvent() {is=title K2Node_Event}
   > Set MBIInitialized(MBIInitialized:bool=true) -> (Output_Get:bool)
   > SetVisibility(
    self=[
     Get ValueAmountOverlay() -> (ValueAmountOverlay:Overlay Object Reference)
     Get BandAvatarIcon4() -> (BandAvatarIcon4:Image Object Reference)
     Get BandAvatarIcon1_inactive() -> (BandAvatarIcon1_inactive:Image Object Reference)
     Get BandAvatarIcon2_inactive() -> (BandAvatarIcon2_inactive:Image Object Reference)
     Get BandAvatarIcon3_inactive() -> (BandAvatarIcon3_inactive:Image Object Reference)
     Get BandAvatarIcon4_inactive() -> (BandAvatarIcon4_inactive:Image Object Reference)
     Get MaxOverdrive() -> (MaxOverdrive:Overlay Object Reference)
     Get BandOverdriveBonus_Text_PointCounterLabel() -> (BandOverdriveBonus_Text_PointCounterLabel:Common Text Object Reference)
     Get BandAvatarIcon1() -> (BandAvatarIcon1:Image Object Reference)
     Get BandAvatarIcon2() -> (BandAvatarIcon2:Image Object Reference)
     Get BandAvatarIcon3() -> (BandAvatarIcon3:Image Object Reference)
    ]
    InVisibility=Collapsed
   )
   > n2 = GetConsoleVariableBoolValue!(execute, VariableName:String=Pilgrim.ShowOverdriveUI) -> (then, ReturnValue:bool)
   > Branch(Condition=@n2)
   > SetActiveWidgetIndex(
    self:Common Visibility Switcher Object Reference=Get CommonVisibilitySwitcher_226() -> (CommonVisibilitySwitcher_226:Common Visibility Switcher Object Reference)
    Index=Get BandScore_Blank_Text() -> (BandScore_Blank_Text:int=0)
   )
   n3 = Get MusicBattleInstance() -> (MusicBattleInstance:Pilgrim Music Battle Instance Object Reference)
   > n4 = For Each Loop!(
    Exec
    Array:Array of Pilgrim Music Battle Player State Structures=Break Pilgrim Music Battle Data!(PilgrimMusicBattleData:"Pilgrim Music Battle Data Structure (by ref)"=Get MusicBattleData!(self:Pilgrim Music Battle Instance Object Reference=@n3) -> (MusicBattleData:Pilgrim Music Battle Data Structure)) -> (MusicBattleStates:Array of Pilgrim Music Battle Player State Structures, BandState:Pilgrim Band State Structure) {is=title K2Node_BreakStruct}
   ) -> (
    LoopBody
    Array Element_PlayerState:Fort Player State Athena Object Reference
    Array Element_Team:Byte
    Array Element_TrackType:EPilgrimTrackType Enum
    Array Element_InstrumentTags:Gameplay Tag Container Structure
    Array Element_PilgrimGameProgress:Pilgrim Game Progress Structure
    Array Element_ServerScoringChanges:Pilgrim Server Scoring Changes Structure
    Array Index
    Completed
   )
   SetVisibility(execute=@n4.Completed, InVisibility=SelfHitTestInvisible)
   n5 = Select!(
    Option 0:Image Object Reference=Get BandAvatarIcon1()
    Option 1:Image Object Reference=Get BandAvatarIcon2()
    Option 2:Image Object Reference=Get BandAvatarIcon3()
    Option 3:Image Object Reference=Get BandAvatarIcon4()
    Index:int=@n4.Array Index
   ) -> (ReturnValue:Image Object Reference)
   n6 = GetDynamicMaterial(execute=@n4.LoopBody, self=@n5)
   n7 = Resolve Soft Reference!(Input:Texture 2D Soft Object Reference=GetSocialAvatarBrush!(self:Fort Player State Zone Object Reference=@n4, bSmallImage:bool=true) -> (ReturnValue:Texture 2D Soft Object Reference)) -> (Output:Texture 2D Object Reference)
   > SetTextureParameterValue(self=@n6, ParameterName=PlayerIcon, Value=@n7)
   > n8 = GetDynamicMaterial(self=Select(
     Option 0=Get BandAvatarIcon1_inactive()
     Option 1=Get BandAvatarIcon2_inactive()
     Option 2=Get BandAvatarIcon3_inactive()
     Option 3=Get BandAvatarIcon4_inactive()
     Index=@n4.Array Index
    ))
   > SetTextureParameterValue(self=@n8, ParameterName=PlayerIcon, Value=@n7)
   > n9 = GetDynamicMaterial(self=@n5)
   > SetVectorParameterValue(
    self=@n9
    ParameterName=BgColor
    Value=GetPinColorBySquadMemberIndex!(SquadIndex:int=@n4.Array Index) -> (ReturnValue:Linear Color Structure)
   )
   n10 = Event OnScoreDataReplicated!() -> (OutputDelegate, then, MusicBattleData:"Pilgrim Music Battle Data Structure (by ref)") {is=title K2Node_Event}
   n11 = Break Pilgrim Music Battle Data(PilgrimMusicBattleData=@n10.MusicBattleData)
   > Branch(Condition="integer >= integer"(A=Length(TargetArray:Array of Pilgrim Music Battle Player State Structures=@n11), B=2))
   > n12 = GetBandmateOverdriveInfo!(execute, self, InputPin:Array of Pilgrim Music Battle Player State Structures=@n11) -> (then, OverdriveStatusChanged:bool)
   > n13 = Branch(Condition=@n12)
   > UpdateVisibility!(execute, self, MusicBattleStates:Array of Pilgrim Music Battle Player State Structures=@n11) -> (then)
   > UpdateScoring!(
    execute
    self
    MusicBattleStates:Array of Pilgrim Music Battle Player State Structures=@n11
    BandState:Pilgrim Band State Structure=@n11.BandState
   ) -> (then)
   > UpdateAnimations!(execute, self, MusicBattleStates:Array of Pilgrim Music Battle Player State Structures=@n11) -> (then)
   n14 = Break Pilgrim Band State!(PilgrimBandState:"Pilgrim Band State Structure (by ref)"=@n11.BandState) -> (
    BandHealth:float=~
    BandScore:float=~
    BandModifierBonus:float=~
    CurrentScoreFromOverdrive:float=~
    TotalScoreFromOverdrive:float=~
    BandStar:int
    NextBandStarProgress:float=~
    BandVarietyBonus:float=~
    BaseBandScore:float=~
    BandMultiplier:int
    bAllPlayersHaveMaxStreak:bool=~
    bAllPlayersHaveOverdriveReady:bool=~
    bAllPlayersHaveOverdriveActive:bool=~
    bAnyPlayersHaveOverdriveActive:bool=~
   ) {is=title K2Node_BreakStruct}
   Branch(
    execute=@n13.else
    Condition="Not Equal (Float)"(A=@n14.CurrentScoreFromOverdrive, B=Get CachedBandOverdriveScore() -> (CachedBandOverdriveScore:double))
   )
   > UpdateScoring(MusicBattleStates=@n11, BandState=@n11.BandState)
   > Set CachedBandOverdriveScore(CachedBandOverdriveScore:double=@n14.CurrentScoreFromOverdrive) -> (Output_Get:double)
  ]}
 GetBandmateOverdriveInfo {nodes [
   n1 = GetBandmateOverdriveInfo!() -> (then, InputPin:Array of Pilgrim Music Battle Player State Structures) {is=title K2Node_FunctionEntry}
   > Set NumberOfBandmatesWithOverdriveActive(NumberOfBandmatesWithOverdriveActive:int=0) -> (Output_Get:int=0)
   > Set NumberOfBandmates(NumberOfBandmates:int=0) -> (Output_Get:int=0)
   > n2 = For Each Loop!(Exec, Array=@n1) -> (
    LoopBody
    Array Element_PlayerState
    Array Element_Team
    Array Element_TrackType
    Array Element_InstrumentTags
    Array Element_PilgrimGameProgress_NumHits:int
    Array Element_PilgrimGameProgress_NumMisses:int
    Array Element_PilgrimGameProgress_NumStreakBreakingMisses:int
    Array Element_PilgrimGameProgress_NumPasses:int
    Array Element_PilgrimGameProgress_bHasFullCombo:bool=~
    Array Element_PilgrimGameProgress_bHasPerfectCombo:bool=~
    Array Element_PilgrimGameProgress_CurrentTick:int
    Array Element_PilgrimGameProgress_NoteStreak:int
    Array Element_PilgrimGameProgress_CurrentScore:float=~
    Array Element_PilgrimGameProgress_RawCurrentScore:float=~
    Array Element_PilgrimGameProgress_CurrentModifierBonus:float=~
    Array Element_PilgrimGameProgress_CurrentStar:int
    Array Element_PilgrimGameProgress_NextStarProgress:float=~
    Array Element_PilgrimGameProgress_StreakScoreMultiplier:int
    Array Element_PilgrimGameProgress_NotesPastLastMultiplier:int
    Array Element_PilgrimGameProgress_NotesToNextMultiplier:int
    Array Element_PilgrimGameProgress_CurrentHealth:float=~
    Array Element_PilgrimGameProgress_IsOverdriveActive:bool=~
    Array Element_PilgrimGameProgress_CurrentOverdrive:float=~
    Array Element_PilgrimGameProgress_CurrentOverdriveSectionNum:int
    Array Element_PilgrimGameProgress_BandScoreContribution:float=~
    Array Element_PilgrimGameProgress_BandModifierBonusContribution:float=~
    Array Element_PilgrimGameProgress_OverdriveScoringState:EOverdriveScoringState Enum
    Array Element_PilgrimGameProgress_MaxScore:float=~
    Array Element_PilgrimGameProgress_RawMaxScore:float=~
    Array Element_PilgrimGameProgress_MaxStreakLength:int
    Array Element_PilgrimGameProgress_AverageStreakMultiplierPerHitOrPass:float=~
    Array Element_PilgrimGameProgress_bIsAtMaxStreak:bool=~
    Array Element_PilgrimGameProgress_CurrentSectionCompletionPercentage:float=~
    Array Element_ServerScoringChanges
    Array Index
    Completed
   )
   Return Node!(execute=@n2.Completed, OverdriveStatusChanged:bool=Get AnyOverdriveChange() -> (AnyOverdriveChange:bool))
   > n3 = Sequence!(execute) -> (then_0, then_1, then_2:Exec)
   > Increment Int(Value=Get NumberOfBandmates() -> (NumberOfBandmates:int=0))
   Branch(execute=@n3.then_1, Condition=@n2.Array Element_PilgrimGameProgress_IsOverdriveActive)
   > Increment Int(Value=Get NumberOfBandmatesWithOverdriveActive() -> (NumberOfBandmatesWithOverdriveActive:int=0))
   n4 = Get BandmateInstrumentIcon3_OverdriveActivated() -> (BandmateInstrumentIcon3_OverdriveActivated:bool)
   n5 = Get BandmateInstrumentIcon1_OverdriveActivated() -> (BandmateInstrumentIcon1_OverdriveActivated:bool)
   n6 = Get BandmateInstrumentIcon2_OverdriveActivated() -> (BandmateInstrumentIcon2_OverdriveActivated:bool)
   n7 = Get BandmateInstrumentIcon4_OverdriveActivated() -> (BandmateInstrumentIcon4_OverdriveActivated:bool)
   Branch(
    execute=@n3.then_2
    Condition="Not Equal (Boolean)"(
     A=@n2.Array Element_PilgrimGameProgress_IsOverdriveActive
     B=Select(Option 0:bool=@n5, Option 1:bool=@n6, Option 2:bool=@n4, Option 3:bool=@n7, Index=@n2.Array Index) -> (ReturnValue:bool)
    ) -> (ReturnValue=~)
   )
   > Set AnyOverdriveChange(AnyOverdriveChange:bool=true) -> (Output_Get:bool)
  ]}
 SetBandMultiplerText {nodes [
   n1 = SetBandMultiplerText!() -> (then, Multipler:int) {is=title K2Node_FunctionEntry}
   > Set Text(
    Text=Format Text!(Format="LOCTEXT(\"{Value}x Band Points!\")", Value:int=@n1) -> (Result)
    self=Get BandOverdriveBonus_Text_PointCounterLabel()
   )
  ]}
 UpdateAnimations {nodes [
   n1 = UpdateAnimations!() -> (then, MusicBattleStates:Array of Pilgrim Music Battle Player State Structures) {is=title K2Node_FunctionEntry}
   > n2 = For Each Loop(Array=@n1) -> (
    Array Element_PilgrimGameProgress_bHasFullCombo=~
    Array Element_PilgrimGameProgress_bHasPerfectCombo=~
    Array Element_PilgrimGameProgress_CurrentScore=~
    Array Element_PilgrimGameProgress_RawCurrentScore=~
    Array Element_PilgrimGameProgress_CurrentModifierBonus=~
    Array Element_PilgrimGameProgress_NextStarProgress=~
    Array Element_PilgrimGameProgress_CurrentHealth=~
    Array Element_PilgrimGameProgress_IsOverdriveActive=~
    Array Element_PilgrimGameProgress_CurrentOverdrive=~
    Array Element_PilgrimGameProgress_BandScoreContribution=~
    Array Element_PilgrimGameProgress_BandModifierBonusContribution=~
    Array Element_PilgrimGameProgress_MaxScore=~
    Array Element_PilgrimGameProgress_RawMaxScore=~
    Array Element_PilgrimGameProgress_AverageStreakMultiplierPerHitOrPass=~
    Array Element_PilgrimGameProgress_bIsAtMaxStreak=~
    Array Element_PilgrimGameProgress_CurrentSectionCompletionPercentage=~
   )
   n3 = Select(
    Option 0=Get BandmateInstrumentIcon1_OverdriveActivated()
    Option 1=Get BandmateInstrumentIcon2_OverdriveActivated()
    Option 2=Get BandmateInstrumentIcon3_OverdriveActivated()
    Option 3=Get BandmateInstrumentIcon4_OverdriveActivated()
    Index=@n2.Array Index
   )
   > n4 = Branch(Condition=AND Boolean(
     A="Equal (Boolean)"(A:bool=@n2.Array Element_PilgrimGameProgress_IsOverdriveActive, B:bool=true) -> (ReturnValue:bool=~) {is=title K2Node_PromotableOperator}
     B="Equal (Boolean)"(A=@n3) -> (ReturnValue=~)
    ))
   > PlayAnimation(InAnimation=Select(
     Option 0:Widget Animation Object Reference=Get BandInstrumentIcon1_Activated() -> (BandInstrumentIcon1_Activated:Widget Animation Object Reference)
     Option 1:Widget Animation Object Reference=Get BandInstrumentIcon2_Activated() -> (BandInstrumentIcon2_Activated:Widget Animation Object Reference)
     Option 2:Widget Animation Object Reference=Get BandInstrumentIcon3_Activated() -> (BandInstrumentIcon3_Activated:Widget Animation Object Reference)
     Option 3:Widget Animation Object Reference=Get BandInstrumentIcon4_Activated() -> (BandInstrumentIcon4_Activated:Widget Animation Object Reference)
     Index=@n2.Array Index
    ) -> (ReturnValue:Widget Animation Object Reference))
   > n5 = Branch(Condition="Equal (Integer)"(A=@n2.Array Index))
   > Set BandmateInstrumentIcon1_OverdriveActivated(BandmateInstrumentIcon1_OverdriveActivated:bool=true) -> (Output_Get:bool)
   n6 = Branch(execute=@n5.else, Condition="Equal (Integer)"(A=@n2.Array Index, B=1))
   > Set BandmateInstrumentIcon2_OverdriveActivated(BandmateInstrumentIcon2_OverdriveActivated:bool=true) -> (Output_Get:bool)
   n7 = Branch(execute=@n6.else, Condition="Equal (Integer)"(A=@n2.Array Index, B=2))
   > Set BandmateInstrumentIcon3_OverdriveActivated(BandmateInstrumentIcon3_OverdriveActivated:bool=true) -> (Output_Get:bool)
   Branch(execute=@n7.else, Condition="Equal (Integer)"(A=@n2.Array Index, B=3))
   > Set BandmateInstrumentIcon4_OverdriveActivated(BandmateInstrumentIcon4_OverdriveActivated:bool=true) -> (Output_Get:bool)
   n8 = Branch(execute=@n4.else, Condition=@n2.Array Element_PilgrimGameProgress_IsOverdriveActive)
   n9 = Branch(execute=@n8.else, Condition="Equal (Integer)"(A=@n2.Array Index))
   > Set BandmateInstrumentIcon1_OverdriveActivated(BandmateInstrumentIcon1_OverdriveActivated=false)
   n10 = Branch(execute=@n9.else, Condition="Equal (Integer)"(A=@n2.Array Index, B=1))
   > Set BandmateInstrumentIcon2_OverdriveActivated(BandmateInstrumentIcon2_OverdriveActivated=false)
   n11 = Branch(execute=@n10.else, Condition="Equal (Integer)"(A=@n2.Array Index, B=2))
   > Set BandmateInstrumentIcon3_OverdriveActivated(BandmateInstrumentIcon3_OverdriveActivated=false)
   Branch(execute=@n11.else, Condition="Equal (Integer)"(A=@n2.Array Index, B=3))
   > Set BandmateInstrumentIcon4_OverdriveActivated(BandmateInstrumentIcon4_OverdriveActivated=false)
  ]}
 UpdateScoring {nodes [
   n1 = UpdateScoring!() -> (
    then
    MusicBattleStates:Array of Pilgrim Music Battle Player State Structures
    BandState_BandHealth:float=~
    BandState_BandScore:float=~
    BandState_BandModifierBonus:float=~
    BandState_CurrentScoreFromOverdrive:float=~
    BandState_TotalScoreFromOverdrive:float=~
    BandState_BandStar:int
    BandState_NextBandStarProgress:float=~
    BandState_BandVarietyBonus:float=~
    BandState_BaseBandScore:float=~
    BandState_BandMultiplier:int
    BandState_bAllPlayersHaveMaxStreak:bool=~
    BandState_bAllPlayersHaveOverdriveReady:bool=~
    BandState_bAllPlayersHaveOverdriveActive:bool=~
    BandState_bAnyPlayersHaveOverdriveActive:bool=~
   ) {is=title K2Node_FunctionEntry}
   > n2 = Branch(Condition="integer >= integer"(A=Get NumberOfBandmatesWithOverdriveActive() -> (NumberOfBandmatesWithOverdriveActive=0), B=1))
   > n3 = Branch(Condition=Get HasInitializedOverdriveBandScore_Active() -> (HasInitializedOverdriveBandScore_Active:bool))
   Set BandPointsEarnedWhileOverdriveActive(execute=[@n3.then, @n7.then], BandPointsEarnedWhileOverdriveActive:int=Truncate(A=@n1.BandState_CurrentScoreFromOverdrive)) -> (Output_Get:int=0)
   n4 = Get MusicBattleInstance()
   > Branch(Condition=IsValid(Object=@n4))
   n5 = Break Pilgrim Game Config Settings!(PilgrimGameConfigSettings:"Pilgrim Game Config Settings Structure (by ref)"=Get ConfigSettings!(self:Pilgrim Settings Object Reference=GetConfig!(self:Pilgrim Music Battle Instance Object Reference=@n4) -> (ReturnValue:Pilgrim Settings Object Reference)) -> (ConfigSettings:Pilgrim Game Config Settings Structure)) -> (
    SlopMs:float=~
    GemBreakEndInputUnlockMs:float=~
    EndOfTrackBuffer:float=~
    MsToNowBar:float=~
    BackgroundTrackDb:float=~
    BackgroundTrackGain:float=~
    SongAudioResponceCueSettingsMap:Map of EPilgrimSongAudioResponseCue Enums to Pilgrim Song Audio Response Cue Settings Structures
    FireMidiEventTicksEarly:float=~
    BaseGemScoreValue:"Map of EPilgrimSongDifficulty Enums to Float (single-precision)s"
    BaseCymbalGemScoreValue:"Map of EPilgrimSongDifficulty Enums to Float (single-precision)s"
    SustainScorePerBeat:int
    bNormalizeScores:bool=~
    NormalizedScaleTowardsHighestInDifficulty:Scalable Float Structure
    NormalizedScaleTowardsHighestInExpert:Scalable Float Structure
    OutroPreAlertSeconds:float=~
    ScoreMultiplierDefinitions:Array of Pilgrim Score Multiplier Info Structures
    PlayerHealthModifiers:Map of EPilgrimSongDifficulty Enums to Player Health Modifiers Structures
    PilgrimHealthCrowdResponseTiers:"Array of Float (single-precision)s"
    IndividualStarPercentageThresholds:Map of EPilgrimSongDifficulty Enums to Star Percentage Thresholds Structures
    BandStarPercentageThresholds:"Array of Float (single-precision)s"
    IsBandMultiplierBasedOnUniqueInstruments:bool=~
    IsBandVarietyBonusScoredDuringPlay:bool=~
    BandOverdriveScoreMultipliers:"Array of Float (single-precision)s"
    BandVarietyScoreMultipliers:"Array of Float (single-precision)s"
    AccuracyTiers:Array of Pilgrim Accuracy Tier Structures
    MaximumDisplayedAccuracyTierIndex:int
    TeamQuickplayInitialHealth:float=~
    TeamQuickplayOverdriveDecayRate:float=~
    TeamQuickplayInitialOverdrive:float=~
    TeamQuickplayOverdrive:float=~
    TeamQuickplayOverdriveScoreMultiplier:float=~
    TeamQuickplayOverdriveHealthGainMultiplier:float=~
    TeamQuickplayRequiredOverdriveToActivate:float=~
    TeamQuickplayOverdriveAddedOnSuccesfulSection:float=~
    bOverdriveClearsGems:bool=~
    bOverdriveClearAutoHoldSustain:bool=~
    GemBreakPendingBars:float=~
    bEnableInputSpecificGemsAndSmashers:bool=~
    InputConfig:Pilgrim Input Settings Object Reference
    bEnableOverdriveDeploys:bool=~
    bEnableOverdrivePhrases:bool=~
    bEnableCaptureSystem:bool=~
    bEnableBandPerformanceScoring:bool=~
    ArenaInitialShields:float=~
    ArenaInitialHealth:float=~
    ArenaInitialAmmo:bool=~
    DifficultyCaptureFactors:"Map of EPilgrimSongDifficulty Enums to Float (single-precision)s"
    PerInstrumentConfigs:Map of EPilgrimTrackType Enums to Pilgrim Settings Instrument Overrides Object References
    StreakToastThresholds:Array of Integers
    InstrumentTypeMissCues:Map of ESparksInstrumentType Enums to Sound Cue Object References
    DefaultInstrumentTypeMissCue:Sound Cue Object Reference
    CreateStandinBandmembers:bool=~
    SongIntroEmbargoMs:float=~
    DeployOverdriveGem_BarsUntilGemSpawn:float=~
    DeployOverdriveGem_TailDurationBar:float=~
    RollsMinInputForgivenessMs:float=~
    SwellsMinInputForgivenessMs:float=~
    ProVocalPitchThresholds:Map of EPilgrimSongDifficulty Enums to Pro Vocal Scoring Pitch Range Structures
    ProVocalSectionCompletionThresholds:"Map of EPilgrimSongDifficulty Enums to Float (single-precision)s"
    ProVocalsSectionMaxValue:"Map of EPilgrimSongDifficulty Enums to Float (single-precision)s"
    ProVocalsSectionRatings:Array of Pro Vocals Section Rating Structures
    ProVocalsScoreMultiplierDefinitions:Array of Pilgrim Score Multiplier Info Structures
    ScoreWeightInSection:Curve Float Object Reference
   ) {is=title K2Node_BreakStruct}
   n6 = Get NumberOfBandmatesWithOverdriveActive() -> (NumberOfBandmatesWithOverdriveActive=0)
   > SetBandMultiplerText!(
    execute
    self
    Multipler:int=Truncate(A="Get (a copy)"(Array:"Array of Float (single-precision)s"=@n5.BandOverdriveScoreMultipliers, Dimension 1=@n6) -> (Output:float))
   ) -> (then) {is=title K2Node_CallFunction}
   Set HasInitializedOverdriveBandScore_Active(execute=@n3.else, HasInitializedOverdriveBandScore_Active:bool=true) -> (Output_Get:bool)
   > Set HasInitializedOverdriveBandScore_PointsEarned(HasInitializedOverdriveBandScore_PointsEarned:bool=true) -> (Output_Get:bool)
   > PlayAnimation(InAnimation=Get BandOverdriveBonus_Visibility() -> (BandOverdriveBonus_Visibility:Widget Animation Object Reference))
   > n7 = Set BandPointsEarnedWhileOverdriveActive(BandPointsEarnedWhileOverdriveActive=0) -> (Output_Get=0)
   Branch(execute=@n2.else, Condition=Get HasInitializedOverdriveBandScore_PointsEarned() -> (HasInitializedOverdriveBandScore_PointsEarned:bool))
   > Set HasInitializedOverdriveBandScore_Active(HasInitializedOverdriveBandScore_Active=false)
   > Set HasInitializedOverdriveBandScore_PointsEarned(HasInitializedOverdriveBandScore_PointsEarned=false)
   > Set Text(
    Text=Format Text(
     Format="LOCTEXT(\"+{Value} Band Points!\")"
     Value=Get BandPointsEarnedWhileOverdriveActive() -> (BandPointsEarnedWhileOverdriveActive:int=0)
    )
    self=Get BandPointsEarnedWhileOverdriveActive_Text() -> (BandPointsEarnedWhileOverdriveActive_Text:Common Text Object Reference)
   )
   > SetActiveWidgetIndex(self=Get CommonVisibilitySwitcher_226(), Index=Get BandScore_PointsEarned_Text() -> (BandScore_PointsEarned_Text:int=0))
   > PlayAnimation(InAnimation=Get BandPointsEarnedWhileOverdriveActive_Visibility() -> (BandPointsEarnedWhileOverdriveActive_Visibility:Widget Animation Object Reference))
  ]}
 UpdateVisibility {nodes [
   n1 = UpdateVisibility!() -> (then, MusicBattleStates:Array of Pilgrim Music Battle Player State Structures) {is=title K2Node_FunctionEntry}
   > n2 = For Each Loop!(Exec, Array=@n1) -> (
    LoopBody
    Array Element_PlayerState
    Array Element_Team
    Array Element_TrackType
    Array Element_InstrumentTags
    Array Element_PilgrimGameProgress:Pilgrim Game Progress Structure
    Array Element_ServerScoringChanges
    Array Index
    Completed
   )
   > n3 = GetConsoleVariableBoolValue(VariableName=Pilgrim.ShowOverdriveUI)
   n4 = Get NumberOfBandmatesWithOverdriveActive() -> (NumberOfBandmatesWithOverdriveActive=0)
   > n5 = Branch(Condition=AND Boolean(A=@n3, B="integer >= integer"(A=@n4, B=1)))
   > SetVisibility(self=[Get ValueAmountOverlay(), Get BandOverdriveBonus_Text_PointCounterLabel()], InVisibility=SelfHitTestInvisible)
   > SetActiveWidgetIndex(self=Get CommonVisibilitySwitcher_226(), Index=Get BandScore_Active_Text() -> (BandScore_Active_Text:int=0))
   > n6 = Branch(Condition=Break Pilgrim Game Progress!(PilgrimGameProgress:"Pilgrim Game Progress Structure (by ref)"=@n2.Array Element_PilgrimGameProgress) -> (
     NumStreakBreakingMisses:int
     bHasPerfectCombo:bool=~
     RawCurrentScore:float=~
     CurrentModifierBonus:float=~
     IsOverdriveActive:bool=~
     BandModifierBonusContribution:float=~
     RawMaxScore:float=~
     MaxStreakLength:int
     AverageStreakMultiplierPerHitOrPass:float=~
     CurrentSectionCompletionPercentage:float=~
    ) {is=title K2Node_BreakStruct}.IsOverdriveActive)
   > n7 = Sequence!(execute) -> (then_0, then_1)
   SetColorAndOpacity(
    execute=@n7.then_1
    self:Image Object Reference=Select(
     Option 0:Image Object Reference=Get BandAvatarIcon1()
     Option 1:Image Object Reference=Get BandAvatarIcon2()
     Option 2:Image Object Reference=Get BandAvatarIcon3()
     Option 3:Image Object Reference=Get BandAvatarIcon4()
     Index=@n2.Array Index
    ) -> (ReturnValue:Image Object Reference)
    InColorAndOpacity:Linear Color Structure=(R=1,G=1,B=1,A=1)
   )
   n8 = "Equal (Integer)"(A=Get NumberOfBandmatesWithOverdriveActive() -> (NumberOfBandmatesWithOverdriveActive=0), B=Get NumberOfBandmates() -> (NumberOfBandmates=0))
   n9 = "integer >= integer"(A=Get NumberOfBandmates() -> (NumberOfBandmates=0), B=2)
   SetVisibility(
    execute=@n7.then_0
    self=Get MaxOverdrive()
    InVisibility=Select!(Option 0:ESlateVisibility Enum=Collapsed, Option 1:ESlateVisibility Enum=Collapsed, Index:bool=AND Boolean(A=@n8, B=@n9)) -> (ReturnValue:ESlateVisibility Enum)
   )
   SetVisibility(
    execute=@n6.else
    self=Select!(
     Option 0:Image Object Reference=Get BandAvatarIcon1_inactive()
     Option 1:Image Object Reference=Get BandAvatarIcon2_inactive()
     Option 2:Image Object Reference=Get BandAvatarIcon3_inactive()
     Option 3:Image Object Reference=Get BandAvatarIcon4_inactive()
     Index:int=@n2.Array Index
    ) -> (ReturnValue:Image Object Reference)
    InVisibility=Visible
   )
   > SetVisibility(
    self=Select(
     Option 0=Get BandAvatarIcon1()
     Option 1=Get BandAvatarIcon2()
     Option 2=Get BandAvatarIcon3()
     Option 3=Get BandAvatarIcon4()
     Index=@n2.Array Index
    )
    InVisibility=Collapsed
   )
   SetVisibility(
    execute=@n5.else
    self=[
     Get BandAvatarIcon1_inactive()
     Get BandAvatarIcon2_inactive()
     Get BandOverdriveBonus_Text_PointCounterLabel()
     Get BandAvatarIcon3_inactive()
     Get BandAvatarIcon4_inactive()
     Get BandAvatarIcon1()
     Get BandAvatarIcon2()
     Get BandAvatarIcon3()
     Get BandAvatarIcon4()
     Get ValueAmountOverlay()
    ]
    InVisibility=Collapsed
   )
  ]}
]
variables [
 1PlayersODActive_Points {type=number, default=0.5}
 2PlayersODActive_Points {type=number, default=0.75}
 3PlayersODActive_Points {type=number, default=0.833}
 4PlayersODActive_Points {type=number, default=0.875}
 BandmateInstrumentIcon1_OverdriveActivated {type=boolean}
 BandmateInstrumentIcon2_OverdriveActivated {type=boolean}
 BandmateInstrumentIcon3_OverdriveActivated {type=boolean}
 BandmateInstrumentIcon4_OverdriveActivated {type=boolean}
 BandPointsEarnedWhileOverdriveActive {type=integer}
 BandScore_Active_Text {type=integer, default=1}
 BandScore_Blank_Text {type=integer}
 BandScore_PointsEarned_Text {type=integer, default=2}
 CachedBandOverdriveScore {type=number, default=-1}
 HasInitializedOverdriveBandScore_Active {type=boolean}
 HasInitializedOverdriveBandScore_PointsEarned {type=boolean}
 MBIInitialized {type=boolean}
 NumberOfBandmates {type=integer}
 NumberOfBandmatesWithOverdriveActive {type=integer}
]
widgets {
 info {inheritedWidgetCount=0, parentClass=/Script/PilgrimCoreUI.PilgrimGameWidgetBase, rootWidgetClass=&Overlay, widgetCount=40}
 widgets [
  Overlay_0 {widgetClassPath=&Overlay, widget=Overlay_0}
  MaxOverdrive {
   widgetClassPath=&Overlay
   bIsVariable=true
   parent=Overlay_0
   widget=MaxOverdrive
   properties {visibility=Collapsed}
   slot {class=OverlaySlot, properties {horizontalAlignment=HAlign_Fill, verticalAlignment=VAlign_Center, padding=<ltrb 0 0 0 22>}}
  }
  MaxBandOverdrive_Gradient {
   widgetClassPath=&Image
   bIsVariable=true
   parent=MaxOverdrive
   widget=MaxBandOverdrive_Gradient
   properties {
    brush {
     bIsDynamicallyLoaded=false
     drawAs=Image
     imageType=NoImage
     mirroring=NoMirror
     resourceName=None
     resourceObject=/PilgrimCore/Widgets/Materials/M_UI_Habanero_Flare
     tiling=NoTile
     imageSize=<320 190>
     margin=<ltrb 0 0 0 0>
     outlineSettings {
      bUseBrushTransparency=false
      roundingType=HalfHeightRadius
      width=0
      color {colorUseRule=UseColor_Specified, specifiedColor=<rgba 0 0 0 0>}
      cornerRadii=<0 0 0 0>
     }
     tintColor {colorUseRule=UseColor_Specified, specifiedColor=<rgba 1 1 1 1>}
     uVRegion {bIsValid=false, max=<0 0>, min=<0 0>}
    }
    colorAndOpacity=<rgba 1 0.7408 0.3125 0.425>
   }
   slot {class=OverlaySlot, properties {horizontalAlignment=HAlign_Center, verticalAlignment=VAlign_Center}}
  }
  MaxBandOverdrive_Triangles_Big {
   widgetClassPath=&Image
   bIsVariable=true
   parent=MaxOverdrive
   widget=MaxBandOverdrive_Triangles_Big
   properties {brush {
     bIsDynamicallyLoaded=false
     drawAs=Image
     imageType=NoImage
     mirroring=NoMirror
     resourceName=None
     resourceObject=/PilgrimCore/Widgets/Materials/M_UI_PilgrimCore_Triangles_Big
     tiling=NoTile
     imageSize=<480 60>
     margin=<ltrb 0 0 0 0>
     outlineSettings {
      bUseBrushTransparency=false
      roundingType=HalfHeightRadius
      width=0
      color {colorUseRule=UseColor_Specified, specifiedColor=<rgba 0 0 0 0>}
      cornerRadii=<0 0 0 0>
     }
     tintColor {colorUseRule=UseColor_Specified, specifiedColor=<rgba 1 1 1 1>}
     uVRegion {bIsValid=false, max=<0 0>, min=<0 0>}
    }}
   slot {class=OverlaySlot, properties {horizontalAlignment=HAlign_Center, verticalAlignment=VAlign_Center}}
  }
  MaxBandOverdrive_Triangles_Med {
   widgetClassPath=&Image
   bIsVariable=true
   parent=MaxOverdrive
   widget=MaxBandOverdrive_Triangles_Med
   properties {brush {
     bIsDynamicallyLoaded=false
     drawAs=Image
     imageType=NoImage
     mirroring=NoMirror
     resourceName=None
     resourceObject=/PilgrimCore/Widgets/Materials/M_UI_PilgrimCore_Triangles_Med
     tiling=NoTile
     imageSize=<460 80>
     margin=<ltrb 0 0 0 0>
     outlineSettings {
      bUseBrushTransparency=false
      roundingType=HalfHeightRadius
      width=0
      color {colorUseRule=UseColor_Specified, specifiedColor=<rgba 0 0 0 0>}
      cornerRadii=<0 0 0 0>
     }
     tintColor {colorUseRule=UseColor_Specified, specifiedColor=<rgba 1 1 1 1>}
     uVRegion {bIsValid=false, max=<0 0>, min=<0 0>}
    }}
   slot {class=OverlaySlot, properties {horizontalAlignment=HAlign_Center, verticalAlignment=VAlign_Center, padding=<ltrb 0 8 0 0>}}
  }
  MaxBandOverdrive_Border {
   widgetClassPath=&Image
   bIsVariable=true
   parent=MaxOverdrive
   widget=MaxBandOverdrive_Border
   properties {brush {
     bIsDynamicallyLoaded=false
     drawAs=Image
     imageType=NoImage
     mirroring=NoMirror
     resourceName=None
     resourceObject=/PilgrimQuickplay/Widgets/Materials/M_UI_Pilgrim_BandODActive
     tiling=NoTile
     imageSize=<256 164>
     margin=<ltrb 0 0 0 0>
     outlineSettings {
      bUseBrushTransparency=false
      roundingType=HalfHeightRadius
      width=0
      color {colorUseRule=UseColor_Specified, specifiedColor=<rgba 0 0 0 0>}
      cornerRadii=<0 0 0 0>
     }
     tintColor {colorUseRule=UseColor_Specified, specifiedColor=<rgba 1 1 1 1>}
     uVRegion {bIsValid=false, max=<0 0>, min=<0 0>}
    }}
   slot {class=OverlaySlot, properties {horizontalAlignment=HAlign_Center, verticalAlignment=VAlign_Center}}
  }
  BandOverdriveParent {
   widgetClassPath=&VerticalBox
   parent=Overlay_0
   widget=BandOverdriveParent
   slot {class=OverlaySlot, properties {horizontalAlignment=HAlign_Center, padding=<ltrb 0 6 0 0>}}
  }
  SizeBox_0 {
   widgetClassPath=&SizeBox
   parent=BandOverdriveParent
   widget=SizeBox_0
   properties {heightOverride=154, visibility=Collapsed, widthOverride=154}
   slot {class=VerticalBoxSlot, properties {horizontalAlignment=HAlign_Center, verticalAlignment=VAlign_Center}}
  }
  ScalablePortionOfBandmateOverdriveUI {widgetClassPath=&Overlay, parent=SizeBox_0, widget=ScalablePortionOfBandmateOverdriveUI, slot {class=SizeBoxSlot}}
  Default {
   widgetClassPath=&Overlay
   parent=ScalablePortionOfBandmateOverdriveUI
   widget=Default
   slot {class=OverlaySlot, properties {horizontalAlignment=HAlign_Center, verticalAlignment=VAlign_Center}}
  }
  BandAvatarIcon1_inactive {
   widgetClassPath=&Image
   bIsVariable=true
   parent=Default
   widget=BandAvatarIcon1_inactive
   properties {
    brush {
     bIsDynamicallyLoaded=false
     drawAs=Image
     imageType=FullColor
     mirroring=NoMirror
     resourceName=None
     resourceObject=/SparksUICommon/Materials/MI_UI_BandMember_Avatar_Diamond_Inactive
     tiling=NoTile
     imageSize=<80 80>
     margin=<ltrb 0 0 0 0>
     outlineSettings {
      bUseBrushTransparency=false
      roundingType=HalfHeightRadius
      width=0
      color {colorUseRule=UseColor_Specified, specifiedColor=<rgba 0 0 0 0>}
      cornerRadii=<0 0 0 0>
     }
     tintColor {colorUseRule=UseColor_Specified, specifiedColor=<rgba 1 1 1 1>}
     uVRegion {bIsValid=false, max=<0 0>, min=<0 0>}
    }
    renderTransform {angle=0, scale=<1 1>, shear=<0 0>, translation=<-36.5 0>}
   }
   slot {class=OverlaySlot, properties {horizontalAlignment=HAlign_Fill, verticalAlignment=VAlign_Center}}
  }
  BandAvatarIcon2_inactive {
   widgetClassPath=&Image
   bIsVariable=true
   parent=Default
   widget=BandAvatarIcon2_inactive
   properties {
    brush {
     bIsDynamicallyLoaded=false
     drawAs=Image
     imageType=FullColor
     mirroring=NoMirror
     resourceName=None
     resourceObject=&MI_UI_BandMember_Avatar_Diamond_Inactive
     tiling=NoTile
     imageSize=<80 80>
     margin=<ltrb 0 0 0 0>
     outlineSettings {
      bUseBrushTransparency=false
      roundingType=HalfHeightRadius
      width=0
      color {colorUseRule=UseColor_Specified, specifiedColor=<rgba 0 0 0 0>}
      cornerRadii=<0 0 0 0>
     }
     tintColor {colorUseRule=UseColor_Specified, specifiedColor=<rgba 1 1 1 1>}
     uVRegion {bIsValid=false, max=<0 0>, min=<0 0>}
    }
    renderTransform {angle=0, scale=<1 1>, shear=<0 0>, translation=<36.5 0>}
   }
   slot {class=OverlaySlot, properties {horizontalAlignment=HAlign_Right, verticalAlignment=VAlign_Fill}}
  }
  BandAvatarIcon3_inactive {
   widgetClassPath=&Image
   bIsVariable=true
   parent=Default
   widget=BandAvatarIcon3_inactive
   properties {
    brush {
     bIsDynamicallyLoaded=false
     drawAs=Image
     imageType=FullColor
     mirroring=NoMirror
     resourceName=None
     resourceObject=&MI_UI_BandMember_Avatar_Diamond_Inactive
     tiling=NoTile
     imageSize=<80 80>
     margin=<ltrb 0 0 0 0>
     outlineSettings {
      bUseBrushTransparency=false
      roundingType=HalfHeightRadius
      width=0
      color {colorUseRule=UseColor_Specified, specifiedColor=<rgba 0 0 0 0>}
      cornerRadii=<0 0 0 0>
     }
     tintColor {colorUseRule=UseColor_Specified, specifiedColor=<rgba 1 1 1 1>}
     uVRegion {bIsValid=false, max=<0 0>, min=<0 0>}
    }
    renderTransform {angle=0, scale=<1 1>, shear=<0 0>, translation=<0 -36.5>}
   }
   slot {class=OverlaySlot, properties {horizontalAlignment=HAlign_Center}}
  }
  BandAvatarIcon4_inactive {
   widgetClassPath=&Image
   bIsVariable=true
   parent=Default
   widget=BandAvatarIcon4_inactive
   properties {
    brush {
     bIsDynamicallyLoaded=false
     drawAs=Image
     imageType=FullColor
     mirroring=NoMirror
     resourceName=None
     resourceObject=&MI_UI_BandMember_Avatar_Diamond_Inactive
     tiling=NoTile
     imageSize=<80 80>
     margin=<ltrb 0 0 0 0>
     outlineSettings {
      bUseBrushTransparency=false
      roundingType=HalfHeightRadius
      width=0
      color {colorUseRule=UseColor_Specified, specifiedColor=<rgba 0 0 0 0>}
      cornerRadii=<0 0 0 0>
     }
     tintColor {colorUseRule=UseColor_Specified, specifiedColor=<rgba 1 1 1 1>}
     uVRegion {bIsValid=false, max=<0 0>, min=<0 0>}
    }
    renderTransform {angle=0, scale=<1 1>, shear=<0 0>, translation=<0 36.5>}
   }
   slot {class=OverlaySlot, properties {horizontalAlignment=HAlign_Center}}
  }
  Active {
   widgetClassPath=&GridPanel
   parent=ScalablePortionOfBandmateOverdriveUI
   widget=Active
   slot {class=OverlaySlot, properties {horizontalAlignment=HAlign_Center, verticalAlignment=VAlign_Center}}
  }
  BandAvatarIcon1 {
   widgetClassPath=&Image
   bIsVariable=true
   parent=Active
   widget=BandAvatarIcon1
   properties {
    brush {
     bIsDynamicallyLoaded=false
     drawAs=Image
     imageType=FullColor
     mirroring=NoMirror
     resourceName=None
     resourceObject=/SparksUICommon/Materials/MI_UI_BandMember_Avatar_Diamond
     tiling=NoTile
     imageSize=<80 80>
     margin=<ltrb 0 0 0 0>
     outlineSettings {
      bUseBrushTransparency=false
      roundingType=HalfHeightRadius
      width=0
      color {colorUseRule=UseColor_Specified, specifiedColor=<rgba 0 0 0 0>}
      cornerRadii=<0 0 0 0>
     }
     tintColor {colorUseRule=UseColor_Specified, specifiedColor=<rgba 1 1 1 1>}
     uVRegion {bIsValid=false, max=<0 0>, min=<0 0>}
    }
    renderTransform {angle=0, scale=<1 1>, shear=<0 0>, translation=<-36.5 0>}
   }
   slot {class=GridSlot}
  }
  BandAvatarIcon2 {
   widgetClassPath=&Image
   bIsVariable=true
   parent=Active
   widget=BandAvatarIcon2
   properties {
    brush {
     bIsDynamicallyLoaded=false
     drawAs=Image
     imageType=FullColor
     mirroring=NoMirror
     resourceName=None
     resourceObject=&MI_UI_BandMember_Avatar_Diamond
     tiling=NoTile
     imageSize=<80 80>
     margin=<ltrb 0 0 0 0>
     outlineSettings {
      bUseBrushTransparency=false
      roundingType=HalfHeightRadius
      width=0
      color {colorUseRule=UseColor_Specified, specifiedColor=<rgba 0 0 0 0>}
      cornerRadii=<0 0 0 0>
     }
     tintColor {colorUseRule=UseColor_Specified, specifiedColor=<rgba 1 1 1 1>}
     uVRegion {bIsValid=false, max=<0 0>, min=<0 0>}
    }
    renderTransform {angle=0, scale=<1 1>, shear=<0 0>, translation=<36.5 0>}
   }
   slot {class=GridSlot}
  }
  BandAvatarIcon3 {
   widgetClassPath=&Image
   bIsVariable=true
   parent=Active
   widget=BandAvatarIcon3
   properties {
    brush {
     bIsDynamicallyLoaded=false
     drawAs=Image
     imageType=FullColor
     mirroring=NoMirror
     resourceName=None
     resourceObject=&MI_UI_BandMember_Avatar_Diamond
     tiling=NoTile
     imageSize=<80 80>
     margin=<ltrb 0 0 0 0>
     outlineSettings {
      bUseBrushTransparency=false
      roundingType=HalfHeightRadius
      width=0
      color {colorUseRule=UseColor_Specified, specifiedColor=<rgba 0 0 0 0>}
      cornerRadii=<0 0 0 0>
     }
     tintColor {colorUseRule=UseColor_Specified, specifiedColor=<rgba 1 1 1 1>}
     uVRegion {bIsValid=false, max=<0 0>, min=<0 0>}
    }
    renderTransform {angle=0, scale=<1 1>, shear=<0 0>, translation=<0 -36.5>}
   }
   slot {class=GridSlot}
  }
  BandAvatarIcon4 {
   widgetClassPath=&Image
   bIsVariable=true
   parent=Active
   widget=BandAvatarIcon4
   properties {
    brush {
     bIsDynamicallyLoaded=false
     drawAs=Image
     imageType=FullColor
     mirroring=NoMirror
     resourceName=None
     resourceObject=&MI_UI_BandMember_Avatar_Diamond
     tiling=NoTile
     imageSize=<80 80>
     margin=<ltrb 0 0 0 0>
     outlineSettings {
      bUseBrushTransparency=false
      roundingType=HalfHeightRadius
      width=0
      color {colorUseRule=UseColor_Specified, specifiedColor=<rgba 0 0 0 0>}
      cornerRadii=<0 0 0 0>
     }
     tintColor {colorUseRule=UseColor_Specified, specifiedColor=<rgba 1 1 1 1>}
     uVRegion {bIsValid=false, max=<0 0>, min=<0 0>}
    }
    renderTransform {angle=0, scale=<1 1>, shear=<0 0>, translation=<0 36.5>}
   }
   slot {class=GridSlot}
  }
  Overlay_123 {
   widgetClassPath=&Overlay
   parent=BandOverdriveParent
   widget=Overlay_123
   slot {class=VerticalBoxSlot, properties {horizontalAlignment=HAlign_Center, verticalAlignment=VAlign_Bottom}}
  }
  CommonVisibilitySwitcher_226 {
   widgetClassPath=&CommonVisibilitySwitcher
   bIsVariable=true
   parent=Overlay_123
   widget=CommonVisibilitySwitcher_226
   properties {activeWidgetIndex=1}
   slot {class=OverlaySlot, properties {horizontalAlignment=HAlign_Fill, verticalAlignment=VAlign_Fill}}
  }
  Blank {
   widgetClassPath=&CommonTextBlock
   bIsVariable=true
   parent=CommonVisibilitySwitcher_226
   widget=Blank
   properties {
    style=/FortUILibrary/Styles/Typography/_TextStyle_SubH4._TextStyle_SubH4_C
    font {
     bForceMonospaced=false
     bMaterialIsStencil=false
     fontMaterial=null
     fontObject=&BurbankSmall
     letterSpacing=0
     monospacedWidth=1
     size=24
     skewAmount=0
     typefaceFontName=Black
     outlineSettings {
      bApplyOutlineToDropShadows=false
      bMiteredCorners=false
      bSeparateFillAlpha=false
      outlineMaterial=null
      outlineSize=0
      outlineColor=<rgba 0 0 0 1>
     }
    }
    shadowOffset=<0 0>
    strikeBrush {
     bIsDynamicallyLoaded=false
     drawAs=NoDrawType
     imageType=NoImage
     mirroring=NoMirror
     resourceName=None
     resourceObject=null
     tiling=NoTile
     imageSize=<32 32>
     margin=<ltrb 0 0 0 0>
     outlineSettings {
      bUseBrushTransparency=false
      roundingType=HalfHeightRadius
      width=0
      color {colorUseRule=UseColor_Specified, specifiedColor=<rgba 0 0 0 0>}
      cornerRadii=<0 0 0 0>
     }
     tintColor {colorUseRule=UseColor_Specified, specifiedColor=<rgba 1 1 1 1>}
     uVRegion {bIsValid=false, max=<0 0>, min=<0 0>}
    }
   }
   slot {class=CommonVisibilitySwitcherSlot}
  }
  ActiveText {
   widgetClassPath=&Overlay
   parent=CommonVisibilitySwitcher_226
   widget=ActiveText
   slot {class=CommonVisibilitySwitcherSlot, properties {horizontalAlignment=HAlign_Left, verticalAlignment=VAlign_Top}}
  }
  Text_SoftShadow {
   widgetClassPath=&Image
   parent=ActiveText
   widget=Text_SoftShadow
   properties {
    visibility=SelfHitTestInvisible
    brush {
     bIsDynamicallyLoaded=false
     drawAs=Image
     imageType=NoImage
     mirroring=NoMirror
     resourceName=None
     resourceObject=/PilgrimCore/Widgets/Materials/M_UI_Pilgrim_RoundedShadow_Round
     tiling=NoTile
     imageSize=<32 32>
     margin=<ltrb 0 0 0 0>
     outlineSettings {
      bUseBrushTransparency=false
      roundingType=HalfHeightRadius
      width=0
      color {colorUseRule=UseColor_Specified, specifiedColor=<rgba 0 0 0 0>}
      cornerRadii=<0 0 0 0>
     }
     tintColor {colorUseRule=UseColor_Specified, specifiedColor=<rgba 1 1 1 0.5>}
     uVRegion {bIsValid=false, max=<0 0>, min=<0 0>}
    }
    renderTransform {angle=0, scale=<1 1>, shear=<-10 0>, translation=<0 0>}
   }
   slot {class=OverlaySlot, properties {horizontalAlignment=HAlign_Fill, verticalAlignment=VAlign_Fill}}
  }
  Glow {
   widgetClassPath=&Image
   parent=ActiveText
   widget=Glow
   properties {
    visibility=SelfHitTestInvisible
    brush {
     bIsDynamicallyLoaded=false
     drawAs=Box
     imageType=NoImage
     mirroring=NoMirror
     resourceName=None
     resourceObject=/PilgrimCore/Widgets/Materials/MI_UI_Pilgrim_OverDrive_Glow_Text
     tiling=NoTile
     imageSize=<72 72>
     margin=<ltrb 0.5 0.5 0.5 0.5>
     outlineSettings {
      bUseBrushTransparency=false
      roundingType=HalfHeightRadius
      width=0
      color {colorUseRule=UseColor_Specified, specifiedColor=<rgba 0 0 0 0>}
      cornerRadii=<0 0 0 0>
     }
     tintColor {colorUseRule=UseColor_Specified, specifiedColor=<rgba 1 1 1 1>}
     uVRegion {bIsValid=false, max=<0 0>, min=<0 0>}
    }
    renderTransform {angle=0, scale=<1 1>, shear=<-10 0>, translation=<0 0>}
   }
   slot {class=OverlaySlot, properties {horizontalAlignment=HAlign_Fill, verticalAlignment=VAlign_Fill}}
  }
  Flare {
   widgetClassPath=&Image
   parent=ActiveText
   widget=Flare
   properties {
    visibility=SelfHitTestInvisible
    brush {
     bIsDynamicallyLoaded=false
     drawAs=Image
     imageType=NoImage
     mirroring=NoMirror
     resourceName=None
     resourceObject=/PilgrimCore/Widgets/Materials/MI_UI_Pilgrim_ODText_Flare
     tiling=NoTile
     imageSize=<32 48>
     margin=<ltrb 0.125 0 0.125 0>
     outlineSettings {
      bUseBrushTransparency=false
      roundingType=HalfHeightRadius
      width=0
      color {colorUseRule=UseColor_Specified, specifiedColor=<rgba 0 0 0 0>}
      cornerRadii=<0 0 0 0>
     }
     tintColor {colorUseRule=UseColor_Specified, specifiedColor=<rgba 1 1 1 1>}
     uVRegion {bIsValid=false, max=<0 0>, min=<0 0>}
    }
    renderTransform {angle=0, scale=<2 1>, shear=<0 0>, translation=<0 0>}
   }
   slot {class=OverlaySlot, properties {horizontalAlignment=HAlign_Fill, verticalAlignment=VAlign_Center}}
  }
  BandOverdriveBonus_Text_Active {
   widgetClassPath=&CommonTextBlock
   bIsVariable=true
   parent=ActiveText
   widget=BandOverdriveBonus_Text_Active
   properties {
    justification=Center
    text=BAND OVERDRIVE BONUS
    font {
     bForceMonospaced=false
     bMaterialIsStencil=false
     fontMaterial=/PilgrimCore/Widgets/Materials/MI_UI_Piligrim_HUDTextColor_ODBonus
     fontObject=&HeadingNow
     letterSpacing=-40
     monospacedWidth=1
     size=18
     skewAmount=0
     typefaceFontName=86 Bold Italic
     outlineSettings {
      bApplyOutlineToDropShadows=false
      bMiteredCorners=false
      bSeparateFillAlpha=true
      outlineMaterial=&M-FontShadow
      outlineSize=3
      outlineColor=<rgba 0 0 0 1>
     }
    }
    shadowOffset=<0 0>
    strikeBrush {
     bIsDynamicallyLoaded=false
     drawAs=NoDrawType
     imageType=NoImage
     mirroring=NoMirror
     resourceName=None
     resourceObject=null
     tiling=NoTile
     imageSize=<32 32>
     margin=<ltrb 0 0 0 0>
     outlineSettings {
      bUseBrushTransparency=false
      roundingType=HalfHeightRadius
      width=0
      color {colorUseRule=UseColor_Specified, specifiedColor=<rgba 0 0 0 0>}
      cornerRadii=<0 0 0 0>
     }
     tintColor {colorUseRule=UseColor_Specified, specifiedColor=<rgba 1 1 1 1>}
     uVRegion {bIsValid=false, max=<0 0>, min=<0 0>}
    }
   }
   slot {class=OverlaySlot, properties {horizontalAlignment=HAlign_Center, verticalAlignment=VAlign_Center, padding=<ltrb 24 16 24 16>}}
  }
  BandPointsRecapOverlay {
   widgetClassPath=&Overlay
   bIsVariable=true
   parent=CommonVisibilitySwitcher_226
   widget=BandPointsRecapOverlay
   slot {class=CommonVisibilitySwitcherSlot, properties {horizontalAlignment=HAlign_Center}}
  }
  Text_SoftShadow_1 {
   widgetClassPath=&Image
   parent=BandPointsRecapOverlay
   widget=Text_SoftShadow_1
   properties {
    visibility=SelfHitTestInvisible
    brush {
     bIsDynamicallyLoaded=false
     drawAs=Image
     imageType=NoImage
     mirroring=NoMirror
     resourceName=None
     resourceObject=&M_UI_Pilgrim_RoundedShadow_Round
     tiling=NoTile
     imageSize=<32 32>
     margin=<ltrb 0 0 0 0>
     outlineSettings {
      bUseBrushTransparency=false
      roundingType=HalfHeightRadius
      width=0
      color {colorUseRule=UseColor_Specified, specifiedColor=<rgba 0 0 0 0>}
      cornerRadii=<0 0 0 0>
     }
     tintColor {colorUseRule=UseColor_Specified, specifiedColor=<rgba 1 1 1 0.5>}
     uVRegion {bIsValid=false, max=<0 0>, min=<0 0>}
    }
    renderTransform {angle=0, scale=<1 1>, shear=<-10 0>, translation=<0 0>}
   }
   slot {class=OverlaySlot, properties {horizontalAlignment=HAlign_Fill, verticalAlignment=VAlign_Fill}}
  }
  Glow_1 {
   widgetClassPath=&Image
   parent=BandPointsRecapOverlay
   widget=Glow_1
   properties {
    visibility=SelfHitTestInvisible
    brush {
     bIsDynamicallyLoaded=false
     drawAs=Box
     imageType=NoImage
     mirroring=NoMirror
     resourceName=None
     resourceObject=&MI_UI_Pilgrim_OverDrive_Glow_Text
     tiling=NoTile
     imageSize=<64 64>
     margin=<ltrb 0.5 0.5 0.5 0.5>
     outlineSettings {
      bUseBrushTransparency=false
      roundingType=HalfHeightRadius
      width=0
      color {colorUseRule=UseColor_Specified, specifiedColor=<rgba 0 0 0 0>}
      cornerRadii=<0 0 0 0>
     }
     tintColor {colorUseRule=UseColor_Specified, specifiedColor=<rgba 1 1 1 1>}
     uVRegion {bIsValid=false, max=<0 0>, min=<0 0>}
    }
    renderTransform {angle=0, scale=<1 1>, shear=<-10 0>, translation=<0 0>}
   }
   slot {class=OverlaySlot, properties {horizontalAlignment=HAlign_Fill, verticalAlignment=VAlign_Fill}}
  }
  Flare_1 {
   widgetClassPath=&Image
   parent=BandPointsRecapOverlay
   widget=Flare_1
   properties {
    visibility=SelfHitTestInvisible
    brush {
     bIsDynamicallyLoaded=false
     drawAs=Image
     imageType=NoImage
     mirroring=NoMirror
     resourceName=None
     resourceObject=&MI_UI_Pilgrim_ODText_Flare
     tiling=NoTile
     imageSize=<32 48>
     margin=<ltrb 0.125 0 0.125 0>
     outlineSettings {
      bUseBrushTransparency=false
      roundingType=HalfHeightRadius
      width=0
      color {colorUseRule=UseColor_Specified, specifiedColor=<rgba 0 0 0 0>}
      cornerRadii=<0 0 0 0>
     }
     tintColor {colorUseRule=UseColor_Specified, specifiedColor=<rgba 1 1 1 1>}
     uVRegion {bIsValid=false, max=<0 0>, min=<0 0>}
    }
    renderTransform {angle=0, scale=<2 1>, shear=<0 0>, translation=<0 0>}
   }
   slot {class=OverlaySlot, properties {horizontalAlignment=HAlign_Fill, verticalAlignment=VAlign_Center}}
  }
  BandPointsEarnedWhileOverdriveActive_Text {
   widgetClassPath=&CommonTextBlock
   bIsVariable=true
   parent=BandPointsRecapOverlay
   widget=BandPointsEarnedWhileOverdriveActive_Text
   properties {
    justification=Center
    text="+{Value} Band Points!"
    font {
     bForceMonospaced=false
     bMaterialIsStencil=false
     fontMaterial=null
     fontObject=&HeadingNow
     letterSpacing=-40
     monospacedWidth=1
     size=18
     skewAmount=0
     typefaceFontName=86 Bold Italic
     outlineSettings {
      bApplyOutlineToDropShadows=false
      bMiteredCorners=false
      bSeparateFillAlpha=false
      outlineMaterial=&M-FontShadow
      outlineSize=3
      outlineColor=<rgba 0 0 0 1>
     }
    }
    shadowOffset=<0 0>
    strikeBrush {
     bIsDynamicallyLoaded=false
     drawAs=NoDrawType
     imageType=NoImage
     mirroring=NoMirror
     resourceName=None
     resourceObject=null
     tiling=NoTile
     imageSize=<32 32>
     margin=<ltrb 0 0 0 0>
     outlineSettings {
      bUseBrushTransparency=false
      roundingType=HalfHeightRadius
      width=0
      color {colorUseRule=UseColor_Specified, specifiedColor=<rgba 0 0 0 0>}
      cornerRadii=<0 0 0 0>
     }
     tintColor {colorUseRule=UseColor_Specified, specifiedColor=<rgba 1 1 1 1>}
     uVRegion {bIsValid=false, max=<0 0>, min=<0 0>}
    }
   }
   slot {class=OverlaySlot, properties {horizontalAlignment=HAlign_Center, verticalAlignment=VAlign_Center}}
  }
  ValueAmountOverlay {
   widgetClassPath=&Overlay
   bIsVariable=true
   parent=BandOverdriveParent
   widget=ValueAmountOverlay
   properties {renderTransform {angle=0, scale=<1 1>, shear=<0 0>, translation=<0 -24>}}
   slot {class=VerticalBoxSlot, properties {horizontalAlignment=HAlign_Center, verticalAlignment=VAlign_Bottom}}
  }
  ValueGlow_1 {
   widgetClassPath=&Image
   parent=ValueAmountOverlay
   widget=ValueGlow_1
   properties {
    visibility=SelfHitTestInvisible
    brush {
     bIsDynamicallyLoaded=false
     drawAs=Image
     imageType=NoImage
     mirroring=NoMirror
     resourceName=None
     resourceObject=/PilgrimCore/Widgets/Materials/MI_UI_Pilgrim_Bandmate_ODTextGlow
     tiling=NoTile
     imageSize=<40 40>
     margin=<ltrb 0.5 0.5 0.5 0.5>
     outlineSettings {
      bUseBrushTransparency=false
      roundingType=HalfHeightRadius
      width=0
      color {colorUseRule=UseColor_Specified, specifiedColor=<rgba 0 0 0 0>}
      cornerRadii=<0 0 0 0>
     }
     tintColor {colorUseRule=UseColor_Specified, specifiedColor=<rgba 1 1 1 1>}
     uVRegion {bIsValid=false, max=<0 0>, min=<0 0>}
    }
   }
   slot {class=OverlaySlot, properties {horizontalAlignment=HAlign_Fill, verticalAlignment=VAlign_Fill, padding=<ltrb 8 8 8 8>}}
  }
  ValueGlow_2 {
   widgetClassPath=&Image
   parent=ValueAmountOverlay
   widget=ValueGlow_2
   properties {
    visibility=SelfHitTestInvisible
    brush {
     bIsDynamicallyLoaded=false
     drawAs=Image
     imageType=NoImage
     mirroring=NoMirror
     resourceName=None
     resourceObject=&MI_UI_Pilgrim_Bandmate_ODTextGlow
     tiling=NoTile
     imageSize=<40 40>
     margin=<ltrb 0.5 0.5 0.5 0.5>
     outlineSettings {
      bUseBrushTransparency=false
      roundingType=HalfHeightRadius
      width=0
      color {colorUseRule=UseColor_Specified, specifiedColor=<rgba 0 0 0 0>}
      cornerRadii=<0 0 0 0>
     }
     tintColor {colorUseRule=UseColor_Specified, specifiedColor=<rgba 1 1 1 1>}
     uVRegion {bIsValid=false, max=<0 0>, min=<0 0>}
    }
    renderTransform {angle=0, scale=<-1 -1>, shear=<0 0>, translation=<0 0>}
   }
   slot {class=OverlaySlot, properties {horizontalAlignment=HAlign_Fill, verticalAlignment=VAlign_Fill, padding=<ltrb 8 8 8 8>}}
  }
  PulseRings {
   widgetClassPath=&Image
   parent=ValueAmountOverlay
   widget=PulseRings
   properties {
    visibility=SelfHitTestInvisible
    brush {
     bIsDynamicallyLoaded=false
     drawAs=Box
     imageType=NoImage
     mirroring=NoMirror
     resourceName=None
     resourceObject=/PilgrimCore/Widgets/Materials/MI_UI_ODText_FX_Rings
     tiling=NoTile
     imageSize=<72 72>
     margin=<ltrb 0.5 0.5 0.5 0.5>
     outlineSettings {
      bUseBrushTransparency=false
      roundingType=HalfHeightRadius
      width=0
      color {colorUseRule=UseColor_Specified, specifiedColor=<rgba 0 0 0 0>}
      cornerRadii=<0 0 0 0>
     }
     tintColor {colorUseRule=UseColor_Specified, specifiedColor=<rgba 1 1 1 1>}
     uVRegion {bIsValid=false, max=<0 0>, min=<0 0>}
    }
   }
   slot {class=OverlaySlot, properties {horizontalAlignment=HAlign_Fill, verticalAlignment=VAlign_Fill, padding=<ltrb -12 -12 -12 -12>}}
  }
  AddLayer {
   widgetClassPath=&Image
   parent=ValueAmountOverlay
   widget=AddLayer
   properties {brush {
     bIsDynamicallyLoaded=false
     drawAs=Image
     imageType=NoImage
     mirroring=NoMirror
     resourceName=None
     resourceObject=/PilgrimCore/Widgets/Materials/MI_UI_Pilgrim_Panel_AddOrange
     tiling=NoTile
     imageSize=<24 24>
     margin=<ltrb 0 0 0 0>
     outlineSettings {
      bUseBrushTransparency=false
      roundingType=HalfHeightRadius
      width=0
      color {colorUseRule=UseColor_Specified, specifiedColor=<rgba 0 0 0 0>}
      cornerRadii=<0 0 0 0>
     }
     tintColor {colorUseRule=UseColor_Specified, specifiedColor=<rgba 1 1 1 1>}
     uVRegion {bIsValid=false, max=<0 0>, min=<0 0>}
    }}
   slot {class=OverlaySlot, properties {horizontalAlignment=HAlign_Fill, verticalAlignment=VAlign_Fill, padding=<ltrb 16 16 16 16>}}
  }
  LanguageFlipperOverlay {
   widgetClassPath=&Overlay
   bIsVariable=true
   parent=ValueAmountOverlay
   widget=LanguageFlipperOverlay
   slot {class=OverlaySlot, properties {horizontalAlignment=HAlign_Center, verticalAlignment=VAlign_Center, padding=<ltrb 24 12 24 12>}}
  }
  HorizontalBox_0 {
   widgetClassPath=&HorizontalBox
   parent=LanguageFlipperOverlay
   widget=HorizontalBox_0
   slot {class=OverlaySlot, properties {horizontalAlignment=HAlign_Fill, verticalAlignment=VAlign_Fill}}
  }
  BandOverdriveBonus_Text_PointCounterLabel {
   widgetClassPath=&CommonTextBlock
   bIsVariable=true
   parent=HorizontalBox_0
   widget=BandOverdriveBonus_Text_PointCounterLabel
   properties {
    flowDirectionPreference=Culture
    justification=Center
    text=x Band Points
    textTransformPolicy=ToUpper
    colorAndOpacity {colorUseRule=UseColor_Specified, specifiedColor=<rgba 0.0006 0 0.0052 0.75>}
    font {
     bForceMonospaced=false
     bMaterialIsStencil=false
     fontMaterial=null
     fontObject=&HeadingNow
     letterSpacing=10
     monospacedWidth=1
     size=12
     skewAmount=0
     typefaceFontName=76 Bold
     outlineSettings {
      bApplyOutlineToDropShadows=false
      bMiteredCorners=false
      bSeparateFillAlpha=false
      outlineMaterial=&M-FontShadow
      outlineSize=0
      outlineColor=<rgba 0.5 0.1667 0 1>
     }
    }
    shadowOffset=<0 0>
    shapedTextOptions {bOverride_TextFlowDirection=true, bOverride_TextShapingMethod=false, textFlowDirection=Culture, textShapingMethod=Auto}
    strikeBrush {
     bIsDynamicallyLoaded=false
     drawAs=NoDrawType
     imageType=NoImage
     mirroring=NoMirror
     resourceName=None
     resourceObject=null
     tiling=NoTile
     imageSize=<32 32>
     margin=<ltrb 0 0 0 0>
     outlineSettings {
      bUseBrushTransparency=false
      roundingType=HalfHeightRadius
      width=0
      color {colorUseRule=UseColor_Specified, specifiedColor=<rgba 0 0 0 0>}
      cornerRadii=<0 0 0 0>
     }
     tintColor {colorUseRule=UseColor_Specified, specifiedColor=<rgba 1 1 1 1>}
     uVRegion {bIsValid=false, max=<0 0>, min=<0 0>}
    }
   }
   slot {class=HorizontalBoxSlot}
  }
 ]
}
