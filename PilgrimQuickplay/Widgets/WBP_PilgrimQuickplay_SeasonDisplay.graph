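# Season badge widget for Pilgrim Quickplay that shows a short season label, such as the placeholder "S10", with a highlight for the current season.
# A material-backed Border holds a fixed-size, scaled CommonText label with a sheared slant.
# SetCurrentSeasonBorder and RevertSeasonBorder swap two default blues between the border's dynamic material RGB1 parameter and the text color.
# The event graph has no logic, and nothing here sets the label or calls either function.
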WidgetBlueprint /PilgrimQuickplay/Widgets/WBP_PilgrimQuickplay_SeasonDisplay
graphs [
 EventGraph {nodes [
   Event PreConstruct()
   Event Construct()
   Event Tick()
  ]}
 RevertSeasonBorder {nodes [
   RevertSeasonBorder() {is=title K2Node_FunctionEntry}
   > n1 = GetDynamicMaterial(self:Border Object Reference=Get SeasonBorder() -> (SeasonBorder:Border Object Reference))
   > SetVectorParameterValue(self=@n1, ParameterName=RGB1, Value=Get DefaultDarkBlue() -> (DefaultDarkBlue:Linear Color Structure))
   > SetColorAndOpacity(
    self=Get SeasonText() -> (SeasonText:Common Text Object Reference)
    InColorAndOpacity=Make Slate Color(SpecifiedColor=Get DefaultLightBlue() -> (DefaultLightBlue:Linear Color Structure))
   )
  ]}
 SetCurrentSeasonBorder {nodes [
   SetCurrentSeasonBorder() {is=title K2Node_FunctionEntry}
   > n1 = GetDynamicMaterial(self=Get SeasonBorder())
   > SetVectorParameterValue(self=@n1, ParameterName=RGB1, Value=Get DefaultLightBlue())
   > SetColorAndOpacity(self=Get SeasonText(), InColorAndOpacity=Make Slate Color(SpecifiedColor=Get DefaultDarkBlue()))
  ]}
]
variables [
 DefaultDarkBlue {type=LinearColor, default=<rgba 0.0103 0.0648 0.3095 1>}
 DefaultLightBlue {type=LinearColor, default=<rgba 0.4969 0.8308 1 1>}
]
widgets {
 info {inheritedWidgetCount=0, parentClass=&UserWidget, rootWidgetClass=&Border, widgetCount=4}
 widgets [
  SeasonBorder {
   widgetClassPath=&Border
   bIsVariable=true
   widget=SeasonBorder
   properties {
    visibility=HitTestInvisible
    background {
     bIsDynamicallyLoaded=false
     drawAs=Image
     imageType=FullColor
     mirroring=NoMirror
     resourceName=None
     resourceObject=/PilgrimCore/Widgets/Materials/M_UI_Pilgrim_Season_Border
     tiling=NoTile
     imageSize=<32 32>
     margin=<ltrb 0 0 0 0>
     outlineSettings {
      bUseBrushTransparency=false
      roundingType=HalfHeightRadius
      width=0
      color {colorUseRule=UseColor_Specified, specifiedColor=<rgba 0 0 0 0>}
      cornerRadii=<0 0 0 0>
     }
     tintColor {colorUseRule=UseColor_Specified, specifiedColor=<rgba 1 1 1 1>}
     uVRegion {bIsValid=false, max=<0 0>, min=<0 0>}
    }
    padding=<ltrb 8 0 8 0>
   }
  }
  SizeBox_0 {
   widgetClassPath=&SizeBox
   parent=SeasonBorder
   widget=SizeBox_0
   properties {bOverride_HeightOverride=true, bOverride_WidthOverride=true, heightOverride=36, widthOverride=40}
   slot {class=BorderSlot, properties {padding=<ltrb 8 0 8 0>}}
  }
  ScaleBox_0 {
   widgetClassPath=&ScaleBox
   parent=SizeBox_0
   widget=ScaleBox_0
   slot {class=SizeBoxSlot, properties {horizontalAlignment=HAlign_Center, verticalAlignment=VAlign_Center}}
  }
  SeasonText {
   widgetClassPath=&CommonTextBlock
   bIsVariable=true
   parent=ScaleBox_0
   widget=SeasonText
   properties {
    justification=Center
    text=S10
    colorAndOpacity {colorUseRule=UseColor_Specified, specifiedColor=<rgba 0.5 0.8333 1 1>}
    font {
     bForceMonospaced=false
     bMaterialIsStencil=false
     fontMaterial=null
     fontObject=&HeadingNow
     letterSpacing=0
     monospacedWidth=1
     size=15
     skewAmount=0
     typefaceFontName=75 Medium
     outlineSettings {
      bApplyOutlineToDropShadows=false
      bMiteredCorners=false
      bSeparateFillAlpha=false
      outlineMaterial=null
      outlineSize=0
      outlineColor=<rgba 1 1 1 1>
     }
    }
    renderTransform {angle=0, scale=<1 1>, shear=<-12 0>, translation=<0 0>}
    shadowOffset=<0 0>
   }
   slot {class=ScaleBoxSlot, properties {horizontalAlignment=HAlign_Fill}}
  }
 ]
}
