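# Pilgrim Quickplay song timeline bar showing the local player's playable sections as progress segments, separated by gem breaks.
# Music clock beats set up the bar once song length is known; MBI_Started events trigger a reset.
# Gem breaks for the player's instrument and difficulty become scaled, offset Images, or one full-length bar when breaks are absent or ignored.
# UpdateProgress maps song time onto each segment's dynamic material OverlayProgress.
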WidgetBlueprint /PilgrimQuickplay/Widgets/WPB_PilgrimQuickplay_SongTimelineBar
graphs [
 AddPerformanceSection {nodes [
   AddPerformanceSection!() -> (then, PerformanceSectionStartMs:double=~, PerformanceSectionLengthMs:double=~, CurrentIndex:int) {is=title K2Node_FunctionEntry}
   > Set MeterDesiredSize_X(MeterDesiredSize_X:double=GetDesiredSize!(self:Widget Object Reference=Get Central_stick() -> (Central_stick:Image Object Reference)) -> (ReturnValue_X:double, ReturnValue_Y:double)) -> (Output_Get:double)
   > n1 = Construct Object from Class(Class=&Image) -> (ReturnValue:Image Object Reference)
   > n2 = AddChild(self=Get Overlay-DynamicElements() -> (Overlay-DynamicElements:Overlay Object Reference), Content=@n1)
   > n3 = Cast To OverlaySlot!(execute, Object=@n2) -> (then, CastFailed, AsOverlay Slot:Overlay Slot Object Reference) {is=title K2Node_DynamicCast}
   > Set VerticalAlignment!(execute, VerticalAlignment:EVerticalAlignment Enum=VAlign_Center, self:Overlay Slot Object Reference=@n3) -> (then, Output_Get:EVerticalAlignment Enum=VAlign_Fill)
   > SetVisibility(self=@n1, InVisibility=SelfHitTestInvisible)
   n4 = float * float(
    A=float / float(A=Get PerformanceSectionLengthMs() -> (PerformanceSectionLengthMs:double), B=Get EndOfSongMs() -> (EndOfSongMs:float))
    B=Get MeterDesiredSize_X() -> (MeterDesiredSize_X:double)
   ) {is=title K2Node_PromotableOperator}
   > SetBrush(
    self=@n1
    InBrush=MakeBrushFromMaterial!(
     Material:Material Interface Object Reference=/PilgrimQuickplay/Widgets/Materials/MI_UI_Pilgrim_SongProgress
     Width:int=Truncate(A=@n4)
     Height:int=Get Performance Section Bar Height() -> (Performance Section Bar Height:int=0)
    ) -> (ReturnValue:Slate Brush Structure)
   )
   n5 = float * float(A=Get MeterDesiredSize_X(), B=float / float(A=Get PerformanceSectionStartMs() -> (PerformanceSectionStartMs:double), B=Get EndOfSongMs()))
   > Set RenderTransform!(
    execute
    RenderTransform:Widget Transform Structure=MakeTransform!(Translation:Vector 2D Structure=MakeVector2D(X=@n5, Y=0), Scale:Vector 2D Structure=(X=1,Y=1), Shear:Vector 2D Structure, Angle:double=0) -> (ReturnValue:Widget Transform Structure)
    self:Widget Object Reference=@n1
   ) -> (then, Output_Get:Widget Transform Structure)
   > AddPerformanceSectionData!(
    execute
    self
    SectionStartTimeMs:double=Get PerformanceSectionStartMs()
    SectionDurationMs:double=Get PerformanceSectionLengthMs()
    SectionIndex:int=Get CurrentIndex() -> (CurrentIndex:int=0)
    SectionImageRef:Image Object Reference=@n1
   ) -> (then)
  ]}
 AddPerformanceSectionData {nodes [
   n1 = AddPerformanceSectionData!() -> (then, SectionStartTimeMs:double=~, SectionDurationMs:double=~, SectionIndex:int, SectionImageRef:Image Object Reference) {is=title K2Node_FunctionEntry}
   > Insert!(
    execute
    TargetArray:Array of Vector 2D Structures=Get PerformanceSectionDataArray() -> (PerformanceSectionDataArray:Array of Vector 2D Structures)
    NewItem:"Vector 2D Structure (by ref)"=MakeVector2D(X=@n1, Y=@n1.SectionDurationMs)
    Index:int=@n1.SectionIndex
   ) -> (then) {is=title K2Node_CallArrayFunction}
   > n2 = GetDynamicMaterial(self=@n1.SectionImageRef)
   > Insert(
    TargetArray:Array of Material Instance Dynamic Object References=Get PerformanceSectionDynamicMaterialArray() -> (PerformanceSectionDynamicMaterialArray:Array of Material Instance Dynamic Object References)
    NewItem:"Material Instance Dynamic Object Reference (by ref)"=@n2
    Index=@n1.SectionIndex
   )
   > SetScalarParameterValue(self=@n2, ParameterName=OverlayProgress, Value=0)
  ]}
 EventGraph {nodes [
   n1 = BeatEvent_Event!() -> (OutputDelegate, then, BeatNumber:int, BeatInBar:int) {is=title K2Node_CustomEvent}
   Branch(execute=[@n1.then, @n3.then], Condition="float <= float"(A=Get EndOfSongMs()) -> (ReturnValue=~))
   n2 = GetSongLengthMs!(self:Music Clock Object Reference=GetMusicClock!(WorldContextObject:Object Reference=GetOwningPlayer()) -> (ReturnValue:Music Clock Object Reference)) -> (ReturnValue:float)
   > Branch(Condition="float > float"(A=@n2))
   > Set EndOfSongMs(EndOfSongMs:float=@n2) -> (Output_Get:float)
   > SetupBreakInfo()
   ResetForNewPlayerOrSong() {is=title K2Node_CustomEvent}
   > n3 = Set EndOfSongMs(EndOfSongMs=0)
   SetupBreakInfo() {is=title K2Node_CustomEvent}
   > Branch(Condition=NOT Boolean(A=Get IsSettingUpBreaks() -> (IsSettingUpBreaks:bool)))
   > Set IsSettingUpBreaks(IsSettingUpBreaks:bool=true) -> (Output_Get:bool)
   n4 = GetSparksMusicPlayspaceOfClass!(
    PlayspaceType:Sparks Music Playspace Class Reference=/Script/PilgrimQuickplayRuntime.PilgrimQuickplayPlayspace
    Actor:Actor Object Reference=GetOwningPlayer()
    bMatchChildTypes:bool=true
   ) -> (ReturnValue:Pilgrim Quickplay Playspace Object Reference)
   > n5 = Is Valid(InputObject=@n4)
   n6 = GetComponentByClass!(
    self:Actor Object Reference=@n4
    ComponentClass:Actor Component Class Reference=/Script/PilgrimQuickplayRuntime.PilgrimQuickplayPlayspaceComponent_PlayerManager
   ) -> (ReturnValue:Pilgrim Quickplay Playspace Component Player Manager Object Reference)
   > n7 = Is Valid(InputObject=@n6)
   n8 = GetLocalPlayerOverride!(self:Pilgrim Quickplay Playspace Component Player Manager Object Reference=@n6) -> (ReturnValue:Player State Object Reference)
   > n9 = Is Valid(InputObject=@n8)
   > n10 = Set RepresentedPlayer(RepresentedPlayer:Player State Object Reference=@n8) -> (Output_Get:Player State Object Reference)
   n11 = GetPilgrimCorePlayerStateComponent!(
    execute=[@n10.then, @n24.then]
    PlayerState:Player State Object Reference=Get RepresentedPlayer() -> (RepresentedPlayer:Player State Object Reference)
   ) -> (then, ReturnValue:Pilgrim Core Player State Component Object Reference)
   n12 = GetInstrumentChoice!(self:Pilgrim Core Player State Component Object Reference=@n11) -> (ReturnValue:EPilgrimTrackType Enum=TrackGuitar)
   > n13 = Branch(Condition=AND Boolean(A="Equal (Enum)"(A:EPilgrimTrackType Enum=@n12, B:EPilgrimTrackType Enum=TrackNone), B=InEditorOrPIE!() -> (ReturnValue:bool)))
   > PrintString!(
    execute
    InString=Defaulting to guitar track type for editor only case of no held instrument
    bPrintToScreen
    bPrintToLog
    TextColor
    Duration
    Key
   ) -> (then)
   > n14 = Set Track Type(Track Type:EPilgrimTrackType Enum=TrackGuitar) -> (Output_Get:EPilgrimTrackType Enum=TrackGuitar)
   n15 = Set Gem Breaks(
    execute=[@n14.then, @n23.then]
    Gem Breaks:Array of Pilgrim Gem Break Structures=GetGemBreaks!(
     self:Pilgrim Gem Analyzer Object Reference=GetPilgrimGemAnalyzer!(WorldContextObject:Object Reference=@n11) -> (ReturnValue:Pilgrim Gem Analyzer Object Reference)
     TrackType:EPilgrimTrackType Enum=Get Track Type() -> (Track Type:EPilgrimTrackType Enum=TrackGuitar)
     Difficulty:EPilgrimSongDifficulty Enum=GetSongDifficulty!(self:Pilgrim Core Player State Component Object Reference=@n11) -> (ReturnValue:EPilgrimSongDifficulty Enum=DifficultyEasy)
    ) -> (OutGemBreaks:Array of Pilgrim Gem Break Structures)
   ) -> (Output_Get:Array of Pilgrim Gem Break Structures)
   > Clear(TargetArray:Array of Material Instance Dynamic Object References=Get PerformanceSectionDynamicMaterialArray())
   > Clear(TargetArray:Array of Vector 2D Structures=Get PerformanceSectionDataArray())
   > Set OldChildren(OldChildren:Array of Widget Object References=GetAllChildren(self=Get Overlay-DynamicElements())) -> (Output_Get:Array of Widget Object References)
   n16 = Length(TargetArray:Array of Pilgrim Gem Break Structures=@n15)
   > n17 = Branch(Condition=OR Boolean(A=Get BarIgnoresBreaks() -> (BarIgnoresBreaks:bool), B="integer < integer"(A=@n16, B=1)))
   > n18 = AddPerformanceSection!(execute, self, PerformanceSectionStartMs:double=0, PerformanceSectionLengthMs:double=Get EndOfSongMs(), CurrentIndex:int=0) -> (then) {is=title K2Node_CallFunction}
   ForceLayoutPrepass(execute=[@n18.then, @n22.Completed], self:Widget Object Reference)
   n19 = Get OldChildren() -> (OldChildren:Array of Widget Object References)
   > n20 = Branch(Condition=Is Not Empty!(TargetArray:Array of Widget Object References=@n19) -> (ReturnValue:bool))
   > n21 = For Each Loop(Array:Array of Widget Object References=@n19) -> (Array Element:Widget Object Reference)
   > RemoveFromParent(self=@n21)
   Clear(execute=[@n21.Completed, @n20.else], TargetArray:Array of Widget Object References=Get OldChildren())
   > UpdateProgress!(execute, self, ProgressMs:double=Get SongProgressMs() -> (SongProgressMs:double)) -> (then)
   > Set IsSettingUpBreaks(IsSettingUpBreaks=false)
   n22 = For Each Loop(Exec=@n17.else, Array:Array of Pilgrim Gem Break Structures=@n15) -> (Array Element:Pilgrim Gem Break Structure)
   > GeneratePerformanceSectionFromGemBreak!(
    execute
    self
    GemBreakRef_in:"Pilgrim Gem Break Structure (by ref)"=@n22
    GemBreakIndex_in:int=@n22.Array Index
    AllGemBreaks_in:Array of Pilgrim Gem Break Structures=@n15
   ) -> (then)
   n23 = Set Track Type(execute=@n13.else, Track Type=@n12) -> (Output_Get=TrackGuitar)
   n24 = Set RepresentedPlayer(
    execute=[@n9.Is Not Valid, @n7.Is Not Valid, @n5.Is Not Valid]
    RepresentedPlayer=Get PlayerState!(self:Controller Object Reference=GetOwningPlayer()) -> (PlayerState:Player State Object Reference)
   )
   Event Construct()
   > ForceLayoutPrepass(self=Get Central_stick())
   n25 = GetOwningPlayer()
   > Bind Event to Beat Event(self:Music Clock Object Reference=GetMusicClock(WorldContextObject=@n25), Delegate=@n1) {is=title K2Node_AddDelegate}
   > n26 = PilgrimWaitForEventRouter!(execute:Exec, ContextActor:Actor Object Reference=@n25) -> (then:Exec, OnEventRouterAvailable:Exec, EventRouter:Gameplay Event Router Component Object Reference) {is=title K2Node_AsyncAction}
   n27 = StartListeningToEvent!(
    execute:Exec=@n26.OnEventRouterAvailable
    Target:Gameplay Event Router Component Object Reference=@n26
    EventType:Script Struct Object Reference=/Script/PilgrimCoreRuntime.PilgrimEvent_MBI_Started
    Context:Object Reference
    EventBubblingRule:EEventBubblingRule Enum
   ) -> (
    then:Exec
    AsyncTaskProxy:Async Action Start Listening To Event Object Reference
    OnEventReceived:Exec
    Payload:Pilgrim Event MBI Started Structure
   )
   ResetForNewPlayerOrSong(execute=@n27.OnEventReceived) {is=title K2Node_CallFunction}
  ]}
 GeneratePerformanceSectionFromGemBreak {nodes [
   n1 = GeneratePerformanceSectionFromGemBreak!() -> (then, GemBreakRef_in:"Pilgrim Gem Break Structure (by ref)", GemBreakIndex_in:int, AllGemBreaks_in:Array of Pilgrim Gem Break Structures) {is=title K2Node_FunctionEntry}
   > Set CurrentBreak(CurrentBreak:Pilgrim Gem Break Structure=@n1) -> (Output_Get:Pilgrim Gem Break Structure)
   > n2 = Set CurrentIndex(CurrentIndex:int=@n1.GemBreakIndex_in) -> (Output_Get:int=0)
   > n3 = Set HasValidPreviousBreak(HasValidPreviousBreak:bool="integer > integer"(A=@n2)) -> (Output_Get:bool)
   > n4 = Branch(Condition=@n3)
   n5 = Get CurrentIndex() -> (CurrentIndex=0)
   > n6 = Set PreviousBreak(PreviousBreak:Pilgrim Gem Break Structure="Get (a ref)"(
     Array:Array of Pilgrim Gem Break Structures=Get AllGemBreaks_in() -> (AllGemBreaks_in:Array of Pilgrim Gem Break Structures)
     Dimension 1=int - int(A=@n5)
    ) -> (Output:"Pilgrim Gem Break Structure (by ref)")) -> (Output_Get:Pilgrim Gem Break Structure)
   n7 = Get CurrentIndex() -> (CurrentIndex=0)
   Set HasValidNextBreak(
    execute=[@n6.then, @n4.else]
    HasValidNextBreak:bool=Is Valid Index!(TargetArray:Array of Pilgrim Gem Break Structures=Get AllGemBreaks_in(), IndexToTest:int=int + int(A=@n7)) -> (ReturnValue:bool)
   ) -> (Output_Get:bool)
   n8 = NOT Boolean(A=Get HasValidPreviousBreak() -> (HasValidPreviousBreak:bool))
   n9 = Break Pilgrim Gem Break!(PilgrimGemBreak:"Pilgrim Gem Break Structure (by ref)"=Get CurrentBreak() -> (CurrentBreak:Pilgrim Gem Break Structure)) -> (
    Id:int
    bIsValidGemBreak:bool=~
    StartBreakAtBar:float=~
    StartBreakAtMs:float=~
    BreakBarDuration:float=~
    BreakMsDuration:float=~
    EndBreakAtBar:float=~
    EndBreakAtMs:float=~
    bIsStartingBreak:bool=~
    bIsEndingBreak:bool=~
   ) {is=title K2Node_BreakStruct}
   n10 = Get!() -> (
    PreviousBreak_Id:int
    PreviousBreak_bIsValidGemBreak:bool=~
    PreviousBreak_StartBreakAtBar:float=~
    PreviousBreak_StartBreakAtMs:float=~
    PreviousBreak_BreakBarDuration:float=~
    PreviousBreak_BreakMsDuration:float=~
    PreviousBreak_EndBreakAtBar:float=~
    PreviousBreak_EndBreakAtMs:float=~
    PreviousBreak_bIsStartingBreak:bool=~
    PreviousBreak_bIsEndingBreak:bool=~
   )
   > Set PerformanceSectionStartMs(PerformanceSectionStartMs:double=SelectFloat(
     A=0
     B=float + float(A=@n10.PreviousBreak_StartBreakAtMs, B=@n10.PreviousBreak_BreakMsDuration)
     bPickA=OR Boolean(A=@n8, B=@n9.bIsStartingBreak)
    )) -> (Output_Get:double)
   > Set PerformanceSectionLengthMs(PerformanceSectionLengthMs:double=float - float(A=@n9.StartBreakAtMs, B=Get PerformanceSectionStartMs())) -> (Output_Get:double)
   > AddPerformanceSection(
    PerformanceSectionStartMs=Get PerformanceSectionStartMs()
    PerformanceSectionLengthMs=Get PerformanceSectionLengthMs()
    CurrentIndex=Get CurrentIndex() -> (CurrentIndex=0)
   )
   > n11 = Branch(Condition=Get HasValidNextBreak() -> (HasValidNextBreak:bool))
   n12 = float + float(A=@n9.StartBreakAtMs, B=@n9.BreakMsDuration)
   Set BreakSectionEndMs(execute=@n11.else, BreakSectionEndMs:double=@n12) -> (Output_Get:double)
   > Branch(Condition="float < float"(A=@n12, B=Get EndOfSongMs()))
   > AddPerformanceSection(
    PerformanceSectionStartMs=@n12
    PerformanceSectionLengthMs=float - float(A=Get EndOfSongMs(), B=@n12)
    CurrentIndex=int + int(A=Get CurrentIndex() -> (CurrentIndex=0))
   )
  ]}
 SetBarWidth {nodes [
   n1 = SetBarWidth!() -> (then, Width:double=370) {is=title K2Node_FunctionEntry}
   n2 = Get Central_stick()
   > SetDesiredSizeOverride!(
    execute
    self:Image Object Reference=Get Central_stick()
    DesiredSize_X:double=@n1
    DesiredSize_Y:double=Break Slate Brush!(SlateBrush:"Slate Brush Structure (by ref)"=Get Brush!(self:Image Object Reference=@n2) -> (Brush:Slate Brush Structure)) -> (
     TintColor:Slate Color Structure
     DrawAs:ESlateBrushDrawType Enum
     Tiling:ESlateBrushTileType Enum
     Mirroring:ESlateBrushMirrorType Enum
     ImageSize_X:double
     ImageSize_Y:double
     Margin:Margin Structure
     ResourceObject:Object Reference
     OutlineSettings:Slate Brush Outline Settings Structure
    ) {is=title K2Node_BreakStruct}.ImageSize_Y
   ) -> (then)
  ]}
 UpdateProgress {nodes [
   n1 = UpdateProgress!() -> (then, ProgressMs:double=~) {is=title K2Node_FunctionEntry}
   > Set SongProgressMs(SongProgressMs:double=@n1) -> (Output_Get:double)
   > Branch(Condition=Is Not Empty(TargetArray:Array of Material Instance Dynamic Object References=Get PerformanceSectionDynamicMaterialArray()))
   > n2 = For Each Loop(Array:Array of Material Instance Dynamic Object References=Get PerformanceSectionDynamicMaterialArray()) -> (Array Element:Material Instance Dynamic Object Reference)
   > Set CurrentProgressElement(CurrentProgressElement:Material Instance Dynamic Object Reference=@n2) -> (Output_Get:Material Instance Dynamic Object Reference)
   n3 = Get PerformanceSectionDataArray()
   > Branch(Condition=Is Valid Index(TargetArray:Array of Vector 2D Structures=@n3, IndexToTest=@n2.Array Index))
   n4 = BreakVector2D(InVec="Get (a ref)"(Array:Array of Vector 2D Structures=@n3, Dimension 1=@n2.Array Index) -> (Output:"Vector 2D Structure (by ref)"))
   > Set SectionStartMs(SectionStartMs:double=@n4) -> (Output_Get:double)
   > n5 = Set SectionLengthMs(SectionLengthMs:double=@n4.Y) -> (Output_Get:double)
   > n6 = Branch(Condition="float <= float"(A=@n5) -> (ReturnValue=~))
   n7 = Get SectionStartMs() -> (SectionStartMs:double)
   n8 = Get SectionLengthMs() -> (SectionLengthMs:double)
   SetScalarParameterValue(
    execute=@n6.else
    self=Get CurrentProgressElement() -> (CurrentProgressElement:Material Instance Dynamic Object Reference)
    ParameterName=OverlayProgress
    Value=MapRangeClamped(Value=Get SongProgressMs(), InRangeA=Get SectionStartMs(), InRangeB=float + float(A=@n7, B=@n8))
   )
  ]}
]
variables [
 BarIgnoresBreaks {type=boolean}
 BreakBarHeight {type=number, default=0.2}
 CurrentSongMs {type=number}
 EndOfSongMs {type=number}
 Gem Breaks {type="array<PilgrimGemBreak>", default []}
 GemAnalyzer {type=PilgrimGemAnalyzer, default=null}
 IsSettingUpBreaks {type=boolean}
 OldChildren {type="array<Widget>", default []}
 Performance Section Bar Height {type=integer, default=16}
 PerformanceBarHeight {type=number, default=1}
 PerformanceSectionDataArray {type="array<Vector2D>", default []}
 PerformanceSectionDynamicMaterialArray {type="array<MaterialInstanceDynamic>", default []}
 PlayedColor {type=LinearColor, default=<rgba 1 1 1 1>}
 RepresentedPlayer {type=PlayerState, default=null}
 SongProgressMs {type=number}
 ToPlayColor {type=LinearColor, default=<rgba 0.3916 0.3916 0.3916 1>}
 Track Type {type=EPilgrimTrackType, default=TrackGuitar}
]
widgets {
 info {inheritedWidgetCount=0, parentClass=&UserWidget, rootWidgetClass=&Overlay, widgetCount=6}
 widgets [
  Overlay-Primary {widgetClassPath=&Overlay, bIsVariable=true, widget=Overlay-Primary}
  Overlay-Line {
   widgetClassPath=&Overlay
   parent=Overlay-Primary
   widget=Overlay-Line
   slot {class=OverlaySlot, properties {horizontalAlignment=HAlign_Fill, verticalAlignment=VAlign_Fill}}
  }
  Left_stick {
   widgetClassPath=&Image
   bIsVariable=true
   parent=Overlay-Line
   widget=Left_stick
   properties {brush {
     bIsDynamicallyLoaded=false
     drawAs=Image
     imageType=FullColor
     mirroring=NoMirror
     resourceName=None
     resourceObject=/PilgrimQuickplay/Widgets/Materials/MI_UI_SongDuration_Line
     tiling=NoTile
     imageSize=<4 12>
     margin=<ltrb 0 0 0 0>
     outlineSettings {
      bUseBrushTransparency=false
      roundingType=HalfHeightRadius
      width=0
      color {colorUseRule=UseColor_Specified, specifiedColor=<rgba 0 0 0 0>}
      cornerRadii=<0 0 0 0>
     }
     tintColor {colorUseRule=UseColor_Specified, specifiedColor=<rgba 1 1 1 1>}
     uVRegion {bIsValid=false, max=<0 0>, min=<0 0>}
    }}
   slot {class=OverlaySlot, properties {verticalAlignment=VAlign_Center}}
  }
  Central_stick {
   widgetClassPath=&Image
   bIsVariable=true
   parent=Overlay-Line
   widget=Central_stick
   properties {brush {
     bIsDynamicallyLoaded=false
     drawAs=Image
     imageType=FullColor
     mirroring=NoMirror
     resourceName=None
     resourceObject=&MI_UI_SongDuration_Line
     tiling=NoTile
     imageSize=<370 4>
     margin=<ltrb 0 0 0 0>
     outlineSettings {
      bUseBrushTransparency=false
      roundingType=HalfHeightRadius
      width=0
      color {colorUseRule=UseColor_Specified, specifiedColor=<rgba 0 0 0 0>}
      cornerRadii=<0 0 0 0>
     }
     tintColor {colorUseRule=UseColor_Specified, specifiedColor=<rgba 1 1 1 1>}
     uVRegion {bIsValid=false, max=<0 0>, min=<0 0>}
    }}
   slot {class=OverlaySlot, properties {horizontalAlignment=HAlign_Fill, verticalAlignment=VAlign_Center}}
  }
  Right_stick {
   widgetClassPath=&Image
   bIsVariable=true
   parent=Overlay-Line
   widget=Right_stick
   properties {brush {
     bIsDynamicallyLoaded=false
     drawAs=Image
     imageType=FullColor
     mirroring=NoMirror
     resourceName=None
     resourceObject=&MI_UI_SongDuration_Line
     tiling=NoTile
     imageSize=<4 12>
     margin=<ltrb 0 0 0 0>
     outlineSettings {
      bUseBrushTransparency=false
      roundingType=HalfHeightRadius
      width=0
      color {colorUseRule=UseColor_Specified, specifiedColor=<rgba 0 0 0 0>}
      cornerRadii=<0 0 0 0>
     }
     tintColor {colorUseRule=UseColor_Specified, specifiedColor=<rgba 1 1 1 1>}
     uVRegion {bIsValid=false, max=<0 0>, min=<0 0>}
    }}
   slot {class=OverlaySlot, properties {horizontalAlignment=HAlign_Right, verticalAlignment=VAlign_Center}}
  }
  Overlay-DynamicElements {
   widgetClassPath=&Overlay
   bIsVariable=true
   parent=Overlay-Primary
   widget=Overlay-DynamicElements
   slot {class=OverlaySlot, properties {horizontalAlignment=HAlign_Fill, verticalAlignment=VAlign_Fill}}
  }
 ]
}
